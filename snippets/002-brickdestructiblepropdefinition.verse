# Knowledge-neutral Verse snippet; project names and asset IDs generalized.
# id: snippet-002

using { /Fortnite.com/Devices }
using { /Verse.org/Simulation }
using { /UnrealEngine.com/Temporary/Diagnostics }
using { /UnrealEngine.com/Temporary/SpatialMath }

# ============================================================================================================================================
# Definition of a Prop that can be destroyed to grant a reward of studs and potentially respawn
# ============================================================================================================================================
BRICK_destructible_prop_definition := class<concrete>():
    var PropTransform : transform = transform{}

    @editable:
        ToolTip := ToolTip_Prop
    var Prop : ?creative_prop = false

    @editable:
        ToolTip := ToolTip_PropAsset
    PropAsset : creative_prop_asset = DefaultCreativePropAsset
    
    @editable:
        ToolTip := ToolTip_AutoRespawn
    var AutoRespawn : logic = true
    
    @editable_number(float):
        MinValue := option{0.0}
        ToolTip := ToolTip_RespawnTime
    var RespawnTime : float = 5.0

    @editable:
        ToolTip := ToolTip_Prop_Reward
    StudReward<public> : BRICK_stud_spawner = BRICK_stud_spawner{}

    var RewardSpawned : logic = true

    # We cache the editor state of the asset to be able to restore it when the prop is destroyed
    Init(InCreativeDevice : creative_device): void=
        if (ValidProp := Prop?):
            set PropTransform = ValidProp.GetTransform()
            set RewardSpawned = false

        StudReward.Init(InCreativeDevice)

    # If the prop is destroyed, spawn the reward and potentially respawn the prop
    RewardIfDestroyed(InDevice : creative_device, InIgnoreAutoRespawn : logic): void=
        if (RewardSpawned? or Prop?.IsValid[]):
            return

        StudReward.SpawnStudsAtLocation(PropTransform.Translation + vector3{ Z := 50.0})
        set RewardSpawned = true

        if (AutoRespawn? and not InIgnoreAutoRespawn?):
            spawn{StartRespawnTimer()}

    # Respawn the prop after a set time
    StartRespawnTimer()<suspends>: void=
        race:
            Sleep(RespawnTime)
            OnRespawned.Await()
        Respawn()

    # Respawn the prop
    OnRespawned : event() = event() {}
    Respawn(): void=
        if (Prop?.IsValid[]):
            return

        SpawnedProp := SpawnProp(PropAsset, PropTransform)
        set Prop = SpawnedProp(0)
        set RewardSpawned = false
        OnRespawned.Signal()


ToolTip_Prop<public><localizes>:message = "The prop that is destroyed to grant the reward."
ToolTip_PropAsset<public><localizes>:message = "The prop asset that is spawned when the prop is destroyed."
ToolTip_AutoRespawn<public><localizes>:message = "If true, the prop will respawn after a set time."
ToolTip_RespawnTime<public><localizes>:message = "The time it takes for the prop to respawn."
ToolTip_Prop_Reward<public><localizes>:message = "The Stud reward that is granted for destroying this prop."
