# Knowledge-neutral Verse snippet; project names and asset IDs generalized.
# id: snippet-003

using { /Fortnite.com/Devices }
using { /Verse.org/Simulation }
using { /UnrealEngine.com/Temporary/Diagnostics }
using { /UnrealEngine.com/Temporary/SpatialMath }

# ============================================================================================================================================
# Manager for destructible props that grant rewards when destroyed
# ============================================================================================================================================
BRICK_destructible_prop_manager := class(creative_device):

    @editable:
        ToolTip := ToolTip_DestructibleProps
    DestructibleProps : []BRICK_destructible_prop_definition = array{}
    
    @editable:
        ToolTip := ToolTip_UpdateRate
    UpdateRateInSeconds : float = 0.5
    
    @editable:
        Categories := array{Category_RespawnSettings}
        ToolTip := ToolTip_OnlyRespawnManually
    CanOnlyRespawnManually : logic = false

    @editable:
        Categories := array{Category_RespawnSettings}
        ToolTip := ToolTip_ManualRespawn
    Trigger_ManualRespawn : trigger_device = trigger_device{}

    OnBegin<override>()<suspends>:void=
        Trigger_ManualRespawn.TriggeredEvent.Subscribe(RespawnAllProps)

        for (Definition : DestructibleProps):
            Definition.Init(Self)

        spawn{Loop_CheckPropValidity()}

    # Check if props have been destroyed by checking if they are still valid
    Loop_CheckPropValidity<private>()<suspends>: void=
        loop:
            Sleep(UpdateRateInSeconds)

            for (DestructibleProp : DestructibleProps):
                DestructibleProp.RewardIfDestroyed(Self, CanOnlyRespawnManually)

    RespawnAllProps(InAgent : ?agent): void=
        for (Definition : DestructibleProps):
            Definition.Respawn()


Category_RespawnSettings<public><localizes>:message = "Manual respawning"

ToolTip_DestructibleProps<public><localizes>:message = "The list of destructible props that grant rewards when destroyed."
ToolTip_UpdateRate<public><localizes>:message = "The rate at which the manager checks if props have been destroyed."
ToolTip_OnlyRespawnManually<public><localizes>:message = "If true, props will only respawn when manually triggered, ignoring the AutoRespawn property."
ToolTip_ManualRespawn<public><localizes>:message = "The trigger that manually respawns all props."
