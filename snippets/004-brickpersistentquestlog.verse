# Knowledge-neutral Verse snippet; project names and asset IDs generalized.
# id: snippet-004

using { /UnrealEngine.com/Temporary/Diagnostics }
using { /Verse.org/Simulation }
using { /Fortnite.com/Devices }


# ====================================================================================================================================
# Stores the quest progress per-player - this is persistent data
# ====================================================================================================================================
var GlobalPersistentData_QuestLog : weak_map(player, BRICK_persistent_quest_log) = map{}

BRICK_persistent_quest_log<public> := class<persistable><final>:
    CompletedQuests : []string = array{}
    ActiveQuests : []string = array{}

(InAgent : agent).LoadBRICKQuestLog()<decides><transacts>: BRICK_persistent_quest_log=
    var Result : ?BRICK_persistent_quest_log = false

    if(FoundData := GlobalPersistentData_QuestLog[player[InAgent]]):
        set Result = option{FoundData}
    else:
        set Result = option{BRICK_persistent_quest_log{}}
        if (InAgent.SaveBRICKQuestLog[Result?]) {}

    Result?

# Set the persistent data stored for this player.
(InAgent : agent).SaveBRICKQuestLog<public>(InQuestLog : BRICK_persistent_quest_log)<decides><transacts>: void=
    var Result : logic = false

    if (set GlobalPersistentData_QuestLog[player[InAgent]] = InQuestLog):
        set Result = true
    else:
        Logger: log = log{Channel:= BRICK_quest_log, DefaultLevel:= log_level.Normal}
        Logger.Print("Failed to save BRICK_fortplayer_persistence_data.", ?Level:= log_level.Error)

    Result?
