# Knowledge-neutral Verse snippet; project names and asset IDs generalized.
# id: snippet-005

using { /UnrealEngine.com/Temporary/Diagnostics }
using { /Verse.org/Simulation }
using { /Fortnite.com/Characters }
using { /Fortnite.com/Devices }

using { BRICKModule_StudRewards}

# ============================================================================================================================================
# Definition of a completable quest with dialog setup and reward
# ============================================================================================================================================
en_BRICK_quest_progress<public> := enum{ None, NotStarted, IsActive, CanBeCompleted, Completed }
BRICK_quest_log<public> := class(log_channel) {}

BRICK_quest_repeatable<public> := class<concrete><unique>(BRICK_quest) {}
BRICK_quest<public> := class<concrete><unique>():
    @editable:
        ToolTip := ToolTip_Quest_Name
    Name<public> : string = "UNIQUE Quest Name"

    @editable
    DisplayText<public> : BRICK_quest_info = BRICK_quest_info{}

    @editable:
        ToolTip := ToolTip_EventBindings
    EventBindings<public> : ?BRICK_quest_eventbindings = false
    
    @editable:
        ToolTip := Tooltip_Tracker
    TrackerDevice<public> : tracker_device = tracker_device{}
    
    @editable:
        ToolTip := Tooltip_Beacon
    GuidingBeacons<public> : []beacon_device = array{}

    @editable:
        ToolTip := ToolTip_Quest_Reward
    Reward<public> : BRICK_stud_spawner = BRICK_stud_spawner{}

    var CompletedObjectives : [agent]logic = map{}
    var TrackerDeviceQuestGiver : tracker_device = tracker_device{}
    var ReturnBeacon : beacon_device = beacon_device{}
    Logger<protected>: log = log{Channel := BRICK_quest_log, DefaultLevel:= log_level.Normal}

    # --------------------------------------------------------------------------------------------------------------------------------------------
    # API
    # --------------------------------------------------------------------------------------------------------------------------------------------
    # This is called by the quest giver to setup the quest with the required devices
    Setup<public>(InCreativeDevice : creative_device, InTrackerDeviceQuestGiver : tracker_device, InReturnBeacon : beacon_device): void=
        set TrackerDeviceQuestGiver = InTrackerDeviceQuestGiver
        set ReturnBeacon = InReturnBeacon
        TrackerDevice.CompleteEvent.Subscribe(OnCompletedObjective)            
        TrackerDevice.SetTitleText(ToMessage(Name))
        TrackerDevice.SetDescriptionText(ToMessage(DisplayText.Description))
        Reward.Init(InCreativeDevice)


    # Start the quest for a player, add it to the quest log and update the tracker device
    StartQuest<public>(InAgent : agent): void=
        if (ValidQuestLog := InAgent.LoadBRICKQuestLog[]):
            TrackerDevice.Assign(InAgent)
        
            for (Beacon : GuidingBeacons):
                Beacon.AddToShowList(InAgent)

            if (ValidTrigger := EventBindings?.Trigger_OnStartedQuest?):
                ValidTrigger.Trigger(InAgent)
            
            # Update persistent quest log
            UpdatedQuestLog := BRICK_persistent_quest_log:
                ActiveQuests := ValidQuestLog.ActiveQuests + array{Name},
                CompletedQuests := ValidQuestLog.CompletedQuests

            if (InAgent.SaveBRICKQuestLog[UpdatedQuestLog]):
                Logger.Print("Quest Log updated.")
        else:
            Logger.Print("StartQuest failed - no quest log found.", ?Level := log_level.Error)

    # Abort the quest for a player, remove it from the quest log and update the tracker device
    AbortQuest<public>(InAgent : agent): void=
        if (ValidQuestLog := InAgent.LoadBRICKQuestLog[]): 

            TrackerDevice.Reset(InAgent)

            for (Beacon : GuidingBeacons):
                Beacon.RemoveFromShowList(InAgent)

            if (ValidTrigger := EventBindings?.Trigger_OnAbortedQuest?):
                ValidTrigger.Trigger(InAgent)

            # Update persistent quest log
            var UpdatedActiveQuests : []string = array{}

            for (ActiveQuest : ValidQuestLog.ActiveQuests; ActiveQuest <> Name):
                set UpdatedActiveQuests += array{ActiveQuest}

            UpdatedQuestLog := BRICK_persistent_quest_log{
                ActiveQuests := UpdatedActiveQuests,
                CompletedQuests := ValidQuestLog.CompletedQuests}

            if (InAgent.SaveBRICKQuestLog[UpdatedQuestLog]) {}
        else:
            Logger.Print("AbortQuest failed - no quest log found.", ?Level := log_level.Error)

    # Complete the quest for a player, remove it from the quest log and update the tracker device, handle rewards
    CompleteQuest<public>(InAgent : agent): void=
        if (ValidQuestLog := InAgent.LoadBRICKQuestLog[]):

            TrackerDevice.Complete(InAgent)
            ReturnBeacon.RemoveFromShowList(InAgent)

            if (BRICK_quest_repeatable[Self]):
                TrackerDevice.Reset(InAgent)

            for (Beacon : GuidingBeacons):
                Beacon.RemoveFromShowList(InAgent)

            if (ValidCharacter := InAgent.GetFortCharacter[]):
                Reward.SpawnStudsAtLocation(ValidCharacter.GetTransform().Translation)
            else:
                Print("CompleteQuest failed - no valid FortCharacter found.")

            if (ValidTrigger := EventBindings?.Trigger_OnCompletedQuest?):
                ValidTrigger.Trigger(InAgent)

            # Update persistent quest log
            var UpdatedActiveQuests : []string = array{}

            for (ActiveQuest : ValidQuestLog.ActiveQuests; ActiveQuest <> Name):
                set UpdatedActiveQuests += array{ActiveQuest}
                
            UpdatedQuestLog := BRICK_persistent_quest_log{
                ActiveQuests := UpdatedActiveQuests,
                CompletedQuests := ValidQuestLog.CompletedQuests + array{Name}}

            if (InAgent.SaveBRICKQuestLog[UpdatedQuestLog]) {}
        else:
            Logger.Print("CompleteQuest failed - no quest log found.", ?Level := log_level.Error)

    # --------------------------------------------------------------------------------------------------------------------------------------------
    # Helper functions to determine the state of the quest
    # --------------------------------------------------------------------------------------------------------------------------------------------
    IsActive<public>(InAgent : agent)<transacts><decides>: void=
        var Result : logic = false

        if (ValidQuestLog := InAgent.LoadBRICKQuestLog[]):
            if (ValidQuestLog.ActiveQuests.Find[Name]):
                set Result = true
        Result?

    CanBeStarted<public>(InAgent : agent)<transacts><decides>: void=
        var Result : logic = true

        if (ValidQuestLog := InAgent.LoadBRICKQuestLog[]):

            if (ValidQuestLog.ActiveQuests.Find[Name]):
                set Result = false

            if (ValidQuestLog.CompletedQuests.Find[Name]):
                if (BRICK_quest_repeatable[Self]) {}
                else: 
                    set Result = false
        else:
            Logger.Print("No quest log found for agent.")
            set Result = false

        Result?
        
    CanBeCompleted<public>(InAgent : agent)<transacts><decides>: void=
        var Result : logic = false

        if (CompletedObjectives[InAgent] = true):
            set Result = true

        Result?

    HasBeenCompleted<public>(InAgent : agent)<transacts><decides>: void=
        var Result : logic = false

        if (ValidQuestLog := InAgent.LoadBRICKQuestLog[]):
            if (ValidQuestLog.CompletedQuests.Find[Name]):
                set Result = true
    
        Result?
    
    OnCompletedObjective<public>(InAgent : agent): void=
        if (set CompletedObjectives[InAgent] = true) {}
        
        TrackerDeviceQuestGiver.Assign(InAgent)
        ReturnBeacon.AddToShowList(InAgent)
        
# --------------------------------------------------------------------------------------------------------------------------------------------
# Quest Info class, used to pass the quest description and dialog texts to the dialog UI
# --------------------------------------------------------------------------------------------------------------------------------------------
BRICK_quest_info<public> := class<concrete>():
    @editable:
        ToolTip := Tooltip_ObjectiveDescription
    Description<public> : string = "Quest description"

    @editable_text_box:
        ToolTip := Tooltip_DialogText_Start
        MultiLine := true
    DialogText_Start<public> : string = ""
    
    @editable_text_box:
        ToolTip := Tooltip_DialogText_Complete
        MultiLine := true
    DialogText_Complete<public> : string = ""
        
# ============================================================================================================================================
# Event binding interface
# ============================================================================================================================================
BRICK_quest_eventbindings<public> := class<concrete>():
    @editable
    Trigger_OnStartedQuest : ?trigger_device = false
    @editable
    Trigger_OnAbortedQuest : ?trigger_device = false
    @editable
    Trigger_OnCompletedQuest : ?trigger_device = false

ToolTip_Quest_Name<public><localizes>:message = "The name of the quest as it will appear in the tracker_device and dialog UI. MUST be unique!"
ToolTip_EventBindings<public><localizes>:message = "Trigger Devices for linking other Fortnite Creative devices to this quest."
Tooltip_Tracker<public><localizes>:message = "The tracker_device that is required to complete for finishing the quest."
Tooltip_Beacon<public><localizes>:message = "This beacon can be used to guide the player towards the objective of the quest."
ToolTip_Quest_Reward<public><localizes>:message = "The Stud reward that will be given to the player when the quest is completed."
Tooltip_ObjectiveDescription<public><localizes>:message = "The description as visible in the tracker_device UI."
Tooltip_DialogText_Start<public><localizes>:message = "The text that will be displayed in the dialog when the quest is started."
Tooltip_DialogText_Complete<public><localizes>:message = "The text that will be displayed in the dialog when the quest is completed."
