# Knowledge-neutral Verse snippet; project names and asset IDs generalized.
# id: snippet-006

using { /UnrealEngine.com/Temporary/Diagnostics }
using { /UnrealEngine.com/Temporary/UI }
using { /UnrealEngine.com/Temporary/SpatialMath }
using { /Verse.org/Simulation }
using { /Fortnite.com/UI }
using { /Verse.org/Assets }
using { /Verse.org/Colors }
using { /Verse.org/Colors/NamedColors }

ButtonCloseText<localizes><public> : message = "Close"
ToMessage<public><localizes>(InString : string)<transacts>: message = "{InString}"

# =====================================================================================================================================
# Quest giver UI
# Dialog Box
# =====================================================================================================================================
en_BRICK_quest_dialog_response<public> := enum{ Accept, Decline, Close }

BRICK_dialog_config<public> := struct():
    Title<public> : ?message = false
    Text : ?message = false
    Button_1_Text : ?message = false
    Button_2_Text : ?message = false
    BackgroundTexture: ?texture = false
    DividerTexture: ?texture = false

BRICK_quest_dialog := class():
    var OwningPlayer : ?agent = false

    var Canvas : ?widget = false
    var Button_1 : text_button_base = button_loud{}
    var Callback_ButtonClicked_1 : ?cancelable = false
    var Button_2 : text_button_base = button_loud{}
    var Callback_ButtonClicked_2 : ?cancelable = false

    Button_Close : text_button_base = button_quiet{ DefaultText := ButtonCloseText}
    var Callback_ButtonClicked_Close : ?cancelable = false

    Logger<protected>: log = log{Channel := BRICK_quest_giver_device_log, DefaultLevel:= log_level.Normal}
    
    # ---------------------------------------------------------------------------------------------------------------------------------
    # Dialog UI Flow:
    # 1) Open
    # 2) Wait for player response
    # 3) Close
    # 4) Emit Dialog response as a command
    # ---------------------------------------------------------------------------------------------------------------------------------
    ShowDialog<public>(InAgent : agent, InDialogConfig : BRICK_dialog_config): void=
        set OwningPlayer = option{InAgent}

        # Setup Buttons
        Callback_Close := Button_Close.OnClick().Subscribe(OnButtonClicked_Close)
        set Callback_ButtonClicked_Close = option{Callback_Close}

        if (ButtonText_1 := InDialogConfig.Button_1_Text?):
            Button_1.SetText(ButtonText_1)
            CallBack_Button1 := Button_1.OnClick().Subscribe(OnButtonClicked_1)
            set Callback_ButtonClicked_1 = option{CallBack_Button1}

        if (ButtonText_2 := InDialogConfig.Button_2_Text?):
            Button_2.SetText(ButtonText_2)
            CallBack_Button2 := Button_2.OnClick().Subscribe(OnButtonClicked_2)
            set Callback_ButtonClicked_2 = option{CallBack_Button2}

        # Create Canvas and add to Player Screen
        set Canvas = option{MakeDialogCanvas(InDialogConfig)}

        if:
            PlayerUI := GetPlayerUI[player[OwningPlayer?]]
            ValidCanvas := Canvas?
        then:
            PlayerUI.AddWidget(ValidCanvas, player_ui_slot{ZOrder := 3, InputMode := ui_input_mode.All})
        else:
            Logger.Print("Failed to show Dialog UI for player.", ?Level := log_level.Error)

    # Accept / Complete Button
    OnButtonClicked_1(Message : widget_message): void=
        CompleteDialog(en_BRICK_quest_dialog_response.Accept)

    # Decline / Abort Button
    OnButtonClicked_2(Message : widget_message): void=
        CompleteDialog(en_BRICK_quest_dialog_response.Decline)

    # Close Button
    OnButtonClicked_Close(Message : widget_message): void=
        CompleteDialog(en_BRICK_quest_dialog_response.Close)


    OnDialogCompleted<public> : event(agent; en_BRICK_quest_dialog_response) = event(agent; en_BRICK_quest_dialog_response){}
    CompleteDialog<public>(InDialogCommand : en_BRICK_quest_dialog_response): void=
        # Unsubscribe from Button clicks
        if (ValidCallback := Callback_ButtonClicked_Close?):
            ValidCallback.Cancel()

        if (ValidCallback := Callback_ButtonClicked_1?):
            ValidCallback.Cancel()

        if (ValidCallback := Callback_ButtonClicked_2?):
            ValidCallback.Cancel()

        # Remove Canvas from Player Screen
        if: 
            ValidOwningPlayer := OwningPlayer?
            PlayerUI := GetPlayerUI[player[ValidOwningPlayer]]
            ValidCanvas := Canvas?
        then:
            PlayerUI.RemoveWidget(ValidCanvas)
            set Canvas = false
            set OwningPlayer = false
            OnDialogCompleted.Signal(ValidOwningPlayer; InDialogCommand)
    # ---------------------------------------------------------------------------------------------------------------------------------


    # ---------------------------------------------------------------------------------------------------------------------------------
    # VERSE UI creation
    # Create a canvas with the Dialog UI
    # ---------------------------------------------------------------------------------------------------------------------------------
    MakeDialogCanvas<protected>(InDialogConfig : BRICK_dialog_config)<transacts>: canvas=
        Anchor_Root := vector2{X := 0.5, Y := 0.6}

        OutCanvas : canvas = canvas:
            Slots := array:
                
                # Background
                canvas_slot:
                    Anchors := anchors{ Minimum := Anchor_Root, Maximum := Anchor_Root}
                    SizeToContent := true
                    Offsets := margin{Top := 0.0, Left := -765.0}
                    Widget := if(BackgroundTexture := InDialogConfig.BackgroundTexture?)
                        {texture_block:
                            DefaultImage := BackgroundTexture
                            DefaultDesiredSize := vector2{X:= 1580.0, Y := 394.0}
                        }
                        else {color_block{
                            DefaultDesiredSize := vector2{X:= 1580.0, Y := 394.0}
                        }}

                # Divider
                canvas_slot:
                    Anchors := anchors{ Minimum := Anchor_Root, Maximum := Anchor_Root}
                    SizeToContent := true
                    Offsets := margin{Top := 0.0, Left := -765.0}
                    Widget := if(DividerTexture := InDialogConfig.DividerTexture?)
                    {texture_block:
                        DefaultImage := DividerTexture
                        DefaultDesiredSize := vector2{X:= 1580.0, Y := 394.0}
                    }
                    else {color_block{
                        DefaultDesiredSize := vector2{X:= 1580.0, Y := 394.0}
                    }}

                # Text
                canvas_slot:
                    Anchors := anchors{ Minimum := Anchor_Root + vector2{ X := -0.0, Y := 0.0}, Maximum := Anchor_Root + vector2{ X := -0.0, Y := 0.0}}
                    Widget := overlay:
                        Slots := array:
                            # Title
                            overlay_slot:
                                VerticalAlignment := vertical_alignment.Top
                                HorizontalAlignment := horizontal_alignment.Left
                                Padding := margin{Top := 80.0, Left := -600.0}
                                Widget := if (ValidContent := InDialogConfig.Title?) 
                                    {text_block:
                                        DefaultText := ValidContent
                                        DefaultTextColor := Orange
                                        DefaultJustification := text_justification.Left
                                        DefaultShadowColor:= MakeColorFromSRGB(0.05, 0.05, 0.05)
                                        DefaultShadowOffset := option{vector2{X := 4.0, Y := 4.0}}
                                        DefaultShadowOpacity := 1.0
                                    }
                                    else {text_block{}}

                            # Dialog Body
                            overlay_slot:
                                VerticalAlignment := vertical_alignment.Top
                                HorizontalAlignment := horizontal_alignment.Left
                                Padding := margin{Top := 135.0, Left := -600.0}
                                Widget := if (ValidContent := InDialogConfig.Text?) 
                                    {text_block:
                                        DefaultText := ValidContent
                                        DefaultTextColor := White
                                        DefaultJustification := text_justification.Left
                                        DefaultShadowColor:= MakeColorFromSRGB(0.05, 0.05, 0.05)
                                        DefaultShadowOffset := option{vector2{X := 2.0, Y := 2.0}}
                                        DefaultShadowOpacity := 1.0
                                    } 
                                    else {text_block{}}

                # Button 1
                canvas_slot:
                    Anchors := anchors{ Minimum := Anchor_Root + vector2{ X := -0.18, Y := 0.25 }, Maximum := Anchor_Root + vector2{ X := -0.18, Y := 0.25 }}
                    Widget := if (ValidContent := InDialogConfig.Button_1_Text?) { Button_1 } else {text_block{}}
                # Button 2
                canvas_slot:
                    Anchors := anchors{ Minimum := Anchor_Root + vector2{ X := 0.0225, Y := 0.25 }, Maximum := Anchor_Root + vector2{ X := 0.0225, Y := 0.25 }}
                    Widget := if (ValidContent := InDialogConfig.Button_2_Text?) { Button_2 } else {text_block{}}
                # Button Close
                canvas_slot:
                    Anchors := anchors{ Minimum := Anchor_Root + vector2{ X := 0.275, Y := 0.05 }, Maximum := Anchor_Root + vector2{ X := 0.275, Y := 0.05  }}
                    Widget := Button_Close
