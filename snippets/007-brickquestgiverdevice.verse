# Knowledge-neutral Verse snippet; project names and asset IDs generalized.
# id: snippet-007

using { /Fortnite.com/Characters }
using { /Fortnite.com/Devices }
using { /Verse.org/Assets }
using { /Verse.org/Simulation }
using { /UnrealEngine.com/Temporary/Diagnostics }

# ====================================================================================================================================
# BRICK quest giver device that manages a linear quest system
# ====================================================================================================================================
BRICK_quest_giver_device_log<public> := class(log_channel) {}

BRICK_quest_giver_device := class(creative_device):

    @editable:
        ToolTip := Tooltip_QuestGiver_Name
    QuestGiverName<public> : string = "Quest Giver"

    @editable:
        ToolTip := Tooltip_Button_Interact
    Button_Interact : button_device = button_device{}

    @editable:
        Categories := array{Category_FinishedGuidance}
        ToolTip := Tooltip_QuestGiver_Tracker
    TrackerDevice<public> : tracker_device = tracker_device{}

    @editable:
        Categories := array{Category_FinishedGuidance}
        ToolTip := Tooltip_BeaconReturn
    ReturnBeacon<public> : beacon_device = beacon_device{}

    @editable:
        ToolTip := Tooltip_AutoProgressDialog
        Categories := array{Category_Settings}
    AutoProgressDialog<public> : logic = true

    @editable:
        ToolTip := Tooltip_ResetProgress
        Categories := array{Category_Settings}
    Trigger_ResetProgress : ?trigger_device = false

    @editable:
        ToolTip := Tooltip_Quests
    Quests : []BRICK_quest = array{}

    @editable:
        Categories := array{Category_Appearance}
        ToolTip := Tooltip_BackgroundTexture
    BackgroundTexture : ?texture = false

    @editable:
        Categories := array{Category_Appearance}
        ToolTip := Tooltip_DividerTexture
    DividerTexture : ?texture = false

    Logger<protected>: log = log{Channel := BRICK_quest_giver_device_log, DefaultLevel:= log_level.Normal}

    OnBegin<override>()<suspends>:void=
        # Bind to button interaction event
        Button_Interact.InteractedWithEvent.Subscribe(OnRequestInteraction)
        Button_Interact.SetInteractionText(ToMessage("Talk to {QuestGiverName}"))

        # Bind to Reset request trigger
        if (ValidTrigger := Trigger_ResetProgress?):
            ValidTrigger.TriggeredEvent.Subscribe(OnResetProgress)
        
        # Setup Tracker Device (Return to QuestGiver message)
        TrackerDevice.SetTitleText(ToMessage("Completed Quest"))
        TrackerDevice.SetDescriptionText(ToMessage("Return to {QuestGiverName}"))

        # Setup Event subscribes for all quests
        for (Quest : Quests):
            Quest.Setup(Self, TrackerDevice, ReturnBeacon)

    # ---------------------------------------------------------------------------------------------------------------------------------
    # Player Interaction with QuestGiver
    # 1) Show Dialog UI
    # 2) Wait for player interaction that closes the UI
    # 3) Process and apply Response from Dialog UI (Start / Complete / Abort quest / Close)
    # ---------------------------------------------------------------------------------------------------------------------------------
    OnRequestInteraction<public>(InAgent : agent): void=
        if (Quests.Length = 0):
            Logger.Print("OnRequestInteraction failed - no quests found.", ?Level := log_level.Error)
            return

        TrackerDevice.Complete(InAgent)
        spawn{InteractWithQuestGiver(InAgent)}

    InteractWithQuestGiver<private>(InAgent : agent)<suspends>: void=
        TrackerDevice.Complete(InAgent)
        QuestProgressData := GetQuestLineProgression(InAgent)

        # Create UI to be shown to the player
        DialogUI := BRICK_quest_dialog{}

        # Configure UI based on QuestLine progression 
        # iE. what button text to display and what Dialog option to show
        Title := if(ValidQuest := QuestProgressData(1)?) {ToMessage(ValidQuest.Name)} else {ToMessage("")}
        Dialog_Start := if(ValidQuest := QuestProgressData(1)?) {ToMessage(ValidQuest.DisplayText.DialogText_Start)} else {ToMessage("I don`t have any quests for you right now.")}
        Dialog_Complete := if(ValidQuest := QuestProgressData(1)?) {ToMessage(ValidQuest.DisplayText.DialogText_Complete)} else {ToMessage("")}

        var Text : ?message = false
        var Button_1_Text : string = ""
        var Button_2_Text : string = ""

        case (QuestProgressData(0)):
            en_BRICK_quest_progress.NotStarted =>
                set Button_1_Text = "Accept Quest "
                set Button_2_Text = "Decline Quest"
                set Text = option{Dialog_Start}
            en_BRICK_quest_progress.IsActive =>
                set Button_2_Text = "Abort Quest"
                set Text = option{Dialog_Start}
            en_BRICK_quest_progress.CanBeCompleted =>
                set Button_1_Text = "Finish Quest"
                set Button_2_Text = "Abort Quest"
                set Text = option{Dialog_Complete}
            en_BRICK_quest_progress.None =>
                set Text = option{Dialog_Start}
            _=>

        Button_1_Message : message = ToMessage(Button_1_Text)
        Button_2_Message : message = ToMessage(Button_2_Text)

        # Data to be passed to the UI class, the BRICK_quest_dialog will then process and display it
        DialogConfig := BRICK_dialog_config:
            Title := option{Title},
            Text := Text, 
            Button_1_Text := if(Button_1_Text.Length > 0) {option{Button_1_Message}} else {false},
            Button_2_Text := if(Button_2_Text.Length > 0) {option{Button_2_Message}} else {false},
            BackgroundTexture := BackgroundTexture,
            DividerTexture := DividerTexture

        # 1) Show UI
        DialogUI.ShowDialog(InAgent, DialogConfig)

        # 2) Wait for player interaction
        DialogResponse := DialogUI.OnDialogCompleted.Await()

        # 3) Process Response from the UI
        #    Which Button did the player press? Start / Complete / Abort / Close
        if (DialogResponse = en_BRICK_quest_dialog_response.Close) {return}

        # Process Response based on the players progress through the questline
        if (ValidQuest := QuestProgressData(1)?):
            case (QuestProgressData(0)):
                # Start new quest
                en_BRICK_quest_progress.NotStarted =>
                    if (DialogResponse = en_BRICK_quest_dialog_response.Accept):
                        ValidQuest.StartQuest(InAgent)

                # Complete active quest
                en_BRICK_quest_progress.CanBeCompleted =>
                    if (DialogResponse = en_BRICK_quest_dialog_response.Accept):
                        ValidQuest.CompleteQuest(InAgent)
                        
                        # Open next quest Dialog (if auto Dialog progression is enabled)
                        if (AutoProgressDialog?) {spawn{InteractWithQuestGiver(InAgent)}}

                # Abort quest
                    if (DialogResponse = en_BRICK_quest_dialog_response.Decline):
                        ValidQuest.AbortQuest(InAgent)

                # Abort quest
                en_BRICK_quest_progress.IsActive =>
                    if (DialogResponse = en_BRICK_quest_dialog_response.Decline):
                        ValidQuest.AbortQuest(InAgent)

                _=>
                    # No Response to process
        else:
            Logger.Print("InteractWithQuestGiver failed - no valid quest found.", ?Level := log_level.Error)

    # ---------------------------------------------------------------------------------------------------------------------------------


    # ---------------------------------------------------------------------------------------------------------------------------------
    # QuestLine Progression
    # Loop over all the quests and return the first quest that can either be started, completed or is active
    # ---------------------------------------------------------------------------------------------------------------------------------
    GetQuestLineProgression<public>(InAgent : agent)<transacts>: tuple(en_BRICK_quest_progress, ?BRICK_quest)=
        if (Quests.Length = 0):
            Logger.Print("GetQuestLineProgression: QuestGiver has no quests.", ?Level := log_level.Warning)

        for (Quest : Quests):
            # Quest can be started
            if (Quest.CanBeStarted[InAgent]):
                return (en_BRICK_quest_progress.NotStarted, option{Quest})

            # Quest is active and can be completed or aborted
            if (Quest.IsActive[InAgent]):
                if (Quest.CanBeCompleted[InAgent]):
                    return (en_BRICK_quest_progress.CanBeCompleted, option{Quest})
                return (en_BRICK_quest_progress.IsActive, option{Quest})

        # No available quests
        return (en_BRICK_quest_progress.None, false)
    # ---------------------------------------------------------------------------------------------------------------------------------

    # This function is called when the Trigger_ResetProgress is activated and resets all quest progress for the activating player
    OnResetProgress<public>(InAgent : ?agent): void=
        if:
            ValidAgent := InAgent?
            ValidQuestLog := ValidAgent.LoadBRICKQuestLog[]
        then:
            UpdatedQuestLog := BRICK_persistent_quest_log:
                ActiveQuests := array{},
                CompletedQuests := array{}
            if (ValidAgent.SaveBRICKQuestLog[UpdatedQuestLog]) {}
        else:
            Logger.Print("ResetPersistentData failed - no quest log found.", ?Level := log_level.Error)


Tooltip_QuestGiver_Name<public><localizes>:message = "The name of the quest giver as it will appear in the tracker UI and button interaction text."
Tooltip_Button_Interact<public><localizes>:message = "The button that the player must press to interact with the quest giver."
Tooltip_QuestGiver_Tracker<public><localizes>:message = "The tracker_device used to display the return to questgiver message."
Tooltip_BeaconReturn<public><localizes>:message = "This beacon can be used to guide the player back to the quest giver once a quest is completed."
Tooltip_AutoProgressDialog<public><localizes>:message = "If true, the quest giver will automatically show the next quest Dialog after the player has completed a quest."
Tooltip_ResetProgress<public><localizes>:message = "Trigger device that will reset the progress in all quests for the activating player."
Tooltip_Quests<public><localizes>:message = "The quests that the quest giver can give to the player."
Tooltip_BackgroundTexture<public><localizes>:message = "The texture to use for the background of the dialog UI."
Tooltip_DividerTexture<public><localizes>:message = "The texture to use for the divider in the dialog UI."

Category_Settings<public><localizes>:message = "Settings"
Category_FinishedGuidance<public><localizes>:message = "Quest finished guidance"
Category_Appearance<public><localizes>:message = "Appearance"
