# Knowledge-neutral Verse snippet; project names and asset IDs generalized.
# id: snippet-008

using { /Fortnite.com/Devices }
using { /Fortnite.com/UI }
using { /Verse.org/Colors }
using { /Verse.org/Colors/NamedColors }
using { /Verse.org/Simulation }
using { /UnrealEngine.com/Temporary/Diagnostics }
using { /UnrealEngine.com/Temporary/SpatialMath }
using { /UnrealEngine.com/Temporary/UI }

# ====================================================================================================================================
# Displays a stud counter on the screen
# Uses a conditional button device to periodically update the stud count
# ====================================================================================================================================
BRICK_stud_counter_device := class(creative_device):

    @editable:
        ToolTip := Tooltip_RegisterTrigger
    RegisterTrigger : trigger_device = trigger_device{}

    @editable:
        ToolTip := Tooltip_HudMessage
    HUDMessageDevice : hud_message_device = hud_message_device{}

    @editable:
        ToolTip := Tooltip_ConditionalButton
    ConditionalButton : conditional_button_device = conditional_button_device{}

    var RegisteredPlayers : []agent = array{}

    OnBegin<override>()<suspends>:void=
        RegisterTrigger.TriggeredEvent.Subscribe(RegisterPlayer)

        # Wait to ensure all players are spawned before registering them
        Sleep(1.0)
        AllPlayers : []player = GetPlayspace().GetPlayers()
        for (FoundPlayer : AllPlayers):
            RegisterPlayer(option{FoundPlayer})

    # A player has to be registered with the device to start updating the stud count
    RegisterPlayer<private>(InAgent : ?agent):void=
        if (ValidAgent := InAgent?):
            if (RegisteredPlayers.Find[InAgent]) {return}
            set RegisteredPlayers += array{ValidAgent}
            spawn{Loop_UpdateStudCountForAgent(ValidAgent)}

    # Periodically update the stud count for a player
    Loop_UpdateStudCountForAgent<private>(InAgent : agent)<suspends>: void=
        var PreviousStudCount : int = -1
        
        loop:
            Sleep(0.15)
            CurrentCount := ConditionalButton.GetItemCount(InAgent, 0)

            # Only update the HUD message if the count has changed and not -1 (while player is dead)
            if (CurrentCount <> PreviousStudCount and CurrentCount <> -1):
                set PreviousStudCount = ConditionalButton.GetItemCount(InAgent, 0)
                HUDMessageDevice.SetText(ToMessage("{CurrentCount}"))
                HUDMessageDevice.Show(InAgent)


ToMessage<public><localizes>(InString : string)<transacts>: message = "{InString}"
Tooltip_RegisterTrigger<public><localizes>:message = "The trigger device that will register the player to the stud counter."
Tooltip_HudMessage<public><localizes>:message = "The HUD message device that will display the stud count."
Tooltip_ConditionalButton<public><localizes>:message = "The button device that will update the stud count."
