# Knowledge-neutral Verse snippet; project names and asset IDs generalized.
# id: snippet-010

using { /Fortnite.com/Devices }
using { /Verse.org/Colors }
using { /Verse.org/Simulation }
using { /Verse.org/Simulation/Tags }
using { /UnrealEngine.com/Assets }
using { /UnrealEngine.com/Temporary/Diagnostics }
using { /UnrealEngine.com/Temporary/SpatialMath }

# ============================================================================================================================================
# Distributes Items by using a pool of ItemSpawners and teleporting them to the desired location
# ============================================================================================================================================
BRICK_stud_system_log<public> := class(log_channel) {}
BRICK_stud_system_tag<public> := class(tag) {}
BRICK_stud_pool_tag := class(BRICK_stud_system_tag) {} # The device MUST have this tag to be recognized as a Stud Pool

BRICK_stud_pool<public> := class(creative_device):
   var ObjectPool_ItemSpawners_Gravity : []pool_object_pickup = array{}
   var ObjectPool_ItemSpawners_Toss : []pool_object_pickup = array{}

   Logger<protected>: log = log{Channel := BRICK_stud_system_log, DefaultLevel:= log_level.Normal}

   OnBegin<override>()<suspends>:void=
      Sleep(0.5)
      
      # Find Pre-spawned ItemSpawners via a gameplay tag and add them to the pool
      var PoolSize : int = 0

      for (FoundObject : Self.FindCreativeObjectsWithTag(BRICK_stud_itemspawner_toss_tag{})):
         if (ValidItemSpawner := item_spawner_device[FoundObject]):
            set PoolSize += 1
            set ObjectPool_ItemSpawners_Toss += array{pool_object_pickup{ItemSpawner := option{ValidItemSpawner}}}

      for (FoundObject : Self.FindCreativeObjectsWithTag(BRICK_stud_itemspawner_gravity_tag{})):
         if (ValidItemSpawner := item_spawner_device[FoundObject]):
            set PoolSize += 1
            set ObjectPool_ItemSpawners_Gravity += array{pool_object_pickup{ItemSpawner := option{ValidItemSpawner}}}

      Logger.Print("Initialized {PoolSize} Item Spawners")

   # Spawn a number of studs at a location using the available pool objects
   SpawnStudsAtLocation<public>(InLocation : vector3, ?InAmount : int = 1, ?InSpawnType : EN_ItemSpawnerConfig = EN_ItemSpawnerConfig.Toss): void=
      for (Index := 1..InAmount):
         if (ValidPoolObject := GetAvailablePoolObject[InSpawnType]):
            spawn{ValidPoolObject.SpawnStudsAtLocation(InLocation)}

   # Retrieve an available pool object from the pool if possible
   GetAvailablePoolObject<private>(InSpawnType : EN_ItemSpawnerConfig)<transacts><decides> : pool_object_pickup=
      var OutObject : ?pool_object_pickup = false

      case (InSpawnType):

         EN_ItemSpawnerConfig.Toss =>
            for (Object : ObjectPool_ItemSpawners_Toss):
               if (Object.IsAvailable?):
                  set OutObject = option{Object}

         EN_ItemSpawnerConfig.Gravity =>
            for (Object : ObjectPool_ItemSpawners_Gravity):
               if (Object.IsAvailable?):
                  set OutObject = option{Object}
      OutObject?

# Public function to retrieve the Stud Pool via the assigned Tag
# NOTE: BRICK_stud_pool is expected to be used as a singleton!
(InCreativeDevice : creative_device).GetBRICKStudPool<public>()<transacts><decides>: BRICK_stud_pool=
   var OutResult : ?BRICK_stud_pool = false

   for (FoundObject : InCreativeDevice.FindCreativeObjectsWithTag(BRICK_stud_pool_tag{})):

      if (ValidItemSpawner := BRICK_stud_pool[FoundObject]):
         set OutResult = option{ValidItemSpawner}
   OutResult?

# --------------------------------------------------------------------------------------------------------------------------------------------
# A single pool object constructed and maintained by the Stud Pool device
# --------------------------------------------------------------------------------------------------------------------------------------------
BRICK_stud_itemspawner_tag := class(BRICK_stud_system_tag) {}
BRICK_stud_itemspawner_toss_tag := class(BRICK_stud_itemspawner_tag) {}
BRICK_stud_itemspawner_gravity_tag := class(BRICK_stud_itemspawner_tag) {}
EN_ItemSpawnerConfig<public> := enum{ Toss, Gravity }

pool_object_pickup := class():
   var ItemSpawner<public> : ?item_spawner_device = false
   var IsAvailable<public> : logic = true
   var Location : vector3 = vector3{}

   SpawnStudsAtLocation(InLocation : vector3)<suspends>: void=
      if (not IsAvailable?):
         return

      set IsAvailable = false
      set Location = InLocation

      if (ValidDevice := ItemSpawner?):
         # Offset the spawning height of the item_spanwer_device to avoid spawning studs in the ceiling
         if (ValidDevice.TeleportTo[InLocation + vector3{Z := - 125.0}, IdentityRotation()]) {}
         Sleep(0.25)
         ValidDevice.SpawnItem()
         Sleep(0.25)
         Release()

   Release<public>(): void=
      if (IsAvailable?):
         return
         
      if {ItemSpawner?.TeleportTo[vector3{}, IdentityRotation()]}
      set IsAvailable = true
