# Knowledge-neutral Verse snippet; project names and asset IDs generalized.
# id: snippet-011

using { /Fortnite.com/Characters }
using { /Fortnite.com/Devices }
using { /Verse.org }
using { /Verse.org/Assets }
using { /Verse.org/Colors }
using { /Verse.org/Random }
using { /Verse.org/Simulation }
using { /UnrealEngine.com/Assets }
using { /UnrealEngine.com/Temporary/Diagnostics }
using { /UnrealEngine.com/Temporary/SpatialMath }

# ============================================================================================================================================
# Attacks for the npc_attack_controller
# ============================================================================================================================================
# Base class for NPC attacks
npc_attack := class<abstract>():
    # Override in derived classes with your custom behavior
    StartAttack(InNPCAgent : agent, InNPCCharacter : fort_character, InTargetCharacter : fort_character)<suspends>: void= {}

    OnAttackCompleted<public> : event() = event(){}
    # Override in derived classes to uninitialize any attack-specific state
    EndAttack(InNPCAgent : agent, InNPCCharacter : fort_character, InTargetCharacter : fort_character): void=
        OnAttackCompleted.Signal()


# Stun the player for a short duration
npc_attack_stun := class<concrete>(npc_attack):
    @editable
    StunDuration : float = 2.0
    @editable
    PropAsset_AttackIndicator : creative_prop_asset = DefaultCreativePropAsset
    @editable
    HUDMessage : hud_message_device = hud_message_device{}
    var SpawnedProp : ?creative_prop = false

    StartAttack<override>(InNPCAgent : agent, InNPCCharacter : fort_character, InTargetCharacter : fort_character)<suspends>: void=
        # Announce Attack
        if (InTargetAgent := InTargetCharacter.GetAgent[]):
            HUDMessage.Show(InTargetAgent)

        # Stun Player
        InTargetCharacter.PutInStasis(stasis_args{AllowTurning := false, AllowFalling := false, AllowEmotes := false})

        # Spawn VFX to indicate the attack at Players location
        SpawnedVFX := SpawnProp(PropAsset_AttackIndicator, InTargetCharacter.GetTransform().Translation + vector3{Z := -40.0}, IdentityRotation())
        if (ValidSpawnedProp := SpawnedVFX(0)?) {set SpawnedProp = SpawnedVFX(0)}

        # Wait for the duration of the stun and then end the Attack
        Sleep(StunDuration)
        EndAttack(InNPCAgent, InNPCCharacter, InTargetCharacter)
    
    EndAttack<override>(InNPCAgent : agent, InNPCCharacter : fort_character, InTargetCharacter : fort_character): void=
        if (ValidProp := SpawnedProp?) {ValidProp.Dispose()}
        # Release Player from Stun
        InTargetCharacter.ReleaseFromStasis()
        (super:)EndAttack(InNPCAgent, InNPCCharacter, InTargetCharacter)


# NPC will stop to "cast" an AOE attack that follows the player for a short while and then explodes, dealing AOE damage with an explosive device
npc_attack_ranged_aoe := class<concrete>(npc_attack):
    @editable
    ExplosiveDevice : explosive_device = explosive_device{}
    @editable
    PropAsset_AttackIndicator : creative_prop_asset = DefaultCreativePropAsset
    var SpawnedProps : []creative_prop = array{}

    # Stop NPC, spawn VFX at NPC and Player location, follow Player with VFX for a short while and then explode the device there
    StartAttack<override>(InNPCAgent : agent, InNPCCharacter : fort_character, InTargetCharacter : fort_character)<suspends>: void=
        (super:)StartAttack(InNPCAgent, InNPCCharacter, InTargetCharacter)
        ExplosiveDevice.Reset()

        race:
            OnAttackCompleted.Await()

            block:
                InNPCCharacter.PutInStasis(stasis_args{AllowTurning := true, AllowFalling := false, AllowEmotes := false})
                var TargetLocation : vector3 = InTargetCharacter.GetTransform().Translation + vector3{Z := -40.0}
                
                # Spawn VFX to indicate the attack at the NPC and Player location
                SourceVFX := SpawnProp(PropAsset_AttackIndicator, InNPCCharacter.GetTransform().Translation + vector3{Z := -40.0}, IdentityRotation())
                if (ValidProp := SourceVFX(0)?):
                    set SpawnedProps += array{ValidProp}
                TargetVFX := SpawnProp(PropAsset_AttackIndicator, TargetLocation, IdentityRotation())

                if (ValidProp := TargetVFX(0)?):
                    set SpawnedProps += array{ValidProp}

                # VFX follows Player for a short while
                for (Index := 0..10):
                    set TargetLocation = InTargetCharacter.GetTransform().Translation + vector3{Z := -40.0}
                    if (ValidProp := TargetVFX(0)?) {ValidProp.MoveTo(TargetLocation, IdentityRotation(), 0.2)}

                # Explode the device after a short delay to deal AOE damage
                if (ExplosiveDevice.TeleportTo[InTargetCharacter.GetTransform().Translation, IdentityRotation()]):
                    Sleep(0.5)
                    ExplosiveDevice.Explode(InNPCAgent)
                    Sleep(0.25)

                EndAttack(InNPCAgent, InNPCCharacter, InTargetCharacter)

    # Undo any state changes like stasis, dispose of spawned props and reset the explosive device
    EndAttack<override>(InNPCAgent : agent, InNPCCharacter : fort_character, InTargetCharacter : fort_character): void=
        InNPCCharacter.ReleaseFromStasis()
        
        for (SpawnedProp : SpawnedProps):
            SpawnedProp.Dispose()

        set SpawnedProps = array{}
        ExplosiveDevice.Reset()
        (super:)EndAttack(InNPCAgent, InNPCCharacter, InTargetCharacter)
