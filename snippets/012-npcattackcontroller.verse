# Knowledge-neutral Verse snippet; project names and asset IDs generalized.
# id: snippet-012

using { /Fortnite.com/Characters }
using { /Fortnite.com/Devices }
using { /Verse.org }
using { /Verse.org/Assets }
using { /Verse.org/Colors }
using { /Verse.org/Random }
using { /Verse.org/Simulation }
using { /UnrealEngine.com/Assets }
using { /UnrealEngine.com/Temporary/Diagnostics }
using { /UnrealEngine.com/Temporary/SpatialMath }

# ============================================================================================================================================
# Attack controller
# -> periodically interrupts the guard and does a random attack
# ============================================================================================================================================
BRICK_npc_attack_controller_log := class(log_channel){}
npc_attack_controller := class(creative_device):
    @editable:
        ToolTip := Tooltip_NPCSpawner
    NPCSpawner : npc_spawner_device = npc_spawner_device{}

    @editable:
        ToolTip := Tooltip_Cinematic
    AwaitCinematicSequence : ?cinematic_sequence_device = false

    @editable_number(float):
        MinValue := option{0.0}
        ToolTip := Tooltip_TimeBetweenAttacksMin
    TimeBetweenAttacks_Min : float = 2.5

    @editable_number(float):
        MinValue := option{0.0}
        ToolTip := Tooltip_TimeBetweenAttacksMax
    TimeBetweenAttacks_Max : float = 5.0

    @editable:
        ToolTip := Tooltip_Attacks
    Attacks : []npc_attack = array{}

    Logger<protected>: log = log{Channel := BRICK_npc_attack_controller_log, DefaultLevel:= log_level.Normal}

    OnBegin<override>()<suspends>: void=
        NPCSpawner.SpawnedEvent.Subscribe(OnNPCSpawned)
        NPCSpawner.EliminatedEvent.Subscribe(OnEliminatedBehavior)

    # When an NPC is spawned, run the behavior on it
    OnNPCSpawned(InAgent : agent): void=
        if (FortChar := InAgent.GetFortCharacter[]):
            spawn{RunBehavior(InAgent, FortChar)}

        else:
            Logger.Print("boss_attack_controller - OnNPCSpawned failed: no valid FortChar on Agent", ?Level:= log_level.Warning)

    # When an NPC is eliminated, stop the behavior on it
    OnNPCEliminated : event() = event(){}
    OnEliminatedBehavior(InResult : device_ai_interaction_result): void=
        OnNPCEliminated.Signal()

    # Periodically interrupt the guard to determine and start an attack on the nearest player
    RunBehavior(InNPCAgent : agent, InNPCCharacter : fort_character)<suspends>: void=
        if (Cinematic := AwaitCinematicSequence?):
            Cinematic.StoppedEvent.Await()

        race:   
            OnNPCEliminated.Await()

            loop:
                Sleep(GetRandomFloat(TimeBetweenAttacks_Min, TimeBetweenAttacks_Max))
                if:
                    TargetPlayerCharacter := GetNearestPlayerCharacter[InNPCCharacter]
                    Attack := GetRandomAttack[]
                then:
                    Attack.StartAttack(InNPCAgent, InNPCCharacter, TargetPlayerCharacter)

        for (Attack : Attacks; TargetPlayerCharacter := GetNearestPlayerCharacter[InNPCCharacter]):
            Attack.EndAttack(InNPCAgent, InNPCCharacter, TargetPlayerCharacter)

    # --------------------------------------------------------------------------------------------------------------------------------------------
    # Helpers
    # --------------------------------------------------------------------------------------------------------------------------------------------
    GetNearestPlayerCharacter<public>(InNPCCharacter : fort_character)<decides><transacts>: fort_character=
        var Result : ?fort_character = false
        var MinDistance : float = 1000000.0

        for (Player : GetPlayspace().GetPlayers(); PlayerCharacter := Player.GetFortCharacter[]):
            Dist := DistanceSquared(InNPCCharacter.GetTransform().Translation, PlayerCharacter.GetTransform().Translation)
            if (Dist < MinDistance):
                set MinDistance = Dist
                set Result = option{PlayerCharacter}

        Result?

    GetRandomAttack()<decides><transacts>: npc_attack=
        if (OutAttack := Attacks[GetRandomInt(0, Attacks.Length - 1)]) then:
            OutAttack
        else:
            Logger.Print("boss_attack_controller - GetRandomAttack failed: no valid attack", ?Level:= log_level.Warning)
            Attacks[0]


Tooltip_NPCSpawner<public><localizes>:message = "The npc_spawner_device that will spawn the NPCs that will use the attacks."
Tooltip_Cinematic<public><localizes>:message = "The cinematic_sequence_device that will be awaited before starting the behavior."
Tooltip_TimeBetweenAttacksMin<public><localizes>:message = "The minimum time between attacks in seconds."
Tooltip_TimeBetweenAttacksMax<public><localizes>:message = "The maximum time between attacks in seconds."
Tooltip_Attacks<public><localizes>:message = "The attacks that the NPCs will use. Define a custom npc_attack to add a new attack."
