# Knowledge-neutral Verse snippet; project names and asset IDs generalized.
# id: snippet-013

using { /Fortnite.com/AI }
using { /Fortnite.com/Characters }
using { /Verse.org/Simulation }
 
# ============================================================================================================================================
# QuestGiver NPC Behavior
# -> This NPC just stands in place
# ============================================================================================================================================
npc_behavior_quest_giver := class(npc_behavior):

    OnBegin<override>()<suspends>:void=
        if:
            NPCCharacter := GetAgent[].GetFortCharacter[]
        then:
            NPCCharacter.PutInStasis(stasis_args{AllowTurning := true, AllowFalling := false, AllowEmotes := false})
