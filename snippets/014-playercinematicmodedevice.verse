# Knowledge-neutral Verse snippet; project names and asset IDs generalized.
# id: snippet-014

using { /Fortnite.com/Devices }
using {/Fortnite.com/Characters }
using { /Verse.org/Simulation }
using { /UnrealEngine.com/Temporary/Diagnostics }
using { /Verse.org/Native }

# A Verse Device script that sets the player into a cinematic mode so its not in the cinematic, and then resets them when the cinematic is done.
# 
# Getting started:
#     https://www.epicgames.com/fortnite/en-US/creative/docs/uefn/Verse/onboarding-guide-to-programming-with-verse-in-unreal-editor-for-fortnite
#
player_cinematic_mode_device := class(creative_device):

    @editable
    Museum_Cinematic_Device:cinematic_sequence_device := cinematic_sequence_device{}

    @editable
    Cinematic_Sequence_Trigger_Switch:switch_device := switch_device{}

    #Stores a reference to the player once we get it from the switch press below
    var PlayerRef:?agent = false

    # Subscribes to the Switch that plays the sequence and the sequence itself
    OnBegin<override>()<suspends>:void=
        #Print("Cinematic Mode Player Device Running")
        Cinematic_Sequence_Trigger_Switch.TurnedOffEvent.Subscribe(HandleCinematicStart)
        Museum_Cinematic_Device.StoppedEvent.Subscribe(HandleCinematicEnd)
    
    #Sets the player back to visible and free to move
    HandleCinematicEnd():void=
        #Print("Cinematic Mode Ended")
        if(RealPlayer := PlayerRef?):
            if:
                FortCharacter:= RealPlayer.GetFortCharacter[]
            then:
                
                FortCharacter.Show()
                FortCharacter.ReleaseFromStasis()
                #Print("Cinematic Mode Disengaged - Player should be free to move and visible.")
    
    #Sets the player to invisible and removes their movement
    HandleCinematicStart(Agent:agent):void=
        #Print("Cinematic Mode Started")
        if:
            FortCharacter:= Agent.GetFortCharacter[]
        then:
            set PlayerRef = option { Agent }
            CinematicStasis := stasis_args{AllowTurning := false, AllowFalling := false, AllowEmotes := false}
            FortCharacter.PutInStasis(CinematicStasis)
            FortCharacter.Hide()
            #Print("Cinematic Mode Engaged - Player should be frozen and invisible")
