# Knowledge-neutral Verse snippet; project names and asset IDs generalized.
# id: snippet-016

# Copyright Epic Games, Inc. All Rights Reserved.

using { /Fortnite.com/Devices }
using { /Verse.org/Simulation }

##################################################################
# We want a bit more control over our capture point lighting than what the default capture device light can do
# This script plays a cinematic sequence device for each capture point depending on which team owns it
# Within the sequence, we set our lighting values to desired color & intensity
##################################################################

capturepoint_lighting_control := class(creative_device):

    @editable
    CapturePoints:[]capture_area_device = array{}
    @editable
    Sequencers_Team1:[]cinematic_sequence_device = array{}
    @editable
    Sequencers_Team2:[]cinematic_sequence_device = array{}

    OnBegin<override>()<suspends>:void=
        for (i -> EachCapPoint : CapturePoints):
            case (i):
                0 => EachCapPoint.ControlChangeEvent.Subscribe(CaptureChange_A)
                1 => EachCapPoint.ControlChangeEvent.Subscribe(CaptureChange_B)
                2 => EachCapPoint.ControlChangeEvent.Subscribe(CaptureChange_C)
                _ => 

    CaptureChange_A(MyAgent:agent):void=
        PointCaptured(MyAgent, 0)
    CaptureChange_B(MyAgent:agent):void=
        PointCaptured(MyAgent, 1)
    CaptureChange_C(MyAgent:agent):void=
        PointCaptured(MyAgent, 2)

    PointCaptured(MyAgent:agent, CaptureIndex:int):void=
        TeamCollection := GetPlayspace().GetTeamCollection()
        if:
            Team1Sequence := Sequencers_Team1[CaptureIndex]
            Team2Sequence := Sequencers_Team2[CaptureIndex]
        then:
            if (TeamCollection.IsOnTeam[MyAgent, TeamCollection.GetTeams()[0]]):
                Team1Sequence.Play()
                Team2Sequence.Stop()
            else:
                Team1Sequence.Stop()
                Team2Sequence.Play()
