# Knowledge-neutral Verse snippet; project names and asset IDs generalized.
# id: snippet-017

# Copyright Epic Games, Inc. All Rights Reserved.

using { /Fortnite.com/Devices }
using { /Verse.org/Simulation }

##################################################################
# Our game features a looping cinematic in the pregame.  This class simply tells that cinematic to stop!
# The second cleanup cinematic resets all the material parameters
##################################################################
cinematic_lobby_controller := class(creative_device):

    @editable
    CineLobby_Sequence:cinematic_sequence_device := cinematic_sequence_device{}
    @editable
    CineLobby_Cleanup_Sequence:cinematic_sequence_device := cinematic_sequence_device{}


    OnBegin<override>()<suspends>:void=
        CineLobby_Sequence.Stop()
        CineLobby_Cleanup_Sequence.Play()
