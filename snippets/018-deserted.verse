# Knowledge-neutral Verse snippet; project names and asset IDs generalized.
# id: snippet-018

# Copyright Epic Games, Inc. All Rights Reserved.

using { /Verse.org/Simulation/Tags }

DominationSpawnTags<public> := module:
    Tag_SpawnGroup<public>:=            class(tag){}
    Tag_SpawnGroup_Initial<public>:=    class(Tag_SpawnGroup){}
    Tag_SpawnGroup_GroupA<public>:=     class(Tag_SpawnGroup){}
    Tag_SpawnGroup_GroupB<public>:=     class(Tag_SpawnGroup){}
    Tag_SpawnGroup_GroupC<public>:=     class(Tag_SpawnGroup){}

    Tag_TeamOwnership<public>:=         class(tag){}
    Tag_TeamOwnership_Team1<public>:=   class(Tag_TeamOwnership){}
    Tag_TeamOwnership_Team2<public>:=   class(Tag_TeamOwnership){}
