# Knowledge-neutral Verse snippet; project names and asset IDs generalized.
# id: snippet-019

# Copyright Epic Games, Inc. All Rights Reserved.

using { /Fortnite.com/Devices }
using { /Fortnite.com/Teams }
using { /Verse.org/Simulation }
using { /Verse.org/Simulation/Tags }

## This device manages the player spawners
## An intial set of spawners starts enabled
## After a short time, the initial spawners are disabled
## From there, team spawns associated with each capture point are enabled/disabled based off a weighting score
##     Owning the point adds 1 to the weighting score (or -1 if team 2)
##     Each ally in the area adds 1 to the weighting score (or -1 if team 2)
##     If the weighting score is > 0, team 1 can spawn near the point.  
##     If the weighting score is < 0, team 2 can spawn near the point

##################################################################
## Event Handler Class - We'll spawn one of these per capture point and bind events through it
##################################################################
capture_point_event_handler := class:

    CapturePointIndex:int
    DominationSpawnManagerRef:domination_spawn_manager
    TeamCollectionRef:fort_team_collection

    var CaptureWeightOfThisPoint:int = 0
    var PlayerWeightOfThisPoint:float = 0.0
    WeightForEachPlayer:float = 0.5

    # If Team 1, we return 1.  If Team 2, we return -1.
    GetTeamDirectionOfAgent<private>(MyAgent:agent)<transacts>:int=
        if (FirstTeam := TeamCollectionRef.GetTeams()[0]):
            if (TeamCollectionRef.IsOnTeam[MyAgent, FirstTeam]):
                return 1
        return -1
    
    # Set state of capture point
    OnCaptured(MyAgent:agent):void=
        set CaptureWeightOfThisPoint = GetTeamDirectionOfAgent(MyAgent)
        DominationSpawnManagerRef.CalculateWeightForPoint(CapturePointIndex)

    # Shift player weighting of point as players enter or leave
    PlayerEnteredLargeZone(MyAgent:agent):void=
        set PlayerWeightOfThisPoint = PlayerWeightOfThisPoint + (GetTeamDirectionOfAgent(MyAgent) * WeightForEachPlayer)
        DominationSpawnManagerRef.CalculateWeightForPoint(CapturePointIndex)
    PlayerLeftLargeZone(MyAgent:agent):void=
        set PlayerWeightOfThisPoint = PlayerWeightOfThisPoint - (GetTeamDirectionOfAgent(MyAgent) * WeightForEachPlayer)
        DominationSpawnManagerRef.CalculateWeightForPoint(CapturePointIndex)
    

##################################################################
## Domination Spawn Manager Class
##################################################################
domination_spawn_manager := class(creative_device):

    # Player Spawner Device Groups - Populated by searching for Verse tags
    var SpawnPointGroupInitial:[]player_spawner_device = array{}
    var SpawnPointGroupsCapPoint:[][]player_spawner_device = array{array{}, array{}, array{}}

    # Capture Zone
    @editable  
    CapturePoints:[]capture_area_device = array{}
    @editable  
    LargeZones:[]player_counter_device = array{}

    # Initial Spawners will disable after this many seconds
    InitialGroupDisableTime:float = 5.0

    # Cap Point Event Handler Class Refs
    var CapturePointListeners:[]capture_point_event_handler = array{} 


    # OnBegin will run when the game starts
    OnBegin<override>()<suspends>:void=
        BeginPlayEventSubscriptions()
        FindAndCacheSpawnDevices()
        DisableInitialSpawnGroup()
        

    # Subscribe to events from our attribute triggers
    BeginPlayEventSubscriptions<private>():void=
        for (I -> EachCapPoint:CapturePoints):
            MyHandler := capture_point_event_handler:
                CapturePointIndex := I
                DominationSpawnManagerRef := Self
                TeamCollectionRef := Self.GetPlayspace().GetTeamCollection()

            set CapturePointListeners = CapturePointListeners + array{MyHandler}
            
            EachCapPoint.ControlChangeEvent.Subscribe(MyHandler.OnCaptured)

            if (EachLargeZone := LargeZones[I]):
                EachLargeZone.CountedEvent.Subscribe(MyHandler.PlayerEnteredLargeZone)
                EachLargeZone.RemovedEvent.Subscribe(MyHandler.PlayerLeftLargeZone)


    # Get all our various Player Spawner devices, and store them in the appropriate arrays
    FindAndCacheSpawnDevices<private>():void=
        InitialSpawnPoints:generator(creative_object_interface) = FindCreativeObjectsWithTag(DominationSpawnTags.Tag_SpawnGroup_Initial{})
        for (FoundDevice : InitialSpawnPoints, DeviceAsSpawner:= player_spawner_device[FoundDevice]):
            set SpawnPointGroupInitial = SpawnPointGroupInitial + array{DeviceAsSpawner}

        GroupTags:[]tag = array:
            DominationSpawnTags.Tag_SpawnGroup_GroupA{}
            DominationSpawnTags.Tag_SpawnGroup_GroupB{}
            DominationSpawnTags.Tag_SpawnGroup_GroupC{}

        for (I -> GroupTag:GroupTags): 
            FoundSpawnGroup:generator(creative_object_interface) := FindCreativeObjectsWithTag(GroupTag)
            for (FoundDevice : FoundSpawnGroup, DeviceAsSpawner:= player_spawner_device[FoundDevice]):
                if:
                    CapPoint := SpawnPointGroupsCapPoint[I] + array{DeviceAsSpawner}
                    set SpawnPointGroupsCapPoint[I] = CapPoint
            
            

    # Turn off initial spawn points     
    DisableInitialSpawnGroup<private>()<suspends>:void=
        Sleep(InitialGroupDisableTime)
        for (SpawnPoint : SpawnPointGroupInitial):
            SpawnPoint.Disable()
        

    # Whenever the CapPointCaptureWeight or CapPointPlayerWeight for a point changes, this runs to see who should be allowed to spawn from the associated spawn group
    CalculateWeightForPoint(PointIndex:int):void=
        var WeightOfPoint:int = 0

        if:
            CapPointData := CapturePointListeners[PointIndex]
            CapPlayerWeightToInt := Floor[CapPointData.PlayerWeightOfThisPoint]
        then:
            set WeightOfPoint = CapPointData.CaptureWeightOfThisPoint + CapPlayerWeightToInt
        
        for (SpawnDevice:SpawnPointGroupsCapPoint[PointIndex]):
            var DeviceIsTeam1:logic = false
            DeviceTags := SpawnDevice.GetTags()
            if (DeviceTags.Has[DominationSpawnTags.Tag_TeamOwnership_Team1{}]):
                set DeviceIsTeam1 = true
            
            if (DeviceIsTeam1?):
                if (WeightOfPoint > 0):
                    SpawnDevice.Enable()
                else:
                    SpawnDevice.Disable()
            else:
                if (WeightOfPoint < 0):
                    SpawnDevice.Enable()
                else:
                    SpawnDevice.Disable()
