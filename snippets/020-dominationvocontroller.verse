# Knowledge-neutral Verse snippet; project names and asset IDs generalized.
# id: snippet-020

# Copyright Epic Games, Inc. All Rights Reserved.

using { /Fortnite.com/Devices }
using { /Fortnite.com/Teams }
using { /Verse.org/Simulation }


##################################################################
## Event Handler Class - We'll spawn one of these per capture point and bind events through it
##################################################################
vo_cappoint_event_handler := class:
    CapturePointIndex:int
    VOControlRef:domination_vo_controller

    # Only play losing control audio if the point has been captured.  Otherwise you get 'losing control' vo for neutral points at the start of game.
    var HasBeenCaptured:logic = false

    OnCaptured(MyAgent:agent):void=
        VOControlRef.UpdateCapturePointTracking(MyAgent, CapturePointIndex)
        VOControlRef.QueueVO_PointCapturedOrLost(MyAgent, CapturePointIndex)

        set HasBeenCaptured = true

    OnStartedCapture(MyAgent:agent):void=
        if (HasBeenCaptured?):
            VOControlRef.QueueVO_PointCaptureStarted(MyAgent, CapturePointIndex)     


##################################################################
## VO Controller Device Class
##################################################################   

audioqueue_type := enum:
    CaptureStarted
    CaptureOrLost
    AllPointsCaptured

audioqueue_entry:= struct:
    InitiatingAgent:agent
    CapturePointID:int
    AudioEventType:audioqueue_type

domination_vo_controller := class(creative_device):

    # Capture Zone
    @editable  
    CapturePoints:[]capture_area_device = array{}

    # Audio Players
    @editable
    VO_PointCaptured:[]audio_player_device = array{}
    @editable
    VO_PointLost_A:[]audio_player_device = array{}
    @editable
    VO_PointLost_B:[]audio_player_device = array{}
    @editable
    VO_PointLost_C:[]audio_player_device = array{}

    @editable
    VO_PointStartedLosing_A:[]audio_player_device = array{}
    @editable
    VO_PointStartedLosing_B:[]audio_player_device = array{}
    @editable
    VO_PointStartedLosing_C:[]audio_player_device = array{}

    @editable
    VO_SelectLoadout:audio_player_device = audio_player_device{}

    @editable
    FriendlyTeamIsDominating:[]audio_player_device = array{}
    @editable
    EnemyTeamIsDominating:[]audio_player_device = array{}

    # Hud Messages
    @editable
    HUDMessage_PointCaptured:[]hud_message_device = array{}
    @editable
    HUDMessage_PointLost:[]hud_message_device = array{}
    @editable
    HUDMessage_LosingPoint:[]hud_message_device = array{}

    # Cycle Audio Clips to avoid repeats
    var AudioIndex_CapOrLost:int = 0
    var AudioIndex_Losing:int = 0

    # Domination State
    var CapturePointOwnership:[]int = array{0,0,0}
    var AnyTeamIsDominating:logic = false

    # VO Play Cooldown
    VOCooldownTime:float = 2.0
    var VOIsOnCooldown:logic = false

    # Audio queue for important messages
    var AudioToPlayQueue:[]audioqueue_entry = array{}

    # Class Selector Device Ref
    @editable
    ClassSelectUI:class_selector_ui_device = class_selector_ui_device{}
    @editable
    ClassSelectOpenedRouter:attribute_evaluator_device = attribute_evaluator_device{}
    var ValidAudioTargets:[]agent = array{}
    var ValidClassSelectAudioTargets:[]agent = array{}


    @editable
    GameIntroHUDMessage:hud_message_device = hud_message_device{}


    # Cache Listener Classes to Prevent GC - In the future you won't need to do this!
    var CachedListeners:[]vo_cappoint_event_handler = array{} 


    OnBegin<override>()<suspends>:void=
        # Create an event handler class for each capture point, and subscribe
        for (I->EachCapPoint : CapturePoints):
            MyHandler := vo_cappoint_event_handler:
                CapturePointIndex := I
                VOControlRef := Self
            EachCapPoint.ControlChangeEvent.Subscribe(MyHandler.OnCaptured)
            EachCapPoint.ControlChangeStartsEvent.Subscribe(MyHandler.OnStartedCapture)
            set CachedListeners = CachedListeners + array{MyHandler}

        # Player spawned in world event (we don't play audio until the player has spawned once and heard the intro VO)
        ClassSelectUI.ClassSelectedEvent.Subscribe(PlayerSpawnedIntoWorld)
        # Player entered class UI
        ClassSelectOpenedRouter.PassEvent.Subscribe(ClassSelectUIOpen)


    ClassSelectUIOpen <private>(MyAgent:agent):void=
        if (not ValidClassSelectAudioTargets.Find[MyAgent]):
            set ValidClassSelectAudioTargets = ValidClassSelectAudioTargets + array{MyAgent}
            VO_SelectLoadout.Play(MyAgent)
        

    PlayerSpawnedIntoWorld<private>(MyAgent:agent):void=
        if (not ValidAudioTargets.Find[MyAgent]):
            spawn{AddPlayerToValidAudioTargets(MyAgent)}


    AddPlayerToValidAudioTargets<private>(MyAgent:agent)<suspends>:void=
        # After a delay to account for initial class popup, permit audio
        Sleep(0.5)
        GameIntroHUDMessage.Show(MyAgent)
        Sleep(2.0)
        if (not ValidAudioTargets.Find[MyAgent]):
            set ValidAudioTargets = ValidAudioTargets + array{MyAgent}


    UpdateCapturePointTracking(MyAgent:agent, CapturePointID:int):void=
        TeamCollection := GetPlayspace().GetTeamCollection()
        if:
            Team1 := TeamCollection.GetTeams()[0]
            TeamCollection.IsOnTeam[MyAgent, Team1]
            set CapturePointOwnership[CapturePointID]=1
        else if:
            set CapturePointOwnership[CapturePointID]= -1

        if (Abs(CapturePointOwnership[0] + CapturePointOwnership[1] + CapturePointOwnership[2]) = 3):
            set AnyTeamIsDominating = true
        else:
            set AnyTeamIsDominating = false
        

    ##################################################################
    # Play VO Methods
    # These should not be used directly. Instead use the Audio Queue
    ##################################################################
    PlayVO_PointCapturedOrLost<private>(MyAgent:agent, CapturePointID:int, SkipDominationCheck:logic):void=
        # Increment VO Index
        set AudioIndex_CapOrLost = AudioIndex_CapOrLost + 1
        
        # For each player in the game, play the appropriate audio & associated HUD message
        TeamCollection := GetPlayspace().GetTeamCollection()
        for (EachPlayer : GetPlayspace().GetPlayers()):
            if (TeamCollection.GetTeamAttitude[MyAgent, EachPlayer] = team_attitude.Friendly):
                HUDMessageToShow:?hud_message_device = option{ HUDMessage_PointCaptured[CapturePointID] }

                if (AnyTeamIsDominating? and not SkipDominationCheck?):
                    PlayVOForPlayer(EachPlayer, FriendlyTeamIsDominating, AudioIndex_CapOrLost, HUDMessageToShow)
                else:
                    PlayVOForPlayer(EachPlayer, VO_PointCaptured, AudioIndex_CapOrLost, HUDMessageToShow)

            else:
                HUDMessageToShow:?hud_message_device = option{ HUDMessage_PointLost[CapturePointID] }

                if (AnyTeamIsDominating? and not SkipDominationCheck?):
                    PlayVOForPlayer(EachPlayer, EnemyTeamIsDominating, AudioIndex_CapOrLost, HUDMessageToShow)
                else:
                    VOPointLostArray := array{VO_PointLost_A, VO_PointLost_B, VO_PointLost_C}
                    if (VOToPlay := VOPointLostArray[CapturePointID]):
                        PlayVOForPlayer(EachPlayer, VOToPlay, AudioIndex_CapOrLost, HUDMessageToShow)
    

    PlayPointCaptureStartedVO<private>(MyAgent:agent, CapturePointID:int):void=
        # Increment VO Index
        set AudioIndex_Losing = AudioIndex_Losing + 1
        
        # For each enemy (we don't tell friendlies capture has tarted), play VO and associated hud message
        TeamCollection := GetPlayspace().GetTeamCollection()
        for (EachPlayer : GetPlayspace().GetPlayers()):
            if (TeamCollection.GetTeamAttitude[MyAgent, EachPlayer] <> team_attitude.Friendly):
                HUDMessageToShow:?hud_message_device = option{ HUDMessage_LosingPoint[CapturePointID] }

                VOPointLostArray := array{VO_PointStartedLosing_A, VO_PointStartedLosing_B, VO_PointStartedLosing_C}
                if (VOToPlay := VOPointLostArray[CapturePointID]):
                    PlayVOForPlayer(EachPlayer, VOToPlay, AudioIndex_Losing, HUDMessageToShow)


    PlayVOForPlayer<private>(MyAgent:agent, DeviceList:[]audio_player_device, ClipIndex:int, OptionalHUDMessage:?hud_message_device):void=
        # Make sure the incoming device list is populated!
        if (DeviceList.Length = 0):
            return
        # Is player a valid audio target?
        if (not ValidAudioTargets.Find[MyAgent]):
            return

        if (Device := DeviceList[Mod[ClipIndex, DeviceList.Length]]):
            Device.Play(MyAgent)
        if (HUDMessage := OptionalHUDMessage?):
            HUDMessage.Show(MyAgent)


    ##################################################################
    # Audio Queue Management
    ##################################################################
    QueueVO_PointCapturedOrLost(MyAgent:agent, CapturePointID:int):void=
        if (AnyTeamIsDominating?):
            EnqueueAudioEntry(MyAgent, CapturePointID, audioqueue_type.AllPointsCaptured)
        else:
            EnqueueAudioEntry(MyAgent, CapturePointID, audioqueue_type.CaptureOrLost)


    QueueVO_PointCaptureStarted(MyAgent:agent, CapturePointID:int):void=
        # Only add these to the queue if they're the only audio to play.
        # They aren't important enough to queue up compared to other audio sources.
        if (AudioToPlayQueue.Length = 0):
            EnqueueAudioEntry(MyAgent, CapturePointID, audioqueue_type.CaptureStarted)


    EnqueueAudioEntry<private>(MyAgent:agent, CaptureID:int, AudioType:audioqueue_type):void=
        set AudioToPlayQueue = AudioToPlayQueue + array:
            audioqueue_entry{InitiatingAgent := MyAgent, CapturePointID := CaptureID, AudioEventType := AudioType}

        spawn { PlayQueuedVO() }


    PlayQueuedVO<private>()<suspends>:void =
        if (VOIsOnCooldown?):
            return

        set VOIsOnCooldown = true

        loop: 
            if:
                NextAudioToPlay := AudioToPlayQueue[0]
                NewQueue := AudioToPlayQueue.RemoveElement[0]
                set AudioToPlayQueue = NewQueue
            then:
                case (NextAudioToPlay.AudioEventType):
                    audioqueue_type.CaptureStarted =>
                        PlayPointCaptureStartedVO(NextAudioToPlay.InitiatingAgent, NextAudioToPlay.CapturePointID)
                    audioqueue_type.CaptureOrLost =>
                        PlayVO_PointCapturedOrLost(NextAudioToPlay.InitiatingAgent, NextAudioToPlay.CapturePointID, true)
                    audioqueue_type.AllPointsCaptured =>
                        PlayVO_PointCapturedOrLost(NextAudioToPlay.InitiatingAgent, NextAudioToPlay.CapturePointID, false)

                # Sleep between each play
                Sleep(VOCooldownTime)
            else:
                break

        set VOIsOnCooldown = false
