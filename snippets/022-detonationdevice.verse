# Knowledge-neutral Verse snippet; project names and asset IDs generalized.
# id: snippet-022

using { /Fortnite.com/Devices }
using { /Verse.org/Native }
using { /Verse.org/Simulation }
using { /UnrealEngine.com/Temporary/Diagnostics }

# Log channel for this experience.
log_detonation := class(log_channel){}

# Enum to describe possible bomb states.
bomb_state<public> := enum {AllUnarmed, BombAArmed, BombBArmed}

# A Verse-authored creative device that can be placed in a level to add bombs that can be armed and disarmed.
# 
# Getting started:
#   https://dev.epicgames.com/community/fortnite/getting-started/verse
#
detonation_device := class(creative_device):
    # Logger for the parkour experience channel.
    Logger:log = log{Channel:=log_detonation}

    # Detonation timer for Bomb A.
    @editable 
    TimedObjectiveA:timed_objective_device = timed_objective_device{}

    # Explosive barrel for Bomb A.
    @editable
    ExplosiveBarrelsA:[]explosive_device = array{}

    # Map indicators for Bomb A.
    @editable
    BombAMapIndicators:[]map_indicator_device = array{}

    # Armed beacon for Bomb A.
    @editable
    BombABeaconArm : beacon_device = beacon_device{}

    # Disarmed beacon for Bomb A.
    @editable
    BombABeaconDisarm:beacon_device = beacon_device{}

    # Detonation timer for Bomb B.
    @editable 
    TimedObjectiveB:timed_objective_device = timed_objective_device{}

    # Explosive barrel for Bomb B.
    @editable
    ExplosiveBarrelsB:[]explosive_device = array{}

    # Map indicators for Bomb B.
    @editable
    BombBMapIndicators:[]map_indicator_device = array{}

    # Armed beacon for Bomb B.
    @editable
    BombBBeaconArm:beacon_device = beacon_device{}

    # Disarmed beacon for Bomb B.
    @editable
    BombBBeaconDisarm:beacon_device = beacon_device{}

    # End Game Device for a tracking player progress on disarming the bombs.
    @editable
    EndGameDevice:end_game_device = end_game_device{}

    # Bomb state member variable for tracking current bomb state.
    var BombState:bomb_state = bomb_state.AllUnarmed

    # Runs when the device is started in a running game.
    OnBegin<override>()<suspends>:void=
        # The race expression is used to run a block of two or more async expressions concurrently (simultaneously).
        # When the fastest expression completes, it "wins the race". Other expressions are then terminated.
        race:
            # The block expression is used to create a new scope context. Expressions within the block will be run as a group.
            block:
                # Wait for Bomb A to be armed.
                ArmingPlayer := TimedObjectiveA.BeganEvent.Await()
                Logger.Print("Bomb A Armed")
                
                # Used to know which beacons to enable and barrels need to explode.
                set BombState = bomb_state.BombAArmed
                
                # Disable the other Timed Objective device.
                TimedObjectiveB.Disable(ArmingPlayer)
                
                # Disable map indicators for Bomb B.
                for (MapIndicator : BombBMapIndicators):
                    MapIndicator.Disable()

            block:
                # Wait for Bomb B to be armed.
                ArmingPlayer := TimedObjectiveB.BeganEvent.Await()
                Logger.Print("Bomb B Armed")
                
                # Used to know which beacons to enable and barrels need to explode.
                set BombState = bomb_state.BombBArmed
                
                # Disable the other Timed Objective device.
                TimedObjectiveA.Disable(ArmingPlayer)

                # Disable map indicators for Bomb A.
                for (MapIndicator : BombAMapIndicators):
                    MapIndicator.Disable()
        
        # Enable the correct Beacon Devices now a bomb is armed.
        UpdateBeacons()
        
        # Wait for the TimedObjective to Complete or be Stopped.
        race:
            # Bomb A detonated.
            BombDetonated(TimedObjectiveA.CompletedEvent.Await())

            # Bomb B detonated.
            BombDetonated(TimedObjectiveB.CompletedEvent.Await())
            
            # Bomb A was disarmed.
            block:
                DisarmingPlayer := TimedObjectiveA.EndedEvent.Await()
                Logger.Print("Bomb Disarmed")
                EndGameDevice.Activate(DisarmingPlayer)

            # Bomb B was disarmed.
            block:
                DisarmingPlayer := TimedObjectiveB.EndedEvent.Await()
                Logger.Print("Bomb Disarmed")
                EndGameDevice.Activate(DisarmingPlayer)

        
    # Disable the unarmed Beacons and enable the Beacon over the armed bomb.
    UpdateBeacons():void=
        BombABeaconArm.Disable()
        BombBBeaconArm.Disable()
        
        if (BombState = bomb_state.BombAArmed):
            BombABeaconDisarm.Enable()
        else:
            BombBBeaconDisarm.Enable()

    # One of the bombs detonated.
    BombDetonated(Agent:agent):void=
        Logger.Print("Bomb Detonated")

        # Determine which set of barrels should explode.
        if (BombState = bomb_state.BombAArmed):
            ExplodeBarrels(ExplosiveBarrelsA, Agent)
        else:
            ExplodeBarrels(ExplosiveBarrelsB, Agent)

        # Trigger end game.
        EndGameDevice.Activate(Agent)
    
    # Make the provided barrels explode.
    ExplodeBarrels(Barrels:[]explosive_device, Agent:agent):void=
        for (Barrel : Barrels):
            Barrel.Explode(Agent)
