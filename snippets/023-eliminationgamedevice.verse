# Knowledge-neutral Verse snippet; project names and asset IDs generalized.
# id: snippet-023

using { /Fortnite.com/Devices }
using { /Fortnite.com/Characters }
using { /Fortnite.com/Game }
using { /Verse.org/Simulation }
using { /Verse.org/Simulation/Tags }
using { /Verse.org/Random }
using { /UnrealEngine.com/Temporary/Diagnostics }

# Log channel for this experience.
log_elimination := class(log_channel){}

# This is a Verse-authored creative device that can be placed in a level.
# This script example is a take on an elimnation game where players have their weapon upgraded each time they eliminate an opponent.
# Victory is achieved by being the first to reach a given number of eliminations.
# 
# Getting started:
#   https://dev.epicgames.com/community/fortnite/getting-started/verse
#
elimination_game_device := class(creative_device):
    # Logger for the elimination experience channel.
    Logger:log = log{Channel := log_elimination}

    # End Game Device for managing end game behavior.
    @editable
    EndGameDevice:end_game_device = end_game_device{}

    # Array of item granters used to grant the weapons.
    @editable
    var WeaponItemGranters:[]item_granter_device = array{}
    
    # This is set later, and will be equal to the the number of weapon item_granters in the island.
    var NumberOfEliminationsToWin:int = 0

    # Map container to track player's progress. This is used to determine which weapon to award to the player.
    var AgentMap:[agent]int = map{}

    OnBegin<override>()<suspends>:void=
        # Use this variable to scale the number of eliminations needed based on how many item granters there are in the experience.
        set NumberOfEliminationsToWin = WeaponItemGranters.Length
        Logger.Print("Number of Weapon Item Granters: {WeaponItemGranters.Length}")
        
        # Randomize the order in which the weapons are granted.
        set WeaponItemGranters = Shuffle(WeaponItemGranters)

        # Get all the players in the experience.
        AllPlayers := GetPlayspace().GetPlayers()
        for (EliminationGamePlayer : AllPlayers):
            if (FortCharacter := EliminationGamePlayer.GetFortCharacter[]):
                # Subscribe to character eliminated event.
                FortCharacter.EliminatedEvent().Subscribe(OnPlayerEliminated)
            
            # Add Players to a Map to track progress.
            if (set AgentMap[EliminationGamePlayer] = 1) {} 
            
            # Grant the first weapon to each player.
            if (FirstItemGranter:item_granter_device = WeaponItemGranters[0]):
                FirstItemGranter.GrantItem(EliminationGamePlayer) 

    # Event that handles when a player is eliminated.
    OnPlayerEliminated(Result:elimination_result):void=
        Logger.Print("Player Eliminated")

        # Look up the eliminating character. Since this is an optional value, this will return an option(agent).
        EliminatingCharacter := Result.EliminatingCharacter

        # Get the agent from the eliminating character option(agent). If the eliminating character is not valid, this will return false.
        if:
            FortCharacter := EliminatingCharacter?
            EliminatingAgent := FortCharacter.GetAgent[]
        then:
            # If the eliminating agent is valid, grant them the next weapon.
            GrantNextWeapon(EliminatingAgent)

    # Check if there is a winner for the game, if not then grant the next weapon.
    GrantNextWeapon(Agent:agent):void=
        # Get the current level for the agent.
        if (var CurrentItemNumber:int = AgentMap[Agent]):
            # If the agent has met the victory conditions, end the game.
            if (IsVictoryConditionMet[CurrentItemNumber]):
                # Game has been won by Agent.
                EndGame(Agent)
            else:
                # Game is not over yet.
                if (ItemGranter := WeaponItemGranters[CurrentItemNumber]):
                    ItemGranter.GrantItem(Agent)

                set CurrentItemNumber = CurrentItemNumber + 1
                
                # Update Agent's current level.
                if (set AgentMap[Agent] = CurrentItemNumber) {}

    # Check if the victory condition has been met and return a logic result.
    IsVictoryConditionMet(EliminationNumber:int)<decides><transacts>:void=
        EliminationNumber = NumberOfEliminationsToWin
    
    # Runs when this device script is stopped or the game ends.
    EndGame(Agent:agent):void=
        EndGameDevice.Activate(Agent)
