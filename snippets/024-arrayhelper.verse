# Knowledge-neutral Verse snippet; project names and asset IDs generalized.
# id: snippet-024

# Helper function to find the value that appears most often in an int array
FindMostFrequentIndex<public>(Indexes:[]int):int =
    var CountMap:[int]int = map{}
    var MostFrequent:int = 0
    var HighestCount:int = 0

    for (Choice : Indexes):
        # If the value already exists in the map then increment the count
        # else add the first appearance to the map
        if (Value := CountMap[Choice]):
            if(set CountMap[Choice] += 1):
        else:
            if(set CountMap[Choice] = 1):

        # If the current count of the value being checked is the highest so far then
        # It is the most frequent. Store the value and continue checking the array
        if (Value := CountMap[Choice], Value > HighestCount):
            set HighestCount = Value
            set MostFrequent = Choice

    MostFrequent
