# Knowledge-neutral Verse snippet; project names and asset IDs generalized.
# id: snippet-025

using { /Fortnite.com/Game }
using { /Fortnite.com/Devices }
using { /Fortnite.com/Characters }
using { /Fortnite.com/FortPlayerUtilities }
using { /Verse.org/Simulation }
using { /Verse.org/Random }
using { /UnrealEngine.com/Temporary/Diagnostics }
using { /UnrealEngine.com/Temporary/SpatialMath }
using{Helpers}

island_selector_log := class(log_channel){}

persist_island_data := class<final><persistable>:
    LastIslandIndex:int = 0
    ExclusionList:[]int = array{}

MakeIslandData<constructor>(PreviousData:persist_island_data)<transacts> := persist_island_data:
    LastIslandIndex := PreviousData.LastIslandIndex
    ExclusionList := PreviousData.ExclusionList

# For this experience players will cycle through different islands. For this functionality
# We need to store some data between rounds which we do with a player map.
# the data stored is the last island played to allow for replaying of the same island
# until a new island is prompted and an exclusion list to prevent the same island from being
# picked repeatedly by random chance.
var IslandManagerMap:weak_map(player, persist_island_data) = map{}

# Island data is editable data that allows the users to setup and reference islands via the editor
# Currently the island data requires users to reference teleporter devices where the players will be
# teleported to upon game start.
# The IslandTeleporters is the default layout for the island and the Inverse allows for a different
# teleporter layout. In this use case teams will switch sides each round.
island_data<public>:= class<concrete>():
    @editable
    DebugName:string = "Name Me"
    @editable
    IslandsTeleporters : []teleporter_device = array{}
    @editable
    InverseIslandsTeleporters : []teleporter_device = array{}

island_selector_device := class(creative_device):
    DebugLog:log = log{Channel := island_selector_log}

   @editable
    Islands:[]island_data = array{}

    @editable
    RoundsToPickIsland:[]round_settings_device = array{}

    @editable
    RandomEveryRound:logic = false

    @editable
    IslandSelectionCompletedTrigger :trigger_device = trigger_device{}

    @editable
    TeleportZone : mutator_zone_device = mutator_zone_device{}

    # Invert the sides on every other round when a new island isn't picked
    var InvertSides:logic = true

    OnBegin<override>()<suspends>:void =
        # Pick a Random Island
        # If no round devices are in the array or the random every round flag is set
        # The logic will pick a new island every round.
        # else a new island will be picked once the round begin event is triggered from
        # a round device
        if (RandomEveryRound? or RoundsToPickIsland.Length <= 0):
            PickRoundIsland()
            set InvertSides = false
        else:
            for (RoundDevice : RoundsToPickIsland):
                RoundDevice.RoundBeginEvent.Subscribe(PickRoundIsland)

        # Add a small delay to allow the round devices time to subscribe
        Sleep(2.0)

        # Loop through the teleporters associated with the selected island
        # make sure they are enabled.
        EnableTeleporters()

        # Enable the mutator zone to teleport all the players
        # to the island.
        TeleportPlayers()

        # A trigger with fire upon completion in case any devices need to perform
        # an action once this process completes
        IslandSelectionCompletedTrigger.Trigger()

    PickRoundIsland():void =
        set InvertSides = false
        # To prevent players from playing the same islands twice in a row we use an exclusion list.
        # to check the previously played islands.
        # This data must persist between rounds so will be kept in a weak map.
        # First we check if a player has played every island and if they have we clear the exclusion list
        ValidateExclusionList()

        # Pick the new island
        RandomIndex := PickRandomIsland()
        DebugLog.PrintDebug("Island Index is {RandomIndex}")

        # Save the player data
        SavePlayerIslandData(RandomIndex)

    PickRandomIsland():int =
        DebugLog.PrintDebug("Starting to pick new random Island")
        var PickedIndex:int = GetRandomInt(0, Islands.Length - 1)

        # Currently to keep persistent data between rounds we need to store this information in the players' map
        # As there is a possibility that each player has different data stored we find the most common data between
        # All the players currently in this experience and use that.

        PlayersLastIndex := for(Player : GetPlayspace().GetPlayers(), Index := IslandManagerMap[Player].LastIslandIndex){Index}
        MostFrequentIndex := FindMostFrequentIndex(PlayersLastIndex)

        DebugLog.PrintDebug("Players agree most frequent is Index {MostFrequentIndex}")

        if (Player := FindPlayerWithMostFrequentIndex[MostFrequentIndex]):
            if (ExclusionList := IslandManagerMap[Player].ExclusionList):
                # Find a random new island to use
                var NewChoice : int = GetRandomInt(0, Islands.Length - 1)
                # If the island is on the exclusion list we need to find one player haven't yet played
                if (ExclusionList.Find[NewChoice]):
                    DebugLog.PrintDebug("island index already used, searching for new one")

                    # Shuffle the array index of all the available islands
                    ShuffledIndexesArray := Shuffle(for (Index -> IslandData: Islands){Index})

                    # Loop through this array until an index that is not in the exclusion list is found
                    var loopbreak:logic = false
                    for (ShuffledIndexes:ShuffledIndexesArray, not ExclusionList.Find[ShuffledIndexes], not loopbreak?):
                        DebugLog.PrintDebug("Found previously unused island")
                        set NewChoice = ShuffledIndexes
                        set loopbreak = true

                set PickedIndex = NewChoice
            else:
                DebugLog.PrintDebug("Unable to find players exclusion list")
        else:
            DebugLog.PrintDebug("Unable to find a player that matches most frequent choice")

        PickedIndex

    FindPlayerWithMostFrequentIndex<private>(MostFrequentIndex:int)<decides><transacts>:player =
        for(Player : GetPlayspace().GetPlayers(), IslandManagerMap[Player].LastIslandIndex = MostFrequentIndex){Player}[0]

    EnableTeleporters<private>():void =
        PlayersLastIndex := for(Player : GetPlayspace().GetPlayers(), Index := IslandManagerMap[Player].LastIslandIndex){Index}
        MostFrequentIndex := FindMostFrequentIndex(PlayersLastIndex)

        if (Island := Islands[MostFrequentIndex]):
            if (InvertSides?):
                for (Teleporter : Island.InverseIslandsTeleporters):
                    DebugLog.PrintDebug("Enabling Inverse teleporters")
                    Teleporter.Enable()    
            else:
                for (Teleporter : Island.IslandsTeleporters):
                    DebugLog.PrintDebug("Enabling teleporters")
                    Teleporter.Enable()
        else:
            DebugLog.PrintDebug("Failing to Get Data for teleporters")

    TeleportPlayers<private>()<suspends>:void =
        DebugLog.PrintDebug("Enabling zone")
        TeleportZone.Enable()

    ValidateExclusionList<private>():void =
        # Currently to keep persistent data between rounds we need to store this information in the player's map
        # As there is a possibility that each player has different data stored we find the most common data between
        # All the players currently in this experience and use that.
        PlayersLastIndex := for(Player : GetPlayspace().GetPlayers(), Index := IslandManagerMap[Player].LastIslandIndex){Index}
        MostFrequentIndex := FindMostFrequentIndex(PlayersLastIndex)

        # Find a player with the same index and use that as the source of truth to clear exclusion data if requirements met
        if:
            Player := FindPlayerWithMostFrequentIndex[MostFrequentIndex]
            ExclusionList := IslandManagerMap[Player].ExclusionList
            ExclusionList.Length >= Islands.Length
        then:
            for:
                InnerPlayer: GetPlayspace().GetPlayers()
                # Clear the exclusion list except for the last played map
                set IslandManagerMap[InnerPlayer] = persist_island_data{LastIslandIndex:= MostFrequentIndex, ExclusionList:= array{MostFrequentIndex}}
            do:
                DebugLog.PrintDebug("resetting players exclusion list")

    SavePlayerIslandData(NewIndex:int):void =
        for (Player : GetPlayspace().GetPlayers()):
            if( PreviousData := IslandManagerMap[Player]):
                if (set IslandManagerMap[Player] = persist_island_data :
                        LastIslandIndex := NewIndex
                        ExclusionList := PreviousData.ExclusionList + array{NewIndex}
                        MakeIslandData<constructor>(PreviousData)
                ):
                    DebugLog.PrintDebug("Updating previous player island data")
            else:
                if (set IslandManagerMap[Player] = persist_island_data :
                        LastIslandIndex := NewIndex
                        MakeIslandData<constructor>(persist_island_data{})
                ):
                    DebugLog.PrintDebug("Creating new player island data")
