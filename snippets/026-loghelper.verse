# Knowledge-neutral Verse snippet; project names and asset IDs generalized.
# id: snippet-026

using { /UnrealEngine.com/Temporary/Diagnostics }

# Quick enable or disable of logs
EnableLogs:logic = true

(Logger:log).PrintDebug<public>(Message:string):void=
    if (EnableLogs?):
        Logger.Print(Message)
