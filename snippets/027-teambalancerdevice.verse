# Knowledge-neutral Verse snippet; project names and asset IDs generalized.
# id: snippet-027

using { /UnrealEngine.com/Temporary/Diagnostics }
using { /Fortnite.com/Devices }
using { /Verse.org/Simulation }
using{Helpers}

team_balancer_log := class(log_channel){}

team_balancer_device := class(creative_device):
    DebugLog:log = log{Channel := team_balancer_log}

    @editable
    TargetTeamSize<private>:int = 2

    OnBegin<override>()<suspends>:void =
        GetPlayspace().PlayerAddedEvent().Subscribe(OnAgentJoin)
        Sleep(2.0)
        BalanceTeams()

    # Protect against networking issues causing late join problems such as unbalanced teams
    # This method will Telekinetic the teams to the TargetTeamSize if any team size goes over this target
    BalanceTeams():void =
        TeamCollection := GetPlayspace().GetTeamCollection()
        for:
            Team : TeamCollection.GetTeams()
            AgentsInTeam := TeamCollection.GetAgents[Team]
            AgentsInTeam.Length > TargetTeamSize
            Difference := AgentsInTeam.Length - TargetTeamSize
            OtherTeam : TeamCollection.GetTeams()
            OtherTeam <> Team
            TeamCollection.GetAgents[OtherTeam].Length < TargetTeamSize
            Index := TargetTeamSize .. TargetTeamSize + Difference
            AgentToMove := AgentsInTeam[Index]
            TeamCollection.AddToTeam[AgentToMove, OtherTeam]
        do:
            DebugLog.PrintDebug("Moving Agent To New Team")

    OnAgentJoin(InPlayer: player):void =
        BalanceTeams()
