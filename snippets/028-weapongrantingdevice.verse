# Knowledge-neutral Verse snippet; project names and asset IDs generalized.
# id: snippet-028

using { /Fortnite.com/Devices }
using { /Verse.org/Simulation }
using { /UnrealEngine.com/Temporary/Diagnostics }
using { /Verse.org/Random }
using{Helpers}

weapon_granting_log := class(log_channel){}
    # For this experience, players will have the same weapons for at least two rounds
    # For this functionality, data is stored between rounds in the WeaponDataMap
    # The data stored is an array containing the index of the current item pool
    # Then the index of the item granter within that pool

var WeaponDataMap:weak_map(player, []tuple(int, int)) = map{}

    # Each Item pool within the weapon-granting device contains an array
    # of item granter devices that the user references in the editor
item_pools := class<concrete>():
    @editable
    Granters<public>:[]item_granter_device = array{}
    @editable
    GrantItemFromPool<public>: logic = true
    @editable
    NumberToGrant<public>: int = 1

    # The weapon-granting device will grant all players the same set of weapons.
    # You can control when the script will grant the weapons to the players, either
    # via the round device or the grant items trigger device.
    # In this experience, we grant a new random set of items to every player
    # every second round.
    # If you look at the device in the editor, you can see round setting devices have been
    # reference for the 1, 3, 5 ... etc,  rounds to grant new weapons and
    # reference for the 2, 4, 6 ... etc,  rounds to grant weapons from the previous round
weapon_granting_device := class(creative_device):
    DebugLog:log = log{Channel := weapon_granting_log}

    @editable
    GranterPools<private>:[]item_pools = array{}

    @editable
    GrantItemsTrigger<private>:trigger_device = trigger_device{}

    @editable
    RoundsToGrant<private>:[]round_settings_device = array{}

    @editable
    LoadPreviousOnRounds<private>:[]round_settings_device = array{}

    @editable
    SaveLoadoutOnGrant<private>:logic = false

    # This list is populated when deciding which items we will grant players.
    # There are two reasons why we store the data here and don't grant the items whilst simultaneously calculating which item to grant.

    # 1 We need to wait for the previous item to finish granting before we grant the next. Otherwise
    # we can't guarantee the order of items in a player's inventory. e.g. Grenade in slot one and Gun in slot 2

    # 2 If we have a wait between granting items and calculating the weapons to grant
    #, if a player is late, join during this wait. The data will be incomplete, and the joiner will be late.
    # wouldn't receive all of the items

    var DelayedGranters<public>:[]item_granter_device = array{}

    # This is a minor optimization for late joiners.
    # Instead of calculating which weapons they need from the Weak Map
    #, the player can use the cached data version if it is complete.
    # This logic will be actual if the data for which weapons to grant is completely setup
    var UseLocalForLateJoiners<private>:logic = false

    # Runs when the device is started in a running game
    OnBegin<override>()<suspends>:void =
        # Subscribe to late joiners
        GetPlayspace().PlayerAddedEvent().Subscribe(OnAgentJoin)

        GrantItemsTrigger.TriggeredEvent.Subscribe(OnSwapTriggered)

        for (RoundDevice : RoundsToGrant):
            RoundDevice.RoundBeginEvent.Subscribe(OnRoundStarted)

        for (LoadRounds : LoadPreviousOnRounds):
            LoadRounds.RoundBeginEvent.Subscribe(OnLoadPrevious)

    OnRoundStarted():void =
        CreateGrantData()
        GrantItemToAllPlayers()

    OnSwapTriggered(InAgent:?agent):void =
        CreateGrantData()
        GrantItemToAllPlayers()

    OnLoadPrevious():void =
        DebugLog.PrintDebug("Loading previous weapons")

        # To keep persistent data between rounds we need to store this information in the player's map
        # As there is a possibility that each player has different data stored, we find the most common data between
        # All the players currently have this experience and use it.

        # Check each player's first weapon index to validate the most frequently picked
        PlayersLastIndex := for(Player : GetPlayspace().GetPlayers(), Index := WeaponDataMap[Player][0](1)){Index}
        MostFrequentIndex := FindMostFrequentIndex(PlayersLastIndex)

        # Use the most frequently occurring data as the source of truth for this session
        if(Player := FindPlayerWithMostFrequentIndex[MostFrequentIndex], WeaponData := WeaponDataMap[Player]):
            set DelayedGranters = for (Indexes : WeaponData, Pool:= GranterPools[Indexes(0)], Granter := Pool.Granters[Indexes(1)]){Granter}
        else:
            DebugLog.PrintDebug("Failed to Load Agreed Player weapons")

        set UseLocalForLateJoiners = true
        GrantItemToAllPlayers()

    CreateGrantData<private>():void =
        var GrantedThisSession : []tuple(int, int) = array{}

        for (ParentIndex -> GranterPool : GranterPools, GranterPool.GrantItemFromPool?, GranterPool.NumberToGrant > 0):
            # Create a shuffled array of indexes to randomise what item you receive
            ShuffledIndexesArray := Shuffle(for (Index -> ItemGranters: GranterPool.Granters){Index})

            set GrantedThisSession += for (Index := 0..GranterPool.NumberToGrant - 1, RandomItemIndex := ShuffledIndexesArray[Index], Granter:= GranterPool.Granters[RandomItemIndex]):
                set DelayedGranters += array{Granter}
                (ParentIndex, RandomItemIndex)

        set UseLocalForLateJoiners = true
        SaveLoadOut(GrantedThisSession)

    GrantItemToAllPlayers():void=
        for(Player : GetPlayspace().GetPlayers()):
            spawn:
                GrantItems(Player, DelayedGranters)

    # This method will cycle through all the selected granters. It will wait for the previous grant to complete
    # before requesting the next item. We do this to guarantee the order of the requests.
    GrantItems(Player:player, Granters:[]item_granter_device)<suspends>:void =
        for (Granter:Granters):
            Granter.GrantItem(Player)
            Granter.ItemGrantedEvent.Await()

    FindPlayerWithMostFrequentIndex<private>(MostFrequentIndex:int)<decides><transacts>:player =
        for(Player : GetPlayspace().GetPlayers(), WeaponDataMap[Player][0](1) = MostFrequentIndex) {Player}[0]

    # Save the granted items to the player map to retrieve on future rounds.
    SaveLoadOut(SaveData: []tuple(int, int)):void =
        if (SaveLoadoutOnGrant?):
            for (Player : GetPlayspace().GetPlayers()):
                if( PreviousData := WeaponDataMap[Player]):
                    if (set WeaponDataMap[Player] = SaveData):
                        DebugLog.PrintDebug("Save - Updating previous player weapon data")
                else:
                    if (set WeaponDataMap[Player] = SaveData):
                        DebugLog.PrintDebug("Save - Creating new player weapon data")
    
    # Method handles granting late joiners the items which the players currently in the session are using.
    OnAgentJoin(InPlayer: player):void =
        DebugLog.PrintDebug("Late Player has joined")

        if (UseLocalForLateJoiners?):
            DebugLog.PrintDebug("Late joiner using local data to grant item")
            spawn:
                GrantItems(InPlayer, DelayedGranters)
        else:
            DebugLog.PrintDebug("Late joiner using Agreed player data to grant item")

            PlayersLastIndex := for(Player : GetPlayspace().GetPlayers(), Index := WeaponDataMap[Player][0](1)){Index}
            MostFrequentIndex := FindMostFrequentIndex(PlayersLastIndex)

            if(Player := FindPlayerWithMostFrequentIndex[MostFrequentIndex], WeaponData := WeaponDataMap[Player]):
                Granters := for (Indexes : WeaponData, Pool:= GranterPools[Indexes(0)], Granter := Pool.Granters[Indexes(1)]){Granter}
                spawn:
                    GrantItems(InPlayer, Granters)
            else:
                DebugLog.PrintDebug("Failed to Load Agreed Player weapons for late joiner")
