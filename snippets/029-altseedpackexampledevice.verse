# Knowledge-neutral Verse snippet; project names and asset IDs generalized.
# id: snippet-029

# Copyright Epic Games, Inc. All Rights Reserved.

using { /Fortnite.com/Devices }
using { /UnrealEngine.com/Marketplace }
using { /Verse.org/Simulation }

using { Definitions.ExampleOffers }

alt_seed_pack_example_device<public> := class(creative_device):

    @editable
    Button : button_device = button_device{}

    @editable
    SwitchOfferButton : button_device = button_device{}

    var SwitchOffer :logic = false

    OnBegin<override>()<suspends>:void=
        Button.InteractedWithEvent.Subscribe(OnButtonInteraction)
        SwitchOfferButton.InteractedWithEvent.Subscribe(OnSwitchOfferInteraction)

    OnSwitchOfferInteraction(Agent:agent):void=
        set SwitchOffer = logic {not SwitchOffer?}

    OnButtonInteraction(Agent:agent):void=
        if (Player := player[Agent]):
            if (not SwitchOffer?):
                spawn{TryBuyOffer(Player, corn_seed_pack_bundle{})}
            else:
                spawn{TryBuyOffer(Player, corn_seed_pack_bundle_alt{})}

    TryBuyOffer(Player:player, Offer : offer)<suspends>:void =
        Result := BuyOffer(Player, Offer)

        if (Result?):
            # Do nothing it should respond in the purchase subscription
        else:
            Print("Failed to buy the offer")
