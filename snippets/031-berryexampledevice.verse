# Knowledge-neutral Verse snippet; project names and asset IDs generalized.
# id: snippet-031

# Copyright Epic Games, Inc. All Rights Reserved.

using { /Fortnite.com/Devices }
using { /UnrealEngine.com/Marketplace }
using { /Verse.org/Simulation }

using { Definitions }

berry_example_device := class(creative_device):

    @editable
    Button : button_device = button_device{}

    @editable
    ItemGranter : item_granter_device = item_granter_device{}
        
    @editable
    AllRemover : item_remover_device = item_remover_device{}

    # Runs when the device is started in a running game
    OnBegin<override>()<suspends>:void=
        Button.InteractedWithEvent.Subscribe(OnButtonInteraction)
        Playspace := GetPlayspace()
        for (Player : Playspace.GetPlayers()):
            OnPlayerJoin(Player)

            spawn {RestoreBerryState(Player)}

    OnPlayerJoin(InPlayer : player):void = 
        GetEntitlementsChangedEvent(InPlayer, Entitlements.feature_example_entitlement).Subscribe(OnEntitlementsChanged)

    OnEntitlementsChanged(Result : tuple(player, []entitlement_change(entitlement))):void=
        for (EntitlementChange : Result(1)):
            Entitlement := EntitlementChange.Entitlement

            if (berry := Entitlements.max_berries[Entitlement]):
                
                # Clear all and re-grant
                AllRemover.Remove(Result(0))
                
                if (EntitlementChange.Change > 0):
                    for (Range := 0 .. EntitlementChange.Quantity - 1):
                        ItemGranter.GrantItem(Result(0))
                

    OnButtonInteraction(Agent:agent):void=
        if (Player := player[Agent]):
           spawn{TryBuyOffer(Player, ExampleOffers.max_berries{})}

    TryBuyOffer(Player:player, Offer : offer)<suspends>:void =
        Result := BuyOffer(Player, Offer)

    # This function restores items to the player on subsequent playthroughs
    RestoreBerryState(Player : player)<suspends>:void=
        Purchases := GetPurchasedEntitlements(Player, Entitlements.max_berries)
        if (PurchasedBerries := Purchases[0]):

            # Clear all and re-grant
            AllRemover.Remove(Player)

            if (PurchasedBerries(1) > 0):
                for (Range := 0 .. PurchasedBerries(1) - 1):
                    ItemGranter.GrantItem(Player)
