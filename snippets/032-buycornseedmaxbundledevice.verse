# Knowledge-neutral Verse snippet; project names and asset IDs generalized.
# id: snippet-032

# Copyright Epic Games, Inc. All Rights Reserved.

using { /Fortnite.com/Devices }
using { /UnrealEngine.com/Marketplace }
using { /Verse.org/Simulation }

using { Definitions }

buy_corn_seed_max_bundle_device := class(creative_device):

    @editable
    Button : button_device = button_device{}

    OnBegin<override>()<suspends>:void=
        Button.InteractedWithEvent.Subscribe(OnButtonInteraction)

    OnButtonInteraction(Agent:agent):void=
        if (Player := player[Agent]):
           spawn{TryBuyOffer(Player)}

    TryBuyOffer(Player : player)<suspends>:void =
        Purchases := GetPurchasedEntitlements(Player, Entitlements.corn_seed_pack)
        var NumPlayerCornSeedPacks : int = 0
        if (Purchase := Purchases[0]):
            set NumPlayerCornSeedPacks = Purchase(1)

        # Limit to at least 1 packet
        # Even if player has max amount the offer will still show on the screen with disabled purchase button
        NumCornSeedPacks := Max(1, Entitlements.corn_seed_pack{}.MaxCount - NumPlayerCornSeedPacks)
        
        Offer := ExampleOffers.max_corn_seed_pack_bundle{
            Offers := array{(ExampleOffers.corn_seed_pack{}, NumCornSeedPacks)},
            Price := MakePriceVBucks(NumCornSeedPacks * OfferInfo.CornSeedPacket.Price)
        }
        Result := BuyOffer(Player, Offer)

        if (Result?):
            # Do nothing it should respond in the purchase subscription
        else:
            Print("Failed to buy the offer")
