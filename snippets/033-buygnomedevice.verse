# Knowledge-neutral Verse snippet; project names and asset IDs generalized.
# id: snippet-033

# Copyright Epic Games, Inc. All Rights Reserved.

using { /Fortnite.com/Devices }
using { /UnrealEngine.com/Marketplace }
using { /Verse.org/Simulation }

using { Definitions }

buy_gnome_device := class(creative_device):

    @editable
    Button : button_device = button_device{}

    @editable
    AltGnomeButton : button_device = button_device{}

    OnBegin<override>()<suspends>:void=
        Button.InteractedWithEvent.Subscribe(OnButtonInteraction)
        AltGnomeButton.InteractedWithEvent.Subscribe(OnReducedGnomeButtonInteraction)
        Playspace := GetPlayspace()

    OnButtonInteraction(Agent:agent):void=
        if (Player := player[Agent]):
            spawn{TryBuyOffer(Player, ExampleOffers.gnome{})}

    OnReducedGnomeButtonInteraction(Agent:agent):void=
        if (Player := player[Agent]):
            spawn{TryBuyOffer(Player, ExampleOffers.gnome{Icon := OfferInfo.Gnome.AlternativeIcon})}

    TryBuyOffer(Player : player, Offer : offer)<suspends>:void =
        Result := BuyOffer(Player, Offer)

        if (Result?):
            # Do nothing it should respond in the purchase subscription
        else:
            Print("Failed to buy the offer")
