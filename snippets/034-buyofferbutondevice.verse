# Knowledge-neutral Verse snippet; project names and asset IDs generalized.
# id: snippet-034

# Copyright Epic Games, Inc. All Rights Reserved.

using { /Fortnite.com/Devices }
using { /UnrealEngine.com/Marketplace }
using { /Verse.org/Simulation }

using { Definitions.ExampleOffers }

offer_type := enum :
    CornSeedPack
    CornSeedPackBundle
    PrimoShovel
    StarterBundle
    # Gnome is dynamic offer
    # Gnome
    # Max corn seed pack bundle is created dynamically - can't set price here
    # CornSeedPackMaxBundle

(OfferType : offer_type).OfferTypeToOffer():offer =
    case (OfferType):
        offer_type.CornSeedPack             =>  corn_seed_pack{}
        offer_type.CornSeedPackBundle       =>  corn_seed_pack_bundle{}
        offer_type.PrimoShovel              =>  primo_shovel{}
        offer_type.StarterBundle            =>  starter_bundle{}
        # Gnome is dynamic offer
        # offer_type.Gnome                    =>  gnome{}
        # Max corn seed pack bundle is created dynamically - can't set price here
        # offer_type.CornSeedPackMaxBundle    =>  max_corn_seed_pack_bundle{}

buy_offer_device<public> := class(creative_device):

    @editable
    Button : button_device = button_device{}

    @editable
    OfferType : offer_type = offer_type.CornSeedPack

    OnBegin<override>()<suspends>:void=
        Button.InteractedWithEvent.Subscribe(OnButtonInteraction)

    OnButtonInteraction(Agent:agent):void=
        if (Player := player[Agent]):
           spawn{TryBuyOffer(Player, OfferType.OfferTypeToOffer())}

    TryBuyOffer(Player:player, Offer : offer)<suspends>:void =
        Result := BuyOffer(Player, Offer)

        if (Result?):
            # Do nothing it should respond in the purchase subscription
        else:
            Print("Failed to buy the offer")
