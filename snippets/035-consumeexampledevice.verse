# Knowledge-neutral Verse snippet; project names and asset IDs generalized.
# id: snippet-035

# Copyright Epic Games, Inc. All Rights Reserved.

using { /Fortnite.com/Devices }
using { /UnrealEngine.com/Marketplace }
using { /Verse.org/Simulation }

using {Definitions}

consume_example_device := class(creative_device):
    
    @editable
    EntitlementToGrant : entitlement_enum = entitlement_enum.CornSeed
    @editable
    EntitlementToConsume : entitlement_enum = entitlement_enum.CornSeedPack
    
    @editable
    Button : button_device = button_device{}

    @editable
    NumberToConsume : int = 1

    @editable
    NumberToGrant : int = 10

    OnBegin<override>()<suspends>:void=
        Button.InteractedWithEvent.Subscribe(OnButtonInteraction)

    OnButtonInteraction(Agent:agent):void=
        if (Player := player[Agent]):
           spawn{TryConsumeEntitlement(Player, EntitlementToConsume.ConvertToEntitlement())}

    TryConsumeEntitlement(Player : player, EntitlementType : concrete_subtype(entitlement))<suspends>:void=
        ConsumeResult := ConsumeEntitlement(Player, EntitlementType, ?Count := NumberToConsume)

        if:
            ConsumeResult?
        then:
            Print("Consumed entitlement!")
            GrantResult := GrantEntitlement(Player, EntitlementToGrant.ConvertToEntitlement(), ?Count := NumberToGrant)
            if (GrantResult?):
                Print("Granted entitlement!")
            else:
                Print("Failed to grant the entitlement")
        else:
            Print("Failed to consume the entitlement")
