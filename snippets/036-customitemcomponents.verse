# Knowledge-neutral Verse snippet; project names and asset IDs generalized.
# id: snippet-036

using { Definitions.Entitlements }
using { /Fortnite.com/Devices }
using { /Fortnite.com/Itemization }
using { /UnrealEngine.com/Marketplace }
using { /UnrealEngine.com/Itemization }
using { /Verse.org/SceneGraph }
using { /Verse.org/Simulation }

# Unique subclass to identify watering cans.
watering_can_item_component := class(item_component) :

# Unique subclass to identify pumpkin seeds.
pumpkin_seed_item_component := class(item_component) :

    GrowPumpkin():void=
        if:
            EntitlementComp := Entity.GetComponent[item_entitlement_component]
            StackableComp := Entity.GetComponent[stackable_component]
        then:
            EntitlementComp.ConsumeEntitlement(1)
            if(StackableComp.StackSize > 1):
                StackableComp.SetStackSize(StackableComp.StackSize-1)
            else:
                Entity.RemoveFromParent()

# This can be deleted and replaced with a basic_stackable_component in the prefab very soon.
simple_stackable_component := class(stackable_component):
    CanMergeInto<override>(TargetEntity:entity)<decides><reads>:void=
        (super:)CanMergeInto[TargetEntity]
        TargetEntity.GetComponent[pumpkin_seed_item_component]

    Split<override>(Amount:int)<transacts><decides>:entity =
        # If we take the whole amount, just return this entity (no need to actually Split)
        if (StackSize = Amount):
            return Entity
        else:
            false?
            Entity # required by the compiler although it will never reach here

# The purpose of this component is to track ownership and block inventory entry if owner does not have the Entitlement.
# It is added to an item entity when it has been spawned by an entitlement grant.
# The component is also responsible for consuming an entitlement when required.
item_entitlement_component := class<final_super>(component):

    # Used to block item drop when the item has an entitlement.
    CanBeDropped:logic = false

    # Which player owns this item?
    PlayerOwner:player

    # Which Entitlement is this entity the embodiment for?
    EntitlementType:concrete_subtype(entitlement)

    # Subscribe to PlayerRemovedEvent so an item with an entitlement component gets removed when the owning player leaves the simulation.
    OnBeginSimulation<override>():void =
        if:
            SimEntity := Entity.GetSimulationEntity[]
            CreativeDevice := creative_device[SimEntity]
        then:
            PlaySpace := CreativeDevice.GetPlayspace()
            PlaySpace.PlayerRemovedEvent().Subscribe(OnPlayerRemovedEvent)

    OnPlayerRemovedEvent(Player:player):void = if(IsOwner[Player]) then Entity.RemoveFromParent()

    # Succeeds if the provided player is the same as the cached PlayerOwner.
    IsOwner(Player:player)<decides><transacts>:void=
        Player = PlayerOwner

    OnReceive<override>(SceneEvent:scene_event):logic=
        if:
            AddItemQueryEvent := add_item_query_event[SceneEvent]
            EntitlementComponent := Entity.GetComponent[item_entitlement_component]
            AddPlayer := GetOwningPlayer[AddItemQueryEvent.Inventory.Entity]
            AddPlayer <> GetOwningPlayer[Entity]
        then:
            AddItemQueryEvent.AddError(not_entitled_to_player_add_item_error{})

        if:
            not CanBeDropped?
            RemoveItemQueryEvent := remove_item_query_event[SceneEvent]
        then:
            RemoveItemQueryEvent.AddError(cannot_drop_item_remove_query_error{})

        false

    ConsumeEntitlement(Count:int):void=
        if:
            ItemComponent := Entity.GetComponent[item_component]
        then:
            spawn{ConsumeEntitlement(PlayerOwner,EntitlementType,?Count := Count)}

# Add item error that can be created when an item is already entitled to another player.
not_entitled_to_player_add_item_error := class(add_item_error) :

# Remove ite error that blocks the removal of this item from the inventory.
cannot_drop_item_remove_query_error := class(remove_item_error) :
