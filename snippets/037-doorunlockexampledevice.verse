# Knowledge-neutral Verse snippet; project names and asset IDs generalized.
# id: snippet-037

# Copyright Epic Games, Inc. All Rights Reserved.

using { /Fortnite.com/Devices }
using { /UnrealEngine.com/Marketplace }
using { /Verse.org/Simulation }

using { Definitions }

door_unlock_example_device := class(creative_device):

    @editable
    Button : button_device = button_device{}

    @editable
    DoorSequence : cinematic_sequence_device = cinematic_sequence_device{}

    # Runs when the device is started in a running game
    OnBegin<override>()<suspends>:void=
        Button.InteractedWithEvent.Subscribe(OnButtonInteraction)
        Playspace := GetPlayspace()
        for (Player : Playspace.GetPlayers()):
            OnPlayerJoin(Player)

            spawn {RestoreDoorState(Player)}

    OnPlayerJoin(InPlayer : player):void = 
        GetEntitlementsChangedEvent(InPlayer, Entitlements.feature_example_entitlement).Subscribe(OnEntitlementChanged)

    OnEntitlementChanged(Result : tuple(player, []entitlement_change(entitlement))):void=
        
        for (EntitlementChange : Result(1)):
            Entitlement := EntitlementChange.Entitlement

            if (Entitlements.vip_access[Entitlement]):   
                if (EntitlementChange.Quantity > 0):
                    DoorSequence.Play(Result(0))
                else:
                    DoorSequence.PlayReverse(Result(0))

    OnButtonInteraction(Agent:agent):void=
        if (Player := player[Agent]):
           spawn{TryBuyOffer(Player, ExampleOffers.vip_access{})}

    TryBuyOffer(Player:player, Offer : offer)<suspends>:void =
        Result := BuyOffer(Player, Offer)

    # This function restores items to the player on subsequent playthroughs
    RestoreDoorState(Player : player)<suspends>:void=
        Purchases := GetPurchasedEntitlements(Player, Entitlements.vip_access)
        if (PurchasedKey := Purchases[0]):
            if (PurchasedKey(1) > 0):
                DoorSequence.Play(Player)
            else:
                DoorSequence.PlayReverse(Player)
