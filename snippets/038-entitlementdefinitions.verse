# Knowledge-neutral Verse snippet; project names and asset IDs generalized.
# id: snippet-038

# Copyright Epic Games, Inc. All Rights Reserved.

using { /Verse.org/Assets }
using { /UnrealEngine.com/Marketplace }

using { Art }

EntitlementInfo<public> := module:
    PrimoShovel<public> := module:
        Name<public><localizes>: message = "Primo shovel"
        Description<public><localizes>: message = "An unbreakable shovel used to dig holes for planting."
        ShortDescription<public><localizes>: message = "Digs holes."

    CornSeed<public> := module:
        Name<public><localizes> : message = "Corn seed"
        Description<public><localizes> : message = "An individual corn seed. Plant to grow one corn stalk."
        ShortDescription<public><localizes> : message = "Plants corn."

    CornSeedPacket<public> := module:
        Name<public><localizes> : message = "Corn seed pack"
        Description<public><localizes> : message = "A pack of corn seeds. Opening a pack yields 10 corn seeds for planting."
        ShortDescription<public><localizes> : message = "Contains 10 corn seeds for planting."

    Gnome<public> := module:
        Name<public><localizes>: message = "Gnome"
        Description<public><localizes>: message = "A handsome garden gnome that adds character to your garden."
        ShortDescription<public><localizes>: message = "A decorative gnome."

    VIPAccess<public> := module:
        Name<public><localizes>: message = "VIP access"
        Description<public><localizes>: message = "Grants access to a restricted area of the island. After purchasing, the VIP door will open automatically."
        ShortDescription<public><localizes>: message = "Grants access to a restricted area of the island."

    MaxBerries<public> := module:
        Name<public><localizes>: message = "Flowberry Capacity"
        Description<public><localizes>: message = "Increases the number of Flowberries a player can hold in their inventory during a match. Purchasing an upgrade fills inventory.  Absolute maximum of six. Flowberry: Juice-filled berry that grants shields and low-grav to you and your allies."
        ShortDescription<public><localizes>: message = "Increase and fill max Flowberries."

    PumpkinSeed<public> := module:
        Name<public><localizes>: message = "Pumpkin Seed"
        Description<public><localizes>: message = "An individual pumpkin seed. Plant one and then water it to grow a pumpkin."
        ShortDescription<public><localizes>: message = "Plants a pumpkin."

    WateringCan<public> := module:
        Name<public><localizes>: message = "Watering Can"
        Description<public><localizes>: message = "A Watering Can. Can be used on planted seeds to help them grow."
        ShortDescription<public><localizes>: message = "Waters seed and plants."

<#>
    Here you can define entitlements available in your game.
    Entitlement definition contains:
        * text info
            * name
            * description
            * short description
        * icon
        * maximum count in inventory
        * whether or not it's consumable
        * if it represents a paid random item
        * if it allows access to an area of your experience
        * if it can be consequential for gameplay
    Each entitlement can be later exposed in shop by defining offer(s) that sell this entitlement.
Entitlements<public> := module:
    using { EntitlementInfo }

    # Using custom entitlement class to distinguish your entitlements from all other entitlements.
    feature_example_entitlement<public> := class<abstract><castable><unique>(entitlement){}

    # Example of consumable entitlement
    corn_seed<public> := class<concrete>(feature_example_entitlement):
        var Name<override> : message = CornSeed.Name
        var Description<override> : message = CornSeed.Description
        var ShortDescription<override> : message = CornSeed.ShortDescription
        var Icon<override> : texture = Textures.T_Corn_Seeds
        # Maximum of 10 corn seeds in the inventory
        MaxCount<override> : int = 100
        # Each corn seed is consumable - usage removes it from Player's inventory
        Consumable<override> : logic = true
        PaidRandomItem<override> : logic = false
        PaidArea<override> : logic = false
        ConsequentialToGameplay<override>:logic = true

    corn_seed_pack<public> := class<concrete>(feature_example_entitlement):
        var Name<override> : message = CornSeedPacket.Name
        var Description<override> : message = CornSeedPacket.Description
        var ShortDescription<override> : message = CornSeedPacket.ShortDescription
        var Icon<override> : texture = Textures.T_Corn_Seed_Packet
        # Maximum of 10 corn seed packets in the inventory
        MaxCount<override> : int = 10
        # Each corn seed packet is consumable - usage removes it from Player's inventory
        # To be implemented: consuming grants 10 corn seed entitlements
        Consumable<override> : logic = true
        PaidRandomItem<override> : logic = false
        PaidArea<override> : logic = false
        ConsequentialToGameplay<override>:logic = true

    # Example of durable entitlement
    primo_shovel<public> := class<concrete>(feature_example_entitlement):
        var Name<override>:message = PrimoShovel.Name
        var Description<override>:message = PrimoShovel.Description
        var ShortDescription<override>:message = PrimoShovel.ShortDescription
        var Icon<override>:texture = Textures.T_Shovel
        # Setting 'Consumable' to false means the entitlement is durable
        Consumable<override> : logic = false
        PaidRandomItem<override> : logic = false
        PaidArea<override> : logic = false
        ConsequentialToGameplay<override>:logic = true

    gnome<public> := class<concrete>(feature_example_entitlement):
        var Name<override>:message = Gnome.Name
        var Description<override>:message = Gnome.Description
        var ShortDescription<override>:message = Gnome.ShortDescription
        var Icon<override>:texture = Textures.T_Gnome_Icon
        MaxCount<override> : int = 3
        Consumable<override> : logic = true
        PaidRandomItem<override> : logic = false
        PaidArea<override> : logic = false
        # As a cosmetic item, it has no gameplay consequences
        ConsequentialToGameplay<override>:logic = false

    vip_access<public> := class<concrete>(feature_example_entitlement):
        var Name<override>:message = VIPAccess.Name
        var Description<override>:message = VIPAccess.Description
        var ShortDescription<override>:message = VIPAccess.ShortDescription
        var Icon<override>:texture = Textures.T_VIP_Access
        Consumable<override> : logic = false
        PaidRandomItem<override> : logic = false
        PaidArea<override> : logic = true
        ConsequentialToGameplay<override>:logic = true

    max_berries<public> := class<concrete>(feature_example_entitlement):
        var Name<override>:message = MaxBerries.Name
        var Description<override>:message = MaxBerries.Description
        var ShortDescription<override>:message = MaxBerries.ShortDescription
        var Icon<override>:texture = Textures.T_Flowberry
        MaxCount<override> : int = 6
        Consumable<override> : logic = true
        PaidRandomItem<override> : logic = false
        PaidArea<override> : logic = false
        ConsequentialToGameplay<override>:logic = true

    pumpkin_seed<public> := class<concrete>(feature_example_entitlement):
        var Name<override>:message = PumpkinSeed.Name
        var Description<override>:message = PumpkinSeed.Description
        var ShortDescription<override>:message = PumpkinSeed.ShortDescription
        var Icon<override>:texture = Textures.T_PumpkinSeed
        MaxCount<override> : int = 10
        Consumable<override> : logic = true
        PaidRandomItem<override> : logic = false
        PaidArea<override> : logic = false
        ConsequentialToGameplay<override>:logic = true

    # Example of durable entitlement
    watering_can<public> := class<concrete>(feature_example_entitlement):
        var Name<override>:message = WateringCan.Name
        var Description<override>:message = WateringCan.Description
        var ShortDescription<override>:message = WateringCan.ShortDescription
        var Icon<override>:texture = Textures.T_WateringCan
        # Setting 'Consumable' to false means the entitlement is durable
        Consumable<override> : logic = false
        PaidRandomItem<override> : logic = false
        PaidArea<override> : logic = false
        ConsequentialToGameplay<override>:logic = true
