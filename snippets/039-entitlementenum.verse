# Knowledge-neutral Verse snippet; project names and asset IDs generalized.
# id: snippet-039

# Copyright Epic Games, Inc. All Rights Reserved.

using { /UnrealEngine.com/Marketplace }

using {Definitions}

entitlement_enum := enum:
    CornSeed
    PrimoShovel
    CornSeedPack
    PumpkinSeed
    WateringCan

(EntitlementEnum :entitlement_enum).ConvertToEntitlement() : concrete_subtype(entitlement)=
    case (EntitlementEnum):
        entitlement_enum.CornSeed => return Entitlements.corn_seed
        entitlement_enum.PrimoShovel => return Entitlements.primo_shovel
        entitlement_enum.CornSeedPack => return Entitlements.corn_seed_pack
        entitlement_enum.PumpkinSeed => return Entitlements.pumpkin_seed
        entitlement_enum.WateringCan => return Entitlements.watering_can
