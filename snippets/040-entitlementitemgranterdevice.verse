# Knowledge-neutral Verse snippet; project names and asset IDs generalized.
# id: snippet-040

using { Definitions.Entitlements }
using { /Fortnite.com/Devices }
using { /UnrealEngine.com/Marketplace }
using { /Verse.org/SceneGraph }
using { /Verse.org/Simulation }
using { /UnrealEngine.com/Itemization }

# This device will create an item and add it to a player inventory when certain entitlements are granted.
# It manages entitlements by adding an item_entitlement_component to the item entity when it is spawned.
entitlement_item_granter_device := class(creative_device):

    OnBegin<override>()<suspends>:void =
        BindAllPlayerEntitlementsToItems()

    # Subscribe to PlayerEntered events: To add items when players join the simulation.
    BindAllPlayerEntitlementsToItems():void =
        PlaySpace := GetPlayspace()
        PlaySpace.PlayerAddedEvent().Subscribe(OnPlayerAddedEvent)
        for(Player:PlaySpace.GetPlayers()):
            OnPlayerAddedEvent(Player)

    # Subscribe to GetEntitlementsChangedEvent events: Used to modify items when the player's entitlements change.
    OnPlayerAddedEvent(Player:player):void =
        GetEntitlementsChangedEvent(Player,feature_example_entitlement).Subscribe(OnEntitlementsChangedEvent)
        spawn{GrantPlayerAllEntitledItems(Player)}

    OnEntitlementsChangedEvent(Player:player, EntitlementChange:[]entitlement_change(feature_example_entitlement)):void =
        for(Change:EntitlementChange, Entitlement := Change.Entitlement):
            CountChange := Change.Change
            if(CountChange > 0):
                GrantEntitlementItem(Player,Entitlement,CountChange)
    
    # Get all Entitlements and attempt to grant a matching item for each count.
    GrantPlayerAllEntitledItems(Player:player)<suspends>:void =
        PlayerEntitlements := GetPurchasedEntitlements(Player,feature_example_entitlement)
        for(Entitlement:PlayerEntitlements):
            EntitlementCount := Entitlement(1)
            GrantEntitlementItem(Player,Entitlement(0),EntitlementCount)
    
    # Grant a single stack of the entitlement item to a player's inventory.
    GrantEntitlementItem(Player:player, Entitlement:entitlement, Count:int):void =
        if:
            TargetInventory := GetAgentInventory[Player]
            SpawnItemType:concrete_subtype(entity) = GetEntityTypeFromEntitlement[Entitlement]
            ItemEntitlementType := GetEntitlementTypeFromEntitlement[Entitlement]
        then:
            GrantItem := SpawnItemType{}
            if(StackableComponent := GrantItem.GetComponent[stackable_component]):
                StackableComponent.SetStackSize(Count)
            IsPumpkinSeed := logic{Content.Item_Seed_Pumpkin[GrantItem]} # In the case of the Pumpkin Seeds, we want to allow dropping.
            NewEntitlementComponent := item_entitlement_component{Entity := GrantItem, EntitlementType := ItemEntitlementType, PlayerOwner := Player, CanBeDropped := IsPumpkinSeed}
            GrantItem.AddComponents(array{NewEntitlementComponent})
            if(AddResult := TargetInventory.AddItemDistribute(GrantItem).GetSuccess[]):
