# Knowledge-neutral Verse snippet; project names and asset IDs generalized.
# id: snippet-041

# Copyright Epic Games, Inc. All Rights Reserved.

using { /Fortnite.com/Devices }
using { /UnrealEngine.com/Marketplace }
using { /Verse.org/Simulation }


grant_entitlement_device<public> := class(creative_device):

    @editable
    EntitlementToGrant : entitlement_enum = entitlement_enum.CornSeed

    @editable
    Button : button_device = button_device{}
    @editable
    NumberToGrant : int = 1

    OnBegin<override>()<suspends>:void=
        
        Button.InteractedWithEvent.Subscribe(OnButtonInteraction)

    OnButtonInteraction(Agent:agent):void=

        if (Player := player[Agent]):
            spawn{TryGrantEntitlement(Player, EntitlementToGrant.ConvertToEntitlement())}

    TryGrantEntitlement(Player:player, Entitlement:concrete_subtype(entitlement))<suspends>:void=
        Result := GrantEntitlement(Player, Entitlement, ?Count := NumberToGrant)

        if (Result?):
            Print("Successfully granted entitlement!")
        else:
            Print("Failed to grant the entitlement")
