# Knowledge-neutral Verse snippet; project names and asset IDs generalized.
# id: snippet-042

# Copyright Epic Games, Inc. All Rights Reserved.
using { Definitions.Entitlements }
using { /UnrealEngine.com/Marketplace }
using { /Verse.org/SceneGraph }
using { /Verse.org/Simulation }
using { /UnrealEngine.com/Itemization }

StringToMessage<localizes>(Value:string):message= "{Value}"

# Used to get the eventual player entity if it exists.
# Relies on the helper function from crafting_system.verse
GetOwningPlayer(ParentEntity:entity)<decides><transacts>:player=
    OwningPlayer := player[GetAncestorAgent[ParentEntity]]

# This helper function will return the first ancestor agent it receives, which will be the player.
# Use this to get the player that owns a subentity.
GetAncestorAgent(Entity:entity)<decides><transacts>:agent=
    Agent := (for (I : Entity.FindAncestorEntities(agent)) { I })[0]

# Used to check up the scene tree to discover the nearest ancestor entity with an item_component.
(Component:component).FindOwningItemEntity()<decides><transacts>:?entity=
    var ItemEntities:[]entity = array{Component.Entity}
    for(Entity:Component.Entity.FindAncestorEntities(entity),ItemEntity := Entity.GetComponent[item_component]):
        set ItemEntities = ItemEntities + array{Entity}
    option{ItemEntities[0]}

# Helper function that gets the first descendant inventory component from an agent.
# This will be the root inventory.
GetAgentInventory(Agent:agent)<decides><transacts>:inventory_component=
    TargetInventory := (for (I : Agent.FindDescendantComponents(inventory_component)) { I })[0]

# Cast an entitlement to get a matching prefab class for granting as an item. 
GetEntityTypeFromEntitlement(Entitlement:entitlement)<decides><transacts>:subtype(entity)=
    if(pumpkin_seed[Entitlement]):
        Content.Item_Seed_Pumpkin
    else if(watering_can[Entitlement]):
        Content.Item_WateringCan
    else:
        entity

# Get the entitlement type from an entitlement instance. 
GetEntitlementTypeFromEntitlement(Entitlement:entitlement)<decides><transacts>:concrete_subtype(entitlement) = 
    if(pumpkin_seed[Entitlement]):
        PumpkinType:concrete_subtype(entitlement) = pumpkin_seed
    else if(watering_can[Entitlement]):
         WateringCanType:concrete_subtype(entitlement) = watering_can
    else:
        false?
        pumpkin_seed

# Convert a generator into an array.
(Generator:generator(t) where t:type).ToArray():[]t=
    for(Element : Generator):
        Element
