# Knowledge-neutral Verse snippet; project names and asset IDs generalized.
# id: snippet-043

# Copyright Epic Games, Inc. All Rights Reserved.

using { /Fortnite.com/Devices }
using { /UnrealEngine.com/Marketplace }
using { /Verse.org/Simulation }

using { Definitions }

hole_state<public> := enum:
    Heap
    Hole
    Corn

PropCategory<localizes> : message = "Props"
SequenceCategory<localizes> : message = "Sequences"

sequence_to_play<public> := class:
    @editable
    Sequence<public> : cinematic_sequence_device

    @editable
    State<public> : hole_state

player_entitlement_state<public> := class:
    var HasShovel : logic = false
    var HasSeed : logic = false

hole_device := class(creative_device):

    @editable
    TargetHoleState : hole_state = hole_state.Heap

    @editable
    Button : button_device = button_device{}

    @editable
    SequencesToPlay : []sequence_to_play = array{}

    @editable:
        Categories := array{PropCategory}
    HeapProp : creative_prop = creative_prop{}

    @editable:
        Categories := array{PropCategory}
    HoleProp : creative_prop = creative_prop{}

    @editable:
        Categories := array{PropCategory}
    CornProp : creative_prop = creative_prop{}

    var PlayerEntitlementStateMap : [player]player_entitlement_state = map{}
    var HoleState : hole_state = hole_state.Heap
    var ShouldInteract : logic = true

    DigHoleMessage<localizes> : message = "Dig hole"
    PlantSeedMessage<localizes> : message = "Plant seed"
    InactiveMessage<localizes> : message = "Inactive"

    OnBegin<override>()<suspends>:void=
        SetHoleState(hole_state.Heap)
        Button.InteractedWithEvent.Subscribe(OnInteractedWith)
        Playspace := GetPlayspace()
        for (Player : Playspace.GetPlayers()):
            GetEntitlementsChangedEvent(Player, Entitlements.feature_example_entitlement).Subscribe(OnEntitlementsChanged)
            if (set PlayerEntitlementStateMap[Player] = player_entitlement_state{}):
                spawn{ValidatePreviousPurchases(Player)}

    OnInteractedWith(Agent : agent):void=
        if:
            Player := player[Agent]
            PlayerEntitlementState := GetPlayerEntitlementState[Player]
        then:

            if (PlayerEntitlementState.HasShovel? and HoleState = hole_state.Heap):
                SetHoleState(hole_state.Hole)
            else if (PlayerEntitlementState.HasSeed? and HoleState = hole_state.Hole):
                spawn{TryPlantSeed(Player)}

    OnEntitlementsChanged(Result : tuple(player, []entitlement_change(entitlement))):void=
        UpdatePlayerEntitlements(Result(0), Result(1))

    ValidatePreviousPurchases(Player:player)<suspends>:void=
        Purchases := GetPurchasedEntitlements(Player, Entitlements.feature_example_entitlement)
        ValidateUpdatePlayerEntitlements(Player, Purchases)

    UpdatePlayerEntitlements(Result : tuple(player, []entitlement_change(entitlement))):void=
        for (Purchase : Result(1)):
            Entitlement := Purchase.Entitlement
            Count := Purchase.Change
            if:
                Shovel := Entitlements.primo_shovel[Entitlement]
                set PlayerEntitlementStateMap[Result(0)].HasShovel = true
            if:
                CornSeed := Entitlements.corn_seed[Entitlement]
                set PlayerEntitlementStateMap[Result(0)].HasSeed = true

    ValidateUpdatePlayerEntitlements(Player :player, Purchases :[]tuple(entitlement, int)):void=
        for (Purchase : Purchases):
            Entitlement := Purchase(0)
            Count := Purchase(1)
            if:
                Shovel := Entitlements.primo_shovel[Entitlement]
                set PlayerEntitlementStateMap[Player].HasShovel = true
            if:
                CornSeed := Entitlements.corn_seed[Entitlement]
                set PlayerEntitlementStateMap[Player].HasSeed = true

    GetPlayerEntitlementState(Player : player)<transacts><decides>:player_entitlement_state=
        PlayerEntitlementStateMap[Player]

    TryPlantSeed(Player : player)<suspends>:void=
        Result := ConsumeEntitlement(Player, Entitlements.corn_seed)
        if (Result?):
            SetHoleState(hole_state.Corn)

    SetHoleState<public>(InHoleState : hole_state):void=
        set HoleState = InHoleState
        if (HoleState = TargetHoleState):
            set ShouldInteract = false
            Button.Disable()
        case (HoleState):
            hole_state.Heap =>
                HeapProp.Show()
                HoleProp.Hide()
                CornProp.Hide()
                Button.SetInteractionText(DigHoleMessage)
            hole_state.Hole =>
                HeapProp.Hide()
                HoleProp.Show()
                CornProp.Hide()
                Button.SetInteractionText(PlantSeedMessage)
            hole_state.Corn =>
                HeapProp.Hide()
                HoleProp.Hide()
                CornProp.Show()
                Button.SetInteractionText(InactiveMessage)

        PlaySequence(HoleState)

    PlaySequence(InHoleState : hole_state):void=
        for (SequenceToPlay : SequencesToPlay):
            if (SequenceToPlay.State = InHoleState):
                SequenceToPlay.Sequence.Play()
                Print("Playing sequence...")
