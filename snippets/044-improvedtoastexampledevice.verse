# Knowledge-neutral Verse snippet; project names and asset IDs generalized.
# id: snippet-044

# Copyright Epic Games, Inc. All Rights Reserved.

using { /Fortnite.com/Devices }
using { /UnrealEngine.com/Marketplace }
using { /Verse.org/Simulation }

using { Definitions.Entitlements }

improved_toast_example_device<public> := class(creative_device):

    @editable
    ToastMessageDevice : hud_message_device = hud_message_device{}

    @editable
    ActiveToastExample : button_device = button_device{}

    var ShowToast : logic = false

    OnBegin<override>()<suspends>:void=
        ActiveToastExample.InteractedWithEvent.Subscribe(OnButtonInteraction)

        for (Player : GetPlayspace().GetPlayers()):
            GetEntitlementsChangedEvent(Player, feature_example_entitlement).Subscribe(OnEntitlementsChanged)

    OnButtonInteraction(Agent:agent):void=
        set ShowToast = true
        if (Player := player[Agent]):
            ToastMessageDevice.Show(Player, StringToMessage("Toast Example - Enabled"))

    OnEntitlementsChanged(Result : tuple(player, []entitlement_change(entitlement))):void=
        for (EntitlementChange : Result(1), ShowToast?, Entitlement := feature_example_entitlement[EntitlementChange.Entitlement], EntitlementChange.Change > 0):
            ToastMessageDevice.Show(Result(0), StringToMessage("Received: {Localize(Entitlement.Name)} - {Localize(Entitlement.ShortDescription)}"))
