# Knowledge-neutral Verse snippet; project names and asset IDs generalized.
# id: snippet-045

using { /UnrealEngine.com/Marketplace }
using { /Verse.org/Simulation }
using { /Verse.org/SceneGraph }
using { /Verse.org/Random }
using {/Verse.org/Presentation}
using { /UnrealEngine.com/Itemization }

# Messages must be declared globally so they can be read for localization.
DefaultInteractionMessage<localizes>:message="Interact"
OwnerInteractionMessage<localizes>(Player:player, ItemName:message)<reads>:message =
    "{Player}'s {ItemName}"

# This interactable component requires a mesh component on this entity or a subentity.
# On successful interaction is adds an entity to the interacting player's inventory.
item_interactable_component := class(interactable_component) :

    var ParentEntity:?entity = false
    var ItemComponent:?item_component = false
    var ItemDescriptionComponent:?description_component = false
    var FullInteractionMessage:message = DefaultInteractionMessage

    OnBeginSimulation<override>():void=

        if(AncestorItemEntity := FindOwningItemEntity[]?):
            set ParentEntity = option{AncestorItemEntity}
            set ItemComponent = option{ParentEntity?.GetComponent[item_component]}
            set ItemDescriptionComponent = option{ParentEntity?.GetComponent[description_component]}
            SucceededEvent.Subscribe(OnSucceededEvent)

        SetFullInteractionMessage()

    # Check for a description_component and use the Name to create a custom message to display.
    # Fallback to DefaultInteractionMessage in absence of description_component or item_entitlement_component.
    SetFullInteractionMessage():void=
        if(Details := ItemDescriptionComponent?):
            if:
                EntitlementComp := Entity.GetComponent[item_entitlement_component]
            then:
                Owner := EntitlementComp.PlayerOwner
                ItemName := Details.Name
                set FullInteractionMessage = OwnerInteractionMessage(Owner,ItemName)
            else:
                set FullInteractionMessage = ItemName := Details.Name
        else:
            set FullInteractionMessage = DefaultInteractionMessage

    InteractMessage<override>(Agent:agent)<reads><decides>:message=
        FullInteractionMessage

    # If successful, add item via distribute to player inventories.
    OnSucceededEvent(Agent:agent):void=
        if:
            ResolveEntitlement[Agent]
            TargetInventory := GetAgentInventory[Agent]
            PickupItem := ItemComponent?.Entity
        then:
            TargetInventory.AddItemDistribute(PickupItem)

    # Check Owner before allowing pickup.
    ResolveEntitlement(Agent:agent)<decides><transacts>:void=
        if(Entity.GetComponent[item_entitlement_component].IsOwner[player[Agent]]):
