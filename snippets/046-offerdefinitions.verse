# Knowledge-neutral Verse snippet; project names and asset IDs generalized.
# id: snippet-046

# Copyright Epic Games, Inc. All Rights Reserved.

using { /Verse.org/Assets }
using { /UnrealEngine.com/Marketplace }

using { Art }

<#>
    Offer info such as Name, Description, Icon can differentiate from Entitlement info.
    Offer info is what the Player will see in the item shop.
OfferInfo<public> := module:
    PrimoShovel<public> := module:
        Name<public><localizes>: message = "Primo shovel"
        Description<public><localizes>: message = "An unbreakable shovel used to dig holes for planting."
        ShortDescription<public><localizes>: message = "Digs holes."
        Price<public> : float = 500.0

    CornSeedPacket<public> := module:
        Name<public><localizes> : message = "Corn seed pack"
        Description<public><localizes> : message = "A pack of corn seeds. Opening a pack yields 10 corn seeds for planting."
        ShortDescription<public><localizes> : message = "Contains 10 corn seeds for planting."
        Count<public> : int = 1
        Price<public> : float = 50.0

    CornSeedPacketBundle<public> := module:
        Name<public><localizes> : message = "Corn seed pack bundle"
        Description<public><localizes> : message = "Two packs of corn seeds. Opening a pack yields 10 corn seeds for planting."
        ShortDescription<public><localizes> : message = "Two packs of corn seeds containing 10 corn seeds for planting."
        Count<public> : int = 2
        Price<public> : float = 100.0

    CornSeedPacketBundleAlt<public> := module:
        Name<public><localizes> : message = "Half priced - Corn seed pack bundle"
        Description<public><localizes> : message = "Get two packs of corn seeds for the price of one! Opening a pack yields 10 corn seeds for planting."
        ShortDescription<public><localizes> : message = "Half price! Two packs of corn seeds containing 10 corn seeds for planting."
        Count<public> : int = 2
        Price<public> : float = 50.0

    MaxCornSeedPacketBundle<public> := module:
        Name<public><localizes> : message = "Max corn seed pack bundle"
        Description<public><localizes> : message = "Just enough packs of corn seeds to fully stock up. Each pack contains 10 seeds. Consume to get corn seed for planting."
        ShortDescription<public><localizes> : message = "Just enough corn seed packs to fill inventory. Corn seed packs contain 10 corn seeds for planting."

    StarterBundle<public> := module:
        Name<public><localizes> : message = "Starter bundle"
        Description<public><localizes> : message = "Everything a new player needs. Get fully stocked to start quickly! A shovel that digs holes, and ten packs of corn seeds each containing 10 corn seeds for planting."
        ShortDescription<public><localizes> : message = "A shovel that digs holes, and ten packs of corn seeds each containing 10 corn seeds for planting."
        Price<public> : float = 1350.0
        CornSeedPacksCount<public> : int = 10

    Gnome<public> := module:
        Name<public><localizes> : message = "Gnome"
        Description<public><localizes> : message = "A handsome garden gnome that adds character to your garden."
        ShortDescription<public><localizes> : message = "A decrative gnome."
        Price<public> : float = 100.0
        AlternativeIcon<public> : texture = Textures.T_UI_Gnome_Alternative

    VIPAccess<public> := module:
        Name<public><localizes> : message = "VIP access"
        Description<public><localizes> : message = "Grants access to a restricted area of the island. After purchasing, the VIP door will open automatically."
        ShortDescription<public><localizes> : message = "Grants access to a restricted area of the island."
        Price<public> : float = 2000.0

    MaxBerries<public> := module:
        Name<public><localizes>: message = "Flowberry Capacity"
        Description<public><localizes>: message = "Increases the number of Flowberries a player can hold in their inventory during a match. Purchasing an upgrade fills inventory.  Absolute maximum of six. Flowberry: Juice-filled berry that grants shields and low-grav to you and your allies."
        ShortDescription<public><localizes>: message = "Increase and fill max Flowberries."
        Price<public> : float = 50.0

<#>
    Here you can define offers available in your game.
    Currently there are two distinct types of offers:
        * entitlement offer - used to define offer for a single entitlement
            * examples:
                - one consumable corn seed
                - one durable shovel
        * bundle offer - used to define offer consisting of multiple entitlements
        * usual but not required - cheaper than buying entitlements separately
            * examples:
                - pack of 10 corn seeds
                - gardener starter pack: shovel and pack of 25 corn seeds
                
ExampleOffers<public> := module:
    # Example of entitlement offer
    # Notice: based on the offer you can't determine if the entitlement is consumable or not
    primo_shovel<public> := class(entitlement_offer):
        var Name<override> : message                = OfferInfo.PrimoShovel.Name
        var Description<override> : message         = OfferInfo.PrimoShovel.Description
        var ShortDescription<override> : message    = OfferInfo.PrimoShovel.ShortDescription
        var Icon<override> : texture                = Textures.T_Shovel
        # Reference entitlement type defined in Entitlements module
        EntitlementType<override> : concrete_subtype(entitlement) = Entitlements.primo_shovel
        # Set entitlement offer price
        Price<override> : price_dimension = MakePriceVBucks(OfferInfo.PrimoShovel.Price)

    # Example of entitlement offer
    # Notice: based on the offer you can't determine if an entitlement is consumable or not
    corn_seed_pack<public> := class(entitlement_offer):
        var Name<override> : message                = OfferInfo.CornSeedPacket.Name
        var Description<override> : message         = OfferInfo.CornSeedPacket.Description
        var ShortDescription<override> : message    = OfferInfo.CornSeedPacket.ShortDescription
        var Icon<override> : texture                = Textures.T_Corn_Seed_Packet
        # Reference entitlement type defined in Entitlements module
        EntitlementType<override> : concrete_subtype(entitlement) = Entitlements.corn_seed_pack
        # Set entitlement offer price
        Price<override> : price_dimension = MakePriceVBucks(OfferInfo.CornSeedPacket.Price)

    # Example of bundle offer
    corn_seed_pack_bundle<public> := class(bundle_offer):
        var Name<override> : message                = OfferInfo.CornSeedPacketBundle.Name
        var Description<override> : message         = OfferInfo.CornSeedPacketBundle.Description
        var ShortDescription<override> : message    = OfferInfo.CornSeedPacketBundle.ShortDescription
        var Icon<override> : texture                = Textures.T_Corn_Multi_Pack
        # Set entitlements and its quantity in bundle
        # You can create bundle of different entitlements
        Offers<override> : []tuple(offer, int) = array{(corn_seed_pack{}, OfferInfo.CornSeedPacketBundle.Count)}
        # Set bundle offer price
        Price<override> : price_dimension = MakePriceVBucks(OfferInfo.CornSeedPacketBundle.Price)

    corn_seed_pack_bundle_alt<public> := class(bundle_offer):
        var Name<override> : message                = OfferInfo.CornSeedPacketBundleAlt.Name
        var Description<override> : message         = OfferInfo.CornSeedPacketBundleAlt.Description
        var ShortDescription<override> : message    = OfferInfo.CornSeedPacketBundleAlt.ShortDescription
        var Icon<override> : texture                = Textures.T_Alt_Offer
        # Set entitlements and its quantity in bundle
        # You can create bundle of different entitlements
        Offers<override> : []tuple(offer, int) = array{(corn_seed_pack{}, OfferInfo.CornSeedPacketBundleAlt.Count)}
        # Set bundle offer price
        Price<override> : price_dimension = MakePriceVBucks(OfferInfo.CornSeedPacketBundleAlt.Price)

    max_corn_seed_pack_bundle<public> := class(bundle_offer):
        var Name<override> : message                = OfferInfo.MaxCornSeedPacketBundle.Name
        var Description<override> : message         = OfferInfo.MaxCornSeedPacketBundle.Description
        var ShortDescription<override> : message    = OfferInfo.MaxCornSeedPacketBundle.ShortDescription
        var Icon<override> : texture                = Textures.T_Max_Packets

    starter_bundle<public> := class(bundle_offer):
        var Name<override> : message                = OfferInfo.StarterBundle.Name
        var Description<override> : message         = OfferInfo.StarterBundle.Description
        var ShortDescription<override> : message    = OfferInfo.StarterBundle.ShortDescription
        var Icon<override> : texture                = Textures.T_Starter_Bundle
        # Set entitlements and its quantity in bundle
        # You can create bundle of different entitlements and add bundle withing bundle
        Offers<override> : []tuple(offer, int) = array{
            (corn_seed_pack{}, OfferInfo.StarterBundle.CornSeedPacksCount),
            (primo_shovel{}, 1)
        }
        # Set bundle offer price
        Price<override> : price_dimension = MakePriceVBucks(OfferInfo.StarterBundle.Price)

    gnome<public> := class(entitlement_offer):
        var Name<override> : message                = OfferInfo.Gnome.Name
        var Description<override> : message         = OfferInfo.Gnome.Description
        var ShortDescription<override> : message    = OfferInfo.Gnome.ShortDescription
        var Icon<override> : texture                = Textures.T_Gnome_Icon
        EntitlementType<override> : concrete_subtype(entitlement) = Entitlements.gnome
        Price<override> : price_dimension = MakePriceVBucks(OfferInfo.Gnome.Price)

    vip_access<public> := class(entitlement_offer):
        var Name<override> : message                = OfferInfo.VIPAccess.Name
        var Description<override> : message         = OfferInfo.VIPAccess.Description
        var ShortDescription<override> : message    = OfferInfo.VIPAccess.ShortDescription
        var Icon<override> : texture                = Textures.T_VIP_Access
        EntitlementType<override> : concrete_subtype(entitlement) = Entitlements.vip_access
        Price<override> : price_dimension = MakePriceVBucks(OfferInfo.VIPAccess.Price)

    max_berries<public> := class(entitlement_offer):
        var Name<override> : message                = OfferInfo.MaxBerries.Name
        var Description<override> : message         = OfferInfo.MaxBerries.Description
        var ShortDescription<override> : message    = OfferInfo.MaxBerries.ShortDescription
        var Icon<override> : texture                = Textures.T_Flowberry
        EntitlementType<override> : concrete_subtype(entitlement) = Entitlements.max_berries
        Price<override> : price_dimension = MakePriceVBucks(OfferInfo.MaxBerries.Price)
