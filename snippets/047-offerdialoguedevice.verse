# Knowledge-neutral Verse snippet; project names and asset IDs generalized.
# id: snippet-047

# Copyright Epic Games, Inc. All Rights Reserved.

using { /Fortnite.com/Devices }
using { /UnrealEngine.com/Marketplace }
using { /Verse.org/Simulation }

using { Definitions.ExampleOffers }

offer_dialogue_device<public> := class(creative_device):

    @editable
    Button : button_device = button_device{}

    OnBegin<override>()<suspends>:void=
        Button.InteractedWithEvent.Subscribe(OnButtonInteraction)

    OnButtonInteraction(Agent:agent):void=
        if (Player := player[Agent]):
            # Shows shop UI with list of selected offers
            spawn{ShowOffersDialog(Player, array{
                corn_seed_pack{},
                corn_seed_pack_bundle{},
                primo_shovel{},
                starter_bundle{}
            })}
