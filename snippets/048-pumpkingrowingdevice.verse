# Knowledge-neutral Verse snippet; project names and asset IDs generalized.
# id: snippet-048

# Copyright Epic Games, Inc. All Rights Reserved.

using { /Fortnite.com/Devices }
using { /Fortnite.com/Items }
using { /UnrealEngine.com/Marketplace }
using { /Verse.org/SceneGraph }
using { /Verse.org/Simulation }
using { /Verse.org/SpatialMath }
using { /UnrealEngine.com/Itemization }
using { /UnrealEngine.com/Temporary/SpatialMath }

PlantSeedDeviceMessage<localizes>:message = "Plant Pumpkin Seed"
WaterSeedDeviceMessage<localizes>:message = "Water Pumpkin Seed"

pumpkin_grow_stage := enum:
    WaitingForSeed
    WaitingForWater
    GrowingPumpkin

# This device binds to a button device and uses it to advance through the growing stages. 
# Each stage will check the activating player's inventory for specific item components.
# The result is a Fortnite Pumpkin that can be picked up but will not persist in player inventories if they leave the simulation.
pumpkin_growning_device := class(creative_device) :

    @editable
    Button:?button_device = false

    @editable
    var GrowthStage:pumpkin_grow_stage = pumpkin_grow_stage.WaitingForSeed 

    OnBegin<override>()<suspends>:void=
        if(CastButton := Button?):
            CastButton.InteractedWithEvent.Subscribe(OnInteractedWithEvent)
            CastButton.SetInteractionText(PlantSeedDeviceMessage)

    OnInteractedWithEvent(Agent:agent):void=
        spawn{AdvanceGrowthStage(Agent,GrowthStage)}
            
    AdvanceGrowthStage(Agent:agent, NewStage:pumpkin_grow_stage)<suspends>:void=
        case(NewStage):

            # In this stage we check the player inventory for any items with the pumpkin_seed_item_component.
            # The pumpkin seed calls a function that consumes the associated entitlement.
            pumpkin_grow_stage.WaitingForSeed => 
                {
                    if:
                        TargetInventory := GetAgentInventory[Agent]
                        PumpkinSeeds := TargetInventory.FindItems(pumpkin_seed_item_component)
                        SeedItemComponent := PumpkinSeeds[0].GetComponent[pumpkin_seed_item_component]
                    then:
                        SeedItemComponent.GrowPumpkin()
                        set GrowthStage = pumpkin_grow_stage.WaitingForWater
                        if(CastButton := Button?):
                            CastButton.SetInteractionText(WaterSeedDeviceMessage)
                }

            # In this stage we check the player inventory for an item with the watering_can_item_component.
            pumpkin_grow_stage.WaitingForWater => 
                {
                    if:
                        TargetInventory := GetAgentInventory[Agent]
                        WateringCans := TargetInventory.FindItems(watering_can_item_component)
                        WateringCans.Length > 0
                    then:
                        set GrowthStage = pumpkin_grow_stage.GrowingPumpkin
                        AdvanceGrowthStage(Agent,GrowthStage)
                }

            # This stage begins automatically following the end of the previous.
            # The button is disabled and a new Fortnite item is spawned. 
            # Once the item is picked up, the button is reenabled and the cycle can restart.
            pumpkin_grow_stage.GrowingPumpkin =>
                {
                    if:
                        CastButton := Button?
                        SimEntity := GetSimulationEntity[]
                    then:
                        CastButton.Disable()

                        # Delay to simulate growing pumpkin.
                        Sleep(2.0)

                        # Spawn a new Fortnite item.
                        new_pumpkin_type:concrete_subtype(entity) = Pumpkin_Creative_V1_Common
                        SpawnedPumpkin := new_pumpkin_type{}
        
                        SpawnTransform := (/Verse.org/SpatialMath:)transform
                            {
                                Translation := FromTransform(Self.GetTransform()).Translation,
                                Rotation := FromTransform(Self.GetTransform()).Rotation,
                                Scale := SpawnedPumpkin.GetGlobalTransform().Scale
                            }
                        SpawnedPumpkin.SetGlobalTransform(SpawnTransform)
                        SimEntity.AddEntities(array{SpawnedPumpkin})
                        
                        # Await the pick up of the Fortnite item before allowing a new cycle to begin.
                        if:
                            PumpkinItemComp := SpawnedPumpkin.GetComponent[item_component]
                            PumpkinStackableComp := SpawnedPumpkin.GetComponent[stackable_component]
                        then:
                            race:
                                block:
                                    PumpkinItemComp.ChangeInventoryEvent.Await()
                                block:
                                    PumpkinStackableComp.ChangeStackSizeEvent.Await()

                        set GrowthStage = pumpkin_grow_stage.WaitingForSeed
                        CastButton.SetInteractionText(PlantSeedDeviceMessage)
                        CastButton.Enable()
                }
