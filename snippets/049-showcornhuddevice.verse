# Knowledge-neutral Verse snippet; project names and asset IDs generalized.
# id: snippet-049

# Copyright Epic Games, Inc. All Rights Reserved.

using { /Fortnite.com/Devices }
using { /UnrealEngine.com/Marketplace }
using { /Verse.org/Simulation }

using { Definitions }

show_corn_hud_device := class(creative_device):
    @editable
    PlayerFeedbackButton : button_device = button_device{}

    @editable
    CornSeedMessageDevice : hud_message_device = hud_message_device{}

    @editable
    ShovelMessageDevice : hud_message_device = hud_message_device{}

    @editable
    SeedPacketButton : button_device = button_device{}

    @editable
    SeedPacketMessageDevice : hud_message_device = hud_message_device{}

    # seed is 0, pack is 1, shovel is 2
    var PlayerMap : [player]tuple(int, int, int) = map{}

    var ShowFirstPlayerFeedbackUIActive : logic = false
    var ShowSeedPacketUIActive : logic = false

    OnBegin<override>()<suspends>:void=
        PlayerFeedbackButton.InteractedWithEvent.Subscribe(OnCornButtonInteraction)
        SeedPacketButton.InteractedWithEvent.Subscribe(OnSeedButtonInteraction)
        
        for (Player : GetPlayspace().GetPlayers()):
            if (set PlayerMap[Player] = (0,0,0)):

            GetEntitlementsChangedEvent(Player, Entitlements.feature_example_entitlement).Subscribe(OnEntitlementsChanged)
            spawn{ValidatePreviousPurchases(Player)}
            
    OnCornButtonInteraction(Agent:agent):void=
        set ShowFirstPlayerFeedbackUIActive = true
        if (Player := player[Agent]):
            UpdateUI(Player)

    OnSeedButtonInteraction(Agent:agent):void=
        set ShowSeedPacketUIActive = true
        if (Player := player[Agent]):
            UpdateUI(Player)

    OnEntitlementsChanged(Result : tuple(player, []entitlement_change(entitlement))):void=
        ChangeOccured := for (EntitlementChange : Result(1)):
            if:
                Entitlements.corn_seed[EntitlementChange.Entitlement]
                NewValue := (PlayerMap[Result(0)](0) + EntitlementChange.Change, PlayerMap[Result(0)](1), PlayerMap[Result(0)](2))
                set PlayerMap[Result(0)] = NewValue
            if:
                Entitlements.corn_seed_pack[EntitlementChange.Entitlement]
                NewValue := (PlayerMap[Result(0)](0), PlayerMap[Result(0)](1) + EntitlementChange.Change, PlayerMap[Result(0)](2))
                set PlayerMap[Result(0)] = NewValue

            if:
                Entitlements.primo_shovel[EntitlementChange.Entitlement]
                NewValue := (PlayerMap[Result(0)](0), PlayerMap[Result(0)](1), PlayerMap[Result(0)](2) + EntitlementChange.Change)
                set PlayerMap[Result(0)] = NewValue

        if (ChangeOccured.Length > 0):
            UpdateUI(Result(0))


    UpdateUI(Player : player):void=
        if (Value := PlayerMap[Player]):
            if (ShowFirstPlayerFeedbackUIActive?):
                CornSeedMessageDevice.Show(Player, StringToMessage("{Value(0)}"))
                if (Value(2) > 0):
                    ShovelMessageDevice.Show(Player, StringToMessage("Owned:"))
                else:
                    ShovelMessageDevice.Show(Player, StringToMessage("Unowned:"))
            if (ShowSeedPacketUIActive?):
                SeedPacketMessageDevice.Show(Player, StringToMessage("{Value(1)}"))

    ValidatePreviousPurchases(Player:player)<suspends>:void=
        Purchases := GetPurchasedEntitlements(Player, Definitions.Entitlements.feature_example_entitlement)

        for (Purchase : Purchases):
            if:
                Entitlements.corn_seed[Purchase(0)]
                NewValue := (Purchase(1), PlayerMap[Player](1), PlayerMap[Player](2))
                set PlayerMap[Player] = NewValue
            if:
                Entitlements.corn_seed_pack[Purchase(0)]
                NewValue := (PlayerMap[Player](0), Purchase(1), PlayerMap[Player](2))
                set PlayerMap[Player] = NewValue

            if:
                Entitlements.primo_shovel[Purchase(0)]
                NewValue := (PlayerMap[Player](0), PlayerMap[Player](1), Purchase(1))
                set PlayerMap[Player] = NewValue
