# Knowledge-neutral Verse snippet; project names and asset IDs generalized.
# id: snippet-050

using { /Fortnite.com/Devices }
using { /Verse.org/Simulation }
using { /Verse.org/Simulation/Tags }
using { /UnrealEngine.com/Temporary/Diagnostics }

# ============================================================================================================================================
# Grants a stud whenever a collectible is picked up (using gameplay tags)
# ============================================================================================================================================

BRICK_collectible_stud_tag<public> := class(BRICK_stud_system_tag) {}

BRICK_collectible_stud_granter := class(creative_device):
    @editable:
        ToolTip := ToolTip_ItemGranter
    ItemGranter : item_granter_device = item_granter_device{}

    # Subscribe to all collectibles with the tag Tag_BRICK_collectible_stud
    OnBegin<override>()<suspends> : void =
        for (FoundObject : Self.FindCreativeObjectsWithTag(BRICK_collectible_stud_tag{})):
            if (ValidCollectible := collectible_object_device[FoundObject]):
                ValidCollectible.CollectedEvent.Subscribe(OnCollectibleCollected)

    # Grant the item when the collectible is picked up
    OnCollectibleCollected(InAgent : agent) : void =
        ItemGranter.GrantItem(InAgent)

ToolTip_ItemGranter<public><localizes> : message = "The item_granter that will grant the stud."
