# Knowledge-neutral Verse snippet; project names and asset IDs generalized.
# id: snippet-053

using { /Fortnite.com/Devices }
using { /Fortnite.com/Characters }
using { /Fortnite.com/Playspaces }
using { /Verse.org/Simulation }
using { /Verse.org/Assets }
using { /Verse.org/Simulation/Tags }
using { /Verse.org/Random }
using { /UnrealEngine.com/Temporary/Diagnostics }
using { /UnrealEngine.com/Temporary/SpatialMath }


fortplayer_log<public> := class(log_channel):
# ============================================================================================================================================
# Provides data and behavior on a per-player basis. Maintained by BRICK_fortplayer_manager.
# ============================================================================================================================================
new_BRICK_fortplayer<constructor><public>(InFortPlayer : ?player, InFortChar : ?fort_character) := BRICK_fortplayer:
    FortPlayer := InFortPlayer
    FortChar := InFortChar

BRICK_fortplayer<public> := class<unique>:
    var UUID<public> : int = 0 # Valid ID after Agent leaves the session
    var FortPlayer<public> : ?player = false
    var FortChar<public> : ?fort_character = false

    # Tick behavior
    var StartWithTickEnabled<protected> : logic = false
    var TickRate<protected> : float = 1.0 # In Seconds

    # Use BRICK_fortplayer_team_assigner or your own implementation
    var Team <protected> : ?int = false

    GetTeam<public>()<transacts> : int = 
        if (ValidTeam := Team?):
            return ValidTeam
        else:
            return -1

    SetTeam<public>(InTeam : int) : void = 
        set Team = option{InTeam}

    Logger<protected>: log = log{Channel := fortplayer_log, DefaultLevel := log_level.Normal}


    # --------------------------------------------------------------------------------------------------------------------------------------------
    # Default Methods
    # --------------------------------------------------------------------------------------------------------------------------------------------
    OnBegin<protected>()<suspends> : void =
        return

    OnEnd<protected>() : void =
        return

    Tick<protected>()<suspends> : void =
        loop:
            Logger.Print("Tick")
            Sleep(TickRate)

    # --------------------------------------------------------------------------------------------------------------------------------------------
    # Persistence Data
    # --------------------------------------------------------------------------------------------------------------------------------------------
    PersistentDataUpdated<public>(InData : BRICK_fortplayer_persistence_data)<transacts> : void =
        return

    # --------------------------------------------------------------------------------------------------------------------------------------------
    # BRICK_fortplayer_manager Methods
    # --------------------------------------------------------------------------------------------------------------------------------------------
    Init_BRICKFortplayer<public>()<suspends> : void =
        OnBegin()
        if (StartWithTickEnabled?):
            race:
                OnUninitBRICKFortplayer.Await()
                Tick()
        Logger.Print("Initialized BRICK_fortplayer: {UUID}")

    OnUninitBRICKFortplayer<public> : event() = event(){}
    UnInit_BRICKFortplayer<public>()<suspends> : void =
        Logger.Print("Uninitialized BRICK_fortplayer: {UUID}")
        OnUninitBRICKFortplayer.Signal()
        OnEnd()
