# Knowledge-neutral Verse snippet; project names and asset IDs generalized.
# id: snippet-054

using { /Fortnite.com/Devices }
using { /Fortnite.com/Characters }
using { /Fortnite.com/Playspaces }
using { /Verse.org/Simulation }
using { /Verse.org/Assets }
using { /Verse.org/Simulation/Tags }
using { /Verse.org/Random }
using { /UnrealEngine.com/Temporary/Diagnostics }
using { /UnrealEngine.com/Temporary/SpatialMath }

using { BRICKUtilities }

# Constructor for creating a new BRICK_fortplayer
new_BRICK_fortplayer_extraction<constructor><public>(InFortPlayer : ?player, InFortChar : ?fort_character, InUUID : int) := BRICK_fortplayer_extraction:
    FortPlayer := InFortPlayer
    FortChar := InFortChar
    UUID := InUUID

BRICK_fortplayer_extraction<public> := class(BRICK_fortplayer):
    
    # Used to track if the player is in the lobby and update its corresponding HUD in the gold studs manager
    var InLobby<public> : logic = true

    # Invoked from the fort player manager after a player has joined the game and has been registered as a BRICK_fortplayer
    OnBegin<override>()<suspends> : void =
        (super:)OnBegin()
        Sleep(1.0)

    # Invoked by the BRICK_fortplayer_manager when a player has left the game and has been removed as a BRICK_fortplayer
    # Note: the agent reference might not be valid any more at this point
    OnEnd<override>() : void =
        (super:)OnEnd()

    # Utility function to retrieve the agent associated with this BRICK_fortplayer
    GetAgent<public>()<transacts> : ?agent =
        if (Agent := FortChar?.GetAgent[]):
            option{Agent}
        else:
            false

    # Utility function to retrieve the transform associated with this BRICK_fortplayer
    GetTransform<public>()<transacts> : transform =
        if (Transform := FortChar?.GetTransform()):
            Transform
        else:
            Logger.Print("GetTransform() on BRICK_fortplayer_extraction failed.", ?Level := log_level.Warning)
            transform{}
