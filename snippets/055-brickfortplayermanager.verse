# Knowledge-neutral Verse snippet; project names and asset IDs generalized.
# id: snippet-055

using { /Fortnite.com/Devices }
using { /Fortnite.com/Characters }
using { /Verse.org/Simulation }
using { /Verse.org/Simulation/Tags }
using { /Verse.org/Random }
using { /UnrealEngine.com/Temporary/Diagnostics }
using { /UnrealEngine.com/Temporary/SpatialMath }

# ============================================================================================================================================
# Maintains a BRICK_fortplayer instance for each FortPlayer, meant to provide centralized access to per-player data and behavior.
#                           --- MUST ASSIGN Tag: fortplayer_manager_tag to the device instance in the level! ---
# ============================================================================================================================================

# Return the assigned BRICK_fortplayer from any agent, if it exists. 
(InAgent : agent).GetBRICKFortPlayer<public>(InDevice : creative_device)<decides><transacts> : BRICK_fortplayer =
    var OutBRICKPlayer : ?BRICK_fortplayer = false    
    if:
        FoundBRICKPlayer := GetBRICKFortPlayerManager[InDevice].GetBRICKFortPlayerFromAgent[InAgent]
        set OutBRICKPlayer = option{FoundBRICKPlayer}
    else:
        Logger: log = log{Channel := fortplayer_manager_log, DefaultLevel := log_level.Normal}
        Logger.Print("GetBRICKFortPlayer failed.", ?Level := log_level.Warning)
    
    OutBRICKPlayer?

# Return all BRICK_fortplayers of a specific team
GetBRICKFortPlayersByTeam<public>(InDevice : creative_device, InTeam : int)<transacts> : []BRICK_fortplayer =
    var OutBRICKPlayers : []BRICK_fortplayer = array{}

    if (FortPlayerManager := GetBRICKFortPlayerManager[InDevice]):
        for (Player : FortPlayerManager.GetAllBRICKFortPlayers(), Player.GetTeam() = InTeam):
            set OutBRICKPlayers += array{Player}

    OutBRICKPlayers

# Get the FortPlayerManager instance from the level using its MarkupTag - if it exists.
GetBRICKFortPlayerManager<public>(InDevice : creative_device)<decides><transacts> : BRICK_fortplayer_manager =
    var OutObject : ?BRICK_fortplayer_manager = false

    for (FoundObject : InDevice.FindCreativeObjectsWithTag(fortplayer_manager_tag{})):
        if (ValidTypedObject := BRICK_fortplayer_manager[FoundObject]):
            set OutObject = option{ValidTypedObject}

    if (not OutObject?):
        Logger: log = log{Channel := fortplayer_manager_log, DefaultLevel := log_level.Normal}
        Logger.Print("BRICK_fortplayer_manager not found! Make sure it is in the level and has the fortplayer_manager_tag applied.", ?Level := log_level.Warning)

    OutObject?


fortplayer_manager_log<public> := class(BRICK_utilities_log):
fortplayer_manager_tag<public> := class(BRICK_utilities_tag){} # --- MUST ASSIGN Tag: fortplayer_manager_tag to the device instance in the level! ---


# ============================================================================================================================================
# Device
# ============================================================================================================================================
BRICK_fortplayer_manager_fndevice_interface<public> := class<concrete>:
    @editable
    Trigger_OnBRICKFortPlayerAdded : ?trigger_device = false
    @editable
    Trigger_ClearPersistentDataForPlayer : ?trigger_device = false


BRICK_fortplayer_manager<public> := class(creative_device):
    @editable
    DelayBeforeAddingPlayers<protected> : float = 2.0 # NOTE: Delaying the Initialization of the BRICK player here as OnBegin is inconsistent across devices.
    @editable
    FNEventBindingInterface : BRICK_fortplayer_manager_fndevice_interface = BRICK_fortplayer_manager_fndevice_interface{}

    var ActiveBRICKPlayers<protected> : []BRICK_fortplayer = array{}
    Logger<protected>: log = log{Channel := fortplayer_manager_log, DefaultLevel := log_level.Normal}

    OnBegin<override>()<suspends> : void =
        Logger.Print("--------------------------------------------------------------------")
        Logger.Print("BRICK_fortplayer_manager initialization started")
        Logger.Print("--------------------------------------------------------------------")
        GetPlayspace().PlayerRemovedEvent().Subscribe(RemovePlayerAsBRICKFortPlayer)

        if (ValidTrigger := FNEventBindingInterface.Trigger_ClearPersistentDataForPlayer?):
            ValidTrigger.TriggeredEvent.Subscribe(ClearPersistentDataForPlayer)

        Sleep(DelayBeforeAddingPlayers)
        AddInitialPlayersAsBRICKFortPlayers()
        GetPlayspace().PlayerAddedEvent().Subscribe(AddPlayerAsBRICKFortPlayer)

        Sleep(0.5) # This is to ensure OnBegin() is finished before signaling OnInitializedPlayers
        OnInitializedPlayers.Signal()
        Logger.Print("--------------------------------------------------------------------")
        Logger.Print("BRICK_fortplayer_manager initialization finished")
        Logger.Print("--------------------------------------------------------------------")

        Sleep(1.0)
        FoundDevices := FindCreativeObjectsWithTag(fortplayer_manager_tag{})


    # --------------------------------------------------------------------------------------------------------------------------------------------
    # Helpers
    # --------------------------------------------------------------------------------------------------------------------------------------------
    GetBRICKFortPlayerFromAgent<public>(InAgent : agent)<decides><transacts> : BRICK_fortplayer =
        var OutPlayer : ?BRICK_fortplayer = false
        if (InPlayer := player[InAgent]):
            for (Index -> BRICKPlayer: ActiveBRICKPlayers):
                if:
                    Player := BRICKPlayer.FortPlayer?
                    Player = InPlayer
                then:
                    set OutPlayer = option{BRICKPlayer}
        else:
            Logger.Print("GetBRICKFortPlayerFromAgent failed: InAgent is not a player.", ?Level := log_level.Warning)

        OutPlayer?

    GetAllBRICKFortPlayers<public>()<transacts> : []BRICK_fortplayer =
        ActiveBRICKPlayers

    GetPlayerTeam(InAgent : agent)<transacts> : int =
        TeamCollection := GetPlayspace().GetTeamCollection()
        if (AgentsTeam := TeamCollection.GetTeam[InAgent]):
            TeamArray := TeamCollection.GetTeams()
            for (TeamIndex -> Team : TeamArray, AgentsTeam = Team):
                Logger.Print("GetPlayerTeam - Team {TeamIndex + 1}.")
                TeamIndex + 1

        Logger.Print("GetPlayerTeam - Failed to find team for new player.", ?Level := log_level.Warning)
        return -1

    ClearPersistentDataForPlayer<public>(InAgent : ?agent) : void =
        if:
            ValidAgent := InAgent?
            ValidAgent.SetBRICKFortPlayerPersistenceData[Self, BRICK_fortplayer_persistence_data{}]
        then:
            Logger.Print("Cleared persistent data for player.", ?Level := log_level.Warning)
        else:
            Logger.Print("ClearPersistentDataForPlayer failed - No BRICKFortPlayer found for agent.", ?Level := log_level.Error)

    # --------------------------------------------------------------------------------------------------------------------------------------------
    # Internal
    # --------------------------------------------------------------------------------------------------------------------------------------------
    OnInitializedPlayers<public>: event() = event(){}
    AddInitialPlayersAsBRICKFortPlayers<private>()<suspends> : void =
        AllPlayers : []player = GetPlayspace().GetPlayers()
        for (FoundPlayer : AllPlayers):
            AddPlayerAsBRICKFortPlayer(FoundPlayer)
            Sleep(0.5)

    OnNewBRICKFortplayerAdded<public>: event(BRICK_fortplayer) = event(BRICK_fortplayer){}
    AddPlayerAsBRICKFortPlayer<private>(InPlayer : player) : void =
        Logger.Print("--------------------------------------------------------------------")
        Logger.Print("Player joined: creating up BRICKFortplayer.")
        if (GetBRICKFortPlayerFromAgent[InPlayer]):
            return

        # UUID to identify the player after the Agent has left the session
        var UUID : int = GetRandomInt(-2147483648, 2147483647)

        # Ensure player UUID stays consistent throughout playsessions
        if (PersistentData := GlobalPersistentData_BRICKFortPlayerData[InPlayer]):
            set UUID = PersistentData.UUID
            Logger.Print("Found existing entry for BRICKFortplayer {UUID}.")
        else:
            if (set GlobalPersistentData_BRICKFortPlayerData[InPlayer] = BRICK_fortplayer_persistence_data{UUID := UUID}):
                Logger.Print("Created new entry for BRICKFortplayer {UUID}.")
            else:
                Logger.Print("Failed to create new entry for BRICKFortplayer {UUID}.", ?Level := log_level.Error)

        NewBRICKFortPlayer := MakeBRICKFortPlayer(InPlayer, UUID)
        set ActiveBRICKPlayers += array{NewBRICKFortPlayer}

        spawn{NewBRICKFortPlayer.Init_BRICKFortplayer()}
        OnNewBRICKFortplayerAdded.Signal(NewBRICKFortPlayer)
        if (ValidTrigger := FNEventBindingInterface.Trigger_OnBRICKFortPlayerAdded?):
            ValidTrigger.Trigger(InPlayer)

        Logger.Print("Player joined: Added BRICKFortplayer {UUID}.")
        Logger.Print("--------------------------------------------------------------------")

    RemovePlayerAsBRICKFortPlayer<protected>(InPlayer : player) : void =
        if (LeavingBRICKPlayer := GetBRICKFortPlayerFromAgent[InPlayer]):
            ActiveBRICKPlayers.RemoveAllElements(LeavingBRICKPlayer)

            for (ActiveBRICKPlayer : ActiveBRICKPlayers):
                if (LeavingBRICKPlayer = ActiveBRICKPlayer):
                    if (ActiveBRICKPlayers.RemoveFirstElement[ActiveBRICKPlayer]){}
                    spawn{ActiveBRICKPlayer.UnInit_BRICKFortplayer()}
                    Logger.Print("Player left: Removed BRICKFortplayer {LeavingBRICKPlayer.UUID}.")
                    return

    # Override this in a child class to use a custom BRICK_fortplayer class.
    MakeBRICKFortPlayer<protected>(InPlayer : player, InUUID : int) : BRICK_fortplayer = 
        if (ValidFortChar := InPlayer.GetFortCharacter[]):
            return BRICK_fortplayer {FortPlayer := option{InPlayer}, FortChar := option{ValidFortChar}, UUID := InUUID}
        else:
            Logger.Print("MakeBRICKFortPlayer failed - No FortCharacter found for player.", ?Level := log_level.Error)
            return BRICK_fortplayer {FortPlayer := option{InPlayer}, FortChar := false}
