# Knowledge-neutral Verse snippet; project names and asset IDs generalized.
# id: snippet-056

using { /Fortnite.com/Characters }
using { /Fortnite.com/Devices }
using { /Fortnite.com/Playspaces }
using { /Verse.org/Simulation }
using { /UnrealEngine.com/Temporary/Diagnostics }

using { BRICKUtilities }

(InAgent : agent).GetExtractionPlayer<public>(InDevice : creative_device)<decides><transacts> : BRICK_fortplayer_extraction =
    ExtractionPlayer := BRICK_fortplayer_extraction[InAgent.GetBRICKFortPlayer[InDevice]]    

# A Verse-authored creative device that can be placed in a level
BRICK_fortplayer_manager_extraction<public> := class(BRICK_fortplayer_manager):    
    OnBegin<override>()<suspends> : void =
        (super:)OnBegin()

    MakeBRICKFortPlayer<override>(InPlayer : player, InUUID : int) : BRICK_fortplayer_extraction = 
        if (ValidFortChar := InPlayer.GetFortCharacter[]):
            BRICK_fortplayer_extraction {FortPlayer := option{InPlayer}, FortChar := option{ValidFortChar}, UUID := InUUID}
        else:
            Logger.Print("MakeBRICKFortPlayer failed - No FortCharacter found for player.", ?Level := log_level.Error)
            BRICK_fortplayer_extraction {FortPlayer := option{InPlayer}, FortChar := false}
