# Knowledge-neutral Verse snippet; project names and asset IDs generalized.
# id: snippet-057

using { /UnrealEngine.com/Temporary/Diagnostics }
using { /Verse.org/Simulation }
using { /Fortnite.com/Devices }

# Return the persistent data stored for this player. If it does not exist, create a new instance and return it.
(InAgent : agent).GetBRICKFortPlayerPersistenceData<public>(InDevice : creative_device)<decides><transacts> : BRICK_fortplayer_persistence_data =
    var OutData : ?BRICK_fortplayer_persistence_data = false
    if:
        BRICKPlayer := InAgent.GetBRICKFortPlayer[InDevice]
        Player := BRICKPlayer.FortPlayer?
    then:
        if (FoundData := GlobalPersistentData_BRICKFortPlayerData[Player]):
            set OutData = option{FoundData}
        else:
            NewData := BRICK_fortplayer_persistence_data{}
            if (set GlobalPersistentData_BRICKFortPlayerData[Player] = NewData)
            {}
            set OutData = option{NewData}
        
    if (not OutData?):
        Logger: log = log{Channel := fortplayer_manager_log, DefaultLevel := log_level.Normal}
        Logger.Print("Failed to get BRICK_fortplayer_persistence_data.", ?Level := log_level.Error)

    OutData?

# Set the persistent data stored for this player.
(InAgent : agent).SetBRICKFortPlayerPersistenceData<public>(InDevice : creative_device, InPlayerData : BRICK_fortplayer_persistence_data)<decides><transacts> : void =
    Logger: log = log{Channel := fortplayer_manager_log, DefaultLevel := log_level.Normal}
    var Success : logic = false
    if:
        BRICKPlayer := InAgent.GetBRICKFortPlayer[InDevice]
        Player := BRICKPlayer.FortPlayer?
    then:
        if (FitsInPlayerMap[InPlayerData]):
            if (set GlobalPersistentData_BRICKFortPlayerData[Player] = InPlayerData)
            {}
            BRICKPlayer.PersistentDataUpdated(InPlayerData)
            set Success = true
        else:
            Logger.Print("Failed to set BRICK_fortplayer_persistence_data: Data exceeds size limit!", ?Level := log_level.Error)
    else:
        Logger.Print("Failed to save BRICK_fortplayer_persistence_data.", ?Level := log_level.Error)

    Success?

# ============================================================================================================================================
# Persistent Data
# ============================================================================================================================================
var GlobalPersistentData_BRICKFortPlayerData : weak_map(player, BRICK_fortplayer_persistence_data) = map{}
BRICK_fortplayer_persistence_data <public> := class<persistable><final>():
    UUID : int = 0
