# Knowledge-neutral Verse snippet; project names and asset IDs generalized.
# id: snippet-058

using { /Fortnite.com/Devices }
using { /Verse.org/Simulation }
using { /UnrealEngine.com/Temporary/Diagnostics }

using { Lobby }
using { BRICKUtilities }

# Manager class that handles gold studs for players in the game
BRICK_goldstuds_manager := class(creative_device):
    @editable:
        ToolTip := ToolTip_StudManager_Stat_GoldStuds
        Categories := array{StudManager_StatCategory}
    Stat_GoldStuds : stat_creator_device = stat_creator_device{}

    @editable:
        ToolTip := ToolTip_StudManager_Stat_BankedGoldStuds
        Categories := array{StudManager_StatCategory}
    Stat_BankedGoldStuds : stat_creator_device = stat_creator_device{}
    
    @editable:
        ToolTip := ToolTip_StudManager_Stat_BankedGoldStuds_VisualsOnly
        Categories := array{StudManager_StatCategory}
    Stat_BankedGoldStuds_VisualsOnly : stat_creator_device = stat_creator_device{}

    @editable:
        ToolTip := ToolTip_StudManager_ConditionalButton_GoldStuds_CheckCount
        Categories := array{StudManager_GeneralDevicesCategory}
    ConditionalButton_GoldStuds_CheckCount : conditional_button_device = conditional_button_device{}

    @editable:
        ToolTip := ToolTip_StudManager_ItemRemover_DepositGoldStuds
        Categories := array{StudManager_GeneralDevicesCategory}
    ItemRemover_DepositGoldStuds : item_remover_device = item_remover_device{}

    @editable:
        ToolTip := ToolTip_StudManager_ItemRemover_RemoveAll
        Categories := array{StudManager_GeneralDevicesCategory}
    ItemRemover_RemoveAll : item_remover_device = item_remover_device{}

    @editable:
        ToolTip := ToolTip_StudManager_CinematicSequence_DepositGoldStuds
        Categories := array{StudManager_GeneralDevicesCategory}
    CinematicSequence_DepositGoldStuds : cinematic_sequence_device = cinematic_sequence_device{}

    @editable:
        ToolTip := ToolTip_StudManager_HUDMessage_DepositeGoldStuds
        Categories := array{StudManager_GeneralDevicesCategory}
    HUDMessage_DepositeGoldStuds : hud_message_device = hud_message_device{}

    @editable:
        ToolTip := ToolTip_StudManager_AudioPlayer_DepositGoldStuds
        Categories := array{StudManager_GeneralDevicesCategory}
    AudioPlayer_DepositGoldStuds : audio_player_device = audio_player_device{}

    @editable:
        ToolTip := ToolTip_StudManager_ItemGranter_LobbyStuds_X1
        Categories := array{StudManager_GeneralDevicesCategory}
    ItemGranter_LobbyStuds_X1 : item_granter_device = item_granter_device{}

    @editable:
        ToolTip := ToolTip_StudManager_ItemGranter_LobbyStuds_X10
        Categories := array{StudManager_GeneralDevicesCategory}
    ItemGranter_LobbyStuds_X10 : item_granter_device = item_granter_device{}

    @editable:
        ToolTip := ToolTip_StudManager_ItemGranter_LobbyStuds_X100
        Categories := array{StudManager_GeneralDevicesCategory}
    ItemGranter_LobbyStuds_X100 : item_granter_device = item_granter_device{}

    @editable:
        ToolTip := ToolTip_StudManager_EliminationManager
        Categories := array{StudManager_GeneralDevicesCategory}
    EliminationManager : elimination_manager_device = elimination_manager_device{}

    @editable:
        ToolTip := ToolTip_StudManager_OnGoldStudsChanged
        Categories := array{StudManager_TriggerEventsCategory}
    Trigger_OnGoldStudsChanged : trigger_device = trigger_device{}

    @editable:
        ToolTip := ToolTip_StudManager_DepositGoldStuds
        Categories := array{StudManager_TriggerEventsCategory}
    Trigger_DepositGoldStuds : trigger_device = trigger_device{}

    @editable:
        ToolTip := ToolTip_StudManager_DepositLobbyStuds
        Categories := array{StudManager_TriggerEventsCategory}
    Trigger_DepositLobbyStuds : trigger_device = trigger_device{}

    @editable:
        ToolTip := ToolTip_StudManager_OnLobbyEnter
        Categories := array{StudManager_TriggerEventsCategory}
    Trigger_OnLobbyEnter : trigger_device = trigger_device{}

    @editable:
        ToolTip := ToolTip_StudManager_OnLobbyExit
        Categories := array{StudManager_TriggerEventsCategory}
    Trigger_OnLobbyExit : trigger_device = trigger_device{}

    @editable:
        ToolTip := ToolTip_StudManager_OnBRICKFortPlayerAdded
        Categories := array{StudManager_TriggerEventsCategory}
    Trigger_OnBRICKFortPlayerAdded : trigger_device = trigger_device{}

    OnBegin<override>()<suspends> : void =
        # Subscribe to the relevant triggers when the game starts
        Trigger_OnGoldStudsChanged.TriggeredEvent.Subscribe(UpdateStuds)
        Trigger_DepositGoldStuds.TriggeredEvent.Subscribe(DepositStuds)
        Trigger_DepositLobbyStuds.TriggeredEvent.Subscribe(DepositLobbyStuds)
        Trigger_OnLobbyEnter.TriggeredEvent.Subscribe(OnLobbyEnter)
        Trigger_OnLobbyExit.TriggeredEvent.Subscribe(OnLobbyExit)
        Trigger_OnBRICKFortPlayerAdded.TriggeredEvent.Subscribe(OnPlayerJoined)

        # Make sure to clear the player's inventory when they leave the game
        GetPlayspace().PlayerRemovedEvent().Subscribe(ClearAllInventory)

    # Set the player's InLobby state to TRUE to determine which HUD stat elements to show
    OnLobbyEnter(InAgent : ?agent) : void = 
        if:
            ValidAgent := InAgent?
            ValidExtractionPlayer := ValidAgent.GetExtractionPlayer[Self]
        then:
            set ValidExtractionPlayer.InLobby = true
            spawn{QueueUpdateStuds_OnLobbyEntered(InAgent)}

    QueueUpdateStuds_OnLobbyEntered(InAgent : ?agent)<suspends> : void =
        # Safety measure to ensure studs are correctly updated when entering the lobby
        Sleep(4.0)
        UpdateStuds(InAgent)

    # Set the player's InLobby state to FALSE to determine which HUD stat elements to show
    OnLobbyExit(InAgent : ?agent) : void = 
        if:
            ValidAgent := InAgent?
            ValidExtractionPlayer := ValidAgent.GetExtractionPlayer[Self]
        then:
            set ValidExtractionPlayer.InLobby = false

    # Update the relevant studs stats and their associated HUD elements when studs are picked up or spent
    UpdateStuds(InAgent : ?agent) : void = 
        if:
            ValidAgent := InAgent?
            ValidExtractionPlayer := ValidAgent.GetExtractionPlayer[Self]
        then:
            ItemAmount : int = ConditionalButton_GoldStuds_CheckCount.GetItemCount(ValidAgent, 0)

            # In the lobby: Hide the current studs HUD by setting its value to 0 but keep the banked studs HUD visible
            if (ValidExtractionPlayer.InLobby?):
                if (Stat_GoldStuds.SetValue[ValidAgent, 0]) # 'Stat Bar Show on HUD' should be set to 'Non-Zero' on the device so the current studs HUD automatically hides when in the lobby
                    {}

                if (Stat_BankedGoldStuds_VisualsOnly.SetValue[ValidAgent, ItemAmount])
                    {}

            # In a level: Show both gold studs and the banked studs HUD
            else if (not ValidExtractionPlayer.InLobby?):
                if (Stat_GoldStuds.SetValue[ValidAgent, ItemAmount])
                    {}
        else:
            Print("Error! No ExtractionPlayer was found for UpdateStuds()")

    
    # Remove studs from the players inventory and add them to the banked studs stat value 
    DepositStuds(InAgent : ?agent) : void = 
        if (ValidAgent := InAgent?):
            ItemAmount : int = ConditionalButton_GoldStuds_CheckCount.GetItemCount(ValidAgent, 0)
            
            if:
                BankedAmount := Stat_BankedGoldStuds.GetValue[ValidAgent]
                Stat_BankedGoldStuds.SetValue[ValidAgent, ItemAmount + BankedAmount]
            then:
                ItemRemover_DepositGoldStuds.Remove(ValidAgent)
                spawn{UpdateBankedStuds(ValidAgent, ItemAmount)}

    UpdateBankedStuds(InAgent : agent, AmountToAdd : int)<suspends> : void =
        Sleep(2.0)
        
        # Play effects only if there are any new studs to be banked
        if (AmountToAdd > 0):
            CinematicSequence_DepositGoldStuds.Play(InAgent)
            HUDMessage_DepositeGoldStuds.Show(InAgent, StringToMessage("{AmountToAdd} Studs added to Bank Vault!"))
            AudioPlayer_DepositGoldStuds.Play(InAgent)

        # Grant the lobby studs after depositing
        GrantLobbyStuds(option{InAgent})

    # Deposit the studs that the player is carrying in the lobby so that they return to the banked studs total
    DepositLobbyStuds(InAgent : ?agent) : void = 
        if (ValidAgent := InAgent?):    
            ItemAmount : int = ConditionalButton_GoldStuds_CheckCount.GetItemCount(ValidAgent, 0)
            
            if (Stat_BankedGoldStuds.SetValue[ValidAgent, ItemAmount]):
                ItemRemover_DepositGoldStuds.Remove(ValidAgent)

    # Make sure to update the banked studs when a player is eliminated
    OnPlayerEliminated(EliminatedAgent : agent) : void =
        if (Player := player[EliminatedAgent]):
            spawn{UpdateBankedStuds(EliminatedAgent, 0)}

    # Clear the player's inventory when they leave the game (instead of dropping all items on the ground)
    ClearAllInventory(LeavingPlayer : player) : void =
        if (Agent := agent[LeavingPlayer]):
            ItemRemover_RemoveAll.Remove(Agent)

    # Invoked when a new player joins the game
    OnPlayerJoined (InAgent : ?agent) : void =
        spawn{QueueGrantLobbyStuds_OnPlayerJoined(InAgent)}

    # Grant banked lobby studs when a player joins the game
    QueueGrantLobbyStuds_OnPlayerJoined(InAgent : ?agent)<suspends> : void =
        # Wait a bit to make sure the player stats have been fully initialized
        Sleep(2.0)
        GrantLobbyStuds(InAgent)

    # Get current banked studs value and grant in lump sums 0f 100s, 10s and 1s for purchases in the lobby
    GrantLobbyStuds(InAgent : ?agent) : void =
        if (ValidAgent := InAgent?):
            if (BankedStuds := Stat_BankedGoldStuds.GetValue[ValidAgent]):
                var Remainder_100 : int = 0
                var Remainder_10 : int = 0
                
                if:
                    StudsX100 := Floor(BankedStuds/100)
                    Remainder := Mod[BankedStuds, 100]
                then:
                    set Remainder_100 = Remainder

                    for (Index := 1..StudsX100, StudsX100 > 0):
                        ItemGranter_LobbyStuds_X100.GrantItem(ValidAgent)

                if:
                    StudsX10 := Floor(Remainder_100/10)
                    Remainder := Mod[Remainder_100, 10]
                then:
                    set Remainder_10 = Remainder
                    
                    for (Index := 1..StudsX10, StudsX10 > 0):
                        ItemGranter_LobbyStuds_X10.GrantItem(ValidAgent)

                if:
                    StudsX1 := Remainder_10
                    StudsX1 > 0
                then:
                    for (Index := 1..StudsX1):
                        ItemGranter_LobbyStuds_X1.GrantItem(ValidAgent)

        else:
            Print("No Agent Provided for GrantLobbyStuds")


# Category definitions for the gold studs manager
StudManager_StatCategory<public><localizes> : message := "Stat Devices"
StudManager_GeneralDevicesCategory<public><localizes> : message := "General Devices"
StudManager_TriggerEventsCategory<public><localizes> : message := "Trigger Events"

# Tooltip definitions for stat devices
ToolTip_StudManager_Stat_GoldStuds<public><localizes> : message = "Stat device to display the player's current studs while in the level. Note: The stat should be set to only shown on the HUD when the value is non-zero."
ToolTip_StudManager_Stat_BankedGoldStuds<public><localizes> : message = "Stat device to store the player's banked studs that have been deposited in the lobby. Persistency should be enabled on this device. Note: The stat is set to never show on the HUD."
ToolTip_StudManager_Stat_BankedGoldStuds_VisualsOnly<public><localizes> : message = "Stat device to display the player's banked studs while in the lobby. Note: The stat should be set to only shown on the HUD when the value is non-zero."

# Tooltip definitions for general devices
ToolTip_StudManager_ConditionalButton_GoldStuds_CheckCount<public><localizes> : message = "Conditional button device that checks the player's current gold studs count and returns the amount of gold studs in their inventory."
ToolTip_StudManager_ItemRemover_DepositGoldStuds<public><localizes> : message = "Item remover device that removes the gold studs from the player when depositing them in the bank."
ToolTip_StudManager_ItemRemover_RemoveAll<public><localizes> : message = "Item remover device that removes all items in the player's inventory."
ToolTip_StudManager_CinematicSequence_DepositGoldStuds<public><localizes> : message = "Cinematic sequence device that plays an animation for the instigator player who deposited studs after extracting to the lobby."
ToolTip_StudManager_HUDMessage_DepositeGoldStuds<public><localizes> : message = "HUD message device that displays a message to the player when they deposit gold studs in the lobby."
ToolTip_StudManager_AudioPlayer_DepositGoldStuds<public><localizes> : message = "Audio player device that plays a sound when the player deposits gold studs in the lobby."
ToolTip_StudManager_ItemGranter_LobbyStuds_X1<public><localizes> : message = "Item granter device that grants the player 1 gold stud in the lobby."
ToolTip_StudManager_ItemGranter_LobbyStuds_X10<public><localizes> : message = "Item granter device that grants the player 10 gold studs in the lobby."
ToolTip_StudManager_ItemGranter_LobbyStuds_X100<public><localizes> : message = "Item granter device that grants the player 100 gold studs in the lobby."
ToolTip_StudManager_EliminationManager<public><localizes> : message = "Keeps track of when players get eliminated."

# Tooltip definitions for trigger events
ToolTip_StudManager_OnGoldStudsChanged<public><localizes> : message = "Triggered when the player's gold studs count changes."
ToolTip_StudManager_DepositGoldStuds<public><localizes> : message = "Triggered when the player deposits their gold studs in the bank."
ToolTip_StudManager_DepositLobbyStuds<public><localizes> : message = "Triggered when the player deposits their gold studs in the lobby."
ToolTip_StudManager_OnLobbyEnter<public><localizes> : message = "Triggered when the player enters the lobby."
ToolTip_StudManager_OnLobbyExit<public><localizes> : message = "Triggered when the player exits the lobby."
ToolTip_StudManager_OnBRICKFortPlayerAdded<public><localizes> : message = "Triggered when a new BRICK Fort Player is added to the game."
