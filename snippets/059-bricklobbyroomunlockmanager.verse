# Knowledge-neutral Verse snippet; project names and asset IDs generalized.
# id: snippet-059

using { /Fortnite.com/Devices }
using { /Verse.org/Simulation }
using { /UnrealEngine.com/Temporary/Diagnostics }

# Manager class that handles unlock zones in the lobby based on the player's level
BRICK_lobby_room_unlock_manager := class(creative_device):
    @editable:
        ToolTip := ToolTip_LobbyRoomUnlockManager_PlayerLevel
    PlayerLevel : stat_creator_device = stat_creator_device{}  

    @editable:
        ToolTip := ToolTip_LobbyRoomUnlockManager_Trigger_OnBRICKFortPlayerAdded
    Trigger_OnBRICKFortPlayerAdded : trigger_device = trigger_device{}

    @editable:
        ToolTip := ToolTip_LobbyRoomUnlockManager_UnlockZones
    UnlockZones : []unlock_zone = array{}
    
    # Subscribes to the relevant events when the game starts
    OnBegin<override>()<suspends> : void =
        Trigger_OnBRICKFortPlayerAdded.TriggeredEvent.Subscribe(OnPlayerJoined)
        PlayerLevel.LevelChangedEvent.Subscribe(OnPlayerLevelChanged)

    # Loop through all unlock zones and update their barriers based on the player's current level
    # When a player is added to a barrier's ignore list, the specific player will not see the barrier and  can pass through the barrier
    OnPlayerLevelChanged(Agent : ?agent, CurrentPlayerLevel : int) : void =
        if (ValidAgent := Agent?):
            for (UnlockZone : UnlockZones):
                if (CurrentPlayerLevel >= UnlockZone.UnlockLevel):
                    UnlockZone.Barrier.AddToIgnoreList(ValidAgent)
                else:
                    UnlockZone.Barrier.RemoveFromIgnoreList(ValidAgent)

    # Ensure that unlock zones are correctly updated when a player joins the game
    OnPlayerJoined(Agent : ?agent) : void =
        if:
            ValidAgent := Agent?
            CurrentPlayerLevel := PlayerLevel.GetLevel[ValidAgent]
        then:
            OnPlayerLevelChanged(Agent, CurrentPlayerLevel)

# Class that represents an unlock zone using a barrier device
unlock_zone := class():
    @editable:
        ToolTip := ToolTip_UnlockZone_UnlockLevel
    UnlockLevel : int = 5

    @editable:
        ToolTip := ToolTip_UnlockZone_Barrier
    Barrier : barrier_device = barrier_device{}


# Tooltip definitions for the editable fields in the lobby room unlock manager
ToolTip_LobbyRoomUnlockManager_PlayerLevel<public><localizes> : message = "Stat device that tracks the player's level."
ToolTip_LobbyRoomUnlockManager_Trigger_OnBRICKFortPlayerAdded<public><localizes> : message = "Trigger device that is invoked from the fort player manager when a player joins the game."
ToolTip_LobbyRoomUnlockManager_UnlockZones<public><localizes> : message = "Array of unlock zones that can be unlocked by players."
ToolTip_UnlockZone_UnlockLevel<public><localizes> : message = "Required player level to unlock this zone."
ToolTip_UnlockZone_Barrier<public><localizes> : message = "Barrier device that prevents players from entering this zone."
