# Knowledge-neutral Verse snippet; project names and asset IDs generalized.
# id: snippet-060

using { /Fortnite.com/Characters }
using { /Fortnite.com/Devices }
using { /Verse.org/Simulation }
using { /UnrealEngine.com/Temporary/Diagnostics }
using { /UnrealEngine.com/Temporary/SpatialMath }

# ============================================================================================================================================
# Device that spawns studs at a location determined by the type of spawner used
# ============================================================================================================================================

BRICK_stud_spawner_device := class(creative_device):
    @editable
    Rewards : []BRICK_stud_spawner = array{}

    OnBegin<override>()<suspends> : void =
        for (Reward : Rewards):
            Reward.Init(Self)

# Definition of a spawning event for studs
BRICK_stud_spawner<public> := class():
    @editable:
        ToolTip := ToolTip_Reward
    Amount : int = 5

    @editable:
        ToolTip := ToolTip_SpawnType
    SpawnType : EN_ItemSpawnerConfig = EN_ItemSpawnerConfig.Toss

    var OwningDevice : creative_device = creative_device{}

    Init<public>(InDevice : creative_device) : void =
        set OwningDevice = InDevice

    SpawnStudsAtLocation<public>(InLocation : vector3) : void =
        if (StudDistributor := OwningDevice.GetBRICKStudPool[]):
            StudDistributor.SpawnStudsAtLocation(InLocation, ?InAmount := Amount, ?InSpawnType := SpawnType)

# Spawn studs at the location of a trigger when it's triggered
BRICK_stud_spawner_trigger := class(BRICK_stud_spawner):
    @editable:
        ToolTip := ToolTip_Trigger
    var Trigger : trigger_device = trigger_device{}

    Init<override>(InDevice : creative_device) : void =
        (super:)Init(InDevice)
        Trigger.TriggeredEvent.Subscribe(OnTriggered)

    OnTriggered(InAgent : ?agent) : void =
        SpawnStudsAtLocation(Trigger.GetTransform().Translation)

# Spawn studs at the location of a button when it's pressed
BRICK_stud_spawner_button := class(BRICK_stud_spawner):
    @editable:
        ToolTip := ToolTip_Button
    var Button : button_device = button_device{}

    Init<override>(InDevice : creative_device) : void =
        (super:)Init(InDevice)
        Button.InteractedWithEvent.Subscribe(OnTriggered)

    OnTriggered(InAgent : agent) : void =
        SpawnStudsAtLocation(Button.GetTransform().Translation)

# Spawn studs at the location of an NPC when it's eliminated
BRICK_stud_spawner_npc := class(BRICK_stud_spawner):
    @editable:
        ToolTip := ToolTip_NPCSpawners
    var NPCSpawners : []npc_spawner_device = array{}

    Init<override>(InDevice : creative_device) : void =
        (super:)Init(InDevice)

        for (NPCSpawner : NPCSpawners):
            NPCSpawner.EliminatedEvent.Subscribe(OnEliminated)

    OnEliminated(InResult : device_ai_interaction_result) : void =
        if (ValidSource := InResult.Target?.GetFortCharacter[].GetTransform().Translation):
            SpawnStudsAtLocation(ValidSource)

# Spawn studs when an assembly is completed or disassembled
EN_AssemblyMode := enum{Assemble, Disassemble}
BRICK_stud_spawner_assembly := class(BRICK_stud_spawner):
    @editable:
        ToolTip := ToolTip_Assembly
    var AssemblyDevices : []assembly_device = array{}

    @editable:
        ToolTip := ToolTip_AssemblyMode
    var CompletionCondition : EN_AssemblyMode = EN_AssemblyMode.Assemble

    Init<override>(InDevice : creative_device) : void =
        (super:)Init(InDevice)

        for (AssemblyDevice : AssemblyDevices):
            case (CompletionCondition):
                EN_AssemblyMode.Assemble =>
                    AssemblyDevice.AssemblyFinishedEvent.Subscribe(OnCompletedCondition)
                EN_AssemblyMode.Disassemble =>
                    AssemblyDevice.DisassemblyFinishedEvent.Subscribe(OnCompletedCondition)

    OnCompletedCondition(InAgent : ?agent) : void =
        if (ValidCharacter := InAgent?.GetFortCharacter[]):
            SpawnStudsAtLocation(ValidCharacter.GetTransform().Translation)


ToolTip_Reward<public><localizes> : message = "The number of studs to be spawned."
ToolTip_SpawnType<public><localizes> : message = "The method used to spawn the studs, use Gravity in tight spaces."
ToolTip_Trigger<public><localizes> : message = "Studs will be spawned at the trigger location when it is triggered."
ToolTip_Button<public><localizes> : message = "Studs will be spawned at the button location when it is pressed."
ToolTip_NPCSpawners<public><localizes> : message = "The list of NPC spawners that will spawn studs when an NPC is eliminated."
ToolTip_Assembly<public><localizes> : message = "The list of assembly devices that will spawn studs when an assembly is completed."
ToolTip_AssemblyMode<public><localizes> : message = "The condition that must be met for the studs to be spawned."
