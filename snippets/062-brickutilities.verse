# Knowledge-neutral Verse snippet; project names and asset IDs generalized.
# id: snippet-062

using { /Verse.org/Concurrency }
using { /Verse.org/Simulation }
using { /Verse.org/Simulation/Tags }
using { /Verse.org/Random }
using { /UnrealEngine.com/Temporary/SpatialMath }
using { /UnrealEngine.com/Temporary/UI }
using { /UnrealEngine.com/Temporary/Diagnostics }
using { /Fortnite.com/Devices }
using { /Fortnite.com/Characters }
using { /Fortnite.com/Devices/CreativeAnimation }	
using { /Fortnite.com/Playspaces }

BRICK_utilities_log<public> := class(log_channel):
BRICK_utilities_tag<public> := class(tag){}

Tooltip_FNBindingsInterface<public><localizes> : message = "trigger_devices to trigger when device specific events occur - allows to use this Verse device with Fortnite devices"

# ============================================================================================================================================
# Math Helpers & Operators
# ============================================================================================================================================
# Check if a float is within a range
(InFloat : float).IsInRange<public>(InLowerBoundIncl : float, InUpperBoundIncl : float)<decides><transacts>: void =
    InLowerBoundIncl <= InFloat <= InUpperBoundIncl

# Divide an int by another int
DivideInt<public>(Dividend : int, Divisor : int)<decides><transacts> : int =
    var OutResult : ?int = false
    if:
        Divisor <> 0
        set OutResult = option{Round[((Dividend * 1.0) / (Divisor * 1.0)) + 0.001]}
    else:
        Logger: log = log{Channel := BRICK_utilities_log, DefaultLevel := log_level.Normal}
        Logger.Print("Division by 0 is not allowed.")
    OutResult?

# SOURCE: https://dev.epicgames.com/community/snippets/1mD/fortnite-float-to-int-cast-and-operators
Float(InInt : int)<reads><allocates>: float =
    InInt * 0.1

operator'+'(I : int, F : float): float =
    I * 1.0 + F

operator'+'(I : float, F : int): float =
    I * 1.0 + F

operator'-'(I : int, F : float): float =
    I * 1.0 - F

operator'-'(I : float, F : int): float =
    F - I * 1.0

(InIndex : int).IncrementWrapped<public>(InMaxValue : int)<computes>: int =
    if (InIndex < InMaxValue):
        InIndex + 1
    else:
        0

(InIndex : int).DecrementWrapped<public>(InMaxValue : int)<computes>: int =
    if (InIndex > 0):
        InIndex - 1
    else:
        InMaxValue

StringToMessage<public><localizes>(InString : string)<transacts>: message = "{InString}"

# Helper function to get the delta from a start time in Seconds
GetElapsedTime<public>(InStartTime : float) : float =
    GetSimulationElapsedTime() - InStartTime

# --------------------------------------------------------------------------------------------------------------------------------------------
# Array helpers
# --------------------------------------------------------------------------------------------------------------------------------------------
(Array : []t where t : subtype(comparable)).AllContainedIn<public>(InArray : []t)<decides><transacts>: void =
    var FoundMatches : int = 0
    for (MaybeMatch : Array, InArray.Find[MaybeMatch]):
        set FoundMatches += 1
    FoundMatches = Array.Length

(Array : []t where t : subtype(comparable)).AnyContainedIn<public>(InArray : []t)<decides><transacts>: void =
    var FoundMatches : int = 0
    for (MaybeMatch : Array, InArray.Find[MaybeMatch]):
        set FoundMatches += 1
    FoundMatches <> 0

(Array : []t where t : subtype(comparable)).NoneContainedIn<public>(InArray : []t)<decides><transacts>: void =
    var FoundMatches : int = 0
    for (MaybeMatch : Array, InArray.Find[MaybeMatch]):
        set FoundMatches += 1
    FoundMatches = 0

# Removes a particular element if it exists in InArray
ManualRemoveElement<public>(InArray : []t, InElement : t where t : subtype(comparable))<transacts>: []t =
    var OutArray : []t = array{}
    for (Element : InArray, Element <> InElement):
        set OutArray += array{Element}
    OutArray

# Reimplementation of the Unreal Engine function
AddElementUnique<public>(InArray : []t, InElement : t where t : subtype(comparable))<transacts>: []t =
    if (InArray.Find[InElement]):
        InArray
    else:
        InArray + array{InElement}
