# Knowledge-neutral Verse snippet; project names and asset IDs generalized.
# id: snippet-063

using { /Fortnite.com/Devices }
using { /Verse.org/Simulation }
using { /UnrealEngine.com/Temporary/Diagnostics }

# Manager class that handles granting experience points to players when they make progress in certain events
BRICK_xp_level_manager := class(creative_device):

    @editable:
        ToolTip := ToolTip_LevelManager_TrackerEvents
    TrackerEvents : []xp_level_manager_trackers_fndevice_interface = array{}

    @editable:
        ToolTip := ToolTip_LevelManager_EliminationEvents
    EliminationEvents : xp_level_manager_elimination_fndevice_interface = xp_level_manager_elimination_fndevice_interface{}

    @editable:
        ToolTip := ToolTip_LevelManager_ExtractionEvent
    ExtractionEvent : xp_level_manager_extraction_fndevice_interface = xp_level_manager_extraction_fndevice_interface{}

    @editable:
        ToolTip := ToolTip_LevelManager_LootChestEvents
    LootChestEvents : []xp_level_manager_loot_chest_fndevice_interface = array{}

    # Subscribe to the relevant events when the game starts
    OnBegin<override>()<suspends> : void =
        # Event trackers
        for (Tracker : TrackerEvents):
            Tracker.Trigger_OnEventProgress.TriggeredEvent.Subscribe(Tracker.GrantXP)

        # Eliminations
        EliminationEvents.EliminationManager.EliminatedEvent.Subscribe(EliminationEvents.GrantXP_EliminatedEvent)
        EliminationEvents.EliminationManager.EliminationEvent.Subscribe(EliminationEvents.GrantXP_EliminationEvent)

        # Extractions
        ExtractionEvent.OnExtractedTrigger.TriggeredEvent.Subscribe(ExtractionEvent.GrantXP)

        # Loot chests
        for (LootChest : LootChestEvents):
            LootChest.ItemGranter.ItemGrantedEvent.Subscribe(LootChest.GrantXP)

# Class for generic tracker events that can grant experience points upon tracker progress
xp_level_manager_trackers_fndevice_interface := class<concrete>():
    @editable:
        ToolTip := ToolTip_TrackerEvent_TrackerDevice
    Trigger_OnEventProgress : trigger_device = trigger_device{}

    @editable:
        ToolTip := ToolTip_TrackerEvent_Trigger_GrantXP
    Trigger_GrantXP : trigger_device = trigger_device{}

    GrantXP(Agent : ?agent) : void =
        if (ValidAgent := Agent?):
            Trigger_GrantXP.Trigger(ValidAgent)

# Class for elimination events that can grant experience points upon eliminations
xp_level_manager_elimination_fndevice_interface := class<concrete>():
    @editable:
        ToolTip := ToolTip_EliminationEvent_EliminationManager
    EliminationManager : elimination_manager_device = elimination_manager_device{}

    @editable:
        ToolTip := ToolTip_EliminationEvent_Trigger_GrantXP_EliminatedEvent
    Trigger_GrantXP_EliminatedEvent : trigger_device = trigger_device{}

    @editable:
        ToolTip := ToolTip_EliminationEvent_Trigger_GrantXP_EliminationEvent
    Trigger_GrantXP_EliminationEvent : trigger_device = trigger_device{}

    GrantXP_EliminatedEvent(EliminatedAgent : agent) : void =
        if (Player := player[EliminatedAgent]):
            Trigger_GrantXP_EliminatedEvent.Trigger(EliminatedAgent)

    GrantXP_EliminationEvent(EliminatorAgent : ?agent) : void =
        if (ValidAgent := EliminatorAgent?):
            Trigger_GrantXP_EliminationEvent.Trigger(ValidAgent)

# Class for extraction events that can grant experience points upon player extracting to lobby
xp_level_manager_extraction_fndevice_interface := class<concrete>():
    @editable:
        ToolTip := ToolTip_ExtractionEvent_OnExtractedTrigger
    OnExtractedTrigger : trigger_device = trigger_device{}

    @editable:
        ToolTip := ToolTip_ExtractionEvent_Trigger_GrantXP
    Trigger_GrantXP : trigger_device = trigger_device{}

    @editable:
        ToolTip := ToolTip_ExtractionEvent_Stat_GoldStuds
    Stat_GoldStuds : stat_creator_device = stat_creator_device{}

    # Grant experience points to the player if the extract with more than 0 gold studs
    GrantXP(Agent : ?agent) : void =
        if:
            ValidAgent := Agent?
            NumberOfStuds : int = Stat_GoldStuds.GetValue[ValidAgent]
            NumberOfStuds > 0
        then:
            Trigger_GrantXP.Trigger(ValidAgent)

# Class for generic tracker events that can grant experience points upon tracker progress
xp_level_manager_loot_chest_fndevice_interface := class<concrete>():
    @editable:
        ToolTip := ToolTip_LootChestEvent_OnItemGranted
    ItemGranter : item_granter_device = item_granter_device{}

    @editable:
        ToolTip := ToolTip_LootChestEvent_Trigger_GrantXP
    Trigger_GrantXP : trigger_device = trigger_device{}

    GrantXP(Agent : agent) : void =
        Trigger_GrantXP.Trigger(Agent)

# Tooltip definitions for the editable fields in the xp level manager
ToolTip_LevelManager_TrackerEvents<public><localizes> : message = "Array of trigger devices that will grant experience points to players when making progress."
ToolTip_LevelManager_EliminationEvents<public><localizes> : message = "Elimination manager device that will grant experience points to players upon elimination."
ToolTip_LevelManager_ExtractionEvent<public><localizes> : message = "Extraction manager device that will grant experience points to players when they extract to the lobby."
ToolTip_TrackerEvent_TrackerDevice<public><localizes> : message = "Trigger device that will grant experience points to players."
ToolTip_TrackerEvent_Trigger_GrantXP<public><localizes> : message = "Trigger device that will be invoked when progress has been made, granting experience points to the player via a powerup stat device."
ToolTip_EliminationEvent_EliminationManager<public><localizes> : message = "Elimination manager device that tracks when players are eliminated or eliminate others."
ToolTip_EliminationEvent_Trigger_GrantXP_EliminatedEvent<public><localizes> : message = "Trigger device that will be invoked when a player is eliminated to grant experience points to the eliminated player."
ToolTip_EliminationEvent_Trigger_GrantXP_EliminationEvent<public><localizes> : message = "Trigger device that will be invoked when a player is eliminated to grant experience points to the eliminator player."
ToolTip_ExtractionEvent_OnExtractedTrigger<public><localizes> : message = "Trigger that gets invoked when a player extracts to the lobby."
ToolTip_ExtractionEvent_Trigger_GrantXP<public><localizes> : message = "Trigger device that will be invoked when a player extracts to the lobby to grant experience points to the player via a powerup stat device."
ToolTip_ExtractionEvent_Stat_GoldStuds<public><localizes> : message = "Stat device that tracks the number of gold studs collected by the player."
ToolTip_LootChestEvent_OnItemGranted<public><localizes> : message = "Trigger device that will be invoked when an item is granted from a loot chest."
ToolTip_LootChestEvent_Trigger_GrantXP<public><localizes> : message = "Trigger device that will be invoked when an item is granted from a loot chest to grant experience points to the player."
ToolTip_LevelManager_LootChestEvents<public><localizes> : message = "Array of loot chest devices that will grant experience points to players when an item is granted."
