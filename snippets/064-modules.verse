# Knowledge-neutral Verse snippet; project names and asset IDs generalized.
# id: snippet-064

BRICKExtraction<public> := module:
    Core<public> := module {}
    Debug<public> := module {}
    Lobby<public> := module {}

BRICKStudRewards<public> := module {}
BRICKUtilities<public> := module {}
