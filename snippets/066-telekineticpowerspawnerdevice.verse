# Knowledge-neutral Verse snippet; project names and asset IDs generalized.
# id: snippet-066

using { /Fortnite.com/Devices }
using { /Fortnite.com/SPACE_OPERA }
using { /Fortnite.com/Items/SPACE_OPERA }
using { /Fortnite.com/Weapons/SPACE_OPERA }
using { /Verse.org/SceneGraph }
using { /Verse.org/Simulation }
using { /UnrealEngine.com/Itemization }
using { /UnrealEngine.com/Temporary/Diagnostics }

using { EnergyBladeWorkbenchCore }

# Enum of telekinetic abilities.
# Used as a shorthand for referencing a specific item entity.
Telekinetic_power := enum<open>
{
    Cloak,
    Decoy,
    Heal,
    Pull,
    Push,
    Sense,
    Shield,
    Throw,
    Zap,

    LightningAugment,
    ThrowAugment,
    TrainingAugment,
    PullAugment,
    PushAugment
}

# Device that spawns telekinetic abilities and adds them to a player's inventory
Telekinetic_power_spawner_device := class(creative_device):

    # The button to bind to.
    # This could be replaced with a volume or similar device.
    # Anything that passes an agent through to the subscriber.
    @editable
    Button : button_device = button_device {}

    # The telekinetic ability to give the player
    @editable
    Power : Telekinetic_power = Telekinetic_power.LightningAugment

    OnBegin<override>():void=

        Button.InteractedWithEvent.Subscribe(OnButtonInteractedWith)

    # Triggers when a player interacts with the Button
    OnButtonInteractedWith<private>(InAgent:agent):void=

        if:
            Inventory := InAgent.GetInventoryComponent[]

            GiveTelekineticAbility[Inventory]

        then:
            Print("Item added")

        else:
            Print("Failed to add item(s)")

    # Spawns a telekinetic ability item and adds it to the target inventory
    GiveTelekineticAbility<private>(InInventory:inventory_component)<decides><transacts>:entity=

        Item := case(Power):
            Telekinetic_power.Cloak => TelekineticAbilityTrick_Weapon_V1 {}
            Telekinetic_power.Decoy => TelekineticAbilityDistraction_Weapon_V1 {}
            Telekinetic_power.Heal => TelekineticAbilityHeal_Weapon_V1 {}
            Telekinetic_power.Pull => TelekineticAbilityPull_Weapon_V1 {}
            Telekinetic_power.Push => TelekineticAbilityPush_Weapon_V1 {}
            Telekinetic_power.Sense => TelekineticAbilityAwareness_Weapon_V1 {}
            Telekinetic_power.Shield => TelekineticAbilitieshield_Weapon_V1 {}
            Telekinetic_power.Throw => TelekineticAbilityThrow_Weapon_V1 {}
            Telekinetic_power.Zap => TelekineticAbilityThrow_Weapon_V1 {}
            Telekinetic_power.LightningAugment => TelekineticAbilityLightning_EnergyBladeAugment_V1 {}
            Telekinetic_power.ThrowAugment => TelekineticAbilitiesaberThrow_EnergyBladeAugment_V1 {}
            Telekinetic_power.TrainingAugment => TelekineticAbilityTraining_Augment_V1 {}
            Telekinetic_power.PullAugment => TelekineticAbilityPull_EnergyBladeAugment_V1 {}
            Telekinetic_power.PushAugment => TelekineticAbilityPush_EnergyBladeAugment_V1 {}

        Result := InInventory.AddItemDistribute(Item);

        Result.GetSuccess[]

        Item
