# Knowledge-neutral Verse snippet; project names and asset IDs generalized.
# id: snippet-067

using { /Fortnite.com/Devices }
using { /Verse.org/SceneGraph }
using { /Verse.org/Simulation }
using { /UnrealEngine.com/Itemization }
using { /UnrealEngine.com/Temporary/Diagnostics }
    
# Finds the first inventory component for a given agent
(Agent:agent).GetInventoryComponent<public>()<decides><transacts>:inventory_component=
    
    Agent.FindDescendantComponents(inventory_component).First[]

# Returns the 0th element in a collection
(Generator:generator(t) where t:type).First<public>()<decides><reads>:t=

    Items := for:
        Item : Generator
    do:
        Item

    Items[0]
