# Knowledge-neutral Verse snippet; project names and asset IDs generalized.
# id: snippet-068

using { /Fortnite.com/Devices }
using { /Fortnite.com/SPACE_OPERA }
using { /Verse.org/SceneGraph }
using { /Verse.org/Simulation }
using { /UnrealEngine.com/Itemization }
using { /UnrealEngine.com/Temporary/Diagnostics }

using { EnergyBladeWorkbenchCore }

# Enum of Kyber Crystals.
# Used as a shorthand for referencing a specific component type.
kyber_crystal<internal> := enum<open>
{
    blue,
    green,
    purple,
    red,
    orange,
    yellow,
    white
}

# Util function to get a string name for a given Kyber Crystal
ToString<internal>(Val:kyber_crystal)<reads>:string=

    case(Val):
        kyber_crystal.blue      => "Blue Kyber Crystal"
        kyber_crystal.green     => "Green Kyber Crystal"
        kyber_crystal.purple    => "Purple Kyber Crystal"
        kyber_crystal.red       => "Red Kyber Crystal"
        kyber_crystal.orange    => "Orange Kyber Crystal"
        kyber_crystal.yellow    => "Yellow Kyber Crystal"
        kyber_crystal.white     => "White Kyber Crystal"
        _ => "Unknown Kyber Crystal"

# Util function to get a specific type of kyber_crystal_component
GetKyberCrystalComponentType<internal>(InCrystal:kyber_crystal)<decides><computes>:subtype(kyber_crystal_component)=

    case(InCrystal):
        kyber_crystal.blue      => kyber_crystal_blue_component
        kyber_crystal.green     => kyber_crystal_green_component
        kyber_crystal.purple    => kyber_crystal_purple_component
        kyber_crystal.red       => kyber_crystal_red_component
        kyber_crystal.orange    => kyber_crystal_orange_component
        kyber_crystal.yellow    => kyber_crystal_yellow_component
        kyber_crystal.white     => kyber_crystal_white_component

# Util function to create a specific type of kyber_crystal_component
CreateKyberCrystalComponent<internal>(InEntity:entity, InCrystal:kyber_crystal)<decides><transacts>:kyber_crystal_component=
    
    case(InCrystal):
        kyber_crystal.blue      => kyber_crystal_blue_component { Entity := InEntity }
        kyber_crystal.green     => kyber_crystal_green_component { Entity := InEntity }
        kyber_crystal.purple    => kyber_crystal_purple_component { Entity := InEntity }
        kyber_crystal.red       => kyber_crystal_red_component { Entity := InEntity }
        kyber_crystal.orange    => kyber_crystal_orange_component { Entity := InEntity }
        kyber_crystal.yellow    => kyber_crystal_yellow_component { Entity := InEntity }
        kyber_crystal.white     => kyber_crystal_white_component { Entity := InEntity }

    # Can't call AddComponents here because AddComponents is no-rollback

# Util function to get a kyber_crystal_component from an entity
(InEntity:entity).GetKyberCrystalComponent<internal>()<decides><reads>:kyber_crystal_component=

    InEntity.GetComponent[kyber_crystal_component]
    
# Util function to get a specific type of kyber_crystal_component from an entity
(InEntity:entity).GetKyberCrystalComponent<internal>(InCrystal:kyber_crystal)<decides><reads>:kyber_crystal_component=

    Type := GetKyberCrystalComponentType[InCrystal]

    InEntity.GetComponent[Type]
