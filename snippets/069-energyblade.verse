# Knowledge-neutral Verse snippet; project names and asset IDs generalized.
# id: snippet-069

using { /Fortnite.com/Devices }
using { /Fortnite.com/SPACE_OPERA }
using { /Fortnite.com/Weapons/SPACE_OPERA }
using { /Verse.org/SceneGraph }
using { /Verse.org/Simulation }
using { /UnrealEngine.com/Itemization }
using { /UnrealEngine.com/Temporary/Diagnostics }

using { EnergyBladeWorkbenchCore }

# Enum of EnergyBlade hilts.
# Used as a shorthand for referencing a specific entity type.
EnergyBlade<internal> := enum<open>
{
    EnergyBladeHeroA,
    VillainB,
    EnergyBladeHeroB,
    EnergyBladeHeroC,
    VillainA,
    Mentor,

    ApprenticeA,
    ApprenticeB,
    Training,
    OpposingApprentice
}

# Util function to get a string name for a given EnergyBlade
ToString<internal>(Val:EnergyBlade)<reads>:string=

    case(Val):
        EnergyBlade.EnergyBladeHeroA    => "energy-blade hero A's EnergyBlade"
        EnergyBlade.VillainB         => "villain B's EnergyBlade"
        EnergyBlade.EnergyBladeHeroB    => "energy-blade hero B's EnergyBlade"
        EnergyBlade.EnergyBladeHeroC   => "energy-blade hero C's EnergyBlade"
        EnergyBlade.VillainA      => "villain A's EnergyBlade"
        EnergyBlade.Mentor       => "mentor's EnergyBlade"
        EnergyBlade.ApprenticeA    => "apprentice EnergyBlade (A)"
        EnergyBlade.ApprenticeB    => "apprentice EnergyBlade (B)"
        EnergyBlade.Training    => "training EnergyBlade"
        EnergyBlade.OpposingApprentice  => "opposing apprentice EnergyBlade"
        _ => "Unknown EnergyBlade"

# Util function to create a specific type of EnergyBlade entity
CreateEnergyBladeEntity<internal>(InEnergyBlade:EnergyBlade)<decides><transacts>:entity=

    case(InEnergyBlade):
        EnergyBlade.EnergyBladeHeroA    => EnergyBladeHeroA_V1 {}
        EnergyBlade.VillainB         => EnergyBladeVillainB_V1 {}
        EnergyBlade.EnergyBladeHeroB    => EnergyBladeHeroB_V1 {}
        EnergyBlade.EnergyBladeHeroC   => EnergyBladeHeroC_V1 {}
        EnergyBlade.VillainA      => EnergyBladeVillainA_V1 {}
        EnergyBlade.Mentor       => EnergyBladeMentor_V1 {}
        EnergyBlade.ApprenticeA    => EnergyBladeApprenticeA_V1 {}
        EnergyBlade.ApprenticeB    => EnergyBladeApprenticeB_V1 {}
        EnergyBlade.Training    => EnergyBladeTraining_V1 {}
        EnergyBlade.OpposingApprentice  => EnergyBladeOpposingApprentice_V1 {}

# Util function to get a EnergyBlade_item_component from an entity
(InEntity:entity).GetEnergyBladeItemComponent<internal>()<decides><reads>:EnergyBlade_item_component=

    InEntity.GetComponent[EnergyBlade_item_component]
