# Knowledge-neutral Verse snippet; project names and asset IDs generalized.
# id: snippet-070

using { /Fortnite.com/Devices }
using { /Fortnite.com/SPACE_OPERA }
using { /Verse.org/SceneGraph }
using { /Verse.org/Simulation }
using { /UnrealEngine.com/Itemization }
using { /UnrealEngine.com/Temporary/Diagnostics }

using { EnergyBladeWorkbenchCore }

# Device to change the arcing intensity of the player's currently equipped EnergyBlade
EnergyBlade_arcing_changer_device := class(creative_device):

    # Button to bind the "Increase Arcing Intensity" event to
    @editable
    IncreaseButton : button_device = button_device {}

    # Button to bind the "Decrease Arcing Intensity" event to
    @editable
    DecreaseButton : button_device = button_device {}

    OnBegin<override>():void=

        IncreaseButton.InteractedWithEvent.Subscribe(IncreaseButton_OnButtonInteractedWith)
        DecreaseButton.InteractedWithEvent.Subscribe(DecreaseButton_OnButtonInteractedWith)
    
    # Triggers when a player interacts with the IncreaseButton
    IncreaseButton_OnButtonInteractedWith<private>(InAgent:agent):void=

        if:
            Inventory := InAgent.GetInventoryComponent[]

        then:
            Items := ChangeEnergyBlades_Arcing(Inventory, 0.1)
            
            # Will be 0 if the player has no EnergyBlades
            Print("Changed {Items.Length} item(s)")

    # Triggers when a player interacts with the DecreaseButton
    DecreaseButton_OnButtonInteractedWith<private>(InAgent:agent):void=

        if:
            Inventory := InAgent.GetInventoryComponent[]

        then:
            Items := ChangeEnergyBlades_Arcing(Inventory, -0.1)
            
            # Will be 0 if the player has no EnergyBlades
            Print("Changed {Items.Length} item(s)")
            
    # Change the arcing intensity of the player's equipped EnergyBlade by the given amount
    ChangeEnergyBlades_Arcing(InInventory:inventory_component, Amount:float):[]entity=

        Items := for:
            Item : InInventory.FindItems()

            # We only want to modify the EnergyBlade that is currently in the player's hand
            ItemComponent := Item.GetEnergyBladeItemComponent[]
            ItemComponent.IsEquipped[]

            Crystal := Item.GetKyberCrystalComponent[]
        do:
            set Crystal.ArcIntensity += Amount

            Print("ArcIntensity: {Crystal.ArcIntensity}")

            Item
