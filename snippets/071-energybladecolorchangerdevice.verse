# Knowledge-neutral Verse snippet; project names and asset IDs generalized.
# id: snippet-071

using { /Fortnite.com/Devices }
using { /Fortnite.com/SPACE_OPERA }
using { /Verse.org/SceneGraph }
using { /Verse.org/Simulation }
using { /UnrealEngine.com/Itemization }
using { /UnrealEngine.com/Temporary/Diagnostics }

using { EnergyBladeWorkbenchCore }

# Device that replaces the kyber_crystal_component on the player's currently equipped EnergyBlade
EnergyBlade_color_changer_device := class(creative_device):

    # The button to bind to.
    # This could be replaced with a volume or similar device.
    # Anything that passes an agent through to the subscriber.
    @editable
    Button : button_device = button_device {}

    # The kind of Kyber Crystal to add to the EnergyBlade.
    @editable
    KyberCrystal : kyber_crystal = kyber_crystal.red

    # The arcing intensity of the created EnergyBlade.
    # The higher the value, the more the blade arcs with energy.
    # A value of zero disables all arcing.
    @editable_number(float):
        MinValue := option { 0.0 }
        MaxValue := option { 1.0 }
    KyberCrystalArcIntensity : float = 0.0

    # The stability of the created EnergyBlade.
    # This replaces the EnergyBlade's default stability.
    # The lower the value, the less stable and "frizzy" the blade is.
    @editable_number(float):
        MinValue := option { 0.0 }
        MaxValue := option { 1.0 }
    KyberCrystalStability : float = 1.0

    OnBegin<override>():void=

        Button.InteractedWithEvent.Subscribe(OnButtonInteractedWith)

    # Triggers when a player interacts with the Button
    OnButtonInteractedWith<private>(InAgent:agent):void=

        if:
            Inventory := InAgent.GetInventoryComponent[]

        then:
            spawn { ChangeEnergyBlades_Color(Inventory) }
            
    # Finds an equipped EnergyBlade in the target inventory and replaces it's kyber_crystal_component
    ChangeEnergyBlades_Color(InInventory:inventory_component)<suspends>:void=

        Items := for:
            Item : InInventory.FindItems()

            # We only want to modify the EnergyBlade that is currently in the player's hand
            ItemComponent := Item.GetEnergyBladeItemComponent[]
            ItemComponent.IsEquipped[]
        do:
            ChangeEnergyBlades_Color(Item)
    
    ChangeEnergyBlades_Color(InEnergyBlade:entity)<suspends>:void=

        OldCrystal := option { InEnergyBlade.GetKyberCrystalComponent[] }

        # Remove the old kyber crystal, if any
        if:
            # Does it already have a kyber crystal?
            Temp := OldCrystal?

        then:
            Temp.RemoveFromEntity()

            # Workaround for a race conditon when trying to replace a component.
            # If you try to remove a component and then replace it with another
            # sometimes the remove and add replicate in the wrong order.
            # If the client receives "add then remove" the add will fail because
            # there is already a component of that type.
            # Once that is fixed this sleep command can be removed and
            # the <suspends> attribute could be removed from the function.
            Sleep(0.1)

            Print("Removed existing Kyber Crystal")

        else:
            Print("Failed to remove existing Kyber Cystal")

            return

        # Add the new kyber crystal and apply settings to it
        if:
            NewCrystal := CreateKyberCrystalComponent[InEnergyBlade, KyberCrystal]

            set NewCrystal.ArcIntensity = KyberCrystalArcIntensity
            set NewCrystal.Stability = KyberCrystalStability

        then:
            InEnergyBlade.AddComponents(array { NewCrystal })
            
            Print("Inserted {KyberCrystal}")
            
        else:
            Print("Failed to replace Kyber Crystal")

            # Put it back

            if:
                Temp := OldCrystal?
            then:
                InEnergyBlade.AddComponents(array { Temp })
