# Knowledge-neutral Verse snippet; project names and asset IDs generalized.
# id: snippet-072

using { /Fortnite.com/Devices }
using { /Fortnite.com/SPACE_OPERA }
using { /Verse.org/SceneGraph }
using { /Verse.org/Simulation }
using { /UnrealEngine.com/Itemization }
using { /UnrealEngine.com/Temporary/Diagnostics }

using { EnergyBladeWorkbenchCore }

# Device to remove all EnergyBlades from a player's inventory when they interact with it
EnergyBlade_remover_device := class(creative_device):

    # The button to bind to.
    # This could be replaced with a volume or similar device.
    # Anything that passes an agent through to the subscriber.
    @editable
    Button : button_device = button_device {}

    OnBegin<override>():void=

        Button.InteractedWithEvent.Subscribe(OnButtonInteractedWith)

    # Triggers when a player interacts with the Button
    OnButtonInteractedWith(InAgent:agent):void=

        if:
            Inventory := InAgent.GetInventoryComponent[]

        then:
            Items := RemoveEnergyBlades(Inventory)

            # Will be 0 if the player has no EnergyBlades
            Print("Removed {Items.Length} item(s)")

        else:
            Print("Failed to remove item(s)")

    # Removes all EnergyBlade items from the target inventory
    RemoveEnergyBlades(InInventory:inventory_component):[]entity=

        Items := for:
            Item : InInventory.FindItems()
            Item.GetEnergyBladeItemComponent[]
            Result := InInventory.RemoveItem(Item)
            Result.GetSuccess[]
        do:
            Item
        
        Items
