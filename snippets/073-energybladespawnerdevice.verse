# Knowledge-neutral Verse snippet; project names and asset IDs generalized.
# id: snippet-073

using { /Fortnite.com/Devices }
using { /Fortnite.com/SPACE_OPERA }
using { /Verse.org/SceneGraph }
using { /Verse.org/Simulation }
using { /UnrealEngine.com/Itemization }
using { /UnrealEngine.com/Temporary/Diagnostics }
using { /Fortnite.com/UI }
using { /UnrealEngine.com/Temporary/UI }
using { /Fortnite.com/Items/SPACE_OPERA }

using { EnergyBladeWorkbenchCore }

# Device to add a EnergyBlade to a player's inventory when they interact with it
EnergyBlade_spawner_device := class(creative_device):

    # The button to bind to.
    # This could be replaced with a volume or similar device.
    # Anything that passes an agent through to the subscriber.
    @editable
    Button : button_device = button_device {}

    # The kind of EnergyBlade to give the player
    @editable
    EnergyBlade : EnergyBlade = EnergyBlade.EnergyBladeHeroC

    # Optionally, the kind of kyber crystal to add to the created EnergyBlade.
    # This replaces the EnergyBlade's default kyber crystal.
    @editable
    KyberCrystalOverride : ?kyber_crystal = false

    # The arcing intensity of the created EnergyBlade.
    # The higher the value, the more the blade arcs with energy.
    # A value of zero disables all arcing.
    @editable_number(float):
        MinValue := option { 0.0 }
        MaxValue := option { 1.0 }
    ArcIntensityOverride : float = 0.0

    # The stability of the created EnergyBlade.
    # This replaces the EnergyBlade's default stability.
    # The lower the value, the less stable and "frizzy" the blade is.
    @editable_number(float):
        MinValue := option { 0.0 }
        MaxValue := option { 1.0 }
    StabilityOverride : float = 1.0

    # The telekinetic ability to add to the EnergyBlade
    @editable
    PowerOverride : Telekinetic_power = Telekinetic_power.PullAugment

    OnBegin<override>():void=

        Button.InteractedWithEvent.Subscribe(OnButtonInteractedWith)

    # Triggers when a player interacts with the Button
    OnButtonInteractedWith<private>(InAgent:agent):void=

        if:
            Inventory := InAgent.GetInventoryComponent[]

        then:
            spawn { TryCreateEnergyBlade(Inventory) }

        else:
            Print("Failed to get inventory")

    # Remove all items containing a specific component type
    RemoveItems(PlayerInventory:inventory_component, ComponentType:subtype(component)):void =
        for (Item : PlayerInventory.FindItems(), Item.GetComponent[ComponentType], Result := PlayerInventory.RemoveItem(Item), Result.GetSuccess[]) {}
        
    # Creates a new lightasber item and adds it to the target inventory
    TryCreateEnergyBlade<private>(InInventory:inventory_component)<suspends>:void=

        # Remove any previous EnergyBlades or  telekinetic abilities
        RemoveItems(InInventory, kyber_crystal_component)
        RemoveItems(InInventory, Telekinetic_power_item_component)

        if:
            EnergyBladeEntity := CreateEnergyBladeEntity[EnergyBlade]

        then:
            
            if:
                # Don't override anything if this isn't set
                KyberCrystal := KyberCrystalOverride?

            then:
                # AddComponents (and RemoveFromEntity) calls are no-rollback so
                # they can't be used within a transaction.
                # In theory, one day no-rollback won't be a thing? In "Max Verse".

                if:
                    # Does it already have a kyber crystal?
                    # Should always be yes, but if it doesn't for some
                    # reason we can just silently skip this step.
                    OldCrystal := EnergyBladeEntity.GetKyberCrystalComponent[]

                then:
                    OldCrystal.RemoveFromEntity()

                    # Workaround for a race conditon when trying to replace a component.
                    # If you try to remove a component and then replace it with another
                    # sometimes the remove and add replicate in the wrong order.
                    # If the client receives "add then remove" the add will fail because
                    # there is already a component of that type.
                    # Once that is fixed this sleep command can be removed and
                    # the <suspends> attribute could be removed from the function.
                    Sleep(0.1)

                if:
                    NewCrystal := CreateKyberCrystalComponent[EnergyBladeEntity, KyberCrystal]

                    set NewCrystal.ArcIntensity = ArcIntensityOverride
                    set NewCrystal.Stability = StabilityOverride

                then:
                    EnergyBladeEntity.AddComponents(array { NewCrystal })
                    
                else:
                    Print("Failed to replace Kyber Crystal")
                    # Prevent entity leak. Destroy EnergyBlade.
                    EnergyBladeEntity.RemoveFromParent()

                Item := case(PowerOverride):
                    Telekinetic_power.Cloak => TelekineticAbilityTrick_EnergyBladeAugment_V1 {}
                    Telekinetic_power.Decoy => TelekineticAbilityDistraction_EnergyBladeAugment_V1 {}
                    Telekinetic_power.Heal => TelekineticAbilityHeal_EnergyBladeAugment_V1 {}
                    Telekinetic_power.Sense => TelekineticAbilityHeal_EnergyBladeAugment_V1 {}
                    Telekinetic_power.Shield => TelekineticAbilitieshield_EnergyBladeAugment_V1 {}
                    Telekinetic_power.Throw => TelekineticAbilityThrow_EnergyBladeAugment_V1 {}
                    Telekinetic_power.Zap => TelekineticAbilityThrow_EnergyBladeAugment_V1 {}
                    Telekinetic_power.LightningAugment => TelekineticAbilityLightning_EnergyBladeAugment_V1 {}
                    Telekinetic_power.ThrowAugment => TelekineticAbilitiesaberThrow_EnergyBladeAugment_V1 {}
                    Telekinetic_power.TrainingAugment => TelekineticAbilityTraining_Augment_V1 {}
                    Telekinetic_power.PullAugment => TelekineticAbilityPull_EnergyBladeAugment_V1 {}
                    Telekinetic_power.PushAugment => TelekineticAbilityPush_EnergyBladeAugment_V1 {}
                    _ => TelekineticAbilityPush_EnergyBladeAugment_V1 {}

                Result := InInventory.AddItemDistribute(Item);
                
            if:
                Result := InInventory.AddItemDistribute(EnergyBladeEntity)
                Result.GetSuccess[]

            else:
                Print("Failed to add EnergyBlade to inventory")
                # Prevent entity leak. Destroy EnergyBlade.
                EnergyBladeEntity.RemoveFromParent()
                return
                
            if:
                Item := EnergyBladeEntity.GetEnergyBladeItemComponent[]

            then:
                Result := Item.Equip()
                # Result.GetSuccess[]

            else:
                Print("Failed to equip EnergyBlade")
                
        else:
            Print("Failed to create EnergyBlade")
