# Knowledge-neutral Verse snippet; project names and asset IDs generalized.
# id: snippet-074

using { /Fortnite.com/Devices }
using { /Fortnite.com/SPACE_OPERA }
using { /Verse.org/SceneGraph }
using { /Verse.org/Simulation }
using { /UnrealEngine.com/Itemization }
using { /UnrealEngine.com/Temporary/Diagnostics }

using { EnergyBladeWorkbenchCore }

# Device to change the stability of the player's currently equipped EnergyBlade
EnergyBlade_stability_changer_device := class(creative_device):

    # Button to bind the "Increase Stability" event to
    @editable
    IncreaseButton : button_device = button_device {}

    # Button to bind the "Decrease Stability" event to
    @editable
    DecreaseButton : button_device = button_device {}

    OnBegin<override>():void=

        IncreaseButton.InteractedWithEvent.Subscribe(IncreaseButton_OnButtonInteractedWith)
        DecreaseButton.InteractedWithEvent.Subscribe(DecreaseButton_OnButtonInteractedWith)

    # Triggers when a player interacts with the IncreaseButton
    IncreaseButton_OnButtonInteractedWith<private>(InAgent:agent):void=

        if:
            Inventory := InAgent.GetInventoryComponent[]

        then:
            Items := ChangeEnergyBlades_Stability(Inventory, 0.1)

            Print("Changed {Items.Length} item(s)")

    # Triggers when a player interacts with the DecreaseButton
    DecreaseButton_OnButtonInteractedWith<private>(InAgent:agent):void=

        if:
            Inventory := InAgent.GetInventoryComponent[]

        then:
            Items := ChangeEnergyBlades_Stability(Inventory, -0.1)

            Print("Changed {Items.Length} item(s)")
            
    ChangeEnergyBlades_Stability(InInventory:inventory_component, Ammount:float):[]entity=

        Items := for:
            Item : InInventory.FindItems()
            
            # We only want to modify the EnergyBlade that is currently in the player's hand
            ItemComponent := Item.GetEnergyBladeItemComponent[]
            ItemComponent.IsEquipped[]

            Crystal := Item.GetKyberCrystalComponent[]
        do:
            set Crystal.Stability += Ammount

            Print("Stability: {Crystal.Stability}")

            Item
