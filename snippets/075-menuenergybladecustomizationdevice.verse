# Knowledge-neutral Verse snippet; project names and asset IDs generalized.
# id: snippet-075

using { /Fortnite.com/Devices }
using { /Fortnite.com/UI }
using { /UnrealEngine.com/Temporary/UI }
using { /UnrealEngine.com/Temporary/SpatialMath }
using { /Verse.org/Simulation }
using { /Verse.org/Assets }
using { /Verse.org/Colors }
using { /Fortnite.com/SPACE_OPERA }
using { /Verse.org/SceneGraph }
using { /UnrealEngine.com/Itemization }
using { /UnrealEngine.com/Temporary/Diagnostics }

using { EnergyBladeWorkbenchCore }

# ============================================================================
# EnergyBlade & telekinetic ability Customization
# 
# This script is divided into three main parts:
# 1. A device that handles triggers, camera/HUD overrides, and updates the player's inventory with their chosen items and components. 
# 2. A session manager that tracks individual player UI states, handles tab navigation, and slider/grid inputs. 
# 3. A UI helper that dynamically builds grids, buttons, and slider pages using Verse UI.
# ============================================================================

# ================================================================
# 1. MAIN DEVICE
# Handles world interactions, camera/HUDstates,and applies the inventory/component changes to the player.
# ================================================================
menu_EnergyBlade_customization_device<public> := class(creative_device):

    @editable ShowTrigger : trigger_device = trigger_device{}
    @editable HUDController : hud_controller_device = hud_controller_device{}
    @editable FixedAngleCamera : gameplay_camera_fixed_angle_device = gameplay_camera_fixed_angle_device{}
    
    # to track active customization sessions for each player
    var ActiveSessions : [player]EnergyBlade_customization_session = map{}

    OnBegin<override>()<suspends> : void =
        ShowTrigger.TriggeredEvent.Subscribe(OnShowTriggered)

    OnShowTriggered(Instigator : ?agent) : void =
        if (TriggeringPlayer := player[Instigator?]):
            
            HUDController.Enable()
            FixedAngleCamera.Enable()

            Session := GetOrCreateSession(TriggeringPlayer)
            Session.Open()

    # Gets existing session or creates a new one if it doesn't exist
    GetOrCreateSession(TargetPlayer : player) : EnergyBlade_customization_session =
        if (ExistingSession := ActiveSessions[TargetPlayer]) then return ExistingSession

        NewSession := EnergyBlade_customization_session{ OwningDevice := Self, SessionPlayer := TargetPlayer }
        if (set ActiveSessions[TargetPlayer] = NewSession) {}
        
        NewSession.Initialize()

        return NewSession

    OnUIClosed<public>(TargetPlayer : player) : void =
        HUDController.Disable()
        FixedAngleCamera.Disable()

    # Applies the selected customization options to the player inventory
    ApplyLoadoutFor(TargetPlayer : player, LoadoutSelection : EnergyBlade_customization_selection)<suspends> : void =

        var ModifyExistingEnergyBlade : logic = false
        Session := GetOrCreateSession(TargetPlayer)

        # If we are only modifying Arcing and Stability, we can just modify our existing EnergyBlade.
        if(ExistingLoadout := Session.PreviouslyAppliedLoadout?):
            if(ExistingLoadout.HiltID = LoadoutSelection.HiltID, ExistingLoadout.CrystalID = LoadoutSelection.CrystalID, ExistingLoadout.AbilityID = LoadoutSelection.AbilityID):
                set ModifyExistingEnergyBlade = true

        if (PlayerInventory := TargetPlayer.GetInventoryComponent[]):
            # If we can just update the existing one, update only arcing and stabilty, no need to grant a new EnergyBlade.
            if(ModifyExistingEnergyBlade?):
                for (Item : PlayerInventory.FindItems(), Item.GetComponent[EnergyBlade_item_component]):
                    if(KyberComp := Item.GetComponent[kyber_crystal_component]):
                        set KyberComp.ArcIntensity = LoadoutSelection.Arcing
                        set KyberComp.Stability = LoadoutSelection.Stability
            else:   
                # Clean up existing EnergyBlades and powers in inventory
                RemoveItems(PlayerInventory, kyber_crystal_component)
                RemoveItems(PlayerInventory, Telekinetic_power_item_component)

                # Create Entity locally
                HiltID := if (LoadoutSelection.HiltID < 0) then ui_config.DefaultHiltID else LoadoutSelection.HiltID
                NewEnergyBlade := ui_config.CreateEnergyBlade(HiltID)
                
                # Remove the default component
                ExistingComponents := NewEnergyBlade.GetComponents()
                var FoundExisting : logic = false
                for (Comp : ExistingComponents):
                    if (KyberComp := kyber_crystal_component[Comp]):
                        KyberComp.RemoveFromEntity()
                        set FoundExisting = true
            
                if (FoundExisting?):
                    Sleep(0.1) 

                # Add New Crystal
                CrystalID := if (LoadoutSelection.CrystalID < 0) then ui_config.DefaultCrystalID else LoadoutSelection.CrystalID

                if (NewCrystal := ui_config.GetColorFromCrystalID(CrystalID, NewEnergyBlade)?):

                    # Apply settings to the component object before attachment
                    set NewCrystal.ArcIntensity = LoadoutSelection.Arcing
                    set NewCrystal.Stability = LoadoutSelection.Stability
                    
                    # Attach new component
                    NewEnergyBlade.AddComponents(array{NewCrystal})

                # Distribute the item
                if (PlayerInventory.AddItemDistribute(NewEnergyBlade).GetSuccess[]):
                    
                    # Telekinetic refresh
                    if (ItemComponent := NewEnergyBlade.GetComponent[EnergyBlade_item_component]):
                        
                        # Telekinetic settings onto the live component again to be safe
                        if (LiveCrystal := NewEnergyBlade.GetComponent[kyber_crystal_component]):
                            set LiveCrystal.ArcIntensity = LoadoutSelection.Arcing
                            set LiveCrystal.Stability = LoadoutSelection.Stability

                        # Telekinetic the game to reload visual actor
                        ItemComponent.Unequip()
                        Sleep(0.1)
                        ItemComponent.Equip()

                # Grant Ability
                AbilityID := LoadoutSelection.AbilityID
                if (AbilityID > -1, ui_config.GivePower(PlayerInventory, AbilityID)?) {}

        set Session.PreviouslyAppliedLoadout = option{LoadoutSelection}

    # Remove all items containing a specific component type
    RemoveItems(PlayerInventory:inventory_component, ComponentType:subtype(component)):void =
        for (Item : PlayerInventory.FindItems(), Item.GetComponent[ComponentType], Result := PlayerInventory.RemoveItem(Item), Result.GetSuccess[]) {}

# Struct to hold the final data chosen by the player
EnergyBlade_customization_selection := struct:
    HiltID : int
    CrystalID : int
    AbilityID : int
    Stability : float
    Arcing : float

# ================================================================
# 2. PLAYER UI STATE MANAGER
# Instantiated per-player. Handles the menu's logic, tab navigation, and real-time updates.
# ================================================================
EnergyBlade_customization_session := class:
    OwningDevice : menu_EnergyBlade_customization_device
    SessionPlayer : player
    var PreviouslyAppliedLoadout : ?EnergyBlade_customization_selection = false
    var CloseButton : ui_helper.Buttons.SaveButton = ui_helper.Buttons.SaveButton{}
    var UpdateToken : int = 0
    var RootOverlayWidget : overlay = overlay{}
    
    # Content pages
    var Page_Hilt : ui_helper.Pages.GridContentPage = ui_helper.Pages.GridContentPage{}
    var Page_Crystal : ui_helper.Pages.GridContentPage = ui_helper.Pages.GridContentPage{}
    var Page_Ability : ui_helper.Pages.GridContentPage = ui_helper.Pages.GridContentPage{}
    var Page_Stability : ui_helper.Pages.SliderContentPage = ui_helper.Pages.SliderContentPage{}
    var Page_Arcing : ui_helper.Pages.SliderContentPage = ui_helper.Pages.SliderContentPage{}

    var TabButtons : [int]ui_helper.Buttons.TabButton = map{}
    
    var ContentPages : [int]ui_helper.Pages.BaseContentPage = map{}
    var ActiveTabID : int = -1

    Initialize() : void =
        InitComponents()
        BuildVisuals()
        SelectTab(ui_config.TabID_Hilt)

        CloseButton.OnClick().Subscribe(OnCloseClicked)
        CloseButton.OnHover().Subscribe(OnCloseHighlight)
        CloseButton.OnUnhover().Subscribe(OnCloseUnhighlight)

    Open() : void =
        if (PlayerUI := GetPlayerUI[SessionPlayer]):
            PlayerUI.RemoveWidget(RootOverlayWidget)
            PlayerUI.AddWidget(RootOverlayWidget, player_ui_slot{ InputMode := ui_input_mode.All })

            # Re-apply button content so images/materials show after widget was removed and re-added
            Page_Hilt.RefreshButtonContent()
            Page_Crystal.RefreshButtonContent()
            Page_Ability.RefreshButtonContent()

            # needed for gamepad support
            PlayerUI.SetFocus(RootOverlayWidget)

    Close() : void =
        if (PlayerUI := GetPlayerUI[SessionPlayer]):
            PlayerUI.RemoveWidget(RootOverlayWidget)

        OwningDevice.OnUIClosed(SessionPlayer)

    OnCloseHighlight(Message:widget_message):void = CloseButton.SetHighlighted(true)
    OnCloseUnhighlight(Message:widget_message):void = CloseButton.SetHighlighted(false)

    OnCloseClicked(Message : widget_message) : void =
        if (Message.Player = SessionPlayer):
            Close()

    InitComponents() : void =
        Page_Hilt.Init(ui_config.GetHilts(), ui_config.ColumnsHilt, ui_config.DefaultHiltID, 0)
        Page_Crystal.Init(ui_config.Crystals(), ui_config.ColumnsCrystal, ui_config.DefaultCrystalID, 1)
        
        RemoveItem := ui_config.MakeItem(-1, ui_config.None, ui_config.Empty)
        FullAbilityList := array{RemoveItem} + ui_config.Abilities()
        Page_Ability.Init(FullAbilityList, ui_config.ColumnsAbility, ui_config.DefaultAbilityID, 2)
        
        Page_Stability.Init(ui_config.DefaultStability, ui_config.StepStability, ui_config.Label_Stability, ui_config.CrystalDesc)
        Page_Arcing.Init(ui_config.DefaultArcing, ui_config.StepArcing, ui_config.Label_Arcing, ui_config.ArcDesc)

        Page_Hilt.SetCallback(OnRealTimeUpdateInt)
        Page_Crystal.SetCallback(OnRealTimeUpdateInt)
        Page_Ability.SetCallback(OnRealTimeUpdateInt)
        Page_Stability.SetCallback(OnRealTimeUpdateFloat)
        Page_Arcing.SetCallback(OnRealTimeUpdateFloat)

        set ContentPages = map: 
            ui_config.TabID_Hilt => Page_Hilt
            ui_config.TabID_Crystal => Page_Crystal
            ui_config.TabID_Ability => Page_Ability
            ui_config.TabID_Stability => Page_Stability
            ui_config.TabID_Arcing => Page_Arcing
            
        for (ID -> Page : ContentPages):
            if (set TabButtons[ID] = ui_helper.Buttons.TabButton{}) {}

    # Wrapper for callbacks coming from Grid Pages
    OnRealTimeUpdateInt(Value:int) : void = 
        UpdateRealTime()

    # Wrapper for callbacks coming from Sliders
    OnRealTimeUpdateFloat(Value:float) : void = 
        UpdateRealTime()

    # Gathers current state and spawns apply logic immediately
    UpdateRealTime() : void =
        # Increment token. This invalidates any previous pending updates
        set UpdateToken += 1
        CurrentToken := UpdateToken
        
        # Gather data
        Selection := EnergyBlade_customization_selection{
            HiltID := Page_Hilt.GetSelectedID(), 
            CrystalID := Page_Crystal.GetSelectedID(), 
            AbilityID := Page_Ability.GetSelectedID(),
            Stability := Page_Stability.GetValue(), 
            Arcing := Page_Arcing.GetValue()
        }

        # Spawn a task that waits before applying
        spawn { ApplyBufferedLoadout(CurrentToken, Selection) }

    ApplyBufferedLoadout(Token : int, Selection : EnergyBlade_customization_selection)<suspends> : void =
        # Check if the token is still the same
        # If the user clicked again during the Sleep, UpdateToken will be higher, so we abort
        if (Token = UpdateToken):
            OwningDevice.ApplyLoadoutFor(SessionPlayer, Selection)

    SelectTab(TargetTabID : int) : void =
        set ActiveTabID = TargetTabID

        for (TabID -> Btn : TabButtons):
            Btn.SetSelected(if (TabID = TargetTabID) then true else false)

        for (TabID -> Page : ContentPages):
            if (TabID = TargetTabID) then Page.Show() else Page.Hide()

    BuildVisuals() : void =       
        SidebarColumn := stack_box{ Orientation := orientation.Vertical }
        SidebarColumn.AddWidget(stack_box_slot{ Widget := UI.Widgets.UW_Header{}, Padding := margin{ Left := -27.0, Bottom := 0.0 }})
        
        AddTabToColumn(SidebarColumn, ui_config.TabID_Hilt, ui_config.Label_Hilt)
        AddTabToColumn(SidebarColumn, ui_config.TabID_Crystal, ui_config.Label_Crystal)
        AddTabToColumn(SidebarColumn, ui_config.TabID_Ability, ui_config.Label_Ability)
        AddTabToColumn(SidebarColumn, ui_config.TabID_Stability, ui_config.Label_Stability)
        AddTabToColumn(SidebarColumn, ui_config.TabID_Arcing, ui_config.Label_Arcing)

        SidebarColumn.AddWidget(stack_box_slot{ Widget := CloseButton.Build(), Padding := margin{Top:=8.0}, HorizontalAlignment := horizontal_alignment.Fill })

        ContentOverlay := overlay{}
        for (ID -> Page : ContentPages):
            ContentOverlay.AddWidget(overlay_slot{ Widget := Page.BuildWidget(), HorizontalAlignment := horizontal_alignment.Fill, VerticalAlignment := vertical_alignment.Fill })

        InnerLayout := overlay{ Slots := array{
            overlay_slot{ Widget := SidebarColumn, HorizontalAlignment := horizontal_alignment.Left, VerticalAlignment := vertical_alignment.Bottom },
            overlay_slot{ Widget := ContentOverlay, HorizontalAlignment := horizontal_alignment.Right, VerticalAlignment := vertical_alignment.Bottom }
        }}

        set RootOverlayWidget = overlay{ Slots := array{
            overlay_slot{ 
                Widget := UI.Widgets.UW_Overlay{}, 
                HorizontalAlignment := horizontal_alignment.Fill, 
                VerticalAlignment := vertical_alignment.Fill 
            },
            overlay_slot{ 
                Widget := InnerLayout, 
                HorizontalAlignment := horizontal_alignment.Fill, 
                VerticalAlignment := vertical_alignment.Fill, 
                Padding := margin{ Left := 48.0, Right := 48.0, Top := 48.0, Bottom := 48.0 } }
        }}

    AddTabToColumn(ColumnStack:stack_box, TabID:int, Label:message):void =
        if (TabBtn := TabButtons[TabID]):
            ColumnStack.AddWidget(stack_box_slot{ Widget := TabBtn.Build(Label), Padding := margin{ Top := 0.0} })
            
            Handler := TabInteractionHandler{ Session:=Self, TabID:=TabID, Button:=TabBtn }
            TabBtn.OnClick().Subscribe(Handler.OnClicked)
            TabBtn.OnHover().Subscribe(Handler.OnHover)
            TabBtn.OnUnhover().Subscribe(Handler.OnUnhover)
            
            if (Page := ContentPages[TabID]):
                Page.SetTabLink(TabBtn)

# Helper class to capture state for tab button events
TabInteractionHandler := class:
    Session : EnergyBlade_customization_session
    TabID : int
    Button : ui_helper.Buttons.TabButton
    
    OnClicked(Message:widget_message):void = Session.SelectTab(TabID)
    OnHover(Message:widget_message):void = Button.SetHighlighted(true)
    OnUnhover(Message:widget_message):void = 
        # Only unhighlight if this isn't the currently active tab
        if (TabID <> Session.ActiveTabID) then Button.SetHighlighted(false)

# ================================================================
# 3. UI HELPER
# Classes used to construct grid layouts, buttons, and slider pages from custom UMG widgets.
# ================================================================
ui_helper := module:
    # Helper to arrange widgets in a grid layout
    GridBuilder<public> := class:
        var VerticalStack<public> : stack_box = stack_box{ Orientation := orientation.Vertical }
        var ActiveRowStack<public> : stack_box = stack_box{ Orientation := orientation.Horizontal }
        var ItemCountInRow<public> : int = 0
        MaxColumns<public> : int = 4

        Setup<public>() : void =
            set VerticalStack = stack_box{ Orientation := orientation.Vertical }
            set ActiveRowStack = stack_box{ Orientation := orientation.Horizontal }
            set ItemCountInRow = 0
            VerticalStack.AddWidget(stack_box_slot{ Widget := ActiveRowStack })

        Add<public>(WidgetToAdd : widget) : void =
            # If row is full, create a new horizontal stack
            if (ItemCountInRow >= MaxColumns):
                set ActiveRowStack = stack_box{ Orientation := orientation.Horizontal }
                set ItemCountInRow = 0
                VerticalStack.AddWidget(stack_box_slot{ Widget := ActiveRowStack, Padding := margin{ Top := 16.0 }, HorizontalAlignment := horizontal_alignment.Left })
            
            ActiveRowStack.AddWidget(stack_box_slot{ Widget := WidgetToAdd, Padding := margin{ Right := 16.0 } })
            set ItemCountInRow += 1

    Buttons<public> := module:
        BaseButton<public> := class:
            var RootWidget : overlay = overlay{}
            InteractableButton : button = button{ Slot := button_slot{ Widget := overlay{} } }
            
            BuildBase<protected>(VisualWidget : widget) : void =
                set RootWidget = overlay{ Slots := array{ 
                    overlay_slot{ Widget := VisualWidget, HorizontalAlignment := horizontal_alignment.Fill, VerticalAlignment := vertical_alignment.Fill }, 
                    overlay_slot{ Widget := InteractableButton, HorizontalAlignment := horizontal_alignment.Fill, VerticalAlignment := vertical_alignment.Fill } 
                }}
            
            GetWidget<public>() : widget = RootWidget
            OnClick<public>() : listenable(widget_message) = InteractableButton.OnClick()
            OnHover<public>() : listenable(widget_message) = InteractableButton.HighlightEvent()
            OnUnhover<public>() : listenable(widget_message) = InteractableButton.UnhighlightEvent()

        # Base class for buttons that have selection states
        SelectionButtonWrapper<public> := class(BaseButton):
            SetSelected<public>(IsSelected:logic):void = {}
            SetHighlighted<public>(IsHighlighted:logic):void = {}
            RefreshContent<public>(ItemData:ui_config.item_data):void = {}

        HiltButton<public> := class(SelectionButtonWrapper):
            UMGWidget : UI.Widgets.UW_Item = UI.Widgets.UW_Item{}
            
            Init<public>(ItemData:ui_config.item_data):void =
                set UMGWidget.SelectionName = ItemData.Label
                SetHiltMaterial(ItemData.ID)
                BuildBase(UMGWidget)
            
            SetHiltMaterial<private>(HiltID:int):void =
                if (HiltID = 0) then set UMGWidget.SelectionMaterial = UI.Materials.M_Hilt_EnergyBladeHeroC_ROTJ{}
                if (HiltID = 1) then set UMGWidget.SelectionMaterial = UI.Materials.M_Hilt_EnergyBladeHeroB{}
                if (HiltID = 2) then set UMGWidget.SelectionMaterial = UI.Materials.M_Hilt_EnergyBladeHeroA{}
                if (HiltID = 3) then set UMGWidget.SelectionMaterial = UI.Materials.M_Hilt_VillainA{}
                if (HiltID = 4) then set UMGWidget.SelectionMaterial = UI.Materials.M_Hilt_VillainB{}
                if (HiltID = 5) then set UMGWidget.SelectionMaterial = UI.Materials.M_Hilt_Mentor{}
                if (HiltID = 6) then set UMGWidget.SelectionMaterial = UI.Materials.M_Hilt_PadawanBlue{}
                if (HiltID = 7) then set UMGWidget.SelectionMaterial = UI.Materials.M_Hilt_PadawanGreen{}
                if (HiltID = 8) then set UMGWidget.SelectionMaterial = UI.Materials.M_Hilt_Training{}
                if (HiltID = 9) then set UMGWidget.SelectionMaterial = UI.Materials.M_Hilt_OpposingApprentice{}
                
            SetSelected<override>(IsSelected:logic):void = { set UMGWidget.isSelected = IsSelected; set UMGWidget.isHighlighted = IsSelected }
            SetHighlighted<override>(IsHighlighted:logic):void = { set UMGWidget.isHighlighted = IsHighlighted }
            RefreshContent<override>(ItemData:ui_config.item_data):void =
                set UMGWidget.SelectionName = ItemData.Label
                SetHiltMaterial(ItemData.ID)

        CrystalButton<public> := class(SelectionButtonWrapper):
            UMGWidget : UI.Widgets.UW_Selection_Crystal = UI.Widgets.UW_Selection_Crystal{}
            
            Init<public>(ItemData:ui_config.item_data):void =
                set UMGWidget.CrystalColor = ui_config.CrystalIDToUMGIndex(ItemData.ID)
                BuildBase(UMGWidget)
                
            SetSelected<override>(IsSelected:logic):void = { set UMGWidget.isSelected = IsSelected; set UMGWidget.isHighlighted = IsSelected }
            SetHighlighted<override>(IsHighlighted:logic):void = { set UMGWidget.isHighlighted = IsHighlighted }
            RefreshContent<override>(ItemData:ui_config.item_data):void =
                set UMGWidget.CrystalColor = ui_config.CrystalIDToUMGIndex(ItemData.ID)

        AbilityButton<public> := class(SelectionButtonWrapper):
            UMGWidget : UI.Widgets.UW_Item = UI.Widgets.UW_Item{}

            Init<public>(ItemData:ui_config.item_data):void =
                set UMGWidget.SelectionName = ItemData.Label
                set UMGWidget.SelectionImage = ItemData.Texture
                set UMGWidget.isRemoveAll = if (ItemData.ID = -1) then true else false
                SetAbilityMaterial(ItemData.ID)
                BuildBase(UMGWidget)
            
            SetAbilityMaterial<private>(AbilityID:int):void =
                if (AbilityID = 200) then set UMGWidget.SelectionMaterial = UI.Materials.M_TelekineticAbility_Push{}
                if (AbilityID = 201) then set UMGWidget.SelectionMaterial = UI.Materials.M_TelekineticAbility_Pull{}
                if (AbilityID = 202) then set UMGWidget.SelectionMaterial = UI.Materials.M_TelekineticAbility_Throw{}
                if (AbilityID = 203) then set UMGWidget.SelectionMaterial = UI.Materials.M_TelekineticAbility_Barrier{}
                if (AbilityID = 204) then set UMGWidget.SelectionMaterial = UI.Materials.M_TelekineticAbility_Lightning{}
                if (AbilityID = 205) then set UMGWidget.SelectionMaterial = UI.Materials.M_TelekineticAbility_Heal{}
                if (AbilityID = 206) then set UMGWidget.SelectionMaterial = UI.Materials.M_TelekineticAbility_Awareness{}
                if (AbilityID = 207) then set UMGWidget.SelectionMaterial = UI.Materials.M_TelekineticAbility_SaberThrow{}
                if (AbilityID = 208) then set UMGWidget.SelectionMaterial = UI.Materials.M_TelekineticAbility_Distraction{}
                if (AbilityID = 209) then set UMGWidget.SelectionMaterial = UI.Materials.M_TelekineticAbility_MindTrick{}

            SetSelected<override>(IsSelected:logic):void = { set UMGWidget.isSelected = IsSelected; set UMGWidget.isHighlighted = IsSelected }
            SetHighlighted<override>(IsHighlighted:logic):void = { set UMGWidget.isHighlighted = IsHighlighted }
            RefreshContent<override>(ItemData:ui_config.item_data):void =
                set UMGWidget.SelectionName = ItemData.Label
                set UMGWidget.SelectionImage = ItemData.Texture
                set UMGWidget.isRemoveAll = if (ItemData.ID = -1) then true else false
                SetAbilityMaterial(ItemData.ID)
        
        TabButton<public> := class(BaseButton):
            UMGWidget : UI.Widgets.UW_Tab = UI.Widgets.UW_Tab{}
            
            Build<public>(Name : message) : widget =
                set UMGWidget.TabName = Name
                set UMGWidget.SelectionName = ui_config.Empty
                BuildBase(UMGWidget)
                return GetWidget()
                
            SetSelected<public>(IsSelected:logic):void = { set UMGWidget.isSelected = IsSelected; set UMGWidget.isHighlighted = IsSelected }
            SetHighlighted<public>(IsHighlighted:logic):void = { set UMGWidget.isHighlighted = IsHighlighted }
            SetSelectionName<public>(Name:message):void = { set UMGWidget.SelectionName = Name }

        SaveButton<public> := class(BaseButton):
            UMGWidget : UI.Widgets.UW_SaveButton = UI.Widgets.UW_SaveButton{}
            Build<public>() : widget = { BuildBase(UMGWidget); return GetWidget() }
            SetHighlighted<public>(IsHighlighted:logic):void = { set UMGWidget.isHighlighted = IsHighlighted }

        IncDecButton<public> := class(BaseButton):
            UMGWidget : UI.Widgets.UW_Button_IncDec = UI.Widgets.UW_Button_IncDec{}

            Init<public>(IsDecrement:logic):void =
                set UMGWidget.isDecrement = IsDecrement
                BuildBase(UMGWidget)
                InteractableButton.HighlightEvent().Subscribe(OnVisualHover)
                InteractableButton.UnhighlightEvent().Subscribe(OnVisualUnhover)

            # Toggles the boolean to trigger the animation
            PlayClickAnim<public>():void =
                if (UMGWidget.Clicked?) then set UMGWidget.Clicked = false else set UMGWidget.Clicked = true

            OnVisualHover(M:widget_message):void = { set UMGWidget.Highlighted = true }
            OnVisualUnhover(M:widget_message):void = { set UMGWidget.Highlighted = false }

    Pages<public> := module:
        
        BaseContentPage<public> := class:
            var RootContainer : overlay = overlay{}
            DetailsPanel : UI.Widgets.UW_Details = UI.Widgets.UW_Details{}
            var AssociatedTabButton : ?Buttons.TabButton = false
            
            BuildWidget<public>() : widget = RootContainer
            Show<public>() : void = RootContainer.SetVisibility(widget_visibility.Visible)
            Hide<public>() : void = RootContainer.SetVisibility(widget_visibility.Collapsed)
            
            SetTabLink<public>(Tab:Buttons.TabButton):void = 
                set AssociatedTabButton = option{Tab}
                RefreshTabName()

            UpdateDetails(Title : message, Description : message) : void =
                set DetailsPanel.Title = Title
                set DetailsPanel.Description = Description

            # Helper to wrap the page content with the details panel on the side
            WrapWithDetails(ContentWidget : widget) : void =
                Stack := stack_box{ Orientation := orientation.Vertical }
                Stack.AddWidget(stack_box_slot{ Widget := DetailsPanel, Padding := margin{ Bottom := 20.0 }, HorizontalAlignment := horizontal_alignment.Right })
                Stack.AddWidget(stack_box_slot{ Widget := ContentWidget, Padding := margin{ Bottom := 100.0 }, HorizontalAlignment := horizontal_alignment.Right, VerticalAlignment := vertical_alignment.Fill })
                set RootContainer = overlay{ Slots := array{ overlay_slot{ Widget := Stack } } }
            
            RefreshTabName<public>():void = {}

        GridContentPage<public> := class(BaseContentPage):
            var Items : []ui_config.item_data = array{}
            var GridButtons : [int]Buttons.SelectionButtonWrapper = map{}
            var SelectedID : int = 0
            var ItemType : int = 0 
            var Callback : ?type{_(:int):void} = false
            var Grid : GridBuilder = GridBuilder{}

            Init<public>(Data:[]ui_config.item_data, ColumnCount:int, DefaultID:int, TargetType:int):void =
                set Items = Data
                set SelectedID = DefaultID
                set ItemType = TargetType
                
                set Grid = GridBuilder{ MaxColumns := ColumnCount }
                Grid.Setup()

                for (Item : Items):
                    var NewButton : Buttons.SelectionButtonWrapper = Buttons.HiltButton{}
                    
                    # Instantiate specific button types based on the page configuration
                    if (ItemType = 0):
                        HiltBtn := Buttons.HiltButton{}; HiltBtn.Init(Item); set NewButton = HiltBtn
                    else if (ItemType = 1):
                        CrystalBtn := Buttons.CrystalButton{}; CrystalBtn.Init(Item); set NewButton = CrystalBtn
                    else:
                        AbilityBtn := Buttons.AbilityButton{}; AbilityBtn.Init(Item); set NewButton = AbilityBtn

                    if (set GridButtons[Item.ID] = NewButton) {}
                    
                    Handler := ItemInteractionHandler{ Page:=Self, ID:=Item.ID, Button:=NewButton }
                    NewButton.OnClick().Subscribe(Handler.OnClick)
                    NewButton.OnHover().Subscribe(Handler.OnHover)
                    NewButton.OnUnhover().Subscribe(Handler.OnUnhover)

                    Grid.Add(NewButton.GetWidget())
                
                WrapWithDetails(Grid.VerticalStack)
                SelectItem(SelectedID)

            RefreshButtonContent<public>():void =
                for (Item : Items):
                    if (Btn := GridButtons[Item.ID]):
                        Btn.RefreshContent(Item)

            SelectItem<public>(TargetID:int):void =
                set SelectedID = TargetID
                
                # Update visual state for all buttons in the grid
                for (ButtonID -> Button : GridButtons):
                     IsSelected := if (ButtonID = TargetID) then true else false
                     Button.SetSelected(IsSelected)
                     
                     # Telekinetic unhighlight if not selected to clean up hover states
                     if (IsSelected?):
                     else:
                        Button.SetHighlighted(false)
                
                RefreshDetails()
                UpdateTabName(TargetID)
                if (ExecuteCallback := Callback?) then ExecuteCallback(TargetID)

            HighlightItem<public>(TargetID:int, State:logic):void =
                if (TargetID <> SelectedID, Button := GridButtons[TargetID]) then Button.SetHighlighted(State)
                if (State?) then ShowDetailsFor(TargetID) else RefreshDetails()

            ShowDetailsFor(TargetID:int):void =
                Label := ui_config.GetLabel(Items, TargetID)
                var Description : message = ui_config.Empty
                for (Item : Items, Item.ID = TargetID):
                    set Description = Item.Description
                UpdateDetails(Label, Description)

            RefreshDetails():void = ShowDetailsFor(SelectedID)
            
            UpdateTabName(TargetID:int):void =
                if (Tab := AssociatedTabButton?) then Tab.SetSelectionName(if (TargetID = -1) then ui_config.None else ui_config.GetLabel(Items, TargetID))

            RefreshTabName<override>():void = UpdateTabName(SelectedID)

            GetSelectedID<public>():int = SelectedID
            SetCallback<public>(CB:type{_(:int):void}):void = { set Callback = option{CB} }

        ItemInteractionHandler := class:
            Page : GridContentPage
            ID : int
            Button : Buttons.SelectionButtonWrapper
            
            OnClick(Message:widget_message):void = Page.SelectItem(ID)
            OnHover(Message:widget_message):void = Page.HighlightItem(ID, true)
            OnUnhover(Message:widget_message):void = Page.HighlightItem(ID, false)

        SliderContentPage<public> := class(BaseContentPage):
            var CurrentValue : float = 0.0
            var StepAmount : float = 0.1
            var Callback : ?type{_(:float):void} = false
            
            var GaugeWidget : UI.Widgets.UW_GaugeSlider = UI.Widgets.UW_GaugeSlider{}
            
            GaugeMaxSegments : float = 10.0

            DecrementButton : Buttons.IncDecButton = Buttons.IncDecButton{}
            IncrementButton : Buttons.IncDecButton = Buttons.IncDecButton{}

            Init<public>(DefaultValue:float, Step:float, Title:message, Description:message):void =
                set CurrentValue = DefaultValue
                set StepAmount = Step
                
                UpdateGaugeVisuals()

                DecrementButton.Init(true)
                IncrementButton.Init(false)
                
                DecrementButton.OnClick().Subscribe(OnDecrement)
                IncrementButton.OnClick().Subscribe(OnIncrement)
                
                MainStack := stack_box{ Orientation := orientation.Vertical }
                
                MainStack.AddWidget(stack_box_slot{ 
                    Widget := GaugeWidget, 
                    HorizontalAlignment := horizontal_alignment.Center, 
                    Padding := margin{ Bottom := -100.0, Top := -70.0 } 
                    })
                    
                ControlsStack := stack_box{ Orientation := orientation.Horizontal }
                ControlsStack.AddWidget(stack_box_slot{ Widget := DecrementButton.GetWidget() })
                ControlsStack.AddWidget(stack_box_slot{ Widget := IncrementButton.GetWidget(), Padding := margin{ Left := 20.0 } })

                MainStack.AddWidget(stack_box_slot{ Widget := ControlsStack, HorizontalAlignment := horizontal_alignment.Center, Padding := margin{ Top := -40.0 } })

                UpdateDetails(Title, Description)
                WrapWithDetails(MainStack)
                UpdateTab()

            OnDecrement(Message:widget_message):void = 
                DecrementButton.PlayClickAnim()
                Modify(-1.0)

            OnIncrement(Message:widget_message):void = 
                IncrementButton.PlayClickAnim()
                Modify(1.0)

            Modify(Multiplier:float):void =
                SafeCurrentSegment := if (R := Round[CurrentValue * GaugeMaxSegments]) then R * 1.0 else 0.0
                
                NewSegment := SafeCurrentSegment + Multiplier
                
                ClampedSegment := if (NewSegment < 0.0) { 0.0 } else if (NewSegment > GaugeMaxSegments) { GaugeMaxSegments } else { NewSegment }
                
                set CurrentValue = ClampedSegment / GaugeMaxSegments
                
                UpdateGaugeVisuals()
                UpdateTab()
                if (ExecuteCallback := Callback?) then ExecuteCallback(CurrentValue)

            UpdateGaugeVisuals():void =
                RawValue := CurrentValue * GaugeMaxSegments
                CurrentIndex := if (Rounded := Round[RawValue]) then Rounded * 1.0 else 0.0

                ClampedIndex := if (CurrentIndex < 0.0) { 0.0 } else if (CurrentIndex > GaugeMaxSegments) { GaugeMaxSegments } else { CurrentIndex }

                set GaugeWidget.ValueSegmentIndex = ClampedIndex

                var Fraction : float = 0.0
                if (GaugeMaxSegments > 0.0):
                      Step := 1.0 / GaugeMaxSegments
                      set Fraction = ClampedIndex * Step
                else:
                      set Fraction = 0.0

                ActualValue := MapFractionToRange(0.0, 1.0, Fraction)
                set GaugeWidget.Value = ActualValue

            UpdateTab():void =
                if (Tab := AssociatedTabButton?) then Tab.SetSelectionName(ui_config.ValueToMessage(CurrentValue))
            
            RefreshTabName<override>():void = UpdateTab()

            GetValue<public>():float = CurrentValue
            SetCallback<public>(CB:type{_(:float):void}):void = { set Callback = option{CB} }

            ClampFractionToZeroOne(FractionValue: float): float =
                if (FractionValue < 0.0) { 0.0 }
                else if (FractionValue > 1.0) { 1.0 }
                else { FractionValue }

            MapFractionToRange(MinimumValue: float, MaximumValue: float, FractionValue: float): float =
                ClampedFractionValue := ClampFractionToZeroOne(FractionValue)
                MinimumValue + (MaximumValue - MinimumValue) * ClampedFractionValue
