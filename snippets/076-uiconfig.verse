# Knowledge-neutral Verse snippet; project names and asset IDs generalized.
# id: snippet-076

using { /Fortnite.com/Devices }
using { /Fortnite.com/UI }
using { /UnrealEngine.com/Temporary/UI }
using { /UnrealEngine.com/Temporary/SpatialMath }
using { /Verse.org/Simulation }
using { /Verse.org/Assets }
using { /Verse.org/Colors }
using { /Fortnite.com/SPACE_OPERA }
using { /Fortnite.com/Items/SPACE_OPERA }
using { /Fortnite.com/Weapons/SPACE_OPERA }
using { /Verse.org/SceneGraph }
using { /UnrealEngine.com/Itemization }
using { /UnrealEngine.com/Temporary/Diagnostics }

UI<public> := module:
    Widgets<public> := module:
    MaterialInstances<public> := module:
    Materials<public> := module:
        Hilts<public> := module{}
        TelekineticAbilities<public> := module{}
    Textures<public> := module:

ui_config := module:

    # Standard structure for items in the UI (Hilts, Crystals, Abilities)
    item_data<public> := struct:
        ID : int
        Label : message
        Description : message = Empty
        Texture : texture = DefaultTexture
        SelectionMaterial : ?subtype(material) = false

    DefaultTexture<public> : texture = UI.Textures.T_Icon_None
    
    # Internal Tab IDs for the menu switching logic
    TabID_Hilt<public> : int = 0
    TabID_Ability<public> : int = 1
    TabID_Crystal<public> : int = 3 
    TabID_Stability<public> : int = 4
    TabID_Arcing<public> : int = 5

    # Default selections when a new session starts
    DefaultHiltID<public> : int = 0
    DefaultAbilityID<public> : int = -1 # None
    DefaultCrystalID<public> : int = 100 # Blue

    # Slider config
    DefaultStability<public> : float = 0.5
    StepStability<public> : float = 0.1
    DefaultArcing<public> : float = 0.0
    StepArcing<public> : float = 0.1

    # Localization Strings
    Label_Hilt<public><localizes> : message = "Hilt"
    Label_Ability<public><localizes> : message = "Ability"
    Label_Crystal<public><localizes> : message = "Crystal"
    Label_Stability<public><localizes> : message = "Stability"
    Label_Arcing<public><localizes> : message = "Arcing"
    
    Empty<public><localizes> : message = ""
    None<public><localizes> : message = "None"
    CrystalDesc<public><localizes> : message = "Configuration for crystal effects."
    ArcDesc<public><localizes> : message = "Configuration for blade arcing effects."
    Decrement<public><localizes> : message = " - "
    Increment<public><localizes> : message = " + "

    # Hilts
    H1_L<localizes>:message = "Hope Hilt"
    H1_D<localizes>:message = ""
    H2_L<localizes>:message = "Legacy Hilt"
    H2_D<localizes>:message = ""
    H3_L<localizes>:message = "Guidance Hilt"
    H3_D<localizes>:message = ""
    H4_L<localizes>:message = "Control Hilt"
    H4_D<localizes>:message = ""
    H5_L<localizes>:message = "Conflict Hilt"
    H5_D<localizes>:message = ""
    H6_L<localizes>:message = "Discipline Hilt"
    H6_D<localizes>:message = ""
    H7_L<localizes>:message = "Valor Hilt"
    H7_D<localizes>:message = ""
    H8_L<localizes>:message = "Wisdom Hilt"
    H8_D<localizes>:message = ""
    H9_L<localizes>:message = "Destiny Hilt"
    H9_D<localizes>:message = ""
    H10_L<localizes>:message = "Ambition Hilt"
    H10_D<localizes>:message = ""

    # Crystals
    C1_L<localizes>:message = "Blue"
    C2_L<localizes>:message = "Green"
    C3_L<localizes>:message = "Yellow"
    C4_L<localizes>:message = "White"
    C5_L<localizes>:message = "Orange"
    C6_L<localizes>:message = "Red"
    C7_L<localizes>:message = "Purple"
    C_D<localizes>:message = ""

    # Abilities
    A1_L<localizes>:message = "Push"
    A1_D<localizes>:message = "Blast away objects in front of you"
    A2_L<localizes>:message = "Pull"
    A2_D<localizes>:message = "Pull objects to you"
    A3_L<localizes>:message = "Throw"
    A3_D<localizes>:message = "Throw nearby objects at your enemies"
    A4_L<localizes>:message = "Barrier"
    A4_D<localizes>:message = "Summon a shield in front of you"
    A5_L<localizes>:message = "Lightning"
    A5_D<localizes>:message = "Channel lightning at your enemies"
    A6_L<localizes>:message = "Heal"
    A6_D<localizes>:message = "Heal yourself and those around you"
    A7_L<localizes>:message = "Awareness"
    A7_D<localizes>:message = "Mark nearby characters and see what they have equipped"
    A8_L<localizes>:message = "Saber Throw"
    A8_D<localizes>:message = "Throw a EnergyBlade and use the Telekinetic to pull it back to your hand"
    A9_L<localizes>:message = "Distraction"
    A9_D<localizes>:message = "Trick others into seeing an illusionary enemy"
    A10_L<localizes>:message = "Mind Trick"
    A10_D<localizes>:message = "Trick others into seeing you as a character of your choosing"
    # A_D<localizes>:message = ""

    # Grid Layout Config
    ColumnsHilt<public> : int = 4
    ColumnsCrystal<public> : int = 4
    ColumnsAbility<public> : int = 4
    
    # Maps specific Crystal IDs (100-106) to the color index on the UMG widget
    # If adding new crystals, add their ID mapping here
    CrystalIDToUMGMap : [int]int = map{ 100=>0, 101=>1, 102=>2, 104=>3, 105=>4, 103=>5, 106=>6 }

    CrystalIDToUMGIndex<public>(ID:int):int = 
        if (Val := CrystalIDToUMGMap[ID]) then return Val else 1

    # Maps Crystal IDs to actual colors for the EnergyBlade component
    GetColorFromCrystalID<public>(ID:int, EnergyBlade:entity)<transacts>:?kyber_crystal_component = 
        if (ID = 100) then return option{kyber_crystal_blue_component{Entity:=EnergyBlade}}
        if (ID = 101) then return option{kyber_crystal_green_component{Entity:=EnergyBlade}}
        if (ID = 102) then return option{kyber_crystal_yellow_component{Entity:=EnergyBlade}}
        if (ID = 103) then return option{kyber_crystal_white_component{Entity:=EnergyBlade}}
        if (ID = 104) then return option{kyber_crystal_orange_component{Entity:=EnergyBlade}}
        if (ID = 105) then return option{kyber_crystal_red_component{Entity:=EnergyBlade}}
        if (ID = 106) then return option{kyber_crystal_purple_component{Entity:=EnergyBlade}}
        return false
    
    MakeItem<public>(ID:int, Lab:message, Desc:message, ?SelMat:?subtype(material) = false):item_data = 
        item_data{ID:=ID, Label:=Lab, Description:=Desc, SelectionMaterial:=SelMat}

    # ID Range: 0 - 99
    # Add new Hilt definitions here. Ensure IDs match GiveEnergyBlade() function
    GetHilts<public>() : []item_data = array:
        MakeItem(0, H1_L, H1_D, ?SelMat:=option{UI.Materials.M_Hilt_EnergyBladeHeroC_ROTJ})
        MakeItem(1, H2_L, H2_D, ?SelMat:=option{UI.Materials.M_Hilt_EnergyBladeHeroB})
        MakeItem(2, H3_L, H3_D, ?SelMat:=option{UI.Materials.M_Hilt_EnergyBladeHeroA})
        MakeItem(3, H4_L, H4_D, ?SelMat:=option{UI.Materials.M_Hilt_VillainA})
        MakeItem(4, H5_L, H5_D, ?SelMat:=option{UI.Materials.M_Hilt_VillainB})
        MakeItem(5, H6_L, H6_D, ?SelMat:=option{UI.Materials.M_Hilt_Mentor})
        MakeItem(6, H7_L, H7_D, ?SelMat:=option{UI.Materials.M_Hilt_PadawanBlue})
        MakeItem(7, H8_L, H8_D, ?SelMat:=option{UI.Materials.M_Hilt_PadawanGreen})
        MakeItem(8, H9_L, H9_D, ?SelMat:=option{UI.Materials.M_Hilt_Training})
        MakeItem(9, H10_L, H10_D, ?SelMat:=option{UI.Materials.M_Hilt_OpposingApprentice})

    # ID Range: 100 - 199
    # DEV NOTE: Add new Crystal colors here. Must update GetColorFromCrystalID() and CrystalIDToUMGMap() above if added
    Crystals<public>() : []item_data = array:
        MakeItem(100, C1_L, C_D) # Blue
        MakeItem(101, C2_L, C_D) # Green
        MakeItem(102, C3_L, C_D) # Yellow
        MakeItem(103, C4_L, C_D) # White
        MakeItem(104, C5_L, C_D) # Orange
        MakeItem(105, C6_L, C_D) # Red
        MakeItem(106, C7_L, C_D) # Purple

    # ID Range: 200 - 299
    # Add new Abilities here. Ensure IDs match GivePower() function
    Abilities<public>() : []item_data = array:
        MakeItem(200, A1_L, A1_D, ?SelMat:=option{UI.Materials.M_TelekineticAbility_Push})
        MakeItem(201, A2_L, A2_D, ?SelMat:=option{UI.Materials.M_TelekineticAbility_Pull})
        MakeItem(202, A3_L, A3_D, ?SelMat:=option{UI.Materials.M_TelekineticAbility_Throw})
        MakeItem(203, A4_L, A4_D, ?SelMat:=option{UI.Materials.M_TelekineticAbility_Barrier})
        MakeItem(204, A5_L, A5_D, ?SelMat:=option{UI.Materials.M_TelekineticAbility_Lightning})
        MakeItem(205, A6_L, A6_D, ?SelMat:=option{UI.Materials.M_TelekineticAbility_Heal})
        MakeItem(206, A7_L, A7_D, ?SelMat:=option{UI.Materials.M_TelekineticAbility_Awareness})
        MakeItem(207, A8_L, A8_D, ?SelMat:=option{UI.Materials.M_TelekineticAbility_SaberThrow})
        MakeItem(208, A9_L, A9_D, ?SelMat:=option{UI.Materials.M_TelekineticAbility_Distraction})
        MakeItem(209, A10_L, A10_D, ?SelMat:=option{UI.Materials.M_TelekineticAbility_MindTrick})

    # Grants the Item Definition based on the ID selected in the UI
    # Update this when adding new Hilt IDs to the 'Hilts' array
    GiveEnergyBlade<public>(InInventory:inventory_component, TypeID:int)<transacts>:?entity =
        Item := case(TypeID):
            0 => EnergyBladeHeroC_V1 {}
            1 => EnergyBladeHeroB_V1 {}
            2 => EnergyBladeHeroA_V1 {}
            3 => EnergyBladeVillainA_V1 {}
            4 => EnergyBladeVillainB_V1 {}
            5 => EnergyBladeMentor_V1 {}
            6 => EnergyBladeApprenticeA_V1 {}
            7 => EnergyBladeApprenticeB_V1 {}
            8 => EnergyBladeTraining_V1 {}
            9 => EnergyBladeOpposingApprentice_V1 {}
            _ => EnergyBladeHeroC_V1 {}
        
        if (InInventory.AddItemDistribute(Item).GetSuccess[]):
             return option{Item}
        return false

    # Grants the ability item based on the ID selected in the UI
    # Update this when adding new Ability IDs to the 'Abilities' array
    GivePower<public>(InInventory:inventory_component, PowerID:int)<transacts>:?entity =
        Item := case(PowerID):
            200 => TelekineticAbilityPush_EnergyBladeAugment_V1 {}
            201 => TelekineticAbilityPull_EnergyBladeAugment_V1 {}
            202 => TelekineticAbilityThrow_EnergyBladeAugment_V1 {}
            203 => TelekineticAbilitieshield_EnergyBladeAugment_V1 {}
            204 => TelekineticAbilityLightning_EnergyBladeAugment_V1 {}
            205 => TelekineticAbilityHeal_EnergyBladeAugment_V1 {}
            206 => TelekineticAbilityAwareness_EnergyBladeAugment_V1 {}
            207 => TelekineticAbilitiesaberThrow_EnergyBladeAugment_V1 {}
            208 => TelekineticAbilityDistraction_EnergyBladeAugment_V1 {}
            209 => TelekineticAbilityTrick_EnergyBladeAugment_V1 {}
            _ => TelekineticAbilityPush_EnergyBladeAugment_V1 {}
        
        if (InInventory.AddItemDistribute(Item).GetSuccess[]):
             return option{Item}
        return false

    CreateEnergyBlade<public>(TypeID:int)<transacts>:entity =
        return case(TypeID):
           0 => EnergyBladeHeroC_V1 {}
            1 => EnergyBladeHeroB_V1 {}
            2 => EnergyBladeHeroA_V1 {}
            3 => EnergyBladeVillainA_V1 {}
            4 => EnergyBladeVillainB_V1 {}
            5 => EnergyBladeMentor_V1 {}
            6 => EnergyBladeApprenticeA_V1 {}
            7 => EnergyBladeApprenticeB_V1 {}
            8 => EnergyBladeTraining_V1 {}
            9 => EnergyBladeOpposingApprentice_V1 {}
            _ => EnergyBladeHeroC_V1 {}

    # Utilities
    GetLabel<public>(Items : []item_data, TargetID : int) : message =
        for (Item : Items, Item.ID = TargetID) do return Item.Label
        return Empty

    Clamp01<public>(V : float) : float = if (V < 0.0) then 0.0 else if (V > 1.0) then 1.0 else V

    IntToMessage<localizes>(I:int) : message = "{I}"

    ValueToMessage<public>(Val : float) : message =
        if (IntVal := Int[Val * 10.0]):
            return IntToMessage(IntVal)
        return IntToMessage(0)
