# Knowledge-neutral Verse snippet; project names and asset IDs generalized.
# id: snippet-077

using { /Fortnite.com/Devices }
using { /Verse.org/Simulation }
using { /UnrealEngine.com/Temporary/Diagnostics }

# Multiplayer QTEs game logic
qte_game_device := class(creative_device):

    # Timer that allows us to react to the end game with a small time offset
    @editable
    EndGameOffsetTimerDevice : timer_device = timer_device{}

    # Array of towers
    @editable
    Towers : []qte_tower_device = array{}

    var GameStarted : logic = false

    OnBegin<override>()<suspends> : void =
        # Buttons on top of each tower will start multiplayer QTEs game
        for (Tower : Towers):
            Tower.StartGameButton.InteractedWithEvent.Subscribe(StartGame)
        
        # Subscribe function to the end game offset timer
        EndGameOffsetTimerDevice.SuccessEvent.Subscribe(OnEndGameOffsetTimerSuccess)

    # ArrayRace will wait for any tower's event each tower LostEvent
    # In case any of them triggers, it will end the game
    # Idea for improvement:
    #   this device should react to all towers reporting failures and finish the game when there's only one tower left
    WaitForEnd()<suspends> : void =
        ArrayRace(Towers, WaitForTower)
        Print("Someone lost!")
        EndGameOffsetTimerDevice.Start()

    WaitForTower(Tower : qte_tower_device)<suspends> : void =
        Tower.FailedEvent.Await()

    # Should be fired when players push the button at the top of any tower
    StartGame(Agent : agent) : void =
        # Ignore if the game already started
        if (GameStarted?):
            return

        set GameStarted = true

        # Activate all towers
        for (Tower : Towers):
            Tower.Start()

        spawn{WaitForEnd()}

    # Should be fired when the game has ended and towers need to be reset
    EndGame() : void =
        # Ignore if the game already ended
        if (not GameStarted?):
            return

        set GameStarted = false

        # Deactivate all towers
        for (Tower : Towers):
            Tower.Stop()

    OnEndGameOffsetTimerSuccess<private>(Agent : ?agent) : void =
        EndGameOffsetTimerDevice.Reset()
        EndGame()
