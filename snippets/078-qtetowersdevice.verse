# Knowledge-neutral Verse snippet; project names and asset IDs generalized.
# id: snippet-078

using { /Fortnite.com/Devices }
using { /Verse.org/Simulation }
using { /UnrealEngine.com/Temporary/Diagnostics }

# Tower logic - a part of multiplayer QTEs game
qte_tower_device := class(creative_device):

    # Button at the top of the tower - starts multiplayer QTEs game for all towers registered in the game's device
    @editable
    StartGameButton : button_device = button_device{}

    # Button near the tower, allowing the player to teleport to the top floor
    @editable
    TeleportToTowerButton : button_device = button_device{}

    # Used to display game status
    @editable
    GameInProgressBillboard : billboard_device = billboard_device{}

    # With Verse we only need one Skilled Interaction Device per tower, because
    # when handling QTE success and failure events, we can affect different floors in the code
    @editable
    SkilledInteraction : skilled_interaction_device = skilled_interaction_device{}

    # With Verse we only need one Timer Device per tower
    # Used to delay start of skilled interaction after falling one floor
    @editable
    NextFloorTimerDevice : timer_device = timer_device{}

    # Player Counter Device - gives us the number of players in the tower
    @editable
    PlayerCounter : player_counter_device = player_counter_device{}

    # Where to move players after the game
    @editable
    EndGameTeleporter : teleporter_device = teleporter_device{}

    # HUD messages for winners and losers
    @editable
    HUDMessageGameWon : hud_message_device = hud_message_device{}
    @editable
    HUDMessageGameLost : hud_message_device = hud_message_device{}
    
    @editable
    TowerFloors : []cinematic_sequence_device = array{}

    FailedEvent<public> : event() = event(){}

    var CurrentFloorIndex : int = 0
    var IsInFailureFlow : logic = false

    # Player to status map:
    # Key: every player that takes part in this tower
    # Value: whether or not player successfully finished QTE
    var CurrentPlayers : [agent]logic = map{}

    OnBegin<override>()<suspends> : void =
        GameInProgressBillboard.HideText()

        NextFloorTimerDevice.SuccessEvent.Subscribe(OnNextFloorTimerSuccess)
        SkilledInteraction.InteractionSucceededEvent.Subscribe(OnInteractionSucceeded)
        SkilledInteraction.InteractionFailedEvent.Subscribe(OnInteractionFailed)
        SkilledInteraction.InteractionCanceledEvent.Subscribe(OnInteractionFailed)

    # Prepares game setup and starts the game
    Start() : void =
        Print("Tower start")

        # Clear players data
        set CurrentPlayers = map{}

        # Initialise the players map data and give a defauly value to the map
        for (Agent : PlayerCounter.GetCountedAgents(), set CurrentPlayers[Agent] = false):

        # Close entrance to the tower and begin game
        StartGameButton.Disable()
        TeleportToTowerButton.Disable()
        GameInProgressBillboard.ShowText()
        BeginQTE()

    # Stops current game and prepares for next round
    Stop() : void =
        Print("Tower stop")

        # Stop the ongoing QTEs
        StopOngoingQTEs()

        # If players are still inside, we can assume they are the winners and move them outside the tower.
        # Losers will be respawned at a checkpoint by falling into the lava
        for (Player->Status : CurrentPlayers):
            # Teleport winners outside of the tower
            EndGameTeleporter.Teleport(Player)
            # Show message about winning the game
            HUDMessageGameWon.Show(Player)

        # Clear players data
        set CurrentPlayers = map{}

        # Open entrance to the tower
        StartGameButton.Enable()
        TeleportToTowerButton.Enable()
        GameInProgressBillboard.HideText()
        NextFloorTimerDevice.Reset()

        # Reset all floors
        for (TowerFloor : TowerFloors):
            TowerFloor.PlayReverse()

        # Set the top floor as the current one
        set CurrentFloorIndex = 0;

    # When players on the floor succeed at multiplayer QTE, we want it to restart after a few seconds
    OnNextFloorTimerSuccess(Agent : ?agent) : void =
        NextFloorTimerDevice.Reset()
        BeginQTE()

    # If everyone succeeds then let's launch another QTE
    OnInteractionSucceeded(Agent : agent) : void =
        if (set CurrentPlayers[Agent] = true):

        if (HasEveryoneSucceeded[]):
            ResetPlayersStatus()
            BeginQTE()

    OnInteractionFailed(Agent : agent) : void =
        if (not IsInFailureFlow?):
            set IsInFailureFlow = true
            ResetPlayersStatus()
            FallOneFloor()

    BeginQTE() : void =
        Print("Begin QTE")
        set IsInFailureFlow = false

        Players := for (Player->Status : CurrentPlayers) {Player}
            
        if (SkilledInteraction.BeginInteraction[Players]):
    
    HasEveryoneSucceeded()<transacts><decides> : void =
        for (Player->Status : CurrentPlayers):
            Status?

    ResetPlayersStatus() : void =
        for (Player->Status : CurrentPlayers, set CurrentPlayers[Player] = false):

    # When players fail at multiplayer QTE, we want them to fall one floor
    FallOneFloor() : void =
        # Stop the ongoing QTEs
        StopOngoingQTEs()

        # Hide the floor players are currently standing on (using a sequence)
        if (TowerFloor := TowerFloors[CurrentFloorIndex]):
            TowerFloor.Play()

            # Check if the next floor exists. If not, players lost the game
            if (CurrentFloorIndex >= TowerFloors.Length - 1):
                # Show message about losing the game
                for:
                    Player->Status : CurrentPlayers
                do:
                    HUDMessageGameLost.Show(Player)

                # Clear players data
                set CurrentPlayers = map{}

                # Send a signal about losing the game after the floor hiding animation ends
                TowerFloor.StoppedEvent.Subscribe(LoseGame)
                return

        # Increment the current floor's index
        set CurrentFloorIndex += 1
        Print("Falling one floor below to floor number {CurrentFloorIndex + 1}")

        # Wait a few seconds before starting a new multiplayer QTE
        NextFloorTimerDevice.Start()

    # Players lost the game
    LoseGame() : void =
        # Multiplayer QTEs game logic should listen to this signal
        FailedEvent.Signal()

    # Stop the ongoing QTEs
    StopOngoingQTEs() : void =
        Players := for (Player->Status : CurrentPlayers):
            Player

        if (SkilledInteraction.EndInteraction[Players]):
            SkilledInteraction.ClearQueue()
