# Knowledge-neutral Verse snippet; project names and asset IDs generalized.
# id: snippet-079

# https://dev.epicgames.com/community/snippets/oYRX/fortnite-generic-sync-and-race-across-elements-in-an-array
using { /Verse.org/Simulation }
 
# ArrayRace will call the Function argument for each element in the Array and wait for only one of the async functions to complete, canceling all the rest.
# A divide-and-conquer concurrency algorithm that divides an array into two and races between the two subarrays.
# This is a recursive implementation, where the function calls itself to race between the divided arrays.
# The base case (the condition to stop the recursion) is the array has only one element.
# This is a generic implementation using parametric types, so you can provide your own type and your own async function as arguments.
ArrayRace(Array:[]t, Function(Element:t)<suspends>:u where t:type, u:type)<suspends>:u=
    # If more than two elements in the array, then split the array in half,
    # and race between the subarrays recursively.
    if:
        Array.Length > 1
        Mid := Floor(Array.Length / 2)
        LeftArray := Array.Slice[0, Mid]
        RightArray := Array.Slice[Mid]
    then:
        race:
            ArrayRace(LeftArray, Function)
            ArrayRace(RightArray, Function)
    else if:
        First := Array[0]
    then:
        # Grab the only element in the array and call the async function with it as the argument.
        Function(First)
    else:
        # A race with no subtasks never completes, so sleep infinitely and call Err() because we shouldn't return a value here.
        Sleep(Inf)
        Err()
