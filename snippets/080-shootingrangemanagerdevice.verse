# Knowledge-neutral Verse snippet; project names and asset IDs generalized.
# id: snippet-080

using { /Fortnite.com/Devices }
using { /Verse.org/Simulation }

# A device that manages shooting range gameplay.
shooting_range_manager_device := class(creative_device):

    @editable
    ScoreManager:score_manager_device = score_manager_device{}

    @editable
    GoodTarget1:shooting_range_target_track_device = shooting_range_target_track_device{}

    @editable
    GoodTarget2:shooting_range_target_track_device = shooting_range_target_track_device{}

    @editable
    GoodTarget3:shooting_range_target_track_device = shooting_range_target_track_device{}

    @editable
    var GoodTargetScore:int = 100

    @editable
    BadTarget1:shooting_range_target_device = shooting_range_target_device{}

    @editable
    BadTarget2:shooting_range_target_device = shooting_range_target_device{}

    @editable
    BadTarget3:shooting_range_target_device = shooting_range_target_device{}

    @editable
    var BadTargetScore:int = -100

    # Runs when the device is started in a running game.
    OnBegin<override>()<suspends>:void=
        # Subscribing to the GoodTarget HitEvents.
        GoodTarget1.HitEvent.Subscribe(OnGoodTarget1Hit)
        GoodTarget2.HitEvent.Subscribe(OnGoodTarget2Hit)
        GoodTarget3.HitEvent.Subscribe(OnGoodTarget3Hit)

        # Subscribing to the BadTarget HitEvents.
        BadTarget1.HitEvent.Subscribe(OnBadTarget1Hit)
        BadTarget2.HitEvent.Subscribe(OnBadTarget2Hit)
        BadTarget3.HitEvent.Subscribe(OnBadTarget3Hit)


    # A hit callback that scores and pops down the first GoodTarget.
    OnGoodTarget1Hit():void=
        AdjustScore(GoodTargetScore)
        GoodTarget1.PopDown()
    
    # A hit callback that scores and pops down the second GoodTarget.
    OnGoodTarget2Hit():void=
        AdjustScore(GoodTargetScore)
        GoodTarget2.PopDown()

    # A hit callback that scores and pops down the third GoodTarget.
    OnGoodTarget3Hit():void=
        AdjustScore(GoodTargetScore)
        GoodTarget3.PopDown()

    # A hit callback that scores and pops down the first BadTarget.
    OnBadTarget1Hit():void=
        AdjustScore(BadTargetScore)
        BadTarget1.PopDown()

    # A hit callback that scores and pops down the second BadTarget.
    OnBadTarget2Hit():void=
        AdjustScore(BadTargetScore)
        BadTarget2.PopDown()

    # A hit callback that scores and pops down the third BadTarget.
    OnBadTarget3Hit():void=
        AdjustScore(BadTargetScore)
        BadTarget3.PopDown()

    # Adjusts the player's score by the provided value.
    AdjustScore(Value:int):void=
        ScoreManager.SetScoreAward(Value)

      # Gets the first player in the playspace.
        if (Player:player = GetPlayspace().GetPlayers()[0]):
            # Awards the points to the player.
            ScoreManager.Activate(Player)
