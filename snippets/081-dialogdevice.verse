# Knowledge-neutral Verse snippet; project names and asset IDs generalized.
# id: snippet-081

using { /Fortnite.com/Devices }
using { /Verse.org/Simulation }
using { /UnrealEngine.com/Temporary/Diagnostics }

# See https://dev.epicgames.com/documentation/en-us/uefn/create-your-own-device-in-verse for how to create a verse device.

sub_dialog<public>:=enum:
    Start,
    SubDialog_A, 
    SubDialog_B,
    Complete

sub_dialog_class := class():
    @editable
    SubDialogType:sub_dialog = sub_dialog.Start

    var IsCompleted:logic = false

event_type<public>:=enum:
    ChangeReputation,
    NextConversation,
    NextObjective,
    StartQuestDJ

dialogpath := class():
    # Conversation Device
    @editable
    ConversationDevice:conversation_device = conversation_device{}
    
    # The specific objective to enable the conversation
    @editable
    Objective:cantina_objectives = cantina_objectives.GameStart

    # Array of optional events, such as giving reputation or changing dialog
    @editable
    Events:[]conversation_event_class = array{}

    # Defines this conversation as a potential dialog
    @editable
    MaybeSubDialog:?sub_dialog_class = false

conversation_event_class := class():
    # The conversation event 
    @editable
    EventIndex:int = 0

    # Event type
    @editable
    EventType:event_type = event_type.ChangeReputation

    # Gives reputation to player
    @editable
    ReputationAmount:?int = false

    # Hud description for gaining or losing reputation
    @editable
    ReputationHudDevice:hud_message_device = hud_message_device{}

    # Optional Dialog to change to
    @editable
    MaybeDialogChange:?sub_dialog = false

    # Optional Dialog to change to
    @editable
    MaybeDialogRemove:?sub_dialog = false

    # Optional quest objective to change to
    @editable
    MaybeNewQuestObjective:?cantina_objectives = false

dialog_device := class(creative_device):
    # The conversations for the NPC
    @editable
    DialogPaths:[]dialogpath = array{}

    # Quest Manager
    @editable
    QuestManager:quest_manager_device = quest_manager_device{}

    # Stat Attitude to increase or decrease reputation
    @editable
    StatAttitude:stat_creator_device = stat_creator_device{}

    # Runs when the device is started in a running game
    OnBegin<override>()<suspends>:void=
        Sleep (0.1)
        if (DialogPaths[0].MaybeSubDialog?):
            UpdateSubDialog(sub_dialog.Start)

        for(DialogPath:DialogPaths):
            DialogPath.ConversationDevice.OnConversationEvent.Subscribe(OnEvent)
            DialogPath.ConversationDevice.CancelEvent.Subscribe(UpdateProgressOnCanceled)

    # Subscribed event based on which event is called and selected
    OnEvent(Agent:agent, EventIndex:int):void=
        for(CurrentEvent:GetCurrentDialogPath[].Events, CurrentEvent.EventIndex = EventIndex):
            case (CurrentEvent.EventType):
                event_type.StartQuestDJ =>
                    set QuestManager.StartQuestWithDJ = true

                event_type.ChangeReputation =>
                    if (StatAttitude.SetValue[Agent, StatAttitude.GetValue[Agent] + CurrentEvent.ReputationAmount?]):
                        CurrentEvent.ReputationHudDevice.Show(Agent)

                event_type.NextConversation =>
                    if:
                        SetSubDialogToComplete(GetCurrentDialogPath[].MaybeSubDialog?.SubDialogType)
                        DialogChange := CurrentEvent.MaybeDialogChange?
                    then:
                        if (SetSubDialogToComplete(CurrentEvent.MaybeDialogRemove?)):

                        CheckForSubDialog(DialogChange)

                event_type.NextObjective =>
                    if (NextObjective := CurrentEvent.MaybeNewQuestObjective?):
                        QuestManager.ProgressQuest(Agent, NextObjective)
    
    # Checks if the next conversation is completed, if so, then swap to the next one
    CheckForSubDialog(RequestedSubDialog:sub_dialog):void=
        var FoundSubDialog:logic = false
        for (Paths:DialogPaths, Paths.MaybeSubDialog?.SubDialogType = RequestedSubDialog, Paths.MaybeSubDialog?.IsCompleted = false):
            set FoundSubDialog = true
            UpdateSubDialog(RequestedSubDialog)
        
        if (not FoundSubDialog?):
            UpdateSubDialog(sub_dialog.Complete)
    
    # Set a specific dialog to complete and is not abled to be called again
    SetSubDialogToComplete(RequestedSubDialog:sub_dialog)<transacts>:void=
        for (Paths:DialogPaths, Paths.MaybeSubDialog?.SubDialogType = RequestedSubDialog, set Paths.MaybeSubDialog?.IsCompleted = true):

    # Upon the conversation device being cancelled, still attempt to update the quest
    UpdateProgressOnCanceled(Agent:agent):void=
        for(CurrentEvent:GetCurrentDialogPath[].Events, CurrentEvent.EventType = event_type.NextObjective, NextObjective := CurrentEvent.MaybeNewQuestObjective?):
            QuestManager.ProgressQuest(Agent, NextObjective)

    <# Enables the selected dialog path
        Any other conversation is disabled
    #>            
    UpdateSubDialog(DialogOption:sub_dialog):void=
        for (Path:DialogPaths):
            if (Path.MaybeSubDialog?.SubDialogType = DialogOption):
                Path.ConversationDevice.Enable()
            else:
                Path.ConversationDevice.Disable()

    <# Enables the correct conversation based on the selected objective
        Any other conversation is disabled
    #>
    UpdateObjectiveDialog(CurrentObjective:cantina_objectives):void=
        for (Paths:DialogPaths):
            if (Paths.Objective = CurrentObjective):
                Paths.ConversationDevice.Enable()
            else:
                Paths.ConversationDevice.Disable()

    # Returns a potential dialog path
    GetCurrentDialogPath()<decides><transacts>:dialogpath=
        EnabledDialogPaths := for(DialogPath:DialogPaths, DialogPath.ConversationDevice.IsEnabled[]):
            DialogPath

        EnabledDialogPaths[0]
