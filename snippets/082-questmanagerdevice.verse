# Knowledge-neutral Verse snippet; project names and asset IDs generalized.
# id: snippet-082

using { /Fortnite.com/Devices }
using { /Verse.org/Simulation }
using { /UnrealEngine.com/Temporary/Diagnostics }

# See https://dev.epicgames.com/documentation/en-us/uefn/create-your-own-device-in-verse for how to create a verse device.

cantina_objectives<public>:=enum:
    GameStart,
    QuestAccepted,
    TalkToDJAboutEskay,
    TalkToDJAboutEskay1_2,
    LookForItem,
    TakeItemToDJ,
    TalkToCaris,
    MeetCaptain,
    QuestCompleted

# An enum list of named objectives for the cantina quest, used to define when devices need to be enabled/disabled
cantina_quest:= class():
    # The current objective
    @editable
    Objective:cantina_objectives = cantina_objectives.GameStart

    # Gives reputation to player
    @editable
    MaybeCantinaReward:?cantina_reward = false

    # Objective description to display current objective
    @editable
    MaybeObjectiveDescription:?string = false

    # Map Indicator to display current objective location
    @editable
    ObjectiveIndicator:map_indicator_device = map_indicator_device{}

cantina_reward:= class():
    # Gives reputation to player
    @editable
    ReputationAmount:int = 0

    # Hud Device when gaining reputation
    @editable
    ReputationHudDevice:hud_message_device = hud_message_device{}

# A Verse-authored creative device that can be placed in a level
quest_manager_device := class(creative_device):

    # An array of potential quest objectives
    @editable
    CantinaQuest:[]cantina_quest = array{}

    # An array of dialog devices
    @editable
    Conversations:[]dialog_device = array{}

    var CurrentQuestIndex:int = 0

    # Stat Attitude to increase or decrease reputation
    @editable
    StatAttitude:stat_creator_device = stat_creator_device{}

    # Tracker device to display messages
    @editable
    QuestTrackerDevice:tracker_device = tracker_device{}

    # Pop-up Device to display when the quest is completed
    @editable
    PopupDevice:popup_dialog_device = popup_dialog_device{}

    # Logic variable to determine who initiated the quest
    var StartQuestWithDJ:logic = false

    # Updates progress of quest, gives potential reputation, and updates quest tracker device
    ProgressQuest(Agent:agent, CantinaObjective:cantina_objectives):void=

        if:
            StatAttitude.SetValue[Agent, StatAttitude.GetValue[Agent] + CantinaQuest[CurrentQuestIndex].MaybeCantinaReward?.ReputationAmount]
            HudDevice := CantinaQuest[CurrentQuestIndex].MaybeCantinaReward?.ReputationHudDevice
        then:
            HudDevice.Show(Agent)

        if (PreviousObjectiveMarker := CantinaQuest[CurrentQuestIndex].ObjectiveIndicator):
            PreviousObjectiveMarker.DeactivateObjectivePulse(Agent)

        # If the player started the quest before talking with Caris, this will update the quest to an alternative version and give appropriate rewards
        if:
            CantinaObjective = cantina_objectives.TalkToDJAboutEskay
            StartQuestWithDJ = false
        then:

            # Locates and rewards the player with reputation when they accept the quest with Caris
            for(Quest:CantinaQuest, Quest.Objective = cantina_objectives.QuestAccepted, StatAttitude.SetValue[Agent, StatAttitude.GetValue[Agent] + Quest.MaybeCantinaReward?.ReputationAmount]):
                
            ProgressQuest(Agent, cantina_objectives.TalkToDJAboutEskay1_2)
        else:

            for(Index -> Quest : CantinaQuest, Quest.Objective = CantinaObjective):
                set CurrentQuestIndex = Index
                if (ObjectiveDescription := Quest.MaybeObjectiveDescription?):
                    QuestTrackerDevice.SetDescriptionText(ToText(ObjectiveDescription))

                Quest.ObjectiveIndicator.ActivateObjectivePulse(Agent)
                    
                if (CurrentQuestIndex = CantinaQuest.Length-1):
                    PopupDevice.Show(Agent)
                    QuestTrackerDevice.Complete(Agent)

                for(SingleConversation:Conversations):
                    TryUpdateDialogs(SingleConversation)

    # Checks if the dialog needs to update before disabling any conversation
    TryUpdateDialogs(DialogDevice:dialog_device):void=
        for (Paths:DialogDevice.DialogPaths, Paths.Objective = CantinaQuest[CurrentQuestIndex].Objective):
            DialogDevice.UpdateObjectiveDialog(Paths.Objective)
