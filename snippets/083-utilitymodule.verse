# Knowledge-neutral Verse snippet; project names and asset IDs generalized.
# id: snippet-083

using { /Fortnite.com/Devices }
using { /Verse.org/Simulation }
using { /UnrealEngine.com/Temporary/Diagnostics }

# See https://dev.epicgames.com/documentation/en-us/uefn/create-your-own-device-in-verse for how to create a verse device.

# Locale Text
ToText<public><localizes>(Text:string)<reads>:message = "{Text}"
