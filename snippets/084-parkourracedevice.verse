# Knowledge-neutral Verse snippet; project names and asset IDs generalized.
# id: snippet-084

using { /Fortnite.com/Devices }
using { /Fortnite.com/Characters }
using { /Verse.org/Simulation}
using { /UnrealEngine.com/Temporary/Diagnostics }

log_parkour := class(log_channel){}

# This is a Verse-authored creative device that can be placed in a level.
# This script example is a take on Parkour where the objective is to gather 4 batteries before the timer expires.
# 
# Getting started:
#   https://dev.epicgames.com/community/fortnite/getting-started/verse
#
parkour_race_device := class(creative_device):
    # Logger for the parkour experience channel.
    Logger : log = log{Channel:=log_parkour}

    # Device reference to the player spawner.
    @editable
    PlayerSpawnDevice:player_spawner_device = player_spawner_device{}

    # End Game Device for a Victory.
    @editable
    EndGameVictoryDevice:end_game_device = end_game_device{}

    # Timer for this game mode.
    @editable
    TimerDevice:timer_device = timer_device{}

    # Hud message feedback for the player when they get a battery.
    @editable
    HUDMessageBattery:hud_message_device = hud_message_device{}

    # An array of the Item Spawners we have.
    @editable
    BatteryItemSpawners:[]item_spawner_device = array{}

    # Spawner used to create secret batteries that can give the player additional time.
    @editable
    SecretBatteryItemSpawner:item_spawner_device = item_spawner_device{}

    # Tunable amount of time to add when the player picks up the secret battery.
    @editable
    SecretBatteryTimeReward:float = 10.0
        
    # Declaration for an integer that represents the number of batteries we've collected so far.
    var BatteriesCollected:int = 0
    
    # Messages to display when a battery is collected.
    BatteryCollectedMessage<localizes>(Amount:int):message = "You collected {Amount} battery"
    BatteriesCollectedMessage<localizes>(Amount:int):message = "You collected {Amount} batteries"

    # Messages for collecting the secret battery and the completion message.
    AllBatteriesCollectedMessage<localizes>:message = "You collected all of the batteries!"
    SecretBatteryCollectedMessage<localizes>:message = "You collected the secret battery, additional time added!"
    
    # Runs when this device_script is started in a running game.
    OnBegin<override>()<suspends>:void=
        # You can output to the log like this to determine what your script is up to.
        Logger.Print("Parkour Race Script Started!")

        # Subscribing to the AgentSpawnedEvent.
        # When the player spawns, the function "HandleAgentSpawned" is called.
        PlayerSpawnDevice.SpawnedEvent.Subscribe(HandleAgentSpawned)

        # Same for the Timer. We subscribe to the time expired event.
        TimerDevice.FailureEvent.Subscribe(HandleTimerExpired)

        # We go through our array of BatteryItemSpawners and for each Item Spawner, we subscribe to the ItemPickupEvent.
        # We do this for each Item Spawner and we don't save the handle like we did for the AgentSpawnedSubscription and TimerExpiredSubscription above.
        # You don't have to save the handle, but without it, you won't be able to cancel the subscription so it will fire every time an item is picked up. 
        for (BatterySpawner : BatteryItemSpawners):
            BatterySpawner.ItemPickedUpEvent.Subscribe(HandleBatteryPickedUp)
        
        <# An alternate way to do the for loop above that uses an integer to walk through the array.
           for (i:int := 0..BatteryItemSpawners.Length - 1):
                if (Battery := BatteryItemSpawners[i]):
                    Battery.ItemPickedUpEvent.Subscribe(HandlebatteryPickedUp) #>
        
        SecretBatteryItemSpawner.ItemPickedUpEvent.Subscribe(HandleSecretBatteryPickedUp)

    
    # A function that is called when the Agent Spawns from the Player Spawn Pad.
    HandleAgentSpawned(Agent:agent):void=
        Logger.Print("Agent Spawned!")

        # Reset the Timer Device to ensure respawning players get the full amount of time.
        TimerDevice.Reset(Agent)
        TimerDevice.Start(Agent)

    
    # Function that is called when a battery item is picked up from the Item Spawners.
    HandleBatteryPickedUp(Agent:agent):void=
        # Increment the number of batteries collected.
        set BatteriesCollected = BatteriesCollected + 1

        # This is how you can output the number of batteries to the log. Useful for debugging.
        Logger.Print("Number of batteries picked up: {BatteriesCollected}")

        # Check to see if there are enough batteries collected to the end the game.
        if:
            BatteriesCollected >= BatteryItemSpawners.Length
        then:
            # Check to see if we've collected 4 (or more) batteries. If so, we have won, call the EndGame function.
            spawn { EndGame(Agent) }
        else:
            # This code runs if the number of batteries is less than 4. Show a Hud Message to update the player on their progress.
            # HUD message is "battery" if only 1 battery is collect. Becomes "batteries" if more than 1 is collected
            if:
                BatteriesCollected = 1
            then:
                HUDMessageBattery.SetText(BatteryCollectedMessage(BatteriesCollected))
                HUDMessageBattery.Show(Agent)
            else:
                HUDMessageBattery.SetText(BatteriesCollectedMessage(BatteriesCollected))
                HUDMessageBattery.Show(Agent)

            # Then we get the next Item Spawner in our array we set up.
            # We do this by "indexing" into the array. It is inside the "if" statement to ensure that NextBatterySpawner is referenced correctly.
            if: 
                NextBatterySpawner := BatteryItemSpawners[BatteriesCollected]
            then:
                # If we got the next item spawner, call SpawnItem which will activate the next battery to get.
                NextBatterySpawner.SpawnItem()


    HandleSecretBatteryPickedUp(Agent:agent):void=
        Logger.Print("Picked up secret battery")

        # Get the time remaining so we can add additional time to it.
        var TimeRemaining:float = TimerDevice.GetActiveDuration( Agent )
        var TimeToAdd:float = (TimeRemaining + SecretBatteryTimeReward)

        # Add additional time to the Timer Device, but don't go over the initial starting time.
        TimerDevice.SetActiveDuration(Min(TimeToAdd, TimerDevice.GetMaxDuration()), Agent )

        HUDMessageBattery.SetText(SecretBatteryCollectedMessage)
        HUDMessageBattery.Show(Agent)


    # Function that is called when the timer expires.
    HandleTimerExpired(MaybeAgent:?agent):void=
        Logger.Print("Timer Ended")

        if (Agent := MaybeAgent?):
            # Eliminate the player
            if: 
                FortCharacter:fort_character = Agent.GetFortCharacter[]
            then:
                FortCharacter.Damage(500.0)

    
    # Asynchronous function that handles the end of the game.
    EndGame(Agent:agent)<suspends>:void=
        HUDMessageBattery.SetText(AllBatteriesCollectedMessage)
        HUDMessageBattery.Show(Agent)

        # Wait for three seconds before ending the game.
        Sleep(3.0)
        EndGameVictoryDevice.Activate(Agent)


    # Runs when this device script is stopped or the game ends.
    OnEnd<override>():void=
        Logger.Print("Verse device stopped!")
