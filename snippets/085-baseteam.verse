# Knowledge-neutral Verse snippet; project names and asset IDs generalized.
# id: snippet-085

# A tutorial for writing this script can be found at https://dev.epicgames.com/documentation/en-us/uefn/verse-prop-hunt-template-in-unreal-editor-for-fortnite
using { /Fortnite.com/Characters }
using { /Fortnite.com/Devices }
using { /Fortnite.com/UI }
using { /UnrealEngine.com/Temporary/Diagnostics }
using { /UnrealEngine.com/Temporary/SpatialMath }
using { /UnrealEngine.com/Temporary/UI }
using { /Verse.org/Colors }
using { /Verse.org/Simulation }

log_team := class(log_channel){}

# This class defines the devices needed for the different teams in the experience.
# This class is abstract so it cannot be used on its own. It has to be inherited by another class.
base_team := class<abstract>:
    Logger:log = log{Channel:=log_team}

    @editable # Used to set a participant to the team.
    ClassSelector:class_and_team_selector_device = class_and_team_selector_device{}
  
    @editable # Used to award score to agents on the team.
    ScoreManager:score_manager_device = score_manager_device{}
    
    @editable # Used to display the team assignment title.
    TeamTitle:hud_message_device = hud_message_device{}
    
    @editable # Used to display the team assignment description.
    TeamDescription:hud_message_device = hud_message_device{}
    
    @editable # Used to subscribe to team member (prop team) or enemy (hunter team) eliminated events.
    TeamManager:team_settings_and_inventory_device = team_settings_and_inventory_device{}

    # This is an array of agents on the team.
    var TeamAgents<private>:[]agent = array{}

    # This event is signaled when the TeamAgents array becomes empty (signaling the end of the round).
    TeamEmptyEvent:event() = event(){}

    # Returns the current TeamAgents array.
    # This is required because the TeamAgents array is private, so other classes cannot access it directly.
    GetAgents()<decides><transacts>:[]agent =
        TeamAgents

    # Return the size of the TeamAgents array
    # This requires a function because the TeamAgents array is private, so other classes cannot access it directly.
    Count()<transacts>:int = 
        TeamAgents.Length

    # Returns an index in the TeamAgents array of an agent, fails otherwise.
    FindOnTeam(Agent:agent)<decides><transacts>: int =
        Index := TeamAgents.Find[Agent] 
        
    # Set the agent to the team and notify the player. 
    InitializeAgent(Agent:agent):void =
        AddAgentToTeam(Agent)
        ClassSelector.ChangeTeamAndClass(Agent)
        DisplayTeamInformation(Agent)
  
    # Add an agent to TeamAgents.
    AddAgentToTeam(AgentToAdd:agent):void =
        if (not FindOnTeam[AgentToAdd]):
            Logger.Print("Adding agent to team.")
            set TeamAgents += array{AgentToAdd}

    # Activates hud message devices to show the player what team they are on.
    DisplayTeamInformation(Agent:agent):void =
        TeamTitle.Show(Agent)
        TeamDescription.Show(Agent)

    # When a participant leaves the match, remove them from the TeamAgents array and check for the end of the round. 
    EliminateAgent(Agent:agent)<suspends>:void =
        Sleep(0.0) # Delaying 1 game tick to ensure the participant is respawned before proceeding.
        RemoveAgentFromTeam(Agent)

    # Remove an agent from TeamAgents.
    # If the agent removed was the last, signal TeamEmptyEvent.
    RemoveAgentFromTeam(AgentToRemove:agent):void =
        set TeamAgents = TeamAgents.RemoveAllElements(AgentToRemove)
        Logger.Print("{Count()} agent(s) on team remaining.")
        if (Count() < 1):
            Logger.Print("No agents on team remaining. Ending the round.")
            TeamEmptyEvent.Signal()
