# Knowledge-neutral Verse snippet; project names and asset IDs generalized.
# id: snippet-087

# A tutorial for writing this script can be found at https://dev.epicgames.com/documentation/en-us/uefn/verse-prop-hunt-template-in-unreal-editor-for-fortnite
using { /Fortnite.com/Characters }
using { /Fortnite.com/Devices }
using { /Fortnite.com/UI }
using { /UnrealEngine.com/Temporary/SpatialMath }
using { /UnrealEngine.com/Temporary/Diagnostics }
using { /UnrealEngine.com/Temporary/UI }
using { /Verse.org/Colors }
using { /Verse.org/Simulation }

log_heart_beat := class(log_channel){}

# These messages are used to notify a prop agent with a message (or to hide it) when they need to move to avoid their heart beat from becoming visible.
HeartBeatWarningMessage<localizes>(Time:int):message = "Heart Beat in {Time} Seconds. Move!"
HeartBeatWarningClear<localizes>:message = ""

# This class exposed the editable properties for the heartbeat to the prop_hunt device.
heart_beat := class<concrete>():
    Logger:log = log{Channel:=log_heart_beat}

    @editable # The number of seconds before a prop agent must move before the heart beat reveals their position.
    MoveTime:float = 15.0

    @editable # The seconds remaining before the heart beat warning appears. Shouldn't be > than HeartBeatTimer.
    WarningTime:float = 5.0

    @editable # An array of heart beat VFX devices. There is one per participant.
    AgentVFX:[]heartbeat_vfx = array{}
    
    @editable # The audio player device used to play the heart beat sound effects (SFX).
    AudioDevice:radio_device = radio_device{}

    # This map associates a UI for displaying the heart beat warning to each prop agent. 
    var WarningUI:[agent]heartbeat_warning_ui = map{}

    # Keeps track of how many participants have an active heartbeat so we can manage the SFX device.
    var NumberOfHeartBeats:int = 0

    # Sets up heart beat UI for the agent if the agent is a player.
    SetUpUI(PropAgent:agent):void =
        if:
            not WarningUI[PropAgent]
            PlayerUI := GetPlayerUI[player[PropAgent]]
        then:
            UIData:heartbeat_warning_ui = heartbeat_warning_ui{}
            UIData.CreateCanvas()
            PlayerUI.AddWidget(UIData.Canvas, player_ui_slot{ZOrder := 1})
            if (set WarningUI[PropAgent] = UIData) {}

    # Activates the heartbeat VFX and SFX for the specified participant.
    Enable(PropAgent:agent, HeartBeatVFXData:heartbeat_vfx):void =
        if:
            # Get the character, which is used to find the prop agent's position in the scene. 
            Character := PropAgent.GetFortCharacter[]
        then:
            # Set the heart beat VFX's position to the prop agent's position.
            HeartBeatVFXData.Activate(Character.GetTransform())
            # Increment the heartbeat count, and if this is the first heartbeat playing, we need to play the audio to get it started.
            set NumberOfHeartBeats += 1
            if (NumberOfHeartBeats = 1) then AudioDevice.Play()
            # Register the prop agent to the audio device so the heart beat audio will play from that position.
            AudioDevice.Register(PropAgent)
        else:
            Logger.Print("Character, Index, or HeartBeatVFXData not found. Cannot start the heartbeat")

    # Clears the heartbeat VFX and SFX for the specified prop agent.
    Disable(PropAgent:agent, HeartBeatVFXData:heartbeat_vfx):void =
        Logger.Print("Disabling heart beat.")
        # Deactivate the VFX visuals.
        HeartBeatVFXData.Deactivate()
        # Unregister the prop agent from the audio device, causing the heart beat audio to stop.
        AudioDevice.Unregister(PropAgent)
        # Decrement the heartbeat counter. This counter should never drop below 0.
        set NumberOfHeartBeats -= 1
        if (NumberOfHeartBeats < 0) then set NumberOfHeartBeats = 0

    # Clears all heartbeat VFX and SFX for all prop agents.
    DisableAll():void =
        Logger.Print("Disabling all heart beats.")
        # Iterate through all VFX and move them to 0,0,0.
        for (HeartBeatVFXDevice : AgentVFX):
            HeartBeatVFXDevice.Deactivate()
        # Unregister all participants from the heart beat audio.
        AudioDevice.UnregisterAll()
        # Reinitialize the heartbeat counter to 0
        set NumberOfHeartBeats = 0

# The heartbeat_warning_ui class contains a struct of data to track the UI canvas and text_block per player as well as the function to create a new heartbeat warning UI canvas.
heartbeat_warning_ui := class:
    var Canvas:canvas = canvas{}
    var Text:text_block = text_block{}

    # Creates the UI canvas for the warning message.
    CreateCanvas():void =
        set Text = text_block{DefaultTextColor := NamedColors.White, DefaultShadowColor := NamedColors.Black}
        set Canvas = canvas:
            Slots := array:
                canvas_slot:
                    Anchors := anchors{Minimum := vector2{X := 0.5, Y := 0.75}, Maximum := vector2{X := 0.5, Y := 0.75}}
                    Offsets := margin{Top := 0.0, Left := 0.0, Right := 0.0, Bottom := 0.0}
                    Alignment := vector2{X := 0.5, Y := 1.0}
                    SizeToContent := true
                    Widget := Text

# The heartbeat_vfx class contains a struct of data to track the VFX's root and vfx_spawner_device objects per participant as well as the functions to set the VFX to a position or reset it.
heartbeat_vfx := class<concrete>:
    @editable # The VFX device for each heart beat.
    VFXDevice:vfx_spawner_device = vfx_spawner_device{}

    # This offset is used to position the heartbeat above a prop agent's head.
    HeartBeatVFXOffset:vector3 = vector3{X := 0.0, Y := 0.0, Z := 110.0}

    # Sets the position of the heart beat VFX then enables the VFX.
    Activate(Transform:transform):void =
        VFXPosition := Transform.Translation + HeartBeatVFXOffset
        if (VFXDevice.TeleportTo[VFXPosition, Transform.Rotation]):
            VFXDevice.Enable()
    
    # Disables the VFX, hiding the heart beat visuals.
    Deactivate():void =
        VFXDevice.Disable()
