# Knowledge-neutral Verse snippet; project names and asset IDs generalized.
# id: snippet-088

# A tutorial for writing this script can be found at https://dev.epicgames.com/documentation/en-us/uefn/verse-prop-hunt-template-in-unreal-editor-for-fortnite
using { /Fortnite.com/Devices }
using { /UnrealEngine.com/Temporary/Diagnostics }
using { /Verse.org/Simulation }

# Inheriting from the base_team class, The hunter_team class contains the device definitions and functions related to the hunter team and its agents. 
hunter_team := class<concrete>(base_team):
    @editable # One hunter agent is created each round for each n participants. Example: 5.0 is 1 per 5 participants. If participants = 6, 2 hunter agents are created.
    HunterAgentPerNumberOfPlayers:float = 5.0 # Minimum 1.1 is enforced to ensure at least 1 prop agent is created.

    @editable # The number of seconds before the hunter agents are spawned, giving the prop agents a head start to hide.
    SpawnDelay:float = 15.0

    @editable # The maximum base points a hunter agent gets for eliminating a prop agent. These points are divided by the number of prop agents remaining.
    MaxEliminationScore:int = 5000

    @editable # The timer device used to give props a grace period to hide.
    WaitTimer:timer_device = timer_device{}

    # Set the agent to a hunter agent. 
    InitializeAgent<override>(NewHunterAgent:agent):void =
        Logger.Print("Setting a new hunter agent.")
        (super:)InitializeAgent(NewHunterAgent)

    # When a hunter agent leaves the match, remove them from the HunterAgents array and check for the end of the round. 
    # Notice that we're overriding this function because we don't need to pass extra data here like we do for prop team.
    EliminateAgent<override>(HunterAgent:agent)<suspends>:void =
        Logger.Print("Hunter agent eliminated.")
        (super:)EliminateAgent(HunterAgent)
