# Knowledge-neutral Verse snippet; project names and asset IDs generalized.
# id: snippet-089

# A tutorial for writing this script can be found at https://dev.epicgames.com/documentation/en-us/uefn/verse-prop-hunt-template-in-unreal-editor-for-fortnite
using { /Fortnite.com/Devices }
using { /Fortnite.com/UI }
using { /UnrealEngine.com/Temporary/Diagnostics }
using { /UnrealEngine.com/Temporary/SpatialMath }
using { /UnrealEngine.com/Temporary/UI }
using { /Verse.org/Colors }
using { /Verse.org/Random }
using { /Verse.org/Simulation }

log_prop_hunt_device := class(log_channel){}

# This class contains the logic for starting and ending a round. It also handles events for participants joining and leaving when a round is in progress.
prop_hunt := class(creative_device):
    Logger:log = log{Channel:=log_prop_hunt_device}

    @editable # An instance of the prop_team class, used to manage the Prop team.
    PropTeam:prop_team = prop_team{}

    @editable # An instance of the hunter_team class, used to manage the Hunter team.
    HunterTeam:hunter_team = hunter_team{}

    @editable # The round manager device is used to end the round when the last prop agent is eliminated.
    RoundSettings:round_settings_device = round_settings_device{}

    @editable # The teleporter in the lobby area. This device is enabled and disabled to allow props and hunters out of the pre-game lobby.
    LobbyTeleporter:teleporter_device = teleporter_device{}

    @editable # The waiting for more players device. This device checks if there are enough participants to start the round.
    WaitingForMorePlayers:waiting_for_more_players = waiting_for_more_players{}

    @editable # This hud controller is used when a round is waiting to start.
    # Set Priority to "Highest" for this device and have a second HUD controller set to < "Highest" to use during the round.
    HUDControllerWaiting:hud_controller_device = hud_controller_device{}

    @editable # The round timer device. This device keeps track of the time left in the round and is started within StartTheHunt().
    RoundTimer:round_timer = round_timer{}

    @editable # The device that controls the game music. 
    GameMusic:radio_device = radio_device{}

    @editable # Message for when the hunter agents are set loose in the game.
    StartTheHuntMessage:hud_message_device = hud_message_device{}

    @editable # The Verse class that controls the opening cinematic.
    Cinematic:billboard_cinematic = billboard_cinematic{}

    # When the device is started in a running game, initialize subscriptions and launch the team setup. This runs at the start of every round.
    OnBegin<override>()<suspends>:void =
        Sleep(1.0)
        Logger.Print("Round loaded.")

        # Play the cinematic for each player at beginning if it is enabled
        if (Cinematic.PlayAtStart?):
            Cinematic.EnterCinematicModeForAll(GetPlayspace().GetPlayers())
            Cinematic.CinematicCompleted.Await()
            Cinematic.EndCinematicModeForAll(GetPlayspace().GetPlayers())
            Logger.Print("Cinematic Completed.")
        else:
            Logger.Print("Cinematic Skipped.")

        GameMusic.Play()
        SetUpTeams()
        race: # When there are no more prop or hunter agents (whichever happens first), or the round timer finishes, the round ends
            PropTeam.TeamEmptyEvent.Await()
            HunterTeam.TeamEmptyEvent.Await()
            RoundTimer.AwaitEnd()
        Logger.Print("Round ending.")
        EndRound()

    # When a round is started, subscribe to team devices, randomly pick the hunter agents, enable the hunter timer, set the prop agents, and teleport them into the game area.
    SetUpTeams()<suspends>:void =
        Logger.Print("Setting up teams.")

        # Subscribe to the prop team score timer, set the score award, and subscribe to the prop team's eliminated event.
        PropTeam.ScoreManager.SetScoreAward(PropTeam.ScorePerSecond)
        PropTeam.TeamManager.TeamMemberEliminatedEvent.Subscribe(OnPropEliminated) # Occurs when a prop agent is eliminated.

        # Subscribe to the hunter team's wait timer and set the duration. Also subscribe to the hunter team's elimination event.
        HunterTeam.WaitTimer.SuccessEvent.Subscribe(HuntersGo)
        HunterTeam.WaitTimer.SetMaxDuration(HunterTeam.SpawnDelay)
        HunterTeam.TeamManager.EnemyEliminatedEvent.Subscribe(OnHunterEliminated) # Occurs when a hunter agent eliminates a prop agent.

        # Initialize the starting hunter and prop agents arrays. Get the participants in the session.
        var StartingHunterAgents:[]agent = array{}
        var StartingPropAgents:[]agent = array{}
        var Participants:[]agent = GetPlayspace().GetParticipants()

        # Enable the HUD appropriate for the waiting for players state. 
        HUDControllerWaiting.Enable()

        # Check if there are enough participants to start the round.
        set Participants = WaitingForMorePlayers.WaitForMinimumNumberOfParticipants(Participants)
        Logger.Print("Round started.")

        # Disable the waiting HUD to use the next highest priority HUD. 
        HUDControllerWaiting.Disable()

        # Now that the round has started, we need to handle participants being added or removed from the match. Subscribe to those events.
        GetPlayspace().ParticipantAddedEvent().Subscribe(OnParticipantAdded)
        GetPlayspace().ParticipantRemovedEvent().Subscribe(OnParticipantRemoved)

        # Calculate the number of starting hunter agents for the game based. 1 Hunter is selected per HunterAgentPerNumberOfPlayers participants.
        NumberOfParticipants:float = 1.0 * Participants.Length # Converts Participants.Length to a float so it can be used in the next Ceil function.
        if (NumberOfStartingHunters:int = Ceil[NumberOfParticipants / Max(1.1, HunterTeam.HunterAgentPerNumberOfPlayers)]):
            # Shuffle the participants and then slice the array to get the starting hunter agents. The remaining participants are the starting prop agents. 
            var RandomizedParticipants:[]agent = Shuffle(Participants)
            set StartingHunterAgents = RandomizedParticipants.Slice[0,NumberOfStartingHunters] or StartingHunterAgents
            set StartingPropAgents = RandomizedParticipants.Slice[NumberOfStartingHunters,RandomizedParticipants.Length] or StartingPropAgents

            # Iterate through the starting hunter agents and assign them to the hunter team. Then start the hunter wait timer.
            Logger.Print("Setting {StartingHunterAgents.Length} hunter agent(s).")
            for (StartingHunterAgent : StartingHunterAgents):
                HunterTeam.InitializeAgent(StartingHunterAgent)
            HunterTeam.WaitTimer.Start()

            # Iterate through the starting prop agents and assign them to the prop team. Teleport them into the play area.
            Logger.Print("Setting {StartingPropAgents.Length} prop agent(s).")
            LobbyTeleporter.Enable()
            for (StartingPropAgent : StartingPropAgents):
                PropTeam.InitializeAgent(StartingPropAgent)
                PropTeam.HeartBeat.SetUpUI(StartingPropAgent)
                LobbyTeleporter.Activate(StartingPropAgent)
            LobbyTeleporter.Disable()
            # Set the props remaining tracker's Target and Value to the current number of props.
            # In the future, we'll only update Value as props are eliminated.
            PropTeam.PropsRemainingTracker.SetTarget(PropTeam.Count())
            PropTeam.UpdatePropsRemainingTracker()

    # When the hunter timer finishes, spawn the function that unleashes the hunter agents into the game and begins the round.
    # HunterInstigator is needed for the event subscription but not used.
    HuntersGo(HunterInstigator:?agent):void =
        spawn{ StartTheHunt() }

    # Enable the lobby teleporter, teleport all the hunter agents to the game area and then run the prop game loop for each prop agent.
    StartTheHunt()<suspends>:void =
        Logger.Print("Teleporting hunter agent(s).")
        LobbyTeleporter.Enable()
        Sleep(0.5) # Delaying half a second to give the teleporter time to enable.
        for (HunterAgent : HunterTeam.GetAgents[]):
            LobbyTeleporter.Activate(HunterAgent)
            StartTheHuntMessage.Show(HunterAgent)
        Logger.Print("Starting prop(s) game logic.")
        for (PropAgent : PropTeam.GetAgents[]):
            spawn {PropTeam.RunPropGameLoop(PropAgent)}
            StartTheHuntMessage.Show(PropAgent)

        # Signal that the round has started to start the round and scoring timers.
        RoundTimer.Start()
        PropTeam.ScoreTimer.Start()

    # When a participant joins the match mid-round, make them a hunter if the round has started.
    OnParticipantAdded(Agent:agent):void =
        Logger.Print("A participant joined the game.")
        if (PropTeam.Count() > 0): # if there are any prop agents, the round has started.
            HunterTeam.InitializeAgent(Agent)

    # When a hunter agent eliminates a prop agent, award score. The score is divided by the number of prop agents remaining.
    OnHunterEliminated(HunterAgent:agent):void =
        Logger.Print("Hunter agent eliminated a prop agent.")
        PropTeamSize := PropTeam.Count()
        if (EliminationAward := Floor(HunterTeam.MaxEliminationScore / PropTeamSize)):
            Logger.Print("Awarding {EliminationAward} points.")
            HunterTeam.ScoreManager.SetScoreAward(EliminationAward)
            HunterTeam.ScoreManager.Activate(HunterAgent)

    # When a prop agent is eliminated, remove the prop from the prop team, check for round end, and set them as a hunter.
    OnPropEliminated(PropAgent:agent):void =
        Logger.Print("Prop agent eliminated.")
        spawn{ PropTeam.EliminateAgent(PropAgent) }
        HunterTeam.InitializeAgent(PropAgent)
  
    # When a participant leaves the match, check which team they were on and then check for round end.
    OnParticipantRemoved(Agent:agent):void =
        Logger.Print("A participant left the game.")
        if (PropTeam.FindOnTeam[Agent]):
            Logger.Print("Participant was a Prop.")
            spawn{ PropTeam.EliminateAgent(Agent) }
        if (HunterTeam.FindOnTeam[Agent]):
            Logger.Print("Participant was a Hunter.")
            spawn{ HunterTeam.EliminateAgent(Agent) }

    # Cleans up the heart beat VFX and then ends the round.
    EndRound():void=
        PropTeam.HeartBeat.DisableAll()
        # Get any player to use as an instigator to end the round.
        if (RoundEndInstigator := GetPlayspace().GetPlayers()[0]):
            Logger.Print("Round ended.")
            RoundSettings.EndRound(RoundEndInstigator)
