# Knowledge-neutral Verse snippet; project names and asset IDs generalized.
# id: snippet-090

# A tutorial for writing this script can be found at https://dev.epicgames.com/documentation/en-us/uefn/verse-prop-hunt-template-in-unreal-editor-for-fortnite
using { /Fortnite.com/Characters }
using { /Fortnite.com/Devices }
using { /Fortnite.com/UI }
using { /UnrealEngine.com/Temporary/Diagnostics }
using { /UnrealEngine.com/Temporary/SpatialMath }
using { /UnrealEngine.com/Temporary/UI }
using { /Verse.org/Colors }
using { /Verse.org/Simulation }

# This message is used to print the number of props remaining to all players during a round.
PropAgentsRemainingMessage<localizes>(Count:int):message = "{Count} Prop(s) Remaining"

# Inheriting from the base_team class, the prop_team class contains the device definitions and functions related to the prop team and its agents.
# Notably, a prop agent's heart beat behavior can be found in this class.
prop_team := class<concrete>(base_team):
    @editable # The score a prop agent receives per second.
    ScorePerSecond:int = 10

    @editable # The minimum distance a prop agent must move to reset the heart beat timer.
    MinimumMoveDistance:float = 100.0

    @editable # The timer device used to award score to a prop.
    ScoreTimer:timer_device = timer_device{}

    @editable # This tracker device is used to display the props remaining to the screen.
    PropsRemainingTracker:tracker_device = tracker_device{}
    
    @editable # Get the properties from the heart_beat class.
    HeartBeat:heart_beat = heart_beat{}

    # Set the agent to a prop agent and assign heart beat warning UI. 
    InitializeAgent<override>(NewPropAgent:agent):void =
        Logger.Print("Setting a new prop agent.")
        (super:)InitializeAgent(NewPropAgent)

    # When a prop agent is eliminated or leaves the match, remove them from the PropAgents array and check for the end of the round.
    # Note that this isn't overridding because we're passing all participants to the function to update the props remaining message.
    EliminateAgent<override>(PropAgent:agent)<suspends>:void =
        Logger.Print("Prop agent eliminated.")
        (super:)EliminateAgent(PropAgent)
        # Update the props remaining number.
        UpdatePropsRemainingTracker()

    # Updates the value of the tracker device showing the number of props remaining.
    UpdatePropsRemainingTracker():void =
        PropsRemainingTracker.SetValue(Count())

    # If the prop agent stops moving then race to see if the prop agent moves beyond the MinimumMoveDistance, the heart beat timer completes, or the prop agent is eliminated.
    RunPropGameLoop(PropAgent:agent)<suspends>:void =
        Logger.Print("Starting prop agent game loop.")
        # Loop forever through the prop behavior and scoring until the prop agent is eliminated or the participant leaves the session.
        race: 
            PropAgent.AwaitNoLongerAProp()
            loop:
                # Wait until the prop agent moves less than the minimum distance, then advance.
                PropAgent.AwaitStopMoving(MinimumMoveDistance) 
                # Until the prop agent moves beyond the minimum distance, countdown to the heart beat and then play the heart beat indefinitely.
                race: 
                    PropAgent.AwaitStartMoving(MinimumMoveDistance)
                    block:
                        CountdownTimer(PropAgent)
                        PropAgent.StartHeartbeat()
                Sleep(0.0) # Once the race completes (the prop agent moves), start the loop again.
            loop:
                # Every second, award the prop agent points.
                Sleep(1.0)
                ScoreManager.Activate(PropAgent)


    # Loop until the prop agent is no longer a part of the PropAgents array. Removal happens if the prop agent is eliminated and turned into a hunter or if the participant leaves the session.
    (PropAgent:agent).AwaitNoLongerAProp()<suspends>:void =
        loop:
            if (not FindOnTeam[PropAgent]):
                Logger.Print("Cancelling prop agent behavior.")
                break
            Sleep(0.0) # Advance to the next game tick.

    # Loops until the agent moves less than the MinimumDistance.
    (PropAgent:agent).AwaitStopMoving(MinimumDistance:float)<suspends>:void =
        Logger.Print("Checking if the agent has moved less than the minimum distance.")
        # Get the initial position of the agent from the agent's character in the scene.
        if (Tracked := PropAgent.GetFortCharacter[]):
            var StartPosition:vector3 = Tracked.GetTransform().Translation
            loop:
                Sleep(0.0) # Get the position of the agent in the next game tick.
                NewPosition := Tracked.GetTransform().Translation
                # If the distance of the new position from the starting position is less than MinimumDistance, the agent has not moved and we break the loop.
                if (Distance(StartPosition, NewPosition) < MinimumDistance):
                    Logger.Print("Agent has moved less than the minimum distance.")
                    break
                # Otherwise, we reset StartPosition to make sure the participant moves from the new position.
                else: 
                    set StartPosition = NewPosition

    # Loops until the agent moves more than the MinimumDistance.
    (PropAgent:agent).AwaitStartMoving(MinimumDistance:float)<suspends>:void =
        Logger.Print("Checking if the agent moves further than the minimum distance.")
        # Get the initial position of the agent from the agent's character in the scene.
        if (Tracked := PropAgent.GetFortCharacter[]):
            StartPosition:vector3 = Tracked.GetTransform().Translation
            loop:
                Sleep(0.0) # Get the position of the agent in the next game tick.
                NewPosition := Tracked.GetTransform().Translation
                # If the distance of the new position from the starting position is greater than or equal to MinimumDistance, the agent has moved and we break the loop.
                if (Distance(StartPosition, NewPosition) >= MinimumDistance):
                    Logger.Print("Agent has moved more than or equal to the minimum distance.")
                    break

    # Delays until HeartBeatWarningTime should start. Then counts down by HeartBeatWarningTime and sets the countdown text. Clears the text when deferred.
    CountdownTimer(PropAgent:agent)<suspends>:void =
        Logger.Print("Starting heart beat countdown.")
        if (UIData := HeartBeat.WarningUI[PropAgent]):
            Sleep(HeartBeat.MoveTime - HeartBeat.WarningTime) # Sleep for the amount of time before the warning appears.
            Logger.Print("Starting heart beat warning.")
            var WarningTimeRemaining:int = Ceil[HeartBeat.WarningTime] or 0
            # A defer happens when the function completes or if it is cancelled, such as if it loses a race.
            # So in this case, the warning text is cleared when the countdown finishes or if the prop agent moves before the countdown finishes.
            defer:
                UIData.Text.SetText(HeartBeatWarningClear)
            # Set the warning text to the time remaining, wait a second, and then decrement the time remaining. If the countdown completes, break the loop.
            loop:
                Logger.Print("Heart beat in {WarningTimeRemaining} seconds.")
                UIData.Text.SetText(HeartBeatWarningMessage(WarningTimeRemaining))
                Sleep(1.0)
                set WarningTimeRemaining -= 1
                if (WarningTimeRemaining <= 0):
                    break
        else:
            Logger.Print("UIData not found.")
              
    # Turns on the heart beat VFX and SFX. Waits infinitely until deferred then disables the heart beat VFX and SFX.
    (PropAgent:agent).StartHeartbeat()<suspends>:void =
        Logger.Print("Spawning heart beat.")
        # Get the index of the prop agent in the PropAgents array to then access the corresponding heart beat VFX actor.
        HeartBeatVFXData := HeartBeat.AgentVFX[FindOnTeam[PropAgent]] or heartbeat_vfx{}
        HeartBeat.Enable(PropAgent, HeartBeatVFXData)
        # When this function is cancelled by the prop agent moving, being eliminated, or the participant leaving the session, disable the heartbeat.
        defer:
            HeartBeat.Disable(PropAgent, HeartBeatVFXData)
        Sleep(Inf) # Do not stop sleeping until the race completes.
