# Knowledge-neutral Verse snippet; project names and asset IDs generalized.
# id: snippet-091

# A tutorial for writing this script can be found at https://dev.epicgames.com/documentation/en-us/uefn/verse-prop-hunt-template-in-unreal-editor-for-fortnite
using { /Fortnite.com/Devices }
using { /Fortnite.com/UI }
using { /UnrealEngine.com/Temporary/Diagnostics }
using { /UnrealEngine.com/Temporary/SpatialMath }
using { /UnrealEngine.com/Temporary/UI }
using { /Verse.org/Colors }
using { /Verse.org/Simulation }

log_round_timer_device := class(log_channel){}

# An int that allows values between the ranges specified. This type is required by player_ui_slot.ZOrder.
round_int_clamped := type{_X:int where 0 <= _X, _X <= 2147483647}

# This message is used to print the time remaining before a round ends.
TimeRemainingMessage<localizes>(Minutes:string, Seconds:string):message = "{Minutes}:{Seconds}"

<#
This class contains all the logic for managing the round time and displaying the time to the screen.
You can use this device with a round_settings_device to actually end the round.
This device manages time without the use of a timer.
To use this class:
    1) Add the file to your project.
    2) Compile Verse Code from the Verse menu on the toolbar.
    3) Drag the device into your island from your island's Content/Creative Devices folder in the Content Browser.
    4) Include the waiting_for_more_players class in another Verse script with:
        @editable
        RoundTimer:round_timer = round_timer{}
    5) Compile Verse Code from the Verse menu on the toolbar.
    6) Connect the device you made in step 3 to the Verse device.
    7) Start the round timer with the following Verse:
        RoundTimer.Start()
    8) Restart or Stop the timer with the equivalent functions.
    9) Wait for the timer to start with:
        RoundTimer.AwaitStart()
    10) Wait for the timer to complete with:
        RoundTimer.AwaitEnd()
        Call the EndRound function on a round_settings_device to actually end the game round.
#>
round_timer := class(creative_device):
    Logger:log = log{Channel:=log_round_timer_device}

    @editable # The time, in minutes, a round lasts.
    RoundTimeInMinutes:float = 5.0

    @editable # The horizontal and vertical position of the timer UI on screen. X 0-1 is left-right and Y 0-1 is top-bottom.
    UIPosition:vector2 = vector2{X:= 0.98, Y:=0.13}

    @editable # The horizontal and vertical position of the timer UI on screen. X 0-1 is left-right and Y 0-1 is top-bottom.
    UIAlignment:vector2 = vector2{X := 1.0, Y := 0.0}

    @editable # The ZOrder of the UI compared to other UI elements.
    UIZOrder:round_int_clamped = 4

    # Signaled when the round has been started.
    RoundStarted:event() = event(){}

    # Signaled when the round is about to be ended.
    RoundEndedEvent:event() = event(){}

    # This map associates a text box for displaying the time to each player.  
    var TimeRemainingTextBlocks:[player]text_block = map{}

    # The time remaining before the round completes, as an integer.
    var TimeRemainingInSeconds:int = 0

    # Waits until the round timer is started.
    AwaitStart()<suspends>:void =
        RoundStarted.Await()
        Logger.Print("Round timer started.")
    
    # Used to start the round timer.
    Start():void = 
        Logger.Print("Starting the round timer.")
        RoundStarted.Signal()
        set TimeRemainingInSeconds = GetRoundTimeInSeconds()
        spawn{ Running() }

    # Restarts the timer to RoundTime
    Restart():void = 
        Logger.Print("Restarting the round timer.")
        set TimeRemainingInSeconds = GetRoundTimeInSeconds()

    # Runs the timer logic.
    Running()<suspends>:void =
        Logger.Print("Round timer running.")
        loop:
            UpdateTimeUI()
            Sleep(1.0)
            # Decrement TimeRemaining by 1 second then check if the time has run out. If so, end the round.
            set TimeRemainingInSeconds -= 1
            if (TimeRemainingInSeconds < 0):
                Stop()
                break
            
    # Stops the timer and ends the round.
    Stop():void = 
        Logger.Print("Ending the round timer.")
        RoundEndedEvent.Signal()

    # Waits until the round timer is just about to end.
    AwaitEnd()<suspends>:void =
        RoundEndedEvent.Await()
        Logger.Print("Round timer ended.")

    # Accepts a time value in minutes and returns the value in seconds.
    GetRoundTimeInSeconds():int =
        Round[RoundTimeInMinutes * 60.0] or 0

    # When the timer completes, update the time remaining and check if time has expired.
    UpdateTimeUI():void =
        # Set Minutes to TimeRemainingInSeconds/60 without the remainder.
        Minutes:int = Floor(TimeRemainingInSeconds / 60 or Err())
        # Set Seconds to the remainder of TimeRemainingInSeconds/60.
        Seconds:int = Mod[TimeRemainingInSeconds, 60] or Err()
        # Convert Minutes and Seconds to strings.
        MinutesAsString := string("{Minutes}")
        # If Seconds < 10, then we need to add a 0 in front of the value so it displays as :0# instead of :#
        SecondsAsString := if (Seconds < 10) then Join(array{string("{0}"),string("{Seconds}")},string()) else string("{Seconds}")

        # Iterate through all players, check if they have a TimeRemainingTextBlock, if not, give them one. Then update the text.
        Players:[]player = GetPlayspace().GetPlayers()
        for (Player : Players):
            TextBlock:text_block = TimeRemainingTextBlocks[Player] or block:
                Logger.Print("Creating new round time text block.")
                SetUpTimeRemainingUI(Player)
            TextBlock.SetText(TimeRemainingMessage(MinutesAsString, SecondsAsString))

    # Accepts a player and then adds a round time ui canvas to their screen and stores their TimeRemainingTextBlock for updating later.
    SetUpTimeRemainingUI(Player:player):text_block =
        Logger.Print("Adding round timer UI to a player.")
        # This is the text_block that prints the time remaining text to the screen.
        TextBlock:text_block = text_block:
            DefaultTextColor := NamedColors.White
            DefaultShadowColor := NamedColors.Black
        if (PlayerUI := GetPlayerUI[Player]):
            if (set TimeRemainingTextBlocks[Player] = TextBlock) {}
            # This is the canvas that holds and positions the text_block on the screen.
            Canvas := canvas:
                Slots := array:
                    canvas_slot:
                        Anchors := anchors{Minimum := UIPosition, Maximum := UIPosition}
                        Offsets := margin{Top := 0.0, Left := 0.0, Right := 0.0, Bottom := 0.0}
                        Alignment := UIAlignment
                        SizeToContent := true
                        Widget := TextBlock
            # The canvas is assigned to the player.
            PlayerUI.AddWidget(Canvas, player_ui_slot{ZOrder := UIZOrder})
        # The text_block is returned so it can be saved to the map and updated later as time ticks down.
        return TextBlock
