# Knowledge-neutral Verse snippet; project names and asset IDs generalized.
# id: snippet-092

# A tutorial for writing this script can be found at https://dev.epicgames.com/documentation/en-us/uefn/verse-prop-hunt-template-in-unreal-editor-for-fortnite
using { /Fortnite.com/Devices }
using { /Fortnite.com/UI }
using { /UnrealEngine.com/Temporary/Diagnostics }
using { /UnrealEngine.com/Temporary/SpatialMath }
using { /UnrealEngine.com/Temporary/UI }
using { /Verse.org/Colors }
using { /Verse.org/Simulation }

log_waiting_for_more_players_device := class(log_channel){}

# An int that allows values between the ranges specified. This type is required by player_ui_slot.ZOrder.
waiting_int_clamped := type{_X:int where 0 <= _X, _X <= 2147483647}

# This message is used to print the number of players needed before a round can begin.
WaitingForMorePlayersMessage<localizes>(Count:int):message = "Waiting for {Count} more player(s)"

# This class is for showing how many players are needed to start the round.
waiting_for_more_players_ui := class:
    var Canvas:canvas
    var TextBlock:text_block

<# 
This class contains all the logic for setting a minimum number of players and checking if there are enough to start the round.
To use this class:
    1) Add the file to your project.
    2) Compile Verse Code from the Verse menu on the toolbar.
    3) Drag the device into your island from your island's Content/Creative Devices folder in the Content Browser.
    4) Connect a Timer device to this device's "WaitingForMorePlayersTimer" property.
    5) Include the waiting_for_more_players class in another Verse script with:
        @editable
        WaitingForMorePlayers:waiting_for_more_players = waiting_for_more_players{}
    6) Compile Verse Code from the Verse menu on the toolbar.
    7) Connect the device you made in step 3 to the Verse device and property you exposed in step 5.
    8) Await the WaitForMinimumNumberOfParticipants functions by passing it an array of participants (players and/or test players). 
        For example:
        Participants = GetPlayspace().GetParticipants()
        CheckForMinimumNumberOfParticipants(Participants)
    9) On IslandSettings, Set the Game Start Countdown to 0.0.
#>
waiting_for_more_players := class(creative_device):
    Logger:log = log{Channel:=log_waiting_for_more_players_device}

    @editable # The minimum number of players needed in the match for a round to start.
    MinimumNumberOfPlayers:int = 2

    @editable # The horizontal and vertical position of the timer UI on screen. X 0-1 is left-right and Y 0-1 is top-bottom.
    UIPosition:vector2 = vector2{X:= 0.5, Y:=0.4}

    @editable # The horizontal and vertical position of the timer UI on screen. X 0-1 is left-right and Y 0-1 is top-bottom.
    UIAlignment:vector2 = vector2{X := 0.5, Y := 0.5}

    @editable # The ZOrder of the UI compared to other UI elements.
    UIZOrder:waiting_int_clamped = 3

    @editable # This timer is used to countdown to round start after waiting for players to join the match.
    WaitingForMorePlayersTimer:timer_device = timer_device{}

    # This map associates a UI canvas for displaying the number of participants needed to start a round to each player. 
    var WaitingForMorePlayersUI:[player]?waiting_for_more_players_ui = map{}

    # Check if there are enough participants to start the round. If not, wait until number of participants >= MinimumNumberOfPlayers.
    WaitForMinimumNumberOfParticipants(Participants:[]agent)<suspends>:[]agent =
        Logger.Print("Waiting if there are enough participants for the round to start.")
        # Creating a new variable so I can modify it as more participants join. Initializing it with an array of participants passed to the function.
        var ParticipantsWaiting:[]agent = Participants
        # If the number of participants is less than the minimum needed to start the round...
        if (ParticipantsWaiting.Length < MinimumNumberOfPlayers):
            loop: # Loop until the number of participants is greater than or equal to the minimum needed.
                Logger.Print("{ParticipantsWaiting.Length}/{MinimumNumberOfPlayers} players needed for the round to start.")
                # Update the waiting for more players UI.
                DisplayWaitingForMorePlayersUI(ParticipantsWaiting)
                Sleep(2.0) # Wait to see if more participants join the match then check if there are enough participants to start the round.
                set ParticipantsWaiting = GetPlayspace().GetParticipants()
                if (ParticipantsWaiting.Length >= MinimumNumberOfPlayers):
                    # If there are now enough participants, clear the waiting for more players UI, 
                    Logger.Print("{ParticipantsWaiting.Length}/{MinimumNumberOfPlayers} participants in round, preparing for round start.")
                    ClearWaitingForMorePlayersUI()
                    # Then break out of the loop.
                    break
        # start the round start countdown, and wait until the countdown completes. 
        WaitingForMorePlayersTimer.Start()
        WaitingForMorePlayersTimer.SuccessEvent.Await()
        Logger.Print("Starting round.")
        # Return the list of participants in the session.
        return ParticipantsWaiting

    # Displays a "Waiting for more players" UI message for each player if they don't already have one. Updates the player counter for all players.
    DisplayWaitingForMorePlayersUI(Participants:[]agent):void =
        PlayersNeededCount := MinimumNumberOfPlayers - Participants.Length
        Logger.Print("{Participants.Length} participants in round, waiting for {PlayersNeededCount} more to join.")
        for (Participant: Participants):
            # If the participant is a player, it can have UI.
            if (Player := player[Participant]):
                # If the player already has a WaitingForMorePlayersUI, get the TextBlock and refresh the text so it shows the correct number of participants needed to start the round.
                if (UIData := WaitingForMorePlayersUI[Player]?):
                    UIData.TextBlock.SetText(WaitingForMorePlayersMessage(PlayersNeededCount))
                # Else create a WaitingForMorePlayersUI for the player.
                else:
                    SetUpWaitingForMorePlayersUI(Player, PlayersNeededCount)
    
    # Accepts a player and player_ui and adds a waiting for more players ui canvas to their screen.
    SetUpWaitingForMorePlayersUI(Player:player, PlayersNeededCount:int):void =
        Logger.Print("Creating 'waiting for more players' UI.")
        if (PlayerUI := GetPlayerUI[Player]):
            # This is the text_block that prints the waiting for more players text to the screen.
            TextBlock:text_block = text_block:
                DefaultText := WaitingForMorePlayersMessage(PlayersNeededCount)
                DefaultTextColor := NamedColors.White
                DefaultShadowColor := NamedColors.Black
            # This is the canvas that holds and positions the text_block on the screen.
            Canvas := canvas:
                Slots := array:
                    canvas_slot:
                        Anchors := anchors{Minimum := UIPosition, Maximum := UIPosition}
                        Offsets := margin{Top := 0.0, Left := 0.0, Right := 0.0, Bottom := 0.0}
                        Alignment := UIAlignment
                        SizeToContent := true
                        Widget := TextBlock
            # The canvas is assigned to the player.
            PlayerUI.AddWidget(Canvas, player_ui_slot{ZOrder := UIZOrder})
            # The text_block is saved to the map so we can update the text later as more players join the game.
            if (set WaitingForMorePlayersUI[Player] = option{ waiting_for_more_players_ui{Canvas := Canvas, TextBlock := TextBlock} }) {}

    # Removes the "Waiting for more players" UI message for each player who has one.
    ClearWaitingForMorePlayersUI():void =
        Logger.Print("Clearing 'waiting for more players' UI.")
        for (Player -> UIData : WaitingForMorePlayersUI):
            if:
                PlayerUI := GetPlayerUI[Player]
                Canvas := UIData?.Canvas
                set WaitingForMorePlayersUI[Player] = false
            then:
                PlayerUI.RemoveWidget(Canvas)
