# Knowledge-neutral Verse snippet; project names and asset IDs generalized.
# id: snippet-093

# This file defines a candy_cane_field that handles a field of candy canes, which can be harvested.
# The candy cane field manages that the candy canes regrow over time.

using { /Fortnite.com/Devices }
using { /Verse.org/Simulation }
using { Persistence }
using { Time }
using { UserInterface }
using { Utility }

# A single candy cane in the candy cane field.
candy_cane := class<concrete>:

    # Prop manipulator used to show and hide the candy cane.
    # Also used to detect when the candy cane is harvested.
    @editable
    PropManipulator:prop_manipulator_device = prop_manipulator_device{}
    
    # The candy cane number.
    @editable
    Number:int = 0

    # Stores if the candy cane has grown.
    var Grown:logic = false
    
candy_cane_field := class(creative_device):

    # The real time device used to store the candy cane harvest times and
    # track the regrowth.
    @editable
    RealTime:real_time = real_time{}

    # UI manager for updating resource and price UIs.
    @editable
    UIManager:ui_manager = ui_manager{}

    # Prop asset containing af effect to show when candy cane is harvested and regrows.
    @editable
    CandyCaneEffect:creative_prop_asset = DefaultCreativePropAsset

    # Audio to play when candy cane regrows.
    @editable
    CandyCaneRegrowAudioPlayer:audio_player_device = audio_player_device{}
    
    # Array of candy canes in this field.
    @editable
    CandyCanes:[]candy_cane = array{}

    # Time in seconds for a candy cane to regrow.
    @editable
    CandyCaneRegrowthTime:float = 45.0

    # Amount of gold earned when harvesting a candy cane.
    @editable
    GoldPerCandyCane:int = 1

    # Event signalled when a candy cane is harvested.
    HarvestEvent:event() = event(){}

    # The selected player that owns the candy cane field. Can be false if nobody owns it yet.
    # The owner will receive gold when candy canes are harvested,
    # and will store information about harvest times.
    var MaybeSelectedPlayer:?player = false

    # Runs when the device is started in a running game.
    OnBegin<override>()<suspends>:void=
        # Subscribe callback when candy cane is harvested.
        for(CandyCane : CandyCanes):
            CandyCane.PropManipulator.HarvestingEvent.SubscribeAgent(OnHarvesting, CandyCane)

    # When the candy cane is harvested, give the selected player gold. 
    OnHarvesting(Agent:agent, CandyCane:candy_cane):void=
        # Signal that candy cane was harvested.
        HarvestEvent.Signal()

        # Hide candy cane.
        set CandyCane.Grown = false
        CandyCane.PropManipulator.HideProps()

        # Show effect.
        spawn:
            ShowCandyCaneEffect(CandyCane)

        # Give player some gold and store harvest time.
        if:
            SelectedPlayer := MaybeSelectedPlayer?
            GrantGold[SelectedPlayer, GoldPerCandyCane]
            SetCandyCaneHarvestTime[SelectedPlayer, CandyCane.Number, RealTime.GetDateAndTime()]
        then:
            # Update all players' UI.
            UIManager.UpdateAllPriceUIs()
            UIManager.UpdateAllResourceUIs()

    # Sets the selected player.
    SelectPlayer<public>(MaybePlayer:?player):void=
        set MaybeSelectedPlayer = MaybePlayer

        # Setup correct state of candy cane field.
        Update()

        if:
            SelectedPlayer := MaybeSelectedPlayer?
        then:
            # Monitor regrowth.
            for(CandyCane : CandyCanes):
                spawn:
                    MonitorRegrowth(CandyCane)

    # Updates all candy canes to reflect the current state.
    Update():void=
        Now := RealTime.GetDateAndTime()
        for(CandyCane : CandyCanes):
            # If the appropriate time since the last harvest has passed, show the candy cane.
            if:
                SelectedPlayer := MaybeSelectedPlayer?
                LatestHarvestTime := GetCandyCaneHarvestTime[SelectedPlayer, CandyCane.Number]
                TimeSinceHarvest := Now.Subtract(LatestHarvestTime)
                not LatestHarvestTime.IsValid[] or TimeSinceHarvest.GetTotalSeconds() >= CandyCaneRegrowthTime
                not CandyCane.Grown?
            then:
                # Show candy cane.
                set CandyCane.Grown = true
                CandyCane.PropManipulator.ShowProps()
            else:
                # Hide candy cane.
                set CandyCane.Grown = false
                CandyCane.PropManipulator.HideProps()

    # Monitors the regrowth of a candy cane.
    # Regrows the candy cane with an effect after the appropriate time has passed.
    MonitorRegrowth(CandyCane:candy_cane)<suspends>:void=
        loop:
            # If the selected player is removed, stop monitoring regrowth.
            if (not MaybeSelectedPlayer?):
                break

            Now := RealTime.GetDateAndTime()
            # If the appropriate time since the last harvest has passed, regrow the candy cane.
            if:
                SelectedPlayer := MaybeSelectedPlayer?
                LatestHarvestTime := GetCandyCaneHarvestTime[SelectedPlayer, CandyCane.Number]
                TimeSinceHarvest := Now.Subtract(LatestHarvestTime)
                not LatestHarvestTime.IsValid[] or TimeSinceHarvest.GetTotalSeconds() >= CandyCaneRegrowthTime
                not CandyCane.Grown?
            then:
                # Show candy cane.
                set CandyCane.Grown = true
                CandyCane.PropManipulator.ShowProps()

                # Show effect.
                spawn:
                    ShowCandyCaneEffect(CandyCane)
        
                # Play audio.
                CandyCaneRegrowAudioPlayer.Play()

            Sleep(1.0)

    # Shows the candy cane effect at the position of the candy cane.
    # Disposes it again after a few seconds.
    ShowCandyCaneEffect(CandyCane:candy_cane)<suspends>:void=
        Transform := CandyCane.PropManipulator.GetTransform()
        Position := Transform.Translation
        Rotation := Transform.Rotation
        SpawnResult := SpawnProp(CandyCaneEffect, Position, Rotation)
        if (EffectProp := SpawnResult(0)?):
            Sleep(3.0)
            EffectProp.Dispose()
