# Knowledge-neutral Verse snippet; project names and asset IDs generalized.
# id: snippet-094

# This file defines a daily_gift, which can be opened once every day.
# The gift is opened by harvesting it and will respawn when the date rolls over.

using { /Fortnite.com/Devices }
using { /Verse.org/Simulation }
using { Persistence }
using { Time }
using { UserInterface }

daily_gift := class(creative_device):

    # The real time device used to track the daily gift opening time.
    @editable
    RealTime:real_time = real_time{}

    # UI manager for updating resource and price UIs.
    @editable
    UIManager:ui_manager = ui_manager{}

    # Prop asset containing an effect to show when the gift is opened (harvested) and respawned.
    @editable
    GiftEffect:creative_prop_asset = DefaultCreativePropAsset

    # Audio to play when gift is opened (harvested).
    @editable
    GiftOpenAudioPlayer:audio_player_device = audio_player_device{}

    # Prop manipulator used to show and hide the gift.
    # Also used to detect when the gift is opened (harvested).
    @editable
    PropManipulator:prop_manipulator_device = prop_manipulator_device{}

    # Amount of gold earned when opening the gift.
    @editable
    Gold:int = 1

    # Stores if the gift is opened.
    var Opened:logic = true

    # The selected player that owns the gift. Can be false if nobody owns it yet.
    # The owner will receive gold when the gift is opened (harvested),
    # and will store information about opening time.
    var MaybeSelectedPlayer:?player = false

    # Runs when the device is started in a running game.
    OnBegin<override>()<suspends>:void=
        # Subscribe callback when gift is opened (harvested).
        PropManipulator.HarvestingEvent.Subscribe(OnHarvesting)

    # When the gift is opened (harvested), give the selected player gold. 
    OnHarvesting(Agent:agent):void=
        # Hide gift.
        set Opened = true
        PropManipulator.HideProps()

        # Show effect.
        spawn:
            ShowGiftEffect()

        # Play audio.
        GiftOpenAudioPlayer.Play()

        # Give player some gold and store opening time.
        if:
            SelectedPlayer := MaybeSelectedPlayer?
            GrantGold[SelectedPlayer, Gold]

            # Professor Flopkins: "Looks like the code to store the opening time is missing!"
            #                     "First, we should get the current date and time."
            #                     "Look in the IsDateRolledOver function for a hint on how to do this."


            # Professor Flopkins: "And then, we should store that time in the persistable player data."
            #                     "I saw a function for doing this in the bottom of player_info.verse."
 
            
        then:
            # Update all players' UI.
            UIManager.UpdateAllPriceUIs()
            UIManager.UpdateAllResourceUIs()

    # Sets the selected player.
    SelectPlayer<public>(MaybePlayer:?player):void=
        set MaybeSelectedPlayer = MaybePlayer

        # Setup correct state of gift.
        Update()

        if:
            SelectedPlayer := MaybeSelectedPlayer?
        then:
            # Monitor respawn.
            spawn:
                MonitorRespawn()

    # Updates the gift to reflect the current state.
    Update():void=
        # If the appropriate time since the last opening has passed, show the gift.
        if:
            MaybeSelectedPlayer?
            IsDateRolledOver[]
            Opened?
        then:
            # Show gift.
            set Opened = false
            PropManipulator.ShowProps()
        else:
            # Hide gift.
            set Opened = true
            PropManipulator.HideProps()

    # Monitors the respawning of gift.
    # Respawns the gift with an effect after the appropriate time has passed.
    MonitorRespawn()<suspends>:void=
        loop:
            # If the selected player is removed, stop monitoring respawn.
            if (not MaybeSelectedPlayer?):
                break

            # If the appropriate time since the last opening has passed, respawn the gift.
            if:
                IsDateRolledOver[]
                Opened?
            then:
                # Show gift.
                set Opened = false
                PropManipulator.ShowProps()

                # Show effect.
                spawn:
                    ShowGiftEffect()
        
            Sleep(1.0)

    # Succeeds if the current date is different from the opening time.
    IsDateRolledOver()<decides><transacts>:void=
        # Get the current date.
        Now := RealTime.GetDateAndTime()

        # Professor Flopkins: "Hmm, the code to determine if the date has rolled over is also missing."
        #                     "Let's see.. first we'll get the selected player."
        #                     "Look in the OnHarvesting function to see how to do that."


        # Professor Flopkins: "Now, we should retrieve the last date and time the gift was opened."
        #                     "There is a function for doing this in player_info.verse."


        # Professor Flopkins: "Finally, we should compare that date with the current date."
        #                     "We want the year, month or date to be different as that means the date"
        #                     "has rolled over."

        

    # Shows the gift effect at the position of the gift.
    # Disposes it again after a few seconds.
    ShowGiftEffect()<suspends>:void=
        Transform := PropManipulator.GetTransform()
        Position := Transform.Translation
        Rotation := Transform.Rotation
        SpawnResult := SpawnProp(GiftEffect, Position, Rotation)
        if (EffectProp := SpawnResult(0)?):
            Sleep(3.0)
            EffectProp.Dispose()
