# Knowledge-neutral Verse snippet; project names and asset IDs generalized.
# id: snippet-096

# This file defines factory_belt that handles the logic and visuals of a factory belt.

using { /Fortnite.com/Devices }
using { /Verse.org/Simulation }
using { Config }
using { Persistence }
using { Utility }

factory_belt<public> := class<concrete>(factory_element):

    # The factory mould that is connected to this belt.
    @editable
    Mould:factory_element = factory_element{}

    # The factory wrapper that is connected to this belt.
    @editable
    Wrapper:factory_element = factory_element{}
    
    # The props that make up the factory belt.
    @editable
    Props:[]creative_prop = array{}

    # Prop that defines the end of the belt.
    # Used by factory manager to determine where the products should end up.
    @editable
    EndProp:creative_prop = creative_prop{}

    # Prop that defines the position where build and upgrade effects should show.
    @editable
    EffectPositionProp:creative_prop = creative_prop{}

    # Sequence that animates the factory belt.
    @editable
    CinematicSequence:cinematic_sequence_device = cinematic_sequence_device{}

    # Audio that plays when belt is running.
    @editable
    RunningAudioPlayer:audio_player_device = audio_player_device{}
    
    # Audio that plays when belt is stopped.
    @editable
    StoppedAudioPlayer:audio_player_device = audio_player_device{}

    # The factory belt number.
    @editable
    Number<public>:int = 0

    # Stores if the factory belt is blocked.
    # This happens when the factory storage is full and there is a product at the
    # end of the belt.
    var Blocked:logic = false

    # Shows the factory belt UI and build preview.
    ShowUI<override>(Player:player):void=
        # Handle things that are in common for all factory elements when showing the UI.
        (super:)ShowUI(Player)
        # Show the UI and build preview.
        UIManager.ShowFactoryBeltUI(Player, Self)
        if:
            SelectedPlayer := MaybeSelectedPlayer?
            Level := GetFactoryBeltLevel[SelectedPlayer, Number]
            Level = 0
        then:
            ShowBuildPreview(EffectPositionProp)

    # Succeeds if factory belt is built.
    IsBuilt<override>()<decides><transacts>:void=
        IsFactoryBeltBuilt[MaybeSelectedPlayer?, Number]

    # Succeeds if factory belt can be upgraded.
    CanUpgrade<override>()<decides><transacts>:void=
        GetFactoryBeltLevel[MaybeSelectedPlayer?, Number] < GetFactoryBeltMaxLevel[Number]
    
    # Gets the price for upgrading factory belt.
    GetUpgradePrice<override>()<decides><transacts>:int=
        GetFactoryBeltUpgradePrice[Number, MaybeSelectedPlayer?]

    # Upgrades factory belt.
    Upgrade<override>():void=
        # Perform the actual upgrade if the selected player can afford it.
        if:
            SelectedPlayer := MaybeSelectedPlayer?
            Gold := GetGold[SelectedPlayer]
            Price := GetFactoryBeltUpgradePrice[Number, SelectedPlayer]
            Gold >= Price
            CanUpgrade[]
            SpendGold[SelectedPlayer, Price]
            if (not IsFactoryBeltBuilt[SelectedPlayer, Number]):
                BuildFactoryBelt[SelectedPlayer, Number]
            UpgradeFactoryBeltLevel[SelectedPlayer, Number]
        then:
            # Handle things that are in common for all factory elements when upgrading.
            (super:)Upgrade()
            # Show build or upgrade effect and play build or upgrade audio.
            if (Level := GetFactoryBeltLevel[SelectedPlayer, Number]):
                ShowBuildOrUpgradeEffect(Level, EffectPositionProp)
                PlayBuildOrUpgradeAudio(Level)
            # Update all players' UI.
            UIManager.UpdateAllPriceUIs()
            UIManager.UpdateAllResourceUIs()
            UIManager.UpdateAllFactoryUIs()

    # Updates the visuals of the factory belt to reflect the current state.
    Update<override>():void=
        # Handle things that are in common for all factory elements when updating.
        (super:)Update()
        if:
            SelectedPlayer := MaybeSelectedPlayer?
            IsFactoryBeltBuilt[SelectedPlayer, Number]
        then:
            # Show the belt if it is built.
            for(Prop : Props):
                Prop.Show()
            # Update the mould and wrapper visuals as the mould and wrapper need to hide any parts of the belt that are inside them.
            Mould.Update()
            Wrapper.Update()
            # If the belt is not blocked and the factory mould is active, animate the belt and play the running audio.
            # The belt should also be animating in the special case that the factory mould is not built yet.
            if:
                not IsBlocked[]
                IsFactoryMouldActive[SelectedPlayer, Number] or not IsFactoryMouldBuilt[SelectedPlayer, Number]
                FactoryBeltSpeed := GetFactoryBeltSpeed[Number, SelectedPlayer]
            then:
                CinematicSequence.Play()
                CinematicSequence.SetPlayRate(FactoryBeltSpeed)
                RunningAudioPlayer.Play()
                StoppedAudioPlayer.Stop()
            else:
                CinematicSequence.Pause()
                RunningAudioPlayer.Stop()
                StoppedAudioPlayer.Play()
        else:
            # Hide the belt if it is unbuilt.
            set Blocked = false
            for(Prop : Props):
                Prop.Hide()
            CinematicSequence.Stop()
            RunningAudioPlayer.Stop()
            StoppedAudioPlayer.Stop()

    # Succeeds if the belt is blocked.
    IsBlocked()<decides><transacts>:void=
        Blocked?

    # Sets if the belt is blocked.
    SetBlocked(NewBlocked:logic):void=
        set Blocked = NewBlocked
