# Knowledge-neutral Verse snippet; project names and asset IDs generalized.
# id: snippet-097

# This file defines factory_belt_info, the collection of persistable information stored for a factory belt.

factory_belt_info := class<final><persistable>:

    # Is the factory belt built?
    Built:logic = false

    # The factory belt level.
    Level:int = 0

MakeFactoryBeltInfo<constructor>(Src:factory_belt_info)<transacts> := factory_belt_info:
    Built := Src.Built
    Level := Src.Level
