# Knowledge-neutral Verse snippet; project names and asset IDs generalized.
# id: snippet-098

# This file defines the factory belt UI, which shows information for a factory belt
# in the factory of the selected player.

using { /Fortnite.com/Devices }
using { /Fortnite.com/UI }
using { /UnrealEngine.com/Temporary/SpatialMath }
using { /UnrealEngine.com/Temporary/UI }
using { /Verse.org/Colors }
using { /Verse.org/Simulation }
using { Config }
using { Factory }
using { Persistence }
using { Utility }

# Contains all widgets relevant for showing, updating and closing the UI.
factory_belt_ui_widgets := struct:
    Canvas<public>:canvas
    LevelTextWidget<public>:text_block
    SpeedTextWidget<public>:text_block
    PriceTextWidget<public>:text_block
    PriceTextureWidget<public>:texture_block
    UpgradeButtonWidget<public>:text_button_base

factory_belt_ui := class(creative_device):

    # Stores the created factory belt UI for each player.
    var UIMap:[player]factory_belt_ui_widgets = map{}
    # Stores the factory belt whose information will be shown on the UI for each player. 
    var FactoryBeltMap:[player]factory_belt = map{}

    # When a player is added, create the factory belt UI.
    PlayerAdded(Player:player):void=
        UIResult := CreateUI()
        if:
            set UIMap[Player] = UIResult

    # When a player is removed, remove the stored data.
    PlayerRemoved(Player:player):void=
        if:
            NewUIMap := UIMap.RemoveKey[Player]
            NewFactoryBeltMap := FactoryBeltMap.RemoveKey[Player]
        then:
            set UIMap = NewUIMap
            set FactoryBeltMap = NewFactoryBeltMap

    # Create the UI.
    CreateUI():factory_belt_ui_widgets=
        BackgroundColorWidget := color_block:
            DefaultColor := NamedColors.DarkGreen
            DefaultOpacity := 0.7
        LevelTextWidget := text_block:
            DefaultTextColor:= NamedColors.White
            DefaultShadowOffset:= option{vector2{X := 2.0, Y := 3.0}}
            DefaultShadowColor:= NamedColors.Black
            DefaultShadowOpacity := 1.0
        TextureWidget := texture_block:
            DefaultImage := Textures.UI.T_Belt
            DefaultDesiredSize:=vector2{X := 400.0, Y := 400.0}
        SpeedTextWidget := text_block:
            DefaultTextColor:= NamedColors.White
            DefaultShadowOffset:= option{vector2{X := 2.0, Y := 3.0}}
            DefaultShadowColor:= NamedColors.Black
            DefaultShadowOpacity := 1.0
        PriceTextWidget := text_block:
            DefaultTextColor:= NamedColors.White
            DefaultShadowOffset:= option{vector2{X := 2.0, Y := 3.0}}
            DefaultShadowColor:= NamedColors.Black
            DefaultShadowOpacity := 1.0
            DefaultJustification:=text_justification.InvariantRight
        PriceTextureWidget := texture_block:
            DefaultImage := Textures.UI.T_Gold
            DefaultDesiredSize:=vector2{X := 40.0, Y := 40.0}
        UpgradeButtonWidget := button_loud{}
        CloseButtonWidget := button_regular:
            DefaultText:= PlainText("Close")

        CanvasWidget := canvas:
            Slots := array:
                canvas_slot:
                    Anchors := anchors{Minimum := vector2{X := 0.0, Y := 0.0}, Maximum := vector2{X := 1.0, Y := 1.0}}
                    Offsets := margin{Top := 0.0, Left := 0.0, Right := 0.0, Bottom := 0.0}
                    Alignment := vector2{X := 0.5, Y := 0.5}
                    Widget := BackgroundColorWidget
                canvas_slot:
                    Anchors := anchors{Minimum := vector2{X := 0.5, Y := 0.5}, Maximum := vector2{X := 0.5, Y := 0.5}}
                    Offsets := margin{Top := -390.0, Left := 0.0, Right := 0.0, Bottom := 0.0}
                    Alignment := vector2{X := 0.5, Y := 0.5}
                    SizeToContent := true
                    Widget := LevelTextWidget
                canvas_slot:
                    Anchors := anchors{Minimum := vector2{X := 0.5, Y := 0.5}, Maximum := vector2{X := 0.5, Y := 0.5}}
                    Offsets := margin{Top := -70.0, Left := 0.0, Right := 0.0, Bottom := 0.0}
                    Alignment := vector2{X := 0.5, Y := 0.5}
                    Widget := TextureWidget
                canvas_slot:
                    Anchors := anchors{Minimum := vector2{X := 0.5, Y := 0.5}, Maximum := vector2{X := 0.5, Y := 0.5}}
                    Offsets := margin{Top := 90.0, Left := 0.0, Right := 0.0, Bottom := 0.0}
                    Alignment := vector2{X := 0.5, Y := 0.5}
                    SizeToContent := true
                    Widget := SpeedTextWidget
                canvas_slot:
                    Anchors := anchors{Minimum := vector2{X := 0.5, Y := 0.5}, Maximum := vector2{X := 0.5, Y := 0.5}}
                    Offsets := margin{Top := 250.0, Left := -180.0, Right := 0.0, Bottom := 0.0}
                    Alignment := vector2{X := 1.0, Y := 0.5}
                    SizeToContent := true
                    Widget := PriceTextWidget
                canvas_slot:
                    Anchors := anchors{Minimum := vector2{X := 0.5, Y := 0.5}, Maximum := vector2{X := 0.5, Y := 0.5}}
                    Offsets := margin{Top := 250.0, Left := -150.0, Right := 0.0, Bottom := 0.0}
                    Alignment := vector2{X := 0.5, Y := 0.5}
                    Widget := PriceTextureWidget
                canvas_slot:
                    Anchors := anchors{Minimum := vector2{X := 0.5, Y := 0.5}, Maximum := vector2{X := 0.5, Y := 0.5}}
                    Offsets := margin{Top := 250.0, Left := 0.0, Right := 0.0, Bottom := 0.0}
                    Alignment := vector2{X := 0.5, Y := 0.5}
                    SizeToContent := true
                    Widget := UpgradeButtonWidget
                canvas_slot:
                    Anchors := anchors{Minimum := vector2{X := 0.5, Y := 0.5}, Maximum := vector2{X := 0.5, Y := 0.5}}
                    Offsets := margin{Top := 350.0, Left := 0.0, Right := 0.0, Bottom := 0.0}
                    Alignment := vector2{X := 0.5, Y := 0.5}
                    SizeToContent := true
                    Widget := CloseButtonWidget

        Canvas := canvas:
            Slots := array:
                canvas_slot:
                    Anchors := anchors{Minimum := vector2{X := 0.0, Y := 0.0}, Maximum := vector2{X := 0.42, Y := 1.0}}
                    Offsets := margin{Top := 50.0, Left := 50.0, Right := 50.0, Bottom := 120.0}
                    Alignment := vector2{X := 0.5, Y := 0.5}
                    Widget := CanvasWidget

        # Subscribe callbacks to get notified when the player clicks the buttons.
        UpgradeButtonWidget.OnClick().Subscribe(OnUpgradeButton)
        CloseButtonWidget.OnClick().SubscribeWidgetMessage(OnCloseButton, 0.0)

        return factory_belt_ui_widgets:
            Canvas := Canvas
            LevelTextWidget := LevelTextWidget
            SpeedTextWidget := SpeedTextWidget
            PriceTextWidget := PriceTextWidget
            PriceTextureWidget := PriceTextureWidget
            UpgradeButtonWidget := UpgradeButtonWidget

    # Show the UI to the player.
    ShowUI(Player:player, FactoryBelt:factory_belt, AudioPlayer:audio_player_device):void=
        if:
            set FactoryBeltMap[Player] = FactoryBelt
            UI := UIMap[Player]
            PlayerUI := GetPlayerUI[Player]
        then:
            # Play UI audio.
            AudioPlayer.Play(Player)

            # Update and show UI.
            UpdateUI(Player)
            PlayerUI.AddWidget(UI.Canvas, player_ui_slot{ZOrder := 0, InputMode := ui_input_mode.All})

    # Update the UI for the player.
    UpdateUI(Player:player):void=
        if:
            UI := UIMap[Player]
            FactoryBelt := FactoryBeltMap[Player]
            SelectedPlayer := FactoryBelt.GetSelectedPlayer[]
            UpgradePrice := GetFactoryBeltUpgradePrice[FactoryBelt.Number, SelectedPlayer]
            Built := logic{IsFactoryBeltBuilt[SelectedPlayer, FactoryBelt.Number]}
            Level := GetFactoryBeltLevel[SelectedPlayer, FactoryBelt.Number]
            MaxLevel := GetFactoryBeltMaxLevel[FactoryBelt.Number]
            CanUpgrade := logic{FactoryBelt.CanUpgrade[]}
            UpgradeEnabled := logic{CanUpgrade?, GetGold[SelectedPlayer] >= UpgradePrice}
            Speed := GetFactoryBeltSpeed[FactoryBelt.Number, SelectedPlayer]
            ProductsPerHour := 3600.0 / (FactoryBeltMinDistanceToProduce / Speed)
            SpeedText := "{Round[ProductsPerHour]} units per hour"
        then:
            # Adapt the UI to whether or not the factory belt has been built yet.
            if (Built?):
                UI.LevelTextWidget.SetText(LevelText(Level, MaxLevel))
                UI.SpeedTextWidget.SetText(PlainText(SpeedText))
                UI.UpgradeButtonWidget.SetText(PlainText("Upgrade"))
            else:
                UI.LevelTextWidget.SetText(PlainText("Build factory belt")) 
                UI.SpeedTextWidget.SetText(PlainText(""))
                UI.UpgradeButtonWidget.SetText(PlainText("Build"))

            # Only show the price if the player can upgrade.
            if (CanUpgrade?):
                UI.PriceTextWidget.SetText(CountText(UpgradePrice))
                UI.PriceTextureWidget.SetVisibility(widget_visibility.Visible)
            else:
                UI.PriceTextWidget.SetText(PlainText(""))
                UI.PriceTextureWidget.SetVisibility(widget_visibility.Hidden)
            UI.UpgradeButtonWidget.SetEnabled(UpgradeEnabled)

    # When the player clicks the upgrade button, upgrade the factory belt.
    # If it was the first upgrade, close the UI with a slight delay.
    OnUpgradeButton(Message:widget_message):void=
        if:
            FactoryBelt := FactoryBeltMap[Message.Player]
            SelectedPlayer := FactoryBelt.GetSelectedPlayer[]
        then:
            FactoryBelt.Upgrade()

            if:
                FactoryBeltLevel := GetFactoryBeltLevel[SelectedPlayer, FactoryBelt.Number]
                FactoryBeltLevel = 1
            then:
                OnCloseButton(Message, 1.0)
            else:
                UpdateUI(Message.Player)

    # When the player clicks the close button, close the UI.
    # Pass on a potential delay for factory belt notification of the closure.
    OnCloseButton(Message:widget_message, CloseDelay:float):void=
        CloseUI(Message.Player, CloseDelay)

    # Close the UI for the player.
    # Potentially, delay the notification to the factory belt.
    CloseUI<public>(Player:player, CloseDelay:float):void=
        if:
            UI := UIMap[Player]
            PlayerUI := GetPlayerUI[Player]
        then:
            spawn:
                DelayedUIClosed(Player, CloseDelay)
            PlayerUI.RemoveWidget(UI.Canvas)

    # Delay the notification to the factory belt.
    DelayedUIClosed<private>(Player:player, Delay:float)<suspends>:void=
        if:
            FactoryBelt := FactoryBeltMap[Player]
        then:
            Sleep(Delay)
            # Notify the factory belt that the UI was closed.
            FactoryBelt.UIClosed(Player)
