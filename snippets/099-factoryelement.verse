# Knowledge-neutral Verse snippet; project names and asset IDs generalized.
# id: snippet-099

# This file defines factory_element that contains all common device references and functions
# of the different factory element types (belt, mould, wrapper and storage).

using { /Fortnite.com/Characters }
using { /Fortnite.com/Devices }
using { /Verse.org/Simulation }
using { Persistence }
using { UserInterface }

factory_element<public> := class(creative_device):

    # UI manager for showing and updating UI.
    @editable
    UIManager:ui_manager = ui_manager{}

    # The trigger that initiates interaction with the factory element. 
    @editable
    Trigger:trigger_device = trigger_device{}

    # Prop indicating that the factory element can be upgraded.
    @editable
    UpgradePossibleProp:creative_prop = creative_prop{}

    # Prop indicating that the factory element cannot be upgraded.
    @editable
    UpgradeNotPossibleProp:creative_prop = creative_prop{}

    # Mutator zone for detecting if any players are inside the factory element when it is built.
    @editable
    MutatorZone:mutator_zone_device = mutator_zone_device{}

    # Array of factory elements that must be built before this one.
    @editable
    DependentFactoryElements:[]factory_element = array{}

    # Prop asset showing a preview of the built factory element when interacting with it
    # before building it.
    @editable
    BuildPreview:creative_prop_asset = DefaultCreativePropAsset

    # Prop asset containing an effect to show when building the factory element.
    @editable
    BuildEffect:creative_prop_asset = DefaultCreativePropAsset

    # Audio that plays when factory element is built.
    @editable
    BuildAudioPlayer:audio_player_device = audio_player_device{}
    
    # Prop asset containing an effect to show when upgrading the factory element.
    @editable
    UpgradeEffect:creative_prop_asset = DefaultCreativePropAsset

    # Audio that plays when factory element is upgraded.
    @editable
    UpgradeAudioPlayer:audio_player_device = audio_player_device{}
    
    # Camera to use while interacting with the factory element.
    @editable
    InteractCamera:gameplay_camera_fixed_point_device = gameplay_camera_fixed_point_device{}

    # Event signalled when upgrading factory element.
    UpgradeEvent<public>:event() = event(){}

    # The selected player that owns the factory element. Can be false if nobody owns it yet.
    # The owner will pay for all upgrades and store the state of the factory element.
    var MaybeSelectedPlayer:?player = false

    # Currently showing build preview prop.
    var CurrentBuildPreviewProp:creative_prop = creative_prop{}

    # Locks the trigger out from being enabled while interacting with factory element.
    var TriggerLocked:logic = false

    # Runs when the device is started in a running game.
    OnBegin<override>()<suspends>:void=
        # Subscribe callback when interaction trigger is triggered.
        Trigger.TriggeredEvent.Subscribe(OnTriggered)
        # Continuously monitor if factory element can be built and upgraded.
        spawn:
            MonitorBuildAndUpgradeAvailability()

    # Sets the selected player.
    SelectPlayer(MaybePlayer:?player):void=
        set MaybeSelectedPlayer = MaybePlayer
        # Remove any players that might be stuck inside the factory element.
        RemovePlayersInside()

    GetSelectedPlayer<public>()<decides><transacts>:player=
        MaybeSelectedPlayer?

    # When the interaction trigger is triggered, show the UI for the factory element.
    OnTriggered(MaybeAgent:?agent):void=
        if:
            Agent := MaybeAgent?
            Player := player[Agent]
        then:
            ShowUI(Player)

    # Sets up the interaction camera, disables and locks the interaction trigger.
    # Overridden in each specific factory element type to actually show the specific UI
    # as well as potentially showing the build preview.
    ShowUI(Player:player):void=
        InteractCamera.AddTo(Player)
        Trigger.Disable()
        set TriggerLocked = true

    # Removes the interaction camera, reenables and unlocks the interaction trigger.
    # Also hides the build preview.
    UIClosed<public>(Player:player):void=
        InteractCamera.RemoveFrom(Player)
        Trigger.Enable()
        set TriggerLocked = false
        HideBuildPreview()

    # Succeeds if factory element is built.
    # Should be overridden by each specific factory element type.
    IsBuilt<public>()<decides><transacts>:void=
        false?

    # Succeeds if the factory element can be built by checking if all
    # factory elements it depends on have been built.
    CanBuild()<decides><transacts>:void=
        MaybeSelectedPlayer?
        for:
            DependentFactoryElement : DependentFactoryElements
            not DependentFactoryElement.IsBuilt[]
        do:
            false?

    # Succeeds if factory element can be upgraded.
    # Should be overridden by each specific factory element type.
    CanUpgrade<public>()<decides><transacts>:void=
        false?

    # Gets the price for upgrading factory element.
    # Should be overridden by each specific factory element type.
    GetUpgradePrice()<decides><transacts>:int=
        0

    # Upgrades factory element.
    # Overridden in each specific factory element type to actually perform the upgrade.
    Upgrade<public>():void=
        # Remove any players that might be stuck inside the upgraded element.
        RemovePlayersInside()
        # Signal that factory element was upgraded.
        UpgradeEvent.Signal()
        # Update the visuals.
        Update()
        # Hide the build preview if it is showing.
        HideBuildPreview()

    # Updates the visuals of the factory element to reflect the current state.
    # Overridden in each specific factory element type to perform the specific visual update.
    Update():void=
        # If the factory element is built, there is no need to continue detecting players
        # inside the element. 
        if (IsBuilt[]):
            MutatorZone.Disable()
        else:
            MutatorZone.Enable()

    # Monitors if factory element can be built and upgraded.
    # Updates the props accordingly.
    MonitorBuildAndUpgradeAvailability()<suspends>:void=
        loop:
            if:
                CanBuild[]
            then:
                if (not TriggerLocked?):
                    Trigger.Enable()
                if:
                    CanUpgrade[]
                    IsUpgradeAffordable[]
                then:
                    UpgradePossibleProp.Show()
                    UpgradeNotPossibleProp.Hide()
                else:
                    UpgradePossibleProp.Hide()
                    UpgradeNotPossibleProp.Show()
            else:
                Trigger.Disable()
                UpgradePossibleProp.Hide()
                UpgradeNotPossibleProp.Hide()
            Sleep(0.1)

    # Succeeds if the upgrade is affordable.
    IsUpgradeAffordable()<decides><transacts>:void=
        SelectedPlayer := MaybeSelectedPlayer?
        Gold := GetGold[SelectedPlayer]
        Price := GetUpgradePrice[]
        Gold >= Price

    # Shows the effect when building or upgrading at the position of the prop.
    ShowBuildOrUpgradeEffect(Level:int, EffectPositionProp:creative_prop):void=
        Effect := if (Level = 1) then BuildEffect else UpgradeEffect
        spawn:
            ShowEffect(Effect, EffectPositionProp)

    # Plays the audio when building or upgrading.
    PlayBuildOrUpgradeAudio(Level:int):void=
        AudioPlayer := if (Level = 1) then BuildAudioPlayer else UpgradeAudioPlayer
        AudioPlayer.Play()
        
    # Spawns the effect at the position of the prop.
    # Disposes it again after a few seconds.
    ShowEffect(Effect:creative_prop_asset, EffectPositionProp:creative_prop)<suspends>:void=
        Transform := EffectPositionProp.GetTransform()
        SpawnResult := SpawnProp(Effect, Transform)
        if (EffectProp := SpawnResult(0)?):
            Sleep(3.0)
            EffectProp.Dispose()

    # Shows the build preview at the position of the prop.
    ShowBuildPreview(PreviewPositionProp:creative_prop):void=
        Transform := PreviewPositionProp.GetTransform()
        SpawnResult := SpawnProp(BuildPreview, Transform)
        if (PreviewProp := SpawnResult(0)?):
            set CurrentBuildPreviewProp = PreviewProp

    # Disposes the build preview.
    HideBuildPreview():void=
        CurrentBuildPreviewProp.Dispose()

    # Removes any players inside the factory element's mutator zone by making them respawn.
    RemovePlayersInside():void=
        for(Player : GetPlayspace().GetPlayers()):
            if:
                MutatorZone.IsInVolume[Player]
                Character := Player.GetFortCharacter[]
            then:
                # Respawn the player.
                Character.Damage(100.0)
