# Knowledge-neutral Verse snippet; project names and asset IDs generalized.
# id: snippet-100

# This file defines factory_intensity_lights that handles the logic and visuals of a number of lights
# showing a steadily decreasing intensity.

using { /Fortnite.com/Devices }
using { /Verse.org/Simulation }

factory_intensity_lights := class<concrete>:

    # The props making up the lights.
    @editable
    LightProps:[]creative_prop = array{}

    # Time in seconds for the intensity to go from full to empty.
    @editable
    LightCoolingTime:float = 1.0

    # Event signalled to stop the current light animation.
    StopEvent:event() = event(){}

    # Stores if the lights are currently shown.
    var IsShowing:logic = false

    # The current intensity as a value between 0.0 and 1.0.
    var CurrentIntensity:float = 0.0

    # Shows the lights.
    Show():void=
        if (not IsShowing?):
            set IsShowing = true
            # Start animating the lights.
            spawn:
                AnimateLights()

    # Hides the lights.
    Hide():void=
        if (IsShowing?):
            set IsShowing = false
            set CurrentIntensity = 0.0
            # Stop any running animation.
            StopEvent.Signal()
            for(LightProp : LightProps):
                LightProp.Hide()

    # Animates the lights by continuously showing the current intensity until interrupted by the stop event.
    AnimateLights()<suspends>:void=
        var PreviousElapsedTime:float = GetSimulationElapsedTime()
        race:
            StopEvent.Await()
            loop:
                # Determine how much time has passed since last iteration of the loop.
                CurrentElapsedTime := GetSimulationElapsedTime()
                DeltaTime := CurrentElapsedTime - PreviousElapsedTime
                set PreviousElapsedTime = CurrentElapsedTime
                # Update the current intensity.
                set CurrentIntensity = Max(0.0, CurrentIntensity - DeltaTime / LightCoolingTime)
                # Determine the last light index.
                if:
                    LastLightIndex := Ceil[LightProps.Length * CurrentIntensity]
                then:
                    # Show all lights until just before the last light index. 
                    for (Index := 0..LastLightIndex-1, LightProp := LightProps[Index]):
                        LightProp.Show()
                    # Hide the rest of the lights. 
                    for (Index := LastLightIndex..LightProps.Length-1, LightProp := LightProps[Index]):
                        LightProp.Hide()

                Sleep(0.1)

    # Sets the current intensity.
    SetIntensity(Intensity:float):void=
        set CurrentIntensity = Intensity
