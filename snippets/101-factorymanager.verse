# Knowledge-neutral Verse snippet; project names and asset IDs generalized.
# id: snippet-101

# This file defines the factory_manager that handles initialising and running a factory.
# It is also able to calculate how much was produced while the player was offline.
# Finally, it can calculate when the factory storage will be full.

using { /Fortnite.com/Devices }
using { /UnrealEngine.com/Temporary/SpatialMath }
using { /Verse.org/Simulation }
using { Config }
using { Persistence }
using { Time }
using { UserInterface }
using { Utility }

# The minimum distance a product must move on the belt before the mould can produce again.
FactoryBeltMinDistanceToProduce<public>:float = 4.0

factory_manager<public> := class(creative_device):

    # The belts of the factory.
    @editable
    FactoryBelts:[]factory_belt = array{}

    # The moulds of the factory.
    @editable
    FactoryMoulds:[]factory_mould = array{}

    # The wrappers of the factory.
    @editable
    FactoryWrappers:[]factory_wrapper = array{}

    # The storage of the factory.
    @editable
    FactoryStorage:factory_storage = factory_storage{}

    # UI manager for showing offline production UI.
    @editable
    UIManager:ui_manager = ui_manager{}

    # The real time device used to calculate offline production, calculate when the storage is full and
    # run the online production.
    @editable
    RealTime:real_time = real_time{}

    # Event signalled when offline production has been calculated.
    # An array containing the number of products produced for each factory line is sent.
    OfflineProductionEvent<public>:event([]int) = event([]int){}

    # The vertical offset from the mould's position to where the belt is located.
    FactoryBeltOffset:vector3 = vector3{X := 0.0, Y := 0.0, Z := 68.0}
    
    # The selected player that owns the factory. Can be false if nobody owns it yet.
    # The owner will pay for all upgrades and store the state of the factory.
    var MaybeSelectedPlayer:?player = false
    
    # The factory products that are currently moving down the factory belts.
    var FactoryProductsOnBelts:[]factory_product = array{}

    # Runs when the device is started in a running game.
    OnBegin<override>()<suspends>:void=
        # Hide factory initially since no factory has been loaded yet.
        Update()

    # Sets the selected player. This will initialise the factory and start running it.
    SelectPlayer<public>(MaybePlayer:?player):void=
        set MaybeSelectedPlayer = MaybePlayer

        # Propagate selected player to all factory elements.
        for(FactoryBelt : FactoryBelts):
            FactoryBelt.SelectPlayer(MaybePlayer)

        for(FactoryMould : FactoryMoulds):
            FactoryMould.SelectPlayer(MaybePlayer)

        for(FactoryWrapper : FactoryWrappers):
            FactoryWrapper.SelectPlayer(MaybePlayer)
    
        FactoryStorage.SelectPlayer(MaybePlayer)

        # Setup correct state of factory.
        Update()

        if:
            SelectedPlayer := MaybeSelectedPlayer?
        then:
            # Calculate offline production.
            ProductsProducedOffline := CalculateOfflineProduction()
            UIManager.ShowOfflineProductionUI(SelectedPlayer, Self, ProductsProducedOffline)

            # Initialise online production.
            InitialiseProduction()

            # Run online production.
            for(Index := 0..FactoryBelts.Length-1):
                if:
                    FactoryBelt := FactoryBelts[Index]
                    FactoryMould := FactoryMoulds[Index]
                    FactoryWrapper := FactoryWrappers[Index]
                then:
                    spawn:
                        RunProduction(FactoryBelt, FactoryMould, FactoryWrapper)

    # Gets the selected player.
    GetSelectedPlayer<public>()<decides><transacts>:player=
        MaybeSelectedPlayer?

    # Updates all factory elements to reflect the current state.
    Update<public>():void=
        for(FactoryBelt : FactoryBelts):
            FactoryBelt.Update()

        for(FactoryMould : FactoryMoulds):
            FactoryMould.Update()

        for(FactoryWrapper : FactoryWrappers):
            FactoryWrapper.Update()
    
        FactoryStorage.Update()

    # Calculates how many products were produced while offline.
    # Any products on the belt when going offline are not counted by this calculation.
    CalculateOfflineProduction<public>():[]int=
        # Stores how many products were produced during offline production.
        var ProductsProducedOffline:[]int = array{}

        if:
            SelectedPlayer := MaybeSelectedPlayer?
            RemainingStorageCapacity := GetFactoryStorageCapacity[SelectedPlayer] - GetAllFactoryProducts[SelectedPlayer]
        then:
            Now := RealTime.GetDateAndTime()
            NowTotalSeconds := Now.GetTotalSeconds()

            # The calculation is complicated by the fact that all factory lines share the same storage.
            # To find the correct result, first determine when the storage will become full.
            # Solve the following equation, where X is the time when the storage is full, S is the storage capacity,
            # A is the time (seconds) per product of the first mould, O is the latest production time (seconds) of the first mould plus delivery time (seconds):
            # (X-O)/A = S
            # For four moulds the equation becomes:
            # (X-O)/A + (X-P)/B + (X-Q)/C + (X-R)/D = S,
            # where A, B, C and D are times per product and O, P, Q and R are latest production times plus delivery times.
            # The general solution for four moulds is given by:
            # X = (ABCDS + ABCR + ABDQ + ACDP + BCDO) / (ABC + ABD + ACD + BCD)
            # The following code constructs these terms in the numerator and denominator while taking into account that some of the moulds are not active.
            var Numerator:float = 1.0 * RemainingStorageCapacity
            var Denominator:float = 0.0
            var NextTermToAdd:float = 1.0
            for (FactoryMould : FactoryMoulds):
                set ProductsProducedOffline += array{0}

                if:
                    IsFactoryMouldActive[SelectedPlayer, FactoryMould.Number]
                    LatestProductionTimeTotalSeconds := GetFactoryMouldLatestProductionTime[SelectedPlayer, FactoryMould.Number].GetTotalSeconds()
                    DeliveryTime := GetFactoryBeltDeliveryTime[FactoryMould.Number, SelectedPlayer]
                    ProductionTime := GetFactoryMouldProductionTime[FactoryMould.Number, SelectedPlayer]
                    MinDistanceTime := FactoryBeltMinDistanceToProduce / GetFactoryBeltSpeed[FactoryMould.Number, SelectedPlayer]
                    ActualProductionTime := Max(ProductionTime, MinDistanceTime)
                then:
                    set Numerator *= ActualProductionTime
                    set Denominator *= ActualProductionTime
                    set Numerator += (LatestProductionTimeTotalSeconds + DeliveryTime) * NextTermToAdd
                    set Denominator += NextTermToAdd
                    set NextTermToAdd *= ActualProductionTime

            TimeWhenFull := Numerator / Denominator

            # If the time is well defined, calculate how much each mould has contributed.
            if (TimeWhenFull.IsFinite[]):
                for (FactoryMould : FactoryMoulds):
                    if:
                        IsFactoryMouldActive[SelectedPlayer, FactoryMould.Number]
                        LatestProductionTimeTotalSeconds := GetFactoryMouldLatestProductionTime[SelectedPlayer, FactoryMould.Number].GetTotalSeconds()
                        DeliveryTime := GetFactoryBeltDeliveryTime[FactoryMould.Number, SelectedPlayer]
                        ProductionTime := GetFactoryMouldProductionTime[FactoryMould.Number, SelectedPlayer]
                        MinDistanceTime := FactoryBeltMinDistanceToProduce / GetFactoryBeltSpeed[FactoryMould.Number, SelectedPlayer]
                        ActualProductionTime := Max(ProductionTime, MinDistanceTime)
                    then:
                        # Determine the total time this mould has to produce and store products.
                        TimeToProduce := Max(0.0, Min(NowTotalSeconds, TimeWhenFull) - (LatestProductionTimeTotalSeconds + DeliveryTime))
                        if:
                            # If the time when the storage becomes full is in the future, ceil the amount of products to ensure that all products are counted as stored.
                            # If the time is in the past, round the amount of products to ensure that the storage is not overflowed.
                            Products := if (NowTotalSeconds > TimeWhenFull) then Round[TimeToProduce / ActualProductionTime] else Ceil[TimeToProduce / ActualProductionTime]
                            set ProductsProducedOffline[FactoryMould.Number] = Products
                            StoreFactoryProduct[SelectedPlayer, FactoryMould.Number, Products]
                            SetFactoryMouldLatestProductionTime[SelectedPlayer, FactoryMould.Number, MakeDateAndTime(LatestProductionTimeTotalSeconds + Products * ActualProductionTime)]
            
            # Signal that offline production was performed.
            OfflineProductionEvent.Signal(ProductsProducedOffline)

        return ProductsProducedOffline

    # Calculates the date and time at which the storage will be full.
    # This shares similarities with CalculateOfflineProduction but does not actually store anything during the calculation.
    # Also, it takes products already moving down the belts into consideration.
    CalculateDateAndTimeForFullStorage<public>():date_and_time=
        Now := RealTime.GetDateAndTime()
        NowTotalSeconds := Now.GetTotalSeconds()

        # Since the factory might not have any active moulds, the storage might never be filled. Therefore start out with the result being infinite. 
        var ResultTotalSeconds:float = Inf

        if:
            SelectedPlayer := MaybeSelectedPlayer?
            AllFactoryProducts := GetAllFactoryProducts[SelectedPlayer]
            Capacity := GetFactoryStorageCapacity[SelectedPlayer]
        then:
            # Keeps track of how much storage space is left.
            var StorageSpaceLeft : int = Capacity - AllFactoryProducts 
            # The calculation is complicated by the fact that all factory lines share the same storage.
            # First, continually calculate the next time each product on the belts will reach the storage.
            # During this time, new products produced could overtake other products due to different belt speeds,
            # so take these into account in the calculation as well.
            var NextStoreTimesTotalSeconds:[]float = array{}
            for(FactoryMould : FactoryMoulds):
                if:
                    IsFactoryMouldActive[SelectedPlayer, FactoryMould.Number]
                    # If just one mould is active, register a valid, finite value. 
                    set ResultTotalSeconds = NowTotalSeconds
                    LatestProductionTime := GetFactoryMouldLatestProductionTime[SelectedPlayer, FactoryMould.Number]
                    LatestProductionTimeTotalSeconds := if (LatestProductionTime.IsValid[]) then LatestProductionTime.GetTotalSeconds() else NowTotalSeconds
                    ProductionTime := GetFactoryMouldProductionTime[FactoryMould.Number, SelectedPlayer]
                    DeliveryTime := GetFactoryBeltDeliveryTime[FactoryMould.Number, SelectedPlayer]
                    StartTime := NowTotalSeconds + Max(0.0, ProductionTime - (NowTotalSeconds - LatestProductionTimeTotalSeconds))
                then:
                    set NextStoreTimesTotalSeconds = NextStoreTimesTotalSeconds + array{StartTime + DeliveryTime}
                else:
                    set NextStoreTimesTotalSeconds = NextStoreTimesTotalSeconds + array{Inf}

            # Any products already on the belt due to production initialisation or running production should also contribute storage times.
            for(FactoryProduct : FactoryProductsOnBelts):
                if:
                    FactoryBelt := FactoryBelts[FactoryProduct.Number]
                    IsFactoryMouldActive[SelectedPlayer, FactoryProduct.Number]
                    FactoryBeltSpeed := GetFactoryBeltSpeed[FactoryProduct.Number, SelectedPlayer]
                    DeliveryTimeLeft := FactoryProduct.DistanceLeft / FactoryBeltSpeed
                then:
                    set NextStoreTimesTotalSeconds = NextStoreTimesTotalSeconds + array{NowTotalSeconds + DeliveryTimeLeft}

            # Calculate and cache the production time for each mould.
            var ActualProductionTimes:[]float = array{}
            for(Index := 0..FactoryMoulds.Length-1):
                if:
                    ProductionTime := GetFactoryMouldProductionTime[Index, SelectedPlayer]
                    MinDistanceTime := FactoryBeltMinDistanceToProduce / GetFactoryBeltSpeed[Index, SelectedPlayer]
                    ActualProductionTime := Max(ProductionTime, MinDistanceTime)
                then:
                    set ActualProductionTimes += array{ActualProductionTime}

            # While there are products on the belt and room left in the storage, find the next, valid storage time and store it as the intermediate result of the calculation.
            loop:
                if (NextStoreTimesTotalSeconds.Length = FactoryMoulds.Length):
                    break

                var NextStoreTimeTotalSeconds:float = Inf
                var NextStoreIndex:int = -1
                for(Index := 0..NextStoreTimesTotalSeconds.Length-1):
                    if:
                        ThisStoreTimeTotalSeconds := NextStoreTimesTotalSeconds[Index]
                        ThisStoreTimeTotalSeconds < NextStoreTimeTotalSeconds
                        set NextStoreTimeTotalSeconds = ThisStoreTimeTotalSeconds
                        set NextStoreIndex = Index

                if:
                    StorageSpaceLeft > 0
                    NextStoreTimeTotalSeconds.IsFinite[]
                    set StorageSpaceLeft -= 1
                    set ResultTotalSeconds = NextStoreTimeTotalSeconds
                    # If this storage comes straight from a mould, update it to the next storage time for this factory line.
                    if (NextStoreIndex < FactoryMoulds.Length):
                        set NextStoreTimesTotalSeconds[NextStoreIndex] += ActualProductionTimes[NextStoreIndex]
                    # Otherwise, the storage comes from a product that was on the belt already, so remove the storage time.
                    else:
                        NewNextStoreTimesTotalSeconds := NextStoreTimesTotalSeconds.RemoveElement[NextStoreIndex]
                        set NextStoreTimesTotalSeconds = NewNextStoreTimesTotalSeconds
                else:
                    break

            # If there is still storage space left, compute when the storage will become full from mould production alone.
            # See CalculateOfflineProduction for an explanation of the equation that is solved.
            if (StorageSpaceLeft > 0):
                var Numerator:float = 1.0 * (Capacity - FactoryProductsOnBelts.Length)
                var Denominator:float = 0.0
                var NextTermToAdd:float = 1.0
                for (FactoryMould : FactoryMoulds):

                    if:
                        IsFactoryMouldActive[SelectedPlayer, FactoryMould.Number]
                        LatestProductionTime := GetFactoryMouldLatestProductionTime[SelectedPlayer, FactoryMould.Number]
                        LatestProductionTimeTotalSeconds := if (LatestProductionTime.IsValid[]) then LatestProductionTime.GetTotalSeconds() else NowTotalSeconds
                        DeliveryTime := GetFactoryBeltDeliveryTime[FactoryMould.Number, SelectedPlayer]
                        ProductionTime := GetFactoryMouldProductionTime[FactoryMould.Number, SelectedPlayer]
                        MinDistanceTime := FactoryBeltMinDistanceToProduce / GetFactoryBeltSpeed[FactoryMould.Number, SelectedPlayer]
                        ActualProductionTime := Max(ProductionTime, MinDistanceTime)
                    then:
                        set Numerator *= ActualProductionTime
                        set Denominator *= ActualProductionTime
                        set Numerator += (LatestProductionTimeTotalSeconds + DeliveryTime) * NextTermToAdd
                        set Denominator += NextTermToAdd
                        set NextTermToAdd *= ActualProductionTime

                # This result is not entirely accurate since different products that are halfway down the belt can sum to one product in the storage
                # making the storage full even though none of the products have actually been delivered.
                set ResultTotalSeconds = Numerator / Denominator

        # If the result is a valid, finite value, convert to a valid date_and_time.
        return MakeDateAndTime(ResultTotalSeconds)


    # Shows the factory belts as initially blocked and full of products if the storage is full when initialising the factory.
    InitialiseProduction():void=
        if:
            SelectedPlayer := MaybeSelectedPlayer?
            not IsRoomInFactoryStorage[SelectedPlayer]
        then:
            for(Index := 0..FactoryBelts.Length-1):
                if:
                    FactoryBelt := FactoryBelts[Index]
                    FactoryMould := FactoryMoulds[Index]
                    FactoryWrapper := FactoryWrappers[Index]
                    IsFactoryBeltBuilt[SelectedPlayer, FactoryBelt.Number]
                    IsFactoryMouldBuilt[SelectedPlayer, FactoryMould.Number]
                    IsFactoryMouldActive[SelectedPlayer, FactoryMould.Number]
                then:
                    # Block the belt.
                    FactoryBelt.SetBlocked(true)
                    FactoryBelt.Update()

                    # Fill the belt with products.
                    var CurrentBeltDistance:float = 0.0
                    if:
                        FactoryBeltSpeed := GetFactoryBeltSpeed[FactoryBelt.Number, SelectedPlayer]
                        FactoryMouldProductionTime := GetFactoryMouldProductionTime[FactoryBelt.Number, SelectedPlayer]
                    then:
                        # Calculate how much the belt moves between productions. This is used to determine the distance between products on the belt.
                        FactoryBeltMovementBetweenProduction := Max(FactoryBeltMinDistanceToProduce, FactoryMouldProductionTime * FactoryBeltSpeed)
                        # Starting from the end of the belt, while there is room left on the belt, place a product with the determined distance to the previous.
                        loop:
                            spawn:
                                Produce(FactoryBelt, FactoryMould, FactoryWrapper, CurrentBeltDistance, ?PlayMouldAndWrapperAnimation := false)
                            set CurrentBeltDistance += FactoryBeltMovementBetweenProduction
                            if (CurrentBeltDistance > FactoryBeltTransportDistance):
                                break

    # Runs the production for a factory line consisting of belt, mould and wrapper.
    RunProduction(FactoryBelt:factory_belt, FactoryMould:factory_mould, FactoryWrapper:factory_wrapper)<suspends>:void=
        if:
            SelectedPlayer := MaybeSelectedPlayer?
            var PreviousElapsedTime:float = GetSimulationElapsedTime()
            var TimePassed:float = 0.0
            var DistanceSinceLatestProduction:float = Inf
            # Determine how much time has passed since the last production of the mould.
            LatestProductionTime := GetFactoryMouldLatestProductionTime[SelectedPlayer, FactoryMould.Number]
            if:
                LatestProductionTime.IsValid[]
            then:
                StartTime := RealTime.GetDateAndTime()
                SecondsSinceLatestProduction := StartTime.Subtract(LatestProductionTime).GetTotalSeconds()
                set TimePassed = SecondsSinceLatestProduction
            else:
                if:
                    FactoryMouldProductionTime := GetFactoryMouldProductionTime[FactoryMould.Number, SelectedPlayer]
                    set TimePassed = FactoryMouldProductionTime
        then:
            loop:
                # If the selected player is removed, stop the production.
                if (not MaybeSelectedPlayer?):
                    break

                # Determine how much time has passed since last iteration of the loop.
                CurrentElapsedTime := GetSimulationElapsedTime()
                DeltaTime := CurrentElapsedTime - PreviousElapsedTime
                set PreviousElapsedTime = CurrentElapsedTime
                # Update the time passed.
                set TimePassed += DeltaTime
                # If the belt is not blocked and the mould is active, the belt will move.
                # Update the distance that it has moved.
                if:
                    not FactoryBelt.IsBlocked[]
                    IsFactoryMouldActive[SelectedPlayer, FactoryMould.Number]
                    FactoryBeltSpeed := GetFactoryBeltSpeed[FactoryBelt.Number, SelectedPlayer]
                then:
                    set DistanceSinceLatestProduction += FactoryBeltSpeed * DeltaTime
                # If the belt is not blocked and the mould is active, determine if enough time has passed
                # and if the belt has moved far enough for the mould to be vacant from the previous production.
                # If so, produce a new product.
                if:
                    not FactoryBelt.IsBlocked[]
                    IsFactoryMouldActive[SelectedPlayer, FactoryMould.Number]
                    FactoryMouldProductionTime := GetFactoryMouldProductionTime[FactoryMould.Number, SelectedPlayer]
                    TimePassed >= FactoryMouldProductionTime and DistanceSinceLatestProduction >= FactoryBeltMinDistanceToProduce
                    set TimePassed = 0.0
                    set DistanceSinceLatestProduction = 0.0
                    SetFactoryMouldLatestProductionTime[SelectedPlayer, FactoryMould.Number, RealTime.GetDateAndTime()]
                then:
                    spawn:
                        Produce(FactoryBelt, FactoryMould, FactoryWrapper, FactoryBeltTransportDistance)
                Sleep(0.0)

    # Produces a product and tracks it all the way down the factory line through a wrapper until it reaches the storage.
    # There is an option to not play mould and wrapper animations in case the production is part of the initialisation.
    Produce(FactoryBelt:factory_belt, FactoryMould:factory_mould, FactoryWrapper:factory_wrapper, InitialDistance:float, ?PlayMouldAndWrapperAnimation:logic = true)<suspends>:void=
        if (PlayMouldAndWrapperAnimation?):
            FactoryMould.Produce()

        # Determine where to spawn the product prop and where it should end up.
        FactoryMouldTransform := FactoryMould.PositionProp.GetTransform()
        BeltStartPosition := FactoryMouldTransform.Translation + FactoryBeltOffset
        BeltEndPosition := FactoryBelt.EndProp.GetTransform().Translation + FactoryBeltOffset

        # Determine where along the journey the wrapper is placed.
        FactoryWrapperTransform := FactoryWrapper.PositionProp.GetTransform()
        WrapperPosition := FactoryWrapperTransform.Translation + FactoryBeltOffset
        DistanceToWrapper := (WrapperPosition - BeltEndPosition).Length()
        var WrapperTriggered:logic = false

        if:
            Direction := (BeltStartPosition - BeltEndPosition).MakeUnitVector[]
            InitialPosition := BeltEndPosition + InitialDistance * BRICKUnitsToFortniteUnits * Direction
        then:
            # Spawn the product prop.
            SpawnResult := SpawnProp(FactoryMould.Product, InitialPosition, IdentityRotation())
            if:
                var Prop:creative_prop = SpawnResult(0)?
            then:
                # Keep track of products currently on the belts.
                FactoryProduct := factory_product{Number := FactoryMould.Number, DistanceLeft := InitialDistance}
                set FactoryProductsOnBelts += array{FactoryProduct}

                var PreviousElapsedTime:float = GetSimulationElapsedTime()
                loop:
                    if:
                        SelectedPlayer := MaybeSelectedPlayer?
                        Prop.IsValid[]
                    then:
                        # Determine how much time has passed since last iteration of the loop.
                        CurrentElapsedTime := GetSimulationElapsedTime()
                        DeltaTime := CurrentElapsedTime - PreviousElapsedTime
                        set PreviousElapsedTime = CurrentElapsedTime
                        # Update the position at which the product should be now depending on if the belt is blocked and the mould is active.
                        if:
                            not FactoryBelt.IsBlocked[]
                            IsFactoryMouldActive[SelectedPlayer, FactoryMould.Number]
                            FactoryBeltSpeed := GetFactoryBeltSpeed[FactoryBelt.Number, SelectedPlayer]
                        then:
                            set FactoryProduct.DistanceLeft -= FactoryBeltSpeed * DeltaTime
                        CurrentPosition := BeltEndPosition + FactoryProduct.DistanceLeft * BRICKUnitsToFortniteUnits * Direction
                        # Check if the product has passed the wrapper. If it has, replace the prop with the wrapper's prop.
                        if:
                            IsFactoryWrapperBuilt[SelectedPlayer, FactoryWrapper.Number]
                            not WrapperTriggered?
                            FactoryProduct.DistanceLeft * BRICKUnitsToFortniteUnits < DistanceToWrapper
                        then:
                            if (PlayMouldAndWrapperAnimation?):
                                FactoryWrapper.Process(SelectedPlayer)
                            set WrapperTriggered = true
                            OldProp := Prop
                            NewSpawnResult := SpawnProp(FactoryWrapper.Product, Prop.GetTransform())
                            if:
                                NewProp := NewSpawnResult(0)?
                                set Prop = NewProp
                            then:
                                # Hide the prop before disposing it to avoid showing BRICK explosion effect.
                                OldProp.Hide()
                                OldProp.Dispose()

                        # Move the product prop a little.
                        Prop.MoveTo(CurrentPosition, IdentityRotation(), 0.1)
                        # When the product has reached the storage, check if there is any room left.
                        if (FactoryProduct.DistanceLeft <= 0.0):
                            # If there is, store the product, unblock the belt and return. The storage will continue handling the product prop.
                            # If there is no room, block the factory belt, and try again later.
                            if:
                                IsRoomInFactoryStorage[SelectedPlayer]
                                StoreFactoryProduct[SelectedPlayer, FactoryMould.Number, 1]
                                NewFactoryProductsOnBelts := FactoryProductsOnBelts.RemoveAll[FactoryProduct, CompareFactoryProducts]
                                set FactoryProductsOnBelts = NewFactoryProductsOnBelts
                            then:
                                FactoryStorage.Store(FactoryMould.Number, Prop)
                                if (FactoryBelt.IsBlocked[]):
                                    FactoryBelt.SetBlocked(false)
                                    FactoryBelt.Update()
                            else:
                                if (not FactoryBelt.IsBlocked[]):
                                    FactoryBelt.SetBlocked(true)
                                    FactoryBelt.Update()
                            if:
                                not FactoryBelt.IsBlocked[]
                            then:
                                return
                    else:
                        # If the selected player is removed or the prop became invalid, dispose of the product and return.
                        if:
                            NewFactoryProductsOnBelts := FactoryProductsOnBelts.RemoveAll[FactoryProduct, CompareFactoryProducts]
                            set FactoryProductsOnBelts = NewFactoryProductsOnBelts
                        # Hide the prop before disposing it to avoid showing BRICK explosion effect.
                        Prop.Hide()
                        Prop.Dispose()
                        return
