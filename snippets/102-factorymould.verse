# Knowledge-neutral Verse snippet; project names and asset IDs generalized.
# id: snippet-102

# This file defines factory_mould that handles the logic and visuals of a factory mould.

using { /Fortnite.com/Devices }
using { /Verse.org/Simulation }
using { Config }
using { Persistence }
using { Time }

factory_mould<public> := class<concrete>(factory_element):

    # The factory belt that is connected to this mould.
    @editable
    Belt:factory_element = factory_element{}

    # The real time device used to reset the production date and time when the player
    # activates the factory mould manually.
    # Also used to update the progress lights.
    @editable
    RealTime:real_time = real_time{}

    # The props that make up the factory mould.
    @editable
    Props:[]creative_prop = array{}

    # The prop that is the section of the factory belt that is covered by the factory mould.
    @editable
    BeltSectionProp:creative_prop = creative_prop{}

    # Prop that defines the position of the mould.
    # Used by factory manager to determine where the products should start.
    @editable
    PositionProp:creative_prop = creative_prop{}

    # Prop that defines the position where build and upgrade effects should show.
    @editable
    EffectPositionProp:creative_prop = creative_prop{}
    
    # Lights that show off the progress until the next time the factory mould produces. 
    @editable
    ProgressLights:factory_progress_lights = factory_progress_lights{}

    # Rotor light that shows if the factory mould is active and the belt is not blocked.
    @editable
    RotorLight:factory_rotor_light = factory_rotor_light{}

    # Sequence that animates the factory mould when producing.
    @editable
    CinematicSequence:cinematic_sequence_device = cinematic_sequence_device{}

    # Prop asset containing an effect to show when producing.
    @editable
    Effect:creative_prop_asset = DefaultCreativePropAsset

    # Audio that plays from console.
    @editable
    ConsoleAudioPlayer:audio_player_device = audio_player_device{}

    # Audio that plays when producing.
    @editable
    RamAudioPlayer:audio_player_device = audio_player_device{}
    
    # Prop asset containing the product prop that is produced by the factory mould.
    @editable
    Product<public>:creative_prop_asset = DefaultCreativePropAsset

    # The factory mould number.
    @editable
    Number<public>:int = 0

    # Runs when the device is started in a running game.
    OnBegin<override>()<suspends>:void=
        # Handle things that are in common for all factory elements when beginning.
        (super:)OnBegin()
        # Continuously update the progress and rotor lights. 
        UpdateLights()

    # Shows the factory mould UI and build preview.
    ShowUI<override>(Player:player):void=
        # Handle things that are in common for all factory elements when showing the UI.
        (super:)ShowUI(Player)
        # Show the UI and build preview.
        UIManager.ShowFactoryMouldUI(Player, Self)
        if:
            SelectedPlayer := MaybeSelectedPlayer?
            Level := GetFactoryMouldLevel[SelectedPlayer, Number]
            Level = 0
        then:
            ShowBuildPreview(EffectPositionProp)
    
    # Succeeds if factory mould is built.
    IsBuilt<override>()<decides><transacts>:void=
        IsFactoryMouldBuilt[MaybeSelectedPlayer?, Number]

    # Succeeds if factory mould can be upgraded.
    CanUpgrade<override>()<decides><transacts>:void=
        GetFactoryMouldLevel[MaybeSelectedPlayer?, Number] < GetFactoryMouldMaxLevel[Number]

    # Gets the price for upgrading factory mould.
    GetUpgradePrice<override>()<decides><transacts>:int=
        GetFactoryMouldUpgradePrice[Number, MaybeSelectedPlayer?]
    
    # Upgrades factory mould.
    Upgrade<override>():void=
        # Perform the actual upgrade if the selected player can afford it.
        if:
            SelectedPlayer := MaybeSelectedPlayer?
            Gold:= GetGold[SelectedPlayer]
            Price := GetFactoryMouldUpgradePrice[Number, SelectedPlayer]
            Gold >= Price
            CanUpgrade[]
            SpendGold[SelectedPlayer, Price]
            if (not IsFactoryMouldBuilt[SelectedPlayer, Number]):
                BuildFactoryMould[SelectedPlayer, Number]
            UpgradeFactoryMouldLevel[SelectedPlayer, Number]
        then:
            # Handle things that are in common for all factory elements when upgrading.
            (super:)Upgrade()
            # Show build or upgrade effect and play build or upgrade audio.
            if (Level := GetFactoryMouldLevel[SelectedPlayer, Number]):
                ShowBuildOrUpgradeEffect(Level, EffectPositionProp)
                PlayBuildOrUpgradeAudio(Level)
            # Update all players' UI.
            UIManager.UpdateAllPriceUIs()
            UIManager.UpdateAllResourceUIs()
            UIManager.UpdateAllFactoryUIs()
    
    # Updates the visuals of factory mould to reflect the current state.
    Update<override>():void=
        # Handle things that are in common for all factory elements when updating.
        (super:)Update()
        if:
            SelectedPlayer := MaybeSelectedPlayer?
            IsFactoryMouldBuilt[SelectedPlayer, Number]
        then:
            # Show the mould if it is built.
            for(Prop : Props):
                Prop.Show()
            # Hide the section of the belt that is covered by the mould.
            BeltSectionProp.Hide()
            # Play the console audio.
            ConsoleAudioPlayer.Play()
        else:
            # Hide the mould if it is unbuilt.
            for(Prop : Props):
                Prop.Hide()
            CinematicSequence.Stop()
            RamAudioPlayer.Stop()
            ConsoleAudioPlayer.Stop()

    # Shows factory mould producing.
    Produce():void=
        CinematicSequence.Play()
        RamAudioPlayer.Play()
        spawn:
            ShowRamEffect()

    # Shows the ram effect with a slight delay timed to the ram animation.
    ShowRamEffect()<suspends>:void=
        Sleep(0.25)
        ShowEffect(Effect, EffectPositionProp)

    # Succeeds if factory mould is active.
    IsActive()<decides><transacts>:void=
        IsFactoryMouldActive[MaybeSelectedPlayer?, Number]

    # Toggles the active state of factory mould.
    ToggleActive<public>():void=
        if:
            SelectedPlayer := MaybeSelectedPlayer?
            not IsFactoryMouldActive[SelectedPlayer, Number]
            SetFactoryMouldActive[SelectedPlayer, Number, true]
            # Reset factory mould production time when activating.
            SetFactoryMouldLatestProductionTime[SelectedPlayer, Number, RealTime.GetDateAndTime()]
        else if:
            SelectedPlayer := MaybeSelectedPlayer?
            IsFactoryMouldActive[SelectedPlayer, Number]
            SetFactoryMouldActive[SelectedPlayer, Number, false]

        # Update visuals for both factory belt and factory mould as
        # the belt depends on the active state of the mould.
        Belt.Update()
        Update()

        # Update all players' UI.
        UIManager.UpdateAllFactoryUIs()

    # Updates the lights of factory mould.
    UpdateLights()<suspends>:void=
        loop:
            # If the factory mould is built and active, show and update the progress lights.
            if:
                SelectedPlayer := MaybeSelectedPlayer?
                IsBuilt[]
                IsFactoryMouldActive[SelectedPlayer, Number]
                LatestProductionTime := GetFactoryMouldLatestProductionTime[SelectedPlayer, Number]
                ProductionTime := GetFactoryMouldProductionTime[Number, SelectedPlayer]
            then:
                Now := RealTime.GetDateAndTime()
                TimePassed := Now.Subtract(LatestProductionTime).GetTotalSeconds()
                ProgressLights.SetProgress(TimePassed / ProductionTime)
                ProgressLights.Show()
                # If the factory belt is also not blocked, show the rotor light.
                if:
                    FactoryBelt := factory_belt[Belt]
                    not FactoryBelt.IsBlocked[]
                then:
                    RotorLight.Show()
                else:
                    RotorLight.Hide()
            else:
                ProgressLights.Hide()
                RotorLight.Hide()
            Sleep(0.0)
