# Knowledge-neutral Verse snippet; project names and asset IDs generalized.
# id: snippet-103

# This file defines factory_mould_info, the collection of persistable information stored for a factory mould.

using { Time }

factory_mould_info := class<final><persistable>:

    # Is the factory mould built?
    Built:logic = false

    # Is the factory mould active?
    Active:logic = false

    # The factory mould level.
    Level:int = 0

    # The latest time that the factory mould produced a product.
    LatestProductionTime:date_and_time = date_and_time{}

MakeFactoryMouldInfo<constructor>(Src:factory_mould_info)<transacts> := factory_mould_info:
    Built := Src.Built
    Active := Src.Active
    Level := Src.Level
    LatestProductionTime := Src.LatestProductionTime
