# Knowledge-neutral Verse snippet; project names and asset IDs generalized.
# id: snippet-104

# This file defines the factory mould UI, which shows information for a factory mould
# in the factory of the selected player.

using { /Fortnite.com/Devices }
using { /Fortnite.com/UI }
using { /UnrealEngine.com/Temporary/SpatialMath }
using { /UnrealEngine.com/Temporary/UI }
using { /Verse.org/Colors }
using { /Verse.org/Simulation }
using { Config }
using { Factory }
using { Persistence }
using { Utility }

# Contains all widgets relevant for showing, updating and closing the UI.
factory_mould_ui_widgets := struct:
    Canvas<public>:canvas
    LevelTextWidget<public>:text_block
    UnbuiltTextureWidget<public>:texture_block
    ProductTextureWidget<public>:texture_block
    ProductionTextWidget<public>:text_block
    ActivateButtonWidget<public>:text_button_base
    PriceTextWidget<public>:text_block
    PriceTextureWidget<public>:texture_block
    UpgradeButtonWidget<public>:text_button_base

factory_mould_ui := class(creative_device):

    # Stores the created factory mould UI for each player.
    var UIMap:[player]factory_mould_ui_widgets = map{}
    # Stores the factory mould whose information will be shown on the UI for each player. 
    var FactoryMouldMap:[player]factory_mould = map{}

    # When a player is added, create the factory mould UI.
    PlayerAdded(Player:player):void=
        UIResult := CreateUI()
        if:
            set UIMap[Player] = UIResult

    # When a player is removed, remove the stored data.
    PlayerRemoved(Player:player):void=
        if:
            NewUIMap := UIMap.RemoveKey[Player]
            NewFactoryMouldMap := FactoryMouldMap.RemoveKey[Player]
        then:
            set UIMap = NewUIMap
            set FactoryMouldMap = NewFactoryMouldMap

    # Create the UI.
    CreateUI():factory_mould_ui_widgets=
        BackgroundColorWidget := color_block:
            DefaultColor := NamedColors.DarkGreen
            DefaultOpacity := 0.7
        LevelTextWidget := text_block:
            DefaultTextColor:= NamedColors.White
            DefaultShadowOffset:= option{vector2{X := 2.0, Y := 3.0}}
            DefaultShadowColor:= NamedColors.Black
            DefaultShadowOpacity := 1.0
        UnbuiltTextureWidget := texture_block:
            DefaultImage := Textures.UI.T_Machine
            DefaultDesiredSize:=vector2{X := 400.0, Y := 400.0}
        ProductTextureWidget := texture_block:
            DefaultImage := Textures.UI.T_Hidden
            DefaultDesiredSize:=vector2{X := 200.0, Y := 200.0}
        ProductionTextWidget := text_block:
            DefaultTextColor:= NamedColors.White
            DefaultShadowOffset:= option{vector2{X := 2.0, Y := 3.0}}
            DefaultShadowColor:= NamedColors.Black
            DefaultShadowOpacity := 1.0
        ActivateButtonWidget := button_loud{}
        PriceTextWidget := text_block:
            DefaultTextColor:= NamedColors.White
            DefaultShadowOffset:= option{vector2{X := 2.0, Y := 3.0}}
            DefaultShadowColor:= NamedColors.Black
            DefaultShadowOpacity := 1.0
            DefaultJustification:=text_justification.InvariantRight
        PriceTextureWidget := texture_block:
            DefaultImage := Textures.UI.T_Gold
            DefaultDesiredSize:=vector2{X := 40.0, Y := 40.0}
        UpgradeButtonWidget := button_loud{}
        CloseButtonWidget := button_regular:
            DefaultText:= PlainText("Close")

        CanvasWidget := canvas:
            Slots := array:
                canvas_slot:
                    Anchors := anchors{Minimum := vector2{X := 0.0, Y := 0.0}, Maximum := vector2{X := 1.0, Y := 1.0}}
                    Offsets := margin{Top := 0.0, Left := 0.0, Right := 0.0, Bottom := 0.0}
                    Alignment := vector2{X := 0.5, Y := 0.5}
                    Widget := BackgroundColorWidget
                canvas_slot:
                    Anchors := anchors{Minimum := vector2{X := 0.5, Y := 0.5}, Maximum := vector2{X := 0.5, Y := 0.5}}
                    Offsets := margin{Top := -390.0, Left := 0.0, Right := 0.0, Bottom := 0.0}
                    Alignment := vector2{X := 0.5, Y := 0.5}
                    SizeToContent := true
                    Widget := LevelTextWidget
                canvas_slot:
                    Anchors := anchors{Minimum := vector2{X := 0.5, Y := 0.5}, Maximum := vector2{X := 0.5, Y := 0.5}}
                    Offsets := margin{Top := -70.0, Left := 0.0, Right := 0.0, Bottom := 0.0}
                    Alignment := vector2{X := 0.5, Y := 0.5}
                    Widget := UnbuiltTextureWidget
                canvas_slot:
                    Anchors := anchors{Minimum := vector2{X := 0.5, Y := 0.5}, Maximum := vector2{X := 0.5, Y := 0.5}}
                    Offsets := margin{Top := -160.0, Left := 0.0, Right := 0.0, Bottom := 0.0}
                    Alignment := vector2{X := 0.5, Y := 0.5}
                    Widget := ProductTextureWidget
                canvas_slot:
                    Anchors := anchors{Minimum := vector2{X := 0.5, Y := 0.5}, Maximum := vector2{X := 0.5, Y := 0.5}}
                    Offsets := margin{Top := -10.0, Left := 0.0, Right := 0.0, Bottom := 0.0}
                    Alignment := vector2{X := 0.5, Y := 0.5}
                    SizeToContent := true
                    Widget := ProductionTextWidget
                canvas_slot:
                    Anchors := anchors{Minimum := vector2{X := 0.5, Y := 0.5}, Maximum := vector2{X := 0.5, Y := 0.5}}
                    Offsets := margin{Top := 150.0, Left := 0.0, Right := 0.0, Bottom := 0.0}
                    Alignment := vector2{X := 0.5, Y := 0.5}
                    SizeToContent := true
                    Widget := ActivateButtonWidget
                canvas_slot:
                    Anchors := anchors{Minimum := vector2{X := 0.5, Y := 0.5}, Maximum := vector2{X := 0.5, Y := 0.5}}
                    Offsets := margin{Top := 250.0, Left := -180.0, Right := 0.0, Bottom := 0.0}
                    Alignment := vector2{X := 1.0, Y := 0.5}
                    SizeToContent := true
                    Widget := PriceTextWidget
                canvas_slot:
                    Anchors := anchors{Minimum := vector2{X := 0.5, Y := 0.5}, Maximum := vector2{X := 0.5, Y := 0.5}}
                    Offsets := margin{Top := 250.0, Left := -150.0, Right := 0.0, Bottom := 0.0}
                    Alignment := vector2{X := 0.5, Y := 0.5}
                    Widget := PriceTextureWidget
                canvas_slot:
                    Anchors := anchors{Minimum := vector2{X := 0.5, Y := 0.5}, Maximum := vector2{X := 0.5, Y := 0.5}}
                    Offsets := margin{Top := 250.0, Left := 0.0, Right := 0.0, Bottom := 0.0}
                    Alignment := vector2{X := 0.5, Y := 0.5}
                    SizeToContent := true
                    Widget := UpgradeButtonWidget
                canvas_slot:
                    Anchors := anchors{Minimum := vector2{X := 0.5, Y := 0.5}, Maximum := vector2{X := 0.5, Y := 0.5}}
                    Offsets := margin{Top := 350.0, Left := 0.0, Right := 0.0, Bottom := 0.0}
                    Alignment := vector2{X := 0.5, Y := 0.5}
                    SizeToContent := true
                    Widget := CloseButtonWidget

        Canvas := canvas:
            Slots := array:
                canvas_slot:
                    Anchors := anchors{Minimum := vector2{X := 0.0, Y := 0.0}, Maximum := vector2{X := 0.42, Y := 1.0}}
                    Offsets := margin{Top := 50.0, Left := 50.0, Right := 50.0, Bottom := 120.0}
                    Alignment := vector2{X := 0.5, Y := 0.5}
                    Widget := CanvasWidget

        # Subscribe callbacks to get notified when the player clicks the buttons.
        ActivateButtonWidget.OnClick().Subscribe(OnActivateButton)
        UpgradeButtonWidget.OnClick().Subscribe(OnUpgradeButton)
        CloseButtonWidget.OnClick().SubscribeWidgetMessage(OnCloseButton, 0.0)

        return factory_mould_ui_widgets:
            Canvas := Canvas
            LevelTextWidget := LevelTextWidget
            UnbuiltTextureWidget := UnbuiltTextureWidget
            ProductTextureWidget := ProductTextureWidget
            ProductionTextWidget := ProductionTextWidget
            ActivateButtonWidget := ActivateButtonWidget
            PriceTextWidget := PriceTextWidget
            PriceTextureWidget := PriceTextureWidget
            UpgradeButtonWidget := UpgradeButtonWidget            

    # Show the UI to the player.
    ShowUI(Player:player, FactoryMould:factory_mould, AudioPlayer:audio_player_device):void=
        if:
            set FactoryMouldMap[Player] = FactoryMould
            UI := UIMap[Player]
            PlayerUI := GetPlayerUI[Player]
        then:
            # Play UI audio.
            AudioPlayer.Play(Player)

            # Update and show UI.
            UpdateUI(Player)
            PlayerUI.AddWidget(UI.Canvas, player_ui_slot{ZOrder := 0, InputMode := ui_input_mode.All})

    # Update the UI for the player.
    UpdateUI(Player:player):void=
        if:
            UI := UIMap[Player]
            FactoryMould := FactoryMouldMap[Player]
            SelectedPlayer := FactoryMould.GetSelectedPlayer[]
            UpgradePrice := GetFactoryMouldUpgradePrice[FactoryMould.Number, SelectedPlayer]
            Built := logic{IsFactoryMouldBuilt[SelectedPlayer, FactoryMould.Number]}
            Level := GetFactoryMouldLevel[SelectedPlayer, FactoryMould.Number]
            MaxLevel := GetFactoryMouldMaxLevel[FactoryMould.Number]
            CanUpgrade := logic{FactoryMould.CanUpgrade[]}
            UpgradeEnabled := logic{CanUpgrade?, GetGold[SelectedPlayer] >= UpgradePrice}
            ProductionTime := GetFactoryMouldProductionTime[FactoryMould.Number, SelectedPlayer]
            ProductsPerHour := 3600.0 / ProductionTime
            Active := logic{IsFactoryMouldActive[SelectedPlayer, FactoryMould.Number]}
            ProductionText := if (Active?) then "{Round[ProductsPerHour]} units per hour" else "Not active"
        then:
            # Adapt the UI to whether or not the factory mould has been built yet.
            if (Built?):
                UI.LevelTextWidget.SetText(LevelText(Level, MaxLevel))
                UI.UnbuiltTextureWidget.SetVisibility(widget_visibility.Hidden)
                UI.ProductTextureWidget.SetVisibility(widget_visibility.Visible)
                case(FactoryMould.Number):
                    0 =>
                        UI.ProductTextureWidget.SetImage(Textures.UI.T_Plane)
                    1 =>
                        UI.ProductTextureWidget.SetImage(Textures.UI.T_TRex)
                    2 =>
                        UI.ProductTextureWidget.SetImage(Textures.UI.T_Castle)
                    _ =>
                        UI.ProductTextureWidget.SetImage(Textures.UI.T_Spaceship)
                UI.ProductionTextWidget.SetText(PlainText(ProductionText))
                UI.ActivateButtonWidget.SetVisibility(widget_visibility.Visible)
                UI.ActivateButtonWidget.SetText(PlainText(if (Active?) then "Deactivate" else "Activate"))
                UI.UpgradeButtonWidget.SetText(PlainText("Upgrade"))
            else:
                UI.LevelTextWidget.SetText(PlainText("Build to start producing"))
                UI.UnbuiltTextureWidget.SetVisibility(widget_visibility.Visible)
                UI.ProductTextureWidget.SetVisibility(widget_visibility.Hidden)
                UI.ProductionTextWidget.SetText(PlainText(""))
                UI.ActivateButtonWidget.SetVisibility(widget_visibility.Hidden)
                UI.UpgradeButtonWidget.SetText(PlainText("Build"))

            # Only show the price if the player can upgrade.
            if (CanUpgrade?):
                UI.PriceTextWidget.SetText(CountText(UpgradePrice))
                UI.PriceTextureWidget.SetVisibility(widget_visibility.Visible)
            else:
                UI.PriceTextWidget.SetText(PlainText(""))
                UI.PriceTextureWidget.SetVisibility(widget_visibility.Hidden)
            UI.UpgradeButtonWidget.SetEnabled(UpgradeEnabled)

    # When the player clicks the activate button, toggle the mould's activation.
    OnActivateButton(Message:widget_message):void=
        if:
            FactoryMould := FactoryMouldMap[Message.Player]
        then:
            FactoryMould.ToggleActive()

            UpdateUI(Message.Player)
    
    # When the player clicks the upgrade button, upgrade the factory mould.
    # If it was the first upgrade, close the UI with a slight delay.
    OnUpgradeButton(Message:widget_message):void=
        if:
            FactoryMould := FactoryMouldMap[Message.Player]
            SelectedPlayer := FactoryMould.GetSelectedPlayer[]
        then:
            FactoryMould.Upgrade()

            if:
                FactoryMouldLevel := GetFactoryMouldLevel[SelectedPlayer, FactoryMould.Number]
                FactoryMouldLevel = 1
            then:
                OnCloseButton(Message, 1.0)
            else:
                UpdateUI(Message.Player)
        
    # When the player clicks the close button, close the UI.
    # Pass on a potential delay for factory mould notification of the closure.
    OnCloseButton(Message:widget_message, CloseDelay:float):void=
        CloseUI(Message.Player, CloseDelay)

    # Close the UI for the player.
    # Potentially, delay the notification to the factory mould.
    CloseUI<public>(Player:player, CloseDelay:float):void=
        if:
            UI := UIMap[Player]
            PlayerUI := GetPlayerUI[Player]
        then:
            spawn:
                DelayedUIClosed(Player, CloseDelay)
            PlayerUI.RemoveWidget(UI.Canvas)

    # Delay the notification to the factory mould.
    DelayedUIClosed<private>(Player:player, Delay:float)<suspends>:void=
        if:
            FactoryMould := FactoryMouldMap[Player]
        then:
            Sleep(Delay)
            # Notify the factory mould that the UI was closed.
            FactoryMould.UIClosed(Player)
