# Knowledge-neutral Verse snippet; project names and asset IDs generalized.
# id: snippet-105

# This file defines factory_product that represents a product moving down a factory belt from the
# factory mould to the factory storage.
# It also contains a function to compare two instances of factory_product.

using { /Fortnite.com/Devices }

factory_product := class:
    
    # The product number.
    Number:int

    # The distance left on the factory belt before reaching the factory storage in BRICK units (knobs).
    var DistanceLeft:float

# Compare two factory_product instances. They are considered the same if the product number
# and distance left are the same.
CompareFactoryProducts(A:factory_product, B:factory_product)<decides><transacts>:void=
    A.Number = B.Number
    A.DistanceLeft = B.DistanceLeft
