# Knowledge-neutral Verse snippet; project names and asset IDs generalized.
# id: snippet-106

# This file defines factory_progress_lights that handles the logic and visuals of a number of lights
# showing the progress of something.

using { /Fortnite.com/Devices }
using { /Verse.org/Simulation }

factory_progress_lights := class<concrete>:

    # The props making up the lights.
    @editable
    LightProps:[]creative_prop = array{}

    # Time in seconds determining how fast the lights should blink.
    @editable
    LightBlinkTime:float = 0.2

    # Event signalled to stop the current light animation.
    StopEvent:event() = event(){}

    # Stores if the lights are currently shown.
    var IsShowing:logic = false

    # The current progress as a value between 0.0 and 1.0.
    var CurrentProgress:float = 0.0

    # Stores if the last light in the progress is on or off.
    var LastLightOn:logic = true

   # Shows the lights.
    Show():void=
        if (not IsShowing?):
            set IsShowing = true
            # Start animating the lights.
            spawn:
                AnimateLights()

    # Hides the lights.
    Hide():void=
        if (IsShowing?):
            set IsShowing = false
            # Stop any running animation.
            StopEvent.Signal()
            for(LightProp : LightProps):
                LightProp.Hide()

    # Animates the lights by continuously showing the current progress until interrupted by the stop event.
    AnimateLights()<suspends>:void=
        race:
            StopEvent.Await()
            loop:
                # If progress is at 1.0 or more, flash all lights.
                if:
                    CurrentProgress >= 1.0
                then:
                    # Blink all lights.
                    if (LastLightOn?):
                        for(LightProp : LightProps):
                            LightProp.Show()
                        set LastLightOn = false
                    else:
                        for(LightProp : LightProps):
                            LightProp.Hide()
                        set LastLightOn = true
                # Otherwise, show progress.
                else:
                    # Determine the last light index.
                    if:
                        LastLightIndex := Floor[LightProps.Length * CurrentProgress]
                    then:
                        # Show all lights until just before the last light index. 
                        for (Index := 0..LastLightIndex-1, LightProp := LightProps[Index]):
                            LightProp.Show()
                        # Blink the last light. 
                        if:
                            LastLightProp := LightProps[LastLightIndex]
                        then:
                            if (LastLightOn?):
                                LastLightProp.Show()
                                set LastLightOn = false
                            else:
                                LastLightProp.Hide()
                                set LastLightOn = true
                        # Hide the rest of the lights. 
                        for (Index := LastLightIndex+1..LightProps.Length-1, LightProp := LightProps[Index]):
                            LightProp.Hide()

                Sleep(LightBlinkTime)

    # Sets the current progress.
    SetProgress(Progress:float):void=
        set CurrentProgress = Progress
