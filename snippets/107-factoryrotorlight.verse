# Knowledge-neutral Verse snippet; project names and asset IDs generalized.
# id: snippet-107

# This file defines factory_rotor_light that handles the logic and visuals of a rotor light.

using { /Fortnite.com/Devices }
using { /Verse.org/Simulation }

factory_rotor_light := class<concrete>:

    # Prop making up the light itself.
    @editable
    LightProp:creative_prop = creative_prop{}

    # Props making up rays of light shooting out from the light.
    @editable
    RayProps:[]creative_prop = array{}

    # Sequence animating the rays of light.
    @editable
    CinematicSequence:cinematic_sequence_device = cinematic_sequence_device{}

    # Stores if the light is currently shown.
    var IsShowing:logic = false

    # Shows the light.
    Show():void=
        if (not IsShowing?):
            set IsShowing = true
            # Show the light and rays.
            LightProp.Show()
            for(RayProp : RayProps):
                RayProp.Show()
            # Animate the rays.
            CinematicSequence.Play()

    # Hides the light.
    Hide():void=
        if (IsShowing?):
            set IsShowing = false
            # Hide the light and rays.
            LightProp.Hide()
            for(RayProp : RayProps):
                RayProp.Hide()
            # Stop ray animation.
            CinematicSequence.Stop()
