# Knowledge-neutral Verse snippet; project names and asset IDs generalized.
# id: snippet-108

# This file defines factory_storage that handles the logic and visuals of a factory storage.

using { /Fortnite.com/Devices }
using { /UnrealEngine.com/Temporary/SpatialMath }
using { /Verse.org/Random }
using { /Verse.org/Simulation }
using { Config }
using { Persistence }
using { Utility }

factory_storage<public> := class<concrete>(factory_element):

    # Prop that makes up the factory storage.
    @editable
    Props:[]creative_prop = array{}

    # Props that are the suction hoods at the end of each factory belt.
    @editable
    SuctionHoodProps:[]creative_prop = array{}

    # Prop that defines the position where build and upgrade effects should show.
    @editable
    EffectPositionProp:creative_prop = creative_prop{}

    # Audio that plays when product is stored.
    @editable
    SuctionHoodAudioPlayers:[]audio_player_device = array{}

    # Sequence that animates the storage compressor when it is running (the storage is not full).
    @editable
    CompressorSequence:cinematic_sequence_device = cinematic_sequence_device{}
    
    # Audio that plays when storage compressor is running (the storage is not full).
    @editable
    CompressorAudioPlayer:audio_player_device = audio_player_device{}
    
    # Lights that show off how much of the storage capacity is used. 
    @editable
    ProgressLights:factory_progress_lights = factory_progress_lights{}

    # Light that is on when compressor is running (the storage is not full).
    @editable
    CompressorRunningLight:creative_prop = creative_prop{}

    # Light that is on when compressor is stopped (the storage is full).
    @editable
    CompressorStoppedLight:creative_prop = creative_prop{}

    # Event signalled when a product is stored in the factory storage.
    # The product number stored is sent.
    StoreEvent<public>:event(int) = event(int){}

    # The vertical offset from the suction hood's position to where the factory product should
    # be pulled up to when stored.
    SuctionHoodOffset:vector3 = vector3{X := 0.0, Y := 0.0, Z := -89.6}

    # The initial transforms of suction hoods.
    var SuctionHoodInitialTransforms:[]transform = array{}

    # Runs when the device is started in a running game.
    OnBegin<override>()<suspends>:void=
        # Handle things that are in common for all factory elements when beginning.
        (super:)OnBegin()
        # Store initial transforms of suction hoods.
        for(SuctionHoodProp : SuctionHoodProps):
            set SuctionHoodInitialTransforms += array{SuctionHoodProp.GetTransform()}

    # Shows the factory storage UI and build preview.
    ShowUI<override>(Player:player):void=
        # Handle things that are in common for all factory elements when showing the UI.
        (super:)ShowUI(Player)
        # Show the UI and build preview.
        UIManager.ShowFactoryStorageUI(Player, Self)
        if:
            SelectedPlayer := MaybeSelectedPlayer?
            Level := GetFactoryStorageLevel[SelectedPlayer]
            Level = 0
        then:
            ShowBuildPreview(EffectPositionProp)

    # Succeeds if factory storage is built.
    IsBuilt<override>()<decides><transacts>:void=
        IsFactoryStorageBuilt[MaybeSelectedPlayer?]

    # Succeeds if factory storage can be upgraded.
    CanUpgrade<override>()<decides><transacts>:void=
        GetFactoryStorageLevel[MaybeSelectedPlayer?] < GetFactoryStorageMaxLevel[]

    # Gets the price for upgrading factory storage.
    GetUpgradePrice<override>()<decides><transacts>:int=
        GetFactoryStorageUpgradePrice[MaybeSelectedPlayer?]

    # Upgrades factory storage.
    Upgrade<override>():void=
        # Perform the actual upgrade if the selected player can afford it.
        if:
            SelectedPlayer := MaybeSelectedPlayer?
            Gold := GetGold[SelectedPlayer]
            Price := GetFactoryStorageUpgradePrice[SelectedPlayer]
            Gold >= Price
            CanUpgrade[]
            SpendGold[SelectedPlayer, Price]
            if (not IsFactoryStorageBuilt[SelectedPlayer]):
                BuildFactoryStorage[SelectedPlayer]
            UpgradeFactoryStorageLevel[SelectedPlayer]
        then:
            # Handle things that are in common for all factory elements when upgrading.
            (super:)Upgrade()
            # Show build or upgrade effect and play build or upgrade audio.
            if (Level := GetFactoryStorageLevel[SelectedPlayer]):
                ShowBuildOrUpgradeEffect(Level, EffectPositionProp)
                PlayBuildOrUpgradeAudio(Level)
            # Update all players' UI.
            UIManager.UpdateAllPriceUIs()
            UIManager.UpdateAllResourceUIs()
            UIManager.UpdateAllFactoryUIs()

    # Updates the visuals of factory storage to reflect the current state.
    Update<override>():void=
        # Handle things that are in common for all factory elements when updating.
        (super:)Update()
        if:
            SelectedPlayer := MaybeSelectedPlayer?
            IsFactoryStorageBuilt[SelectedPlayer]
        then:
            # Show the storage if it is built.
            for(Prop : Props):
                Prop.Show()
            # Update storage compressor.
            UpdateStorageCompressor()
            # Update and show the progress lights.
            UpdateProgressLights()
            ProgressLights.Show()
        else:
            # Hide the storage if it is unbuilt.
            for(Prop : Props):
                Prop.Hide()
            CompressorAudioPlayer.Stop()
            ProgressLights.Hide()
            CompressorRunningLight.Hide()
            CompressorStoppedLight.Hide()
                
    # Shows factory storage storing the prop.
    Store(ProductNumber:int, Prop:creative_prop):void=
        # Signal that factory product was stored.
        StoreEvent.Signal(ProductNumber)

        # Update storage compressor.
        UpdateStorageCompressor()

        # Update the progress lights.
        UpdateProgressLights()

        # Animate prop into suction hood and play audio. 
        if:
            SuctionHoodProp := SuctionHoodProps[ProductNumber]
            SuctionHoodInitialTransform := SuctionHoodInitialTransforms[ProductNumber]
            SuctionHoodAudioPlayer := SuctionHoodAudioPlayers[ProductNumber]
        then:
            spawn:
                AnimateStorage(Prop, SuctionHoodProp, SuctionHoodInitialTransform)
            SuctionHoodAudioPlayer.Play()

        # Update all players' UI.
        UIManager.UpdateAllFactoryStorageUIs()
        UIManager.UpdateAllOrderUIs()

    # Shows that products were removed from the factory storage. 
    Remove<public>(ProductNumber:int, Amount:int):void=
        UpdateStorageCompressor()
        UpdateProgressLights()

    # Update the storage compressor animations, audio and lights to reflect if there is room in storage.
    UpdateStorageCompressor():void=
        if:
            SelectedPlayer := MaybeSelectedPlayer?
            IsRoomInFactoryStorage[SelectedPlayer]
        then:
            CompressorSequence.Play()
            CompressorAudioPlayer.Play()
            CompressorRunningLight.Show()
            CompressorStoppedLight.Hide()
        else:
            CompressorSequence.Pause()
            CompressorAudioPlayer.Stop()
            CompressorRunningLight.Hide()
            CompressorStoppedLight.Show()


    # Updates the progress lights to reflect the current available storage capacity.
    UpdateProgressLights():void=
        if:
            SelectedPlayer := MaybeSelectedPlayer?
            Capacity := GetFactoryStorageCapacity[SelectedPlayer]
            AllProducts := GetAllFactoryProducts[SelectedPlayer]
        then:
            ProgressLights.SetProgress((1.0 * AllProducts) / (1.0 * Capacity))

    # Animates the factory product prop from the factory belt to the suction hood before disposing of it.
    # Also, shakes the suction hood.
    AnimateStorage(Prop:creative_prop, SuctionHoodProp:creative_prop, SuctionHoodInitialTransform:transform)<suspends>:void=
        PropTransform := Prop.GetTransform()
        StartRotation := PropTransform.Rotation
        TargetPosition := SuctionHoodInitialTransform.Translation + SuctionHoodOffset
        TargetRotation := StartRotation.RotateBy(MakeRotationFromYawPitchRollDegrees(GetRandomFloat(-45.0, 45.0), GetRandomFloat(-10.0, 10.0), GetRandomFloat(-10.0, 10.0)))

        sync:
            block:
                Prop.MoveTo(TargetPosition, TargetRotation, 0.2)
                # Hide the prop before disposing it to avoid showing BRICK explosion effect.
                Prop.Hide()
                Prop.Dispose()
            block:
                if:
                    SuctionHoodProp.TeleportTo[SuctionHoodInitialTransform]
                then:
                    SuctionHoodProp.Shake(4.0, 10.0, 5, 0.3)
