# Knowledge-neutral Verse snippet; project names and asset IDs generalized.
# id: snippet-109

# This file defines factory_storage_info, the collection of persistable information stored for the factory storage.

factory_storage_info := class<final><persistable>:

    # Is the factory storage built?
    Built:logic = false

    # The factory storage level.
    Level:int = 0

    # The amounts of factory products in storage.
    ProductsInStorage:[]int = array{0, 0, 0, 0}

MakeFactoryStorageInfo<constructor>(Src:factory_storage_info)<transacts> := factory_storage_info:
    Built := Src.Built
    Level := Src.Level
    ProductsInStorage := Src.ProductsInStorage
