# Knowledge-neutral Verse snippet; project names and asset IDs generalized.
# id: snippet-110

# This file defines the factory storage UI, which shows information for the factory storage
# in the factory of the selected player.

using { /Fortnite.com/Devices }
using { /Fortnite.com/UI }
using { /UnrealEngine.com/Temporary/SpatialMath }
using { /UnrealEngine.com/Temporary/UI }
using { /Verse.org/Colors }
using { /Verse.org/Simulation }
using { Config }
using { Factory }
using { Persistence }
using { Utility }

# Contains all widgets relevant for showing, updating and closing the UI.
factory_storage_ui_widgets := struct:
    Canvas<public>:canvas 
    LevelTextWidget<public>:text_block
    UnbuiltTextureWidget<public>:texture_block
    CapacityBackgroundColorWidget<public>:color_block
    CapacityFillColorWidget<public>:color_block
    CapacityTextWidget<public>:text_block
    ProductTextWidgets<public>:[]text_block
    ProductTextureWidgets<public>:[]texture_block
    PriceTextWidget<public>:text_block
    PriceTextureWidget<public>:texture_block
    UpgradeButtonWidget<public>:text_button_base

factory_storage_ui := class(creative_device):

    # Stores the created factory storage UI for each player.
    var UIMap:[player]factory_storage_ui_widgets = map{}
    # Stores the factory storage whose information will be shown on the UI for each player. 
    var FactoryStorageMap:[player]factory_storage = map{}

    # When a player is added, create the factory storage UI.
    PlayerAdded(Player:player):void=
        UIResult := CreateUI()
        if:
            set UIMap[Player] = UIResult

    # When a player is removed, remove the stored data.
    PlayerRemoved(Player:player):void=
        if:
            NewUIMap := UIMap.RemoveKey[Player]
            NewFactoryStorageMap := FactoryStorageMap.RemoveKey[Player]
        then:
            set UIMap = NewUIMap
            set FactoryStorageMap = NewFactoryStorageMap

    # Create the UI.
    CreateUI():factory_storage_ui_widgets=
        BackgroundColorWidget := color_block:
            DefaultColor := NamedColors.DarkGreen
            DefaultOpacity := 0.7
        LevelTextWidget := text_block:
            DefaultTextColor:= NamedColors.White
            DefaultShadowOffset:= option{vector2{X := 2.0, Y := 3.0}}
            DefaultShadowColor:= NamedColors.Black
            DefaultShadowOpacity := 1.0
        UnbuiltTextureWidget := texture_block:
            DefaultImage := Textures.UI.T_Storage
            DefaultDesiredSize:=vector2{X := 400.0, Y := 400.0}
        CapacityBackgroundColorWidget := color_block:
            DefaultColor := MakeColorFromSRGB(0.078, 0.149, 0.102)
            DefaultOpacity := 1.0
            DefaultDesiredSize:=vector2{X := 500.0, Y := 60.0}
        CapacityFillColorWidget := color_block:
            DefaultColor := MakeColorFromSRGB(0.212, 0.576, 0.325)
            DefaultOpacity := 1.0
        CapacityTextWidget := text_block:
            DefaultTextColor:= NamedColors.White
            DefaultShadowOffset:= option{vector2{X := 2.0, Y := 3.0}}
            DefaultShadowColor:= NamedColors.Black    
            DefaultShadowOpacity := 1.0
        Product0TextWidget := text_block:
            DefaultTextColor:= NamedColors.White
            DefaultJustification:=text_justification.InvariantRight
            DefaultShadowOffset:= option{vector2{X := 2.0, Y := 3.0}}
            DefaultShadowColor:= NamedColors.Black
            DefaultShadowOpacity := 1.0
        Product1TextWidget := text_block:
            DefaultTextColor:= NamedColors.White
            DefaultJustification:=text_justification.InvariantRight
            DefaultShadowOffset:= option{vector2{X := 2.0, Y := 3.0}}
            DefaultShadowColor:= NamedColors.Black
            DefaultShadowOpacity := 1.0
        Product2TextWidget := text_block:
            DefaultTextColor:= NamedColors.White
            DefaultJustification:=text_justification.InvariantRight
            DefaultShadowOffset:= option{vector2{X := 2.0, Y := 3.0}}
            DefaultShadowColor:= NamedColors.Black
            DefaultShadowOpacity := 1.0
        Product3TextWidget := text_block:
            DefaultTextColor:= NamedColors.White
            DefaultJustification:=text_justification.InvariantRight
            DefaultShadowOffset:= option{vector2{X := 2.0, Y := 3.0}}
            DefaultShadowColor:= NamedColors.Black
            DefaultShadowOpacity := 1.0
        Product0TextureWidget := texture_block:
            DefaultImage := Textures.UI.T_Plane
            DefaultDesiredSize:=vector2{X := 120.0, Y := 120.0}
        Product1TextureWidget := texture_block:
            DefaultImage := Textures.UI.T_TRex
            DefaultDesiredSize:=vector2{X := 120.0, Y := 120.0}
        Product2TextureWidget := texture_block:
            DefaultImage := Textures.UI.T_Castle
            DefaultDesiredSize:=vector2{X := 120.0, Y := 120.0}
        Product3TextureWidget := texture_block:
            DefaultImage := Textures.UI.T_Spaceship
            DefaultDesiredSize:=vector2{X := 120.0, Y := 120.0}
        PriceTextWidget := text_block:
            DefaultTextColor:= NamedColors.White
            DefaultShadowOffset:= option{vector2{X := 2.0, Y := 3.0}}
            DefaultShadowColor:= NamedColors.Black
            DefaultShadowOpacity := 1.0
            DefaultJustification:=text_justification.InvariantRight
        PriceTextureWidget := texture_block:
            DefaultImage := Textures.UI.T_Gold
            DefaultDesiredSize:=vector2{X := 40.0, Y := 40.0}
        UpgradeButtonWidget := button_loud{}
        CloseButtonWidget := button_regular:
            DefaultText:= PlainText("Close")

        CanvasWidget := canvas:
            Slots := array:
                canvas_slot:
                    Anchors := anchors{Minimum := vector2{X := 0.0, Y := 0.0}, Maximum := vector2{X := 1.0, Y := 1.0}}
                    Offsets := margin{Top := 0.0, Left := 0.0, Right := 0.0, Bottom := 0.0}
                    Alignment := vector2{X := 0.5, Y := 0.5}
                    Widget := BackgroundColorWidget
                canvas_slot:
                    Anchors := anchors{Minimum := vector2{X := 0.5, Y := 0.5}, Maximum := vector2{X := 0.5, Y := 0.5}}
                    Offsets := margin{Top := -390.0, Left := 0.0, Right := 0.0, Bottom := 0.0}
                    Alignment := vector2{X := 0.5, Y := 0.5}
                    SizeToContent := true
                    Widget := LevelTextWidget
                canvas_slot:
                    Anchors := anchors{Minimum := vector2{X := 0.5, Y := 0.5}, Maximum := vector2{X := 0.5, Y := 0.5}}
                    Offsets := margin{Top := -70.0, Left := 0.0, Right := 0.0, Bottom := 0.0}
                    Alignment := vector2{X := 0.5, Y := 0.5}
                    Widget := UnbuiltTextureWidget
                canvas_slot:
                    Anchors := anchors{Minimum := vector2{X := 0.5, Y := 0.5}, Maximum := vector2{X := 0.5, Y := 0.5}}
                    Offsets := margin{Top := -310.0, Left := 0.0, Right := 0.0, Bottom := 0.0}
                    Alignment := vector2{X := 0.5, Y := 0.5}
                    Widget := CapacityBackgroundColorWidget
                canvas_slot:
                    Anchors := anchors{Minimum := vector2{X := 0.5, Y := 0.5}, Maximum := vector2{X := 0.5, Y := 0.5}}
                    Offsets := margin{Top := -310.0, Left := -245.0, Right := 0.0, Bottom := 0.0}
                    Alignment := vector2{X := 0.0, Y := 0.5}
                    Widget := CapacityFillColorWidget
                canvas_slot:
                    Anchors := anchors{Minimum := vector2{X := 0.5, Y := 0.5}, Maximum := vector2{X := 0.5, Y := 0.5}}
                    Offsets := margin{Top := -310.0, Left := 0.0, Right := 0.0, Bottom := 0.0}
                    Alignment := vector2{X := 0.5, Y := 0.5}
                    SizeToContent := true
                    Widget := CapacityTextWidget
                canvas_slot:
                    Anchors := anchors{Minimum := vector2{X := 0.5, Y := 0.5}, Maximum := vector2{X := 0.5, Y := 0.5}}
                    Offsets := margin{Top := -190.0, Left := -30.0, Right := 0.0, Bottom := 0.0}
                    Alignment := vector2{X := 1.0, Y := 0.5}
                    SizeToContent := true
                    Widget := Product0TextWidget
                canvas_slot:
                    Anchors := anchors{Minimum := vector2{X := 0.5, Y := 0.5}, Maximum := vector2{X := 0.5, Y := 0.5}}
                    Offsets := margin{Top := -80.0, Left := -30.0, Right := 0.0, Bottom := 0.0}
                    Alignment := vector2{X := 1.0, Y := 0.5}
                    SizeToContent := true
                    Widget := Product1TextWidget
                canvas_slot:
                    Anchors := anchors{Minimum := vector2{X := 0.5, Y := 0.5}, Maximum := vector2{X := 0.5, Y := 0.5}}
                    Offsets := margin{Top := 30.0, Left := -30.0, Right := 0.0, Bottom := 0.0}
                    Alignment := vector2{X := 1.0, Y := 0.5}
                    SizeToContent := true
                    Widget := Product2TextWidget
                canvas_slot:
                    Anchors := anchors{Minimum := vector2{X := 0.5, Y := 0.5}, Maximum := vector2{X := 0.5, Y := 0.5}}
                    Offsets := margin{Top := 140.0, Left := -30.0, Right := 0.0, Bottom := 0.0}
                    Alignment := vector2{X := 1.0, Y := 0.5}
                    SizeToContent := true
                    Widget := Product3TextWidget
                canvas_slot:
                    Anchors := anchors{Minimum := vector2{X := 0.5, Y := 0.5}, Maximum := vector2{X := 0.5, Y := 0.5}}
                    Offsets := margin{Top := -190.0, Left := 50.0, Right := 0.0, Bottom := 0.0}
                    Alignment := vector2{X := 0.5, Y := 0.5}
                    Widget := Product0TextureWidget
                canvas_slot:
                    Anchors := anchors{Minimum := vector2{X := 0.5, Y := 0.5}, Maximum := vector2{X := 0.5, Y := 0.5}}
                    Offsets := margin{Top := -80.0, Left := 50.0, Right := 0.0, Bottom := 0.0}
                    Alignment := vector2{X := 0.5, Y := 0.5}
                    Widget := Product1TextureWidget
                canvas_slot:
                    Anchors := anchors{Minimum := vector2{X := 0.5, Y := 0.5}, Maximum := vector2{X := 0.5, Y := 0.5}}
                    Offsets := margin{Top := 30.0, Left := 50.0, Right := 0.0, Bottom := 0.0}
                    Alignment := vector2{X := 0.5, Y := 0.5}
                    Widget := Product2TextureWidget
                canvas_slot:
                    Anchors := anchors{Minimum := vector2{X := 0.5, Y := 0.5}, Maximum := vector2{X := 0.5, Y := 0.5}}
                    Offsets := margin{Top := 140.0, Left := 50.0, Right := 0.0, Bottom := 0.0}
                    Alignment := vector2{X := 0.5, Y := 0.5}
                    Widget := Product3TextureWidget
                canvas_slot:
                    Anchors := anchors{Minimum := vector2{X := 0.5, Y := 0.5}, Maximum := vector2{X := 0.5, Y := 0.5}}
                    Offsets := margin{Top := 250.0, Left := -180.0, Right := 0.0, Bottom := 0.0}
                    Alignment := vector2{X := 1.0, Y := 0.5}
                    SizeToContent := true
                    Widget := PriceTextWidget
                canvas_slot:
                    Anchors := anchors{Minimum := vector2{X := 0.5, Y := 0.5}, Maximum := vector2{X := 0.5, Y := 0.5}}
                    Offsets := margin{Top := 250.0, Left := -150.0, Right := 0.0, Bottom := 0.0}
                    Alignment := vector2{X := 0.5, Y := 0.5}
                    Widget := PriceTextureWidget
                canvas_slot:
                    Anchors := anchors{Minimum := vector2{X := 0.5, Y := 0.5}, Maximum := vector2{X := 0.5, Y := 0.5}}
                    Offsets := margin{Top := 250.0, Left := 0.0, Right := 0.0, Bottom := 0.0}
                    Alignment := vector2{X := 0.5, Y := 0.5}
                    SizeToContent := true
                    Widget := UpgradeButtonWidget
                canvas_slot:
                    Anchors := anchors{Minimum := vector2{X := 0.5, Y := 0.5}, Maximum := vector2{X := 0.5, Y := 0.5}}
                    Offsets := margin{Top := 350.0, Left := 0.0, Right := 0.0, Bottom := 0.0}
                    Alignment := vector2{X := 0.5, Y := 0.5}
                    SizeToContent := true
                    Widget := CloseButtonWidget

        Canvas := canvas:
            Slots := array:
                canvas_slot:
                    Anchors := anchors{Minimum := vector2{X := 0.0, Y := 0.0}, Maximum := vector2{X := 0.42, Y := 1.0}}
                    Offsets := margin{Top := 50.0, Left := 50.0, Right := 50.0, Bottom := 120.0}
                    Alignment := vector2{X := 0.5, Y := 0.5}
                    Widget := CanvasWidget

        # Subscribe callbacks to get notified when the player clicks the buttons.
        UpgradeButtonWidget.OnClick().Subscribe(OnUpgradeButton)
        CloseButtonWidget.OnClick().SubscribeWidgetMessage(OnCloseButton, 0.0)

        return factory_storage_ui_widgets:
            Canvas := Canvas
            LevelTextWidget := LevelTextWidget
            UnbuiltTextureWidget := UnbuiltTextureWidget
            CapacityBackgroundColorWidget := CapacityBackgroundColorWidget
            CapacityFillColorWidget := CapacityFillColorWidget
            CapacityTextWidget := CapacityTextWidget
            ProductTextWidgets := array{Product0TextWidget, Product1TextWidget, Product2TextWidget, Product3TextWidget}
            ProductTextureWidgets := array{Product0TextureWidget, Product1TextureWidget, Product2TextureWidget, Product3TextureWidget}
            PriceTextWidget := PriceTextWidget
            PriceTextureWidget := PriceTextureWidget
            UpgradeButtonWidget := UpgradeButtonWidget

    # Show the UI to the player.
    ShowUI(Player:player, FactoryStorage:factory_storage, AudioPlayer:audio_player_device):void=
        if:
            set FactoryStorageMap[Player] = FactoryStorage
            UI := UIMap[Player]
            PlayerUI := GetPlayerUI[Player]
        then:
            # Play UI audio.
            AudioPlayer.Play(Player)

            # Update and show UI.
            UpdateUI(Player)
            PlayerUI.AddWidget(UI.Canvas, player_ui_slot{ZOrder := 0, InputMode := ui_input_mode.All})

    # Update the UI for the player.
    UpdateUI(Player:player):void=
        if:
            UI := UIMap[Player]
            FactoryStorage := FactoryStorageMap[Player]
            SelectedPlayer := FactoryStorage.GetSelectedPlayer[]
            UpgradePrice := GetFactoryStorageUpgradePrice[SelectedPlayer]
            Built := logic{IsFactoryStorageBuilt[SelectedPlayer]}
            Level := GetFactoryStorageLevel[SelectedPlayer]
            MaxLevel := GetFactoryStorageMaxLevel[]
            CanUpgrade := logic{FactoryStorage.CanUpgrade[]}
            UpgradeEnabled := logic{CanUpgrade?, GetGold[SelectedPlayer] >= UpgradePrice}
            Capacity := GetFactoryStorageCapacity[SelectedPlayer]
            AllFactoryProducts := GetAllFactoryProducts[SelectedPlayer]
            Product0 := GetFactoryProduct[SelectedPlayer, 0]
            Product1 := GetFactoryProduct[SelectedPlayer, 1]
            Product2 := GetFactoryProduct[SelectedPlayer, 2]
            Product3 := GetFactoryProduct[SelectedPlayer, 3]
            Product0Built := logic{IsFactoryMouldBuilt[SelectedPlayer, 0]}
            Product1Built := logic{IsFactoryMouldBuilt[SelectedPlayer, 1]}
            Product2Built := logic{IsFactoryMouldBuilt[SelectedPlayer, 2]}
            Product3Built := logic{IsFactoryMouldBuilt[SelectedPlayer, 3]}
            Product0TextWidget := UI.ProductTextWidgets[0]
            Product1TextWidget := UI.ProductTextWidgets[1]
            Product2TextWidget := UI.ProductTextWidgets[2]
            Product3TextWidget := UI.ProductTextWidgets[3]
            Product0TextureWidget := UI.ProductTextureWidgets[0]
            Product1TextureWidget := UI.ProductTextureWidgets[1]
            Product2TextureWidget := UI.ProductTextureWidgets[2]
            Product3TextureWidget := UI.ProductTextureWidgets[3]
        then:
            # Adapt the UI to whether or not the factory storage has been built yet.
            if (Built?):
                UI.LevelTextWidget.SetText(LevelText(Level, MaxLevel))
                UI.UnbuiltTextureWidget.SetVisibility(widget_visibility.Hidden)
                UI.CapacityBackgroundColorWidget.SetVisibility(widget_visibility.Visible)
                UI.CapacityFillColorWidget.SetVisibility(widget_visibility.Visible)
                UI.CapacityFillColorWidget.SetDesiredSize(vector2{X := 490.0 * (AllFactoryProducts * 1.0) / (Capacity * 1.0), Y := 50.0})
                UI.CapacityFillColorWidget.SetColor(if (AllFactoryProducts = Capacity) then MakeColorFromSRGB(0.784, 0.0, 0.0) else MakeColorFromSRGB(0.212, 0.576, 0.325))

                UI.CapacityTextWidget.SetText(PlainText("{AllFactoryProducts} / {Capacity}"))

                Product0TextWidget.SetText(CountText(Product0))
                Product1TextWidget.SetText(CountText(Product1))
                Product2TextWidget.SetText(CountText(Product2))
                Product3TextWidget.SetText(CountText(Product3))
    
                # Only show the product texture if the player has the ability to produce that product.
                Product0TextureWidget.SetVisibility(widget_visibility.Visible)
                if (Product0Built?):
                    Product0TextureWidget.SetImage(Textures.UI.T_Plane)
                else:
                    Product0TextureWidget.SetImage(Textures.UI.T_Hidden)
    
                Product1TextureWidget.SetVisibility(widget_visibility.Visible)
                if (Product1Built?):
                    Product1TextureWidget.SetImage(Textures.UI.T_TRex)
                else:
                    Product1TextureWidget.SetImage(Textures.UI.T_Hidden)
    
                Product2TextureWidget.SetVisibility(widget_visibility.Visible)
                if (Product2Built?):
                    Product2TextureWidget.SetImage(Textures.UI.T_Castle)
                else:
                    Product2TextureWidget.SetImage(Textures.UI.T_Hidden)
    
                Product3TextureWidget.SetVisibility(widget_visibility.Visible)
                if (Product3Built?):
                    Product3TextureWidget.SetImage(Textures.UI.T_Spaceship)
                else:
                    Product3TextureWidget.SetImage(Textures.UI.T_Hidden)

                UI.UpgradeButtonWidget.SetText(PlainText("Upgrade"))

            else:
                UI.LevelTextWidget.SetText(PlainText("Build to start storing toys"))
                UI.UnbuiltTextureWidget.SetVisibility(widget_visibility.Visible)
                UI.CapacityBackgroundColorWidget.SetVisibility(widget_visibility.Hidden)
                UI.CapacityFillColorWidget.SetVisibility(widget_visibility.Hidden)
                UI.CapacityTextWidget.SetText(PlainText(""))

                Product0TextWidget.SetText(PlainText(""))
                Product1TextWidget.SetText(PlainText(""))
                Product2TextWidget.SetText(PlainText(""))
                Product3TextWidget.SetText(PlainText(""))

                Product0TextureWidget.SetVisibility(widget_visibility.Hidden)
                Product1TextureWidget.SetVisibility(widget_visibility.Hidden)
                Product2TextureWidget.SetVisibility(widget_visibility.Hidden)
                Product3TextureWidget.SetVisibility(widget_visibility.Hidden)

                UI.UpgradeButtonWidget.SetText(PlainText("Build"))

            # Only show the price if the player can upgrade.
            if (CanUpgrade?):
                UI.PriceTextWidget.SetText(CountText(UpgradePrice))
                UI.PriceTextureWidget.SetVisibility(widget_visibility.Visible)
            else:
                UI.PriceTextWidget.SetText(PlainText(""))
                UI.PriceTextureWidget.SetVisibility(widget_visibility.Hidden)                
            UI.UpgradeButtonWidget.SetEnabled(UpgradeEnabled)

    # When the player clicks the upgrade button, upgrade the factory storage.
    # If it was the first upgrade, close the UI with a slight delay.
    OnUpgradeButton(Message:widget_message):void=
        if:
            FactoryStorage := FactoryStorageMap[Message.Player]
            SelectedPlayer := FactoryStorage.GetSelectedPlayer[]
        then:
            FactoryStorage.Upgrade()

            if:
                FactoryStorageLevel := GetFactoryStorageLevel[SelectedPlayer]
                FactoryStorageLevel = 1
            then:
                OnCloseButton(Message, 1.0)
            else:
                UpdateUI(Message.Player)

    # When the player clicks the close button, close the UI.
    # Pass on a potential delay for factory storage notification of the closure.
    OnCloseButton(Message:widget_message, CloseDelay:float):void=
        CloseUI(Message.Player, CloseDelay)

    # Close the UI for the player.
    # Potentially, delay the notification to the factory storage.
    CloseUI<public>(Player:player, CloseDelay:float):void=
        if:
            UI := UIMap[Player]
            PlayerUI := GetPlayerUI[Player]
        then:
            spawn:
                DelayedUIClosed(Player, CloseDelay)
            PlayerUI.RemoveWidget(UI.Canvas)

    # Delay the notification to the factory storage.
    DelayedUIClosed<private>(Player:player, Delay:float)<suspends>:void=
        if:
            FactoryStorage := FactoryStorageMap[Player]
        then:
            Sleep(Delay)
            # Notify the factory storage that the UI was closed.
            FactoryStorage.UIClosed(Player)
