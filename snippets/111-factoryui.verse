# Knowledge-neutral Verse snippet; project names and asset IDs generalized.
# id: snippet-111

# This file defines the factory UI, which shows information for the factory of the selected player.
# Whenever the UI is updated, it handles the countdown to when the factory's storage will be full.

using { /Fortnite.com/Devices }
using { /Fortnite.com/UI }
using { /UnrealEngine.com/Temporary/SpatialMath }
using { /UnrealEngine.com/Temporary/UI }
using { /Verse.org/Colors }
using { /Verse.org/Simulation }
using { Factory }
using { Persistence }
using { Time }
using { Utility }

# Contains all widgets relevant for showing, closing and updating the UI.
factory_ui_widgets := struct:
    Canvas<public>:canvas
    StorageTextWidget<public>:text_block
    StorageTextureWidget<public>:texture_block
    StopEvent<public>:event()

factory_ui := class(creative_device):
    # The real time device used to animate the countdown.
    @editable
    RealTime:real_time = real_time{}

    # Stores the created factory UI for each player.
    var UIMap:[player]factory_ui_widgets = map{}
    # Stores the factory manager whose information will be shown on the UI for each player. 
    var FactoryManagerMap:[player]factory_manager = map{}

    # When a player is added, create the factory UI.
    PlayerAdded(Player:player):void=
        UIResult := CreateUI()
        if:
            PlayerUI := GetPlayerUI[Player]
            set UIMap[Player] = UIResult

    # When a player is removed, remove the stored data.
    PlayerRemoved(Player:player):void=
        if:
            NewUIMap := UIMap.RemoveKey[Player]
            NewFactoryManagerMap := FactoryManagerMap.RemoveKey[Player]
        then:
            set UIMap = NewUIMap
            set FactoryManagerMap = NewFactoryManagerMap

    # Create the UI.
    CreateUI():factory_ui_widgets=
        StorageTextWidget := text_block:
            DefaultTextColor:= NamedColors.White
            DefaultJustification:=text_justification.InvariantRight
            DefaultShadowOffset:= option{vector2{X := 2.0, Y := 3.0}}
            DefaultShadowColor:= NamedColors.Black
            DefaultShadowOpacity := 1.0
        StorageTextureWidget := texture_block:
            DefaultImage := Textures.UI.T_StorageContainer
            DefaultDesiredSize:=vector2{X := 120.0, Y := 120.0}
        Canvas := canvas:
            Slots := array:
                canvas_slot:
                    Anchors := anchors{Minimum := vector2{X := 1.0, Y := 0.0}, Maximum := vector2{X := 1.0, Y := 0.0}}
                    Offsets := margin{Top := 50.0, Left := -175.0, Right := 0.0, Bottom := 0.0}
                    Alignment := vector2{X := 1.0, Y := 0.0}
                    SizeToContent := true
                    Widget := StorageTextWidget
                canvas_slot:
                    Anchors := anchors{Minimum := vector2{X := 1.0, Y := 0.0}, Maximum := vector2{X := 1.0, Y := 0.0}}
                    Offsets := margin{Top := 90.0, Left := -95.0, Right := 0.0, Bottom := 0.0}
                    Alignment := vector2{X := 0.5, Y := 0.5}
                    Widget := StorageTextureWidget

        return factory_ui_widgets:
            Canvas := Canvas
            StorageTextWidget := StorageTextWidget
            StorageTextureWidget := StorageTextureWidget
            StopEvent := event(){}

    # Show the UI to the player. The UI will show the information for the given factory manager.
    ShowUI(Player:player, FactoryManager:factory_manager):void=
        if:
            set FactoryManagerMap[Player] = FactoryManager
            UI := UIMap[Player]
            PlayerUI := GetPlayerUI[Player]
        then:
            UpdateUI(Player)
            PlayerUI.AddWidget(UI.Canvas)

    # Update the UI for the player.
    UpdateUI(Player:player):void=
        if:
            UI := UIMap[Player]
            FactoryManager := FactoryManagerMap[Player]
            SelectedPlayer := FactoryManager.GetSelectedPlayer[]
            StorageBuilt := logic{IsFactoryStorageBuilt[SelectedPlayer]}
        then:
            # Stop any previous time animations.
            UI.StopEvent.Signal()

            # Only show the countdown to when the factory's storage is full if the storage is built.
            if (StorageBuilt?):
                FullStorageTime := FactoryManager.CalculateDateAndTimeForFullStorage()
                # If nothing is producing, the storage would never become full and therefore
                # the time would be invalid. Only start the countdown animation if the time is valid. 
                if (FullStorageTime.IsValid[]):
                    spawn:
                        AnimateStorageFullCountdown(UI.StorageTextWidget, UI.StopEvent, FullStorageTime, SelectedPlayer)
                else:
                    UI.StorageTextWidget.SetTextColor(NamedColors.Red)
                    UI.StorageTextWidget.SetText(PlainText("Production deactivated"))
                UI.StorageTextureWidget.SetVisibility(widget_visibility.Visible)
            else:
                UI.StorageTextWidget.SetText(PlainText(""))
                UI.StorageTextureWidget.SetVisibility(widget_visibility.Hidden)
    
    # Close the UI for the player.
    CloseUI<public>(Player:player):void=
        if:
            UI := UIMap[Player]
            PlayerUI := GetPlayerUI[Player]
        then:
            # Stop any time animations.
            UI.StopEvent.Signal()
            PlayerUI.RemoveWidget(UI.Canvas)

    # Animate a countdown to the given time.
    # Since the time is an estimate, do not show the last two minutes.
    AnimateStorageFullCountdown(Text:text_block, StopEvent:event(), StorageFullTime:date_and_time, SelectedPlayer:player)<suspends>:void=
        # Keep animating the text value until interrupted by the stop event,
        # or until the storage is full.
        race:
            StopEvent.Await()
            block:
                loop:
                    Now := RealTime.GetDateAndTime()
                    Difference := StorageFullTime.Subtract(Now)
                    if (IsRoomInFactoryStorage[SelectedPlayer]):
                        # If there is more than two minutes left of the estimated time, show the time.
                        if (Difference.GetTotalSeconds() > 120.0):
                            Text.SetTextColor(NamedColors.White)
                            Text.SetText(PlainText("{Difference.ToTruncatedString()} to full"))
                        else:
                            Text.SetTextColor(NamedColors.Yellow)
                            Text.SetText(PlainText("Storage soon full"))
                    else:
                        Text.SetTextColor(NamedColors.Red)
                        Text.SetText(PlainText("Storage full"))
                        break
                    Sleep(0.1)
