# Knowledge-neutral Verse snippet; project names and asset IDs generalized.
# id: snippet-112

# This file defines factory_wrapper that handles the logic and visuals of a factory wrapper.

using { /Fortnite.com/Devices }
using { /Verse.org/Simulation }
using { Config }
using { Persistence }

factory_wrapper<public> := class<concrete>(factory_element):

    # Props that make up the factory wrapper.
    @editable
    Props:[]creative_prop = array{}

    # The prop that is the section of the factory belt that is covered by the factory wrapper.
    @editable
    BeltSectionProp:creative_prop = creative_prop{}

    # Prop that defines the position of the wrapper.
    # Used by factory manager to determine where the products should be wrapped.
    @editable
    PositionProp:creative_prop = creative_prop{}

    # Prop that defines the position where build and upgrade effects should show.
    @editable
    EffectPositionProp:creative_prop = creative_prop{}
    
    # Lights that show off the time since the last wrapping.
    @editable
    IntensityLights:factory_intensity_lights = factory_intensity_lights{}

    # Sequence that animates the factory wrapper when wrapping.
    @editable
    CinematicSequence:cinematic_sequence_device = cinematic_sequence_device{}

    # Prop asset containing an effect to show when wrapping.
    @editable
    Effect:creative_prop_asset = DefaultCreativePropAsset
    
    # Audio that plays from console.
    @editable
    ConsoleAudioPlayer:audio_player_device = audio_player_device{}
    
    # Audio that plays when wrapping.
    @editable
    RamAudioPlayer:audio_player_device = audio_player_device{}
    
    # Prop asset containing the product prop that is the outcome of the factory wrapper.
    @editable
    Product<public>:creative_prop_asset = DefaultCreativePropAsset

    # The factory wrapper number.
    @editable
    Number<public>:int = 0

    # Shows the factory wrapper UI and build preview.
    ShowUI<override>(Player:player):void=
        (super:)ShowUI(Player)
        UIManager.ShowFactoryWrapperUI(Player, Self)
        if:
            SelectedPlayer := MaybeSelectedPlayer?
            Level := GetFactoryWrapperLevel[SelectedPlayer, Number]
            Level = 0
        then:
            ShowBuildPreview(EffectPositionProp)

    # Succeeds if factory wrapper is built.
    IsBuilt<override>()<decides><transacts>:void=
        IsFactoryMouldBuilt[MaybeSelectedPlayer?, Number]

    # Succeeds if factory wrapper can be upgraded.
    CanUpgrade<override>()<decides><transacts>:void=
        GetFactoryWrapperLevel[MaybeSelectedPlayer?, Number] < GetFactoryWrapperMaxLevel[Number]

    # Gets the price for upgrading factory wrapper.
    GetUpgradePrice<override>()<decides><transacts>:int=
        GetFactoryWrapperUpgradePrice[Number, MaybeSelectedPlayer?]

    # Upgrades factory wrapper.
    Upgrade<override>():void=
        # Perform the actual upgrade if the selected player can afford it.
        if:
            SelectedPlayer := MaybeSelectedPlayer?
            Gold := GetGold[SelectedPlayer]
            Price := GetFactoryWrapperUpgradePrice[Number, SelectedPlayer]
            Gold >= Price
            CanUpgrade[]
            SpendGold[SelectedPlayer, Price]
            if (not IsFactoryWrapperBuilt[SelectedPlayer, Number]):
                BuildFactoryWrapper[SelectedPlayer, Number]
            UpgradeFactoryWrapperLevel[SelectedPlayer, Number]
        then:
            # Handle things that are in common for all factory elements when upgrading.
            (super:)Upgrade()
            # Show build or upgrade effect and play build or upgrade audio.
            if (Level := GetFactoryWrapperLevel[SelectedPlayer, Number]):
                ShowBuildOrUpgradeEffect(Level, EffectPositionProp)
                PlayBuildOrUpgradeAudio(Level)
            # Update all players' UI.
            UIManager.UpdateAllPriceUIs()
            UIManager.UpdateAllResourceUIs()

    # Updates the visuals of factory wrapper to reflect the current state.
    Update<override>():void=
        # Handle things that are in common for all factory elements when updating.
        (super:)Update()
        if:
            SelectedPlayer := MaybeSelectedPlayer?
            IsFactoryWrapperBuilt[SelectedPlayer, Number]
        then:
            # Show the wrapper if it is built.
            for(Prop : Props):
                Prop.Show()
            # Hide the section of the belt that is covered by the wrapper.
            BeltSectionProp.Hide()
            # Play the console audio.
            ConsoleAudioPlayer.Play()
            # Show the intensity lights.
            IntensityLights.Show()
        else:
            # Hide the wrapper if it is unbuilt.
            for(Prop : Props):
                Prop.Hide()
            CinematicSequence.Stop()
            ConsoleAudioPlayer.Stop()
            IntensityLights.Hide()

    # Shows factory wrapper wrapping.
    Process(Player:player):void=
        CinematicSequence.Play()
        RamAudioPlayer.Play()
        spawn:
            ShowRamEffect()
        spawn:
            UpdateIntensityLights()

    # Shows the ram effect with a slight delay timed to the ram animation.
    ShowRamEffect()<suspends>:void=
        Sleep(0.25)
        ShowEffect(Effect, EffectPositionProp)

    # Updates the intensity lights with a slight delay timed to the ram animation.
    UpdateIntensityLights()<suspends>:void=
        Sleep(0.25)
        IntensityLights.SetIntensity(1.0)
