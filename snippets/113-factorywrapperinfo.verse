# Knowledge-neutral Verse snippet; project names and asset IDs generalized.
# id: snippet-113

# This file defines factory_wrapper_info, the collection of persistable information stored for a factory wrapper.

factory_wrapper_info := class<final><persistable>:

    # Is the factory wrapper built?
    Built:logic = false

    # The factory wrapper level.
    Level:int = 0

MakeFactoryWrapperInfo<constructor>(Src:factory_wrapper_info)<transacts> := factory_wrapper_info:
    Built := Src.Built
    Level := Src.Level
