# Knowledge-neutral Verse snippet; project names and asset IDs generalized.
# id: snippet-115

# This file defines the game_manager that handles setting up the game, manages adding and removing players,
# and loads the factory owned by the selected player.

using { /Fortnite.com/Devices }
using { /Verse.org/Simulation }
using { Factory }
using { UserInterface }
using { Utility }

# Change the module accessibility of the UI texture assets.
Textures := module:
    UI<public> := module{}

game_manager := class(creative_device):

    # UI manager for showing game UI.
    @editable
    UIManager:ui_manager = ui_manager{}

    # Billboard showing the owner of the factory.
    @editable
    FactoryBillboard:billboard_device = billboard_device{}

    # Popup dialog telling players that the factory is reset when the factory owner leaves.
    @editable
    FactoryResetPopup:popup_dialog_device = popup_dialog_device{} 

    # The factory manager used to run the factory.
    @editable
    FactoryManager:factory_manager = factory_manager{}

    # Order that the factory can deliver to.
    @editable
    Order:order = order{}

    # Candy CandyCane fields that can be harvested.
    @editable
    CandyCaneFields:[]candy_cane_field = array{}

    # Tutorial guiding the player through the game.
    @editable
    Tutorial:tutorial = tutorial{}

    # Jukebox playing awesome music.
    @editable
    Jukebox:jukebox = jukebox{}

    # Test used when debugging and testing the game.
    @editable
    Test:test = test{}

    # Daily gift given to the player.
    @editable
    DailyGift:daily_gift = daily_gift{}
    
    # Prop indicating that the factory can be loaded.
    @editable
    LoadFactoryButtonProp:creative_prop = creative_prop{}

    # The trigger that initiates factory loading.
    @editable
    LoadFactoryTrigger:trigger_device = trigger_device{}

    # The selected player that owns the factory and all other game elements. Can be false if nobody owns them yet.
    var MaybeSelectedPlayer:?player = false

    # Runs when the device is started in a running game.
    OnBegin<override>()<suspends>:void=
        # Subscribe callback when factory loading trigger is triggered.
        LoadFactoryTrigger.TriggeredEvent.Subscribe(OnLoadFactoryTriggered)

        # Subscribe callback when players are added or removed.
        Playspace := GetPlayspace()
        Playspace.PlayerAddedEvent().Subscribe(OnPlayerAdded)
        Playspace.PlayerRemovedEvent().Subscribe(OnPlayerRemoved)

        # Make sure that initial players are registered.
        for (Player : Playspace.GetPlayers()):
            OnPlayerAdded(Player)

        # Load factory automatically if there is only one player.
        if:
            Players := Playspace.GetPlayers()
            Players.Length = 1
            Player := Players[0]
        then:
            LoadFactory(Player)

    # Sets the selected player. This will propagate the selected player to all game elements.
    SelectPlayer(MaybePlayer:?player):void=
        set MaybeSelectedPlayer = MaybePlayer

        # Propagate selected player to all game elements. 
        Order.SelectPlayer(MaybePlayer)
        for(CandyCaneField : CandyCaneFields):
            CandyCaneField.SelectPlayer(MaybePlayer)
        Jukebox.SelectPlayer(MaybePlayer)

        # It is important that tutorial sets the selected player before the factory manager
        # since the factory manager does offline production when player is selected.
        Tutorial.SelectPlayer(MaybePlayer)

        # Setting the selected player on the factory manager will initialise the factory for that player. 
        FactoryManager.SelectPlayer(MaybePlayer)

        Test.SelectPlayer(MaybePlayer)

        DailyGift.SelectPlayer(MaybePlayer)

    # When the factory loading trigger is triggered, show the UI for loading a factory.
    OnLoadFactoryTriggered(MaybeAgent:?agent):void=
        if:
            Agent := MaybeAgent?
            Player := player[Agent]
        then:
            LoadFactoryTrigger.Disable()
            UIManager.ShowLoadFactoryUI(Player, Self)

    # Load the factory owned by the player.
    LoadFactory<public>(Player:player):void=
        # Update the billboard with the factory owner.
        FactoryBillboard.SetText(PlayerText("", Player, "'s\ngift factory"))

        # Register the player as the selected player. This will trigger
        SelectPlayer(option{Player})

        # Show factory and resource UI to all players.
        UIManager.ShowAllFactoryUIs(FactoryManager)
        UIManager.ShowAllResourceUIs(Player)

        # Hide button and disable trigger for loading factory.
        LoadFactoryButtonProp.Hide()
        LoadFactoryTrigger.Disable()

    # Reenables the factory loading trigger. 
    UIClosed<public>():void=
        LoadFactoryTrigger.Enable()
        
    # When player is added, propagate the information to relevant game elements and update UI.
    OnPlayerAdded(Player:player):void=
        UIManager.PlayerAdded(Player)
        Tutorial.PlayerAdded(Player)

        # Show factory and resource UI to player if factory is loaded.
        if:
            SelectedPlayer := MaybeSelectedPlayer?
        then:
            UIManager.ShowFactoryUI(Player, FactoryManager)
            UIManager.ShowResourceUI(Player, SelectedPlayer)

    # When player is removed, check if that player owned the factory.
    # If so, reset the game state.
    OnPlayerRemoved(Player:player):void=
        UIManager.PlayerRemoved(Player)
        if:
            SelectedPlayer := MaybeSelectedPlayer?
            SelectedPlayer = Player
        then:
            # Reset the billbord showing the factory owner.
            FactoryBillboard.SetText(PlainText("Vacant Toy\nFactory Lot"))
    
            # Reset the selected player. This will unload the current factory.
            SelectPlayer(false)

            # Close all UIs for all players.
            UIManager.CloseAllGameUIs()

            # Show popup explaining that selected player left.
            FactoryResetPopup.Show()

            # Show button and enable trigger for loading factory.
            LoadFactoryButtonProp.Show()
            LoadFactoryTrigger.Enable()
