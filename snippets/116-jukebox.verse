# Knowledge-neutral Verse snippet; project names and asset IDs generalized.
# id: snippet-116

# This file defines a jukebox that can be interacted with and play music for a price.

using { /Fortnite.com/Devices }
using { /Verse.org/Simulation }
using { Persistence }
using { UserInterface }

# A song on the jukebox.
jukebox_song := class<concrete>:

    # Title of the song.
    @editable
    Title:string = "Title"

    # Price to play the song.
    @editable
    Price:int = 5

    # Audio player that will actually play the song.
    @editable
    AudioPlayer:audio_player_device = audio_player_device{}

jukebox := class(creative_device):

    # The trigger that initiates interaction with the jukebox.
    @editable
    Trigger:trigger_device = trigger_device{}

    # Prop indicating that the jukebox can be interacted with.
    @editable
    ButtonProp:creative_prop = creative_prop{}

    # Time in seconds to play a selected song.
    @editable
    SongPlayTime:float = 60.0

    # Array of songs the jukebox can play.
    @editable
    Songs:[]jukebox_song = array{}

    # UI manager for showing and updating UI.
    @editable
    UIManager:ui_manager = ui_manager{}

    # Camera to use while interacting with the jukebox.
    @editable
    InteractCamera:gameplay_camera_fixed_point_device = gameplay_camera_fixed_point_device{}

    # Event signalled when a song is played.
    PlayEvent:event() = event(){}

    # The selected player that owns the jukebox. Can be false if nobody owns it yet.
    # The owner pays for songs that are played.
    var MaybeSelectedPlayer:?player = false

    # The audio player playing the previously selected song. Can be false if no song has been played yet.
    var MaybePreviousAudioPlayer:?audio_player_device = false

    # Runs when the device is started in a running game.
    OnBegin<override>()<suspends>:void=
        # Subscribe callback when interaction trigger is triggered.
        Trigger.TriggeredEvent.Subscribe(OnJukeboxTriggered)

    # Sets the selected player.
    SelectPlayer<public>(MaybePlayer:?player):void=
        set MaybeSelectedPlayer = MaybePlayer

        # If a player has been selected, show and enable interaction.
        if (MaybePlayer?):
            ButtonProp.Show()
            Trigger.Enable()
        else:
            ButtonProp.Hide()
            Trigger.Disable()

    # Plays the selected song.
    PlaySong<public>(Number:int):void=
        # Play the song if the selected player can afford it.
        if:
            SelectedPlayer := MaybeSelectedPlayer?
            Gold := GetGold[SelectedPlayer]
            Price := Songs[Number].Price
            Gold >= Price
            SpendGold[SelectedPlayer, Price]
        then:
            # Signal that a new song was selected.
            PlayEvent.Signal()

            # Play the song.
            if:
                AudioPlayer := Songs[Number].AudioPlayer
            then:
                AudioPlayer.Play()
                set MaybePreviousAudioPlayer = option{AudioPlayer}
                # Finish the song later.
                spawn:
                    FinishSong()
            # Update all players' UI.
            UIManager.UpdateAllPriceUIs()
            UIManager.UpdateAllResourceUIs()
            
    # Finishes the song when the time has passed or it is interrupted by another song being played.
    FinishSong()<suspends>:void=
        race:
            PlayEvent.Await()
            Sleep(SongPlayTime)
        
        # Stop previous song.
        if:
            PreviousAudioPlayer := MaybePreviousAudioPlayer?
        then:
            PreviousAudioPlayer.Stop()
    
    # When the interaction trigger is triggered, disable the trigger,
    # set up the interaction camera and show the UI. 
    OnJukeboxTriggered(MaybeAgent:?agent):void=
        if:
            Agent := MaybeAgent?
            Player := player[Agent]
        then:
            InteractCamera.AddTo(Player)
            Trigger.Disable()
            UIManager.ShowJukeboxUI(Player, Self)

    # Removes the interaction camera and reenables the interaction trigger.
    UIClosed<public>(Player:player):void=
        InteractCamera.RemoveFrom(Player)
        Trigger.Enable()
