# Knowledge-neutral Verse snippet; project names and asset IDs generalized.
# id: snippet-117

# This file defines the jukebox UI, which prompts the player to play a song on the jukebox.

using { /Fortnite.com/Devices }
using { /Fortnite.com/UI }
using { /UnrealEngine.com/Temporary/SpatialMath }
using { /UnrealEngine.com/Temporary/UI }
using { /Verse.org/Colors }
using { /Verse.org/Simulation }
using { Persistence }
using { Utility }

# Contains all widgets relevant for showing, updating and closing the UI.
jukebox_ui_widgets := struct:
    Canvas<public>:canvas
    SongTitleTextWidgets<public>:[]text_block
    SongPriceTextWidgets<public>:[]text_block
    PlayButtonWidgets<public>:[]text_button_base

jukebox_ui := class(creative_device):

    # Stores the created jukebox UI for each player.
    var UIMap:[player]jukebox_ui_widgets = map{}
    # Stores the jukebox whose songs can be played from the UI for each player.
    var JukeboxMap:[player]jukebox = map{}

    # When a player is added, create the jukebox UI.
    PlayerAdded(Player:player):void=
        UIResult := CreateUI()
        if:
            set UIMap[Player] = UIResult

    # When a player is removed, remove the stored data.
    PlayerRemoved(Player:player):void=
        if:
            NewUIMap := UIMap.RemoveKey[Player]
            NewJukeboxMap := JukeboxMap.RemoveKey[Player]
        then:
            set UIMap = NewUIMap
            set JukeboxMap = NewJukeboxMap

    # Create the UI.
    CreateUI():jukebox_ui_widgets=
        BackgroundColorWidget := color_block:
            DefaultColor := NamedColors.DarkGreen
            DefaultOpacity := 0.7
        HeaderTextWidget := text_block:
            DefaultText:= PlainText("Play some awesome tunes")
            DefaultTextColor:= NamedColors.White
            DefaultShadowOffset:= option{vector2{X := 2.0, Y := 3.0}}
            DefaultShadowColor:= NamedColors.Black
            DefaultShadowOpacity := 1.0
        Song0PriceTextWidget := text_block:
            DefaultTextColor:= NamedColors.White
            DefaultJustification:=text_justification.InvariantRight
            DefaultShadowOffset:= option{vector2{X := 2.0, Y := 3.0}}
            DefaultShadowColor:= NamedColors.Black
            DefaultShadowOpacity := 1.0
        Song1PriceTextWidget := text_block:
            DefaultTextColor:= NamedColors.White
            DefaultJustification:=text_justification.InvariantRight
            DefaultShadowOffset:= option{vector2{X := 2.0, Y := 3.0}}
            DefaultShadowColor:= NamedColors.Black
            DefaultShadowOpacity := 1.0
        Song2PriceTextWidget := text_block:
            DefaultTextColor:= NamedColors.White
            DefaultJustification:=text_justification.InvariantRight
            DefaultShadowOffset:= option{vector2{X := 2.0, Y := 3.0}}
            DefaultShadowColor:= NamedColors.Black
            DefaultShadowOpacity := 1.0
        Song3PriceTextWidget := text_block:
            DefaultTextColor:= NamedColors.White
            DefaultJustification:=text_justification.InvariantRight
            DefaultShadowOffset:= option{vector2{X := 2.0, Y := 3.0}}
            DefaultShadowColor:= NamedColors.Black
            DefaultShadowOpacity := 1.0
        Song0PriceTextureWidget := texture_block:
            DefaultImage := Textures.UI.T_Gold
            DefaultDesiredSize:=vector2{X := 40.0, Y := 40.0}
        Song1PriceTextureWidget := texture_block:
            DefaultImage := Textures.UI.T_Gold
            DefaultDesiredSize:=vector2{X := 40.0, Y := 40.0}
        Song2PriceTextureWidget := texture_block:
            DefaultImage := Textures.UI.T_Gold
            DefaultDesiredSize:=vector2{X := 40.0, Y := 40.0}
        Song3PriceTextureWidget := texture_block:
            DefaultImage := Textures.UI.T_Gold
            DefaultDesiredSize:=vector2{X := 40.0, Y := 40.0}

        Play0ButtonWidget := button_loud:
            DefaultText := PlainText("Play")
        Play1ButtonWidget := button_loud:
            DefaultText := PlainText("Play")
        Play2ButtonWidget := button_loud:
            DefaultText := PlainText("Play")
        Play3ButtonWidget := button_loud:
            DefaultText := PlainText("Play")
        Song0TitleTextWidget := text_block:
            DefaultTextColor:= NamedColors.White
            DefaultJustification:=text_justification.InvariantRight
            DefaultShadowOffset:= option{vector2{X := 2.0, Y := 3.0}}
            DefaultShadowColor:= NamedColors.Black
            DefaultShadowOpacity := 1.0
        Song1TitleTextWidget := text_block:
            DefaultTextColor:= NamedColors.White
            DefaultJustification:=text_justification.InvariantRight
            DefaultShadowOffset:= option{vector2{X := 2.0, Y := 3.0}}
            DefaultShadowColor:= NamedColors.Black
            DefaultShadowOpacity := 1.0
        Song2TitleTextWidget := text_block:
            DefaultTextColor:= NamedColors.White
            DefaultJustification:=text_justification.InvariantRight
            DefaultShadowOffset:= option{vector2{X := 2.0, Y := 3.0}}
            DefaultShadowColor:= NamedColors.Black
            DefaultShadowOpacity := 1.0
        Song3TitleTextWidget := text_block:
            DefaultTextColor:= NamedColors.White
            DefaultJustification:=text_justification.InvariantRight
            DefaultShadowOffset:= option{vector2{X := 2.0, Y := 3.0}}
            DefaultShadowColor:= NamedColors.Black
            DefaultShadowOpacity := 1.0
        CloseButtonWidget := button_regular:
            DefaultText:= PlainText("Close")
        
        CanvasWidget := canvas:
            Slots := array:
                canvas_slot:
                    Anchors := anchors{Minimum := vector2{X := 0.0, Y := 0.0}, Maximum := vector2{X := 1.0, Y := 1.0}}
                    Offsets := margin{Top := 0.0, Left := 0.0, Right := 0.0, Bottom := 0.0}
                    Alignment := vector2{X := 0.5, Y := 0.5}
                    Widget := BackgroundColorWidget
                canvas_slot:
                    Anchors := anchors{Minimum := vector2{X := 0.5, Y := 0.5}, Maximum := vector2{X := 0.5, Y := 0.5}}
                    Offsets := margin{Top := -390.0, Left := 0.0, Right := 0.0, Bottom := 0.0}
                    Alignment := vector2{X := 0.5, Y := 0.5}
                    SizeToContent := true
                    Widget := HeaderTextWidget
                canvas_slot:
                    Anchors := anchors{Minimum := vector2{X := 0.5, Y := 0.5}, Maximum := vector2{X := 0.5, Y := 0.5}}
                    Offsets := margin{Top := -210.0, Left := -360.0, Right := 0.0, Bottom := 0.0}
                    Alignment := vector2{X := 1.0, Y := 0.5}
                    SizeToContent := true
                    Widget := Song0PriceTextWidget
                canvas_slot:
                    Anchors := anchors{Minimum := vector2{X := 0.5, Y := 0.5}, Maximum := vector2{X := 0.5, Y := 0.5}}
                    Offsets := margin{Top := -100.0, Left := -360.0, Right := 0.0, Bottom := 0.0}
                    Alignment := vector2{X := 1.0, Y := 0.5}
                    SizeToContent := true
                    Widget := Song1PriceTextWidget
                canvas_slot:
                    Anchors := anchors{Minimum := vector2{X := 0.5, Y := 0.5}, Maximum := vector2{X := 0.5, Y := 0.5}}
                    Offsets := margin{Top := 10.0, Left := -360.0, Right := 0.0, Bottom := 0.0}
                    Alignment := vector2{X := 1.0, Y := 0.5}
                    SizeToContent := true
                    Widget := Song2PriceTextWidget
                canvas_slot:
                    Anchors := anchors{Minimum := vector2{X := 0.5, Y := 0.5}, Maximum := vector2{X := 0.5, Y := 0.5}}
                    Offsets := margin{Top := 120.0, Left := -360.0, Right := 0.0, Bottom := 0.0}
                    Alignment := vector2{X := 1.0, Y := 0.5}
                    SizeToContent := true
                    Widget := Song3PriceTextWidget
                canvas_slot:
                    Anchors := anchors{Minimum := vector2{X := 0.5, Y := 0.5}, Maximum := vector2{X := 0.5, Y := 0.5}}
                    Offsets := margin{Top := -210.0, Left := -330.0, Right := 0.0, Bottom := 0.0}
                    Alignment := vector2{X := 0.5, Y := 0.5}
                    Widget := Song0PriceTextureWidget
                canvas_slot:
                    Anchors := anchors{Minimum := vector2{X := 0.5, Y := 0.5}, Maximum := vector2{X := 0.5, Y := 0.5}}
                    Offsets := margin{Top := -100.0, Left := -330.0, Right := 0.0, Bottom := 0.0}
                    Alignment := vector2{X := 0.5, Y := 0.5}
                    Widget := Song1PriceTextureWidget
                canvas_slot:
                    Anchors := anchors{Minimum := vector2{X := 0.5, Y := 0.5}, Maximum := vector2{X := 0.5, Y := 0.5}}
                    Offsets := margin{Top := 10.0, Left := -330.0, Right := 0.0, Bottom := 0.0}
                    Alignment := vector2{X := 0.5, Y := 0.5}
                    Widget := Song2PriceTextureWidget
                canvas_slot:
                    Anchors := anchors{Minimum := vector2{X := 0.5, Y := 0.5}, Maximum := vector2{X := 0.5, Y := 0.5}}
                    Offsets := margin{Top := 120.0, Left := -330.0, Right := 0.0, Bottom := 0.0}
                    Alignment := vector2{X := 0.5, Y := 0.5}
                    Widget := Song3PriceTextureWidget
                canvas_slot:
                    Anchors := anchors{Minimum := vector2{X := 0.5, Y := 0.5}, Maximum := vector2{X := 0.5, Y := 0.5}}
                    Offsets := margin{Top := -210.0, Left := -200.0, Right := 0.0, Bottom := 0.0}
                    Alignment := vector2{X := 0.5, Y := 0.5}
                    SizeToContent := true
                    Widget := Play0ButtonWidget
                canvas_slot:
                    Anchors := anchors{Minimum := vector2{X := 0.5, Y := 0.5}, Maximum := vector2{X := 0.5, Y := 0.5}}
                    Offsets := margin{Top := -100.0, Left := -200.0, Right := 0.0, Bottom := 0.0}
                    Alignment := vector2{X := 0.5, Y := 0.5}
                    SizeToContent := true
                    Widget := Play1ButtonWidget
                canvas_slot:
                    Anchors := anchors{Minimum := vector2{X := 0.5, Y := 0.5}, Maximum := vector2{X := 0.5, Y := 0.5}}
                    Offsets := margin{Top := 10.0, Left := -200.0, Right := 0.0, Bottom := 0.0}
                    Alignment := vector2{X := 0.5, Y := 0.5}
                    SizeToContent := true
                    Widget := Play2ButtonWidget
                canvas_slot:
                    Anchors := anchors{Minimum := vector2{X := 0.5, Y := 0.5}, Maximum := vector2{X := 0.5, Y := 0.5}}
                    Offsets := margin{Top := 120.0, Left := -200.0, Right := 0.0, Bottom := 0.0}
                    Alignment := vector2{X := 0.5, Y := 0.5}
                    SizeToContent := true
                    Widget := Play3ButtonWidget                    
                canvas_slot:
                    Anchors := anchors{Minimum := vector2{X := 0.5, Y := 0.5}, Maximum := vector2{X := 0.5, Y := 0.5}}
                    Offsets := margin{Top := -210.0, Left := -70.0, Right := 0.0, Bottom := 0.0}
                    Alignment := vector2{X := 0.0, Y := 0.5}
                    SizeToContent := true
                    Widget := Song0TitleTextWidget
                canvas_slot:
                    Anchors := anchors{Minimum := vector2{X := 0.5, Y := 0.5}, Maximum := vector2{X := 0.5, Y := 0.5}}
                    Offsets := margin{Top := -100.0, Left := -70.0, Right := 0.0, Bottom := 0.0}
                    Alignment := vector2{X := 0.0, Y := 0.5}
                    SizeToContent := true
                    Widget := Song1TitleTextWidget
                canvas_slot:
                    Anchors := anchors{Minimum := vector2{X := 0.5, Y := 0.5}, Maximum := vector2{X := 0.5, Y := 0.5}}
                    Offsets := margin{Top := 10.0, Left := -70.0, Right := 0.0, Bottom := 0.0}
                    Alignment := vector2{X := 0.0, Y := 0.5}
                    SizeToContent := true
                    Widget := Song2TitleTextWidget
                canvas_slot:
                    Anchors := anchors{Minimum := vector2{X := 0.5, Y := 0.5}, Maximum := vector2{X := 0.5, Y := 0.5}}
                    Offsets := margin{Top := 120.0, Left := -70.0, Right := 0.0, Bottom := 0.0}
                    Alignment := vector2{X := 0.0, Y := 0.5}
                    SizeToContent := true
                    Widget := Song3TitleTextWidget
                canvas_slot:
                    Anchors := anchors{Minimum := vector2{X := 0.5, Y := 0.5}, Maximum := vector2{X := 0.5, Y := 0.5}}
                    Offsets := margin{Top := 350.0, Left := 0.0, Right := 0.0, Bottom := 0.0}
                    Alignment := vector2{X := 0.5, Y := 0.5}
                    SizeToContent := true
                    Widget := CloseButtonWidget

        Canvas := canvas:
            Slots := array:
                canvas_slot:
                    Anchors := anchors{Minimum := vector2{X := 0.0, Y := 0.0}, Maximum := vector2{X := 0.5, Y := 1.0}}
                    Offsets := margin{Top := 50.0, Left := 50.0, Right := 50.0, Bottom := 120.0}
                    Alignment := vector2{X := 0.5, Y := 0.5}
                    Widget := CanvasWidget

        # Subscribe callbacks to get notified when the player clicks the buttons.
        Play0ButtonWidget.OnClick().SubscribeWidgetMessage(OnPlayButton, 0)
        Play1ButtonWidget.OnClick().SubscribeWidgetMessage(OnPlayButton, 1)
        Play2ButtonWidget.OnClick().SubscribeWidgetMessage(OnPlayButton, 2)
        Play3ButtonWidget.OnClick().SubscribeWidgetMessage(OnPlayButton, 3)
        CloseButtonWidget.OnClick().Subscribe(OnCloseButton)

        return jukebox_ui_widgets:
            Canvas := Canvas
            SongTitleTextWidgets := array{Song0TitleTextWidget, Song1TitleTextWidget, Song2TitleTextWidget, Song3TitleTextWidget}
            SongPriceTextWidgets := array{Song0PriceTextWidget, Song1PriceTextWidget, Song2PriceTextWidget, Song3PriceTextWidget}
            PlayButtonWidgets := array{Play0ButtonWidget, Play1ButtonWidget, Play2ButtonWidget, Play3ButtonWidget}

    # Show the UI to the player.
    ShowUI(Player:player, Jukebox:jukebox, AudioPlayer:audio_player_device):void=
        if:
            set JukeboxMap[Player] = Jukebox
            UI := UIMap[Player]
            PlayerUI := GetPlayerUI[Player]
        then:
            # Play UI audio.
            AudioPlayer.Play(Player)

            # Update and show UI.
            UpdateUI(Player)
            PlayerUI.AddWidget(UI.Canvas, player_ui_slot{ZOrder := 0, InputMode := ui_input_mode.All})

    # Update the UI for the player.
    UpdateUI(Player:player):void=
        if:
            UI := UIMap[Player]
            Jukebox := JukeboxMap[Player]
            SelectedPlayer := Jukebox.MaybeSelectedPlayer?
            Gold := GetGold[SelectedPlayer]
            Title0 := Jukebox.Songs[0].Title
            Title1 := Jukebox.Songs[1].Title
            Title2 := Jukebox.Songs[2].Title
            Title3 := Jukebox.Songs[3].Title
            Price0 := Jukebox.Songs[0].Price
            Price1 := Jukebox.Songs[1].Price
            Price2 := Jukebox.Songs[2].Price
            Price3 := Jukebox.Songs[3].Price
            Play0Enabled := logic{Gold >= Price0}
            Play1Enabled := logic{Gold >= Price1}
            Play2Enabled := logic{Gold >= Price2}
            Play3Enabled := logic{Gold >= Price3}
            Song0TitleTextWidget := UI.SongTitleTextWidgets[0]
            Song1TitleTextWidget := UI.SongTitleTextWidgets[1]
            Song2TitleTextWidget := UI.SongTitleTextWidgets[2]
            Song3TitleTextWidget := UI.SongTitleTextWidgets[3]
            Song0PriceTextWidget := UI.SongPriceTextWidgets[0]
            Song1PriceTextWidget := UI.SongPriceTextWidgets[1]
            Song2PriceTextWidget := UI.SongPriceTextWidgets[2]
            Song3PriceTextWidget := UI.SongPriceTextWidgets[3]
            Play0ButtonWidget := UI.PlayButtonWidgets[0]
            Play1ButtonWidget := UI.PlayButtonWidgets[1]
            Play2ButtonWidget := UI.PlayButtonWidgets[2]
            Play3ButtonWidget := UI.PlayButtonWidgets[3]
        then:
            Song0TitleTextWidget.SetText(PlainText(Title0))
            Song1TitleTextWidget.SetText(PlainText(Title1))
            Song2TitleTextWidget.SetText(PlainText(Title2))
            Song3TitleTextWidget.SetText(PlainText(Title3))
            Song0PriceTextWidget.SetText(CountText(Price0))
            Song1PriceTextWidget.SetText(CountText(Price1))
            Song2PriceTextWidget.SetText(CountText(Price2))
            Song3PriceTextWidget.SetText(CountText(Price3))
            Play0ButtonWidget.SetEnabled(Play0Enabled)
            Play1ButtonWidget.SetEnabled(Play1Enabled)
            Play2ButtonWidget.SetEnabled(Play2Enabled)
            Play3ButtonWidget.SetEnabled(Play3Enabled)

    # When the player clicks the play button, close the UI and play that song on the jukebox.
    OnPlayButton(Message:widget_message, Number:int):void=
        if:
            Jukebox := JukeboxMap[Message.Player]
        then:
            OnCloseButton(Message)

            Jukebox.PlaySong(Number)

    # When the player clicks the close button, close the UI.
    OnCloseButton(Message:widget_message):void=
        CloseUI(Message.Player)

    # Close the UI for the player.
    CloseUI<public>(Player:player):void=
        if:
            Jukebox := JukeboxMap[Player]
            UI := UIMap[Player]
            PlayerUI := GetPlayerUI[Player]
        then:
            Jukebox.UIClosed(Player)
            PlayerUI.RemoveWidget(UI.Canvas)
