# Knowledge-neutral Verse snippet; project names and asset IDs generalized.
# id: snippet-118

# This file defines the load factory UI, which prompts the player to load his factory. 

using { /Fortnite.com/Devices }
using { /Fortnite.com/UI }
using { /UnrealEngine.com/Temporary/SpatialMath }
using { /UnrealEngine.com/Temporary/UI }
using { /Verse.org/Colors }
using { /Verse.org/Simulation }
using { Persistence }
using { Utility }

# Contains all widgets relevant for showing and closing the UI.
load_factory_ui_widgets := struct:
    Canvas<public>:canvas

load_factory_ui := class(creative_device):

    # Stores the created load factory UI for each player.
    var UIMap:[player]load_factory_ui_widgets = map{}
    # Stores the game manager that will handle loading the factory.
    var MaybeGameManager:?game_manager = false

    # When a player is added, create the load factory UI.
    PlayerAdded(Player:player):void=
        UIResult := CreateUI()
        if:
            set UIMap[Player] = UIResult

    # When a player is removed, remove the stored data.
    PlayerRemoved(Player:player):void=
        if:
            NewUIMap := UIMap.RemoveKey[Player]
        then:
            set UIMap = NewUIMap

    # Create the UI.
    CreateUI():load_factory_ui_widgets=
        BackgroundColorWidget := color_block:
            DefaultColor := NamedColors.DarkGreen
            DefaultOpacity := 0.7
        HeaderTextWidget := text_block:
            DefaultText := PlainText("Start building your factory")
            DefaultTextColor:= NamedColors.White
            DefaultShadowOffset:= option{vector2{X := 2.0, Y := 3.0}}
            DefaultShadowColor:= NamedColors.Black
            DefaultShadowOpacity := 1.0
        FactoryTextureWidget := texture_block:
            DefaultImage := Textures.UI.T_Factory
            DefaultDesiredSize:=vector2{X := 1000.0, Y := 500.0}
        LoadButtonWidget := button_loud:
            DefaultText := PlainText("Load")
        CloseButtonWidget := button_regular:
            DefaultText:= PlainText("Close")
        
        CanvasWidget := canvas:
            Slots := array:
                canvas_slot:
                    Anchors := anchors{Minimum := vector2{X := 0.0, Y := 0.0}, Maximum := vector2{X := 1.0, Y := 1.0}}
                    Offsets := margin{Top := 0.0, Left := 0.0, Right := 0.0, Bottom := 0.0}
                    Alignment := vector2{X := 0.5, Y := 0.5}
                    Widget := BackgroundColorWidget
                canvas_slot:
                    Anchors := anchors{Minimum := vector2{X := 0.5, Y := 0.5}, Maximum := vector2{X := 0.5, Y := 0.5}}
                    Offsets := margin{Top := -390.0, Left := 0.0, Right := 0.0, Bottom := 0.0}
                    Alignment := vector2{X := 0.5, Y := 0.5}
                    SizeToContent := true
                    Widget := HeaderTextWidget
                canvas_slot:
                    Anchors := anchors{Minimum := vector2{X := 0.5, Y := 0.5}, Maximum := vector2{X := 0.5, Y := 0.5}}
                    Offsets := margin{Top := -70.0, Left := 0.0, Right := 0.0, Bottom := 0.0}
                    Alignment := vector2{X := 0.5, Y := 0.5}
                    Widget := FactoryTextureWidget
                canvas_slot:
                    Anchors := anchors{Minimum := vector2{X := 0.5, Y := 0.5}, Maximum := vector2{X := 0.5, Y := 0.5}}
                    Offsets := margin{Top := 250.0, Left := 0.0, Right := 0.0, Bottom := 0.0}
                    Alignment := vector2{X := 0.5, Y := 0.5}
                    SizeToContent := true
                    Widget := LoadButtonWidget
                canvas_slot:
                    Anchors := anchors{Minimum := vector2{X := 0.5, Y := 0.5}, Maximum := vector2{X := 0.5, Y := 0.5}}
                    Offsets := margin{Top := 350.0, Left := 0.0, Right := 0.0, Bottom := 0.0}
                    Alignment := vector2{X := 0.5, Y := 0.5}
                    SizeToContent := true
                    Widget := CloseButtonWidget

        Canvas := canvas:
            Slots := array:
                canvas_slot:
                    Anchors := anchors{Minimum := vector2{X := 0.0, Y := 0.0}, Maximum := vector2{X := 1.0, Y := 1.0}}
                    Offsets := margin{Top := 50.0, Left := 50.0, Right := 50.0, Bottom := 120.0}
                    Alignment := vector2{X := 0.5, Y := 0.5}
                    Widget := CanvasWidget

        # Subscribe callbacks to get notified when the player clicks the buttons.
        LoadButtonWidget.OnClick().Subscribe(OnLoadButton)
        CloseButtonWidget.OnClick().Subscribe(OnCloseButton)

        return load_factory_ui_widgets:
            Canvas := Canvas

    # Show the UI to the player.
    ShowUI(Player:player, GameManager:game_manager):void=
        if:
            set MaybeGameManager = option{GameManager}
            UI := UIMap[Player]
            PlayerUI := GetPlayerUI[Player]
        then:
            PlayerUI.AddWidget(UI.Canvas, player_ui_slot{ZOrder := 0, InputMode := ui_input_mode.All})

    # When the player clicks the load button, close the UI and load the factory for that player.
    OnLoadButton(Message:widget_message):void=
        if:
            GameManager := MaybeGameManager?
        then:
            OnCloseButton(Message)

            GameManager.LoadFactory(Message.Player)

    # When the player clicks the close button, close the UI.
    OnCloseButton(Message:widget_message):void=
        CloseUI(Message.Player)

    # Close the UI for the player.
    CloseUI<public>(Player:player):void=
        if:
            GameManager := MaybeGameManager?
            UI := UIMap[Player]
            PlayerUI := GetPlayerUI[Player]
        then:
            # Notify the game manager that the UI was closed.
            GameManager.UIClosed()
            PlayerUI.RemoveWidget(UI.Canvas)
