# Knowledge-neutral Verse snippet; project names and asset IDs generalized.
# id: snippet-119

# This file defines the offline production UI, which shows the player how many products were
# produced by the given factory while the player was away.

using { /Fortnite.com/Devices }
using { /Fortnite.com/UI }
using { /UnrealEngine.com/Temporary/SpatialMath }
using { /UnrealEngine.com/Temporary/UI }
using { /Verse.org/Colors }
using { /Verse.org/Simulation }
using { Factory }
using { Persistence }
using { Utility }

# Contains all widgets relevant for showing and closing the UI.
offline_production_ui_widgets := struct:
    Canvas<public>:canvas
    StorageFullTextWidget<public>:text_block
    ProductTextWidgets<public>:[]text_block
    ProductTextureWidgets<public>:[]texture_block

offline_production_ui := class(creative_device):

    # Stores the created offline production UI for each player.
    var UIMap:[player]offline_production_ui_widgets = map{}

    # When a player is added, create the offline production UI.
    PlayerAdded(Player:player):void=
        UIResult := CreateUI()
        if:
            set UIMap[Player] = UIResult

    # When a player is removed, remove the stored data.
    PlayerRemoved(Player:player):void=
        if:
            NewUIMap := UIMap.RemoveKey[Player]
        then:
            set UIMap = NewUIMap

    # Create the UI.
    CreateUI():offline_production_ui_widgets=
        BackgroundColorWidget := color_block:
            DefaultColor := NamedColors.DarkGreen
            DefaultOpacity := 0.7
        HeaderTextWidget := text_block:
            DefaultText:= PlainText("Your factory produced the following toys while you were away")
            DefaultTextColor:= NamedColors.White
            DefaultShadowOffset:= option{vector2{X := 2.0, Y := 3.0}}
            DefaultShadowColor:= NamedColors.Black
            DefaultShadowOpacity := 1.0
        StorageFullTextWidget := text_block:
            DefaultText:= PlainText("Storage is full! Upgrade to increase capacity")
            DefaultTextColor:= NamedColors.Red
            DefaultShadowOffset:= option{vector2{X := 2.0, Y := 3.0}}
            DefaultShadowColor:= NamedColors.Black
            DefaultShadowOpacity := 1.0
        Product0TextWidget := text_block:
            DefaultTextColor:= NamedColors.White
            DefaultJustification:=text_justification.InvariantRight
            DefaultShadowOffset:= option{vector2{X := 2.0, Y := 3.0}}
            DefaultShadowColor:= NamedColors.Black
            DefaultShadowOpacity := 1.0
        Product1TextWidget := text_block:
            DefaultTextColor:= NamedColors.White
            DefaultJustification:=text_justification.InvariantRight
            DefaultShadowOffset:= option{vector2{X := 2.0, Y := 3.0}}
            DefaultShadowColor:= NamedColors.Black
            DefaultShadowOpacity := 1.0
        Product2TextWidget := text_block:
            DefaultTextColor:= NamedColors.White
            DefaultJustification:=text_justification.InvariantRight
            DefaultShadowOffset:= option{vector2{X := 2.0, Y := 3.0}}
            DefaultShadowColor:= NamedColors.Black
            DefaultShadowOpacity := 1.0
        Product3TextWidget := text_block:
            DefaultTextColor:= NamedColors.White
            DefaultJustification:=text_justification.InvariantRight
            DefaultShadowOffset:= option{vector2{X := 2.0, Y := 3.0}}
            DefaultShadowColor:= NamedColors.Black
            DefaultShadowOpacity := 1.0
        Product0TextureWidget := texture_block:
            DefaultImage := Textures.UI.T_Plane
            DefaultDesiredSize:=vector2{X := 120.0, Y := 120.0}
        Product1TextureWidget := texture_block:
            DefaultImage := Textures.UI.T_TRex
            DefaultDesiredSize:=vector2{X := 120.0, Y := 120.0}
        Product2TextureWidget := texture_block:
            DefaultImage := Textures.UI.T_Castle
            DefaultDesiredSize:=vector2{X := 120.0, Y := 120.0}
        Product3TextureWidget := texture_block:
            DefaultImage := Textures.UI.T_Spaceship
            DefaultDesiredSize:=vector2{X := 120.0, Y := 120.0}
        CloseButtonWidget := button_regular:
            DefaultText:= PlainText("Awesome!")

        Canvas := canvas:
            Slots := array:
                canvas_slot:
                    Anchors := anchors{Minimum := vector2{X := 0.0, Y := 0.0}, Maximum := vector2{X := 1.0, Y := 1.0}}
                    Offsets := margin{Top := 50.0, Left := 50.0, Right := 50.0, Bottom := 120.0}
                    Alignment := vector2{X := 0.5, Y := 0.5}
                    Widget := BackgroundColorWidget
                canvas_slot:
                    Anchors := anchors{Minimum := vector2{X := 0.5, Y := 0.5}, Maximum := vector2{X := 0.5, Y := 0.5}}
                    Offsets := margin{Top := -390.0, Left := 0.0, Right := 0.0, Bottom := 0.0}
                    Alignment := vector2{X := 0.5, Y := 0.5}
                    SizeToContent := true
                    Widget := HeaderTextWidget
                canvas_slot:
                    Anchors := anchors{Minimum := vector2{X := 0.5, Y := 0.5}, Maximum := vector2{X := 0.5, Y := 0.5}}
                    Offsets := margin{Top := -310.0, Left := 0.0, Right := 0.0, Bottom := 0.0}
                    Alignment := vector2{X := 0.5, Y := 0.5}
                    SizeToContent := true
                    Widget := StorageFullTextWidget
                canvas_slot:
                    Anchors := anchors{Minimum := vector2{X := 0.5, Y := 0.5}, Maximum := vector2{X := 0.5, Y := 0.5}}
                    Offsets := margin{Top := -210.0, Left := -30.0, Right := 0.0, Bottom := 0.0}
                    Alignment := vector2{X := 1.0, Y := 0.5}
                    SizeToContent := true
                    Widget := Product0TextWidget
                canvas_slot:
                    Anchors := anchors{Minimum := vector2{X := 0.5, Y := 0.5}, Maximum := vector2{X := 0.5, Y := 0.5}}
                    Offsets := margin{Top := -100.0, Left := -30.0, Right := 0.0, Bottom := 0.0}
                    Alignment := vector2{X := 1.0, Y := 0.5}
                    SizeToContent := true
                    Widget := Product1TextWidget
                canvas_slot:
                    Anchors := anchors{Minimum := vector2{X := 0.5, Y := 0.5}, Maximum := vector2{X := 0.5, Y := 0.5}}
                    Offsets := margin{Top := 10.0, Left := -30.0, Right := 0.0, Bottom := 0.0}
                    Alignment := vector2{X := 1.0, Y := 0.5}
                    SizeToContent := true
                    Widget := Product2TextWidget
                canvas_slot:
                    Anchors := anchors{Minimum := vector2{X := 0.5, Y := 0.5}, Maximum := vector2{X := 0.5, Y := 0.5}}
                    Offsets := margin{Top := 120.0, Left := -30.0, Right := 0.0, Bottom := 0.0}
                    Alignment := vector2{X := 1.0, Y := 0.5}
                    SizeToContent := true
                    Widget := Product3TextWidget
                canvas_slot:
                    Anchors := anchors{Minimum := vector2{X := 0.5, Y := 0.5}, Maximum := vector2{X := 0.5, Y := 0.5}}
                    Offsets := margin{Top := -210.0, Left := 50.0, Right := 0.0, Bottom := 0.0}
                    Alignment := vector2{X := 0.5, Y := 0.5}
                    Widget := Product0TextureWidget
                canvas_slot:
                    Anchors := anchors{Minimum := vector2{X := 0.5, Y := 0.5}, Maximum := vector2{X := 0.5, Y := 0.5}}
                    Offsets := margin{Top := -100.0, Left := 50.0, Right := 0.0, Bottom := 0.0}
                    Alignment := vector2{X := 0.5, Y := 0.5}
                    Widget := Product1TextureWidget
                canvas_slot:
                    Anchors := anchors{Minimum := vector2{X := 0.5, Y := 0.5}, Maximum := vector2{X := 0.5, Y := 0.5}}
                    Offsets := margin{Top := 10.0, Left := 50.0, Right := 0.0, Bottom := 0.0}
                    Alignment := vector2{X := 0.5, Y := 0.5}
                    Widget := Product2TextureWidget
                canvas_slot:
                    Anchors := anchors{Minimum := vector2{X := 0.5, Y := 0.5}, Maximum := vector2{X := 0.5, Y := 0.5}}
                    Offsets := margin{Top := 120.0, Left := 50.0, Right := 0.0, Bottom := 0.0}
                    Alignment := vector2{X := 0.5, Y := 0.5}
                    Widget := Product3TextureWidget
                canvas_slot:
                    Anchors := anchors{Minimum := vector2{X := 0.5, Y := 0.5}, Maximum := vector2{X := 0.5, Y := 0.5}}
                    Offsets := margin{Top := 350.0, Left := 0.0, Right := 0.0, Bottom := 0.0}
                    Alignment := vector2{X := 0.5, Y := 0.5}
                    SizeToContent := true
                    Widget := CloseButtonWidget

        # Subscribe callback to get notified when the player clicks the button.
        CloseButtonWidget.OnClick().Subscribe(OnCloseButton)

        return offline_production_ui_widgets:
            Canvas := Canvas
            StorageFullTextWidget := StorageFullTextWidget
            ProductTextWidgets := array{Product0TextWidget, Product1TextWidget, Product2TextWidget, Product3TextWidget}
            ProductTextureWidgets := array{Product0TextureWidget, Product1TextureWidget, Product2TextureWidget, Product3TextureWidget}

    # Show the UI to the player.
    ShowUI(Player:player, FactoryManager:factory_manager, ProductsProducedOffline:[]int, AudioPlayer:audio_player_device):void=
        if:
            UI := UIMap[Player]
            PlayerUI := GetPlayerUI[Player]
            SelectedPlayer := FactoryManager.GetSelectedPlayer[]
            FactoryStorageBuilt := logic{IsFactoryStorageBuilt[SelectedPlayer]}
            # Only show UI if factory storage was built.
            FactoryStorageBuilt?
            RoomInFactoryStorage := logic{IsRoomInFactoryStorage[SelectedPlayer]}
            Product0 := ProductsProducedOffline[0]
            Product1 := ProductsProducedOffline[1]
            Product2 := ProductsProducedOffline[2]
            Product3 := ProductsProducedOffline[3]
            Product0TextWidget := UI.ProductTextWidgets[0]
            Product1TextWidget := UI.ProductTextWidgets[1]
            Product2TextWidget := UI.ProductTextWidgets[2]
            Product3TextWidget := UI.ProductTextWidgets[3]
            Product0TextureWidget := UI.ProductTextureWidgets[0]
            Product1TextureWidget := UI.ProductTextureWidgets[1]
            Product2TextureWidget := UI.ProductTextureWidgets[2]
            Product3TextureWidget := UI.ProductTextureWidgets[3]
            Product0Built := logic{IsFactoryMouldBuilt[SelectedPlayer, 0]}
            Product1Built := logic{IsFactoryMouldBuilt[SelectedPlayer, 1]}
            Product2Built := logic{IsFactoryMouldBuilt[SelectedPlayer, 2]}
            Product3Built := logic{IsFactoryMouldBuilt[SelectedPlayer, 3]}
        then:
            # Play UI audio.
            AudioPlayer.Play(Player)

            # If there is not room in storage, show warning to the player.
            if (RoomInFactoryStorage?):
                UI.StorageFullTextWidget.SetVisibility(widget_visibility.Hidden)
            else:
                UI.StorageFullTextWidget.SetVisibility(widget_visibility.Visible)

            Product0TextWidget.SetText(CountText(Product0))
            Product1TextWidget.SetText(CountText(Product1))
            Product2TextWidget.SetText(CountText(Product2))
            Product3TextWidget.SetText(CountText(Product3))

            # Only show the product texture if the player has the ability to produce that product.
            if (Product0Built?):
                Product0TextureWidget.SetImage(Textures.UI.T_Plane)
            else:
                Product0TextureWidget.SetImage(Textures.UI.T_Hidden)

            if (Product1Built?):
                Product1TextureWidget.SetImage(Textures.UI.T_TRex)
            else:
                Product1TextureWidget.SetImage(Textures.UI.T_Hidden)

            if (Product2Built?):
                Product2TextureWidget.SetImage(Textures.UI.T_Castle)
            else:
                Product2TextureWidget.SetImage(Textures.UI.T_Hidden)

            if (Product3Built?):
                Product3TextureWidget.SetImage(Textures.UI.T_Spaceship)
            else:
                Product3TextureWidget.SetImage(Textures.UI.T_Hidden)

            PlayerUI.AddWidget(UI.Canvas, player_ui_slot{ZOrder := 0, InputMode := ui_input_mode.All})

    # When the player clicks the close button, close the UI.
    OnCloseButton(Message:widget_message):void=
        CloseUI(Message.Player)

    # Close the UI for the player.
    CloseUI<public>(Player:player):void=
        if:
            UI := UIMap[Player]
            PlayerUI := GetPlayerUI[Player]
        then:
            PlayerUI.RemoveWidget(UI.Canvas)
