# Knowledge-neutral Verse snippet; project names and asset IDs generalized.
# id: snippet-120

# This file defines order that handles the logic and visuals of an order.

using { /Fortnite.com/Characters }
using { /Fortnite.com/Devices }
using { /Fortnite.com/Devices/CreativeAnimation }
using { /UnrealEngine.com/Temporary/SpatialMath }
using { /Verse.org/Random }
using { /Verse.org/Simulation }
using { Factory }
using { Persistence }
using { Time }
using { UserInterface }
using { Utility }

# Contains information for one delivery to an order.
delivery := struct:
    ProductNumber<public>:int
    Amount<public>:int

order := class(creative_device):
    # The factory moulds the order will get the product prop assets from when a delivery is made.
    @editable
    FactoryMoulds:[]factory_mould = array{}

    # The factory wrappers the order will get the product prop assets from when a delivery is made.
    @editable
    FactoryWrappers:[]factory_wrapper = array{}

    # The factory storage the order will remove products from when a delivery is made.
    @editable
    FactoryStorage:factory_storage = factory_storage{}

    # The trigger that initiates interaction with the order. 
    @editable
    Trigger:trigger_device = trigger_device{}

    # Prop indicating that a delivery can be made to the order.
    @editable
    DeliveryPossibleProp:creative_prop = creative_prop{}

    # Prop indicating that a delivery cannot be made to the order.
    @editable
    DeliveryNotPossibleProp:creative_prop = creative_prop{}

    # Mutator zone for detecting if any players are inside the order when it is respawned
    # at the end of the order completion animation.
    @editable
    MutatorZone:mutator_zone_device = mutator_zone_device{}

    # Prop that is the place where products will be delivered from. 
    @editable
    TapProp:creative_prop = creative_prop{}

    # Audio that plays when delivering.
    @editable
    DeliverAudioPlayer:audio_player_device = audio_player_device{}

    # Props that define the positions where spawn effects should show.
    @editable
    EffectPositionProps:[]creative_prop = array{}

    # Prop asset containing an effect to show when the order spawns. 
    @editable
    SpawnEffect:creative_prop_asset = DefaultCreativePropAsset
    
    # Prop that is the place where products will be delivered to. 
    @editable
    SackProp:creative_prop = creative_prop{}

    # The sequence playing when an order is completed.
    @editable
    CinematicSequence:cinematic_sequence_device = cinematic_sequence_device{}

    # Audio that plays when an order is completed.
    @editable
    OrderCompletionAudioPlayer:audio_player_device = audio_player_device{}

    # Audio that plays when the sleigh spawns.
    @editable
    SleighSpawnAudioPlayer:audio_player_device = audio_player_device{}
    
    # UI manager for showing and updating UI.
    @editable
    UIManager:ui_manager = ui_manager{}

    # The real time device used to set the date and time when creating a new order.
    @editable
    RealTime:real_time = real_time{}

    # Camera to use while interacting with the order.
    @editable
    InteractCamera:gameplay_camera_fixed_point_device = gameplay_camera_fixed_point_device{}

    # Event signalled when delivering to the order.
    DeliverEvent:event(delivery) = event(delivery){}

    # Time in seconds between deliveries when delivering more than one factory product.
    DeliveryTime:float = 0.2

    # The vertical offset from the tap's position to where the factory product should
    # be spawned when delivering.
    TapOffset:vector3 = vector3{X := 0.0, Y := 0.0, Z := -121.6}

    # The vertical offset from the sack's position to where the factory product should
    # be fall to when delivering.
    SackOffset:vector3 = vector3{X := 0.0, Y := 0.0, Z := 44.8}

    # The vertical offset the tap should be pushed back when delivering.
    TapPushedBackOffset:vector3 = vector3{X := 0.0, Y := 0.0, Z := 6.4}

    # The selected player that owns the order. Can be false if nobody owns it yet.
    # The owner supplies products from his factory storage and receives gold
    # when delivering to the order.
    var MaybeSelectedPlayer:?player = false

    # The initial transform of tap.
    var TapInitialTransform:transform = transform{}

    # The initial transform of sack.
    var SackInitialTransform:transform = transform{}

    # Locks the trigger out from being enabled while interacting with order.
    var TriggerLocked:logic = false

    # The number of factory products in queue to be delivered.
    var ProductsLeftToDeliver:int = 0

    # The prop asset representing the factory product in queue to be delivered.
    var PropAssetToDeliver:creative_prop_asset = DefaultCreativePropAsset

    # Runs when the device is started in a running game.
    OnBegin<override>()<suspends>:void=
        # Store initial transforms of tap and sack.
        set TapInitialTransform = TapProp.GetTransform()
        set SackInitialTransform = SackProp.GetTransform()
        # Subscribe callback when interaction trigger is triggered.
        Trigger.TriggeredEvent.Subscribe(OnOrderTriggered)
        # Continuously monitor if order can be delivered to.
        spawn:
            MonitorDeliveryPossible()
        # Continuously handle delivery of queued factory products.
        spawn:
            HandleDelivery()

    # Sets the selected player.
    SelectPlayer<public>(MaybePlayer:?player):void=
        set MaybeSelectedPlayer = MaybePlayer

        # Setup correct state of order.
        Update()
        
    # Updates the visuals of the order to reflect the current state.
    # Creates a new order if the current order is completed.
    Update():void=
        if:
            SelectedPlayer := MaybeSelectedPlayer?
            OrderTime := GetOrderTime[SelectedPlayer]
            not OrderTime.IsValid[] or IsOrderCompleted[SelectedPlayer]
            PlaceNewOrder[SelectedPlayer, RealTime.GetDateAndTime()]

    # Makes a delivery. Try to deliver as many as the value.
    Deliver<public>(ProductNumber:int, MaxProductsToDeliver:int):void=
        # Perform the actual delivery if possible.
        if:
            SelectedPlayer := MaybeSelectedPlayer?
            FactoryProduct := GetFactoryProduct[SelectedPlayer, ProductNumber]
            var ProductsToDeliver:int = Min(MaxProductsToDeliver, FactoryProduct)
            ProductsOrdered := GetProductsOrdered[SelectedPlayer, ProductNumber]
            ProductsDelivered := GetProductsDelivered[SelectedPlayer, ProductNumber]
            if (ProductsDelivered < ProductsOrdered):
                set ProductsToDeliver = Min(ProductsToDeliver, ProductsOrdered - ProductsDelivered)
            ProductsToDeliver > 0
            WithdrawFactoryProduct[SelectedPlayer, ProductNumber, ProductsToDeliver]
            DeliverProduct[SelectedPlayer, ProductNumber, ProductsToDeliver]
        then:
            # Signal that a delivery was made.
            DeliverEvent.Signal(delivery{ProductNumber := ProductNumber, Amount := ProductsToDeliver})

            # Show that factory storage removed some products.
            FactoryStorage.Remove(ProductNumber, ProductsToDeliver)

            # Queue up delivery of products. The delivery will be handled by the dedicated delivery task.
            # If factory wrapper is built, use the product that is the outcome of the wrapper.
            # Otherwise, use the product of the factory mould.
            if (IsFactoryWrapperBuilt[SelectedPlayer, ProductNumber]):
                if:
                    PropAsset := FactoryWrappers[ProductNumber].Product
                then:
                    set ProductsLeftToDeliver = ProductsToDeliver
                    set PropAssetToDeliver = PropAsset 
            else:
                if:
                    PropAsset := FactoryMoulds[ProductNumber].Product
                then:
                    set ProductsLeftToDeliver = ProductsToDeliver
                    set PropAssetToDeliver = PropAsset 
            # If the order is completed, show order completion animation.
            # Otherwise, update the order right away.
            if (IsOrderCompleted[SelectedPlayer]):
                spawn:
                    AnimateOrderCompletion(ProductsToDeliver)
            else:
                Update()

            # Update all players' UI.
            UIManager.UpdateAllFactoryStorageUIs()
            UIManager.UpdateAllPriceUIs()
            UIManager.UpdateAllResourceUIs()
            UIManager.UpdateAllFactoryUIs()

    # Handles delivering queued factory products.
    HandleDelivery()<suspends>:void=
        loop:
            if (ProductsLeftToDeliver > 0):
                set ProductsLeftToDeliver -=1
                spawn:
                    AnimateDelivery(PropAssetToDeliver)
                Sleep(DeliveryTime)
            else:
                Sleep(0.0)

    # Animates the delivery of a factory product from tap to sack before disposing of it.
    # Also, pushes back the tap and shakes the sack.
    AnimateDelivery(PropAsset:creative_prop_asset)<suspends>:void=
        StartPosition := TapInitialTransform.Translation + TapOffset
        StartRotation := MakeRotationFromYawPitchRollDegrees(GetRandomFloat(-180.0, 180.0), GetRandomFloat(-5.0, 5.0), GetRandomFloat(-5.0, 5.0))

        # Play deliver audio.
        DeliverAudioPlayer.Play()

        if:
            Prop := SpawnProp(PropAsset, StartPosition, StartRotation)(0)?
        then:
            PropTransform := Prop.GetTransform()
            TargetPosition := SackInitialTransform.Translation + SackOffset
            TargetRotation := StartRotation.RotateBy(MakeRotationFromYawPitchRollDegrees(GetRandomFloat(-45.0, 45.0), GetRandomFloat(-30.0, 30.0), GetRandomFloat(-30.0, 30.0)))

            sync:
                block:
                    Prop.MoveTo(TargetPosition, TargetRotation, 0.3)
                    # Hide the prop before disposing it to avoid showing BRICK explosion effect.
                    Prop.Hide()
                    Prop.Dispose()
                block:
                    if:
                        TapPushedBackPosition := TapInitialTransform.Translation + TapPushedBackOffset
                        TapProp.TeleportTo[TapPushedBackPosition, TapInitialTransform.Rotation]
                    then:
                        TapProp.MoveTo(TapInitialTransform, 0.15)
                block:
                    Sleep(0.2)
                    if:
                        SackProp.TeleportTo[SackInitialTransform]
                    then:
                        SackProp.Shake(3.0, 5.0, 5, 0.3)

    # Animates the order completion. When the animation completes, all props reset back to
    # initial positions.
    AnimateOrderCompletion(ProductsToDeliver:int)<suspends>:void=
        # Wait for products to be delivered.
        Sleep(ProductsToDeliver * DeliveryTime)
        # Play animation.
        CinematicSequence.Play()
        # Play order completion audio.
        OrderCompletionAudioPlayer.Play()
        # Wait for animation to stop
        CinematicSequence.StoppedEvent.Await()
        # Play sleigh spawn audio.
        SleighSpawnAudioPlayer.Play()
         # Remove any players that might be stuck inside the reset order props.
        RemovePlayersInside()
        # Show spawn effect.
        ShowSpawnEffect()
        # Complete the order.
        Complete()

    # Shows the spawn effect. One for each position prop.
    ShowSpawnEffect():void=
        for(Prop : EffectPositionProps):
            spawn:
                ShowEffect(SpawnEffect, Prop)
    
    # Spawns the effect at the position of the prop.
    # Disposes it again after a few seconds.
    ShowEffect(Effect:creative_prop_asset, EffectPositionProp:creative_prop)<suspends>:void=
        Transform := EffectPositionProp.GetTransform()
        SpawnResult := SpawnProp(Effect, Transform)
        if (EffectProp := SpawnResult(0)?):
            Sleep(3.0)
            EffectProp.Dispose()    

    # When the interaction trigger is triggered, disable and lock the trigger,
    # setup the interaction camera and show the UI.
    OnOrderTriggered(MaybeAgent:?agent):void=
        if:
            Agent := MaybeAgent?
            Player := player[Agent]
        then:
            InteractCamera.AddTo(Player)
            Trigger.Disable()
            set TriggerLocked = true
            UIManager.ShowOrderUI(Player, Self)

    # Removes the interaction camera, reenables and unlocks the interaction trigger.
    UIClosed<public>(Player:player):void=
        InteractCamera.RemoveFrom(Player)
        Trigger.Enable()
        set TriggerLocked = false

    # Completes the order.
    Complete():void=
        # Create a new order.
        Update()
        # Update all players' UI.
        UIManager.UpdateAllOrderUIs()

    # Monitors if delivery to the order is possible.
    # Updates the props accordingly.
    MonitorDeliveryPossible()<suspends>:void=
        loop:
            if:
                SelectedPlayer := MaybeSelectedPlayer?
            then:
                if (not TriggerLocked?):
                    Trigger.Enable()
                if:
                    IsAnyProductInFactoryStorage[SelectedPlayer]
                then:
                    DeliveryPossibleProp.Show()
                    DeliveryNotPossibleProp.Hide()
                else:
                    DeliveryPossibleProp.Hide()
                    DeliveryNotPossibleProp.Show()
            else:
                Trigger.Disable()
                DeliveryPossibleProp.Hide()
                DeliveryNotPossibleProp.Hide()
            Sleep(0.1)

    # Removes any players inside the order's mutator zone by making them respawn.
    RemovePlayersInside():void=
        for(Player : GetPlayspace().GetPlayers()):
            if:
                MutatorZone.IsInVolume[Player]
                Character := Player.GetFortCharacter[]
            then:
                # Respawn the player.
                Character.Damage(100.0)
