# Knowledge-neutral Verse snippet; project names and asset IDs generalized.
# id: snippet-121

# This file contains the data used to configure how orders are generated in the game.
# It also contains functions for getting and generating the next order. 

using { /Verse.org/Random }
using { /Verse.org/Simulation }
using { Persistence }

# Contains information for one line in an order.
order_line<public> := struct:
    ProductNumber<public>:int
    ProductsOrdered<public>:int
    BasePrice<public>:int

# Defines the first orders a player will get.
# This gives a predictable beginning to the game and ensures that placed orders play well with the tutorial.
# This also ensures that the player becomes accustomed to the game before random orders are generated.
FirstOrders: [][]tuple(int, int, int) =
    # Each row is an order.
    # Each column is an order line (product number, amount ordered, base price per unit).
    array{  array{(0, 1, 5)},
            array{(0, 5, 5)},
            array{(0, 10, 5), (1, 5, 10)},
            array{(0, 20, 10)}}

# Defines the base prices for each factory product.
ProductBasePrices: []int =
    # Each  column is a product.
    array{ 5, 10, 25, 100}

# Defines the scrap prices for each factory product.
ProductScrapPrices: []int =
    # Each column is a product.
    array{ 1, 1, 1, 1}

# Gets the scrap price for a product.
GetProductScrapPrice<public>(ProductNumber:int)<decides><transacts>:int=
    ProductScrapPrices[ProductNumber]

# Gets the next order for the player.
GetOrder<public>(Player:player)<decides><transacts>:[]order_line=
    OrderNumber := GetOrderNumber[Player]
    # If the order number is one of the first orders, create the order from the data.
    # Otherwise, generate the order randomly.
    if:
        OrderNumber < FirstOrders.Length
        OrderLines := FirstOrders[OrderNumber]
    then:
        for(OrderLine : OrderLines):
            order_line{ProductNumber := OrderLine(0), ProductsOrdered := OrderLine(1), BasePrice := OrderLine(2)}
    else:
        GenerateOrder[Player, OrderNumber]

# Generates an order randomly from the products that the factory can currently produce.
GenerateOrder(Player:player, OrderNumber:int)<decides><transacts>:[]order_line=
    # Determine the products available.
    var ProductsAvailable: []int = 
        for(ProductNumber := 0..3, IsFactoryMouldBuilt[Player, ProductNumber]):
            ProductNumber

    # Make a random amount of order lines.
    # Assumes that at least one product is available.
    var OrderLineCount:int = GetRandomInt(1, Min(3, ProductsAvailable.Length))

    var Order:[]order_line = array{}
    for (OrderLine := 1..OrderLineCount):
        # Select a random available product and add to order.
        RandomProductIndex := GetRandomInt(0, ProductsAvailable.Length-1)
        ProductNumber := ProductsAvailable[RandomProductIndex]
        ProductsOrdered := GetProductsOrdered[ProductNumber, OrderNumber]
        BasePrice := GetBasePrice[ProductNumber, OrderNumber]
        set Order = Order + array{order_line{ProductNumber := ProductNumber, ProductsOrdered := ProductsOrdered, BasePrice := BasePrice}}
        # Remove product from available products.
        NewProductsAvailable := ProductsAvailable.RemoveElement[RandomProductIndex]
        set ProductsAvailable = NewProductsAvailable
    Order

# Gets a random amount of products ordered.
GetProductsOrdered(ProductNumber:int, OrderNumber:int)<decides><transacts>:int=
    # This formula ensures that more products will be ordered later in the game.
    # It also ensures that products with a higher number (more expensive to produce) are ordered in a smaller quantity.
    Ceil((OrderNumber + GetRandomInt(0, 5)) * 5 / (ProductNumber + 1))

# Gets a random base price for a product.
GetBasePrice(ProductNumber:int, OrderNumber:int)<decides><transacts>:int=
    # This formula ensures that base prices are higher later in the game.
    ProductBasePrice := ProductBasePrices[ProductNumber]
    Ceil((OrderNumber + GetRandomInt(0, 20)) / 10) * ProductBasePrice
