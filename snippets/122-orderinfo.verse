# Knowledge-neutral Verse snippet; project names and asset IDs generalized.
# id: snippet-122

# This file defines order_info, the collection of persistable information stored for the current order.

using { Time }

order_info := class<final><persistable>:

    # The order lines of the order.
    OrderLineInfos:[]order_line_info = array{}

    # The time the order was placed.
    OrderTime:date_and_time = date_and_time{}

MakeOrderInfo<constructor>(Src:order_info)<transacts> := order_info:
    OrderLineInfos := Src.OrderLineInfos
    OrderTime := Src.OrderTime
