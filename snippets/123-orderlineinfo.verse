# Knowledge-neutral Verse snippet; project names and asset IDs generalized.
# id: snippet-123

# This file defines order_line_info, the collection of persistable information stored for each line of the current order.

order_line_info := class<final><persistable>:

    # The factory product number.
    ProductNumber:int = -1

    # The amount of factory products ordered.
    ProductsOrdered:int = 0

    # The amount of factory products already delivered.
    ProductsDelivered:int = 0

    # The base price offered for each factory product.
    BasePrice:int = 0

MakeOrderLineInfo<constructor>(Src:order_line_info)<transacts> := order_line_info:
    ProductNumber := Src.ProductNumber
    ProductsOrdered := Src.ProductsOrdered
    ProductsDelivered := Src.ProductsDelivered
    BasePrice := Src.BasePrice
