# Knowledge-neutral Verse snippet; project names and asset IDs generalized.
# id: snippet-124

# This file defines the order UI, which shows information for the current order of the selected player.
# The UI also allows the player to deliver products from the factory's storage.

using { /Fortnite.com/Devices }
using { /Fortnite.com/UI }
using { /UnrealEngine.com/Temporary/SpatialMath }
using { /UnrealEngine.com/Temporary/UI }
using { /Verse.org/Assets }
using { /Verse.org/Colors }
using { /Verse.org/Simulation }
using { Config }
using { Persistence }
using { Utility }

# Contains all widgets relevant for showing, closing and updating the UI.
order_ui_widgets := struct:
    Canvas<public>:canvas
    ProductTextWidgets<public>:[]text_block
    ProductTextureWidgets<public>:[]texture_block
    OrderBackgroundTextureWidget<public>:texture_block
    OrderHeaderTextWidget<public>:text_block
    OrderAmountTextWidgets<public>:[]text_block
    OrderTextureWidgets<public>:[]texture_block
    OrderStrikethroughTextureWidgets<public>:[]texture_block
    OrderPriceTextWidgets<public>:[]text_block
    OrderPriceTextureWidgets<public>:[]texture_block
    DeliverButtonWidgets<public>:[]text_button_base

order_ui := class(creative_device):

    # The maximum amount of products to deliver per click on a button.
    MaxProductsToDeliverPerClick : int = 10

    # Stores the created order UI for each player.
    var UIMap:[player]order_ui_widgets = map{}
    # Stores the order whose information will be shown on the UI for each player.
    var OrderMap:[player]order = map{}

    # When a player is added, create the order UI.
    PlayerAdded(Player:player):void=
        UIResult := CreateUI()
        if:
            set UIMap[Player] = UIResult

    # When a player is removed, remove the stored data.
    PlayerRemoved(Player:player):void=
        if:
            NewUIMap := UIMap.RemoveKey[Player]
            NewOrderMap := OrderMap.RemoveKey[Player]
        then:
            set UIMap = NewUIMap
            set OrderMap = NewOrderMap

    # Create the UI.
    CreateUI():order_ui_widgets =
        DeliverBackgroundColorWidget := color_block:
            DefaultColor := NamedColors.DarkGreen
            DefaultOpacity := 0.7
        HeaderTextWidget := text_block:
            DefaultText:= PlainText("Ship your toys to GiftGiver")
            DefaultTextColor:= NamedColors.White
            DefaultShadowOffset:= option{vector2{X := 2.0, Y := 3.0}}
            DefaultShadowColor:= NamedColors.Black
            DefaultShadowOpacity := 1.0
        Deliver0ButtonWidget := button_loud{}
        Deliver1ButtonWidget := button_loud{}
        Deliver2ButtonWidget := button_loud{}
        Deliver3ButtonWidget := button_loud{}
        Product0TextWidget := text_block:
            DefaultTextColor:= NamedColors.White
            DefaultJustification:=text_justification.InvariantRight
            DefaultShadowOffset:= option{vector2{X := 2.0, Y := 3.0}}
            DefaultShadowColor:= NamedColors.Black
            DefaultShadowOpacity := 1.0
        Product1TextWidget := text_block:
            DefaultTextColor:= NamedColors.White
            DefaultJustification:=text_justification.InvariantRight
            DefaultShadowOffset:= option{vector2{X := 2.0, Y := 3.0}}
            DefaultShadowColor:= NamedColors.Black
            DefaultShadowOpacity := 1.0
        Product2TextWidget := text_block:
            DefaultTextColor:= NamedColors.White
            DefaultJustification:=text_justification.InvariantRight
            DefaultShadowOffset:= option{vector2{X := 2.0, Y := 3.0}}
            DefaultShadowColor:= NamedColors.Black
            DefaultShadowOpacity := 1.0
        Product3TextWidget := text_block:
            DefaultTextColor:= NamedColors.White
            DefaultJustification:=text_justification.InvariantRight
            DefaultShadowOffset:= option{vector2{X := 2.0, Y := 3.0}}
            DefaultShadowColor:= NamedColors.Black
            DefaultShadowOpacity := 1.0
        Product0TextureWidget := texture_block:
            DefaultImage := Textures.UI.T_Plane
            DefaultDesiredSize:=vector2{X := 120.0, Y := 120.0}
        Product1TextureWidget := texture_block:
            DefaultImage := Textures.UI.T_TRex
            DefaultDesiredSize:=vector2{X := 120.0, Y := 120.0}
        Product2TextureWidget := texture_block:
            DefaultImage := Textures.UI.T_Castle
            DefaultDesiredSize:=vector2{X := 120.0, Y := 120.0}
        Product3TextureWidget := texture_block:
            DefaultImage := Textures.UI.T_Spaceship
            DefaultDesiredSize:=vector2{X := 120.0, Y := 120.0}
        CloseButtonWidget := button_regular:
            DefaultText:= PlainText("Close")
        OrderBackgroundTextureWidget := texture_block:
            DefaultImage := Textures.UI.T_Order
            DefaultDesiredSize :=vector2{X := 540.0, Y := 620.0}
        OrderHeaderTextWidget := text_block:
            DefaultTextColor:= NamedColors.White
            DefaultShadowOffset:= option{vector2{X := 2.0, Y := 3.0}}
            DefaultShadowColor:= NamedColors.Black
            DefaultShadowOpacity := 1.0
        Order0AmountTextWidget := text_block:
            DefaultTextColor:= NamedColors.White
            DefaultJustification:=text_justification.InvariantRight
            DefaultShadowOffset:= option{vector2{X := 2.0, Y := 3.0}}
            DefaultShadowColor:= NamedColors.Black
            DefaultShadowOpacity := 1.0
        Order1AmountTextWidget := text_block:
            DefaultTextColor:= NamedColors.White
            DefaultJustification:=text_justification.InvariantRight
            DefaultShadowOffset:= option{vector2{X := 2.0, Y := 3.0}}
            DefaultShadowColor:= NamedColors.Black
            DefaultShadowOpacity := 1.0
        Order2AmountTextWidget := text_block:
            DefaultTextColor:= NamedColors.White
            DefaultJustification:=text_justification.InvariantRight
            DefaultShadowOffset:= option{vector2{X := 2.0, Y := 3.0}}
            DefaultShadowColor:= NamedColors.Black
            DefaultShadowOpacity := 1.0
        Order3AmountTextWidget := text_block:
            DefaultTextColor:= NamedColors.White
            DefaultJustification:=text_justification.InvariantRight
            DefaultShadowOffset:= option{vector2{X := 2.0, Y := 3.0}}
            DefaultShadowColor:= NamedColors.Black
            DefaultShadowOpacity := 1.0
        Order0TextureWidget := texture_block:
            DefaultImage := Textures.UI.T_Plane
            DefaultDesiredSize:=vector2{X := 80.0, Y := 80.0}
        Order1TextureWidget := texture_block:
            DefaultImage := Textures.UI.T_TRex
            DefaultDesiredSize:=vector2{X := 80.0, Y := 80.0}
        Order2TextureWidget := texture_block:
            DefaultImage := Textures.UI.T_Castle
            DefaultDesiredSize:=vector2{X := 80.0, Y := 80.0}
        Order3TextureWidget := texture_block:
            DefaultImage := Textures.UI.T_Spaceship
            DefaultDesiredSize:=vector2{X := 80.0, Y := 80.0}
        Order0StrikethroughTextureWidget := texture_block:
            DefaultImage := Textures.UI.T_Strikethrough
            DefaultDesiredSize :=vector2{X := 256.0, Y := 32.0}
        Order1StrikethroughTextureWidget := texture_block:
            DefaultImage := Textures.UI.T_Strikethrough
            DefaultDesiredSize :=vector2{X := 256.0, Y := 32.0}
        Order2StrikethroughTextureWidget := texture_block:
            DefaultImage := Textures.UI.T_Strikethrough
            DefaultDesiredSize :=vector2{X := 256.0, Y := 32.0}
        Order3StrikethroughTextureWidget := texture_block:
            DefaultImage := Textures.UI.T_Strikethrough
            DefaultDesiredSize :=vector2{X := 256.0, Y := 32.0}
        Order0PriceTextWidget := text_block:
            DefaultTextColor:= NamedColors.White
            DefaultJustification:=text_justification.InvariantRight
            DefaultShadowOffset:= option{vector2{X := 2.0, Y := 3.0}}
            DefaultShadowColor:= NamedColors.Black
            DefaultShadowOpacity := 1.0
        Order1PriceTextWidget := text_block:
            DefaultTextColor:= NamedColors.White
            DefaultJustification:=text_justification.InvariantRight
            DefaultShadowOffset:= option{vector2{X := 2.0, Y := 3.0}}
            DefaultShadowColor:= NamedColors.Black
            DefaultShadowOpacity := 1.0
        Order2PriceTextWidget := text_block:
            DefaultTextColor:= NamedColors.White
            DefaultJustification:=text_justification.InvariantRight
            DefaultShadowOffset:= option{vector2{X := 2.0, Y := 3.0}}
            DefaultShadowColor:= NamedColors.Black
            DefaultShadowOpacity := 1.0
        Order3PriceTextWidget := text_block:
            DefaultTextColor:= NamedColors.White
            DefaultJustification:=text_justification.InvariantRight
            DefaultShadowOffset:= option{vector2{X := 2.0, Y := 3.0}}
            DefaultShadowColor:= NamedColors.Black
            DefaultShadowOpacity := 1.0
        Order0PriceTextureWidget := texture_block:
            DefaultImage := Textures.UI.T_Gold
            DefaultDesiredSize:=vector2{X := 40.0, Y := 40.0}
        Order1PriceTextureWidget := texture_block:
            DefaultImage := Textures.UI.T_Gold
            DefaultDesiredSize:=vector2{X := 40.0, Y := 40.0}
        Order2PriceTextureWidget := texture_block:
            DefaultImage := Textures.UI.T_Gold
            DefaultDesiredSize:=vector2{X := 40.0, Y := 40.0}
        Order3PriceTextureWidget := texture_block:
            DefaultImage := Textures.UI.T_Gold
            DefaultDesiredSize:=vector2{X := 40.0, Y := 40.0}

        DeliverCanvasWidget := canvas:
            Slots := array:
                canvas_slot:
                    Anchors := anchors{Minimum := vector2{X := 0.0, Y := 0.0}, Maximum := vector2{X := 1.0, Y := 1.0}}
                    Offsets := margin{Top := 0.0, Left := 0.0, Right := 0.0, Bottom := 0.0}
                    Alignment := vector2{X := 0.5, Y := 0.5}
                    Widget := DeliverBackgroundColorWidget
                canvas_slot:
                    Anchors := anchors{Minimum := vector2{X := 0.5, Y := 0.5}, Maximum := vector2{X := 0.5, Y := 0.5}}
                    Offsets := margin{Top := -390.0, Left := 0.0, Right := 0.0, Bottom := 0.0}
                    Alignment := vector2{X := 0.5, Y := 0.5}
                    SizeToContent := true
                    Widget := HeaderTextWidget
                canvas_slot:
                    Anchors := anchors{Minimum := vector2{X := 0.5, Y := 0.5}, Maximum := vector2{X := 0.5, Y := 0.5}}
                    Offsets := margin{Top := -210.0, Left := -50.0, Right := 0.0, Bottom := 0.0}
                    Alignment := vector2{X := 1.0, Y := 0.5}
                    SizeToContent := true
                    Widget := Deliver0ButtonWidget
                canvas_slot:
                    Anchors := anchors{Minimum := vector2{X := 0.5, Y := 0.5}, Maximum := vector2{X := 0.5, Y := 0.5}}
                    Offsets := margin{Top := -100.0, Left := -50.0, Right := 0.0, Bottom := 0.0}
                    Alignment := vector2{X := 1.0, Y := 0.5}
                    SizeToContent := true
                    Widget := Deliver1ButtonWidget
                canvas_slot:
                    Anchors := anchors{Minimum := vector2{X := 0.5, Y := 0.5}, Maximum := vector2{X := 0.5, Y := 0.5}}
                    Offsets := margin{Top := 10.0, Left := -50.0, Right := 0.0, Bottom := 0.0}
                    Alignment := vector2{X := 1.0, Y := 0.5}
                    SizeToContent := true
                    Widget := Deliver2ButtonWidget
                canvas_slot:
                    Anchors := anchors{Minimum := vector2{X := 0.5, Y := 0.5}, Maximum := vector2{X := 0.5, Y := 0.5}}
                    Offsets := margin{Top := 120.0, Left := -50.0, Right := 0.0, Bottom := 0.0}
                    Alignment := vector2{X := 1.0, Y := 0.5}
                    SizeToContent := true
                    Widget := Deliver3ButtonWidget                    
                canvas_slot:
                    Anchors := anchors{Minimum := vector2{X := 0.5, Y := 0.5}, Maximum := vector2{X := 0.5, Y := 0.5}}
                    Offsets := margin{Top := -210.0, Left := 100.0, Right := 0.0, Bottom := 0.0}
                    Alignment := vector2{X := 1.0, Y := 0.5}
                    SizeToContent := true
                    Widget := Product0TextWidget
                canvas_slot:
                    Anchors := anchors{Minimum := vector2{X := 0.5, Y := 0.5}, Maximum := vector2{X := 0.5, Y := 0.5}}
                    Offsets := margin{Top := -100.0, Left := 100.0, Right := 0.0, Bottom := 0.0}
                    Alignment := vector2{X := 1.0, Y := 0.5}
                    SizeToContent := true
                    Widget := Product1TextWidget
                canvas_slot:
                    Anchors := anchors{Minimum := vector2{X := 0.5, Y := 0.5}, Maximum := vector2{X := 0.5, Y := 0.5}}
                    Offsets := margin{Top := 10.0, Left := 100.0, Right := 0.0, Bottom := 0.0}
                    Alignment := vector2{X := 1.0, Y := 0.5}
                    SizeToContent := true
                    Widget := Product2TextWidget
                canvas_slot:
                    Anchors := anchors{Minimum := vector2{X := 0.5, Y := 0.5}, Maximum := vector2{X := 0.5, Y := 0.5}}
                    Offsets := margin{Top := 120.0, Left := 100.0, Right := 0.0, Bottom := 0.0}
                    Alignment := vector2{X := 1.0, Y := 0.5}
                    SizeToContent := true
                    Widget := Product3TextWidget
                canvas_slot:
                    Anchors := anchors{Minimum := vector2{X := 0.5, Y := 0.5}, Maximum := vector2{X := 0.5, Y := 0.5}}
                    Offsets := margin{Top := -210.0, Left := 180.0, Right := 0.0, Bottom := 0.0}
                    Alignment := vector2{X := 0.5, Y := 0.5}
                    Widget := Product0TextureWidget
                canvas_slot:
                    Anchors := anchors{Minimum := vector2{X := 0.5, Y := 0.5}, Maximum := vector2{X := 0.5, Y := 0.5}}
                    Offsets := margin{Top := -100.0, Left := 180.0, Right := 0.0, Bottom := 0.0}
                    Alignment := vector2{X := 0.5, Y := 0.5}
                    Widget := Product1TextureWidget
                canvas_slot:
                    Anchors := anchors{Minimum := vector2{X := 0.5, Y := 0.5}, Maximum := vector2{X := 0.5, Y := 0.5}}
                    Offsets := margin{Top := 10.0, Left := 180.0, Right := 0.0, Bottom := 0.0}
                    Alignment := vector2{X := 0.5, Y := 0.5}
                    Widget := Product2TextureWidget
                canvas_slot:
                    Anchors := anchors{Minimum := vector2{X := 0.5, Y := 0.5}, Maximum := vector2{X := 0.5, Y := 0.5}}
                    Offsets := margin{Top := 120.0, Left := 180.0, Right := 0.0, Bottom := 0.0}
                    Alignment := vector2{X := 0.5, Y := 0.5}
                    Widget := Product3TextureWidget
                canvas_slot:
                    Anchors := anchors{Minimum := vector2{X := 0.5, Y := 0.5}, Maximum := vector2{X := 0.5, Y := 0.5}}
                    Offsets := margin{Top := 350.0, Left := 0.0, Right := 0.0, Bottom := 0.0}
                    Alignment := vector2{X := 0.5, Y := 0.5}
                    SizeToContent := true
                    Widget := CloseButtonWidget

        OrderCanvasWidget := canvas:
            Slots := array:
                canvas_slot:
                    Anchors := anchors{Minimum := vector2{X := 0.5, Y := 0.5}, Maximum := vector2{X := 0.5, Y := 0.5}}
                    Offsets := margin{Top := -100.0, Left := 0.0, Right := 0.0, Bottom := 0.0}
                    Alignment := vector2{X := 0.5, Y := 0.5}
                    Widget := OrderBackgroundTextureWidget
                canvas_slot:
                    Anchors := anchors{Minimum := vector2{X := 0.5, Y := 0.5}, Maximum := vector2{X := 0.5, Y := 0.5}}
                    Offsets := margin{Top := -310.0, Left := 0.0, Right := 0.0, Bottom := 0.0}
                    Alignment := vector2{X := 0.5, Y := 0.5}
                    SizeToContent := true
                    Widget := OrderHeaderTextWidget
                canvas_slot:
                    Anchors := anchors{Minimum := vector2{X := 0.5, Y := 0.5}, Maximum := vector2{X := 0.5, Y := 0.5}}
                    Offsets := margin{Top := -230.0, Left := -40.0, Right := 0.0, Bottom := 0.0}
                    Alignment := vector2{X := 1.0, Y := 0.5}
                    SizeToContent := true
                    Widget := Order0AmountTextWidget
                canvas_slot:
                    Anchors := anchors{Minimum := vector2{X := 0.5, Y := 0.5}, Maximum := vector2{X := 0.5, Y := 0.5}}
                    Offsets := margin{Top := -130.0, Left := -40.0, Right := 0.0, Bottom := 0.0}
                    Alignment := vector2{X := 1.0, Y := 0.5}
                    SizeToContent := true
                    Widget := Order1AmountTextWidget
                canvas_slot:
                    Anchors := anchors{Minimum := vector2{X := 0.5, Y := 0.5}, Maximum := vector2{X := 0.5, Y := 0.5}}
                    Offsets := margin{Top := -30.0, Left := -40.0, Right := 0.0, Bottom := 0.0}
                    Alignment := vector2{X := 1.0, Y := 0.5}
                    SizeToContent := true
                    Widget := Order2AmountTextWidget
                canvas_slot:
                    Anchors := anchors{Minimum := vector2{X := 0.5, Y := 0.5}, Maximum := vector2{X := 0.5, Y := 0.5}}
                    Offsets := margin{Top := 70.0, Left := -40.0, Right := 0.0, Bottom := 0.0}
                    Alignment := vector2{X := 1.0, Y := 0.5}
                    SizeToContent := true
                    Widget := Order3AmountTextWidget
                canvas_slot:
                    Anchors := anchors{Minimum := vector2{X := 0.5, Y := 0.5}, Maximum := vector2{X := 0.5, Y := 0.5}}
                    Offsets := margin{Top := -230.0, Left := 10.0, Right := 0.0, Bottom := 0.0}
                    Alignment := vector2{X := 0.5, Y := 0.5}
                    Widget := Order0TextureWidget
                    ZOrder := 1
                canvas_slot:
                    Anchors := anchors{Minimum := vector2{X := 0.5, Y := 0.5}, Maximum := vector2{X := 0.5, Y := 0.5}}
                    Offsets := margin{Top := -130.0, Left := 10.0, Right := 0.0, Bottom := 0.0}
                    Alignment := vector2{X := 0.5, Y := 0.5}
                    Widget := Order1TextureWidget
                    ZOrder := 1
                canvas_slot:
                    Anchors := anchors{Minimum := vector2{X := 0.5, Y := 0.5}, Maximum := vector2{X := 0.5, Y := 0.5}}
                    Offsets := margin{Top := -30.0, Left := 10.0, Right := 0.0, Bottom := 0.0}
                    Alignment := vector2{X := 0.5, Y := 0.5}
                    Widget := Order2TextureWidget
                    ZOrder := 1
                canvas_slot:
                    Anchors := anchors{Minimum := vector2{X := 0.5, Y := 0.5}, Maximum := vector2{X := 0.5, Y := 0.5}}
                    Offsets := margin{Top := 70.0, Left := 10.0, Right := 0.0, Bottom := 0.0}
                    Alignment := vector2{X := 0.5, Y := 0.5}
                    Widget := Order3TextureWidget
                    ZOrder := 1
                canvas_slot:
                    Anchors := anchors{Minimum := vector2{X := 0.5, Y := 0.5}, Maximum := vector2{X := 0.5, Y := 0.5}}
                    Offsets := margin{Top := -230.0, Left := -190.0, Right := 0.0, Bottom := 0.0}
                    Alignment := vector2{X := 0.0, Y := 0.5}
                    Widget := Order0StrikethroughTextureWidget
                    ZOrder := 2
                canvas_slot:
                    Anchors := anchors{Minimum := vector2{X := 0.5, Y := 0.5}, Maximum := vector2{X := 0.5, Y := 0.5}}
                    Offsets := margin{Top := -130.0, Left := -190.0, Right := 0.0, Bottom := 0.0}
                    Alignment := vector2{X := 0.0, Y := 0.5}
                    Widget := Order1StrikethroughTextureWidget
                    ZOrder := 2
                canvas_slot:
                    Anchors := anchors{Minimum := vector2{X := 0.5, Y := 0.5}, Maximum := vector2{X := 0.5, Y := 0.5}}
                    Offsets := margin{Top := -30.0, Left := -190.0, Right := 0.0, Bottom := 0.0}
                    Alignment := vector2{X := 0.0, Y := 0.5}
                    Widget := Order2StrikethroughTextureWidget
                    ZOrder := 2
                canvas_slot:
                    Anchors := anchors{Minimum := vector2{X := 0.5, Y := 0.5}, Maximum := vector2{X := 0.5, Y := 0.5}}
                    Offsets := margin{Top := 70.0, Left := -190.0, Right := 0.0, Bottom := 0.0}
                    Alignment := vector2{X := 0.0, Y := 0.5}
                    Widget := Order3StrikethroughTextureWidget
                    ZOrder := 2
                canvas_slot:
                    Anchors := anchors{Minimum := vector2{X := 0.5, Y := 0.5}, Maximum := vector2{X := 0.5, Y := 0.5}}
                    Offsets := margin{Top := -230.0, Left := 150.0, Right := 0.0, Bottom := 0.0}
                    Alignment := vector2{X := 1.0, Y := 0.5}
                    SizeToContent := true
                    Widget := Order0PriceTextWidget
                canvas_slot:
                    Anchors := anchors{Minimum := vector2{X := 0.5, Y := 0.5}, Maximum := vector2{X := 0.5, Y := 0.5}}
                    Offsets := margin{Top := -130.0, Left := 150.0, Right := 0.0, Bottom := 0.0}
                    Alignment := vector2{X := 1.0, Y := 0.5}
                    SizeToContent := true
                    Widget := Order1PriceTextWidget
                canvas_slot:
                    Anchors := anchors{Minimum := vector2{X := 0.5, Y := 0.5}, Maximum := vector2{X := 0.5, Y := 0.5}}
                    Offsets := margin{Top := -30.0, Left := 150.0, Right := 0.0, Bottom := 0.0}
                    Alignment := vector2{X := 1.0, Y := 0.5}
                    SizeToContent := true
                    Widget := Order2PriceTextWidget
                canvas_slot:
                    Anchors := anchors{Minimum := vector2{X := 0.5, Y := 0.5}, Maximum := vector2{X := 0.5, Y := 0.5}}
                    Offsets := margin{Top := 70.0, Left := 150.0, Right := 0.0, Bottom := 0.0}
                    Alignment := vector2{X := 1.0, Y := 0.5}
                    SizeToContent := true
                    Widget := Order3PriceTextWidget
                canvas_slot:
                    Anchors := anchors{Minimum := vector2{X := 0.5, Y := 0.5}, Maximum := vector2{X := 0.5, Y := 0.5}}
                    Offsets := margin{Top := -230.0, Left := 180.0, Right := 0.0, Bottom := 0.0}
                    Alignment := vector2{X := 0.5, Y := 0.5}
                    Widget := Order0PriceTextureWidget
                canvas_slot:
                    Anchors := anchors{Minimum := vector2{X := 0.5, Y := 0.5}, Maximum := vector2{X := 0.5, Y := 0.5}}
                    Offsets := margin{Top := -130.0, Left := 180.0, Right := 0.0, Bottom := 0.0}
                    Alignment := vector2{X := 0.5, Y := 0.5}
                    Widget := Order1PriceTextureWidget
                canvas_slot:
                    Anchors := anchors{Minimum := vector2{X := 0.5, Y := 0.5}, Maximum := vector2{X := 0.5, Y := 0.5}}
                    Offsets := margin{Top := -30.0, Left := 180.0, Right := 0.0, Bottom := 0.0}
                    Alignment := vector2{X := 0.5, Y := 0.5}
                    Widget := Order2PriceTextureWidget
                canvas_slot:
                    Anchors := anchors{Minimum := vector2{X := 0.5, Y := 0.5}, Maximum := vector2{X := 0.5, Y := 0.5}}
                    Offsets := margin{Top := 70.0, Left := 180.0, Right := 0.0, Bottom := 0.0}
                    Alignment := vector2{X := 0.5, Y := 0.5}
                    Widget := Order3PriceTextureWidget

        Canvas := canvas:
            Slots := array:
                canvas_slot:
                    Anchors := anchors{Minimum := vector2{X := 0.0, Y := 0.0}, Maximum := vector2{X := 0.42, Y := 1.0}}
                    Offsets := margin{Top := 50.0, Left := 50.0, Right := 50.0, Bottom := 120.0}
                    Alignment := vector2{X := 0.5, Y := 0.5}
                    Widget := DeliverCanvasWidget
                canvas_slot:
                    Anchors := anchors{Minimum := vector2{X := 0.6, Y := 0.0}, Maximum := vector2{X := 1.0, Y := 1.0}}
                    Offsets := margin{Top := 50.0, Left := 50.0, Right := 50.0, Bottom := 120.0}
                    Alignment := vector2{X := 0.5, Y := 0.5}
                    Widget := OrderCanvasWidget
    
        # Subscribe callbacks to get notified when the player clicks the buttons.
        Deliver0ButtonWidget.OnClick().SubscribeWidgetMessage(OnDeliverButton, 0)
        Deliver1ButtonWidget.OnClick().SubscribeWidgetMessage(OnDeliverButton, 1)
        Deliver2ButtonWidget.OnClick().SubscribeWidgetMessage(OnDeliverButton, 2)
        Deliver3ButtonWidget.OnClick().SubscribeWidgetMessage(OnDeliverButton, 3)
        CloseButtonWidget.OnClick().Subscribe(OnCloseButton)

        return order_ui_widgets:
            Canvas := Canvas
            ProductTextWidgets := array{Product0TextWidget, Product1TextWidget, Product2TextWidget, Product3TextWidget}
            ProductTextureWidgets := array{Product0TextureWidget, Product1TextureWidget, Product2TextureWidget, Product3TextureWidget}
            OrderBackgroundTextureWidget := OrderBackgroundTextureWidget
            OrderHeaderTextWidget := OrderHeaderTextWidget
            OrderAmountTextWidgets := array{Order0AmountTextWidget, Order1AmountTextWidget, Order2AmountTextWidget, Order3AmountTextWidget}
            OrderTextureWidgets := array{Order0TextureWidget, Order1TextureWidget, Order2TextureWidget, Order3TextureWidget}
            OrderStrikethroughTextureWidgets := array{Order0StrikethroughTextureWidget, Order1StrikethroughTextureWidget, Order2StrikethroughTextureWidget, Order3StrikethroughTextureWidget}
            OrderPriceTextWidgets := array{Order0PriceTextWidget, Order1PriceTextWidget, Order2PriceTextWidget, Order3PriceTextWidget}
            OrderPriceTextureWidgets := array{Order0PriceTextureWidget, Order1PriceTextureWidget, Order2PriceTextureWidget, Order3PriceTextureWidget}
            DeliverButtonWidgets := array{Deliver0ButtonWidget, Deliver1ButtonWidget, Deliver2ButtonWidget, Deliver3ButtonWidget}

    # Show the UI to the player.
    ShowUI(Player:player, Order:order, AudioPlayer:audio_player_device):void=
        if:
            set OrderMap[Player] = Order
            UI := UIMap[Player]
            PlayerUI := GetPlayerUI[Player]
        then:
            # Play UI audio.
            AudioPlayer.Play(Player)

            # Update and show UI.
            UpdateUI(Player)
            PlayerUI.AddWidget(UI.Canvas, player_ui_slot{ZOrder := 0, InputMode := ui_input_mode.All})

    # Update the UI for the player.
    UpdateUI(Player:player):void=
        if:
            UI := UIMap[Player]
            Order := OrderMap[Player]
            SelectedPlayer := Order.MaybeSelectedPlayer?
            OrderNumber := GetOrderNumber[SelectedPlayer]
            OrderCompleted := logic{IsOrderCompleted[SelectedPlayer]}
            Product0 := GetFactoryProduct[SelectedPlayer, 0]
            Product1 := GetFactoryProduct[SelectedPlayer, 1]
            Product2 := GetFactoryProduct[SelectedPlayer, 2]
            Product3 := GetFactoryProduct[SelectedPlayer, 3]
            Product0Built := logic{IsFactoryMouldBuilt[SelectedPlayer, 0]}
            Product1Built := logic{IsFactoryMouldBuilt[SelectedPlayer, 1]}
            Product2Built := logic{IsFactoryMouldBuilt[SelectedPlayer, 2]}
            Product3Built := logic{IsFactoryMouldBuilt[SelectedPlayer, 3]}
            ProductsOrdered0 := GetProductsOrdered[SelectedPlayer, 0]
            ProductsOrdered1 := GetProductsOrdered[SelectedPlayer, 1]
            ProductsOrdered2 := GetProductsOrdered[SelectedPlayer, 2]
            ProductsOrdered3 := GetProductsOrdered[SelectedPlayer, 3]
            ProductsDelivered0 := GetProductsDelivered[SelectedPlayer, 0]
            ProductsDelivered1 := GetProductsDelivered[SelectedPlayer, 1]
            ProductsDelivered2 := GetProductsDelivered[SelectedPlayer, 2]
            ProductsDelivered3 := GetProductsDelivered[SelectedPlayer, 3]
            ProductsToDeliver0 := Min(MaxProductsToDeliverPerClick, Product0)
            ProductsToDeliver1 := Min(MaxProductsToDeliverPerClick, Product1)
            ProductsToDeliver2 := Min(MaxProductsToDeliverPerClick, Product2)
            ProductsToDeliver3 := Min(MaxProductsToDeliverPerClick, Product3)
            Deliver0Count := if (ProductsDelivered0 < ProductsOrdered0) then Min(ProductsToDeliver0, ProductsOrdered0 - ProductsDelivered0) else ProductsToDeliver0
            Deliver1Count := if (ProductsDelivered1 < ProductsOrdered1) then Min(ProductsToDeliver1, ProductsOrdered1 - ProductsDelivered1) else ProductsToDeliver1
            Deliver2Count := if (ProductsDelivered2 < ProductsOrdered2) then Min(ProductsToDeliver2, ProductsOrdered2 - ProductsDelivered2) else ProductsToDeliver2
            Deliver3Count := if (ProductsDelivered3 < ProductsOrdered3) then Min(ProductsToDeliver3, ProductsOrdered3 - ProductsDelivered3) else ProductsToDeliver3
            Deliver0Text := "{if (Deliver0Count > 0, not OrderCompleted?) then (if (ProductsDelivered0 < ProductsOrdered0) then "Deliver " else "Scrap ") + ToString(Deliver0Count) else "Deliver"}"
            Deliver1Text := "{if (Deliver1Count > 0, not OrderCompleted?) then (if (ProductsDelivered1 < ProductsOrdered1) then "Deliver " else "Scrap ") + ToString(Deliver1Count) else "Deliver"}"
            Deliver2Text := "{if (Deliver2Count > 0, not OrderCompleted?) then (if (ProductsDelivered2 < ProductsOrdered2) then "Deliver " else "Scrap ") + ToString(Deliver2Count) else "Deliver"}"
            Deliver3Text := "{if (Deliver3Count > 0, not OrderCompleted?) then (if (ProductsDelivered3 < ProductsOrdered3) then "Deliver " else "Scrap ") + ToString(Deliver3Count) else "Deliver"}"
            Deliver0Enabled := logic{Product0 > 0, not OrderCompleted?}
            Deliver1Enabled := logic{Product1 > 0, not OrderCompleted?}
            Deliver2Enabled := logic{Product2 > 0, not OrderCompleted?}
            Deliver3Enabled := logic{Product3 > 0, not OrderCompleted?}
            Price0 := GetDeliveryPrice[SelectedPlayer, 0]
            Price1 := GetDeliveryPrice[SelectedPlayer, 1]
            Price2 := GetDeliveryPrice[SelectedPlayer, 2]
            Price3 := GetDeliveryPrice[SelectedPlayer, 3]
            Product0TextWidget := UI.ProductTextWidgets[0]
            Product1TextWidget := UI.ProductTextWidgets[1]
            Product2TextWidget := UI.ProductTextWidgets[2]
            Product3TextWidget := UI.ProductTextWidgets[3]
            Product0TextureWidget := UI.ProductTextureWidgets[0]
            Product1TextureWidget := UI.ProductTextureWidgets[1]
            Product2TextureWidget := UI.ProductTextureWidgets[2]
            Product3TextureWidget := UI.ProductTextureWidgets[3]
            Deliver0ButtonWidget := UI.DeliverButtonWidgets[0]
            Deliver1ButtonWidget := UI.DeliverButtonWidgets[1]
            Deliver2ButtonWidget := UI.DeliverButtonWidgets[2]
            Deliver3ButtonWidget := UI.DeliverButtonWidgets[3]
            OrderAmount0TextWidget := UI.OrderAmountTextWidgets[0]
            OrderAmount1TextWidget := UI.OrderAmountTextWidgets[1]
            OrderAmount2TextWidget := UI.OrderAmountTextWidgets[2]
            OrderAmount3TextWidget := UI.OrderAmountTextWidgets[3]
            Order0TextureWidget := UI.OrderTextureWidgets[0]
            Order1TextureWidget := UI.OrderTextureWidgets[1]
            Order2TextureWidget := UI.OrderTextureWidgets[2]
            Order3TextureWidget := UI.OrderTextureWidgets[3]
            Order0StrikethroughTextureWidget := UI.OrderStrikethroughTextureWidgets[0]
            Order1StrikethroughTextureWidget := UI.OrderStrikethroughTextureWidgets[1]
            Order2StrikethroughTextureWidget := UI.OrderStrikethroughTextureWidgets[2]
            Order3StrikethroughTextureWidget := UI.OrderStrikethroughTextureWidgets[3]
            OrderPrice0TextWidget := UI.OrderPriceTextWidgets[0]
            OrderPrice1TextWidget := UI.OrderPriceTextWidgets[1]
            OrderPrice2TextWidget := UI.OrderPriceTextWidgets[2]
            OrderPrice3TextWidget := UI.OrderPriceTextWidgets[3]
            OrderPrice0TextureWidget := UI.OrderPriceTextureWidgets[0]
            OrderPrice1TextureWidget := UI.OrderPriceTextureWidgets[1]
            OrderPrice2TextureWidget := UI.OrderPriceTextureWidgets[2]
            OrderPrice3TextureWidget := UI.OrderPriceTextureWidgets[3]
        then:
            Deliver0ButtonWidget.SetText(PlainText(Deliver0Text))
            Deliver0ButtonWidget.SetEnabled(Deliver0Enabled)
            Deliver1ButtonWidget.SetText(PlainText(Deliver1Text))
            Deliver1ButtonWidget.SetEnabled(Deliver1Enabled)
            Deliver2ButtonWidget.SetText(PlainText(Deliver2Text))
            Deliver2ButtonWidget.SetEnabled(Deliver2Enabled)
            Deliver3ButtonWidget.SetText(PlainText(Deliver3Text))
            Deliver3ButtonWidget.SetEnabled(Deliver3Enabled)

            Product0TextWidget.SetText(CountText(Product0))
            Product1TextWidget.SetText(CountText(Product1))
            Product2TextWidget.SetText(CountText(Product2))
            Product3TextWidget.SetText(CountText(Product3))

            # Only show the product texture if the player has the ability to produce that product.
            if (Product0Built?):
                Product0TextureWidget.SetImage(Textures.UI.T_Plane)
            else:
                Product0TextureWidget.SetImage(Textures.UI.T_Hidden)

            if (Product1Built?):
                Product1TextureWidget.SetImage(Textures.UI.T_TRex)
            else:
                Product1TextureWidget.SetImage(Textures.UI.T_Hidden)

            if (Product2Built?):
                Product2TextureWidget.SetImage(Textures.UI.T_Castle)
            else:
                Product2TextureWidget.SetImage(Textures.UI.T_Hidden)

            if (Product3Built?):
                Product3TextureWidget.SetImage(Textures.UI.T_Spaceship)
            else:
                Product3TextureWidget.SetImage(Textures.UI.T_Hidden)

            # Update order.
            # Start by hiding all the order lines.
            HideOrderLine(OrderAmount0TextWidget, Order0TextureWidget, Order0StrikethroughTextureWidget, OrderPrice0TextWidget, OrderPrice0TextureWidget)
            HideOrderLine(OrderAmount1TextWidget, Order1TextureWidget, Order1StrikethroughTextureWidget, OrderPrice1TextWidget, OrderPrice1TextureWidget)
            HideOrderLine(OrderAmount2TextWidget, Order2TextureWidget, Order2StrikethroughTextureWidget, OrderPrice2TextWidget, OrderPrice2TextureWidget)
            HideOrderLine(OrderAmount3TextWidget, Order3TextureWidget, Order3StrikethroughTextureWidget, OrderPrice3TextWidget, OrderPrice3TextureWidget)

            # Hide the entire order if it is completed.
            # Otherwise, show the relevant order lines.
            if (OrderCompleted?):
                UI.OrderBackgroundTextureWidget.SetVisibility(widget_visibility.Hidden)
                UI.OrderHeaderTextWidget.SetText(PlainText(""))
            else:
                UI.OrderBackgroundTextureWidget.SetVisibility(widget_visibility.Visible)
                UI.OrderHeaderTextWidget.SetText(PlainText("GiftGiver's order #{OrderNumber}"))

                var NextOrderLineIndex:int = 0
                set NextOrderLineIndex = UpdateOrderLine(UI, ProductsOrdered0, ProductsDelivered0, Price0, Textures.UI.T_Plane, NextOrderLineIndex)
                set NextOrderLineIndex = UpdateOrderLine(UI, ProductsOrdered1, ProductsDelivered1, Price1, Textures.UI.T_TRex, NextOrderLineIndex)
                set NextOrderLineIndex = UpdateOrderLine(UI, ProductsOrdered2, ProductsDelivered2, Price2, Textures.UI.T_Castle, NextOrderLineIndex)
                set NextOrderLineIndex = UpdateOrderLine(UI, ProductsOrdered3, ProductsDelivered3, Price3, Textures.UI.T_Spaceship, NextOrderLineIndex)

    # Hide the order line on the order.
    HideOrderLine(AmountText:text_block, Texture:texture_block, Strikethrough:texture_block, PriceText:text_block, PriceTexture:texture_block):void=
        AmountText.SetText(PlainText(""))
        Texture.SetVisibility(widget_visibility.Hidden)
        Strikethrough.SetVisibility(widget_visibility.Hidden)
        PriceText.SetText(PlainText(""))
        PriceTexture.SetVisibility(widget_visibility.Hidden)

    # Update the order line on the order.
    UpdateOrderLine(UI:order_ui_widgets, Ordered:int, Delivered:int, Price:int, Texture:texture, OrderLineIndex:int):int=
        if:
            Ordered > 0
            OrderAmountTextWidget := UI.OrderAmountTextWidgets[OrderLineIndex]
            OrderTextureWidget := UI.OrderTextureWidgets[OrderLineIndex]
            OrderStrikethroughTextureWidget := UI.OrderStrikethroughTextureWidgets[OrderLineIndex]
            OrderPriceTextWidget := UI.OrderPriceTextWidgets[OrderLineIndex]
            OrderPriceTextureWidget := UI.OrderPriceTextureWidgets[OrderLineIndex]
        then:
            OrderAmountTextWidget.SetText(PlainText("{Delivered} / {Ordered}"))
            OrderTextureWidget.SetVisibility(widget_visibility.Visible)
            OrderTextureWidget.SetImage(Texture)
            # If all products of this line have been delivered, show a line crossing out
            # the product and make the price red.
            if (Delivered = Ordered):
                OrderStrikethroughTextureWidget.SetVisibility(widget_visibility.Visible)
                OrderPriceTextWidget.SetTextColor(NamedColors.Red)
            else:
                OrderStrikethroughTextureWidget.SetVisibility(widget_visibility.Hidden)
                OrderPriceTextWidget.SetTextColor(NamedColors.White)
            OrderPriceTextWidget.SetText(PlainText("x {Price}"))
            OrderPriceTextureWidget.SetVisibility(widget_visibility.Visible)
            OrderLineIndex + 1
        else:
            OrderLineIndex

    # When the player clicks the deliver button, deliver those products to the order.
    OnDeliverButton(Message:widget_message, Number:int):void=
        if:
            Order := OrderMap[Message.Player]
        then:
            Order.Deliver(Number, MaxProductsToDeliverPerClick)

            UpdateUI(Message.Player)

    # When the player clicks the close button, close the UI.
    OnCloseButton(Message:widget_message):void=
        CloseUI(Message.Player)

    # Close the UI for the player.
    CloseUI<public>(Player:player):void=
        if:
            Order := OrderMap[Player]
            UI := UIMap[Player]
            PlayerUI := GetPlayerUI[Player]
        then:
            Order.UIClosed(Player)
            PlayerUI.RemoveWidget(UI.Canvas)
