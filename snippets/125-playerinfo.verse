# Knowledge-neutral Verse snippet; project names and asset IDs generalized.
# id: snippet-125

# This file defines player_info, the collection of persistable information stored for each player.
# It also contains functions for creating, updating and returning information about a player.

using { /Verse.org/Simulation }
using { Config }
using { Time }

# Maps each player to their persisted information.
var PlayerInfoMap:weak_map(player, player_info) = map{}

# This tracks all persistable information for a player.
player_info := class<final><persistable>:

    # The version of the current player info.
    Version:int = 0

    # Latest harvest times of all candy CandyCanes.
    CandyCaneHarvestTimes:[]date_and_time =  array{date_and_time{}, date_and_time{}, date_and_time{}, date_and_time{}, date_and_time{}, date_and_time{}, date_and_time{}, date_and_time{}, date_and_time{}, date_and_time{}}
    
    # The amount of gold the player has.
    Gold:int = 0

    # The state of the factory.
    FactoryBeltInfos:[]factory_belt_info = array{factory_belt_info{}, factory_belt_info{}, factory_belt_info{}, factory_belt_info{}}
    FactoryMouldInfos:[]factory_mould_info = array{factory_mould_info{}, factory_mould_info{}, factory_mould_info{}, factory_mould_info{}}
    FactoryWrapperInfos:[]factory_wrapper_info = array{factory_wrapper_info{}, factory_wrapper_info{}, factory_wrapper_info{}, factory_wrapper_info{}}
    FactoryStorageInfo:factory_storage_info = factory_storage_info{}

    # The current order.
    OrderNumber:int = 0
    OrderInfo:order_info = order_info{}

    # Professor Flopkins: "Hmm, we should store the date and time."
    #                     "Maybe we could use something similar to line 18?"

    

# Creates a new player_info with the same values as the previous player_info.
MakePlayerInfo<constructor>(Src:player_info)<transacts> := player_info:
    Version := Src.Version
    CandyCaneHarvestTimes := Src.CandyCaneHarvestTimes
    Gold := Src.Gold
    FactoryBeltInfos := Src.FactoryBeltInfos
    FactoryMouldInfos := Src.FactoryMouldInfos
    FactoryWrapperInfos := Src.FactoryWrapperInfos
    FactoryStorageInfo := Src.FactoryStorageInfo
    OrderNumber := Src.OrderNumber
    OrderInfo := Src.OrderInfo
    # Professor Flopkins: "We need to make sure to copy over the latest gift opening time"
    #                     "It should look completely similar to the other lines in this constructor."


# Creates player information for the player if none exists. 
CheckPlayerInfoForPlayer(Player:player)<decides><transacts>:void=
    Player.IsActive[]
    if(not PlayerInfoMap[Player]):
        set PlayerInfoMap[Player] = player_info{}

# Sets the harvest time of the candy CandyCane for the player. 
SetCandyCaneHarvestTime<public>(Player:player, CandyCaneNumber:int, CandyCaneHarvestTime:date_and_time)<decides><transacts>:void=
    CheckPlayerInfoForPlayer[Player]
    SourceInfo := PlayerInfoMap[Player]
    var NewCandyCaneHarvestTimes:[]date_and_time = SourceInfo.CandyCaneHarvestTimes
    set NewCandyCaneHarvestTimes[CandyCaneNumber] = CandyCaneHarvestTime

    set PlayerInfoMap[Player] = player_info:
        CandyCaneHarvestTimes := NewCandyCaneHarvestTimes
        MakePlayerInfo<constructor>(SourceInfo)

# Gets the harvest time of the candy CandyCane for the player. 
GetCandyCaneHarvestTime<public>(Player:player, CandyCaneNumber:int)<decides><transacts>:date_and_time=
    CheckPlayerInfoForPlayer[Player]
    PlayerInfoMap[Player].CandyCaneHarvestTimes[CandyCaneNumber]

# Grant the amount of gold to the player.
GrantGold<public>(Player:player, Amount:int)<decides><transacts>:void=
    CheckPlayerInfoForPlayer[Player]
    SourceInfo := PlayerInfoMap[Player]

    set PlayerInfoMap[Player] = player_info:
        Gold := SourceInfo.Gold + Amount
        MakePlayerInfo<constructor>(SourceInfo)

# Spend the amount of gold for the player.
SpendGold<public>(Player:player, Amount:int)<decides><transacts>:void=
    CheckPlayerInfoForPlayer[Player]
    SourceInfo := PlayerInfoMap[Player]

    # Fails if player does not have enough gold.
    SourceInfo.Gold >= Amount

    set PlayerInfoMap[Player] = player_info:
        Gold := SourceInfo.Gold - Amount
        MakePlayerInfo<constructor>(SourceInfo)

# Gets the amount of gold the player has.
GetGold<public>(Player:player)<decides><transacts>:int=
    CheckPlayerInfoForPlayer[Player]
    PlayerInfoMap[Player].Gold

# Builds the factory belt for the player.
BuildFactoryBelt<public>(Player:player, FactoryBeltNumber:int)<decides><transacts>:void=
    CheckPlayerInfoForPlayer[Player]
    SourceInfo := PlayerInfoMap[Player]
    SourceFactoryBeltInfo := SourceInfo.FactoryBeltInfos[FactoryBeltNumber]

    # Fails if belt is already built.
    SourceFactoryBeltInfo.Built = false

    var NewFactoryBeltInfos:[]factory_belt_info = SourceInfo.FactoryBeltInfos
    set NewFactoryBeltInfos[FactoryBeltNumber] = factory_belt_info:
        Built := true
        MakeFactoryBeltInfo<constructor>(SourceFactoryBeltInfo)

    set PlayerInfoMap[Player] = player_info:
        FactoryBeltInfos := NewFactoryBeltInfos
        MakePlayerInfo<constructor>(SourceInfo)

# Succeeds if the factory belt is built for the player.
IsFactoryBeltBuilt<public>(Player:player, FactoryBeltNumber:int)<decides><transacts>:void=
    CheckPlayerInfoForPlayer[Player]
    PlayerInfoMap[Player].FactoryBeltInfos[FactoryBeltNumber].Built?

# Upgrades the level of the factory belt for the player.
UpgradeFactoryBeltLevel<public>(Player:player, FactoryBeltNumber:int)<decides><transacts>:void=
    CheckPlayerInfoForPlayer[Player]
    SourceInfo := PlayerInfoMap[Player]
    var NewFactoryBeltInfos:[]factory_belt_info = SourceInfo.FactoryBeltInfos
    SourceFactoryBeltInfo := SourceInfo.FactoryBeltInfos[FactoryBeltNumber]
    set NewFactoryBeltInfos[FactoryBeltNumber] = factory_belt_info:
        Level := SourceFactoryBeltInfo.Level + 1
        MakeFactoryBeltInfo<constructor>(SourceFactoryBeltInfo)

    set PlayerInfoMap[Player] = player_info:
        FactoryBeltInfos := NewFactoryBeltInfos
        MakePlayerInfo<constructor>(SourceInfo)

# Gets the level of the factory belt for the player. 
GetFactoryBeltLevel<public>(Player:player, FactoryBeltNumber:int)<decides><transacts>:int=
    CheckPlayerInfoForPlayer[Player]
    PlayerInfoMap[Player].FactoryBeltInfos[FactoryBeltNumber].Level

# Builds the factory mould for the player.
# Also, activates the factory mould.
BuildFactoryMould<public>(Player:player, FactoryMouldNumber:int)<decides><transacts>:void=
    SourceInfo := PlayerInfoMap[Player]
    SourceFactoryMouldInfo := SourceInfo.FactoryMouldInfos[FactoryMouldNumber]

    # Fails if mould is already built.
    SourceFactoryMouldInfo.Built = false

    var NewFactoryMouldInfos:[]factory_mould_info = SourceInfo.FactoryMouldInfos
    set NewFactoryMouldInfos[FactoryMouldNumber] = factory_mould_info:
        Built := true
        Active := true
        MakeFactoryMouldInfo<constructor>(SourceFactoryMouldInfo)

    set PlayerInfoMap[Player] = player_info:
        FactoryMouldInfos := NewFactoryMouldInfos
        MakePlayerInfo<constructor>(SourceInfo)

# Succeeds if the factory mould is built for the player.
IsFactoryMouldBuilt<public>(Player:player, FactoryMouldNumber:int)<decides><transacts>:void=
    CheckPlayerInfoForPlayer[Player]
    PlayerInfoMap[Player].FactoryMouldInfos[FactoryMouldNumber].Built?

# Sets the activation of the factory mould for the player.
SetFactoryMouldActive<public>(Player:player, FactoryMouldNumber:int, Active:logic)<decides><transacts>:void=
    CheckPlayerInfoForPlayer[Player]
    SourceInfo := PlayerInfoMap[Player]
    SourceFactoryMouldInfo := SourceInfo.FactoryMouldInfos[FactoryMouldNumber]

    # Fails if mould is not built.
    SourceFactoryMouldInfo.Built = true

    var NewFactoryMouldInfos:[]factory_mould_info = SourceInfo.FactoryMouldInfos
    set NewFactoryMouldInfos[FactoryMouldNumber] = factory_mould_info:
        Active := Active
        MakeFactoryMouldInfo<constructor>(SourceFactoryMouldInfo)

    set PlayerInfoMap[Player] = player_info:
        FactoryMouldInfos := NewFactoryMouldInfos
        MakePlayerInfo<constructor>(SourceInfo)

# Succeeds if the factory mould is active for the player.
IsFactoryMouldActive<public>(Player:player, FactoryMouldNumber:int)<decides><transacts>:void=
    CheckPlayerInfoForPlayer[Player]
    PlayerInfoMap[Player].FactoryMouldInfos[FactoryMouldNumber].Active?
    
# Upgrades the level of the factory mould for the player.
UpgradeFactoryMouldLevel<public>(Player:player, FactoryMouldNumber:int)<decides><transacts>:void=
    CheckPlayerInfoForPlayer[Player]
    SourceInfo := PlayerInfoMap[Player]
    var NewFactoryMouldInfos:[]factory_mould_info = SourceInfo.FactoryMouldInfos
    SourceFactoryMouldInfo := SourceInfo.FactoryMouldInfos[FactoryMouldNumber]
    set NewFactoryMouldInfos[FactoryMouldNumber] = factory_mould_info:
        Level := SourceFactoryMouldInfo.Level + 1
        MakeFactoryMouldInfo<constructor>(SourceFactoryMouldInfo)

    set PlayerInfoMap[Player] = player_info:
        FactoryMouldInfos := NewFactoryMouldInfos
        MakePlayerInfo<constructor>(SourceInfo)

# Gets the level of the factory mould for the player. 
GetFactoryMouldLevel<public>(Player:player, FactoryMouldNumber:int)<decides><transacts>:int=
    CheckPlayerInfoForPlayer[Player]
    PlayerInfoMap[Player].FactoryMouldInfos[FactoryMouldNumber].Level

# Sets the latest production time of the factory mould for the player.
SetFactoryMouldLatestProductionTime<public>(Player:player, FactoryMouldNumber:int, LatestProductionTime:date_and_time)<decides><transacts>:void=
    CheckPlayerInfoForPlayer[Player]
    SourceInfo := PlayerInfoMap[Player]
    var NewFactoryMouldInfos:[]factory_mould_info = SourceInfo.FactoryMouldInfos
    SourceFactoryMouldInfo := SourceInfo.FactoryMouldInfos[FactoryMouldNumber]
    set NewFactoryMouldInfos[FactoryMouldNumber] = factory_mould_info:
        LatestProductionTime := LatestProductionTime
        MakeFactoryMouldInfo<constructor>(SourceFactoryMouldInfo)

    set PlayerInfoMap[Player] = player_info:
        FactoryMouldInfos := NewFactoryMouldInfos
        MakePlayerInfo<constructor>(SourceInfo)

# Gets the latest production time of the factory mould for the player.
GetFactoryMouldLatestProductionTime<public>(Player:player, FactoryMouldNumber:int)<decides><transacts>:date_and_time=
    CheckPlayerInfoForPlayer[Player]
    PlayerInfoMap[Player].FactoryMouldInfos[FactoryMouldNumber].LatestProductionTime

# Builds the factory wrapper for the player.
BuildFactoryWrapper<public>(Player:player, FactoryWrapperNumber:int)<decides><transacts>:void=
    CheckPlayerInfoForPlayer[Player]
    SourceInfo := PlayerInfoMap[Player]
    SourceFactoryWrapperInfo := SourceInfo.FactoryWrapperInfos[FactoryWrapperNumber]

    # Fails if wrapper is already built.
    SourceFactoryWrapperInfo.Built = false

    var NewFactoryWrapperInfos:[]factory_wrapper_info = SourceInfo.FactoryWrapperInfos
    set NewFactoryWrapperInfos[FactoryWrapperNumber] = factory_wrapper_info:
        Built := true
        MakeFactoryWrapperInfo<constructor>(SourceFactoryWrapperInfo)

    set PlayerInfoMap[Player] = player_info:
        FactoryWrapperInfos := NewFactoryWrapperInfos
        MakePlayerInfo<constructor>(SourceInfo)

# Succeeds if the factory wrapper is built for the player.
IsFactoryWrapperBuilt<public>(Player:player, FactoryWrapperNumber:int)<decides><transacts>:void=
    CheckPlayerInfoForPlayer[Player]
    PlayerInfoMap[Player].FactoryWrapperInfos[FactoryWrapperNumber].Built?
    
# Upgrades the level of the factory wrapper for the player.
UpgradeFactoryWrapperLevel<public>(Player:player, FactoryWrapperNumber:int)<decides><transacts>:void=
    CheckPlayerInfoForPlayer[Player]
    SourceInfo := PlayerInfoMap[Player]
    var NewFactoryWrapperInfos:[]factory_wrapper_info = SourceInfo.FactoryWrapperInfos
    SourceFactoryWrapperInfo := SourceInfo.FactoryWrapperInfos[FactoryWrapperNumber]
    set NewFactoryWrapperInfos[FactoryWrapperNumber] = factory_wrapper_info:
        Level := SourceFactoryWrapperInfo.Level + 1
        MakeFactoryWrapperInfo<constructor>(SourceFactoryWrapperInfo)

    set PlayerInfoMap[Player] = player_info:
        FactoryWrapperInfos := NewFactoryWrapperInfos
        MakePlayerInfo<constructor>(SourceInfo)

# Gets the level of the factory wrapper for the player. 
GetFactoryWrapperLevel<public>(Player:player, FactoryWrapperNumber:int)<decides><transacts>:int=
    CheckPlayerInfoForPlayer[Player]
    PlayerInfoMap[Player].FactoryWrapperInfos[FactoryWrapperNumber].Level

# Builds the factory storage for the player.
BuildFactoryStorage<public>(Player:player)<decides><transacts>:void=
    CheckPlayerInfoForPlayer[Player]
    SourceInfo := PlayerInfoMap[Player]
    SourceFactoryStorageInfo := SourceInfo.FactoryStorageInfo

    # Fails if storage is already built.
    SourceFactoryStorageInfo.Built = false

    set PlayerInfoMap[Player] = player_info:
        FactoryStorageInfo := factory_storage_info:
            Built := true
            MakeFactoryStorageInfo<constructor>(SourceFactoryStorageInfo)
        MakePlayerInfo<constructor>(SourceInfo)

# Succeeds if the factory storage is built for the player.
IsFactoryStorageBuilt<public>(Player:player)<decides><transacts>:void=
    CheckPlayerInfoForPlayer[Player]
    PlayerInfoMap[Player].FactoryStorageInfo.Built?

# Succeeds if there is any available capacity left in the factory storage for the player.
IsRoomInFactoryStorage<public>(Player:player)<decides><transacts>:void=
    CheckPlayerInfoForPlayer[Player]
    IsFactoryStorageBuilt[Player]
    GetFactoryStorageCapacity[Player] > GetAllFactoryProducts[Player]

# Succeeds if there are products stored in the factory storage for the player.
IsAnyProductInFactoryStorage<public>(Player:player)<decides><transacts>:void=
    CheckPlayerInfoForPlayer[Player]
    GetAllFactoryProducts[Player] > 0

# Stores the amount of factory products in the factory storage for the player.
StoreFactoryProduct<public>(Player:player, ProductNumber:int, Amount:int)<decides><transacts>:void=
    CheckPlayerInfoForPlayer[Player]
    SourceInfo := PlayerInfoMap[Player]
    SourceFactoryStorageInfo := SourceInfo.FactoryStorageInfo

    # Fails if there is not room.
    GetAllFactoryProducts[Player] + Amount <= GetFactoryStorageCapacity[Player]

    SourceProductsInStorage := SourceFactoryStorageInfo.ProductsInStorage[ProductNumber]

    var NewProductsInStorage:[]int = SourceFactoryStorageInfo.ProductsInStorage
    set NewProductsInStorage[ProductNumber] = SourceProductsInStorage + Amount

    set PlayerInfoMap[Player] = player_info:
        FactoryStorageInfo := factory_storage_info:
            ProductsInStorage := NewProductsInStorage
            MakeFactoryStorageInfo<constructor>(SourceFactoryStorageInfo)
        MakePlayerInfo<constructor>(SourceInfo)

# Gets the factory product amount in the factory storage for the player.
GetFactoryProduct<public>(Player:player, ProductNumber:int)<decides><transacts>:int=
    CheckPlayerInfoForPlayer[Player]
    PlayerInfoMap[Player].FactoryStorageInfo.ProductsInStorage[ProductNumber]

# Gets the total amount of factory products in the factory storage for the player.
GetAllFactoryProducts<public>(Player:player)<decides><transacts>:int=
    CheckPlayerInfoForPlayer[Player]
    SourceInfo := PlayerInfoMap[Player]
    SourceFactoryStorageInfo := SourceInfo.FactoryStorageInfo

    var TotalProductsInStorage:int = 0
    for(ProductsInStorage : SourceFactoryStorageInfo.ProductsInStorage):
        set TotalProductsInStorage += ProductsInStorage

    TotalProductsInStorage
    
# Withdraws the amount of factory products from the factory storage for the player.
WithdrawFactoryProduct<public>(Player:player, ProductNumber:int, Amount:int)<decides><transacts>:void=
    CheckPlayerInfoForPlayer[Player]
    SourceInfo := PlayerInfoMap[Player]
    SourceFactoryStorageInfo := SourceInfo.FactoryStorageInfo
    SourceProductsInStorage := SourceFactoryStorageInfo.ProductsInStorage[ProductNumber]

    # Fails if there is not enough of a product.
    SourceProductsInStorage >= Amount

    var NewProductsInStorage:[]int = SourceFactoryStorageInfo.ProductsInStorage
    set NewProductsInStorage[ProductNumber] = SourceProductsInStorage - Amount

    set PlayerInfoMap[Player] = player_info:
        FactoryStorageInfo := factory_storage_info:
            ProductsInStorage := NewProductsInStorage
            MakeFactoryStorageInfo<constructor>(SourceFactoryStorageInfo)
        MakePlayerInfo<constructor>(SourceInfo)

# Upgrades the level of the factory storage for the player.
UpgradeFactoryStorageLevel<public>(Player:player)<decides><transacts>:void=
    CheckPlayerInfoForPlayer[Player]
    SourceInfo := PlayerInfoMap[Player]
    SourceFactoryStorageInfo := SourceInfo.FactoryStorageInfo

    set PlayerInfoMap[Player] = player_info:
        FactoryStorageInfo := factory_storage_info:
            Level := SourceFactoryStorageInfo.Level + 1
            MakeFactoryStorageInfo<constructor>(SourceFactoryStorageInfo)
        MakePlayerInfo<constructor>(SourceInfo)

# Gets the level of the factory storage for the player. 
GetFactoryStorageLevel<public>(Player:player)<decides><transacts>:int=
    CheckPlayerInfoForPlayer[Player]
    PlayerInfoMap[Player].FactoryStorageInfo.Level

# Gets the current order number for the player. 
GetOrderNumber<public>(Player:player)<decides><transacts>:int=
    CheckPlayerInfoForPlayer[Player]
    PlayerInfoMap[Player].OrderNumber

GetOrderLineInfoIndex(OrderInfo:order_info, ProductNumber:int)<transacts>:int=
    var OrderLineInfoIndex:int = -1
    for:
        Index := 0..OrderInfo.OrderLineInfos.Length-1
        OrderInfo.OrderLineInfos[Index].ProductNumber = ProductNumber
    do:
        set OrderLineInfoIndex = Index
    OrderLineInfoIndex

# Delivers the amount of factory products from the factory storage for the player.
# The player is granted the amount of gold dictated by the current order. 
DeliverProduct<public>(Player:player, ProductNumber:int, Amount:int)<decides><transacts>:void=
    CheckPlayerInfoForPlayer[Player]
    SourceInfo := PlayerInfoMap[Player]
    SourceOrderInfo := SourceInfo.OrderInfo

    var GoldEarnings:int = 0
    var ProductsNotInOrderDelivered:int = Amount

    var NewOrderLineInfos:[]order_line_info = SourceOrderInfo.OrderLineInfos
    if:
        OrderLineInfoIndex := GetOrderLineInfoIndex(SourceOrderInfo, ProductNumber)
        SourceOrderLineInfo := SourceOrderInfo.OrderLineInfos[OrderLineInfoIndex]
        ProductsLeftInOrder := SourceOrderLineInfo.ProductsOrdered - SourceOrderLineInfo.ProductsDelivered
        ProductsInOrderDelivered := Min(Amount, ProductsLeftInOrder)
        set ProductsNotInOrderDelivered -= ProductsInOrderDelivered

        # Apply multiplier from factory wrapper to base price.
        Multiplier := GetFactoryWrapperMultiplier[ProductNumber, Player]
        set GoldEarnings += Int[ProductsInOrderDelivered * SourceOrderLineInfo.BasePrice * Multiplier]

        set NewOrderLineInfos[OrderLineInfoIndex] = order_line_info:
            ProductsDelivered := SourceOrderLineInfo.ProductsDelivered + ProductsInOrderDelivered
            MakeOrderLineInfo<constructor>(SourceOrderLineInfo)

    set PlayerInfoMap[Player] = player_info:
        OrderInfo := order_info:
            OrderLineInfos := NewOrderLineInfos
            MakeOrderInfo<constructor>(SourceOrderInfo)
        MakePlayerInfo<constructor>(SourceInfo)

    # Any products delivered that are not part of the order will only earn the player a scrap price.
    FactoryProductScrapPrice := GetProductScrapPrice[ProductNumber]
    set GoldEarnings += ProductsNotInOrderDelivered * FactoryProductScrapPrice

    GrantGold[Player, GoldEarnings]

# Gets the delivery price for the factory product in the current order for the player.
GetDeliveryPrice<public>(Player:player, ProductNumber:int)<decides><transacts>:int=
    CheckPlayerInfoForPlayer[Player]
    OrderInfo := PlayerInfoMap[Player].OrderInfo

    OrderLineInfoIndex := GetOrderLineInfoIndex(OrderInfo, ProductNumber)

    if:
        OrderLineInfo := OrderInfo.OrderLineInfos[OrderLineInfoIndex]
        OrderLineInfo.ProductsDelivered < OrderLineInfo.ProductsOrdered
        Multiplier := GetFactoryWrapperMultiplier[ProductNumber, Player]
    then:
        # Apply multiplier from factory wrapper to base price.
        Int[OrderLineInfo.BasePrice * Multiplier]
    else:
        # If the order asks for no more of this factory product, list the scrap price. 
        GetProductScrapPrice[ProductNumber]

# Gets the time when the current order was placed for the player.
GetOrderTime<public>(Player:player)<decides><transacts>:date_and_time=
    CheckPlayerInfoForPlayer[Player]
    PlayerInfoMap[Player].OrderInfo.OrderTime

# Gets the amount of factory products ordered in the current order for the player.
GetProductsOrdered<public>(Player:player, ProductNumber:int)<decides><transacts>:int=
    CheckPlayerInfoForPlayer[Player]
    OrderInfo := PlayerInfoMap[Player].OrderInfo

    OrderLineInfoIndex := GetOrderLineInfoIndex(OrderInfo, ProductNumber)

    OrderInfo.OrderLineInfos[OrderLineInfoIndex].ProductsOrdered or 0
    
# Gets the amount of factory products already delivered to the current order for the player.
GetProductsDelivered<public>(Player:player, ProductNumber:int)<decides><transacts>:int=
    CheckPlayerInfoForPlayer[Player]
    OrderInfo := PlayerInfoMap[Player].OrderInfo

    OrderLineInfoIndex := GetOrderLineInfoIndex(OrderInfo, ProductNumber)

    OrderInfo.OrderLineInfos[OrderLineInfoIndex].ProductsDelivered or 0

# Succeeds if the current order is completed for the player.
IsOrderCompleted<public>(Player:player)<decides><transacts>:void=
    CheckPlayerInfoForPlayer[Player]
    OrderInfo := PlayerInfoMap[Player].OrderInfo

    for:
        OrderLineInfo : OrderInfo.OrderLineInfos
        OrderLineInfo.ProductsDelivered < OrderLineInfo.ProductsOrdered
    do:
        false?

# Places a new order for the player.
PlaceNewOrder<public>(Player:player, OrderTime:date_and_time)<decides><transacts>:void=
    CheckPlayerInfoForPlayer[Player]
    SourceInfo := PlayerInfoMap[Player]
    OrderNumber := SourceInfo.OrderNumber

    # Create an order based on the player's progression and the products that can be produced.
    Order := GetOrder[Player]
    
    var OrderLineInfos:[]order_line_info = array{}
    for(OrderLine : Order):
        OrderLineInfo :=  order_line_info:
            ProductNumber := OrderLine.ProductNumber
            ProductsOrdered := OrderLine.ProductsOrdered
            BasePrice := OrderLine.BasePrice
        set OrderLineInfos += array{OrderLineInfo}

    OrderInfo := order_info:
        OrderLineInfos := OrderLineInfos
        OrderTime := OrderTime

    set PlayerInfoMap[Player] = player_info:
            OrderNumber := OrderNumber + 1
            OrderInfo := OrderInfo
            MakePlayerInfo<constructor>(SourceInfo)

# Sets the gift opening time for the player.
SetGiftOpeningTime<public>(Player:player, GiftOpeningTime:date_and_time)<decides><transacts>:void=
    CheckPlayerInfoForPlayer[Player]
    SourceInfo := PlayerInfoMap[Player]

    # Professor Flopkins: "Oh no, someone forgot to set the gift opening time!"
    #                     "We should make a new player_info and set the gift opening time."
    #                     "Check out SetCandyCaneHarvestTime or GrantGold to get an idea."



# Gets the gift opening time for the player. 
GetGiftOpeningTime<public>(Player:player)<decides><transacts>:date_and_time=
    CheckPlayerInfoForPlayer[Player]

    # Professor Flopkins: "Yikes, looks like the returned date and time has nothing to do with the"
    #                     "persistable player data."
    date_and_time{}
    
    # Professor Flopkins: "We should retrieve it from the player info map instead!"
    #                     "Look into GetCandyCaneHarvestTime or GetGold for inspiration."
