# Knowledge-neutral Verse snippet; project names and asset IDs generalized.
# id: snippet-126

# This file defines the real_time device that can be used for obtaining the current
# date and time in Coordinated Universal Time (UTC).
# It supports adding an offset to the current time for debugging and testing purposes.
# It also controls the hands of the clock.

using { /Fortnite.com/Devices }
using { /UnrealEngine.com/Temporary/SpatialMath }
using { /Verse.org/Simulation }

real_time<public> := class(creative_device):

    # The prop representing the clock.
    # Used for calculating the rotations of the clock hands.
    @editable
    ClockProp:creative_prop = creative_prop{}

    # The hour hand prop.
    @editable
    HourHandProp:creative_prop = creative_prop{}

    # The minute hand prop.
    @editable
    MinuteHandProp:creative_prop = creative_prop{}

    # The second hand prop.
    @editable
    SecondHandProp:creative_prop = creative_prop{}

    # The tracker device that keeps track of the current debug time offset.
    # It is persistent to ensure that the offset is kept between games.
    @editable
    DebugTimeOffsetSecondsTracker:tracker_device = tracker_device{}

    # Runs when the device is started in a running game.
    OnBegin<override>()<suspends>:void=
        # Continuously update the hands of the clock with the current time.
        loop:
            ClockTransform := ClockProp.GetTransform()
            Now := GetDateAndTime()
            RotationAxis := vector3{X := 1.0, Y := 0.0, Z := 0.0}

            # Calculate the angle in radians.
            HourHandAngle := 1.0 * Now.Hours / 12.0 * (2.0 * PiFloat)
            LocalHourHandRotation := MakeRotation(RotationAxis, HourHandAngle)
            # Make sure that the hand rotates correctly no matter how the clock is oriented.
            GlobalHourHandRotation := LocalHourHandRotation.RotateBy(ClockTransform.Rotation)

            MinuteHandAngle := 1.0 * Now.Minutes / 60.0 * (2.0 * PiFloat)
            LocalMinuteHandRotation := MakeRotation(RotationAxis, MinuteHandAngle)
            GlobalMinuteHandRotation := LocalMinuteHandRotation.RotateBy(ClockTransform.Rotation)

            SecondHandAngle := Now.Seconds / 60.0 * (2.0 * PiFloat)
            LocalSecondHandRotation := MakeRotation(RotationAxis, SecondHandAngle)
            GlobalSecondHandRotation := LocalSecondHandRotation.RotateBy(ClockTransform.Rotation)

            if:
                HourHandProp.TeleportTo[HourHandProp.GetTransform().Translation, GlobalHourHandRotation]
                MinuteHandProp.TeleportTo[MinuteHandProp.GetTransform().Translation, GlobalMinuteHandRotation]
                SecondHandProp.TeleportTo[SecondHandProp.GetTransform().Translation, GlobalSecondHandRotation]

            Sleep(1.0)

    # Returns the current date and time.
    GetDateAndTime<public>()<transacts>:date_and_time=
        # Convert from Unix epoch (1/1/1970 00:00:00) to the beginning of the Common Era (1/1/1 00:00:00) since that is the epoch MakeDateAndTime expects.
        Seconds := GetSecondsSinceEpoch() + 62135596800.0 + 1.0 * DebugTimeOffsetSecondsTracker.GetValue()
        return MakeDateAndTime(Seconds)

    # Adds a number of days to the debug time offset.
    AddDebugTimeOffsetDays<public>(OffsetDays:int):void=
        CurrentDebugTimeOffsetSeconds := DebugTimeOffsetSecondsTracker.GetValue()
        DebugTimeOffsetSecondsTracker.SetValue(CurrentDebugTimeOffsetSeconds + OffsetDays * 86400)
        
    # Adds a number of hours to the debug time offset.
    AddDebugTimeOffsetHours<public>(OffsetHours:int):void=
        CurrentDebugTimeOffsetSeconds := DebugTimeOffsetSecondsTracker.GetValue()
        DebugTimeOffsetSecondsTracker.SetValue(CurrentDebugTimeOffsetSeconds + OffsetHours * 3600)
        
    # Adds a number of minutes to the debug time offset.
    AddDebugTimeOffsetMinutes<public>(OffsetMinutes:int):void=
        CurrentDebugTimeOffsetSeconds := DebugTimeOffsetSecondsTracker.GetValue()
        DebugTimeOffsetSecondsTracker.SetValue(CurrentDebugTimeOffsetSeconds + OffsetMinutes * 60)
        
    # Adds a number of seconds to the debug time offset.
    AddDebugTimeOffsetSeconds<public>(OffsetSeconds:int):void=
        CurrentDebugTimeOffsetSeconds := DebugTimeOffsetSecondsTracker.GetValue()
        DebugTimeOffsetSecondsTracker.SetValue(CurrentDebugTimeOffsetSeconds + OffsetSeconds)
