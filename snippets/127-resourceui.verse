# Knowledge-neutral Verse snippet; project names and asset IDs generalized.
# id: snippet-127

# This file defines the resource UI, which shows how many resources the selected player has.
# Whenever the UI is updated, it handles the animation of the current resource count. 

using { /Fortnite.com/Devices }
using { /Fortnite.com/UI }
using { /UnrealEngine.com/Temporary/SpatialMath }
using { /UnrealEngine.com/Temporary/UI }
using { /Verse.org/Colors }
using { /Verse.org/Simulation }
using { Persistence }
using { Utility }

# Contains all widgets relevant for showing, closing and updating the UI.
resource_ui_widgets := struct:
    Canvas<public>:canvas
    ResourceTextWidget<public>:text_block
    ResourceTextureWidget<public>:texture_block
    StopEvent<public>:event()

resource_ui := class(creative_device):

    # Audio that plays when the resource is counting.  
    @editable
    ResourceCountAudioPlayer:audio_player_device = audio_player_device{}

    # Stores the created resource UI for each player.
    var UIMap:[player]resource_ui_widgets = map{}
    # Stores the selected player whose resources are shown on the UI for each player.
    var SelectedPlayerMap:[player]player = map{}

    # When a player is added, create the resource UI.
    PlayerAdded(Player:player):void=
        UIResult := CreateUI()
        if:
            PlayerUI := GetPlayerUI[Player]
            set UIMap[Player] = UIResult

    # When a player is removed, remove the stored data.
    PlayerRemoved(Player:player):void=
        if:
            NewUIMap := UIMap.RemoveKey[Player]
            NewSelectedPlayerMap := SelectedPlayerMap.RemoveKey[Player]
        then:
            set UIMap = NewUIMap
            set SelectedPlayerMap = NewSelectedPlayerMap

    # Create the UI.
    CreateUI():resource_ui_widgets=
        ResourceTextWidget := text_block:
            DefaultTextColor:= NamedColors.White
            DefaultJustification:=text_justification.InvariantRight
            DefaultShadowOffset:= option{vector2{X := 2.0, Y := 3.0}}
            DefaultShadowColor:= NamedColors.Black
            DefaultShadowOpacity := 1.0
        ResourceTextureWidget := texture_block:
            DefaultImage := Textures.UI.T_Gold
            DefaultDesiredSize:=vector2{X := 40.0, Y := 40.0}
        Canvas := canvas:
            Slots := array:
                canvas_slot:
                    Anchors := anchors{Minimum := vector2{X := 0.5, Y := 1.0}, Maximum := vector2{X := 0.5, Y := 1.0}}
                    Offsets := margin{Top := -50.0, Left := 20.0, Right := 0.0, Bottom := 0.0}
                    Alignment := vector2{X := 1.0, Y := 1.0}
                    SizeToContent := true
                    Widget := ResourceTextWidget
                canvas_slot:
                    Anchors := anchors{Minimum := vector2{X := 0.5, Y := 1.0}, Maximum := vector2{X := 0.5, Y := 1.0}}
                    Offsets := margin{Top := -72.0, Left := 55.0, Right := 0.0, Bottom := 0.0}
                    Alignment := vector2{X := 0.5, Y := 0.5}
                    Widget := ResourceTextureWidget

        return resource_ui_widgets:
            Canvas := Canvas
            ResourceTextWidget := ResourceTextWidget
            ResourceTextureWidget := ResourceTextureWidget
            StopEvent := event(){}

    # Show the UI to the player. The UI will show the resources for the selected player.
    ShowUI(Player:player, SelectedPlayer:player):void=
        if:
            set SelectedPlayerMap[Player] = SelectedPlayer
            UI := UIMap[Player]
            PlayerUI := GetPlayerUI[Player]
        then:
            UpdateUI(Player)
            PlayerUI.AddWidget(UI.Canvas)

    # Update the UI for the player.
    UpdateUI(Player:player):void=
        if:
            UI := UIMap[Player]
            SelectedPlayer := SelectedPlayerMap[Player]
            NewCount := GetGold[SelectedPlayer]
        then:
            # Stop any previous count animations.
            UI.StopEvent.Signal()
            # Determine the currently showing resource count.
            CurrentText := UI.ResourceTextWidget.GetText()
            if:
                CurrentlyShowingCount := ParseInt[CurrentText]
            then:
                # If the new count is different, start a count animation. 
                if (CurrentlyShowingCount <> NewCount):
                    spawn:
                        AnimateCount(Player, UI.ResourceTextWidget, UI.ResourceTextureWidget, UI.StopEvent, CurrentlyShowingCount, NewCount)
            else:
                # If no count was showing, just set the text directly.
                UI.ResourceTextWidget.SetText(CountText(NewCount))
    
    # Close the UI for the player.
    CloseUI<public>(Player:player):void=
        if:
            UI := UIMap[Player]
            PlayerUI := GetPlayerUI[Player]
        then:
            # Stop any count animations.
            UI.StopEvent.Signal()
            PlayerUI.RemoveWidget(UI.Canvas)

    # Animate the given resource count text and texture.
    AnimateCount(Player:player, Text:text_block, Texture:texture_block, StopEvent:event(), StartCount:int, TargetCount:int)<suspends>:void=
        sync:
            # Only animate count text color once.
            AnimateCountTextColor(Text)
            block:
                # Keep animating the text value, the count texture and play the count audio until interrupted by the stop event
                # or until the target count has been reached.
                race:
                    StopEvent.Await()
                    block:
                        var CurrentlyShowingCount : int = StartCount
                        loop:
                            # Increase the count by a fraction of the remaining difference.
                            Difference := TargetCount - CurrentlyShowingCount
                            if:
                                Count := Max(3, Floor[Difference * 0.09 + 1.0])
                            then:
                                set CurrentlyShowingCount = Min(CurrentlyShowingCount + Count, TargetCount)
                                Text.SetText(CountText(CurrentlyShowingCount))
                            Sleep(0.0)
                            if (CurrentlyShowingCount = TargetCount):
                                break
                    loop:
                        spawn:
                            AnimateCountTexture(Texture)
                        Sleep(0.4)
                    loop:
                        ResourceCountAudioPlayer.Play(Player)
                        Sleep(0.1)

    # Bounce the count texture.
    AnimateCountTexture(Texture:texture_block)<suspends>:void=
        Texture.SetDesiredSize(vector2{X := 53.33, Y := 53.33})
        Sleep(0.03)
        Texture.SetDesiredSize(vector2{X := 66.66, Y := 66.66})
        Sleep(0.03)
        Texture.SetDesiredSize(vector2{X := 80.0, Y := 80.0})
        Sleep(0.03)
        Texture.SetDesiredSize(vector2{X := 72.0, Y := 72.0})
        Sleep(0.03)
        Texture.SetDesiredSize(vector2{X := 64.0, Y := 64.0})
        Sleep(0.03)
        Texture.SetDesiredSize(vector2{X := 56.0, Y := 56.0})
        Sleep(0.03)
        Texture.SetDesiredSize(vector2{X := 48.0, Y := 48.0})
        Sleep(0.03)
        Texture.SetDesiredSize(vector2{X := 40.0, Y := 40.0})

    # Make the count text blink with a yellow color.
    AnimateCountTextColor(Text:text_block)<suspends>:void=
        Text.SetTextColor(color{R := 0.99, G := 0.916, B := 0.7})
        Sleep(0.03)
        Text.SetTextColor(color{R := 0.98, G := 0.823, B := 0.4})
        Sleep(0.03)
        Text.SetTextColor(color{R := 0.97, G := 0.72, B := 0.1})
        Sleep(0.03)
        Text.SetTextColor(color{R := 0.976, G := 0.776, B := 0.28})
        Sleep(0.03)
        Text.SetTextColor(color{R := 0.982, G := 0.832, B := 0.46})
        Sleep(0.03)
        Text.SetTextColor(color{R := 0.988, G := 0.888, B := 0.64})
        Sleep(0.03)
        Text.SetTextColor(color{R := 0.994, G := 0.944, B := 0.82})
        Sleep(0.03)
        Text.SetTextColor(NamedColors.White)
