# Knowledge-neutral Verse snippet; project names and asset IDs generalized.
# id: snippet-128

# This file defines a speaking_character that can show information to the player based on proximity.
# The character can show one prop to attract the player and another when the player comes even closer.

using { /Fortnite.com/Characters }
using { /Fortnite.com/Devices }
using { /UnrealEngine.com/Temporary/SpatialMath }
using { /Verse.org/Simulation }
using { Utility }

# Enumeration describing the state of the speaking character.
speaking_character_state := enum:
    None,
    Idle,
    Attract,
    Speak

speaking_character := class(creative_device):

    # Prop that makes up the speaking character. 
    @editable
    Prop:creative_prop = creative_prop{}

    # Maximum distance to show the attract prop.
    @editable
    DistanceToAttract:float = 1200.0

    # Maximum distance to show the speak prop.
    @editable
    DistanceToSpeak:float = 500.0

    # Prop to show when attracting.
    @editable
    AttractProp:creative_prop = creative_prop{}

    # Audio to play when attracting.
    @editable
    AttractAudioPlayer:audio_player_device = audio_player_device{}
    
    # Prop to show when speaking.
    @editable
    SpeakProp:creative_prop = creative_prop{}

    # Audio to play when speaking.
    @editable
    SpeakAudioPlayer:audio_player_device = audio_player_device{}

    # Event signalled when character speaks.
    SpeakEvent:event() = event(){}

    # Event signalled to stop the current speech animation.
    StopEvent:event() = event(){}

    # The vertical offset used to animate the prop when showing.
    ShowOffset:vector3 = vector3{X := 0.0, Y := 0.0, Z := 20.0}

    # The vertical offset used to hide away the prop.
    HideOffset:vector3 = vector3{X := 0.0, Y := 0.0, Z := -500.0}

    # Current state of the character.
    var State:speaking_character_state = speaking_character_state.None

    # The initial transforms of the attract and speak props.
    var InitialAttractPropTransform:transform = transform{}
    var InitialSpeakPropTransform:transform = transform{}

    # Runs when the device is started in a running game.
    OnBegin<override>()<suspends>:void=
        # Store initial transforms of attract and speak props.
        set InitialAttractPropTransform = AttractProp.GetTransform()
        set InitialSpeakPropTransform = SpeakProp.GetTransform()

        # Initialise the attract and speak props.
        UpdateState(speaking_character_state.Idle)

        # Continuously find the distance of the closest player and update the attract and speak props
        # accordingly.
        PropTransform := Prop.GetTransform()
        loop:
            var MinCharacterDistanceSquared:float = Inf
            for(Player : GetPlayspace().GetPlayers(), Character := Player.GetFortCharacter[]):
                CharacterTransform := Character.GetTransform()
                CharacterDistanceSquared := DistanceSquared(PropTransform.Translation, CharacterTransform.Translation)
                set MinCharacterDistanceSquared = Min(MinCharacterDistanceSquared, CharacterDistanceSquared)
            MinCharacterDistance := Sqrt(MinCharacterDistanceSquared)
            if (MinCharacterDistance <= DistanceToSpeak):
                UpdateState(speaking_character_state.Speak)
            else if (MinCharacterDistance <= DistanceToAttract):
                UpdateState(speaking_character_state.Attract)
            else:
                UpdateState(speaking_character_state.Idle)
            Sleep(0.1)

    # Updates the visibility of the attract and speak props and plays audio based on the state.
    UpdateState(NewState:speaking_character_state):void=
        if (State <> NewState):
            set State = NewState
            # Stop current speech animation.
            StopEvent.Signal()
            # Hide existing props.
            HideSpeechProp(AttractProp, InitialAttractPropTransform)
            HideSpeechProp(SpeakProp, InitialSpeakPropTransform)
            # Show correct prop based on state.
            case(State):
                speaking_character_state.Attract =>
                    ShowSpeechProp(AttractProp, InitialAttractPropTransform, true)
                    AttractAudioPlayer.Play()
                speaking_character_state.Speak =>
                    SpeakEvent.Signal()
                    ShowSpeechProp(SpeakProp, InitialSpeakPropTransform, false)
                    SpeakAudioPlayer.Play()
                _ =>
                    block:

    # Hides a prop by teleporting it away.
    HideSpeechProp(SpeechProp:creative_prop, Transform:transform):void=
        if:
            SpeechProp.IsValid[]
            SpeechProp.TeleportTo[Transform.Translation + HideOffset, Transform.Rotation]

    # Shows a prop by teleporting it in and moving it.
    # Optionally keeps animating the prop.
    ShowSpeechProp(SpeechProp:creative_prop, Transform:transform, Animate:logic):void=
        if:
            SpeechProp.IsValid[]
            SpeechProp.TeleportTo[Transform.Translation + ShowOffset, Transform.Rotation]
        then:
            spawn:
                SpeechProp.MoveTo(Transform, 0.1)
            if (Animate?):
                spawn:
                    AnimateSpeechProp(SpeechProp)

    # Animates the speech prop until interrupted by stop event.
    AnimateSpeechProp(SpeechProp:creative_prop)<suspends>:void=
        race:
            StopEvent.Await()
            loop:
                Sleep(2.0)
                SpeechProp.Shake(vector3{X := 0.0, Y := 10.0, Z := 10.0}, vector3{X := 0.0, Y := 0.0, Z := 2.0}, 10, 1.0)
