# Knowledge-neutral Verse snippet; project names and asset IDs generalized.
# id: snippet-129

# This file contains a test device that sets up some useful buttons for testing the game.
# One button allows the players to skip ahead in time to see offline production without leaving the game.
# A second button grants the players 1000 gold.

using { /Fortnite.com/Devices }
using { /Verse.org/Simulation }
using { Factory }
using { Persistence }
using { Time }
using { UserInterface }

test := class(creative_device):

    # The real time device to add the debug time offset to.
    @editable
    RealTime:real_time = real_time{}

    # Button for skipping one hour ahead.
    @editable
    SkipHourButton:button_device = button_device{}

    # Button for granting gold.
    @editable
    GrantGoldButton:button_device = button_device{}

    # Factory manager that should perform the offline production. 
    @editable
    FactoryManager:factory_manager = factory_manager{}

    # UI manager for showing and updating UI.
    @editable
    UIManager:ui_manager = ui_manager{}

    # The player that has been selected to own the factory and get gold.
    # Can be false if nobody has been selected yet.
    var MaybeSelectedPlayer:?player = false

    # Runs when the device is started in a running game.
    OnBegin<override>()<suspends>:void=
        # Subscribe callbacks to get notified when a player interacts with the buttons.
        SkipHourButton.InteractedWithEvent.Subscribe(OnSkipHourButtonInteractedWith)
        GrantGoldButton.InteractedWithEvent.Subscribe(OnGrantGoldButtonInteractedWith)

    # Sets the selected player.
    SelectPlayer<public>(MaybePlayer:?player):void=
        set MaybeSelectedPlayer = MaybePlayer

    # When the skip button is interacted with, skip ahead in time, perform the offline production,
    # and show and update the relevant UIs.
    OnSkipHourButtonInteractedWith(Agent:agent):void=
        if:
            Player := player[Agent]
            SelectedPlayer := MaybeSelectedPlayer?
        then:
            RealTime.AddDebugTimeOffsetHours(1)
            ProductsProducedOffline := FactoryManager.CalculateOfflineProduction()
            FactoryManager.Update()
            UIManager.ShowOfflineProductionUI(Player, FactoryManager, ProductsProducedOffline)

            # Update the production times of all built moulds to reflect the new time.
            Now := RealTime.GetDateAndTime()
            for(MouldIndex := 0..3, IsFactoryMouldBuilt[SelectedPlayer, MouldIndex]):
                if:
                    SetFactoryMouldLatestProductionTime[SelectedPlayer, MouldIndex, Now]

            UIManager.UpdateAllFactoryStorageUIs()
            UIManager.UpdateAllOrderUIs()
            UIManager.UpdateAllFactoryUIs()
    
    # When the grant gold button is interacted with, grant the gold to the selected player
    # and update the relevant UIs.
    OnGrantGoldButtonInteractedWith(Agent:agent):void=
        if:
            SelectedPlayer := MaybeSelectedPlayer?
            GrantGold[SelectedPlayer, 1000]
        then:
            UIManager.UpdateAllPriceUIs()
            UIManager.UpdateAllResourceUIs()
