# Knowledge-neutral Verse snippet; project names and asset IDs generalized.
# id: snippet-130

# This file defines a time_interval for representing a time interval.
# It is what you get when subtracting two date_and_time.

using { /Verse.org/Simulation }
using { Utility }

# A time interval
time_interval<public> := class<concrete>:
    @editable
    Days<public>:int = 0
    @editable
    Hours<public>:int = 0
    @editable
    Minutes<public>:int = 0
    @editable
    Seconds<public>:float = 0.0

    # Returns the total amount of seconds in this interval.
    GetTotalSeconds<public>()<transacts>:float=
        return Seconds + Minutes * 60.0 + Hours * 3600.0 + Days * 86400.0

    # Returns the given positive time interval as a nicely formatted string.
    # Truncate any leading zeroes.
    ToTruncatedString<public>():string=
        if:
            WholeSeconds := Floor[Seconds]
        then:
            if (Days > 0):
                return ToString(Self)
            else if (Hours > 0):
                return "{Hours}:{PadNumber(Minutes, 2, '0')}:{PadNumber(WholeSeconds, 2, '0')}"
            else if (Minutes > 0):
                return "{Minutes}:{PadNumber(WholeSeconds, 2, '0')}"
            else:
                return "0:{PadNumber(WholeSeconds, 2, '0')}"
        else:
            return ""

# Returns the given positive time interval as a nicely formatted string. 
ToString<public>(TI:time_interval):string=
    if:
        WholeSeconds := Floor[TI.Seconds]
    then:
        return "{TI.Days}:{PadNumber(TI.Hours, 2, '0')}:{PadNumber(TI.Minutes, 2, '0')}:{PadNumber(WholeSeconds, 2, '0')}"
    else:
        return ""

# Create a time_interval given the total amount of seconds. 
MakeTimeInterval<public>(TotalSeconds:float)<transacts>:time_interval=
    if:
        TotalSeconds < 0.0
    then:
        if:
            Days := Ceil[TotalSeconds / 86400.0]
            Hours := Ceil[(TotalSeconds - Days * 86400.0) / 3600.0]
            Minutes := Ceil[(TotalSeconds - Days * 86400.0  - Hours * 3600.0) / 60.0]
            Seconds := TotalSeconds - Days * 86400.0  - Hours * 3600.0 - Minutes * 60.0
        then:
            return MakeTimeInterval(Days, Hours, Minutes, Seconds)
    else:
        if:
            Days := Floor[TotalSeconds / 86400.0]
            Hours := Floor[(TotalSeconds - Days * 86400.0) / 3600.0]
            Minutes := Floor[(TotalSeconds - Days * 86400.0  - Hours * 3600.0) / 60.0]
            Seconds := TotalSeconds - Days * 86400.0  - Hours * 3600.0 - Minutes * 60.0
        then:
            return MakeTimeInterval(Days, Hours, Minutes, Seconds)

    return time_interval{}

# Create a time_interval given days, hours, minutes and seconds.
# There is no validation that all values have the same sign.
MakeTimeInterval<public>(Days:int, Hours:int, Minutes:int, Seconds:float)<transacts>:time_interval=
    return time_interval:
        Days := Days
        Hours := Hours
        Minutes := Minutes
        Seconds := Seconds
