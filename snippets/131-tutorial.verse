# Knowledge-neutral Verse snippet; project names and asset IDs generalized.
# id: snippet-131

# This file contains a tutorial device that guides the player through playing the game.
# The device tracks the player's progress and serves up the next tutorial step using
# tracker_devices to generate explanatory UI and map_indicator_devices to guide the
# player around the map.

using { /Fortnite.com/Devices }
using { /UnrealEngine.com/Temporary/SpatialMath }
using { /Verse.org/Simulation }
using { Config }
using { Factory }
using { Persistence }
using { UserInterface }
using { Utility }

tutorial := class(creative_device):

    # Tracker for detecting offline production.
    @editable
    OfflineProductionPerformedTracker:tracker_device = tracker_device{}

    # Tracker for detecting if Sir Pawlar has spoken.
    @editable
    SirPawlarSpokenTracker:tracker_device = tracker_device{}
    
    # Tracker for tutorial step telling player to talk to Sir Pawlar.
    @editable
    TalkToSirPawlarTracker:tracker_device = tracker_device{}

    # Map indicator for above step.
    @editable
    TalkToSirPawlarMapIndicator:map_indicator_device = map_indicator_device{}

    # Sir Pawlar's speaking character.
    @editable
    SirPawlar:speaking_character = speaking_character{}

    # Tracker for tutorial step telling player to load factory.
    @editable
    LoadFactoryTracker:tracker_device = tracker_device{}

    # Map indicator for above step.
    @editable
    LoadFactoryMapIndicator:map_indicator_device = map_indicator_device{}

    # Tracker for tutorial step telling player to harvest candy CandyCanes.
    @editable
    CandyCaneFieldTracker:tracker_device = tracker_device{}

    # Map indicator for above step.
    @editable
    CandyCaneFieldMapIndicator:map_indicator_device = map_indicator_device{}

    # The candy CandyCane field to harvest in above step.
    @editable
    CandyCaneField:candy_cane_field = candy_cane_field{}

    # Tracker for tutorial step telling player to build first factory belt.
    @editable
    FactoryBeltTracker:tracker_device = tracker_device{}

    # Map indicator for above step.
    @editable
    FactoryBeltMapIndicator:map_indicator_device = map_indicator_device{}

    # Factory belt to build in above step.
    @editable
    FactoryBelt:factory_belt = factory_belt{}

    # Tracker for tutorial step telling player to build first factory mould.
    @editable
    FactoryMouldTracker:tracker_device = tracker_device{}

    # Map indicator for above step.
    @editable
    FactoryMouldMapIndicator:map_indicator_device = map_indicator_device{}

    # Factory mould to build in above step.
    @editable
    FactoryMould:factory_mould = factory_mould{}

    # Tracker for tutorial step telling player to build factory storage.
    @editable
    FactoryStorageTracker:tracker_device = tracker_device{}

    # Map indicator for above step.
    @editable
    FactoryStorageMapIndicator:map_indicator_device = map_indicator_device{}

    # Factory storage to build in above step.
    @editable
    FactoryStorage:factory_storage = factory_storage{}

    # Tracker for tutorial step telling player to make first delivery to order.
    @editable
    OrderTracker:tracker_device = tracker_device{}

    # Map indicator for above step.
    @editable
    OrderMapIndicator:map_indicator_device = map_indicator_device{}

    # Order to deliver to in above step.
    @editable
    Order:order = order{}

    # Tracker for tutorial step telling player to wait for first product.
    @editable
    ProductTracker:tracker_device = tracker_device{}

    # Tracker for tutorial step telling player to go offline and come back.
    @editable
    FactoryOfflineProductionTracker:tracker_device = tracker_device{}

    # Factory manager to monitor for offline production in above step.
    @editable
    FactoryManager:factory_manager = factory_manager{}

    # Tracker for tutorial step telling player to upgrade first factory mould.
    @editable
    FactoryMouldUpgradeTracker:tracker_device = tracker_device{}

    # Tracker for tutorial step telling player to build second factory belt.
    @editable
    SecondFactoryBeltTracker:tracker_device = tracker_device{}

    # Map indicator for above step.
    @editable
    SecondFactoryBeltMapIndicator:map_indicator_device = map_indicator_device{}

    # Factory belt to build in above step.
    @editable
    SecondFactoryBelt:factory_belt = factory_belt{}

    # Tracker for tutorial step telling player to head back to UEFN and seek creator challenges from seals.
    @editable
    TutorialCompleteTracker:tracker_device = tracker_device{}

    # The player that has been selected to own the factory and order.
    # Can be false if nobody has been selected yet.
    var MaybeSelectedPlayer:?player = false

    # Runs when the device is started in a running game.
    OnBegin<override>()<suspends>:void=
        NoMapIndicator : ?map_indicator_device = false 
        # Subscribe callbacks to get notified when a tracker is completed.
        TalkToSirPawlarTracker.CompleteEvent.SubscribeAgent(OnTutorialStepTrackerComplete, (TalkToSirPawlarTracker, option{TalkToSirPawlarMapIndicator}))
        LoadFactoryTracker.CompleteEvent.SubscribeAgent(OnTutorialStepTrackerComplete, (LoadFactoryTracker, option{LoadFactoryMapIndicator}))
        CandyCaneFieldTracker.CompleteEvent.SubscribeAgent(OnTutorialStepTrackerComplete, (CandyCaneFieldTracker, option{CandyCaneFieldMapIndicator}))
        FactoryBeltTracker.CompleteEvent.SubscribeAgent(OnTutorialStepTrackerComplete, (FactoryBeltTracker, option{FactoryBeltMapIndicator}))
        FactoryMouldTracker.CompleteEvent.SubscribeAgent(OnTutorialStepTrackerComplete, (FactoryMouldTracker, option{FactoryMouldMapIndicator}))
        FactoryStorageTracker.CompleteEvent.SubscribeAgent(OnTutorialStepTrackerComplete, (FactoryStorageTracker, option{FactoryStorageMapIndicator}))
        OrderTracker.CompleteEvent.SubscribeAgent(OnTutorialStepTrackerComplete, (OrderTracker, option{OrderMapIndicator}))
        ProductTracker.CompleteEvent.SubscribeAgent(OnTutorialStepTrackerComplete, (ProductTracker, NoMapIndicator))
        FactoryOfflineProductionTracker.CompleteEvent.SubscribeAgent(OnTutorialStepTrackerComplete, (FactoryOfflineProductionTracker, NoMapIndicator))
        FactoryMouldUpgradeTracker.CompleteEvent.SubscribeAgent(OnTutorialStepTrackerComplete, (FactoryMouldUpgradeTracker, option{FactoryMouldMapIndicator}))
        SecondFactoryBeltTracker.CompleteEvent.SubscribeAgent(OnTutorialStepTrackerComplete, (SecondFactoryBeltTracker, option{SecondFactoryBeltMapIndicator}))

    # Sets the selected player.
    SelectPlayer<public>(MaybePlayer:?player):void=
        set MaybeSelectedPlayer = MaybePlayer

    # When a player is added, determine the tutorial progress.
    PlayerAdded<public>(Player:player):void=
        # Continuously monitor if offline production is performed or Sir Pawlar is speaking.
        spawn:
            MonitorOfflineProductionPerformedAndSirPawlarSpeaking(Player)
        # Slightly delay determining the initial tutorial progress.
        spawn:
            DelayedDetermineTutorialProgress(Player, 2.0)

    # Determines the tutorial progress with a slight delay.
    DelayedDetermineTutorialProgress(Player:player, Delay:float)<suspends>:void=
        Sleep(Delay)
        if (Player.IsActive[]):
           DetermineTutorialProgress(Player)

    # Determines the tutorial progress.
    DetermineTutorialProgress(Agent:agent):void=
        # Firstly, has the player talked to Sir Pawlar?
        if:
            not SirPawlarSpokenTracker.HasReachedTarget[Agent]
        then:
            # No. Tell the player to talk to Sir Pawlar.
            TalkToSirPawlarStep(Agent)
            return

        # The player has talked to Sir Pawlar. Continue the tutorial.
        if:
            SelectedPlayer := MaybeSelectedPlayer?
            Gold := GetGold[SelectedPlayer]
            Belt0Built := logic{IsFactoryBeltBuilt[SelectedPlayer, 0]}
            Belt0UpgradePrice := GetFactoryBeltUpgradePrice[0, SelectedPlayer]
            Belt1Built := logic{IsFactoryBeltBuilt[SelectedPlayer, 1]}
            Mould0Built := logic{IsFactoryMouldBuilt[SelectedPlayer, 0]}
            Mould0Level := GetFactoryMouldLevel[SelectedPlayer, 0]
            Mould0UpgradePrice := GetFactoryMouldUpgradePrice[0, SelectedPlayer]
            StorageBuilt := logic{IsFactoryStorageBuilt[SelectedPlayer]}
            StorageUpgradePrice := GetFactoryStorageUpgradePrice[SelectedPlayer]
            Product0 := GetFactoryProduct[SelectedPlayer, 0]
            OrderNumber := GetOrderNumber[SelectedPlayer]
            ProductsDelivered0 := GetProductsDelivered[SelectedPlayer, 0]
            FactoryOfflineProductionPerformed := logic{OfflineProductionPerformedTracker.HasReachedTarget[Agent]}
        then:
            if (not Belt0Built?):
                # First belt still unbuilt. Can the player afford it?
                if (Gold >= Belt0UpgradePrice):
                    # Yes. Tell the player to build it.
                    UpgradeFactoryElementStep(Agent, FactoryBelt, FactoryBeltTracker, FactoryBeltMapIndicator)
                else:
                    # No. Tell the player to harvest candy CandyCanes.
                    HarvestCandyCanesStep(Agent)
            
            # First belt is built. Is the first mould built?

            else if (not Mould0Built?):
                # First mould still unbuilt. Can the player afford it?
                if (Gold >= Mould0UpgradePrice):
                    # Yes. Tell the player to build it.
                    UpgradeFactoryElementStep(Agent, FactoryMould, FactoryMouldTracker, FactoryMouldMapIndicator)
                else:
                    # No. Tell the player to harvest candy CandyCanes.
                    HarvestCandyCanesStep(Agent)

            # First belt and mould are built. Is storage built?
            
            else if (not StorageBuilt?):
                # Storage still unbuilt. Can the player afford it?
                if (Gold >= StorageUpgradePrice):
                    # Yes. Tell the player to build it.
                    UpgradeFactoryElementStep(Agent, FactoryStorage, FactoryStorageTracker, FactoryStorageMapIndicator)
                else:
                    # No. Tell the player to harvest candy CandyCanes.
                    HarvestCandyCanesStep(Agent)

            # First belt, mould and storage are built. Has the player delivered anything?
            else if (OrderNumber = 1 and ProductsDelivered0 = 0):
                # Player has not delivered anything. Does the player have anything to deliver?
                if (Product0 > 0):
                    # Yes. Tell the player to deliver.
                    DeliverProductStep(Agent)
                else:
                    # No. Tell the player to wait.
                    WaitForProductStep(Agent)

            # First belt, mould and storage are built. Player has delivered something. Has player's factory produced offline?
            else if (not FactoryOfflineProductionPerformed?):
                # Factory has not produced offline. Tell player to go offline.
                FactoryOfflineProductionStep(Agent)

            # First belt, mould and storage are built. Player has delivered something and produced offline. Is first mould upgraded?
            else if (Mould0Level = 1):
                # Player has not upgraded first mould. Tell the player to upgrade it.
                UpgradeFactoryElementStep(Agent, FactoryMould, FactoryMouldUpgradeTracker, FactoryMouldMapIndicator)

            # First mould has been upgraded. Has the player built the second belt?
            else if (not Belt1Built?):
                # Second belt still unbuilt. Tell the player to build it.
                UpgradeFactoryElementStep(Agent, SecondFactoryBelt, SecondFactoryBeltTracker, SecondFactoryBeltMapIndicator)

            # The tutorial is done. Tell the player to go back to UEFN and seek creator challenges.
            else:
                TutorialCompleteStep(Agent)
        else:
            # Factory has not been loaded yet. Tell the player to load the factory.
            LoadFactoryStep(Agent)

    # Monitors if offline production is performed and if Sir Pawlar is speaking.
    MonitorOfflineProductionPerformedAndSirPawlarSpeaking(Agent:agent)<suspends>:void=
        if (Player := player[Agent]):
            race:
                # Break out if the player leaves.
                loop:
                    if (not Player.IsActive[]):
                        break
                    Sleep(1.0)
                # Wait for the offline production event and check if something was produced.
                loop:
                    OfflineProduction := FactoryManager.OfflineProductionEvent.Await()
                    if:
                        Product0 := OfflineProduction[0]
                        Product1 := OfflineProduction[1]
                        Product2 := OfflineProduction[2]
                        Product3 := OfflineProduction[3]
                        Product0 > 0 or Product1 > 0 or Product2 > 0 or Product3 > 0
                    then:
                        OfflineProductionPerformedTracker.Increment(Agent)
                # Wait for speak event from Sir Pawlar.
                loop:
                    SirPawlar.SpeakEvent.Await()
                    SirPawlarSpokenTracker.Increment(Agent)
        
    # Handles tutorial step telling player to talk to Sir Pawlar.
    TalkToSirPawlarStep(Agent:agent):void=
        # Reset and assign the tracker to make the UI appear.
        TalkToSirPawlarTracker.Reset(Agent)
        TalkToSirPawlarTracker.Assign(Agent)
        # Currently, the objective pulse must be activated multiple times for it to work.
        TalkToSirPawlarMapIndicator.ActivateObjectivePulse(Agent)
        TalkToSirPawlarMapIndicator.ActivateObjectivePulse(Agent)

        # Continuously monitor if player has talked to Sir Pawlar.
        spawn:
            MonitorSirPawlarSpoken(Agent)
        

    # Handles tutorial step telling player to load factory.
    LoadFactoryStep(Agent:agent):void=
        # Reset and assign the tracker to make the UI appear.
        LoadFactoryTracker.Reset(Agent)
        LoadFactoryTracker.Assign(Agent)
        # Currently, the objective pulse must be activated multiple times for it to work.
        LoadFactoryMapIndicator.ActivateObjectivePulse(Agent)
        LoadFactoryMapIndicator.ActivateObjectivePulse(Agent)

        # Continuously monitor if factory is loaded.
        spawn:
            MonitorFactoryLoaded(Agent)

    # Handles tutorial step telling player to harvest candy CandyCanes.
    HarvestCandyCanesStep(Agent:agent):void=
        # Reset and assign the tracker to make the UI appear.
        CandyCaneFieldTracker.Reset(Agent)
        CandyCaneFieldTracker.Assign(Agent)
        # Currently, the objective pulse must be activated multiple times for it to work.
        CandyCaneFieldMapIndicator.ActivateObjectivePulse(Agent)
        CandyCaneFieldMapIndicator.ActivateObjectivePulse(Agent)

        # Continuously monitor for candy CandyCane harvest.
        spawn:
            MonitorCandyCaneHarvesting(Agent)

    # Handles all tutorial steps involving building or upgrading a factory element.
    UpgradeFactoryElementStep(Agent:agent, FactoryElement:factory_element, Tracker:tracker_device, MapIndicator:map_indicator_device):void=
        # Reset and assign the tracker to make the UI appear.
        Tracker.Reset(Agent)
        Tracker.Assign(Agent)
        # Currently, the objective pulse must be activated multiple times for it to work.
        MapIndicator.ActivateObjectivePulse(Agent)
        MapIndicator.ActivateObjectivePulse(Agent)

        # Continuously monitor for factory element upgrade.
        spawn:
            MonitorFactoryElementUpgrade(Agent, FactoryElement, Tracker, MapIndicator)

    # Handles tutorial step telling player to make first delivery to order.
    DeliverProductStep(Agent:agent):void=
        # Reset and assign the tracker to make the UI appear.
        OrderTracker.Reset(Agent)
        OrderTracker.Assign(Agent)
        # Currently, the objective pulse must be activated multiple times for it to work.
        OrderMapIndicator.ActivateObjectivePulse(Agent)
        OrderMapIndicator.ActivateObjectivePulse(Agent)

        # Continuously monitor for delivery.
        spawn:
            MonitorProductDelivery(Agent)

    # Handles tutorial step telling player to wait for first product.
    WaitForProductStep(Agent:agent):void=
        # Reset and assign the tracker to make the UI appear.
        ProductTracker.Reset(Agent)
        ProductTracker.Assign(Agent)

        # Continuously monitor for product storage.
        spawn:
            MonitorProductStorage(Agent)

    # Handles tutorial step telling player to go offline and come back.
    FactoryOfflineProductionStep(Agent:agent):void=
        # Reset and assign the tracker to make the UI appear.
        FactoryOfflineProductionTracker.Reset(Agent)
        FactoryOfflineProductionTracker.Assign(Agent)

        # Continuously monitor if offline production has been performed.
        spawn:
            MonitorOfflineProduction(Agent)

    # Handles tutorial step telling player to go back to UEFN to seek creator challenges.
    TutorialCompleteStep(Agent:agent):void=
        # Reset and assign the tracker to make the UI appear.
        TutorialCompleteTracker.Reset(Agent)
        TutorialCompleteTracker.Assign(Agent)

        # Continuously monitor if tutorial is interrupted.
        spawn:
            MonitorTutorialInterruption(Agent)

    # Monitors if Sir Pawlar has spoken.
    MonitorSirPawlarSpoken(Agent:agent)<suspends>:void=
        if (Player := player[Agent]):
            race:
                loop:
                    # Break out if the player leaves.
                    if (not Player.IsActive[]):
                        break
                    Sleep(1.0)
                # Break out when completed.
                loop:
                    CompletedAgent := TalkToSirPawlarTracker.CompleteEvent.Await()
                    if (CompletedAgent = Agent):
                        break
                # Check if Sir Pawlar spoken tracker has reached target as that 
                # means Sir Pawlar has spoken.
                loop:
                    Sleep(2.0)
                    if:
                        SirPawlarSpokenTracker.HasReachedTarget[Agent]
                    then:
                        TalkToSirPawlarTracker.Increment(Agent)

    # Monitors if factory has been loaded.
    MonitorFactoryLoaded(Agent:agent)<suspends>:void=
        if (Player := player[Agent]):
            race:
                # Break out if the player leaves.
                loop:
                    if (not Player.IsActive[]):
                        break
                    Sleep(1.0)
                # Break out when completed.
                loop:
                    CompletedAgent := LoadFactoryTracker.CompleteEvent.Await()
                    if (CompletedAgent = Agent):
                        break
                # Check if player owning factory has been selected as that means the factory was loaded.
                loop:
                    Sleep(2.0)
                    if:
                        MaybeSelectedPlayer?
                    then:
                        LoadFactoryTracker.Increment(Agent)

    # Monitors candy CandyCane harvest.
    MonitorCandyCaneHarvesting(Agent:agent)<suspends>:void=
        if (Player := player[Agent]):
            race:
                # Break out if the player leaves or the selected player was reset.
                loop:
                    if (not Player.IsActive[] or not MaybeSelectedPlayer?):
                        # Remove the current tracker and map indicator.
                        CandyCaneFieldTracker.Remove(Agent)
                        CandyCaneFieldMapIndicator.DeactivateObjectivePulse(Agent)
                        # Determine a new tutorial step.
                        DelayedDetermineTutorialProgress(Player, 2.0)
                        break
                    Sleep(1.0)
                # Break out when completed.
                loop:
                    CompletedAgent := CandyCaneFieldTracker.CompleteEvent.Await()
                    if (CompletedAgent = Agent):
                        break
                # Wait for the harvest event and increment the tracker.
                loop:
                    CandyCaneField.HarvestEvent.Await()
                    CandyCaneFieldTracker.Increment(Agent)

    # Monitors factory element upgrades.
    MonitorFactoryElementUpgrade(Agent:agent, FactoryElement:factory_element, Tracker:tracker_device, MapIndicator:map_indicator_device)<suspends>:void=
        if (Player := player[Agent]):
            race:
                # Break out if the player leaves or the selected player was reset.
                loop:
                    if (not Player.IsActive[] or not MaybeSelectedPlayer?):
                        # Remove the current tracker and map indicator.
                        Tracker.Remove(Agent)
                        MapIndicator.DeactivateObjectivePulse(Agent)
                        # Determine a new tutorial step.
                        DelayedDetermineTutorialProgress(Player, 2.0)
                        break
                    Sleep(1.0)
                # Break out when completed.
                loop:
                    CompletedAgent := Tracker.CompleteEvent.Await()
                    if (CompletedAgent = Agent):
                        break
                # Wait for the upgrade event.
                loop:
                    FactoryElement.UpgradeEvent.Await()
                    Tracker.Increment(Agent)

    # Monitors product delivery.
    MonitorProductDelivery(Agent:agent)<suspends>:void=
        if (Player := player[Agent]):
            race:
                # Break out if the player leaves or the selected player was reset.
                loop:
                    if (not Player.IsActive[] or not MaybeSelectedPlayer?):
                        # Remove the current tracker and map indicator.
                        OrderTracker.Remove(Agent)
                        OrderMapIndicator.DeactivateObjectivePulse(Agent)
                        # Determine a new tutorial step.
                        DelayedDetermineTutorialProgress(Player, 2.0)
                        break
                    Sleep(1.0)
                # Break out when completed.
                loop:
                    CompletedAgent := OrderTracker.CompleteEvent.Await()
                    if (CompletedAgent = Agent):
                        break
                # Wait for the deliver event and check if the right product was delivered.
                loop:
                    Delivery := Order.DeliverEvent.Await()
                    if (Delivery.ProductNumber = 0, Delivery.Amount > 0):
                        OrderTracker.Increment(Agent)

    # Monitors products being put in storage.
    MonitorProductStorage(Agent:agent)<suspends>:void=
        if (Player := player[Agent]):
            race:
                # Break out if the player leaves or the selected player was reset.
                loop:
                    if (not Player.IsActive[] or not MaybeSelectedPlayer?):
                        # Remove the current tracker.
                        ProductTracker.Remove(Agent)
                        # Determine a new tutorial step.
                        DelayedDetermineTutorialProgress(Player, 2.0)
                        break
                    Sleep(1.0)
                # Break out when completed.
                loop:
                    CompletedAgent := ProductTracker.CompleteEvent.Await()
                    if (CompletedAgent = Agent):
                        break
                # Wait for the store event and check if the right product was stored.
                loop:
                    ProductStored := FactoryStorage.StoreEvent.Await()
                    if (ProductStored = 0):
                        ProductTracker.Increment(Agent)

    # Monitors if offline production has occurred.
    MonitorOfflineProduction(Agent:agent)<suspends>:void=
        if (Player := player[Agent]):
            race:
                loop:
                    # Break out if the player leaves or the selected player was reset.
                    if (not Player.IsActive[] or not MaybeSelectedPlayer?):
                        # Remove the current tracker.
                        FactoryOfflineProductionTracker.Remove(Agent)
                        # Determine a new tutorial step.
                        DelayedDetermineTutorialProgress(Player, 2.0)
                        break
                    Sleep(1.0)
                # Break out when completed.
                loop:
                    CompletedAgent := FactoryOfflineProductionTracker.CompleteEvent.Await()
                    if (CompletedAgent = Agent):
                        break
                # Check if offline production performed tracker has reached target as that 
                # means offline production has been performed.
                loop:
                    Sleep(2.0)
                    if:
                        OfflineProductionPerformedTracker.HasReachedTarget[Agent]
                    then:
                        FactoryOfflineProductionTracker.Increment(Agent)

    MonitorTutorialInterruption(Agent:agent)<suspends>:void=
        if (Player := player[Agent]):
            loop:
                # Break out if the player leaves or the selected player was reset.
                if (not Player.IsActive[] or not MaybeSelectedPlayer?):
                    # Remove the current tracker.
                    TutorialCompleteTracker.Remove(Agent)
                    # Determine a new tutorial step.
                    DelayedDetermineTutorialProgress(Player, 2.0)
                    break
                Sleep(1.0)

    # Handles when the tutorial step tracker is complete.
    OnTutorialStepTrackerComplete(Agent:agent, TrackerAndMaybeMapIndicator:tuple(tracker_device, ?map_indicator_device)):void=
        spawn:
            CompleteTutorialStep(Agent, TrackerAndMaybeMapIndicator(0), TrackerAndMaybeMapIndicator(1))

    # Completes the tutorial step by deactivating any map indicator, removing the tracker and determining
    # the next tutorial step.
    CompleteTutorialStep(Agent:agent, Tracker:tracker_device, MaybeMapIndicator:?map_indicator_device)<suspends>:void=
        if:
            MapIndicator := MaybeMapIndicator?
        then:
            MapIndicator.DeactivateObjectivePulse(Agent)

        # Wait a little to allow the tutorial step UI to complete.
        Sleep(2.0)

        Tracker.Remove(Agent)

        # Show next tutorial step.
        DetermineTutorialProgress(Agent)
