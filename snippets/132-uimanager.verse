# Knowledge-neutral Verse snippet; project names and asset IDs generalized.
# id: snippet-132

# This file defines the ui_manager that handles all UI created through Verse.
# It contains functions related to showing, updating and closing the various UIs.
# See the individual UI files for more details.

using { /Fortnite.com/Devices }
using { /Verse.org/Simulation }
using { Factory }

ui_manager<public> := class(creative_device):
    @editable
    ShowUIAudioPlayer:audio_player_device = audio_player_device{}

    @editable
    LoadFactoryUI:load_factory_ui = load_factory_ui{}

    @editable
    ResourceUI:resource_ui = resource_ui{}

    @editable
    FactoryUI:factory_ui = factory_ui{}

    @editable
    OfflineProductionUI:offline_production_ui = offline_production_ui{}

    @editable
    FactoryBeltUI:factory_belt_ui = factory_belt_ui{}

    @editable
    FactoryMouldUI:factory_mould_ui = factory_mould_ui{}

    @editable
    FactoryWrapperUI:factory_wrapper_ui = factory_wrapper_ui{}

    @editable
    FactoryStorageUI:factory_storage_ui = factory_storage_ui{}

    @editable
    OrderUI:order_ui = order_ui{}

    @editable
    JukeboxUI:jukebox_ui = jukebox_ui{}
    
    PlayerAdded<public>(Player:player):void=
        LoadFactoryUI.PlayerAdded(Player)
        ResourceUI.PlayerAdded(Player)
        FactoryUI.PlayerAdded(Player)
        OfflineProductionUI.PlayerAdded(Player)
        FactoryBeltUI.PlayerAdded(Player)
        FactoryMouldUI.PlayerAdded(Player)
        FactoryWrapperUI.PlayerAdded(Player)
        FactoryStorageUI.PlayerAdded(Player)
        OrderUI.PlayerAdded(Player)
        JukeboxUI.PlayerAdded(Player)

    PlayerRemoved<public>(Player:player):void=
        LoadFactoryUI.PlayerRemoved(Player)
        ResourceUI.PlayerRemoved(Player)
        FactoryUI.PlayerRemoved(Player)
        OfflineProductionUI.PlayerRemoved(Player)
        FactoryBeltUI.PlayerRemoved(Player)
        FactoryMouldUI.PlayerRemoved(Player)
        FactoryWrapperUI.PlayerRemoved(Player)
        FactoryStorageUI.PlayerRemoved(Player)
        OrderUI.PlayerRemoved(Player)
        JukeboxUI.PlayerRemoved(Player)

    ShowLoadFactoryUI<public>(Player:player, GameManager:game_manager):void=
        ShowUIAudioPlayer.Play(Player)
        LoadFactoryUI.ShowUI(Player, GameManager)

    ShowResourceUI<public>(Player:player, SelectedPlayer:player):void=
        ResourceUI.ShowUI(Player, SelectedPlayer)

    ShowAllResourceUIs<public>(SelectedPlayer:player):void=
        for(Player : GetPlayspace().GetPlayers()):
            ShowResourceUI(Player, SelectedPlayer)

    ShowFactoryUI<public>(Player:player, FactoryManager:factory_manager):void=
        FactoryUI.ShowUI(Player, FactoryManager)

    ShowAllFactoryUIs<public>(FactoryManager:factory_manager):void=
        for(Player : GetPlayspace().GetPlayers()):
            ShowFactoryUI(Player, FactoryManager)
        
    ShowOfflineProductionUI<public>(Player:player, FactoryManager:factory_manager, ProductsProducedOffline:[]int):void=
        OfflineProductionUI.ShowUI(Player, FactoryManager, ProductsProducedOffline, ShowUIAudioPlayer)

    ShowFactoryBeltUI<public>(Player:player, FactoryBelt:factory_belt):void=
        FactoryBeltUI.ShowUI(Player, FactoryBelt, ShowUIAudioPlayer)

    ShowFactoryMouldUI<public>(Player:player, FactoryMould:factory_mould):void=
        FactoryMouldUI.ShowUI(Player, FactoryMould, ShowUIAudioPlayer)
            
    ShowFactoryWrapperUI<public>(Player:player, FactoryWrapper:factory_wrapper):void=
        FactoryWrapperUI.ShowUI(Player, FactoryWrapper, ShowUIAudioPlayer)
    
    ShowFactoryStorageUI<public>(Player:player, FactoryStorage:factory_storage):void=
        FactoryStorageUI.ShowUI(Player, FactoryStorage, ShowUIAudioPlayer)
    
    ShowOrderUI<public>(Player:player, Order:order):void=
        OrderUI.ShowUI(Player, Order, ShowUIAudioPlayer)

    ShowJukeboxUI<public>(Player:player, Jukebox:jukebox):void=
        JukeboxUI.ShowUI(Player, Jukebox, ShowUIAudioPlayer)

    UpdateAllPriceUIs<public>():void=
        for(Player : GetPlayspace().GetPlayers()):
            FactoryBeltUI.UpdateUI(Player)
            FactoryMouldUI.UpdateUI(Player)
            FactoryWrapperUI.UpdateUI(Player)
            FactoryStorageUI.UpdateUI(Player)
            JukeboxUI.UpdateUI(Player)

    UpdateAllResourceUIs<public>():void=
        for(Player : GetPlayspace().GetPlayers()):
            ResourceUI.UpdateUI(Player)

    UpdateAllFactoryUIs<public>():void=
        for(Player : GetPlayspace().GetPlayers()):
            FactoryUI.UpdateUI(Player)

    UpdateAllFactoryStorageUIs<public>():void=
        for(Player : GetPlayspace().GetPlayers()):
            FactoryStorageUI.UpdateUI(Player)

    UpdateAllOrderUIs<public>():void=
        for(Player : GetPlayspace().GetPlayers()):
            OrderUI.UpdateUI(Player)

    CloseAllGameUIs<public>():void=
        for(Player : GetPlayspace().GetPlayers()):
            ResourceUI.CloseUI(Player)
            FactoryUI.CloseUI(Player)
            OfflineProductionUI.CloseUI(Player)
            FactoryBeltUI.CloseUI(Player, 0.0)
            FactoryMouldUI.CloseUI(Player, 0.0)
            FactoryWrapperUI.CloseUI(Player, 0.0)
            FactoryStorageUI.CloseUI(Player, 0.0)
            OrderUI.CloseUI(Player)
            JukeboxUI.CloseUI(Player)
