# Knowledge-neutral Verse snippet; project names and asset IDs generalized.
# id: snippet-133

# This file contains the data used to configure the factory in the game.
# This includes upgrade prices, max levels, production speeds, storage capacities and more.
# It also contains functions for getting all this data. 

using { /Verse.org/Simulation }
using { Persistence }

# Blubberchops: "Looks like this is the place to modify the upgrade prices for belts."
#               "Let's make them cheaper!"

# Defines the upgrade price per level for each factory belt.
FactoryBeltPrices: [][]int =
    # Each column is a belt.       0             1               2               3
    # Each row is a level.
    array{  array{                 5,           25,            500,          5000},
            array{                25,          125,           2500,         25000},
            array{               100,          500,          10000,        100000},
            array{               200,         1000,          20000,        200000},
            array{               300,         1500,          30000,        300000},
            array{               400,         2000,          40000,        400000},
            array{               500,         2500,          50000,        500000},
            array{               750,         3750,          75000,        750000},
            array{              1000,         5000,         100000,       1000000},
            array{              1500,         7500,         150000,       1500000},
            array{              2000,        10000,         200000,       2000000},
            array{              2500,        12500,         250000,       2500000},
            array{              3000,        15000,         300000,       3000000},
            array{              4000,        20000,         400000,       4000000},
            array{              5000,        25000,         500000,       5000000}}

# Defines the upgrade price per level for each factory mould.
FactoryMouldPrices : [][]int = 
    # Each column is a mould.      0             1               2               3
    # Each row is a level.
    array{  array{                 5,           25,            500,          5000},
            array{                15,           75,           1500,         15000},
            array{                25,          125,           2500,         25000},
            array{                50,          250,           5000,         50000},
            array{               100,          500,          10000,        100000},
            array{               150,          750,          15000,        150000},
            array{               200,         1000,          20000,        200000},
            array{               250,         1250,          25000,        250000},
            array{               300,         1500,          30000,        300000},
            array{               350,         1750,          35000,        350000},
            array{               400,         2000,          40000,        400000},
            array{               450,         2250,          45000,        450000},
            array{               500,         2500,          50000,        500000},
            array{               600,         3000,          60000,        600000},
            array{               700,         3500,          70000,        700000},
            array{               800,         4000,          80000,        800000},
            array{               900,         4500,          90000,        900000},
            array{              1000,         5000,         100000,       1000000},
            array{              1250,         6250,         125000,       1250000},
            array{              1500,         7500,         150000,       1500000},
            array{              1750,         8750,         175000,       1750000},
            array{              2000,        10000,         200000,       2000000},
            array{              2250,        11250,         225000,       2250000},
            array{              2500,        12500,         250000,       2500000},
            array{              2750,        13750,         275000,       2750000},
            array{              3000,        15000,         300000,       3000000},
            array{              3500,        17500,         350000,       3500000},
            array{              4000,        20000,         400000,       4000000},
            array{              4500,        22500,         450000,       4500000},
            array{              5000,        25000,         500000,       5000000}}

# Defines the upgrade price per level for each factory wrapper.
FactoryWrapperPrices : [][]int = 
    # Each column is a wrapper.    0             1               2               3
    # Each row is a level.
    array{  array{                25,          100,           2000,         20000},
            array{                50,          250,           5000,         50000},
            array{               100,          500,          10000,        100000},
            array{               500,         2500,          50000,        250000},
            array{              1000,         5000,         100000,        500000},
            array{              2500,        12500,         250000,       1000000},
            array{              5000,        25000,         500000,       2500000},
            array{             10000,        50000,        1000000,       5000000},
            array{             15000,        75000,        1500000,       7500000},
            array{             20000,       100000,        2000000,       9999999}}

# Defines the upgrade price per level for the factory storage.
FactoryStoragePrices: []int =
    # Each column is a level.
    array{ 10, 25, 50, 100, 250, 500, 1000, 2500, 5000, 10000, 25000, 50000, 100000, 250000, 500000}

# Defines the length of factory belts in BRICK units (knobs).
# If the length of the belts is changed, this value should be updated.
FactoryBeltTransportDistance<public> : float = 60.0

# Defines the base amount of products produced per minute for each factory mould.
FactoryMouldBaseProductsPerMinute : []float = 
    # Each column is a mould.
    array{ 2.0, 1.0, 0.5, 0.25}

# Gets the max level for the factory belt.
GetFactoryBeltMaxLevel<public>(FactoryBeltNumber:int)<decides><transacts>:int=
    FactoryBeltPrices.Length
    
# Gets the max level for the factory mould.
GetFactoryMouldMaxLevel<public>(FactoryMouldNumber:int)<decides><transacts>:int=
    FactoryMouldPrices.Length

# Gets the max level for the factory wrapper.
GetFactoryWrapperMaxLevel<public>(FactoryWrapperNumber:int)<decides><transacts>:int=
    FactoryWrapperPrices.Length

# Gets the max level for the factory storage.
GetFactoryStorageMaxLevel<public>()<decides><transacts>:int=
    FactoryStoragePrices.Length

# Gets the upgrade price of the factory belt for the player.
GetFactoryBeltUpgradePrice<public>(FactoryBeltNumber:int, Player:player)<decides><transacts>:int=
    FactoryBeltLevel := GetFactoryBeltLevel[Player, FactoryBeltNumber]
    FactoryBeltIndex := Min(FactoryBeltLevel, FactoryBeltPrices.Length-1)
    FactoryBeltPrices[FactoryBeltIndex][FactoryBeltNumber]
    
# Gets the upgrade price of the factory mould for the player.
GetFactoryMouldUpgradePrice<public>(FactoryMouldNumber:int, Player:player)<decides><transacts>:int=
    FactoryMouldLevel := GetFactoryMouldLevel[Player, FactoryMouldNumber]
    FactoryMouldIndex := Min(FactoryMouldLevel, FactoryMouldPrices.Length-1)
    FactoryMouldPrices[FactoryMouldIndex][FactoryMouldNumber]

# Gets the upgrade price of the factory wrapper for the player.
GetFactoryWrapperUpgradePrice<public>(FactoryWrapperNumber:int, Player:player)<decides><transacts>:int=
    FactoryWrapperLevel := GetFactoryWrapperLevel[Player, FactoryWrapperNumber]
    FactoryWrapperIndex := Min(FactoryWrapperLevel, FactoryWrapperPrices.Length-1)
    FactoryWrapperPrices[FactoryWrapperIndex][FactoryWrapperNumber]
    
# Gets the upgrade price of the factory storage for the player.
GetFactoryStorageUpgradePrice<public>(Player:player)<decides><transacts>:int=
    FactoryStorageLevel := GetFactoryStorageLevel[Player]
    FactoryStorageIndex := Min(FactoryStorageLevel, FactoryStoragePrices.Length-1)
    FactoryStoragePrices[FactoryStorageIndex]

# Gets the speed of the factory belt for the player.
GetFactoryBeltSpeed<public>(FactoryBeltNumber:int, Player:player)<decides><transacts>:float=
    FactoryBeltLevel := GetFactoryBeltLevel[Player, FactoryBeltNumber]
    # This formula slowly makes the factory belt faster as the level increases.
    # It starts with a value of 0.5 at level 1 and hits a value of 5.0 at level 15.
    FactoryBeltLevel * 0.25 + 0.25
    
# Gets the speed of the factory mould for the player.
GetFactoryMouldSpeed<public>(FactoryMouldNumber:int, Player:player)<decides><transacts>:float=
    FactoryMouldLevel := GetFactoryMouldLevel[Player, FactoryMouldNumber]
    # This formula very slowly makes the factory mould faster as the level increases.
    # It starts with a value of 1.0 at level 1 and hits a value of 30.0 at level 30.
    FactoryMouldLevel * 1.0 + 0.0

# Gets the price multiplier of the factory wrapper for the player.
GetFactoryWrapperMultiplier<public>(FactoryWrapperNumber:int, Player:player)<decides><transacts>:float=
    FactoryWrapperLevel := GetFactoryWrapperLevel[Player, FactoryWrapperNumber]
    # This formula slowly makes the factory wrapper multiplier higher as the level increases.
    # It starts with a value of 1.4 at level 1 and hits a value of 5.0 at level 10.
    FactoryWrapperLevel * 0.4 + 1.0

# Gets the time in seconds it takes for a product to make it from the factory mould to the factory storage for the player.
GetFactoryBeltDeliveryTime<public>(FactoryBeltNumber:int, Player:player)<decides><transacts>:float=
    FactoryBeltSpeed := GetFactoryBeltSpeed[FactoryBeltNumber, Player]
    FactoryBeltTransportDistance / FactoryBeltSpeed

# Gets the time in seconds it takes for the factory mould to produce a product for the player.
GetFactoryMouldProductionTime<public>(FactoryMouldNumber:int, Player:player)<decides><transacts>:float=
    FactoryMouldSpeed := GetFactoryMouldSpeed[FactoryMouldNumber, Player]
    FactoryMouldProductsPerMinute := FactoryMouldSpeed * FactoryMouldBaseProductsPerMinute[FactoryMouldNumber]
    60.0 / FactoryMouldProductsPerMinute

# Gets the factory storage capacity for the player.
GetFactoryStorageCapacity<public>(Player:player)<decides><transacts>:int=
    FactoryStorageLevel := GetFactoryStorageLevel[Player]
    # This formula makes the factory storage capactity grow exponentially.
    # It starts with a value of 10.0 at level 1 and doubles for each level.
    Ceil[10.0 * Pow(2.0, FactoryStorageLevel * 1.0 - 1.0)]
