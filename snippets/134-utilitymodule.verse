# Knowledge-neutral Verse snippet; project names and asset IDs generalized.
# id: snippet-134

# This file defines the utility module that contains a number of extension functions, constants and functions for performing various tasks.

using { /Fortnite.com/Devices }
using { /Fortnite.com/Devices/CreativeAnimation }
using { /UnrealEngine.com/Temporary/SpatialMath }
using { /UnrealEngine.com/Temporary/UI }
using { /Verse.org/Random }
using { /Verse.org/Simulation }

Utility<public> := module:

    # The factor to convert from BRICK units (knobs) to Fortnite units (centimeters)
    BRICKUnitsToFortniteUnits<public> : float = 0.8 * 20.0

    # Extension functions for passing arguments with events.
    # Code taken from: https://dev.epicgames.com/community/snippets/d8k/fortnite-wrapping-subscribe-to-pass-additional-data-to-listeners.
    (Listenable : listenable(agent)).SubscribeAgent<public>(OutputFunc : tuple(agent, t)->void, ExtraData : t where t:type) : cancelable =
        Wrapper := wrapper_agent(t){ExtraData := ExtraData, OutputFunc := OutputFunc}
        Listenable.Subscribe(Wrapper.InputFunc)
        
    wrapper_agent(t : type) := class():
        ExtraData : t
        OutputFunc : tuple(agent, t) -> void
        InputFunc(Agent : agent):void = OutputFunc(Agent, ExtraData)

    (Listenable : listenable(widget_message)).SubscribeWidgetMessage<public>(OutputFunc : tuple(widget_message, t)->void, ExtraData : t where t:type) : cancelable =
        Wrapper := wrapper_widget_message(t){ExtraData := ExtraData, OutputFunc := OutputFunc}
        Listenable.Subscribe(Wrapper.InputFunc)
        
    wrapper_widget_message(t : type) := class():
        ExtraData : t
        OutputFunc : tuple(widget_message, t) -> void
        InputFunc(Message : widget_message):void = OutputFunc(Message, ExtraData)
    
    # Extension function for removing a key from a map.
    (Input:[t]t2 where t:subtype(comparable), t2:type).RemoveKey<public>(KeyToRemove:t)<decides><transacts>:[t]t2=
        var Output:[t]t2 = map{}
        for (Key -> Value : Input, Key <> KeyToRemove):
            set Output = ConcatenateMaps(Output, map{Key => Value})
        Output

    # Extension function for removing all matching elements from an array.
    (Input:[]t where t:type).RemoveAll<public>(Element:t, Compare(A:t, B:t)<decides><transacts>:void)<decides><transacts>:[]t=
        var Output:[]t = array{}
        for(InputElement : Input):
            if (not Compare[InputElement, Element]):
                set Output += array{InputElement}

        Output

    # Functions for localizing text.
    PlainText<public><localizes>(Text:string):message = "{Text}"

    CountText<public><localizes>(Count:int):message = "{Count}"

    LevelText<public><localizes>(Level:int, MaxLevel:int):message = "Level {Level} / {MaxLevel}"

    PlayerText<public><localizes>(PrefixText:string, Player:player, SuffixText:string):message = "{PrefixText}{Player}{SuffixText}"

    # Function for creating a string from a positive number by padding it with a character. 
    PadNumber<public>(PositiveNumber:int, PadToLength:int, Character:char)<transacts>:[]char=
        var Result : []char = "{PositiveNumber}"

        if (Result.Length < PadToLength):
            PadCount := PadToLength - Result.Length
            for(Count := 1..PadCount):
                set Result = array{Character} + Result

        Result

    # Function for rounding float to a certain positive number of decimals.
    # Code taken from: https://forums.unrealengine.com/t/how-to-round-float-to-x-decimal-places/821189/4
    (Input:float).ToString<public>(Decimals:int)<transacts>:string=
        Multiplier := Pow(10.0, Decimals * 1.0)
        if:
            RoundedValue := (Round[Input * Multiplier] * 1.0) / Multiplier
            BeforeDecimal := Floor[RoundedValue]
            AfterDecimal := Abs(Round[(RoundedValue - BeforeDecimal * 1.0) * Multiplier])
            var AfterDecimalString:string = ToString(AfterDecimal)
            
            for(X := 0..Decimals-AfterDecimalString.Length-1):
                set AfterDecimalString = "0" + AfterDecimalString
        then:
            "{BeforeDecimal}.{AfterDecimalString}"
        else:
            ToString(Input)

    # Function for parsing an integer from a string.
    ParseInt<public>(String : string)<decides><transacts>:int=
        String.Length > 0

        var Result : int = 0
        var Digit : int = 1
        for(Index := 0..String.Length-1):
            CharIndex := String.Length-1 - Index
            Char := String[CharIndex]
            case(Char):
                '-' => 
                    set Result *= -1
                '0' =>
                    set Digit *= 10
                '1' =>
                    set Result += 1 * Digit
                    set Digit *= 10
                '2' =>
                    set Result += 2 * Digit
                    set Digit *= 10
                '3' =>
                    set Result += 3 * Digit
                    set Digit *= 10
                '4' =>
                    set Result += 4 * Digit
                    set Digit *= 10
                '5' =>
                    set Result += 5 * Digit
                    set Digit *= 10
                '6' =>
                    set Result += 6 * Digit
                    set Digit *= 10
                '7' =>
                    set Result += 7 * Digit
                    set Digit *= 10
                '8' =>
                    set Result += 8 * Digit
                    set Digit *= 10
                '9' =>
                    set Result += 9 * Digit
                    set Digit *= 10
                '.' =>
                ',' =>
                _ => false?
        Result

    # Function for making a creative_prop shake.
    (Prop:creative_prop).Shake<public>(MaxPositionOffset:float, MaxRotationAngle:float, Count:int, Duration:float)<suspends>:void=
        Prop.Shake(vector3{X:=MaxPositionOffset, Y:=MaxPositionOffset, Z:=MaxPositionOffset}, vector3{X:=MaxRotationAngle, Y:=MaxRotationAngle, Z:=MaxRotationAngle}, Count, Duration)

    # Function for making a creative_prop shake.
    # The max offset and angles are specified in the prop's local space.
    (Prop:creative_prop).Shake<public>(MaxPositionOffset:vector3, MaxRotationYawPitchRollAngles:vector3, Count:int, Duration:float)<suspends>:void=
        if:
            AnimationController := Prop.GetAnimationController[]
        then:
            TimePerShake := Duration / (1.0 * Count)
            var Keyframes : []keyframe_delta = array{}
            Transform := Prop.GetTransform()
            StartPosition := Transform.Translation
            StartRotation := Transform.Rotation
            var PreviousPosition:vector3 = StartPosition
            var PreviousRotation:rotation = StartRotation
            for(Step := 1..Count):
                CurrentAmount := 1.0 - (1.0 * Step) / (1.0 * Count)
                CurrentPositionOffset := 
                    StartRotation.GetLocalForward() * GetRandomFloat(-CurrentAmount, CurrentAmount) * MaxPositionOffset.X +
                    StartRotation.GetLocalRight() * GetRandomFloat(-CurrentAmount, CurrentAmount) * MaxPositionOffset.Y +
                    StartRotation.GetLocalUp() * GetRandomFloat(-CurrentAmount, CurrentAmount) * MaxPositionOffset.Z
                CurrentYaw := GetRandomFloat(-CurrentAmount, CurrentAmount) * MaxRotationYawPitchRollAngles.X
                CurrentPitch := GetRandomFloat(-CurrentAmount, CurrentAmount) * MaxRotationYawPitchRollAngles.Y
                CurrentRoll := GetRandomFloat(-CurrentAmount, CurrentAmount) * MaxRotationYawPitchRollAngles.Z
                CurrentRotationAngle := MakeRotationFromYawPitchRollDegrees(CurrentYaw, CurrentPitch, CurrentRoll)
                CurrentRotation := CurrentRotationAngle.RotateBy(StartRotation)
                CurrentPosition := StartPosition + CurrentPositionOffset
                Keyframe := keyframe_delta:
                    DeltaLocation := CurrentPosition - PreviousPosition
                    DeltaRotation := MakeShortestRotationBetween(PreviousRotation, CurrentRotation)
                    Time := TimePerShake
                    Interpolation := InterpolationTypes.EaseInOut
                set Keyframes += array{Keyframe}
                set PreviousPosition = CurrentPosition
                set PreviousRotation = CurrentRotation
            AnimationController.SetAnimation(Keyframes, ?Mode := animation_mode.OneShot)
            AnimationController.Play()
            Sleep(Duration)
