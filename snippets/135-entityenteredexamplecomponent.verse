# Knowledge-neutral Verse snippet; project names and asset IDs generalized.
# id: snippet-135

using { /Verse.org }
using { /Verse.org/Native }
using { /Verse.org/SceneGraph }
using { /Verse.org/Simulation }
using { /Verse.org/SceneGraph/KeyframedMovement }
using { /Verse.org/SpatialMath }
using { /Fortnite.com/Game }

entity_entered_example_component<public> := class<final_super>(component):

    @editable
    Keyframes:[]keyframed_movement_delta = array{}

    var StartTransform:transform = transform{}

    OnBeginSimulation<override>():void =
        (super:)OnBeginSimulation()
        if:
            FortRoundManager := Entity.GetFortRoundManager[]
        then:
            FortRoundManager.SubscribeRoundStarted(OnRoundStarted)
        

    OnRoundStarted():void =
        set StartTransform = Entity.GetGlobalTransform()
        if:
            Mesh := Entity.GetComponent[mesh_component]
        then:
            spawn {PlayCowAbductionAnimation()}


    PlayCowAbductionAnimation()<suspends>:void =
        if:
            Mesh := Entity.GetComponent[mesh_component]
            MovementComponent := Entity.GetComponent[keyframed_movement_component]
        then:
            loop:
                Entity.SetGlobalTransform(StartTransform)
                set Mesh.Visible = true
                MovementComponent.SetKeyframes(Keyframes, oneshot_keyframed_movement_playback_mode{})
                MovementComponent.Play()
                Mesh.EntityEnteredEvent.Await()
                Sleep(0.2)
                MovementComponent.Stop()
                set Mesh.Visible = false
                Sleep(2.0)
