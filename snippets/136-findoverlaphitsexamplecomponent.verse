# Knowledge-neutral Verse snippet; project names and asset IDs generalized.
# id: snippet-136

using { /Verse.org }
using { /Verse.org/Native }
using { /Verse.org/SceneGraph }
using { /Verse.org/Simulation }
using { /Verse.org/Colors }
using { /Verse.org/SceneGraph/KeyframedMovement }
using { /Verse.org/SpatialMath }
using { /Fortnite.com/Game }
using { /Fortnite.com/Devices }


find_overlaphits_example_component<public> := class<final_super>(component):

    @editable
    Trigger:volume_device = volume_device{}

    var IsAbducting:logic = false
  
    OnBeginSimulation<override>():void =
        (super:)OnBeginSimulation()
        if:
            FortRoundManager := Entity.GetFortRoundManager[]
        then:
            FortRoundManager.SubscribeRoundStarted(OnRoundStarted)

    OnRoundStarted():void=
        Trigger.AgentEntersEvent.Subscribe(OnTriggerEntered)
        Trigger.AgentExitsEvent.Subscribe(OnTriggerExited)

    OnTriggerEntered(Agent:agent):void =
        if:
            not IsAbducting?
            MovementComponent := Entity.GetComponent[keyframed_movement_component]
        then:
            MovementComponent.Pause()
            EnableAbductionBeam()
            PerformOverlapCheck()
            

    OnTriggerExited(Agent:agent):void =
        if:
            not IsAbducting?
            MovementComponent := Entity.GetComponent[keyframed_movement_component]
        then:
            MovementComponent.Play()
            DisableAbductionBeam()

            
    PerformOverlapCheck():void =
        CollisionCapsule := collision_capsule{Radius := 36.0, Length := 328.0}
        var CollisionTransform:transform = Entity.GetGlobalTransform()
        set CollisionTransform.Translation.Up = CollisionTransform.Translation.Up - 164.0
        
        # Perform the overlap check from the entity that contains the mesh_component
        for:
            Overlap : Entity.FindOverlapHits(CollisionTransform, CollisionCapsule)
            # Cast to see if what was overlapped was the Cow
            CowMesh := Meshes.SM_Toy_Cow[Overlap.TargetComponent]
            CowPrefab := CowMesh.Entity
        do:
            spawn { AbductCow(CowPrefab) }


    EnableAbductionBeam():void =
        for:
            Mesh : Entity.FindDescendantComponents(Meshes.S_EV_SimpleLightBeam_01)
        do:
            Mesh.Enable()
        
        for:
            Light : Entity.FindDescendantComponents(spot_light_component)
        do:
            Light.Enable()

        for:
            Audio : Entity.FindDescendantComponents(Sounds.MSS_Ship_Beam)
        do:
            Audio.Play()

    DisableAbductionBeam():void =
        for:
            Mesh : Entity.FindDescendantComponents(Meshes.S_EV_SimpleLightBeam_01)
        do:
            Mesh.Disable()
        
        for:
            Light : Entity.FindDescendantComponents(spot_light_component)
        do:
            Light.Disable()
        for:
            Audio : Entity.FindDescendantComponents(Sounds.MSS_Ship_Beam)
        do:
            Audio.Stop()
            

    AbductCow(CowEntity:entity)<suspends>:void =
        # Get the components on the Cow Prefab
        if:
            CowMesh := CowEntity.GetComponent[mesh_component]
            MovementComponent := CowEntity.GetComponent[keyframed_movement_component]
            SoundComponent := CowEntity.GetComponent[sound_component]
        then:
            set IsAbducting = true

            # Get the delta between Cow and UFO
            DeltaTransform:transform = transform:
                Translation:= Entity.GetGlobalTransform().Translation - CowEntity.GetGlobalTransform().Translation
                Scale := vector3{Left:= 0.0, Up:= 0.0, Forward:= 0.0}
            
            # Create a key frame
            Keyframe := keyframed_movement_delta:
                Transform := DeltaTransform
                Duration := 2.0
                Easing := ease_in_cubic_bezier_easing_function{}

            MovementComponent.SetKeyframes(array{ Keyframe }, oneshot_keyframed_movement_playback_mode{})
            MovementComponent.Play()

            SoundComponent.Play()

            # Wait for Entity Entered Event
            CowMesh.EntityEnteredEvent.Await()
    
            # Remove Cow from world
            CowEntity.RemoveFromParent()
    
            # Resume UFO Patrol
            set IsAbducting = false
            if:
                UFOMovementComponent := Entity.GetComponent[keyframed_movement_component]
            then:
                UFOMovementComponent.Play()
                DisableAbductionBeam()
