# Knowledge-neutral Verse snippet; project names and asset IDs generalized.
# id: snippet-137

using { /Verse.org }
using { /Verse.org/Native }
using { /Verse.org/SceneGraph }
using { /Verse.org/Simulation }
using { /Verse.org/Colors }
using { /Verse.org/SceneGraph/KeyframedMovement }
using { /Verse.org/SpatialMath }
using { /Fortnite.com/Game }
using { /Fortnite.com/Devices }


find_sweephits_example_component<public> := class<final_super>(component):

    @editable
    Trigger:volume_device = volume_device{}

    var IsAbducting:logic = false
  
    OnBeginSimulation<override>():void =
        (super:)OnBeginSimulation()
        if:
            FortRoundManager := Entity.GetFortRoundManager[]
        then:
            FortRoundManager.SubscribeRoundStarted(OnRoundStarted)

    OnRoundStarted():void=
        Trigger.AgentEntersEvent.Subscribe(OnTriggerEntered)
        Trigger.AgentExitsEvent.Subscribe(OnTriggerExited)

    OnTriggerEntered(Agent:agent):void=
        # When a cow is inside the abduction area, stop the ship moving and start the abduction beam.
        if:
            not IsAbducting?
            MovementComponent := Entity.GetComponent[keyframed_movement_component]
        then:
            MovementComponent.Pause()
            EnableAbductionBeam()

            # Perform the sweep from the UFO Mesh
            if (Child := Entity.GetEntities()[0]):
                DisplacementVector := vector3{Left:=0.0, Up:=-300.0, Forward:=0.0}
                FirstSweepHitEntity := FindFirstSweepHit(Child, DisplacementVector)

                # If the First Hit Entity is the Cow Mesh, then abduct the Cow
                if (HitEntity := FirstSweepHitEntity?; HitEntity.GetComponent[Meshes.SM_Toy_Cow]):
                    spawn { AbductCow(HitEntity) }

                    
    OnTriggerExited(Agent:agent):void =
        # Resume UFO patrol if not in the process of abducting a cow.
        if:
            not IsAbducting?
            MovementComponent := Entity.GetComponent[keyframed_movement_component]
        then:
            MovementComponent.Play()
            DisableAbductionBeam()

    
    # Returns the first Entity hit by FindSweepHits
    FindFirstSweepHit(InEntity:entity, DisplacementVector:vector3):?entity =
        for (SweepHit : InEntity.FindSweepHits(DisplacementVector)):
            return option{ SweepHit.TargetComponent.Entity }
        return false


    EnableAbductionBeam():void =
        for:
            Mesh : Entity.FindDescendantComponents(Meshes.S_EV_SimpleLightBeam_01)
        do:
            Mesh.Enable()
        
        for:
            Light : Entity.FindDescendantComponents(spot_light_component)
        do:
            Light.Enable()
            
        for:
            Audio : Entity.FindDescendantComponents(Sounds.MSS_Ship_Beam)
        do:
            Audio.Play()

    DisableAbductionBeam():void =
        for:
            Mesh : Entity.FindDescendantComponents(Meshes.S_EV_SimpleLightBeam_01)
        do:
            Mesh.Disable()
        
        for:
            Light : Entity.FindDescendantComponents(spot_light_component)
        do:
            Light.Disable()
        for:
            Audio : Entity.FindDescendantComponents(Sounds.MSS_Ship_Beam)
        do:
            Audio.Stop()
            

    AbductCow(CowEntity:entity)<suspends>:void =
        if:
            # Get the components on the Cow Prefab
            CowMesh := CowEntity.GetComponent[mesh_component]
            MovementComponent := CowEntity.GetComponent[keyframed_movement_component]
            SoundComponent := CowEntity.GetComponent[sound_component]
        then:
            set IsAbducting = true
    
            # Get the delta between cow and UFO
            DeltaTransform := transform:
                Translation:= Entity.GetGlobalTransform().Translation - CowEntity.GetGlobalTransform().Translation
                Scale := vector3{ Left := 0.0, Up := 0.0, Forward := 0.0 }
    
            # Create a key frame
            Keyframe := keyframed_movement_delta:
                Transform := DeltaTransform
                Duration := 2.0
                Easing := ease_in_cubic_bezier_easing_function{}
    
            MovementComponent.SetKeyframes(array{ Keyframe }, oneshot_keyframed_movement_playback_mode{})
            MovementComponent.Play()

            SoundComponent.Play()
    
            # Wait for Entity Entered Event
            CowMesh.EntityEnteredEvent.Await()
    
            # Remove Cow from world
            CowEntity.RemoveFromParent()
    
            # Resume UFO Patrol
            set IsAbducting = false
            if:
                UFOMovementComponent := Entity.GetComponent[keyframed_movement_component]
            then:
                UFOMovementComponent.Play()
                DisableAbductionBeam()
