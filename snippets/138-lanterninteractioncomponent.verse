# Knowledge-neutral Verse snippet; project names and asset IDs generalized.
# id: snippet-138

using { /Verse.org }
using { /Verse.org/Native }
using { /Verse.org/SceneGraph }
using { /Verse.org/Simulation }

LightPost := module:
    Materials<public> := module:

lantern_interaction_component<public> := class<final_super>(component):

    var MaterialInstance:LightPost.Materials.MI_Lantern_01 = LightPost.Materials.MI_Lantern_01{}
    var IsLightOn:logic = true

    OnBeginSimulation<override>():void =
        (super:)OnBeginSimulation()

        InteractabeleComponents := Entity.FindDescendantComponents(interactable_component)
        for (InteractableComponent : InteractabeleComponents):
            InteractableComponent.SucceededEvent.Subscribe(OnInteractFinished)

        MeshComponents := Entity.FindDescendantComponents(LightPost.SM_Lightpost_Lantern_01)
        for (MeshComponent : MeshComponents):
            set MeshComponent.M_Lantern = MaterialInstance


    OnInteractFinished(Agent:agent):void =
        if (IsLightOn?):
            TurnLightOff()
        else:
            TurnLightOn()

    
    TurnLightOn():void =
        set IsLightOn = true
        set MaterialInstance.Emissive_Multiply = 1.0
        for:
            Light : Entity.FindDescendantComponents(light_component)
        do:
            Light.Enable()

        for:
            Particle : Entity.FindDescendantComponents(particle_system_component)
        do:
            Particle.Play()

        for:
            Audio : Entity.FindDescendantComponents(Sounds.MSS_LightBuzz_Loop)
        do:
            Audio.Play()

        for:
            Audio : Entity.FindDescendantComponents(Sounds.MSS_LightBuzzMoths_Loop)
        do:
            Audio.Play()

        for:
            Audio : Entity.FindDescendantComponents(Sounds.MSS_LightBuzz_Ignite)
        do:
            Audio.Play()


    TurnLightOff():void = 
        set IsLightOn = false
        set MaterialInstance.Emissive_Multiply = 0.0
        for:
            Light : Entity.FindDescendantComponents(light_component)
        do:
            Light.Disable()

        for:
            Particle : Entity.FindDescendantComponents(particle_system_component)
        do:
            Particle.Stop()

        for:
            Audio : Entity.FindDescendantComponents(Sounds.MSS_LightBuzz_Loop)
        do:
            Audio.Stop()

        for:
            Audio : Entity.FindDescendantComponents(Sounds.MSS_LightBuzzMoths_Loop)
        do:
            Audio.Stop()
