# Knowledge-neutral Verse snippet; project names and asset IDs generalized.
# id: snippet-139

using { /Verse.org }
using { /Verse.org/Native }
using { /Verse.org/SceneGraph }
using { /Verse.org/Simulation }
using { /Verse.org/SceneGraph/KeyframedMovement }
using { /Fortnite.com/Game }

movement_mode<public> := enum:
    OneShot
    Loop
    PingPong

simple_movement_component<public> := class<final_super>(component):

    @editable
    var Keyframes: []keyframed_movement_delta = array{}

    @editable
    var AutoPlay: logic = true

    @editable
    MovementMode: movement_mode = movement_mode.Loop

    OnBeginSimulation<override>():void =
        (super:)OnBeginSimulation()
        if:
            KeyframedMovementComponent := Entity.GetComponent[keyframed_movement_component]
        then:
            InitializeKeyframedMovementComponent(KeyframedMovementComponent)
        else:
            NewKeyFramedMovementComponent := keyframed_movement_component { Entity := Entity }
            Entity.AddComponents(array{ NewKeyFramedMovementComponent })
            InitializeKeyframedMovementComponent(NewKeyFramedMovementComponent)

        if:
            FortRoundManager := Entity.GetFortRoundManager[]
        then:
            FortRoundManager.SubscribeRoundStarted(OnRoundStarted)


    InitializeKeyframedMovementComponent(InKeyframedMovementComponent:keyframed_movement_component):void =
        var PlaybackMode:keyframed_movement_playback_mode = oneshot_keyframed_movement_playback_mode{}
        
        case (MovementMode):
            movement_mode.OneShot =>
                set PlaybackMode = oneshot_keyframed_movement_playback_mode{}
            movement_mode.Loop =>
                set PlaybackMode =  loop_keyframed_movement_playback_mode{}
            movement_mode.PingPong =>
                set PlaybackMode = pingpong_keyframed_movement_playback_mode{}

        InKeyframedMovementComponent.SetKeyframes(Keyframes, PlaybackMode)
    
    OnRoundStarted()<suspends>:void =
        if:
            AutoPlay?
            KeyframedMovementComponent := Entity.GetComponent[keyframed_movement_component]
        then:
            KeyframedMovementComponent.Play()
