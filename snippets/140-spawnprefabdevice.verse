# Knowledge-neutral Verse snippet; project names and asset IDs generalized.
# id: snippet-140

using { /Fortnite.com/Devices }
using { /Verse.org/Simulation }
using { /Verse.org/SceneGraph }
using { /Verse.org/SpatialMath }
using { /UnrealEngine.com/Temporary/Diagnostics }

SpawnPrefabDevice := class(creative_device):

    @editable
    TriggerVolume:volume_device = volume_device{}

    OnBegin<override>()<suspends>:void =
        SpawnTransform:transform = transform:
            Translation := vector3:
                Left := -9084.0
                Up := -8.0
                Forward := -919.0

        if:
            SimulationEntity := GetSimulationEntity[]
        then:
            loop:
                SpawnedPrefab:entity = LightPost.P_LightPost{}
                TriggerVolume.AgentEntersEvent.Await()
                SimulationEntity.AddEntities(array{SpawnedPrefab})
                SpawnedPrefab.SetGlobalTransform(SpawnTransform)

                TriggerVolume.AgentExitsEvent.Await()
                SpawnedPrefab.RemoveFromParent()
