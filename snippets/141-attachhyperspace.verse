# Knowledge-neutral Verse snippet; project names and asset IDs generalized.
# id: snippet-141

using { /Fortnite.com/Devices }
using { /Verse.org/Simulation }
using { /UnrealEngine.com/Temporary/Diagnostics }
using { /Fortnite.com/Characters }
# using { /Verse.org/SpatialMath }
using { /UnrealEngine.com/Temporary/SpatialMath }

# A Verse-authored creative device that can be placed in a level
hyperspace_sequence_device := class(creative_device):
    
    @editable HyperspacePlane : creative_prop = creative_prop{}
    @editable VolumeStart : volume_device = volume_device{}
    
    # Cinematics to control vfx
    @editable BeginHyperspace : cinematic_sequence_device = cinematic_sequence_device{}
    @editable EndHyperspace : cinematic_sequence_device = cinematic_sequence_device{}
    @editable HyperspaceDuration : float = 6.0
    
    # Control data layer that contains hangar
    @editable DataLayerController : cinematic_sequence_device = cinematic_sequence_device{}
    
    # Get player rotation when entering volume
    var StartingRotation : rotation = IdentityRotation()
    
    # This controls the offset of the Hyperspace plane from the position of the player.
    # x left/right  y front/back  z up /down
    @editable AttachmentOffset : vector3 = vector3{X := 0.0, Y := -1500.0, Z := 700.0}
       
    GetAgent(Agent : agent): void=
        spawn{RunHyperspaceSequence(Agent)}

    RunHyperspaceSequence(Agent : agent)<suspends>:void=
        # This attaches the hyperspace plane to a position in front of the player and continually updates the position.
        if(FortCharacter := Agent.GetFortCharacter[]):
            # Get initial player orientation
            set StartingRotation = FortCharacter.GetTransform().Rotation
            # Play BeginHyperspace level sequence
            BeginHyperspace.Play()
            # Hide data layer containing the Hangar
            DataLayerController.Play()
            race:
                # Continually update the attach position
                loop:
                    Sleep(0.01)
                    # get player transform and add attachment offset to get position in front of player.
                    PlayerTransform := FortCharacter.GetTransform()
                    ConvertedAttachLocation := PlayerTransform.Translation + AttachmentOffset
                    HyperspacePlane.MoveTo(ConvertedAttachLocation, StartingRotation, 0.1)

                # TimeOut
                Sleep(HyperspaceDuration)

            BeginHyperspace.Stop()
            EndHyperspace.Play()
            # length of HyperspaceExit NS
            Sleep(3.0)
            EndHyperspace.Stop()

    # Runs when the device is started in a running game
    OnBegin<override>()<suspends>:void=
        VolumeStart.AgentEntersEvent.Subscribe(GetAgent)
