# Knowledge-neutral Verse snippet; project names and asset IDs generalized.
# id: snippet-142

# A divide-and-conquer sorting algorithm that divides an array into two, sorts each divided array, and then merges the arrays together.
# This is a recursive implementation, where the function calls itself to merge sort the divided arrays.
# The base case (the condition to stop the recursion) is the array has only one element.
# This is a generic implementation using parametric types, so you can provide your own type and your own comparison function as arguments.
MergeSort<public>(Array:[]t, Compare(L:t, R:t)<decides><transacts>:t where t:type)<transacts>:[]t=
    Length:int = Array.Length

    if:
        Length > 1 # Verify there is more than one element in the array, otherwise we've reached the base case.
        Mid:int = Floor(Length / 2) # Get the middle index of the array.
        Left:[]t = Array.Slice[0, Mid] # Split the array in half. This keeps elements from the beginning to Mid - 1 index.
        Right:[]t = Array.Slice[Mid] # Split the array in half. This keeps elements from Mid index to the end of the array.
    then:
        # Call MergeSort on the left half of the array.
        LeftSorted:[]t = MergeSort(Left, Compare)
        # Call MergeSort on the right half of the array.
        RightSorted:[]t = MergeSort(Right, Compare)

        # Combine the two arrays and return the result.
        Merge(LeftSorted, RightSorted, Compare)
    else:
        # Return the array passed in because we've reached the base case.
        Array

# A helper function for MergeSort that combines the divided arrays in an order based on the Compare function.
Merge(Left:[]t, Right:[]t, Compare(L:t, R:t)<decides><transacts>:t where t:type)<transacts>:[]t=
    var LeftIndex:int = 0
    var RightIndex:int = 0
    var MergedArray:[]t = array{}
    
    # Loop through all the elements in the arrays to add them to the MergedArray variable.
    loop:
        if (LeftElement := Left[LeftIndex], RightElement := Right[RightIndex]):
            # Check the element in the left half array with the element in the right half array.
            # Uses the Compare function passed in as an argument
            if (Compare[LeftElement, RightElement]):
                set MergedArray += array{LeftElement}
                set LeftIndex += 1
            else:
                set MergedArray += array{RightElement}
                set RightIndex += 1
        else:
            # We've added all of the elements from one the arrays at this point.
            # Now check which array still has elements to merge in and add all remaining elements.
            if (LeftIndex >= Left.Length):
                option{set MergedArray += Right.Slice[RightIndex]}
            else:
                option{set MergedArray += Left.Slice[LeftIndex]}

            # Exit out of the loop because we've finished adding all elements.
            break
        
    # Return the merged array.
    MergedArray
