# Knowledge-neutral Verse snippet; project names and asset IDs generalized.
# id: snippet-143

# This file contains the code that updates the billboards, player references, and UI on the island
# to display a player's stats from their player stats table. It also handles adding wins and points to a
# player's stats table, and setting their lifetime best lap time.

using { /Fortnite.com/Devices }
using { /Verse.org/Simulation}
using { SortingAlgorithms }
using { PlayerStatistics }

# The message to display on the stats billboard.
StatsMessage<localizes>(CurrentPlayer:message, Points:message, Podiums:message, BestLapTime:message):message=
    "{CurrentPlayer}:\n{Points}\n{Podiums}\n{BestLapTime}"
PlayerText<localizes>(CurrentPlayer:agent):message = "Player {CurrentPlayer}"
PointsText<localizes>(Points:int):message = "Total Points {Points}"
PodiumsText<localizes>(Podiums:int):message = "{Podiums} Times on the Podium"
BestLapTimeText<localizes>(BestLapTime:float):message = "Best Lap {BestLapTime}"
NoLapTimeText<localizes>:message = "Best Lap N/A"

# Updates the given billboard device to display the stats of the given player.
UpdateStatsBillboard<public>(Player:agent, StatsBillboard:billboard_device):void=
    if:
        CurrentPlayerStats := GetPlayerStats[Player]
    then:
        BestLapText:message = if(IsValidBestLapTime[CurrentPlayerStats.BestLapTime]):
            BestLapTimeText(CurrentPlayerStats.BestLapTime)
        else:
            NoLapTimeText
        PlayerStatsText := StatsMessage(
            PlayerText(Player),
            PointsText(CurrentPlayerStats.Points),
            PodiumsText(CurrentPlayerStats.Podiums),
            BestLapText)
        StatsBillboard.SetText(PlayerStatsText)

# Updates the Popup UI to display the lifetime stats of the given player during the
# race starting sequence.
UpdatePopupUI<public>(Player:agent, PopupDialogUI:popup_dialog_device):void=
    if:
        CurrentPlayerStats := GetPlayerStats[Player]
    then:
        PopupDialogUI.SetButtonText(PlayerText(Player), 0)
        PopupDialogUI.SetButtonText(PointsText(CurrentPlayerStats.Points), 1)
        PopupDialogUI.SetButtonText(PodiumsText(CurrentPlayerStats.Podiums), 2)
        BestLapText:message = if(IsValidBestLapTime[CurrentPlayerStats.BestLapTime]):
            BestLapTimeText(CurrentPlayerStats.BestLapTime)
        else:
            NoLapTimeText
        PopupDialogUI.SetButtonText(BestLapText, 3)

# Update the stats billboards by sorting them based on the amount of lifetime points
# each player has.
UpdateStatsBillboards<public>(Players:[]agent, StatsBillboards:[]billboard_device):void=
    var PlayerAndStatsArray:[]player_and_stats =
        for (Agent:Players, RacePlayer := player[Agent], PlayerStats := GetPlayerStats[RacePlayer]):
            player_and_stats:
                Player := RacePlayer
                StatsTable := PlayerStats

    # Compare and sort players based on their total points, which is the overall measure
    # of "Best" player in the lobby. You can swap out the comparison function here to fit
    # the needs of your experience.
    SortedPlayersAndStats := SortingAlgorithms.MergeSort(
        PlayerAndStatsArray,
        MorePointsComparison)

    for(PlayerIndex -> PlayerAndStats : SortedPlayersAndStats, StatsBillboard := StatsBillboards[PlayerIndex]):
        UpdateStatsBillboard(PlayerAndStats.Player, StatsBillboard)

# Update the player references devices by sorting them based on the amount
# of lifetime points each player has.
UpdatePlayerReferences<public>(Players:[]player, PlayerReferences:[]player_reference_device):void=
    var PlayerAndStatsArray:[]player_and_stats =
        for (Agent:Players, RacePlayer := player[Agent], PlayerStats := GetPlayerStats[RacePlayer]):
            player_and_stats:
                Player := RacePlayer
                StatsTable := PlayerStats
   
    # Compare and sort players based on their total points, which is the overall measure
    # of "Best" player in the lobby. You can swap out the comparison function here to fit
    # the needs of your experience.
    SortedPlayersAndStats := SortingAlgorithms.MergeSort(
        PlayerAndStatsArray,
        MorePointsComparison)
   
    for(PlayerIndex -> PlayerAndStats : SortedPlayersAndStats, PlayerReference := PlayerReferences[PlayerIndex]):
        PlayerReference.Register(PlayerAndStats.Player)
