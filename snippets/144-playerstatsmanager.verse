# Knowledge-neutral Verse snippet; project names and asset IDs generalized.
# id: snippet-144

# This file handles the code for initializing, updating, and returning player_stats_tables
# for each player. It also defines an abstract stat_type class to use for updating stats, and the
# StatType module to use when displaying stats.

using { /Fortnite.com/Devices }
using { /Verse.org/Simulation }
using { /UnrealEngine.com/Temporary/Diagnostics }

# Return the player_stats_table for the provided Agent.
GetPlayerStats<public>(Agent:agent)<decides><transacts>:player_stats_table=
    var PlayerStats:player_stats_table = player_stats_table{}
    if:
        Player := player[Agent]
        Player.IsActive[]
        PlayerStatsTable := PlayerStatsMap[Player]
        set PlayerStats = MakePlayerStatsTable(PlayerStatsTable)
    PlayerStats

# Initialize stats for all current players.
InitializeAllPlayerStats<public>(Players:[]player):void =
    for (Player : Players):
        InitializePlayerStat(Player)

# Initialize stats for the given player.
InitializePlayerStat<public>(Player:player):void=
    if:
        Player.IsActive[]
        not PlayerStatsMap[Player]
        set PlayerStatsMap[Player] = player_stats_table{}
    else:
        Test.ProjectLog("Unable to initialize player stats")
        
# Create a new player_stats_table with the updated losses,
# then set that table in the PlayerStatsMap.
RecordPlayerBestLap<public>(Agent:agent, BestLapTime:float):void=
    if:
        Player := player[Agent]
        Player.IsActive[]
        PlayerStatsTable := PlayerStatsMap[Player]

        set PlayerStatsMap[Player] = player_stats_table:
            BestLapTime := BestLapTime
            MakePlayerStatsTable<constructor>(PlayerStatsTable)
    else:
        Test.ProjectLog("Unable to record player best lap")

        # Adds to the given Agent's StatToAdd and updates both their stats table
# in PlayerStatsManager and the billboard in the level.
AddPoints<public>(Agent:agent, NewPoints:int):void=
    if:
        Player := player[Agent]
        Player.IsActive[]
        PlayerStatsTable := PlayerStatsMap[Player]
        CurrentPoints := PlayerStatsTable.Points

        set PlayerStatsMap[Player] = player_stats_table:
            Points := CurrentPoints + NewPoints
            MakePlayerStatsTable<constructor>(PlayerStatsTable)
    else:
        Test.ProjectLog("Unable to record player points")

# Adds to the given Agent's StatToAdd and updates both their stats table
# in PlayerStatsManager and the billboard in the level.
AddWin<public>(Agent:agent,  NewPodiums:int):void=
    if:
        Player := player[Agent]
        Player.IsActive[]
        PlayerStatsTable := PlayerStatsMap[Player]
        CurrentPodiums := PlayerStatsTable.Podiums

        set PlayerStatsMap[Player] = player_stats_table:
            Podiums := CurrentPodiums + NewPodiums
            MakePlayerStatsTable<constructor>(PlayerStatsTable)

    else:
        Test.ProjectLog("Unable to record player podiums")

# Sets the given Agent's BestLapTime in PlayerStatsManager if the
# LapTime is less than the Agent's current BestLapTime
SetLapTime<public>(Agent:agent, LapTime:float):void=
    if:
        Player := player[Agent]
        Player.IsActive[]
        PlayerStatsTable := PlayerStatsMap[Player]
        CurrentBestLapTime := PlayerStatsTable.BestLapTime
        LapTime < CurrentBestLapTime or not IsValidBestLapTime[CurrentBestLapTime]
        IsValidBestLapTime[LapTime]

        set PlayerStatsMap[Player] = player_stats_table:
            BestLapTime := LapTime
            MakePlayerStatsTable<constructor>(PlayerStatsTable)
    else:
        Test.ProjectLog("Did not record player best lap")
