# Knowledge-neutral Verse snippet; project names and asset IDs generalized.
# id: snippet-145

# This file defines a player_stats_table, a collection of persistable player statistics.
# It also contains functions to compare stats tables by each of the stats to order players
# when sorting.

using { /Fortnite.com/Devices }
using { /Verse.org/Simulation }
using { /UnrealEngine.com/Temporary/Diagnostics }

# Structure for passing a player and their stats as arguments.
player_and_stats<public> := struct:
    Player<public>:player
    StatsTable<public>:player_stats_table

# Tracks different persistable stats for each player.
player_stats_table<public>:= class<final><persistable>:
    # The version of the current stats table.
    Version<public>:int = 0
   
    # The Points of a player.
    Points<public>:int = 0

    # The number of Podiums for a player.
    Podiums<public>:int = 0

    # The best lap time for a player (lower is better).
    BestLapTime<public>:float = -1.0

# Returns true if the lap time is valid (greater than 0).
IsValidBestLapTime<public>(BestLapTime:float)<decides><transacts>:void=
    BestLapTime > 0.0

# Returns Left if Left has greater Points than Right.
MorePointsComparison<public>(Left:player_and_stats, Right:player_and_stats)<decides><transacts>:player_and_stats=
    Left.StatsTable.Points > Right.StatsTable.Points
    Left

# Returns Left if Left has less Points than Right.
LessPointsComparison<public>(Left:player_and_stats, Right:player_and_stats)<decides><transacts>:player_and_stats=
    Left.StatsTable.Points < Right.StatsTable.Points
    Left

# Returns Left if Left has a greater number of Podiums than Right.
MorePodiumsComparison<public>(Left:player_and_stats, Right:player_and_stats)<decides><transacts>:player_and_stats=
    Left.StatsTable.Points > Right.StatsTable.Points
    Left

# Returns Left if Left has a lesser number of Podiums than Right.
LessPodiumsComparison<public>(Left:player_and_stats, Right:player_and_stats)<decides><transacts>:player_and_stats=
    Left.StatsTable.Points < Right.StatsTable.Points
    Left

# Returns Left if Left has a slower BestLapTime time than Right.
# Note this is backwards from the other stats since a lower lap time is better.
SlowerLapTimeComparison(Left:player_and_stats, Right:player_and_stats)<decides><transacts>:player_and_stats=
    Left.StatsTable.Points > Right.StatsTable.Points
    Left

# Returns Left if Left faster BestLapTime than Right.
# Note this is backwards from the other stats since a lower lap time is better.
FasterLapTimeComparison(Left:player_and_stats, Right:player_and_stats)<decides><transacts>:player_and_stats=
    Left.StatsTable.Points < Right.StatsTable.Points
    Left

# Creates a new player_stats_table with the same values as the previous player_stats_table.
MakePlayerStatsTable<constructor>(OldTable:player_stats_table)<transacts> := player_stats_table:
    Version := OldTable.Version
    Points := OldTable.Points
    Podiums := OldTable.Podiums
    BestLapTime := OldTable.BestLapTime
   
# Maps players to a table of their player stats.
var PlayerStatsMap:weak_map(player, player_stats_table) = map{}
