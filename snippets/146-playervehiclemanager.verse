# Knowledge-neutral Verse snippet; project names and asset IDs generalized.
# id: snippet-146

# This file contains functions that assign players to a given vehicle to put players
# in cars when the race begins.

using { /Fortnite.com/Devices }
using { /Verse.org/Simulation }

# Assigns each agent in the InPlayers array to the the vehicle at the matching index in the InVehicles
# array by calling AssignPlayerToVehicle on each.
AssignPlayersToVehicles<public>(Players:[]agent, Vehicles:[]vehicle_spawner_device):void=
    for:
        Index -> Player : Players
        Vehicle := Vehicles[Index]
    do:
        AssignPlayerToVehicle(Player, Vehicle)

# Assigns the given agent to the given vehicle.
AssignPlayerToVehicle<public>(Player:agent, Vehicle:vehicle_spawner_device):void=
    Vehicle.AssignDriver(Player)
