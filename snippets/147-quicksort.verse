# Knowledge-neutral Verse snippet; project names and asset IDs generalized.
# id: snippet-147

# A divide-and-conquer sorting algorithm that divides an array into subarrays by selecting a pivot element.
# Elements less than the pivot are sorted to left of the pivot, and elements greater than the pivot are sorted to the right.
# The algorithm then divides into subarrays on either side of the pivot, and continues the process, selecting a new pivot each time.
# At the end each of the sorted subarrays is merged together to get a fully sorted array.
# This is a recursive implementation, where the function calls itself to quick sort the divided arrays.
# The base case (the condition to stop the recursion) is the array has only one element.
# This implementation always selects for the rightmost element as the initial pivot. 
# This is a generic implementation using parametric types, so you can provide your own type and your own comparison function as arguments.
QuickSort<public>(Array:[]t, Compare(L:t, R:t)<decides><transacts>:t where t:type)<transacts>:[]t=
    QuickSortAtPivot(Array, 0, Array.Length - 1, Compare)

QuickSortAtPivot<public>(Array:[]t, LowIndex:int, HighIndex:int, Compare(L:t, R:t)<decides><transacts>:t where t:type)<transacts>:[]t=
    if:
        # Check if the length of the array is greater than one. If not we've
        # reached the base case and should stop recursion.
        LowIndex < HighIndex
        # Sort the elements in the array by using the HighIndex (rightmost) element
        # as the Pivot. Return the index of the Pivot after sorting along with the sorted array.
        PartitionTuple := Partition(Array, LowIndex, HighIndex, Compare)
        PivotIndex := PartitionTuple(0)
        SortedArray := PartitionTuple(1)
        # Split the array into subarrays left and right of the pivot value.
        LeftHalf := SortedArray.Slice[LowIndex, PivotIndex]
        PivotVal := array{SortedArray[PivotIndex]}
        RightHalf := SortedArray.Slice[PivotIndex + 1]
    then:
        # Recursively sort the Left and Right subarrays.
        LeftArray := QuickSortAtPivot(LeftHalf, LowIndex, PivotIndex - 1, Compare)
        RightArray := QuickSortAtPivot(RightHalf, 0, RightHalf.Length - 1, Compare)

        # Rejoin the the Left and Right subarrays with the Pivot value.
        LeftArray + PivotVal + RightArray
    else:
        Array

# A helper function for QuickSort that swaps elements by comparing them to the element at the Pivot. Elements less than the Pivot
# are sorted to the left, and elements greater than the pivot are sorted to the right.
# This algorithm uses two index values, SmallerIndex and LargerIndex. The LargerIndex is the index of the element to
# compare to the Pivot value, and the SmallerIndex is the index of the element to swap with the Pivot when the algorithm completes.    
Partition(Array:[]t, LowIndex:int, HighIndex:int, Compare(L:t, R:t)<decides><transacts>:t where t:type)<transacts>:tuple(int, []t)=
    # Index of the smaller element, 
    # right-side position of the pivot.
    var SmallerIndex:int := LowIndex - 1
    var TempArray:[]t = Array
    if:
        # The Pivot, the item at the HighIndex (rightmost) index in the array.
        var Pivot:t := TempArray[HighIndex]
    then:
        # The index of the element to compare against the Pivot.
        var LargerIndex:int := LowIndex
        loop:
            if:
                # Check if we've reached the end of the array.
                LargerIndex > HighIndex
            then:
                break
            else:
                if:
                    # The element at LargerIndex, which will be compared to the Pivot value.
                    ComparedVal := TempArray[LargerIndex]
                    
                    # If the element at LargerIndex is less than the pivot value,
                    # swap the elements at SmallerIndex and LargerIndex
                    Compare[ComparedVal, Pivot]
                then:
                    set SmallerIndex = SmallerIndex + 1
                    if:
                        StoredVal := TempArray[SmallerIndex]
                        # Swap the elements at SmallerIndex and LargerIndex.
                        Swap1 := TempArray.ReplaceElement[SmallerIndex, ComparedVal]
                        Swap2 := Swap1.ReplaceElement[LargerIndex, StoredVal]
                        set TempArray = Swap2
                    else:
                        Test.ProjectLog("Unable to swap elements at smaller and larger indices during sorting.")
            # Increment the LargerIndex each iteration of the loop.
            set LargerIndex = LargerIndex + 1  
        # Finally, swap the element at SmallerIndex with the element at the Pivot.
        if:
            set SmallerIndex = SmallerIndex + 1
            # Save a reference to the elements at SmallerIndex and HighIndex.
            StoredVal := TempArray[SmallerIndex]
            PivotVal := TempArray[HighIndex]
            # Swap the elements at SmallerIndex and HighIndex.
            Swap3 := TempArray.ReplaceElement[SmallerIndex, PivotVal]
            Swap4 := Swap3.ReplaceElement[HighIndex, StoredVal]
            set TempArray = Swap4
        else:
            Test.ProjectLog("Unable to swap the element at the smaller and high indices during sorting.")

    # Return the index of the new Pivot and the modified array.
    (SmallerIndex, TempArray)
