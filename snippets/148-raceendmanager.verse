# Knowledge-neutral Verse snippet; project names and asset IDs generalized.
# id: snippet-148

# This file contains functions related to when a player finishes the race, such as updating
# their stats and awarding them points based on their placement. It also handles what happens when
# a player leaves the game, and updates player reference devices in the level whenever a player passes a checkpoint.

using { /Fortnite.com/Devices }
using { /Fortnite.com/Game }
using { /Verse.org/Simulation }
using { /UnrealEngine.com/Temporary/Diagnostics }
using { PlayerStatistics }
using { PlayerLeaderboard }

# Displays a player's stats on a billboard.
StatsMessage<localizes>(CurrentPlayer:agent, Points:int, Podiums:int, BestLapTime:float):message = "{CurrentPlayer}'s Stats:\n Points: {Points}\n Podiums: {Podiums}\n BestLapTime: {BestLapTime}\n"
   
# A Verse-authored creative device that can be placed in a level
race_end_manager := class(creative_device):

    # An array of each of the checkpoint devices in the level.
    @editable
    Checkpoints:[]race_checkpoint_device = array{}

    # The Maximum Number of Players in the experience.
    # Used for Points calculation to decrease points earned
    # when racing with less than the MaxNumberOfPlayers.
    @editable
    MaxNumberOfPlayers:int = 8

    # Tracks when players complete a race, with the players in the first 3 spots being awarded a win.
    @editable
    RaceManager:race_manager_device = race_manager_device{}

    # Displays the player's stats from the player_stats_table.
    @editable
    StatsBillboards:[]billboard_device = array{}

    # The placement of a player must be at or below to award a win.
    @editable
    PlacementRequiredForWin:int = 3

    # The number of points a player in each place earns.
    # Adjust this to award your players the desired amount of score
    # based on their placement.
    @editable
    PointsPerPlace:[]int = array{10, 8, 6, 5, 4, 3, 2, 1}

    # The timer device that tracks when a race starts and records time when a player finishes the race.
    @editable
    RaceTimer:timer_device = timer_device{}

    # The timer device that counts down the end of the race after the first player finishes.
    @editable
    RaceEndTimer:timer_device = timer_device{}

    # Ends the game when the last player finishes the race.
    @editable
    EndGameDevice:end_game_device = end_game_device{}

    # Awards players score when they pass a checkpoint.
    @editable
    ScoreManager:score_manager_device = score_manager_device{}
    
    # The number of players in the current race.
    var NumPlayersThisRace:int = 8

    # The spot of the player who just finished the race.
    # The first three players to finish the race will be awarded a win.
    var CurrentFinishOrder:int = 0

    # Runs when the device is started in a running game
    OnBegin<override>()<suspends>:void=
        GetPlayspace().PlayerRemovedEvent().Subscribe(OnPlayerRemoved)

        # Get the players in the current race and create a player_stat_table
        # for each of them.
        Players := GetAllValidPlayers(GetPlayspace().GetPlayers())
        set NumPlayersThisRace = Players.Length
        InitializeAllPlayerStats(Players)
        UpdateStatsBillboards(Players, StatsBillboards)
       
        # Wait for the race to start.
        RaceManager.RaceBeganEvent.Await()

        # Start the race timer for each player.
        for (Player : Players):
            RaceTimer.Start(Player)

        spawn{WaitForPlayerToFinishLap()}

        # Wait for all players to finish the race.
        ArraySync(Players, WaitForPlayerToFinishRace)

        # When the race ends, end the game using the first place finisher.
        if:
            OrderedPlayers := GetPlayersByFinishOrderInLastRound(Players)
            FirstPlacePlayer := OrderedPlayers[0]
        then:
            EndGameDevice.Activate(FirstPlacePlayer)

    # When a player is removed from the race, reset their circuit to prevent their
    # stats from showing up on billboards and player references.
    OnPlayerRemoved(Player:player):void=
        # Reset circuit info when player leaves the game.
        if:
            ResetCircuitInfo[Player]
        else:
            Test.ProjectLog("Unable to reset circuit info for player")

        # Remove player's car when they leave.
        if:
            Info := RoundInfo[GetSession()]
            StartingPosition := Info.PlayerStartPositions[Player]
        then:
            StartingPosition.Car.Disable()

    # When a player finishes a lap or the race, calculate their lap time and pass it to
    # SetLapTime. Then reset the timer for that player to record their next lap time.
    WaitForPlayerToFinishLap()<suspends>:void=
        loop:
            LapPlayer := race:
                RaceManager.LapCompletedEvent.Await()
                RaceManager.RaceCompletedEvent.Await()

            # Calculate a player's lap time.
            ActiveTime := RaceTimer.GetActiveDuration(LapPlayer)
            MaxTime := RaceTimer.GetMaxDuration()
            PlayerTime := MaxTime - ActiveTime

            # Pass the lap time to player_leaderboards and reset the time.
            SetLapTime(LapPlayer, PlayerTime)
            RaceTimer.Reset(LapPlayer)
            RaceTimer.Start(LapPlayer)

    # When a player finishes the race, record a finish in their stats table.
    WaitForPlayerToFinishRace(Player:agent)<suspends>:void=
        race:
            # Waiting for this player to finish the race and then record the finish.
            loop:
                FinishingPlayer := RaceManager.RaceCompletedEvent.Await()
                if:
                    FinishingPlayer = Player
                then:
                    RecordPlayerFinish(Player)
                    break
            # Waiting for this player to leave the game.
            loop:
                LeavingPlayer := GetPlayspace().PlayerRemovedEvent().Await()
                if:
                    LeavingPlayer = Player
                then:
                    break
            # Waiting for the race end timer to finish if a player has already finished the race.
            block:
                RaceEndTimer.SuccessEvent.Await()
                RecordPlayerFinish(Player)

    # When a player finishes the race, award them points based on their placement, and award them a win if
    # their placement was better than the PlacementRequiredForWin. Then update their round info in the
    # RoundInfo map.
    RecordPlayerFinish(Player:agent):void=
        PlayerFinishOrder:int = CurrentFinishOrder
        set CurrentFinishOrder += 1

        # If this is the first player to finish the race, then start the race end timer.
        if:
            PlayerFinishOrder = 0
        then:
            RaceEndTimer.Start()

        if:
            PointsToAward := PointsPerPlace[PlayerFinishOrder]
        then:
            AddPoints(Player, PointsToAward)
            # If the player's finishing spot was less than or equal to the PlacementRequiredToWin,
            # award them a win and record it in their player_stats_table.
            if:
                PlayerFinishOrder < PlacementRequiredForWin
            then:
                AddWin(Player, 1)
            if:
                RecordPlayerFinishOrder[Player, PlayerFinishOrder]
            else:
                Test.ProjectLog("Unable to record player finish order")

            if:
                CurrentRound := RoundInfo[GetSession()].CurrentRound
                UpdateRound[Player, CurrentRound]
            else:
                Test.ProjectLog("Unable to update completed round for player")
