# Knowledge-neutral Verse snippet; project names and asset IDs generalized.
# id: snippet-149

# This file defines the player_circuit_info, a persistable class that tracks a player's
# finish order and last completed round during the race.It also defines the round_info class to track player
# starting positions and the current round number. It uses both of these to set up the raceline
# starting order of players by ordering players based on who came before in the previous round.

using { /Fortnite.com/Devices }
using { /Fortnite.com/FortPlayerUtilities }
using { /Verse.org/Random }
using { /Verse.org/Simulation }

# A persistable map that maps each player to the number of and in
# what order they finished the previous round.
var CircuitInfo<public>:weak_map(player, player_circuit_info) = map{}

# Maps the current session to its associated round info.
var RoundInfo:weak_map(session, round_info) = map{}

# Maps each player to their start position during the race, and tracks the
# current number.
round_info := class:
    var PlayerStartPositions:[agent]start_position = map{}
    CurrentRound:int = -1

# Tracks the number of and in what order a player finished the previous round.
player_circuit_info<public> := class<final><persistable>:
    Version:int = 0
    LastRoundFinishOrder:int = -1
    LastCompletedRound<public>:int = -1

GetRound<public>()<decides><transacts>:int=
    RoundInfo[GetSession()].CurrentRound

# Returns the highest last completed round among all players.
GetLastCompletedRound<public>(Players:[]player, TotalRounds:int)<transacts>:int=
    var LastCompletedRound:int = -1
    for:
        Player : Players
        Player.IsActive[]
        PlayerCircuitInfo := CircuitInfo[Player]
    do:
        # If player's recorded info is greater than the total rounds for whatever reason,
        # then need to reset the player's circuit info because they shouldn't have more than what's allowed.
        if:
            # Total rounds should not be set to 0. 
            PlayerCircuitInfo.LastCompletedRound >= TotalRounds - 1
        then:
            # Try to reset the player's circuit info so it's fresh.
            if:
                ResetCircuitInfo[Player]
            else:
                Test.ProjectLog("Unable to reset circuit info for player")

        # Update LastCompletedRound if this player has the highest last completed round.
        else if:
            PlayerCircuitInfo.LastCompletedRound > LastCompletedRound
        then:
            set LastCompletedRound = PlayerCircuitInfo.LastCompletedRound

    LastCompletedRound

# Creates a new player_circuit_info from the given older player_circuit_info.
MakePlayerCircuitInfo<constructor>(OldPlayerCircuitInfo:player_circuit_info)<transacts> := player_circuit_info:
    Version := OldPlayerCircuitInfo.Version
    LastRoundFinishOrder := OldPlayerCircuitInfo.LastRoundFinishOrder
    LastCompletedRound := OldPlayerCircuitInfo.LastCompletedRound

# Updates a player's player_circuit_info with their last completed round.
UpdateRound<public>(Agent:agent, CompletedRound:int)<decides><transacts>:void=
    Player := player[Agent]
    Player.IsActive[]
    PlayerCircuitInfo := CircuitInfo[Player]
    set CircuitInfo[Player] = player_circuit_info:
        LastCompletedRound := CompletedRound
        MakePlayerCircuitInfo<constructor>(PlayerCircuitInfo)

# Creates a new player_circuit_info for the given player with the order they finished the round in.
RecordPlayerFinishOrder<public>(Agent:agent, FinishOrder:int)<decides><transacts>:void=
    Player := player[Agent]
    Player.IsActive[]
    PlayerCircuitInfo:player_circuit_info = if:
        Info := CircuitInfo[Player]
    then:
        Info
    else:
        player_circuit_info{}
   
    set CircuitInfo[Player] = player_circuit_info:
        LastRoundFinishOrder := FinishOrder
        MakePlayerCircuitInfo<constructor>(PlayerCircuitInfo)

# Resets a player's player_circuit_info.
ResetCircuitInfo<public>(Agent:agent)<decides><transacts>:void=
    Player := player[Agent]
    Player.IsActive[]
    PlayerCircuitInfo := CircuitInfo[Player]

    set CircuitInfo[Player] = player_circuit_info:
        LastRoundFinishOrder := -1
        LastCompletedRound := -1
        MakePlayerCircuitInfo<constructor>(PlayerCircuitInfo)

# Returns true if this is the first round of the game.
IsFirstRound<public>(RoundToCheck:int)<decides><transacts>:void=
    RoundToCheck <= 0

# Returns true if PlayerOne has finished the previous round before PlayerTwo.
HasFinishedBeforeInLastRound<public>(PlayerOne:player, PlayerTwo:player)<decides><transacts>:player=
    PlayerOne.IsActive[]
    PlayerTwo.IsActive[]
    PlayerOneFinishOrder := CircuitInfo[player[PlayerOne]].LastRoundFinishOrder
    PlayerTwoFinishOrder := CircuitInfo[player[PlayerTwo]].LastRoundFinishOrder

    PlayerOneFinishOrder < PlayerTwoFinishOrder
    PlayerOne

# Orders and returns players by their finish order in the previous round.
# During the first round, players are given random starting placements.
GetPlayerStartOrder<public>(Players:[]player):[]player=
    var OrderedPlayers:[]player = Players

    if:
        IsFirstRound[GetRound[]]
    then:
        # Randomize player order because it's first round.
        set OrderedPlayers = Shuffle(OrderedPlayers)
    else:
        # Order players based on how well they did in previous round.
        set OrderedPlayers = GetPlayersByFinishOrderInLastRound(Players)
       
    OrderedPlayers

# Order by last round's finish order, smallest to largest
GetPlayersByFinishOrderInLastRound(Players:[]player)<transacts>:[]player=
    SortingAlgorithms.MergeSort(Players, HasFinishedBeforeInLastRound)

RecordCurrentRound<public>(Players:[]player, TotalRounds:int):void=
    var CurrentRoundInfo:round_info =
        if:
            Info := RoundInfo[GetSession()]
        then:
            Info
        else:
            LastCompletedRound := GetLastCompletedRound(Players, TotalRounds)
            round_info:
                CurrentRound := LastCompletedRound + 1

    if:
        set RoundInfo[GetSession()] = CurrentRoundInfo
    else:
        Test.ProjectLog("Unable to record round info in session weak map.")

# Updates the sessions' RoundInfo by recording the round number and the order that players
# started the race in.
RecordPlayerStartOrder<public>(Players:[]player, StartingPositions:[]start_position):void=
    var CurrentRoundInfo:round_info = if:
        Info := RoundInfo[GetSession()]
    then:
        Info
    else:
        round_info{}

    # For each player, update the sessions' RoundInfo with the player's starting position.
    for (Index -> Player : Players, StartPosition := StartingPositions[Index]):
        if:
            set CurrentRoundInfo.PlayerStartPositions[Player] = StartPosition
        else:
            Test.ProjectLog("Unable to set player's start position.")
    if:
        set RoundInfo[GetSession()] = CurrentRoundInfo
    else:
        Test.ProjectLog("Unable to record round info in session weak map.")

# Get all players that are able to race.
GetAllValidPlayers(Players:[]player):[]player=
    # Valid players are ones that are active and not spectating.
    for:
        Player : Players
        Player.IsActive[]
        not Player.IsSpectator[]
    do:
        Player
