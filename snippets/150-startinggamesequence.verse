# Knowledge-neutral Verse snippet; project names and asset IDs generalized.
# id: snippet-150

# This file handles the logic for the pregame lobby and the cinematics that play at the beginning of a race.
# It controls the length of the starting lineup based on the number of players, and plays an intro for each
# player by displaying their stats.

using { /Fortnite.com/Characters }
using { /Fortnite.com/Devices }
using { /Fortnite.com/FortPlayerUtilities }
using { /Verse.org/Native }
using { /Verse.org/Simulation }
using { /UnrealEngine.com/Temporary/Diagnostics }
using { PlayerLeaderboard }

# A starting position on the race track.
start_position<public> := class<concrete>:

    # The car players drive during the race.
    @editable
    Car:vehicle_spawner_pickup_truck_device = vehicle_spawner_pickup_truck_device{}

    # Prevents players from driving out of their starting positions.
    @editable
    Barrier:barrier_device = barrier_device{}

# A Verse-authored creative device that can be placed in a level
starting_game_sequence := class(creative_device):

    # The amount of time to wait in the pregame lobby.
    @editable
    PreGameTime:float = 15.0

    # The total number of rounds in the race. 
    @editable
    TotalRounds:type {Rounds:int where 1 <= Rounds} = 3

    # Counts down to the start of the race during the pregame lobby.
    @editable
    PreGameTimer:timer_device = timer_device{}

    # Visual representations of each player in the pregame lobby.
    @editable
    LobbyPlayerReferences:[]player_reference_device = array{}

    # Display each player's stats in the pregame lobby.
    @editable
    LobbyLeaderboards:[]billboard_device = array{}

    # The position each player starts in, which includes their car,
    # a player reference to visualize them, and a barrier to restrict
    # them during the starting lineup.
    @editable
    StartingPositions:[]start_position = array{}

    # Tracks player analytics during the round.
    @editable
    RoundAnalytics:[]analytics_device = array{}

    # Handles the logic for the race gamemode.
    @editable
    RaceManager:race_manager_device = race_manager_device{}

    # The cinematic that intros the players and their stats.
    @editable
    StartingLineupCinematic:cinematic_sequence_device = cinematic_sequence_device{}

    # The cinematic that we cut to after the lineup and before the race starts.
    @editable
    RaceStartCinematic:cinematic_sequence_device = cinematic_sequence_device{}

    # Displays a player's stats during the lineup cinematic.
    @editable
    PopupDialog:popup_dialog_device = popup_dialog_device{}

    # Signals when a player's intro starts during the lineup cinematic.
    @editable
    StartPlayerIntroEvent:trigger_device = trigger_device{}

    # Signals when a player's intro ends during the lineup cinematic.
    @editable
    EndPlayerIntroEvent:trigger_device = trigger_device{}

    # Runs when the device is started in a running game.
    OnBegin<override>()<suspends>:void=
        var Players:[]player = GetAllValidPlayers(GetPlayspace().GetPlayers())

        RecordCurrentRound(Players, TotalRounds)
       
        # If this is the first round, start the pregame timer
        # to give the players time to explore the lobby. Otherwise skip to the race.
        if:
            IsFirstRound[GetRound[]]
        then:
             # Set up the pre-game lobby for players to view the leaderboard.
            SetupLobby(Players)

            PreGameTimer.SetActiveDuration(PreGameTime)
            PreGameTimer.Start()
            PreGameTimer.SuccessEvent.Await()
        
        # Determine what order players should start the race in and assign them to vehicles.
        set Players = GetPlayerStartOrder(GetAllValidPlayers(GetPlayspace().GetPlayers()))
        SetupStartingPositions(Players)
        RecordPlayerStartOrder(Players, StartingPositions)

        # Sleep for a second to allow the car to finish spawning before assigning a player to it.
        Sleep(1.0)

        Vehicles:[]vehicle_spawner_device =
            for (StartingPosition : StartingPositions):
                StartingPosition.Car
        PlayerVehicle.AssignPlayersToVehicles(Players, Vehicles)

        # Play cinematics before the race starts.
        PlayCinematicLineup(Players)

        # Return control to players to start the race.
        for (StartingPosition : StartingPositions):
            StartingPosition.Barrier.Disable()

        # Record which players are in the round at the start for tracking retention across rounds.
        SubmitRoundAnalytics(Players)

        # Begin the race!
        RaceManager.Begin()

    # Enable each starting position that has a valid player with the appropriate devices.
    SetupStartingPositions(Players:[]player)<suspends>:void=
        for:
            Index -> Player : Players
            StartingPosition := StartingPositions[Index]
            Vehicle := StartingPosition.Car
            Barrier := StartingPosition.Barrier
        do:
            Vehicle.Enable()
            Barrier.Enable()

    # Update player references and billboards in the lobby.
    SetupLobby(Players:[]player):void=
        UpdateStatsBillboards(Players, LobbyLeaderboards)
        UpdatePlayerReferences(Players, LobbyPlayerReferences)

    # Submit round analytics for each player.
    SubmitRoundAnalytics(Players:[]player):void=
        for (Player : Players, AnalyticsDevice := RoundAnalytics[GetRound[]]):
            AnalyticsDevice.Submit(Player)

    # Play the cinematic lineup and player intros based on the number of players.
    PlayCinematicLineup(Players:[]agent)<suspends>:void=

        # Race between all player intros and the starting lineup cinematic finishing.
        # Then plays the RaceStartCinematic to start the race.
        race:
            PlayAndWaitForCinematic(StartingLineupCinematic)
            WaitForAllPlayerIntros(Players)

        # Make sure the popup dialog is hidden when this function ends.
        PopupDialog.Disable()
        PlayAndWaitForCinematic(RaceStartCinematic)

    # Wait for each player intro to finish during the starting lineup.
    # The cinematic signals when a player’s intro starts and ends.
    # This function will wait the player’s intro to start to update the UI,
    # and wait for as many end player intros as there are players in the game.
    WaitForAllPlayerIntros(Players:[]agent)<suspends>:void=
        PlayersAndOrder:[]tuple(agent, int) =
            for (Index -> Player : Players):
                (Player, Index)
        ArraySync(PlayersAndOrder, WaitForPlayerIntro)

    # Wait for the player's intro start and display their info.
    # Cancel the wait if they leave.
    WaitForPlayerIntro(Player:agent, StartOrder:int)<suspends>:void=
        var IntroCounter:int = 0
        race:
            # Waiting for this player to finish the race and then record the finish.
            loop:
                sync:
                    block:
                        StartPlayerIntroEvent.TriggeredEvent.Await()
                        if (IntroCounter = StartOrder):
                            PlayerLeaderboard.UpdatePopupUI(Player, PopupDialog)
                    EndPlayerIntroEvent.TriggeredEvent.Await()
                if (IntroCounter = StartOrder):
                    break
                set IntroCounter += 1
            # Waiting for this player to leave the game.
            loop:
                LeavingPlayer := GetPlayspace().PlayerRemovedEvent().Await()
                if:
                    LeavingPlayer = Player
                then:
                    break

# Play the cinematic and wait for it to stop. Stopped if cinematic ends early.
PlayAndWaitForCinematic(Cinematic:cinematic_sequence_device)<suspends>:void=
    defer:
        # Go to end of cinematic and stop the cinematic to keep any changes
        # if this coroutine is canceled early.
        Cinematic.GoToEndAndStop()
    Cinematic.Play()
    Cinematic.StoppedEvent.Await()
