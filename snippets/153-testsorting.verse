# Knowledge-neutral Verse snippet; project names and asset IDs generalized.
# id: snippet-153

# This file contains tests for verifying the sorting algorithm behaves as expected.
# The tests also include profiling the code to see the performance of the algorithm.

using { /Fortnite.com/Devices }
using { /UnrealEngine.com/Temporary/Diagnostics }
using { /Verse.org/Simulation }
using { /Verse.org/Random }
 
test_log_channel<public> := class(log_channel):
 
# A project-wide "Logger" to print messages from functions that are not in a class with a log.
# The non-Logger Print is <no_rollback>, so it can't be used in a <transacts> function.
ProjectLog<public>(Message:[]char, ?Level:log_level = log_level.Normal)<transacts>:void=
    Logger := log{Channel := test_log_channel}
    Logger.Print(Message, ?Level := Level)
 
# A Verse-authored creative device that can be placed in a level.
# Place in the level to run tests and profile code.
test_sorting := class(creative_device):
 
    # Runs when the device is started in a running game
    OnBegin<override>()<suspends>:void= 
        # Example arrays to use for the test.
        var ArrayLength:int = GetRandomInt(10, 100)
        Test1ExpectedIntArray:[]int = for (Count := 0..ArrayLength):
            Count
        Test1ActualIntArray:[]int = Shuffle(Test1ExpectedIntArray)
 
        set ArrayLength = GetRandomInt(10, 100)
        Test2ExpectedIntArray:[]int = for (Count := 0..ArrayLength):
            ArrayLength - Count
        Test2ActualIntArray:[]int = Shuffle(Test2ExpectedIntArray)
 
        Test3ExpectedIntArray:[]int = array{0, 1, 1, 2, 3, 3}
        Test3ActualIntArray:[]int = Shuffle(Test3ExpectedIntArray)
 
        set ArrayLength = GetRandomInt(10, 100)
        Test4ExpectedPlayerStats:[]player_stats = for (Count := 0..ArrayLength):
            player_stats:
                Wins := ArrayLength - Count
                RoundsPlayed := GetRandomInt(Count, Count * 30)
        Test4ActualPlayerStats:[]player_stats = Shuffle(Test4ExpectedPlayerStats)
 
        # Run the tests and report if they failed or succeeded.
        if:
            RunArrayTest["Integer array in order from smallest to largest", Test1ActualIntArray, Test1ExpectedIntArray, IsIntSmallerThan, IntArrayToString]
            RunArrayTest["Integer array in order from largest to smallest", Test2ActualIntArray, Test2ExpectedIntArray, IsIntGreaterThan, IntArrayToString]
            RunArrayTest["Integer array in order from smallest to largest with repeats", Test3ActualIntArray, Test3ExpectedIntArray, IsIntSmallerThan, IntArrayToString]
            RunArrayTest["Player stats array in order from largest # wins to smallest", Test4ActualPlayerStats, Test4ExpectedPlayerStats, IsWinsGreaterThan, PlayerStatsWinsArrayToString]
        then:
            ProjectLog("All tests passed.")
        else:
            ProjectLog("One or more tests failed.")
           
# Test function for sorting arrays and profiling the code.
RunArrayTest(Description:string, ActualArray:[]t, ExpectedArray:[]t, Compare(L:t, R:t)<decides><transacts>:t, ArrayToString(:[]t)<transacts>:string where t:subtype(comparable))<decides><transacts>:void=
    # Perform merge sort and profile the code.
    ResultArray := profile(Description):
        SortingAlgorithms.MergeSort(ActualArray, Compare)
 
    # Print out the Actual, Expected, and Result arrays for troubleshooting.
    ProjectLog("Actual: {ArrayToString(ActualArray)}")
    ProjectLog("Expected: {ArrayToString(ExpectedArray)}")
    ProjectLog("Result: {ArrayToString(ResultArray)}")
 
    # Verify the test passed and sorted the array as expected.
    for (Index -> Result : ResultArray, Expected := ExpectedArray[Index]):
        Result = Expected
 
# Helper function to convert an integer array to a string.
IntArrayToString(Array:[]int)<transacts>:string=
    var ConvertedToString:string = "["
   
    for(Index -> Element : Array):
        set ConvertedToString += "{Element}"
        if (Index < Array.Length - 1):
            set ConvertedToString += ", "
       
    set ConvertedToString += "]"

# Helper function to convert the player stats comparison values to a string.
PlayerStatsWinsArrayToString(Array:[]player_stats)<transacts>:string=
    var ConvertedToString:string = "["
   
    for(Index -> Element : Array):
        set ConvertedToString += "{Element.Wins}"
        if (Index < Array.Length - 1):
            set ConvertedToString += ", "
       
    set ConvertedToString += "]"
 
# Comparison function to use as argument for sorting integers from smallest to largest number.
IsIntSmallerThan(Left:int, Right:int)<decides><transacts>:int=
    Left < Right

# Comparison function to use as argument for sorting integers from largest to smallest number.
IsIntGreaterThan(Left:int, Right:int)<decides><transacts>:int=
    Left > Right
 
# An example class for testing sorting.
player_stats := class<unique>:
    Wins:int = 0
    RoundsPlayed:int = 0
 
# Comparison function to use as argument for sorting wins from smallest to largest.
IsWinsSmallerThan(Left:player_stats, Right:player_stats)<decides><transacts>:player_stats=
    Left.Wins < Right.Wins
    Left
 
# Comparison function to use as argument for sorting wing from largest to smallest.
IsWinsGreaterThan(Left:player_stats, Right:player_stats)<decides><transacts>:player_stats=
    Left.Wins > Right.Wins
    Left
