# Knowledge-neutral Verse snippet; project names and asset IDs generalized.
# id: snippet-154

using { /Fortnite.com/Devices }
using { /Verse.org/Simulation }
using { /Verse.org/Random }
using { /UnrealEngine.com/Temporary/Diagnostics }

CoffinCategory<public><localizes>:message := "Coffin settings"
GameplayCategory<public><localizes>:message := "Gameplay"
FailureCategory<public><localizes>:message := "Failure"

CoffinTooltip<public><localizes>:message := "Reference to coffin device managed by this device."
RandomTravelTargetTooltip<public><localizes>:message := "If enabled, coffin will transport the player to another random coffin."
InspectButtonTooltip<public><localizes>:message := "Button placed nearby the coffin. Can be used by seeker to check if someone is hiding inside."
ResetTooltip<public><localizes>:message := "Trigger that resets experience."
CueTooltip<public><localizes>:message := "Audio player that will be played when seeker inspects empty coffin."
MessageTooltip<public><localizes>:message := "Message that will be displayed when seeker inspects empty coffin."
FreezeTooltip<public><localizes>:message := "Mutator zone that will freeze seeker when they inspect empty coffin."
FreezeTimeTooltip<public><localizes>:message := "For how long seeker will be frozen."

coffin_controller := class(creative_device):

    @editable:
        Categories := array{CoffinCategory}
        ToolTip := CoffinTooltip
    Coffin : hiding_prop_device = hiding_prop_device{}

    @editable:
        Categories := array{CoffinCategory}
        ToolTip := RandomTravelTargetTooltip
    ShouldRandomizeTravelTarget : logic = false

    @editable:
        Categories := array{GameplayCategory}
        ToolTip := InspectButtonTooltip
    Inspect : button_device = button_device{}

    @editable:
        Categories := array{GameplayCategory}
        ToolTip := ResetTooltip
    Reset : trigger_device = trigger_device{}

    @editable:
        Categories := array{FailureCategory}
        ToolTip := MessageTooltip
    Message : hud_message_device = hud_message_device{}

    @editable:
        Categories := array{FailureCategory}
        ToolTip := CueTooltip
    Cue : audio_player_device = audio_player_device{}

    @editable:
        Categories := array{FailureCategory}
        ToolTip := FreezeTooltip
    FreezeZone : mutator_zone_device = mutator_zone_device{}

    @editable_slider(float):
        Categories := array{FailureCategory}
        ToolTip := FreezeTimeTooltip
        MinValue := option{0.0}
        MaxValue := option{60.0}
        SliderDelta := option{1.0}
    FreezeTime : float = 5.0

    var NumberOfPlayers : int = 0

    OnBegin<override>()<suspends>:void=
        Coffin.BeginPlayerHideEvent.Subscribe(OnPlayerEnter)
        Coffin.EndPlayerHideEvent.Subscribe(OnPlayerExit)
        Inspect.InteractedWithEvent.Subscribe(OnInspect)
        Reset.TriggeredEvent.Subscribe(OnReset)

        # Toggle mutator zone - looks like it needs "warm up" for the first time
        FreezeZone.Enable()
        FreezeZone.Disable()

        if (ShouldRandomizeTravelTarget?):
            RandomizeTravelTarget()

    OnPlayerEnter(Player : player):void=
        set NumberOfPlayers += 1

    OnPlayerExit(Player : player):void=
        set NumberOfPlayers -= 1

    OnReset(Agent : ?agent):void=
        set NumberOfPlayers = 0
        Coffin.EjectAllHiddenPlayers()
        FreezeZone.Disable()

    OnInspect(Agent : agent):void=
        if (IsCoffinOccupied[]):
            Coffin.EjectAllHiddenPlayers()
        else:
            Cue.Play(Agent)
            Message.Show(Agent)
            spawn{FreezeSeeker()}

    FreezeSeeker()<suspends>:void=
        FreezeZone.Enable()
        Sleep(FreezeTime)
        FreezeZone.Disable()

    IsCoffinOccupied()<transacts><decides>:void=
        NumberOfPlayers > 0

    RandomizeTravelTarget():void=
        if (not Coffin.HiddenTravelGroup?):
            set Coffin.HiddenTravelGroup = option{GetRandomInt(0, 5)}

        loop:
            TravelTargetGroup := GetRandomInt(0, 5)
            if (TravelTargetGroup <> Coffin.HiddenTravelGroup?):
                set Coffin.HiddenTravelTargetGroup = option{TravelTargetGroup}
                break
