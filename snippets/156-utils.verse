# Knowledge-neutral Verse snippet; project names and asset IDs generalized.
# id: snippet-156

# Source: https://dev.epicgames.com/community/snippets/d8k/fortnite-wrapping-subscribe-to-pass-additional-data-to-listeners
using { /Verse.org/Simulation }
 
(Listenable : listenable(agent)).SubscribeAgent(OutputFunc : tuple(agent, t)->void, ExtraData : t where t:type) : cancelable =
    Wrapper := wrapper_agent(t){ExtraData := ExtraData, OutputFunc := OutputFunc}
    Listenable.Subscribe(Wrapper.InputFunc)
 
wrapper_agent(t : type) := class():
    ExtraData : t;
    OutputFunc : tuple(agent, t) -> void
    InputFunc(Agent : agent):void = OutputFunc(Agent, ExtraData)
 
(Listenable : listenable(tuple())).SubscribeEmpty(OutputFunc : t -> void, ExtraData : t where t:type) : cancelable =
    Wrapper := wrapper_empty(t) {ExtraData := ExtraData, OutputFunc := OutputFunc}
    Listenable.Subscribe(Wrapper.InputFunc)
 
wrapper_empty(t : type) := class():
    ExtraData : t;
    OutputFunc : t -> void
    InputFunc():void = OutputFunc(ExtraData)
