# Knowledge-neutral Verse snippet; project names and asset IDs generalized.
# id: snippet-157

using { /Fortnite.com/Devices }
using { /Verse.org/Simulation }
using { /Verse.org/Random }
using { /UnrealEngine.com/Temporary/Diagnostics }

PlainText<public><localizes>(Text:string):message = "{Text}"

AnswersCategory<public><localizes>:message := "Answers"

answer := enum<open>:
    A
    B
    C
    D

AnswerToIndex(Answer : answer)<transacts>:int=
    case (Answer):
        answer.A => 0
        answer.B => 1
        answer.C => 2
        answer.D => 3
        _ => -1

question := class:

    @editable_text_box:
        MultiLine := true
    Description : string = ""

    @editable:
        Categories := array{AnswersCategory}
    A : string = ""

    @editable:
        Categories := array{AnswersCategory}
    B : string = ""

    @editable:
        Categories := array{AnswersCategory}
    C : string = ""

    @editable:
        Categories := array{AnswersCategory}
    D : string = ""

    @editable
    CorrectAnswer : answer = answer.A

BillboardsCategory<public><localizes>:message := "Billboards"
VotingCategory<public><localizes>:message := "Voting"
ResultCategory<public><localizes>:message := "Result"

QuestionsTooltip<public><localizes>:message := "List of questions and answers presented to the player."
AnswerCubeTooltip<public><localizes>:message := "Billboard that is displayed in front of cube."
AnswerBoardTooltip<public><localizes>:message := "Billboard that is displayed on big board above lava pit."
AnswerButtonTooltip<public><localizes>:message := "Buttons used to vote for each answer."
EndMessageTooltip<public><localizes>:message := "Message device that is used to display results to the players."
TeleporterTooltip<public><localizes>:message := "Teleporters that are used to teleport winners out of the cube."
PropTooltip<public><localizes>:message := "Prop manipulators used to hide floor in loser's cubes."
VotingStartTooltip<public><localizes>:message := "Button that begins next voting session."
VotingEndTooltip<public><localizes>:message := "Sends trigger when voting ends."
VotingTimerTooltip<public><localizes>:message := "Timer that tracks voting countdown."

voting_device := class(creative_device):

    @editable:
        ToolTip := QuestionsTooltip
    Questions : []question = array{}

    @editable:
        Categories := array{BillboardsCategory}
    QuestionLeftBillboard : billboard_device = billboard_device{}

    @editable:
        Categories := array{BillboardsCategory}
    QuestionRightBillboard : billboard_device = billboard_device{}

    @editable:
        Categories := array{BillboardsCategory}
        ToolTip := AnswerCubeTooltip
    AnswerCubeBillboards : []billboard_device = array{}

    @editable:
        Categories := array{BillboardsCategory}
        ToolTip := AnswerBoardTooltip
    AnswerBoardBillboards : []billboard_device = array{}

    @editable:
        Categories := array{VotingCategory}
        ToolTip := AnswerButtonTooltip
    AnswerButtons : []button_device = array{}

    @editable:
        Categories := array{ResultCategory}
        ToolTip := EndMessageTooltip
    EndMessage : hud_message_device = hud_message_device{}

    @editable:
        Categories := array{ResultCategory}
        ToolTip := PropTooltip
    CubePropManipulators : []prop_manipulator_device = array{}

    @editable:
        Categories := array{ResultCategory}
        ToolTip := TeleporterTooltip
    WinnerTeleporters : []teleporter_device = array{}

    @editable:
        Categories := array{VotingCategory}
        ToolTip := VotingStartTooltip
    VoteStartButton : button_device = button_device{}

    @editable:
        Categories := array{VotingCategory}
        ToolTip := VotingEndTooltip
    VoteEndTrigger : trigger_device = trigger_device{}

    @editable:
        Categories := array{VotingCategory}
        ToolTip := VotingTimerTooltip
    VoteTimer : timer_device = timer_device{}

    var CurrentQuestion : question = question{}
    var QuestionID : int = 0
    var Winners : []agent = array{}
    var Losers : []agent = array{}

    OnBegin<override>()<suspends>:void=
        VoteStartButton.InteractedWithEvent.Subscribe(StartVoting)
        VoteEndTrigger.TriggeredEvent.Subscribe(EndVoting)
        SubscribeVotes()
        HideText()

    # Resets voting area state, displays next question and starts voting timer.
    StartVoting(Agent : agent):void=
        DisableTeleporters()
        ShowCubes()
        set Winners = array{}
        set Losers = array{}
        VoteTimer.Start()
        if (Question := GetNextQuestion[]):
            set CurrentQuestion = Question
            UpdateDescription(Question.Description)
            UpdateAnswer(answer.A, Question.A)
            UpdateAnswer(answer.B, Question.B)
            UpdateAnswer(answer.C, Question.C)
            UpdateAnswer(answer.D, Question.D)

    # Handles win/lose condition
    EndVoting(Agent : ?agent):void=
        CorrectAnswerIndex := AnswerToIndex(CurrentQuestion.CorrectAnswer)
        Answers := array{CurrentQuestion.A, CurrentQuestion.B, CurrentQuestion.C, CurrentQuestion.D}

        if (CorrectAnswer := Answers[CorrectAnswerIndex]):
            for (Loser : Losers):
                EndMessage.SetText(PlainText("You were wrong! Correct answer: {CorrectAnswer}"))
                EndMessage.Show(Loser)
        
        for (Winner : Winners):
            EndMessage.SetText(PlainText("You were right! Get ready for the next round!"))
            EndMessage.Show(Winner)

        # Teleport winner
        if:
            WinnerTeleporter := WinnerTeleporters[CorrectAnswerIndex]
        then:
            WinnerTeleporter.Enable()

        # Throw losers into lava
        for (PropIndex->PropManipulator : CubePropManipulators, PropIndex <> CorrectAnswerIndex):
                PropManipulator.HideProps()

    # Adds player to winners or losers list according to their answer
    OnVote(Agent : agent, Answer : answer)<transacts>:void=
        if (CurrentQuestion.CorrectAnswer = Answer):
            set Winners += array{Agent}
        else:
            set Losers += array{Agent}

    # Updates question displayed to players on billboard
    UpdateDescription(Text : string):void=
        QuestionLeftBillboard.SetText(PlainText(Text))
        QuestionRightBillboard.SetText(PlainText(Text))
        QuestionLeftBillboard.ShowText()
        QuestionRightBillboard.ShowText()

    # Updates answer displayed to players on billboard
    UpdateAnswer(Answer : answer, Text : string):void=
        if:
            Cube := AnswerCubeBillboards[AnswerToIndex(Answer)]
            Board := AnswerBoardBillboards[AnswerToIndex(Answer)]
        then:
            Cube.SetText(PlainText(Text))
            Board.SetText(PlainText(Text))
            Cube.ShowText()
            Board.ShowText()

    # Hides all text related to question
    HideText():void=
        QuestionLeftBillboard.HideText()
        QuestionRightBillboard.HideText()
        for (Billboard : AnswerCubeBillboards):
            Billboard.HideText()
        for (Billboard : AnswerBoardBillboards):
            Billboard.HideText()

    # Binds button interaction to corresponding answer
    SubscribeVote(Answer : answer):void=
        if (Button := AnswerButtons[AnswerToIndex(Answer)]):
            Button.InteractedWithEvent.SubscribeAgent(OnVote, Answer)

    # Bind every answer to button
    SubscribeVotes():void=
        SubscribeVote(answer.A)
        SubscribeVote(answer.B)
        SubscribeVote(answer.C)
        SubscribeVote(answer.D)

    # Disables all winner teleporters
    DisableTeleporters():void=
        for (Teleporter : WinnerTeleporters):
            Teleporter.Disable()

    # Enables floor in every cube
    ShowCubes():void=
        for (CubeManipulator : CubePropManipulators):
            CubeManipulator.ShowProps()

    # Fails if questions array is empty
    GetNextQuestion()<transacts><decides>:question=
        Question := Questions[QuestionID]
        set QuestionID += 1
        if (QuestionID >= Questions.Length):
            set QuestionID = 0
        Question
