# Knowledge-neutral Verse snippet; project names and asset IDs generalized.
# id: snippet-158

using { /Fortnite.com/Devices }
using { /Verse.org/Simulation }
using { /UnrealEngine.com/Temporary/Diagnostics }

# Create a custom log channel for the show_template_device. This helps with log filtering in complex games with lots of log sources.
log_show_template_device := class(log_channel){}

# A Verse-authored creative device that can be placed in a level
show_template_device := class(creative_device):

    # Logger that uses custom log channel.
    Logger:log = log{Channel := log_show_template_device}

    # The chair device the player should sit in.
    @editable
    TheChair:chair_device = chair_device{}

    # The sequence that drives animation and audio on the character device in the level.
    @editable
    MainSequence:cinematic_sequence_device = cinematic_sequence_device{}

    # The sequence that drives the TV mode sequence for viewing in the level.
    @editable
    TVModeSequence:cinematic_sequence_device = cinematic_sequence_device{}

    # A list of alternative camera sequences that are provided to view from during the main sequence.
    @editable
    CameraSwitches:[]cinematic_sequence_device = array{}

    # An input trigger that will return us to free look while in the chair.
    @editable
    ReturnToFreeLook:input_trigger_device = input_trigger_device{}

    # An input trigger that will return us to TV mode when appropriate in the chair.
    @editable
    ReturnToTVMode:input_trigger_device = input_trigger_device{}

    # An input trigger that will choose the next camera while in the chair.
    @editable
    NextCamera:input_trigger_device = input_trigger_device{}

    # An input trigger that will choose the previous camera while in the chair.
    @editable
    PreviousCamera:input_trigger_device = input_trigger_device{}

    # The alternative camera sequence that is playing if valid.
    var CurrentSequence:?cinematic_sequence_device = false

    # Helps us track when the main sequence on the character device is playing.
    var MainSequencePlaying:logic = false

    # When we are in TV mode.
    var InTVMode:logic = false

    # Keeps track of what camera we are viewing when using other cameras besides TV Mode.
    var CurrentCameraIndex:int = -1

    # Subscriptions to all the events we listen for while running.
    var ReturnToFreeLookSubscription:?cancelable = false
    var NextCameraSubscription:?cancelable = false
    var PrevCameraSubscription:?cancelable = false
    var ReturnToTVModeSubscription:?cancelable = false
    var ChairExitSubscription:?cancelable = false

    # Runs when the device is started in a running game.
    OnBegin<override>()<suspends>:void =
        Logger.Print("Standup Template device started")

        TheChair.SeatedEvent.Subscribe(OnSeated)

    # This function handles the player sitting down and starting up the performance if it isn't already running and setting up the input triggers.
    OnSeated(Agent:agent):void =
        Logger.Print("Player sat down")

        # Register for the input trigger for free look and subscribe to the pressed event when sitting.
        ReturnToFreeLook.Register(Agent)
        set ReturnToFreeLookSubscription = option{ReturnToFreeLook.PressedEvent.Subscribe(DoReturnToFreeLook)}

        # Register for the input trigger for the next camera and subscribe to the pressed event when sitting.
        NextCamera.Register(Agent)
        set NextCameraSubscription = option{NextCamera.PressedEvent.Subscribe(DoNextCamera)}

        # Register for the input trigger for the previous camera and subscribe to the pressed event when sitting.
        PreviousCamera.Register(Agent)
        set PrevCameraSubscription = option{PreviousCamera.PressedEvent.Subscribe(DoPreviousCamera)}

        # Subscribe to the input trigger for tv mode but do not register for it until we need to.
        set ReturnToTVModeSubscription = option{ReturnToTVMode.PressedEvent.Subscribe(DoReturnToTVMode)}

        # Subscribe to the event for leaving the chair but not register for it until we need to
        set ChairExitSubscription = option{TheChair.ExitedEvent.Subscribe(OnChairExited)}

        # If the main sequence is not playing on the character device in the level, run it, otherwise if it is running then just return to the TV mode viewing experience.
        if (MainSequencePlaying = false):
            spawn{RunSequence(Agent)}
        else:
            DoReturnToTVMode(Agent)

    # Cancel a subscription if it is valid.
    CancelSubscription(Subscription:?cancelable):void =
        if (SubscriptionToCancel := Subscription?):
            SubscriptionToCancel.Cancel()

    # Handles the player leaving the chair and removing access to the input triggers that are available while in the chair.
    OnChairExited(Agent:agent):void =
        Logger.Print("Player got up")

        # Stop the TV camera view when we leave the chair if it is active.
        if (InTVMode = true):
            TVModeSequence.Stop(Agent)
            set InTVMode = false

        # Stop the camera sequence we were using if it is active when we leave the chair.
        if (TempSeq := CurrentSequence?):
            TempSeq.Stop(Agent)

        # Clear whatever the current sequence was when we leave the chair.
        set CurrentSequence = false

        # Cancel all of our subscriptions when we leave the chair.
        CancelSubscription(ReturnToFreeLookSubscription)
        CancelSubscription(NextCameraSubscription)        
        CancelSubscription(PrevCameraSubscription)
        CancelSubscription(ReturnToTVModeSubscription)
        CancelSubscription(ChairExitSubscription)
    
        # Unregister us from all the input triggers when we leave the chair.
        ReturnToFreeLook.Unregister(Agent)
        NextCamera.Unregister(Agent)
        PreviousCamera.Unregister(Agent)
        ReturnToTVMode.Unregister(Agent)

    # Handles running the main sequence which runs the character and TV mode sequences for viewing and then calls a function to await finishing.
    RunSequence(Agent:agent)<suspends>:void =

        # Sleep for a second because we just entered the chair and the animation to sit takes a moment.
        Sleep(1.0)

        # Run the Main sequence on the character device and the TV mode viewing sequence.
        Logger.Print("Main Sequence Playing")
        set MainSequencePlaying = true
        MainSequence.Play()
        TVModeSequence.Play(Agent)

        set InTVMode = true

        # Call this to await the ending of the main sequence.
        AwaitMainSequencingEnding(Agent)

    # When the main sequence finishes, we clear the flag so that if the player sits back down, it will play again.
    AwaitMainSequencingEnding(Agent:agent)<suspends>:void =
        MainSequence.StoppedEvent.Await()
        Logger.Print("Main Sequence Ended")
        set MainSequencePlaying = false
        
        # Kick the player out of the chair after the performance
        TheChair.Eject(Agent)
    
    # Determine if we are in TV mode or another camera sequence and return control to the main player camera.
    DoReturnToFreeLook(Agent: agent):void =
        Logger.Print("Return to Free Look")

        # If TV Mode is active, stop it.
        if (InTVMode?):
            TVModeSequence.Stop(Agent)
            set InTVMode = false

        # If we are playing another viewing experience, stop it.
        if (TempSeq := CurrentSequence?):
            TempSeq.Stop(Agent)
            set CurrentSequence = false

        # Register the input triggers for returning to tv mode and unregister for free look.
        ReturnToTVMode.Register(Agent)
        ReturnToFreeLook.Unregister(Agent)

    # Returns us to our TV viewing mode sequence by checking where the main sequence is and aligning our playback to that point.
    DoReturnToTVMode(Agent:agent):void =
        Logger.Print("Return to TV Mode")

        # If we are already in TV mode no need to return to it.
        if (InTVMode?):
            return

        # If we have another camera sequence playing, stop it and clear associated values.
        if (TempSeq := CurrentSequence?):
            TempSeq.Stop(Agent)
            set CurrentSequence = false

        # Figure out where the main sequence is in its playback and start up the TV mode sequence at the same spot to provide the proper viewing experience.
        CurrentSeekTime := MainSequence.GetPlaybackTime()
        TVModeSequence.SetPlaybackTime(CurrentSeekTime)
        TVModeSequence.Play(Agent)
        set InTVMode = true

        # Unregister the input triggers for returning to tv mode and register for free look.
        ReturnToTVMode.Unregister(Agent)
        ReturnToFreeLook.Register(Agent)

    # Switches us between camera sequences that have been specified by stopping any current ones and moving to the next appropriate one on the list.
    DoCameraSwitch(Agent:agent, Value:int):void =
        if (CameraSwitch := CameraSwitches[Value]):
            Logger.Print("Switching Cameras to {Value}")

            # Stop any currently playing other camera sequence.
            if (PlayingSequence := CurrentSequence?):
                PlayingSequence.Stop(Agent)

            # Start up the new camera viewing sequence.
            CameraSwitch.Play(Agent)
            set CurrentSequence = option{CameraSwitch}

            # Register for TV mode and Free look input triggers while viewing other cameras.
            ReturnToTVMode.Register(Agent)
            ReturnToFreeLook.Register(Agent)
    
    # Switches to the next camera on our list, or the first if we are in TV mode.
    DoNextCamera(Agent:agent):void =
        Logger.Print("Next Camera")

        # If we are currently viewing from TV Mode, end that sequence and clear associated values.
        if (InTVMode?):
            TVModeSequence.Stop(Agent)
            set InTVMode = false
            set CurrentCameraIndex = -1

        # Figure out the next camera value based on the current camera and switch to it if valid.
        if (NextCameraValue := Mod[CurrentCameraIndex + 1, CameraSwitches.Length]):
            set CurrentCameraIndex = NextCameraValue

            DoCameraSwitch(Agent, CurrentCameraIndex)
            
    # Switches us to the previous camera on the list or last camera if we are leaving TV mode.
    DoPreviousCamera(Agent:agent):void =
        Logger.Print("Prev Camera")

        # If we are currently viewing from TV Mode, end that sequence and clear associated values.
        if (InTVMode?):
            TVModeSequence.Stop(Agent)
            set InTVMode = false
            set CurrentCameraIndex = 0

        # Figure out the previous camera value based on the current camera and switch to it if valid.
        if (NextCameraValue := Mod[CurrentCameraIndex - 1, CameraSwitches.Length]):
            set CurrentCameraIndex = NextCameraValue

            DoCameraSwitch(Agent, CurrentCameraIndex)
