# Knowledge-neutral Verse snippet; project names and asset IDs generalized.
# id: snippet-159

using { /Verse.org/Simulation }
using { /Verse.org/Simulation/Tags }
using { /Verse.org/Colors }
using { /UnrealEngine.com/Temporary/Diagnostics }
using { /Fortnite.com/Devices }


# Script that handles music and turn on lights when guards are alerted
stronghold_alert_manager := class(creative_device):
    
    # Reference to the Game Manager to monitor perception events
    @editable
    StrongholdGameManager:stronghold_game_manager := stronghold_game_manager{}

    # Reference to Radio that plays combat music
    @editable
    RadioCombat:radio_device := radio_device{}

    # Reference to Radio that plays alerted music
    @editable
    RadioAlerted:radio_device := radio_device{}

    # VFX to play when alarm/flare is shot
    @editable
    FlareAlarmVFXCreator:vfx_creator_device := vfx_creator_device{}
    
    # Lights to turn on when alerted
    @editable
    CustomizableLightDevicesAlerted: []customizable_light_device = array{}

    # Lights to turn on during combat
    @editable
    CustomizableLightDevicesCombat: []customizable_light_device = array{}
    
    # Change the camp to alerted when player is lost / killed
    WaitForAlerted()<suspends>:void=
        # do not go back to alerted after fallback
        if (StrongholdGameManager.FallbackTriggered?):
            Sleep(Inf)
        StrongholdGameManager.GuardsUnawareEvent.Await()
        Sleep(3.0)
        SetAlertedMood()

    # Change the camp to combat when player is spotted
    WaitForCombat()<suspends>:void=
        race:
            StrongholdGameManager.PlayerDetectedEvent.Await()
            StrongholdGameManager.FallbackEvent.Await()
        Sleep(2.0)
        SetCombatMood()

    # Runs when the device is started in a running game
    OnBegin<override>()<suspends>:void=
        MonitorStrongholdAlertStatus()

    # Main Loop checking if stronghold is in combat or alerted
    MonitorStrongholdAlertStatus()<suspends>:void=
        loop:
            WaitForCombat()
            WaitForAlerted()
    
    # Sets Base to Combat by toggling lights red and playing high intensity music
    SetCombatMood():void=
        # Loop through Combat Lights and turn them on
        for(LightsToTurnOn: CustomizableLightDevicesCombat):
            LightsToTurnOn.TurnOn()
        # Loop through Alert Lights and turn them off
        for(LightsToTurnOff: CustomizableLightDevicesAlerted):
            LightsToTurnOff.TurnOff()
        # Turn on combat audio and turn off alerted audio
        RadioCombat.Play()
        RadioAlerted.Stop()
        FlareAlarmVFXCreator.Toggle()
        Log("Alert manager - Combat Lights & Music On")

    # Sets Base to Alerted by toggling lights yellow and playing  tense music
    SetAlertedMood():void=
        for(LightsToTurnOn: CustomizableLightDevicesAlerted):
            LightsToTurnOn.TurnOn()
        for(LightsToTurnOff: CustomizableLightDevicesCombat):
            LightsToTurnOff.TurnOff()
        RadioCombat.Stop()
        RadioAlerted.Play()
        Log("Alert manager - Alert Lights & Music On")
