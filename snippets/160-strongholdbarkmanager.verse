# Knowledge-neutral Verse snippet; project names and asset IDs generalized.
# id: snippet-160

using { /Fortnite.com/Devices }
using { /Fortnite.com/Game }
using { /Fortnite.com/Characters }
using { /Verse.org/Random }
using { /Verse.org/Simulation }
using { /UnrealEngine.com/Temporary/Diagnostics }
using { /UnrealEngine.com/Temporary/SpatialMath }

# Audio bark that can be played on a NPC
audio_npc_bark := class<concrete>:

    # Audio device to play barks
    @editable
    BarkDevice:audio_player_device := audio_player_device{}

    # Option to allow NPCs to repeat the bark
    @editable
    CanRepeat:logic = true

    # Delay between the event and the beginning of the bark
    @editable
    Delay:float = 0.0

    # Delay before repeating this bark
    @editable
    Cooldown:float = 5.0

    # Bark name for logging
    @editable
    BarkID:string = "Missing ID"

    # Is the cooldown timer elapsed
    var<private> IsInCooldown:logic = false

    # Event to stop the bark
    StopBarkEvent<private>:event() = event(){}

    PlayBark(Agent:agent)<suspends>:void=
        var IsPlaying:logic = false;
        defer:
            if (IsPlaying?):
                set IsPlaying = false
                set IsInCooldown = false
                BarkDevice.Stop(Agent)
                Log("Bark manager - PlayBark - {BarkID} - Stopped")
        race:
            block:
                StopBarkEvent.Await()
                return
            block:
                AwaitAgentDown(Agent)
                return
            block:
                if (Delay > 0.0):
                    Sleep(Delay)
                if (IsInCooldown?):
                    return
                BarkDevice.Play(Agent)
                set IsPlaying = true
                set IsInCooldown = true
                Sleep(2.0)
                set IsPlaying = false
                Log("Bark manager - PlayBark - {BarkID} - Success")

        if (CanRepeat?):
            Sleep(Cooldown)
            set IsInCooldown = false

    StopBark():void=
        StopBarkEvent.Signal()

    AwaitAgentDown<private>(Agent:agent)<suspends>:void=
        if (Character := Agent.GetFortCharacter[]):
            loop:
                if (Character.GetHealth() <= 0.0):
                    return
                Character.DamagedEvent().Await()
    
# Script that handles barks from guards                
stronghold_bark_manager := class(creative_device):
    
    # Reference to the Game Manager to monitor perception events
    @editable
    StrongholdGameManager:stronghold_game_manager := stronghold_game_manager{}

    # Audio Player Devices
    @editable
    BarkNPCDown:audio_npc_bark = audio_npc_bark{BarkID := "Man Down", Delay := 0.3}

    @editable
    BarkFallback:audio_npc_bark = audio_npc_bark{BarkID := "Fallback", CanRepeat := false, Delay := 3.0}
    
    @editable
    BarkNeedBackup:audio_npc_bark = audio_npc_bark{BarkID := "Need Backup", CanRepeat := false, Delay := 2.0}

    @editable
    BarkGoToLeash:audio_npc_bark = audio_npc_bark{BarkID := "Reinforcements En Route", CanRepeat := false, Delay := 4.0}

    @editable
    BarkDamageTaken:audio_npc_bark = audio_npc_bark{BarkID := "Took Damage", Delay := 0.2}

    @editable
    BarkDamagePlayer:audio_npc_bark = audio_npc_bark{BarkID := "Hit Player", Delay := 0.2}

    @editable
    BarkEliminatedPlayer:audio_npc_bark = audio_npc_bark{BarkID := "Eliminated Player", Delay := 0.3}

    @editable
    BarkPlayerSpotted:audio_npc_bark = audio_npc_bark{BarkID := "Spotted Player", CanRepeat := false}

    @editable
    BarkPlayerLost:audio_npc_bark = audio_npc_bark{BarkID := "Lost Player", Cooldown := 10.0}

    @editable
    BarkGuardSuspicious:audio_npc_bark = audio_npc_bark{BarkID := "Suspicious", Cooldown := 10.0}

    @editable
    BarkGuardUnaware:audio_npc_bark = audio_npc_bark{BarkID := "Unaware", Cooldown := 10.0}

    # Variable to store if guards were looking for targets
    var<private> HasLostTarget:logic := false

    # Runs when the device is started in a running game
    OnBegin<override>()<suspends>:void=
        ConfigureBarks()

        sync:
            AwaitReinforcements()
            AwaitFallback()
            PlayAwarenessBarks()

    PlayBark(Bark:audio_npc_bark, Guard:agent):void=
        spawn {Bark.PlayBark(Guard)}

    # Play a bark when reinforcement is called
    AwaitReinforcements<private>()<suspends>:void=
        AlertedGuard := StrongholdGameManager.ReinforcementsCalledEvent.Await()
        PlayBark(BarkNeedBackup, AlertedGuard)

    # Play a bark when guards regroup in the stronghold
    AwaitFallback<private>()<suspends>:void=
        StrongholdGameManager.FallbackEvent.Await()
        if:
            Guard := StrongholdGameManager.AlertedGuards[0]
        then:
            PlayBark(BarkFallback, Guard)

    PlayAwarenessBarks<private>()<suspends>:void=
        loop:
            race:
                PlayGuardsSuspiciousBark()
                PlayPlayerSpottedBark()
                PlayPlayerLostBark()
                PlayGuardsUnawareBark()

    PlayPlayerSpottedBark<private>()<suspends>:void=
        Guard:=StrongholdGameManager.PlayerDetectedEvent.Await();
        set HasLostTarget = false
        PlayBark(BarkPlayerSpotted, Guard)

    PlayPlayerLostBark<private>()<suspends>:void=
        Guard:=StrongholdGameManager.PlayerLostEvent.Await();
        set HasLostTarget = true
        PlayBark(BarkPlayerLost, Guard)

    PlayGuardsSuspiciousBark<private>()<suspends>:void=
        Guard:=StrongholdGameManager.GuardsSuspiciousEvent.Await();
        PlayBark(BarkGuardSuspicious, Guard)

    PlayGuardsUnawareBark<private>()<suspends>:void=
        Guard:=StrongholdGameManager.GuardsUnawareEvent.Await();
        if (HasLostTarget?):
            set HasLostTarget = false
            if (not StrongholdGameManager.FallbackTriggered?):
                PlayBark(BarkGuardUnaware, Guard)

    SubscribeToGuardSpawnerEvents(GuardSpawnerDevice:guard_spawner_device):void = 
        GuardSpawnerDevice.DamagedEvent.Subscribe(OnGuardDamaged)
        GuardSpawnerDevice.EliminatedEvent.Subscribe(OnGuardEliminated)
        GuardSpawnerDevice.EliminatingEvent.Subscribe(OnPlayerEliminated)
            
    # Configure all barks
    ConfigureBarks():void=
        
        # Subscribe To Player Damage Event
        AllPlayers := GetPlayspace().GetPlayers()
        for (StrongholdPlayer : AllPlayers, StrongholdPC := StrongholdPlayer.GetFortCharacter[]):
            StrongholdPC.DamagedEvent().Subscribe(OnPlayerDamaged)
        
        # Run through guards spawner list from stronghold manager and subscribe to all key events
        for (GuardSpawner : StrongholdGameManager.GuardsInitialSpawners):
            SubscribeToGuardSpawnerEvents(GuardSpawner)
        for (GuardSpawner : StrongholdGameManager.GuardsReinforcementSpawners):
            SubscribeToGuardSpawnerEvents(GuardSpawner)

        # Have a separate case for when the reinforcements spawn
        if:
            FirstReinforcementSpawner := StrongholdGameManager.GuardsReinforcementSpawners[0]
        then:
            FirstReinforcementSpawner.SpawnedEvent.Subscribe(HandleReinforcementSpawned)
        
    # Guard is down, try to play a bark on the closest alerted guard
    OnGuardEliminated(InteractionResult:device_ai_interaction_result):void=
        if (EliminatedGuard := InteractionResult.Target?):

            # Find closest alive guard to play this bark
            var ClosestGuard:?agent = false
            if:
                set ClosestGuard = option{StrongholdGameManager.AlertedGuards[0]}
                EliminatedGuardCharacter := EliminatedGuard.GetFortCharacter[]
            then:
                for (AlertedGuard : StrongholdGameManager.AlertedGuards, AlertedGuardCharacter := AlertedGuard.GetFortCharacter[]):
                    if:
                        not ClosestGuard? = AlertedGuard
                        ClosestGuardCharacter := ClosestGuard?.GetFortCharacter[]
                        DistanceSquaredToAlertedGuard := DistanceSquared(AlertedGuardCharacter.GetTransform().Translation, EliminatedGuardCharacter.GetTransform().Translation)
                        DistanceSquaredToClosestGuard := DistanceSquared(ClosestGuardCharacter.GetTransform().Translation, EliminatedGuardCharacter.GetTransform().Translation)
                        DistanceSquaredToAlertedGuard < DistanceSquaredToClosestGuard
                    then:
                        set ClosestGuard = option{AlertedGuard}

            if (Guard := ClosestGuard?):
                spawn {BarkNPCDown.PlayBark(Guard)}
    
    # Guard is hit, try to play a bark if the guard is not down              
    OnGuardDamaged(InteractionResult:device_ai_interaction_result):void=
        if (Guard := InteractionResult.Target?):
            spawn {BarkDamageTaken.PlayBark(Guard)}

    # Player is hit, try to play a bark on the guard that damaged the player   
    OnPlayerDamaged(DamageResult:damage_result):void=
        if:
            fort_character[DamageResult.Target].GetHealth() > 0.0
            Guard := DamageResult.Instigator?.GetInstigatorAgent[]
        then:
            spawn {BarkDamagePlayer.PlayBark(Guard)}

    # Player is down, try to play a bark on the guard that eliminated the player   
    OnPlayerEliminated(InteractionResult:device_ai_interaction_result):void=
        if (Guard := InteractionResult.Source?):
            spawn {BarkEliminatedPlayer.PlayBark(Guard)}

    HandleReinforcementSpawned(Guard:agent):void=
        spawn {BarkGoToLeash.PlayBark(Guard)}
