# Knowledge-neutral Verse snippet; project names and asset IDs generalized.
# id: snippet-161

using { /Fortnite.com/Devices }
using { /Fortnite.com/Characters }
using { /Fortnite.com/FortPlayerUtilities }
using { /Fortnite.com/UI }
using { /Verse.org/Simulation }
using { /UnrealEngine.com/Temporary/Diagnostics }
using { /UnrealEngine.com/Temporary/SpatialMath}
using { /UnrealEngine.com/Temporary/UI }

# Script that sets the player into a cinematic mode so its not in the cinematic, and then resets them when the cinematic is done
stronghold_billboard_cinematic := class(creative_device):

    # A reference to the device that contains the level sequence to be played
    @editable
    CinematicSequenceDeviceToPlay:cinematic_sequence_device := cinematic_sequence_device{}

    # A reference to the device that will hide the HUD
    @editable
    HudController:hud_controller_device := hud_controller_device{}

    # Should the Skip Intro button be visible during the cinematic
    @editable
    ShowSkipIntroButton:logic := true

    # Should Devices and Billboards be visible during the cinematic
    @editable
    ShowDevices:logic := true

    # References to the billboard that will be visible during the cinematic
    @editable
    BillboardsToShow:[]billboard_device := array{}

    # References to the devices that will be hidden after the cinematic
    @editable
    DevicesToHide:[]creative_device := array{}
    @editable
    RadiosToHide:[]radio_device := array{}
    @editable
    AudioPlayersToHide:[]audio_player_device := array{}

    # Reference to Radio that plays music
    @editable
    Music:radio_device := radio_device{}

    # Reference to the Game Manager to start gameplay after the cinematic
    @editable
    StrongholdGameManager:stronghold_game_manager := stronghold_game_manager{}

    # A localizable message to display as text in the UI
    SkipCinematicText<localizes> : message = "Skip intro"

    var SkipButtonPerPlayer : [player]?canvas = map{}

    # Runs at the beginning of the script.
    OnBegin<override>()<suspends>:void=

        # Handles when the cinematic sequence device finishes playing
        CinematicSequenceDeviceToPlay.StoppedEvent.Subscribe(HandleCinematicEnd)

        # Sets the player to invisible and removes their movement
        CinematicStasis := stasis_args{AllowTurning := false, AllowFalling := false, AllowEmotes := false}

        for (StrongholdPlayer : GetPlayspace().GetPlayers(), not StrongholdPlayer.IsSpectator[], PlayerCharacter:= StrongholdPlayer.GetFortCharacter[], PlayerUI := GetPlayerUI[StrongholdPlayer]):
            PlayerCharacter.PutInStasis(CinematicStasis)
            PlayerCharacter.Hide()

            if (ShowSkipIntroButton?):
                SkipButton := CreateSkipButton(PlayerUI)
                option { set SkipButtonPerPlayer[StrongholdPlayer] = option{SkipButton} }

        HudController.Enable()

        if (ShowDevices?):
            ShowBillboardsAndDevices()
        else:
            HideBillboardsAndDevices()

        CinematicSequenceDeviceToPlay.Play()
        Music.Play()
        Log("Cinematic Mode Engaged - Players should be frozen and invisible")

    # Sets the player back to visible and free to move
    HandleCinematicEnd():void=

        Music.Stop()

        for (StrongholdPlayer : GetPlayspace().GetPlayers(), not StrongholdPlayer.IsSpectator[], PlayerCharacter:= StrongholdPlayer.GetFortCharacter[], PlayerUI := GetPlayerUI[StrongholdPlayer]):
            PlayerCharacter.ReleaseFromStasis()
            PlayerCharacter.Show()

            if (SkipButtonUI := SkipButtonPerPlayer[StrongholdPlayer]?):
                PlayerUI.RemoveWidget(SkipButtonUI)
                option { set SkipButtonPerPlayer[StrongholdPlayer] = false }

        HudController.Disable()

        HideBillboardsAndDevices()

        Log("Cinematic Mode Disengaged - Players should be free to move and visible.")

        spawn{ StrongholdGameManager.StartGameplay() }

    HandleSkipIntroInteraction(Message : widget_message) : void =

        CinematicSequenceDeviceToPlay.Stop(Message.Player)

    # A canvas widget that displays a Skip Button 
    CreateSkipButton(PlayerUI:player_ui):canvas=

        SkipButton :button_regular = button_regular{DefaultText := SkipCinematicText}
        SkipButton.OnClick().Subscribe(HandleSkipIntroInteraction)
        SkipButton.SetEnabled(true)

        SkipButtonCanvas : canvas = canvas:
            Slots := array:
                canvas_slot:
                    Anchors := anchors{Minimum := vector2{X := 0.9, Y := 0.9}, Maximum := vector2{X := 0.9, Y := 0.9}}
                    Offsets := margin{Top := 0.0, Left := 0.0, Right := 0.0, Bottom := 0.0}
                    Alignment := vector2{X := 1.0, Y := 1.0}
                    SizeToContent := true
                    Widget := SkipButton

        PlayerUISlot:=player_ui_slot{ZOrder:=0, InputMode:=ui_input_mode.All}
        PlayerUI.AddWidget(SkipButtonCanvas, PlayerUISlot)

        return SkipButtonCanvas


    ShowBillboardsAndDevices():void=
        for (Billboard : BillboardsToShow):
            Billboard.ShowText()
            Show()
        for (Device : DevicesToHide):
            Device.Show()
        for (Radio : RadiosToHide):
            Radio.Show()
        for (AudioPlayer : AudioPlayersToHide):
            AudioPlayer.Show()

    HideBillboardsAndDevices():void=
        for (Billboard : BillboardsToShow):
            Billboard.HideText()
        Hide()
        for (Device : DevicesToHide):
            Device.Hide()
        for (Radio : RadiosToHide):
            Radio.Hide()
        for (AudioPlayer : AudioPlayersToHide):
            AudioPlayer.Hide()
