# Knowledge-neutral Verse snippet; project names and asset IDs generalized.
# id: snippet-162

using { /Fortnite.com/AI }
using { /Fortnite.com/Characters }
using { /Fortnite.com/Devices }
using { /Fortnite.com/FortPlayerUtilities }
using { /Fortnite.com/Game }
using { /UnrealEngine.com/Temporary/Diagnostics }
using { /UnrealEngine.com/Temporary/SpatialMath }
using { /Verse.org/Simulation }
using { /Verse.org/Verse }


# The Stronghold is a game mode in which the goal is for players to eliminate all hostile enemies at a heavily guarded Stronghold
# The Stronghold Game Manager Verse device is used to manage, monitor, and control the AIs at the Stronghold
stronghold_game_manager := class(creative_device):
    
    # Device reference to guard spawner device to keep track of for eliminations
    @editable
    GuardsInitialSpawners:[]guard_spawner_device := array{}

    # Device reference to guard spawner device to keep track of for eliminations for multiplayer scaling
    @editable
    GuardsInitialSpawnersAdditional:[]guard_spawner_device := array{}

    # Device reference to reinforcement guard spawner device to trigger when one of the Stronghold guards is alerted
    @editable
    GuardsReinforcementSpawners:[]guard_spawner_device := array{}

    # Device reference to reinforcement guard spawner device to trigger when one of the Stronghold guards is alerted for multiplayer scaling
    @editable
    GuardsReinforcementSpawnersAdditional:[]guard_spawner_device := array{}

    # Device reference to display and track objectives
    @editable
    ObjectiveTracker:tracker_device := tracker_device{}

    # Device reference to display reinforcement in-game message
    @editable
    MessageDeviceReinforcement:hud_message_device := hud_message_device{}

    # Device reference to display fallback in-game message
    @editable
    MessageDeviceFallback:hud_message_device := hud_message_device{}

    # Device reference to end the game with a victory if the players completed the Stronghold without being detected
    @editable
    EndGameVictoryDeviceUndetected:end_game_device := end_game_device{}

    # Device reference to end the game with a victory if the players completed the Stronghold while being detected
    @editable
    EndGameVictoryDeviceDetected:end_game_device := end_game_device{}

    # Device reference to end the game with a fail if the players ran out of retries
    @editable
    EndGameFailDevice:end_game_device := end_game_device{}

    # Adjustable number of player lives
    @editable
    var PlayerRetries:int = 2

    # Device to reference Stronghold leash position
    @editable
    ReinforcementLeashReference:stronghold_leash_position := stronghold_leash_position{}

    # Device to reference Fallback leash position
    @editable
    FallbackLeashReference:stronghold_leash_position := stronghold_leash_position{}

    # Leashes that must be disabled after fallback
    @editable
    LeashesToDisableForFallback:[]stronghold_leash_position := array{}

    # Device for the explosion
    @editable
    ExplosiveDevice:explosive_device := explosive_device{}

    # Guards perception is monitored by this script, the other scripts can subscribe to those events

    # Event broadcasted when a guard call for reinforcement
    ReinforcementsCalledEvent:event(agent) = event(agent){}

    # Event broadcasted when guards defend the center of the Stronghold
    FallbackEvent:event() = event(){}

    # Event broadcasted when a guard becomes suspicious
    GuardsSuspiciousEvent:event(agent) = event(agent){}

    # Event broadcasted when all guards become unaware
    GuardsUnawareEvent:event(agent) = event(agent){}

    # Event broadcasted when a player is detected
    PlayerDetectedEvent:event(agent) = event(agent){}

    # Event broadcasted when all guards have lost their target
    PlayerLostEvent:event(agent) = event(agent){}

    # Lists of guards in a specific alert state to monitor perception changes

    # Variable to store reinforcement guards
    var<private> NumGuardsSpawned:int := 0

    # Variable to store all Stronghold guards
    var<private> StrongholdGuards:[]agent := array{}

    # Variable to store reinforcement guards
    var<private> ReinforcementGuards:[]agent := array{}

    # List of guards currently suspicious
    var<private> SuspiciousGuards : []agent = array{}

    # List of guards currently alerted
    var<private> AlertedGuards : []agent = array{}

    # List of guards currently investigating
    var<private> InvestigatingGuards : []agent = array{}

    # Initial guard spawners, will include additional spawners with multiplayer session
    var<private> GuardsInitialSpawnersInternal:[]guard_spawner_device = array{}

    # Reinforcement guard spawners, will include additional spawners with multiplayer session
    var<private> GuardsReinforcementSpawnersInternal:[]guard_spawner_device = array{}

    # Gameplay logic variables

    # Variable to track the number of eliminations from all Stronghold guard spawners
    var<private> GuardsEliminated:int := 0

    # Variable to track if the reinforcement were called or not
    var<private> ReinforcementTriggered:logic := false

    # Variable to track if the fallback was triggered
    var<private> FallbackTriggered:logic := false

    # Variable to store the first player agent that gets detected by the guards
    var<private> DetectedPlayer:?player := false
    

    # Runs when the device is started in a running game.
    OnBegin<override>()<suspends>:void=
        Log("Stronghold verse manager is starting")

        # Check active player for difficulty scaling
        var NumberOfActivePlayers:int = 0
        for (StrongholdPlayer : GetPlayspace().GetPlayers(), not StrongholdPlayer.IsSpectator[], StrongholdPC := StrongholdPlayer.GetFortCharacter[]):
            set NumberOfActivePlayers += 1
            # Subscribing to player elimination event
            StrongholdPC.EliminatedEvent().Subscribe(OnPlayerEliminated)

        Log("Game manager - Total player number is {NumberOfActivePlayers}")

        set GuardsInitialSpawnersInternal = GuardsInitialSpawners
        set GuardsReinforcementSpawnersInternal = GuardsReinforcementSpawners

        # Add additional Guard Spawner when there is more than 2 players
        if (NumberOfActivePlayers > 2):
            set GuardsInitialSpawnersInternal += GuardsInitialSpawnersAdditional
            set GuardsReinforcementSpawnersInternal += GuardsReinforcementSpawnersAdditional
            
        var NumInitialGuards:int = 0
        for (GuardSpawner : GuardsInitialSpawnersInternal):
            GuardSpawner.Enable()
            SubscribeToGuardSpawnerEvents(GuardSpawner);
            set NumInitialGuards += GuardSpawner.GetSpawnLimit()

        ObjectiveTracker.SetTarget(NumInitialGuards)
        
        for (GuardReinforcementSpawner : GuardsReinforcementSpawnersInternal):
            SubscribeToGuardSpawnerEvents(GuardReinforcementSpawner);

            # Subscribing to reinforcement spawned event
            GuardReinforcementSpawner.SpawnedEvent.Subscribe(OnReinforcementSpawned)
            GuardReinforcementSpawner.AlertedEvent.Subscribe(OnReinforcementAlerted)
            GuardReinforcementSpawner.UnawareEvent.Subscribe(OnReinforcementUnaware)

    SubscribeToGuardSpawnerEvents(SpawnerDevice:guard_spawner_device):void = 
        SpawnerDevice.SpawnedEvent.Subscribe(OnGuardSpawned)
        SpawnerDevice.EliminatedEvent.Subscribe(OnGuardEliminated)
        SpawnerDevice.SuspiciousEvent.Subscribe(OnGuardSuspicious)
        SpawnerDevice.AlertedEvent.Subscribe(OnGuardAlerted)
        SpawnerDevice.TargetLostEvent.Subscribe(OnGuardLostTarget)
        SpawnerDevice.UnawareEvent.Subscribe(OnGuardUnaware)
        
    # Start tracking eliminated guards and trigger the explosion  
    StartGameplay()<suspends>:void = 
        ObjectiveTracker.AssignToAll()
        Sleep(3.0)
        if (FirstPlayer:=GetPlayspace().GetPlayers()[0]):
            ExplosiveDevice.Explode(FirstPlayer)

    # Runs when guard spawner receives an alerted event and considers only the first alert event
    OnGuardAlerted(InteractionResult:device_ai_interaction_result):void=
        if:
            not ReinforcementTriggered?
            set DetectedPlayer = option{player[InteractionResult.Target?]}
            Guard:=InteractionResult.Source?
        then:
            Log("Game manager - Stronghold is alerted. Incoming reinforcement")

            var NumGuards:int = ObjectiveTracker.GetTarget()

            # Enabling the reinforcement guard spawner device ensures that we spawn the amount of guards configured in the guard spawner device.
            for (GuardReinforcementSpawner : GuardsReinforcementSpawnersInternal):
                GuardReinforcementSpawner.Enable()
                set NumGuards += GuardReinforcementSpawner.GetSpawnLimit()
    
            ObjectiveTracker.SetTarget(NumGuards)

            # Displaying in-game message for detection and incoming reinforcement
            MessageDeviceReinforcement.Show()
            set ReinforcementTriggered = true
            # Signal Reinforcement event
            ReinforcementsCalledEvent.Signal(Guard)
       
        # Add the guard to the list of alerted guards if it hasn't been previously added
        if(Guard:=InteractionResult.Source?):
            if (not AlertedGuards.Find[Guard]):
                set AlertedGuards += array{Guard}

            set SuspiciousGuards = SuspiciousGuards.RemoveFirstElement[Guard] or SuspiciousGuards
            set InvestigatingGuards = InvestigatingGuards.RemoveFirstElement[Guard] or InvestigatingGuards

            # Broadcast the Player Detected Event when one guard is alerted
            if (AlertedGuards.Length = 1):
                PlayerDetectedEvent.Signal(Guard)

    # Runs when reinforcement guard spawner receives an alerted event
    OnReinforcementAlerted(InteractionResult:device_ai_interaction_result):void=      
        if:
            not FallbackTriggered?
            Guard:=InteractionResult.Source?
        then:
            # Clear leash for reinforcement on alerted so they attack their target
            ReinforcementLeashReference.ClearLeashOnGuard(Guard)
            Log("Game manager - Reinforcement is alerted. Clear Leash.")

    # Runs when reinforcement guard spawner receives an unaware event
    OnReinforcementUnaware(Guard:agent):void=    
        if (not FallbackTriggered?):
            # Set back the leash
            Log("Game manager - Reinforcement is unaware. Reset leash.")
            ReinforcementLeashReference.ApplyLeashOnGuard(Guard)

    # Runs when guard spawner receives an unaware event
    OnGuardSuspicious(Guard:agent):void=
        if (not SuspiciousGuards.Find[Guard]):
            set SuspiciousGuards += array{Guard}

            # Broadcast the Suspicious Event when one guard is suspicious
            if:
                SuspiciousGuards.Length = 1
                AlertedGuards.Length = 0
                InvestigatingGuards.Length = 0
            then:
                GuardsSuspiciousEvent.Signal(Guard)

    # Runs when guard spawner receives an unaware event
    OnGuardUnaware(Guard:agent):void=
        set AlertedGuards = AlertedGuards.RemoveFirstElement[Guard] or AlertedGuards
        set SuspiciousGuards = SuspiciousGuards.RemoveFirstElement[Guard] or SuspiciousGuards
        set InvestigatingGuards = InvestigatingGuards.RemoveFirstElement[Guard] or InvestigatingGuards

        # Broadcast the Unaware Event when no guard is suspicious, alerted or investigating
        if:
            SuspiciousGuards.Length = 0
            AlertedGuards.Length = 0
            InvestigatingGuards.Length = 0
        then:
            GuardsUnawareEvent.Signal(Guard)  

    # When a guard loses track of the player, remove it from the alerted guards list, when all guards have lost player, signal the event
    OnGuardLostTarget(InteractionResult:device_ai_interaction_result):void=
        if (Guard := InteractionResult.Source?):
            if (not InvestigatingGuards.Find[Guard]):
                set InvestigatingGuards += array{Guard}

            # Broadcast the Player Lost Event when no guard is alerted                
            if (RevisedAlertedGuards := AlertedGuards.RemoveFirstElement[Guard]):
                set AlertedGuards = RevisedAlertedGuards
                if(AlertedGuards.Length = 0):
                    PlayerLostEvent.Signal(Guard)


    # Runs when a reinforcement guard is spawned. Each reinforcement guard is forced to attack the player that alerted the Stronghold
    OnReinforcementSpawned(Guard:agent):void=
        set ReinforcementGuards += array{Guard}
        ReinforcementLeashReference.ApplyLeashOnGuard(Guard)

        # Assigns the player that alerted the Stronghold guards as the target
        if (Target := DetectedPlayer?):
            for (GuardReinforcementSpawner : GuardsReinforcementSpawnersInternal):
                GuardReinforcementSpawner.ForceAttackTarget(Target, ?ForgetTime:=30.0)
            Log("Game manager - ForceAttackTarget given to reinforcement guard")

    
    # Runs when any guard from the Stronghold is spawned
    OnGuardSpawned(Guard:agent):void=
        set StrongholdGuards += array{Guard}
        set NumGuardsSpawned += 1
    
    # Runs when initial or reinforcement spawners receive an elimination
    OnGuardEliminated(InteractionResult:device_ai_interaction_result):void=
        set GuardsEliminated += 1
        Log("Game manager - Stronghold guard eliminated count = {GuardsEliminated} / {NumGuardsSpawned}")

        if (EliminatedAgent := InteractionResult.Target?):
            # Remove eliminated guard from the alerted guards list
            set AlertedGuards = AlertedGuards.RemoveFirstElement[EliminatedAgent] or AlertedGuards
            set SuspiciousGuards = SuspiciousGuards.RemoveFirstElement[EliminatedAgent] or SuspiciousGuards
            set InvestigatingGuards = InvestigatingGuards.RemoveFirstElement[EliminatedAgent] or InvestigatingGuards
            set StrongholdGuards = StrongholdGuards.RemoveFirstElement[EliminatedAgent] or StrongholdGuards
        else:
            Log("Game manager - No interaction target for elimination", ?Level:=log_level.Error)

        if (EliminationAgent := InteractionResult.Source?):
            OnGuardEliminatedByPlayer(EliminationAgent)
        else:
            Log("Game manager - No interaction source for elimination.", ?Level:=log_level.Error)

    OnGuardEliminatedByPlayer(EliminationPlayer:agent):void=

        # Increasing progress value for tracker device for each elimination
        ObjectiveTracker.Increment(EliminationPlayer)

        if (ReinforcementTriggered?):

            if (NumGuardsSpawned - GuardsEliminated = 3):
                StartFallback()
                
            # Ends the game mode if all guards were eliminated without the reinforcements
            if (GuardsEliminated >= NumGuardsSpawned):
                EndGameVictoryDeviceDetected.Activate(EliminationPlayer)
                Log("Game manager - Completed Stronghold detected")
        else:
            # Ends the game mode if all guards were eliminated with the reinforcements
            if (GuardsEliminated >= NumGuardsSpawned):
                EndGameVictoryDeviceUndetected.Activate(EliminationPlayer)
                Log("Game manager - Completed Stronghold undetected")

    # Assigns a new fallback leash when a few alerted guards are remaining to defend the center of the Stronghold
    StartFallback():void=
        # Displaying in-game message for guards retreating inside the Stronghold building
        MessageDeviceFallback.Show()
        set FallbackTriggered = true

        for (LeashDevice : LeashesToDisableForFallback):  
            LeashDevice.DisableLeashAndPatrolPaths()

        FallbackLeashPosition := FallbackLeashReference.GetTransform().Translation
        FallbackEvent.Signal()
        for (Guard : StrongholdGuards):  
            FallbackLeashReference.ApplyLeashOnGuard(Guard)
        Log("Game manager - Fallback leash is assigned")
    
        
    # Runs when a player elimination event is received
    OnPlayerEliminated(EliminationResult:elimination_result):void=
        set PlayerRetries -= 1
        if (PlayerRetries = 0, Agent := EliminationResult.EliminatedCharacter.GetAgent[]):
            EndGameFailDevice.Activate(Agent)
            Log("Game manager - Player was eliminated")
