# Knowledge-neutral Verse snippet; project names and asset IDs generalized.
# id: snippet-163

using { /Fortnite.com/AI }
using { /Fortnite.com/Characters }
using { /Fortnite.com/Devices }
using { /Verse.org/Random }
using { /Verse.org/Simulation }
using { /UnrealEngine.com/Temporary/Diagnostics }
using { /UnrealEngine.com/Temporary/SpatialMath }

# Script that handles the explosion sequence
stronghold_intro_explosion := class(creative_device):

    @editable
    StrongholdGameManager:stronghold_game_manager := stronghold_game_manager{}

    # Device for guard spawner who goes to investigate the explosion outside of the base
    @editable
    GuardsInvestigateSpawner:guard_spawner_device := guard_spawner_device{}

    # Device for the explosion
    @editable
    ExplosiveDevice:explosive_device := explosive_device{}

    # Device for the fire volume around the vehicle
    @editable
    FireVolumeAroundVehicle:fire_volume_device := fire_volume_device{}

    # Device for the damage volume around the vehicle
    @editable
    DamageVolumeAroundVehicle:damage_volume_device := damage_volume_device{}
    
    # Device for the vehicle explosion
    @editable
    VehicleSpawnerIntroExplosion:vehicle_spawner_big_rig_device := vehicle_spawner_big_rig_device{}

    # Bark to play when the investigation start
    @editable
    AudioPlayerInvestigateBark:audio_player_device := audio_player_device{}

    # A reference to the device that contains the level sequence to be played
    @editable
    CinematicSequenceDeviceToPlay:cinematic_sequence_device := cinematic_sequence_device{}

    StartDamageVolume<private>()<suspends>:void=

        VehicleSpawnerIntroExplosion.VehicleSpawnedEvent.Await()
        Log("Intro Sequence - Enable Damage Volume")
        DamageVolumeAroundVehicle.Enable()
        ExplosiveDevice.ExplodedEvent.Await()
        Log("Intro Sequence - Start post-explosion")
        DamageVolumeAroundVehicle.Disable()
        FireVolumeAroundVehicle.Ignite()
        CinematicSequenceDeviceToPlay.Play()
        Log("Intro Sequence - Damage Volume Disabled")

    # Runs when the device is started in a running game
    OnBegin<override>()<suspends>:void=

        spawn{ StartDamageVolume() }

        Log("Intro Sequence - Waiting for Guard to Spawn")
        GuardToGoInvestigate:=GuardsInvestigateSpawner.SpawnedEvent.Await()

        race:
            # stop the intro if guard is alerted
            block:
                GuardsInvestigateSpawner.AlertedEvent.Await()
                Log("Intro Sequence - Guard Alerted")
                return

            # stop the intro if guard is eliminated
            block:
                GuardsInvestigateSpawner.EliminatedEvent.Await()
                Log("Intro Sequence - Guard Eliminated")
                return

            block:

                if:
                    Navigatable:=GuardToGoInvestigate.GetFortCharacter[].GetNavigatable[]
                    FocusInterface:=GuardToGoInvestigate.GetFortCharacter[].GetFocusInterface[]
                then:
                    Log("Intro Sequence - Guard Spawned - Waiting for explosion")
                    race:
                        ExplosiveDevice.ExplodedEvent.Await()
                        Navigatable.Wait()

                    Navigatable.Wait(?Duration:=3.0)
                    AudioPlayerInvestigateBark.Play(GuardToGoInvestigate)

                    Log("Intro Sequence - Investigate Bark Played - Setting NavTarget")
                    Navigatable.Wait(?Duration:=2.0)
                    Log("Intro Sequence - Walking to Explosion")
                    InvestigatePosition:= MakeNavigationTarget(Self.GetTransform().Translation)
                    Navigatable.NavigateTo(InvestigatePosition,?AllowPartialPath:=true, ?ReachRadius:=100.0)

                    race:
                        Sleep(6.0)
                        loop:
                            RandomOffset := vector3{ X:=GetRandomFloat(-1000.0, 1000.0), Y:=GetRandomFloat(-1000.0, 1000.0), Z:=GetRandomFloat(50.0, 200.0)}
                            race:
                                FocusInterface.MaintainFocus(Self.GetTransform().Translation + RandomOffset)
                                Sleep(GetRandomFloat(0.5, 1.5))

                else:
                    Log("Intro Sequence - No navigation interface", ?Level:=log_level.Error)
