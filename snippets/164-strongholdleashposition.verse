# Knowledge-neutral Verse snippet; project names and asset IDs generalized.
# id: snippet-164

using { /Fortnite.com/AI }
using { /Fortnite.com/Characters }
using { /Fortnite.com/Devices }
using { /UnrealEngine.com/Temporary/Diagnostics }
using { /UnrealEngine.com/Temporary/SpatialMath }
using { /Verse.org/Colors }
using { /Verse.org/Random }
using { /Verse.org/Simulation }

# Defines a leash volume that can be assigned to guards
stronghold_leash_position := class(creative_device):

    # Leash is applied by default to all guards spawned by those devices
    @editable
    GuardsSpawners:[]guard_spawner_device := array{}

    # Guards on those patrol paths can go outside the leash (only one device per path)
    @editable
    PatrolPaths:[]ai_patrol_path_device := array{}

    # Set the leash inner radius. This value must be in centimeters. 
    # This defines the volume that must be reached when this leash is assigned to guards
    @editable 
    LeashInnerRadius<private>:float = 2300.0

    # Set the leash outer radius. This value must be in centimeters
    # This defines the volume in which guards must stay in when this leash is assigned
    @editable
    LeashOuterRadius<private>:float = 2400.0

    # List of guards currently assigned to this leash
    var<private> Guards : []agent = array{}

    OnBegin<override>()<suspends>:void=
        DebugDraw : debug_draw = debug_draw{}
        DebugDraw.DrawCylinder(GetTransform().Translation, GetTransform().Translation + vector3{X:=0.0, Y:=0.0, Z:=500.0}, ?NumSegments:=64, ?Radius:= LeashInnerRadius, ?Color:=NamedColors.LightGreen, ?DrawDurationPolicy:=debug_draw_duration_policy.Persistent, ?Thickness:= 3.0)
        DebugDraw.DrawCylinder(GetTransform().Translation, GetTransform().Translation + vector3{X:=0.0, Y:=0.0, Z:=500.0}, ?NumSegments:=64, ?Radius:= LeashOuterRadius, ?Color:=NamedColors.Orange, ?DrawDurationPolicy:=debug_draw_duration_policy.Persistent, ?Thickness:= 3.0)

        for (GuardSpawner : GuardsSpawners):
            GuardSpawner.SpawnedEvent.Subscribe(ApplyLeashOnGuard)

        for (PatrolPath : PatrolPaths):
            PatrolPath.PatrolPathStartedEvent.Subscribe(PatrolPathStarted)
            PatrolPath.PatrolPathStoppedEvent.Subscribe(PatrolPathStopped)

    ApplyLeashOnGuard(Guard:agent):void=
        Log("Leash position - Applying leash {LeashInnerRadius} / {LeashOuterRadius} on guard")
        if (Leashable:=Guard.GetFortCharacter[].GetFortLeashable[]):
            Leashable.SetLeashPosition(GetTransform().Translation, LeashInnerRadius, LeashOuterRadius)
    
            set Guards += array{Guard}

    ClearLeashOnGuard(Guard:agent):void=
        Log("Leash position - Clearing leash {LeashInnerRadius} / {LeashOuterRadius} on guard")
        if (Leashable:=Guard.GetFortCharacter[].GetFortLeashable[]):
            Leashable.ClearLeash()
        
        set Guards = Guards.RemoveFirstElement[Guard] or Guards

    DisableLeashAndPatrolPaths():void=
        Log("Leash position - Disabling leash {LeashInnerRadius} / {LeashOuterRadius}")
        for (Guard : Guards):
            if (Leashable:=Guard.GetFortCharacter[].GetFortLeashable[]):
                Leashable.ClearLeash()

        for (PatrolPath : PatrolPaths):
            PatrolPath.Disable()
        
        set Guards = array{}

    PatrolPathStarted(Guard:agent):void=
        if:
            Guards.Find[Guard]
            Leashable:=Guard.GetFortCharacter[].GetFortLeashable[]
        then:
            Log("Leash position - Patrol started - Clearing leash {LeashInnerRadius} / {LeashOuterRadius} on guard")
            Leashable.ClearLeash()

    PatrolPathStopped(Guard:agent):void=
        if:
            Guards.Find[Guard]
            Leashable:=Guard.GetFortCharacter[].GetFortLeashable[]
        then:
            Log("Leash position - Patrol stopped - Setting leash {LeashInnerRadius} / {LeashOuterRadius} on guard")
            Leashable.SetLeashPosition(GetTransform().Translation, LeashInnerRadius, LeashOuterRadius)
