# Knowledge-neutral Verse snippet; project names and asset IDs generalized.
# id: snippet-165

using { /UnrealEngine.com/Temporary/Diagnostics }

# Logger for the Stronghold experience. This log channel prints the Stronghold game's progression to the output log of UEFN.

stronghold_log_channel := class(log_channel){}

Log<internal>(Text:string, ?Level:log_level=log_level.Debug):void=

    Logger:log = log{Channel:=stronghold_log_channel}
    Logger.Print(Text, ?Level:=Level)
