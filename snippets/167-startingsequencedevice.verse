# Knowledge-neutral Verse snippet; project names and asset IDs generalized.
# id: snippet-167

using { /Fortnite.com/Devices }
using { /Verse.org/Simulation }
using { /Fortnite.com/Characters }
using { /UnrealEngine.com/Temporary/UI }
using { /UnrealEngine.com/Temporary/Diagnostics }

# See https://dev.epicgames.com/documentation/en-us/uefn/create-your-own-device-in-verse for how to create a verse device.

# Registers the channel for debug logs.
starting_sequence_log := class(log_channel){}

# Coordinates audio for the start of the game and enables the minimap
# when a player interacts with a console.
starting_sequence_device := class(creative_device):

    # A device for printing debug log information.
    Logger:log = log{Channel := starting_sequence_log}

    # HUD message device for displaying quest message.
    @editable
    AssignQuestHudMsg:hud_message_device = hud_message_device{}

    # The button to activate the upper deck map.
    @editable
    CrewQuartersButton:button_device = button_device{}

    # How long to wait before enabling the minimap and HUD.
    @editable
    MapAndHUDDelay:float = 1.0

    # Audio played when map indicator is added.
    @editable
    MapIndicatorAudio:audio_player_device = audio_player_device{}

    # HUD controller that only displays minimap.
    @editable
    MapOnlyHud:hud_controller_device = hud_controller_device{}
    
    # Map for the lower deck.
    @editable
    NewUpperDeckMap:map_controller_device = map_controller_device{}

    # Enables multiple ambient audio zones around the ship.
    @editable
    TriggerAudioChannelEnable:trigger_device = trigger_device{}

    # Starts playing ambient audio from multiple zones around the ship.
    @editable
    TriggerAudioChannelPlay:trigger_device = trigger_device{}

    # Initial audio that plays when the player spawns.
    @editable
    WelcomeAudio:audio_player_device = audio_player_device{}

    # Runs when the device is started in a running game.
    OnBegin<override>()<suspends>:void =

        # Transmit on the channel to enable ambient audio volumes throughout the ship.
        TriggerAudioChannelEnable.Trigger()
        # Transmit on the channel to start playing audio from ambient audio volumes.
        TriggerAudioChannelPlay.Trigger()
    
        # Wait for the player to interact with the console in the crew quarters.
        CrewQuartersButton.InteractedWithEvent.Await()

        # Cancel the welcome audio if its already playing.
        WelcomeAudio.Stop()

        # Enable the upper deck map controller, then enable the minimap.
        Sleep(MapAndHUDDelay)
        NewUpperDeckMap.Enable()
        MapOnlyHud.Enable()
        Sleep(MapAndHUDDelay)

        # Show the quest message and play the map indicator audio.
        AssignQuestHudMsg.Show()
        MapIndicatorAudio.Play()
