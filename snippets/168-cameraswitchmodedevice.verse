# Knowledge-neutral Verse snippet; project names and asset IDs generalized.
# id: snippet-168

# This file handles swapping between different cameras when an input is pressed.

using { /Fortnite.com/Devices }
using { /Verse.org/Simulation }
using { /Fortnite.com/Characters }
using { /UnrealEngine.com/Temporary/Diagnostics }

# See https://dev.epicgames.com/documentation/en-us/uefn/create-your-own-device-in-verse for how to create a verse device.
# Find more learning content from the character_environment project, including the MetaHuman and clothing, at the link https://fn.gg/character_environment-Learning-Assets

# Registers the channel for debug logs.
camera_switch_mode_log := class(log_channel){}

# Container class for a gameplay camera and whether it should
# hide the player character when the camera is added to a player.
gameplay_camera := class<concrete>():

    # The camera device associated with this class.
    @editable
    Camera:gameplay_camera_orbit_device = gameplay_camera_orbit_device{}

    # Whether this camera should hide the player character.
    @editable
    ShouldHideCharacter:logic = false

# Swaps between different cameras when an input trigger is activated.
camera_switch_mode_device := class(creative_device):

    # A device for printing debug log information.
    Logger:log = log{Channel := camera_switch_mode_log}

    # How long to wait before hiding the player character model
    # when a camera is added to a player.
    @editable
    HideDelay:float = 0.20

    # How long to wait before showing the player character model
    # when a camera is removed from a player.
    @editable
    ShowDelay:float = 0.3

    # Listens for a given input and swaps the camera between different modes when the input is pressed.
    @editable
    SwapCameraInputTrigger:input_trigger_device = input_trigger_device{}

    # The array of cameras to swap between.
    @editable
    Cameras:[]gameplay_camera = array{}

    # Runs when the device is started in a running game
    OnBegin<override>()<suspends>:void=

        # Spawn a camera loop per player.
        Players := GetPlayspace().GetPlayers()
        for(Player : Players):
            spawn{CameraLoop(Player)}

    # Listen for the input trigger, and swap through camers in the Cameras array.
    CameraLoop(Agent:agent)<suspends>:void=

        # Index of the current camera in the Cameras array.
        var CameraIndex:int = 0

        # Reference to the current camera.
        var MaybeCurrentCamera:?gameplay_camera_device = false

        # Continuously listen for the input trigger, then swap 
        # to the next camera. Return to the default camera when 
        # reaching the end of the array.
        loop:
            SwapCameraInputTrigger.PressedEvent.Await()
            # If the player has a camera, remove it from their camera stack.
            if:
                CurrentCamera := MaybeCurrentCamera?
            then:
                CurrentCamera.RemoveFrom(Agent)
                Logger.Print("Removed a camera at {CameraIndex}.")
            if:
                # Get the next camera in the Cameras array.
                NextCamera := Cameras[CameraIndex]
            then:
                # Add the camera to the agent.
                NextCamera.Camera.AddTo(Agent)
                set MaybeCurrentCamera = option{NextCamera.Camera}
                set CameraIndex += 1
                Logger.Print("Added a camera at {CameraIndex}.")
                # Check if the next camera should hide the agent's fort_character.
                # Otherwise, show the fort_character.
                if:
                    NextCamera.ShouldHideCharacter?
                then:
                    spawn{HideCharacter(Agent)}
                else:
                    ShowCharacter(Agent)
            else:
                # Otherwise return to the default camera and reset the
                # CameraIndex count.
                set CameraIndex = 0
                set MaybeCurrentCamera = false
                ShowCharacter(Agent)
                Logger.Print("Reset Cameras")
            # Sleep for a short amount of time to prevent spamming 
            # the input trigger.
            Sleep(1.0)

    # Wait for a HideDelay amount of seconds, the hide the 
    # agent's fort_character.
    HideCharacter(Agent:agent)<suspends>:void=
        Sleep(HideDelay)
        if (FortCharacter := Agent.GetFortCharacter[]):
            FortCharacter.Hide()

    # Wait for a HideDelay amount of seconds, the show the 
    # agent's fort_character.
    ShowCharacter(Agent:agent)<suspends>:void=
        Sleep(ShowDelay)
        if (FortCharacter := Agent.GetFortCharacter[]):
            FortCharacter.Show()
