# Knowledge-neutral Verse snippet; project names and asset IDs generalized.
# id: snippet-169

using { /Fortnite.com/Devices }
using { /Verse.org/Simulation }

Hide_on_Start := class(creative_device):

    @editable
    PropsToHide : []creative_prop = array{}

    OnBegin<override>()<suspends>:void=
        for (Prop : PropsToHide):
            Prop.Hide()
