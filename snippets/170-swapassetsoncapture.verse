# Knowledge-neutral Verse snippet; project names and asset IDs generalized.
# id: snippet-170

using { /Fortnite.com/Devices }
using { /Verse.org/Simulation }

asset_swap_device := class(creative_device):
    # Array to store the 4 assets that can be swapped
    @editable
    var Assets : []creative_prop = array{}

    # Track which asset is currently active
    var CurrentIndex : int = 0

    # Reference to the mutator zone
    @editable
    MutatorZone : mutator_zone_device = mutator_zone_device{}

    OnBegin<override>()<suspends>:void =
        # Subscribe to the mutator zone's AgentEntersEvent
        MutatorZone.AgentEntersEvent.Subscribe(OnAgentEntersZone)

    # Handler for when an agent enters the mutator zone
    OnAgentEntersZone(Agent:agent):void =
        SwapToNextAsset()

    # Swaps to the next asset in the array
    SwapToNextAsset():void =
        # Disable current asset if valid index
        if (CurrentIndex >= 0 and CurrentIndex < Assets.Length):
            if (CurrentAsset := Assets[CurrentIndex]):
                CurrentAsset.Hide()

        # Update index (wrap around to 0 if at end)
        if (CurrentIndex >= Assets.Length - 1):
            set CurrentIndex = 0
        else:
            set CurrentIndex += 1

        # Enable new asset if valid index
        if (CurrentIndex >= 0 and CurrentIndex < Assets.Length):
            if (NewAsset := Assets[CurrentIndex]):
                NewAsset.Show()
