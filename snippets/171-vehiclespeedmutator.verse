# Knowledge-neutral Verse snippet; project names and asset IDs generalized.
# id: snippet-171

using { /Fortnite.com/Devices }
using { /Fortnite.com/Vehicles }
using { /Fortnite.com/Characters }
using { /Verse.org/Simulation }

vehicle_mod_zone := class(creative_device):

    @editable
    MutatorZone : mutator_zone_device = mutator_zone_device{}

    @editable
    MovementMod : movement_modulator_device = movement_modulator_device{}

    OnBegin<override>()<suspends>:void =
        # Subscribe to zone events
        MutatorZone.AgentEntersEvent.Subscribe(OnAgentEntersZone)
        MutatorZone.AgentExitsEvent.Subscribe(OnAgentExitsZone)

    OnAgentEntersZone(Agent:agent):void =
        # Only apply movement mod if agent is in a vehicle
        if (HasVehicle[Agent]):
            MovementMod.Enable()

    OnAgentExitsZone(Agent:agent):void =
        # Remove movement mod when agent leaves zone
        MovementMod.Disable()

    # Helper function to check if agent is in a vehicle
    HasVehicle<private>(Agent:agent)<decides><transacts>:void =
        if (FortCharacter := Agent.GetFortCharacter[]):
            if (Vehicle := FortCharacter.GetVehicle[]):
                true
            else:
                false?
        else:
            false?
