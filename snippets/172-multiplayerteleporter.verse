# Knowledge-neutral Verse snippet; project names and asset IDs generalized.
# id: snippet-172

using { /Fortnite.com/Devices }
using { /Verse.org/Simulation }
using { /Fortnite.com/Playspaces }
using { /Fortnite.com/Characters }
using { /UnrealEngine.com/Temporary/SpatialMath }
using { /UnrealEngine.com/Temporary/Diagnostics }

# This device is used to handle teleporting players around the map to ensure all players are always in the action!
multiplayer_teleporter := class(creative_device):

    #The trigger device used to initiate teleporting players
    @editable
    TriggerDevice : trigger_device = trigger_device{}

    # The world axis in which the offset is applied to the teleporting character i.e. X: 0 Y:1 Z:0 will apply the offset in the Y axis
    @editable
    TeleportOffsetAxis : vector3 = vector3{X := 0.0, Y := 1.0, Z := 0.0}

    # Runs when the device is started in a running game
    OnBegin<override>()<suspends> : void=   
        TriggerDevice.TriggeredEvent.Subscribe(TeleportPlayers)

    TeleportPlayers(TriggeringAgent : ?agent)<transacts> : void = 
    
        # Check if the triggering player's agent is valid and then use it's FortCharacter to set TriggeringCharacter
        if (TriggeringCharacter := TriggeringAgent?.GetFortCharacter[]):
            var TeleportOffset : float = 0.0
            CurrentPlayers := GetPlayspace().GetPlayers()
            # Iterate through all of the players, checking to make sure they aren't the triggering player, then teleport them to the triggering player
            for (Player : CurrentPlayers, Player.GetFortCharacter[] <> TriggeringCharacter, CharacterToTeleport := Player.GetFortCharacter[]):
                set TeleportOffset += 100.0  
                # Teleport PlayerToTeleport to TriggeringPlayer's position, if the teleport fails to find the required space, players will remain where they are
                if (CharacterToTeleport.TeleportTo[TriggeringCharacter.GetTransform().Translation + (TeleportOffsetAxis * TeleportOffset),
                TriggeringCharacter.GetTransform().Rotation]):
