# Knowledge-neutral Verse snippet; project names and asset IDs generalized.
# id: snippet-174

using { /Fortnite.com/Devices }
using { /Fortnite.com/Characters }
using { /Verse.org/Simulation }
using { /UnrealEngine.com/Temporary/Diagnostics }

triad_invisibility_log_channel := class(log_channel){}

invisibility_manager := class(creative_device):

    Logger:log = log{Channel := triad_invisibility_log_channel}

    # Array of players spawners for the Infiltrators team 
    @editable
    PlayersSpawners:[]player_spawner_device = array{}
    
    # Whether the visibility of the infiltrators is shared with teammates. 
    @editable
    IsVisibilityShared:logic = true

    # How long the infiltrators are visible for after being damaged. 
    @editable
    VulnerableSeconds:float = 3.0
    
    # How quickly infiltrators flicker after being damaged.
    @editable
    FlickerRateSeconds:float = 0.4
    
    # Array of all teams in the game.
    var AllTeams:[]team = array{}

    # Map of players to the amount of seconds they have left to keep blinking.
    var PlayerVisibilitySeconds:[agent]float = map{}

    OnBegin<override>()<suspends>:void=
        # Wait for teams to be balanced before subscribing to events that make the players invisible.
        Logger.Print("Waiting for teams to be balanced...")
    
    # Starts the invisibility manager logic. Called from triad_infiltration class after team balancing finishes
    StartInvisibilityManager<public>(GameTeams:[]team, AllPlayers:[]player, Infiltrators:team):void=
        Logger.Print("Invisibility script started!")
        set AllTeams = GameTeams
        for(PlayerSpawner:PlayersSpawners):
            PlayerSpawner.SpawnedEvent.Subscribe(OnPlayerSpawn)
        
        # For each player, if they spawned on the infiltrator team, spawn an OnInfiltratorDamaged function for that
        # player. Then make their character invisible. 
        for(TeamPlayer:AllPlayers):
            if:
                FortCharacter:fort_character = TeamPlayer.GetFortCharacter[] 
                CurrentTeam:team := GetPlayspace().GetTeamCollection().GetTeam[TeamPlayer]
                Logger.Print("Got this player's current team")
                Infiltrators = CurrentTeam
                set PlayerVisibilitySeconds[TeamPlayer] = 0.0
                Logger.Print("Added player to PlayerVisibilitySeconds")
            then:
                spawn{OnInfiltratorDamaged(TeamPlayer)}
                Logger.Print("Player spawned as an infiltrator, making them invisible")
                FortCharacter.Hide()
            else:
                Logger.Print("This player isn't an infiltrator)")


    # Flickers an agent's visibility by repeatedly hiding and showing their fort_character 
    FlickerCharacter(InCharacter:fort_character)<suspends>:void=
        Logger.Print("FlickerCharacter() invoked")
        # Loop hiding and showing the character to create a flickering effect.
        var TimeRemaining:float = 0.0
        loop:
            InCharacter.Hide()
            Sleep(FlickerRateSeconds)
            InCharacter.Show()
            Sleep(FlickerRateSeconds)
            # Each loop, decrease the amount of time the character is flickering by FlickerRateSeconds. 
            # If Remaining time hits 0, break out of the loop.
            if:
                NewTime := set PlayerVisibilitySeconds[InCharacter.GetAgent[]] -= FlickerRateSeconds * 2 
                set TimeRemaining = NewTime
            then:
                Logger.Print("Time Remaining: {TimeRemaining}")
            if:
                TimeRemaining <= 0.0
            then:
                InCharacter.Hide()
                break                           

    # Flickers an agent's visibility whenever they take damage
    OnInfiltratorDamaged(InAgent:agent)<suspends>:void=
        Logger.Print("Attempting to start flickering this character")
        TeamCollection := GetPlayspace().GetTeamCollection()
        if (FortCharacter := InAgent.GetFortCharacter[]):
            loop:
                if(IsVisibilityShared?, CurrentTeam := TeamCollection.GetTeam[InAgent], TeamAgents := TeamCollection.GetAgents[CurrentTeam]):
                    # For each teammate, set them in PlayerVisibility seconds and spawn a FlickerEvent.
                    for(Teammate:TeamAgents):
                        Logger.Print("Calling StartOrResetFlickering on a Teammate")
                        StartOrResetFlickering(Teammate)
                else:
                    # Just flicker the damaged character.
                    Logger.Print("Calling StartOrResetFlickering on InAgent")
                    StartOrResetFlickering(InAgent)
                FortCharacter.DamagedEvent().Await()

    # Starts a new flicker event if the agent was invisible, otherwise
    # resets the agent's ongoing flickering.
    StartOrResetFlickering(InAgent:agent):void=
        if (not IsFlickering[InAgent], FortCharacter := InAgent.GetFortCharacter[]):
            Logger.Print("Attempting to start NEW FlickerEvent for this character")
            # New flickering started.
            if (set PlayerVisibilitySeconds[InAgent] = VulnerableSeconds):
                spawn{FlickerCharacter(FortCharacter)}
                Logger.Print("Spawned a FlickerEvent for this character")
        else:
            # Reset ongoing flickering.
            if (set PlayerVisibilitySeconds[InAgent] = VulnerableSeconds):
                Logger.Print("Reset character's FlickerTimer to VulnerableSeconds")
       
    # Returns whether the player has any time left to flicker
    IsFlickering(InAgent:agent)<decides><transacts>:void=
        PlayerVisibilitySeconds[InAgent] > 0.0

    # Spawns an OnInfiltratorDamaged function when a new infiltrator joins the game
    OnInfiltratorJoined<public>(InAgent:agent):void=
        spawn{OnInfiltratorDamaged(InAgent)}
        
    # Handles a player spawning from an infiltrator spawn pad
    OnPlayerSpawn(SpawnedAgent:agent):void=
        Logger.Print("A player just spawned from an infiltrator spawn pad!")
        if:
            FortCharacter:fort_character = SpawnedAgent.GetFortCharacter[] 
            CurrentTeam := GetPlayspace().GetTeamCollection().GetTeam[SpawnedAgent] 
            AllTeams[0] = CurrentTeam
            Logger.Print("Player spawned as an infiltrator, making them invisible")
        then:
            FortCharacter.Hide()
