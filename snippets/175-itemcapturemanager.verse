# Knowledge-neutral Verse snippet; project names and asset IDs generalized.
# id: snippet-175

using { /Fortnite.com/Devices }
using { /Verse.org/Simulation }
using { /Fortnite.com/Characters }
using { /UnrealEngine.com/Temporary/Diagnostics }
using { /UnrealEngine.com/Temporary/SpatialMath }

triad_item_capture_log_channel := class(log_channel){}

item_capture_manager := class(creative_device):

    Logger:log = log{Channel := triad_item_capture_log_channel}

    # Capture item spawner that spawns the item to capture.
    @editable
    CaptureItemSpawner:capture_item_spawner_device = capture_item_spawner_device{}

    # Prop that floats above a players head when they're holding the item from.
    # the CaptureItemSpawner.
    @editable
    CaptureItemIndicator:creative_prop = creative_prop{}

    # Indicator that displays on the map where the objectives for each team are.
    @editable
    MapIndicator:map_indicator_device = map_indicator_device{}

    # How often the CaptureItemIndicator updates its position.
    @editable
    UpdateRateSeconds:float = 0.15

    # How high above a player's head the CaptureItemIndicator floats.
    @editable
    VerticalOffset:float = 180.0

    # Displays a message when a player grabs the Capture Item.
    @editable
    ItemGrabbedMessageDevice:hud_message_device = hud_message_device{}

    # The amount of time to wait before returning the CaptureItem and Map indicators. 
    # A negative time indicates that the indicators will never return unless the objective is
    # picked up again.
    @editable
    ReturnTime:float = 10.0

    # Awards score when a player captures the capture Item.
    @editable
    ScoreManagerDevice:score_manager_device = score_manager_device{}

    OnBegin<override>()<suspends>:void=
        CaptureItemSpawner.ItemPickedUpEvent.Subscribe(OnItemPickedUp)
        CaptureItemSpawner.ItemCapturedEvent.Subscribe(OnItemCaptured)
        CaptureItemSpawner.ItemDroppedEvent.Subscribe(OnItemDropped)
        SpawnerTransform := CaptureItemSpawner.GetTransform()
        
        # Teleport back to spawner, hiding the CaptureItemIndicator beneath the map out of site.
        CaptureItemIndicator.MoveTo(SpawnerTransform.Translation + vector3{Z := VerticalOffset * 10.0}, SpawnerTransform.Rotation, UpdateRateSeconds)
        MapIndicator.MoveTo(SpawnerTransform.Translation + vector3{Z := VerticalOffset * 10.0}, SpawnerTransform.Rotation, UpdateRateSeconds)

        Logger.Print("Returned Beacon to capture spawner")
    
    # Signal each player when a player grabs the objective.
    OnItemPickedUp(InAgent:agent):void=
        Logger.Print("Objective Grabbed")
        if(FortCharacter := InAgent.GetFortCharacter[]):
            ItemGrabbedMessageDevice.Show()
            spawn{FollowCharacter(FortCharacter)}

    # When a player drops an item, spawn a WaitForReturn() function
    # if the ReturnTime is greater than 0.
    OnItemDropped(InAgent:agent):void=
        Logger.Print("Objective Dropped")
        if(ReturnTime >= 0.0):
            spawn{WaitForReturn()}
        else:
            Logger.Print("The dropped objective does not return")
        
    # When the item is captured, award score to the capturing team, and return the indicators.
    OnItemCaptured(CapturingAgent:agent):void=
        Logger.Print("Objective Captured")
        ScoreManagerDevice.Activate()
        ReturnIndicators()

    # Wait a ReturnTime amount of Time, then return the indicators.
    WaitForReturn()<suspends>:void=
        Logger.Print("Waiting to return the indicators...")
        # Return the CaptureItem and Map indicators if the capture item
        # is not picked up before time expires.
        ShouldReturn:logic := race:
                block:
                    Sleep(ReturnTime)
                    true
                block:
                    CaptureItemSpawner.ItemPickedUpEvent.Await()
                    false
        
        if(ShouldReturn?):
            ReturnIndicators()
    
    # Causes the CaptureItemIndicator to continuously follow a player above their head.
    # Races between the update loop for the CaptureItemIndictator, and whether the player
    # captures the item, drops the item, or is eliminated.
    FollowCharacter(FortCharacter:fort_character)<suspends>:void=
        Logger.Print("Spawned FollowCharacter function")
        race:
            loop:
                Transform := FortCharacter.GetTransform()
                spawn{CaptureItemIndicator.MoveTo(Transform.Translation + vector3{Z := VerticalOffset}, Transform.Rotation, UpdateRateSeconds)}
                spawn{MapIndicator.MoveTo(Transform.Translation + vector3{Z := VerticalOffset}, Transform.Rotation, UpdateRateSeconds)}
                # We want to make sure this loop runs only once per simulation update, so we sleep for one game tick.
                Sleep(0.0)
            CaptureItemSpawner.ItemCapturedEvent.Await()
            CaptureItemSpawner.ItemDroppedEvent.Await()
            FortCharacter.EliminatedEvent().Await()
        Logger.Print("Objective dropped or captured")

    # Returns the map and capture item indicators back to their initial positions above the spawners.
    ReturnIndicators():void=
        SpawnerTransform := CaptureItemSpawner.GetTransform()
        # Teleport back to spawner, hiding the CaptureItemIndicator and MapIndicator above the map out of site.
        spawn{CaptureItemIndicator.MoveTo(SpawnerTransform.Translation + vector3{Z := VerticalOffset * 10.0}, SpawnerTransform.Rotation, UpdateRateSeconds)}
        spawn{MapIndicator.MoveTo(SpawnerTransform.Translation + vector3{Z := VerticalOffset * 10.0}, SpawnerTransform.Rotation, UpdateRateSeconds)}
        Logger.Print("Returned Indicators to capture spawner")
