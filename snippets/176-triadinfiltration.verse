# Knowledge-neutral Verse snippet; project names and asset IDs generalized.
# id: snippet-176

using { /Fortnite.com/Devices }
using { /Fortnite.com/FortPlayerUtilities }
using { /Verse.org/Simulation }
using { /Verse.org/Random }
using { /UnrealEngine.com/Temporary/Diagnostics }

triad_infiltration_log_channel := class(log_channel){}

triad_infiltration := class(creative_device):

    Logger:log = log{Channel := triad_infiltration_log_channel}

    # To avoid players not being able to join a team, you should set the maximum number
    # of players in the island settings to the sum of all of the Maximum(Team) variables.

    # Maximum number of players on the Infiltrators Team.
    @editable
    MaximumInfiltrators:int = 2

    # Maxmimum number of players on the Attackers Team.
    @editable
    MaximumAttackers:int = 4

    # Maximum number of players on the Defenders Team.
    @editable
    MaximumDefenders:int = 4

    # Array of Teleporters that teleport players to their team's spawn once the game starts. 
    @editable
    Teleporters:[]teleporter_device = array{}
    
    # Reference to the invisibility_manager script that controls infiltrator invisibility.  
    @editable
    InvisibilityManager:invisibility_manager = invisibility_manager{}

    # Array of weapon granters for each team.
    @editable
    var WeaponGranters:[]item_granter_device = array{}

    # Array of player spawners for each team.
    @editable
    PlayersSpawners:[]player_spawner_device = array{}

    @editable
    Cinematic:billboard_cinematic = billboard_cinematic{}
    
    # Reference to the infiltrators team.
    var MaybeInfiltrators:?team = false

    # Reference to the attackers team.
    var MaybeAttackers:?team = false

    # Rerfernece to the defenders team.
    var MaybeDefenders:?team = false

    # Array of all teams in the game.
    var AllTeams:[]team = array{}
    
    # Map of teams to their maximum number of players.
    var TeamsAndTotals:[team]int = map{}

    OnBegin<override>()<suspends>:void =
        # Play cinematic at beginning if user has it enabled
        spawn{Cinematic.EnterCinematicModeForAll(GetPlayspace().GetPlayers())}

        # Get all the Teams.
        set AllTeams = GetPlayspace().GetTeamCollection().GetTeams()
        var AllPlayers:[]player = GetPlayspace().GetPlayers()
        # Save the teams to later reference them.
        set MaybeInfiltrators = option{AllTeams[0]}
        set MaybeAttackers = option{AllTeams[1]}
        set MaybeDefenders = option{AllTeams[2]}
        
        if:
            Infiltrators := MaybeInfiltrators?
            Attackers := MaybeAttackers?
            Defenders :=  MaybeDefenders?
            Logger.Print("Found all three teams")
            set TeamsAndTotals[Infiltrators] = MaximumInfiltrators
            set TeamsAndTotals[Attackers] = MaximumAttackers
            set TeamsAndTotals[Defenders] = MaximumDefenders
            Logger.Print("Set all three teams in TeamsAndTotals")
        then:
            #Subscribe to PlayerAddedEvent to allow team rebalancing when a new player joins the game.
            GetPlayspace().PlayerAddedEvent().Subscribe(OnPlayerAdded)
            for(PlayerSpawner:PlayersSpawners):
                PlayerSpawner.SpawnedEvent.Subscribe(OnPlayerSpawn)
            
            BalanceTeams()
            Logger.Print("Teams balanced, calling invisibility script")
            InvisibilityManager.StartInvisibilityManager(AllTeams, AllPlayers, Infiltrators)
            Sleep(0.25)
            TeleportPlayersToStartLocations()
        else:
            Logger.Print("Couldn't find all teams, make sure to assign the correct teams in your island settings.")

    # Grants players a weapon based on the index of their team in the Teams array
    # by indexing into the WeaponGranters array.
    GrantTeamWeapon(InPlayer:player):void=
        if(CurrentTeam := GetPlayspace().GetTeamCollection().GetTeam[InPlayer]):
            for(TeamIndex -> PlayerTeam:AllTeams, PlayerTeam = CurrentTeam):
                if(WeaponGranter := WeaponGranters[TeamIndex]):
                    WeaponGranter.GrantItem(InPlayer)
                    Logger.Print("Granted the a Player on team {TeamIndex + 1} a weapon")
    
    # Runs when any player spawns from a spawn pad. 
    # Calls GrantTeamWeapon using the provided SpawnedAgent.
    OnPlayerSpawn(SpawnedAgent:agent):void=
        if(SpawnedPlayer := player[SpawnedAgent]):
            Logger.Print("Attempting to grant spawned player a weapon")
            GrantTeamWeapon(SpawnedPlayer)

    # Handles a new player joining the game.
    OnPlayerAdded(InPlayer:player):void=
        Logger.Print("A new Player joined, assigning them to a team")
        FortTeamCollection := GetPlayspace().GetTeamCollection()
        
        # Assign the new player to the smallest team, asymmetrically.
        BalancePlayer(InPlayer)

        for:
            TeamIndex -> PlayerTeam:AllTeams
            PlayerTeam = FortTeamCollection.GetTeam[InPlayer]
            TeamTeleporter := Teleporters[TeamIndex]
            Transform := TeamTeleporter.GetTransform()
        do:
            InPlayer.Respawn(Transform.Translation, Transform.Rotation)
            Logger.Print("Teleported the spawned player to their start location")
            # If the player was an infiltrator, call OnInfiltratorJoined in
            # InvisibilityManager.
            if(PlayerTeam = MaybeInfiltrators?):
                InvisibilityManager.OnInfiltratorJoined(InPlayer)
 
    # Balances all players on all teams in the game
    BalanceTeams():void=
        Logger.Print("Beginning to balance teams")
        var AllPlayers:[]player := GetPlayspace().GetPlayers()
        set AllPlayers = Shuffle(AllPlayers)
        Logger.Print("AllPlayers Length is {AllPlayers.Length}")

        for (TeamPlayer:AllPlayers):
            BalancePlayer(TeamPlayer)
     

    # For each player, iterate through the list of teams and assign them to the 
    # team with the least amount of players, or their starting team in case of ties.
    BalancePlayer(InPlayer:player):void=
        Logger.Print("Beginning to balance player")
        var TeamToAssign:?team = false
        set TeamToAssign = FindTeamWithLargestDifference()
        if (AssignedTeam := TeamToAssign?, GetPlayspace().GetTeamCollection().AddToTeam[InPlayer, AssignedTeam]):
            Logger.Print("Assigned player to a new team")
        else:
            Logger.Print("This player was already on the smallest team")

    # Finds the team with the largest difference in their number of players from their
    # maximum number of players.
    FindTeamWithLargestDifference():?team =
        Logger.Print("Attempting to find smallest team")
        var TeamToAssign:?team = false
        var LargestDifference:int = 0
        for:
            CandidateTeamIndex -> CandidateTeam:AllTeams
            CurrentTeamSize := GetPlayspace().GetTeamCollection().GetAgents[CandidateTeam].Length 
            MaximumTeamSize := TeamsAndTotals[CandidateTeam]
        do:
            Logger.Print("Checking a team...")
            Logger.Print("Maximum size of team {CandidateTeamIndex + 1} is {MaximumTeamSize}")
            DifferenceFromMaximum := MaximumTeamSize - CurrentTeamSize 
            Logger.Print("Difference from maximum is {DifferenceFromMaximum}")
            if(LargestDifference < DifferenceFromMaximum):
                set LargestDifference = DifferenceFromMaximum
                set TeamToAssign = option{CandidateTeam}
                Logger.Print("Found team {CandidateTeamIndex + 1} with difference {DifferenceFromMaximum}")

        return TeamToAssign 
        
    # Teleports players to their team's spawn after team balancing finishes.
    TeleportPlayersToStartLocations():void=
        Logger.Print("Teleporting players to start locations")
        for:
            TeamIndex -> PlayerTeam:AllTeams 
            TeamPlayers := GetPlayspace().GetTeamCollection().GetAgents[PlayerTeam]
            TeamTeleporter := Teleporters[TeamIndex]
        do:
            for(TeamPlayer:TeamPlayers):
                TeamTeleporter.Teleport(TeamPlayer)
                Logger.Print("Teleported this player to their start location")
