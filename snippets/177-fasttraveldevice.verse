# Knowledge-neutral Verse snippet; project names and asset IDs generalized.
# id: snippet-177

using { /Fortnite.com/Characters }
using { /Fortnite.com/Devices }
using { /Fortnite.com/Game }
using { /Fortnite.com/UI }
using { /UnrealEngine.com/Temporary/UI }
using { /Verse.org/Input }
using { /Verse.org/Input/UI }
using { /Verse.org/Simulation }

DefaultCategoryName<localizes> : message = "Category"
DefaultDestinationTitle<localizes> : message = "Destination"
EmptyLabel<localizes> : message = ""

Category1Name<localizes> : message = "Creative Devices"
Category2Name<localizes> : message = "Verse UI"
Category3Name<localizes> : message = "Scene Graph UI"
Category4Name<localizes> : message = "UI Art"
Category5Name<localizes> : message = "UI Feature Examples"

fast_travel_destination := struct<concrete>:

    @editable
    Title : message = DefaultDestinationTitle

    @editable
    Description : message = EmptyLabel

    @editable
    Teleporter : teleporter_device = teleporter_device{}

fast_travel_category := struct<concrete>:

    @editable
    Name : message = DefaultCategoryName

    @editable
    Destinations : []fast_travel_destination = array{}

fast_travel_device := class(creative_device):  

    DestinationsPerPage : int = 6 
    CategorySlots : int = 5   

    AnimTriggerHoldSeconds : float = 0.1

    # This input trigger device is used to open the menu
    @editable
    ShowInput : input_trigger_device = input_trigger_device{}

    # Edit these details in the device details panel 
    @editable
    Categories : []fast_travel_category = array:
        fast_travel_category{Name := Category1Name}
        fast_travel_category{Name := Category2Name}
        fast_travel_category{Name := Category3Name}  
        fast_travel_category{Name := Category4Name}
        fast_travel_category{Name := Category5Name}

    MyWidget : UserInterfaces.Shared.Widgets.UW_FastTravel_Modal = UserInterfaces.Shared.Widgets.UW_FastTravel_Modal{}

    var CurrentAgent : ?agent = false
    var IsOpen : logic = false
    var ScreenWidget : ?overlay = false   
    var MenuSubscriptions : []cancelable_event = array{}
    var IsMobileLayout : logic = false     
    var ActiveCategory : int = 0
    var ActivePage : int = 0
    var PageDestinations : []fast_travel_destination = array{}

    OnBegin<override>()<suspends> : void =
        DisableAllTeleporters()

        ShowInput.Enable()
        ShowInput.PressedEvent.Subscribe(OnShowPressed)

        for (Player : GetPlayspace().GetPlayers()):
            ShowInput.Register(Player)
        GetPlayspace().PlayerAddedEvent().Subscribe(OnPlayerAdded)

    OnPlayerAdded(Player : player) : void =
        ShowInput.Register(Player)

    OnShowPressed(Agent : agent) : void =
        if (IsOpen = true):
            return

        set CurrentAgent = option{Agent}
        set IsOpen = true

        set ActiveCategory = Min(Max(ActiveCategory, 0), Max(Categories.Length - 1, 0))

        FullScreenUI : overlay = overlay:
            Slots := array:
                overlay_slot:
                    Widget := MyWidget  
                    HorizontalAlignment := horizontal_alignment.Fill
                    VerticalAlignment := vertical_alignment.Fill

        if (Player := player[Agent]):
            SyncInputMethod(Player)

            if (PlayerUI := GetPlayerUI[Player]):
                PlayerUI.AddWidget(FullScreenUI, player_ui_slot{ InputMode := ui_input_mode.All })
                set ScreenWidget = option{FullScreenUI}  

                PlayerUI.SetFocus(MyWidget)

            # Add the mappings context for the tab/page/close input actions
            if (PlayerInput := GetPlayerInput[Player]):
                PlayerInput.AddInputMapping(UI.MenuNavigationMapping)

            spawn { WatchInputMethod(Player) }

        set MenuSubscriptions = array:
            SubscribeToEvent(MyWidget.OnCategorySelected, OnCategorySelected)
            SubscribeToEvent(MyWidget.OnNavigateTab, OnNavigateTab)   
            SubscribeToEvent(MyWidget.OnNavigatePage, OnNavigatePage)
            SubscribeToEvent(MyWidget.OnDestinationSelected, OnDestinationSelected)
            SubscribeToEvent(MyWidget.OnCloseClicked, OnCloseClicked)   

        # Reset anim states
        set MyWidget.AnimTrigger_Startup = false  
        set MyWidget.AnimTrigger_NextPage = false
        set MyWidget.AnimTrigger_PrevPage = false

        ApplyCategories()
        ApplyPage()

        spawn { PulseStartupAnim() }
   
    CloseMenu() : void =
        if (IsOpen = false):
            return

        set IsOpen = false

        if (Player := player[CurrentAgent?]):
            if (PlayerInput := GetPlayerInput[Player]):
                PlayerInput.RemoveInputMapping(UI.MenuNavigationMapping)
            if (PlayerUI := GetPlayerUI[Player], Screen := ScreenWidget?):  
                PlayerUI.RemoveWidget(Screen)

        set ScreenWidget = false   

        for (Subscription : MenuSubscriptions):
            Subscription.Cancel()
        set MenuSubscriptions = array{}

        set CurrentAgent = false

    SyncInputMethod(Player : player) : void =
        var UsingTouch : logic = false

        if (PlayerInput := GetPlayerInput[Player]):
            case (PlayerInput.PreferredInputMethod):   
                input_method.Touch => set UsingTouch = true
                input_method.Gamepad => set UsingTouch = false
                input_method.KeyboardAndMouse => set UsingTouch = false
                _ => {}

        if (UsingTouch = IsMobileLayout):
            return

        set IsMobileLayout = UsingTouch
        set MyWidget.IsMobile = UsingTouch

    WatchInputMethod(Player : player)<suspends> : void =   
        loop:
            if (IsOpen = false):
                break

            SyncInputMethod(Player)
            Sleep(0.0)

    # Every place in the tab being looked at
    CurrentDestinations() : []fast_travel_destination =
        var Result : []fast_travel_destination = array{}
        if (Category := Categories[ActiveCategory]):     
            set Result = Category.Destinations
        Result

    # Determine how many pages this tab needs
    PageCount() : int =
        Count := CurrentDestinations().Length
        var Pages : int = 1
        if (Ratio := (Count + DestinationsPerPage - 1) / DestinationsPerPage):
            set Pages = Max(1, Floor(Ratio))
        Pages

    ApplyPage() : void =
        Destinations := CurrentDestinations()
        Pages := PageCount()

        set ActivePage = Min(Max(ActivePage, 0), Pages - 1)
        set MyWidget.CurrentCategory = ActiveCategory + 1  
        set MyWidget.CurrentPage = ActivePage + 1
        set MyWidget.MaxPages = Pages

        set MyWidget.PrevPageVisibility = if (ActivePage > 0) { true } else { false }  
        set MyWidget.NextPageVisibility = if (ActivePage < Pages - 1) { true } else { false }

        var Remaining : []fast_travel_destination = array{}
        if (Tail := Destinations.Slice[ActivePage * DestinationsPerPage]):
            set Remaining = Tail

        # Fill each row
        var OnScreen : []fast_travel_destination = array{}
        for (Slot := 0..DestinationsPerPage - 1):
            var MaybeDestination : ?fast_travel_destination = false  
            if (Destination := Remaining[Slot]):
                set MaybeDestination = option{Destination}
                set OnScreen = OnScreen + array{Destination}   
            ApplyDestinationSlot(Slot + 1, MaybeDestination)

        set PageDestinations = OnScreen
        set MyWidget.HighlightedDestination = 0

    # Apply destination data
    ApplyDestinationSlot(SlotNumber : int, MaybeDestination : ?fast_travel_destination) : void =
        var Title : message = EmptyLabel
        var Description : message = EmptyLabel
        var Visible : logic = false  
  
        if (Destination := MaybeDestination?):
            set Title = Destination.Title
            set Description = Destination.Description
            set Visible = true

        case (SlotNumber):
            1 =>
                set MyWidget.Destination1Title = Title
                set MyWidget.Destination1Desc = Description
                set MyWidget.IsDestination1Visible = Visible
            2 =>
                set MyWidget.Destination2Title = Title
                set MyWidget.Destination2Desc = Description  
                set MyWidget.IsDestination2Visible = Visible
            3 =>
                set MyWidget.Destination3Title = Title
                set MyWidget.Destination3Desc = Description
                set MyWidget.IsDestination3Visible = Visible
            4 =>
                set MyWidget.Destination4Title = Title   
                set MyWidget.Destination4Desc = Description
                set MyWidget.IsDestination4Visible = Visible
            5 =>  
                set MyWidget.Destination5Title = Title
                set MyWidget.Destination5Desc = Description
                set MyWidget.IsDestination5Visible = Visible
            6 =>
                set MyWidget.Destination6Title = Title
                set MyWidget.Destination6Desc = Description
                set MyWidget.IsDestination6Visible = Visible
            _ => {}

    ApplyCategories() : void =
        for (Index := 0..CategorySlots - 1):
            var MaybeCategory : ?fast_travel_category = false
            if (Category := Categories[Index]):    
                set MaybeCategory = option{Category}
            ApplyCategorySlot(Index + 1, MaybeCategory)

    ApplyCategorySlot(SlotNumber : int, MaybeCategory : ?fast_travel_category) : void =
        var Name : message = EmptyLabel

        if (Category := MaybeCategory?):  
            set Name = Category.Name

        case (SlotNumber):
            1 => set MyWidget.Category1Text = Name
            2 => set MyWidget.Category2Text = Name
            3 => set MyWidget.Category3Text = Name  
            4 => set MyWidget.Category4Text = Name
            5 => set MyWidget.Category5Text = Name  
            _ => {}

    GoToCategory(Index : int) : void =
        if (Index < CategorySlots, Categories[Index]):   
            PreviousCategory := ActiveCategory
            set ActiveCategory = Index
            set ActivePage = 0    
            ApplyPage()
            PulseTransition(PreviousCategory, ActiveCategory)

    OnCategorySelected(CategoryIndex : tuple(int)) : void =
        GoToCategory(ToZeroBased(CategoryIndex(0)))

    OnNavigateTab(TabIndex : tuple(int)) : void =  
        GoToCategory(ActiveCategory + TabIndex(0))  

    OnNavigatePage(PageIndex : tuple(int)) : void =
        PriorPage := ActivePage     
        set ActivePage = ActivePage + PageIndex(0)
        ApplyPage()

        PulseTransition(PriorPage, ActivePage)  

    OnDestinationSelected(DestinationIndex : tuple(int)) : void =
        Slot := ToZeroBased(DestinationIndex(0))
        if (Destination := PageDestinations[Slot]):
            set MyWidget.HighlightedDestination = Slot + 1
            TravelTo(Destination, Slot + 1)
            CloseMenu()
  
    OnCloseClicked(Event : tuple()) : void =
        CloseMenu()

    DisableAllTeleporters() : void =
        for (Category : Categories, Destination : Category.Destinations):
            Destination.Teleporter.Disable()

    TravelTo(Destination : fast_travel_destination, SlotNumber : int) : void =
        if (Agent := CurrentAgent?):
            spawn { TeleportAgent(Destination.Teleporter, Agent) }
        else:
            Print("Could not travel - no agent.")

    TeleportAgent(Teleporter : teleporter_device, Agent : agent)<suspends> : void =
        Teleporter.Enable()
        Sleep(0.0)
        Teleporter.Teleport(Agent)
        Sleep(0.0)
        Teleporter.Disable()

    # Plays the animation to match the direction we are moving
    PulseTransition(From : int, To : int) : void = 
        if (To > From):
            spawn { PulseNextPageAnim() } 
        else if (To < From):
            spawn { PulsePrevPageAnim() }    

    # Animation flags
    PulseStartupAnim()<suspends> : void =  
        set MyWidget.AnimTrigger_Startup = true
        Sleep(AnimTriggerHoldSeconds)
        set MyWidget.AnimTrigger_Startup = false

    PulseNextPageAnim()<suspends> : void =
        set MyWidget.AnimTrigger_NextPage = true 
        Sleep(AnimTriggerHoldSeconds)
        set MyWidget.AnimTrigger_NextPage = false

    PulsePrevPageAnim()<suspends> : void =
        set MyWidget.AnimTrigger_PrevPage = true
        Sleep(AnimTriggerHoldSeconds)   
        set MyWidget.AnimTrigger_PrevPage = false

    ToZeroBased(Payload : int) : int =
        Payload - 1
