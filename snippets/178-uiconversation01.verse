# Knowledge-neutral Verse snippet; project names and asset IDs generalized.
# id: snippet-178

using { /Verse.org/Colors }
using { /Verse.org/Simulation }
using { /UnrealEngine.com/Temporary/UI }
using { /Verse.org/Assets }

ui_conversation01 := class():

    UnavailableStatusTextColor : color = color {R := 0.577581, G := 0.577581, B := 0.577581}
    DefaultStatusTextColor : color = color {R := 1.0, G := 1.0, B := 1.0}

    var ConversationWidget : UserInterfaces.UIFeatureExamples.Conversations.Widgets.UW_Conversation01 = UserInterfaces.UIFeatureExamples.Conversations.Widgets.UW_Conversation01 {}

    BuildUI(InAgent : agent): void = 
        if (Player := player[InAgent]):
            if (PlayerUI := GetPlayerUI[Player]):
                PlayerUI.AddWidget(ConversationWidget)

    RemoveUI(InPlayer : player) : void =
        if (PlayerUI := GetPlayerUI[InPlayer]):
            PlayerUI.RemoveWidget(ConversationWidget)

    SetCanSpeakState(InMessage: message): void =
        set ConversationWidget.HasFailedToProcess = false
        set ConversationWidget.IsProcessing = false
        set ConversationWidget.IsPlayingAudio = false
        set ConversationWidget.ShowPersonaIcon = false     
        set ConversationWidget.StatusText = InMessage   
        set ConversationWidget.ShowStatusActionWidget = true
        set ConversationWidget.TimeLeftForAudioRecordingNormalized = 1.0
        set ConversationWidget.IconOpacity = 1.0
        set ConversationWidget.StatusTextColor = DefaultStatusTextColor

    SetUnavailableState(InMessage: message): void =
        set ConversationWidget.HasFailedToProcess = true
        set ConversationWidget.IsProcessing = false
        set ConversationWidget.IsPlayingAudio = false
        set ConversationWidget.ShowPersonaIcon = false     
        set ConversationWidget.StatusText = InMessage   
        set ConversationWidget.ShowStatusActionWidget = false
        set ConversationWidget.IconOpacity = 0.5
        set ConversationWidget.StatusTextColor = UnavailableStatusTextColor
        set ConversationWidget.TimeLeftForAudioRecordingNormalized = 1.0

    SetRecordingState(InMessage: message): void =
        set ConversationWidget.HasFailedToProcess = false
        set ConversationWidget.IsProcessing = false
        set ConversationWidget.IsPlayingAudio = false
        set ConversationWidget.ShowPersonaIcon = false     
        set ConversationWidget.StatusText = InMessage   
        set ConversationWidget.ShowStatusActionWidget = true
        set ConversationWidget.IconOpacity = 1.0
        set ConversationWidget.StatusTextColor = DefaultStatusTextColor

    SetProcessingState(InMessage: message): void =
        set ConversationWidget.HasFailedToProcess = false
        set ConversationWidget.IsProcessing = true
        set ConversationWidget.IsPlayingAudio = false
        set ConversationWidget.ShowPersonaIcon = false     
        set ConversationWidget.StatusText = InMessage   
        set ConversationWidget.ShowStatusActionWidget = false
        set ConversationWidget.TimeLeftForAudioRecordingNormalized = 0.0
        set ConversationWidget.IconOpacity = 0.5
        set ConversationWidget.StatusTextColor = DefaultStatusTextColor

    SetSpeakingState(InMessage: message): void =
        set ConversationWidget.HasFailedToProcess = false
        set ConversationWidget.IsProcessing = false
        set ConversationWidget.IsPlayingAudio = true
        set ConversationWidget.ShowPersonaIcon = true     
        set ConversationWidget.StatusText = InMessage   
        set ConversationWidget.ShowStatusActionWidget = false
        set ConversationWidget.TimeLeftForAudioRecordingNormalized = 0.0
        set ConversationWidget.IconOpacity = 1.0
        set ConversationWidget.StatusTextColor = DefaultStatusTextColor

    UpdateRecordingProgress(InValue : float) : void =
        set ConversationWidget.TimeLeftForAudioRecordingNormalized = InValue

    UpdatePersonaAvatar(Avatar : texture) : void = 
        set ConversationWidget.PersonaIcon = Avatar

    UpdatePersonaName(Name : message) : void = 
        set ConversationWidget.PersonaName = Name
        set ConversationWidget.ShowPersonaName = true
