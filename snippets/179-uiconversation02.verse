# Knowledge-neutral Verse snippet; project names and asset IDs generalized.
# id: snippet-179

using { /Verse.org/Colors }
using { /Verse.org/Simulation }
using { /UnrealEngine.com/Temporary/UI }
using { /Verse.org/Assets }

ui_conversation02 := class():    

    UnavailableStatusTextColor : color = color {R := 0.577581, G := 0.577581, B := 0.577581}
    DefaultStatusTextColor : color = color {R := 1.0, G := 1.0, B := 1.0}

    var ConversationWidget : UserInterfaces.UIFeatureExamples.Conversations.Widgets.UW_Conversation02 = UserInterfaces.UIFeatureExamples.Conversations.Widgets.UW_Conversation02 {}

    BuildUI(InAgent : agent): void = 
        if (Player := player[InAgent]):
            if (PlayerUI := GetPlayerUI[Player]):
                PlayerUI.AddWidget(ConversationWidget)

    RemoveUI(InPlayer : player) : void =
        if (PlayerUI := GetPlayerUI[InPlayer]):
            PlayerUI.RemoveWidget(ConversationWidget)

    SetCanSpeakState(InMessage: message): void =
        set ConversationWidget.HasFailedToProcess = false
        set ConversationWidget.IsPlayingAudio = false
        set ConversationWidget.StatusText = InMessage   
        set ConversationWidget.ShowStatusActionWidget = true
        set ConversationWidget.IsWaiting = true
        set ConversationWidget.IsProcessing = false    
        set ConversationWidget.IsRecording = false
        set ConversationWidget.StatusTextColor = DefaultStatusTextColor

    SetUnavailableState(InMessage: message): void =
        set ConversationWidget.HasFailedToProcess = true
        set ConversationWidget.IsPlayingAudio = false
        set ConversationWidget.StatusText = InMessage   
        set ConversationWidget.ShowStatusActionWidget = false
        set ConversationWidget.IsWaiting = false
        set ConversationWidget.IsProcessing = false    
        set ConversationWidget.IsRecording = false
        set ConversationWidget.StatusTextColor = UnavailableStatusTextColor

    SetRecordingState(InMessage: message): void =
        set ConversationWidget.HasFailedToProcess = false
        set ConversationWidget.IsPlayingAudio = false
        set ConversationWidget.StatusText = InMessage   
        set ConversationWidget.ShowStatusActionWidget = true
        set ConversationWidget.IsWaiting = false
        set ConversationWidget.IsProcessing = false    
        set ConversationWidget.IsRecording = true
        set ConversationWidget.StatusTextColor = DefaultStatusTextColor
    
    SetProcessingState(InMessage: message): void =
        set ConversationWidget.HasFailedToProcess = false
        set ConversationWidget.IsPlayingAudio = false
        set ConversationWidget.StatusText = InMessage   
        set ConversationWidget.ShowStatusActionWidget = false
        set ConversationWidget.IsWaiting = false
        set ConversationWidget.IsProcessing = true    
        set ConversationWidget.IsRecording = false
        set ConversationWidget.StatusTextColor = DefaultStatusTextColor

    SetSpeakingState(InMessage: message): void =
        set ConversationWidget.HasFailedToProcess = false
        set ConversationWidget.IsPlayingAudio = true
        set ConversationWidget.StatusText = InMessage   
        set ConversationWidget.ShowStatusActionWidget = false
        set ConversationWidget.IsWaiting = false
        set ConversationWidget.IsProcessing = false    
        set ConversationWidget.IsRecording = false
        set ConversationWidget.StatusTextColor = DefaultStatusTextColor

    UpdateRecordingProgress(InValue : float) : void =
        set ConversationWidget.TimeLeftForAudioRecordingNormalized = InValue

    UpdatePersonaAvatar(Avatar : texture) : void = 
        set ConversationWidget.PersonaAvatar = Avatar
