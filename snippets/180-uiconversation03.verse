# Knowledge-neutral Verse snippet; project names and asset IDs generalized.
# id: snippet-180

using { /Verse.org/Simulation }
using { /UnrealEngine.com/Temporary/UI }
using { /Verse.org/Assets }

ui_conversation03 := class():

    ConversationWidget : UserInterfaces.UIFeatureExamples.Conversations.Widgets.UW_Conversation03 = UserInterfaces.UIFeatureExamples.Conversations.Widgets.UW_Conversation03{}

    BuildUI(InAgent : agent) : void = 
        if (Player := player[InAgent]):
            if (PlayerUI := GetPlayerUI[Player]):
                PlayerUI.AddWidget(ConversationWidget)

    RemoveUI(InPlayer : player) : void =
        if (PlayerUI := GetPlayerUI[InPlayer]):
            PlayerUI.RemoveWidget(ConversationWidget)

    SetCanSpeakState(InMessage: message): void = 
        set ConversationWidget.TimeLeftForAudioRecordingNormalized = 0.0
        set ConversationWidget.IsProcessing = false
        set ConversationWidget.IsPlayingAudio = false
        set ConversationWidget.IsRecording = false
        set ConversationWidget.StatusMessage = InMessage
        set ConversationWidget.ShowActionWidget = true  
        set ConversationWidget.PersonaMessage = StringToMessage("")

    SetUnavailableState(InMessage: message): void = 
        set ConversationWidget.TimeLeftForAudioRecordingNormalized = 0.0
        set ConversationWidget.IsProcessing = false
        set ConversationWidget.IsPlayingAudio = false
        set ConversationWidget.IsRecording = false
        set ConversationWidget.StatusMessage = InMessage
        set ConversationWidget.ShowActionWidget = false  
        set ConversationWidget.PersonaMessage = StringToMessage("")

    SetRecordingState(InMessage: message): void =
        set ConversationWidget.IsProcessing = false
        set ConversationWidget.IsPlayingAudio = false
        set ConversationWidget.StatusMessage = InMessage
        set ConversationWidget.IsRecording = true
        set ConversationWidget.ShowActionWidget = true
        set ConversationWidget.PersonaMessage = StringToMessage("")

    SetProcessingState(InMessage: message): void =
        set ConversationWidget.TimeLeftForAudioRecordingNormalized = 0.0
        set ConversationWidget.IsProcessing = true
        set ConversationWidget.IsPlayingAudio = false
        set ConversationWidget.IsRecording = false
        set ConversationWidget.StatusMessage = InMessage
        set ConversationWidget.ShowActionWidget = false
        set ConversationWidget.PersonaMessage = StringToMessage("")

    SetSpeakingState(): void =
        set ConversationWidget.IsProcessing = false
        set ConversationWidget.IsRecording = false
        set ConversationWidget.IsPlayingAudio = true
        set ConversationWidget.ShowActionWidget = false
        set ConversationWidget.PersonaMessage = StringToMessage("")
        set ConversationWidget.StatusMessage = StringToMessage("")
        set ConversationWidget.TimeLeftForAudioRecordingNormalized = 0.0

    ShowPersonaAnswer(InPersonaMessage : message): void = 
        set ConversationWidget.PersonaMessage = InPersonaMessage

    UpdateRecordingProgress(InValue : float) : void =
        set ConversationWidget.TimeLeftForAudioRecordingNormalized = InValue

    UpdatePersonaAvatar(Avatar : texture) : void = 
        set ConversationWidget.PersonaAvatar = Avatar

    UpdatePersonaName(Name : message) : void = 
        set ConversationWidget.PersonaName = Name
