# Knowledge-neutral Verse snippet; project names and asset IDs generalized.
# id: snippet-181

using { /Verse.org/Colors }
using { /UnrealEngine.com/Temporary/UI }
using { /Verse.org/Simulation }
using { /Verse.org/Assets }

ui_conversation04 := class():

    UnavailableStatusTextColor : color = color {R := 0.577581, G := 0.577581, B := 0.577581}
    DefaultStatusTextColor : color = color {R := 1.0, G := 1.0, B := 1.0}

    var ConversationWidget : UserInterfaces.UIFeatureExamples.Conversations.Widgets.UW_Conversation04 = UserInterfaces.UIFeatureExamples.Conversations.Widgets.UW_Conversation04 {}

    BuildUI(InAgent : agent) : void =
        if (Player := player[InAgent]):
            if (PlayerUI := GetPlayerUI[Player]):
                PlayerUI.AddWidget(ConversationWidget)
                set ConversationWidget.ShowAnim = true

    RemoveUI(InPlayer : player) : void =
        set ConversationWidget.ShowAnim = false
        spawn {PlayHideAnim(InPlayer)}

    PlayHideAnim(InPlayer : player)<suspends> : void =
        var Exit : logic = false
        loop:
            if (Exit = false):
                set ConversationWidget.HideAnim = true
                Sleep(0.5)
                set ConversationWidget.HideAnim = false
                set Exit = true
                if (PlayerUI := GetPlayerUI[InPlayer]):
                    PlayerUI.RemoveWidget(ConversationWidget)
            else:
                break;


    SetCanSpeakState(InMessage: message): void =
        set ConversationWidget.HasFailedToProcess = false
        set ConversationWidget.IsProcessing = false
        set ConversationWidget.IsPlayingAudio = false        
        set ConversationWidget.IsRecording = false
        set ConversationWidget.ShowStatusActionWidget = true
        set ConversationWidget.TimeLeftForAudioRecordingNormalized = 1.0
        set ConversationWidget.StatusText = InMessage        
        set ConversationWidget.StatusTextColor = DefaultStatusTextColor

    SetUnavailableState(InMessage: message): void = 
        set ConversationWidget.HasFailedToProcess = true
        set ConversationWidget.IsProcessing = false
        set ConversationWidget.IsPlayingAudio = false        
        set ConversationWidget.IsRecording = false
        set ConversationWidget.ShowStatusActionWidget = false
        set ConversationWidget.TimeLeftForAudioRecordingNormalized = 0.0
        set ConversationWidget.StatusText = InMessage        
        set ConversationWidget.StatusTextColor = UnavailableStatusTextColor

    SetRecordingState(InMessage: message): void =
        set ConversationWidget.HasFailedToProcess = false
        set ConversationWidget.IsProcessing = false
        set ConversationWidget.IsPlayingAudio = false        
        set ConversationWidget.IsRecording = true
        set ConversationWidget.ShowStatusActionWidget = true
        set ConversationWidget.StatusText = InMessage        
        set ConversationWidget.StatusTextColor = DefaultStatusTextColor

    SetProcessingState(InMessage: message): void =
        set ConversationWidget.HasFailedToProcess = false
        set ConversationWidget.IsProcessing = true
        set ConversationWidget.IsPlayingAudio = false        
        set ConversationWidget.IsRecording = false
        set ConversationWidget.ShowStatusActionWidget = false
        set ConversationWidget.StatusText = InMessage        
        set ConversationWidget.StatusTextColor = DefaultStatusTextColor
        set ConversationWidget.TimeLeftForAudioRecordingNormalized = 0.0

    SetSpeakingState(): void =
        set ConversationWidget.HasFailedToProcess = false
        set ConversationWidget.IsProcessing = false
        set ConversationWidget.IsPlayingAudio = true        
        set ConversationWidget.IsRecording = false
        set ConversationWidget.ShowStatusActionWidget = false
        set ConversationWidget.StatusText = StringToMessage("")       
        set ConversationWidget.StatusTextColor = DefaultStatusTextColor
        set ConversationWidget.TimeLeftForAudioRecordingNormalized = 0.0

    UpdateRecordingProgress(InValue : float) : void =
        set ConversationWidget.TimeLeftForAudioRecordingNormalized = InValue

    UpdatePersonaAvatar(Avatar : texture) : void = 
        set ConversationWidget.PersonaIcon = Avatar

    UpdatePersonaName(Name : message) : void = 
        set ConversationWidget.PersonaName = Name
