# Knowledge-neutral Verse snippet; project names and asset IDs generalized.
# id: snippet-182

using { /Fortnite.com/Devices }
using { /Verse.org/Simulation }
using { /Verse.org/SceneGraph }
using { /Verse.org/AgentGroup }
using { /Verse.org/Chat }
using { /UnrealEngine.com/Conversations }
using { /Fortnite.com/Characters }
using { /Verse.org/Assets }
using { /Verse.org/Presentation }
using { /Verse.org/Input }
using { /Fortnite.com/Input/Character }
using { /UnrealEngine.com/Temporary/UI }
using { /Fortnite.com/Vehicles }
using { /Fortnite.com/Input }

team_member_info := class(has_voice_member_info) {}

ui_conversations_device := class(creative_device):

    @editable
    HUDController : hud_controller_device = hud_controller_device {}

    @editable
    Volume : volume_device = volume_device {}

    @editable
    ShowConversation03UITrigger : trigger_device = trigger_device {}

    @editable
    ShowConversation04UIVolume : volume_device = volume_device {}

    @editable
    ShowConversation01UIVolume : volume_device = volume_device {}

    @editable
    ShowConversation02UIVolume : volume_device = volume_device {}

    @editable
    Conversation03DoorSequence : cinematic_sequence_device = cinematic_sequence_device{}

    @editable
    Conversation03Camera : gameplay_camera_fixed_angle_device = gameplay_camera_fixed_angle_device{}

    @editable
    Conversation03Controls : gameplay_controls_third_person_device = gameplay_controls_third_person_device{}

    @editable
    PersonaSpawner : npc_spawner_device = npc_spawner_device {}

    var MaybeNPCAgent : ?agent = false
    var MaybePersonaEntity : ?entity = false
    var PersonaCompOpt : ?persona_component = false
    var MaybeAgentGroup : ?agent_group(team_member_info) = false
    var MaybeVoiceChannel : ?voice_channel(team_member_info) = false

    var PersonaName : message = UnknownMsg

    var CanSpeakTo : logic = false

    var IsRecording : logic = false

    var IsProcessing : logic = false

    var IsPlayingAudio : logic = false

    var HasFailedToProcess : logic = true

    RecordingDuration : float = 5000.0

    var PersonaIcon : ?texture = false

    TalkDialogMessage<localizes> : message = "Hold the key to record voice"
    TalkMessage<localizes> : message = "Talk (Hold)"
    RecordingMessage<localizes> : message = "Recording"
    UnavailableMessage<localizes> : message = "Unavailable"
    ProcessingMessage<localizes> : message = "Processing"
    SpeakingMessage<localizes> : message = "Speaking"

    Conversation03 : ui_conversation03 = ui_conversation03 {}
    Conversation01 : ui_conversation01 = ui_conversation01 {}
    Conversation02 : ui_conversation02 = ui_conversation02 {}
    Conversation04 : ui_conversation04 = ui_conversation04 {}

    var PersonaSpawnsSubscriptionOption : ?cancelable = false
    var PlayerEntersConversationsVolumeSubscriptionOption : ?cancelable = false  
    var PlayerExitsConversationsVolumeSubscriptionOption : ?cancelable = false
    var PlayerTriggersConversation03TriggerSubscriptionOption : ?cancelable = false
    var PlayerEntersConversation04VolumeSubscriptionOption : ?cancelable = false
    var PlayerExitsConversation04VolumeSubscriptionOption : ?cancelable = false
    var PlayerEntersConversation01VolumeSubscriptionOption : ?cancelable = false
    var PlayerExitsConversation01VolumeSubscriptionOption : ?cancelable = false
    var PlayerEntersConversation02VolumeSubscriptionOption : ?cancelable = false
    var PlayerExitsConversation02VolumeSubscriptionOption : ?cancelable = false
    var LeaveConversationsSubscriptionOption : ?cancelable = false

    OnBegin<override>()<suspends> : void = 
        set PersonaSpawnsSubscriptionOption = option {PersonaSpawner.SpawnedEvent.Subscribe(OnNPCSpawned)}
        set PlayerEntersConversationsVolumeSubscriptionOption = option {Volume.AgentEntersEvent.Subscribe(OnInitialiseConversations)}
        set PlayerExitsConversationsVolumeSubscriptionOption = option {Volume.AgentExitsEvent.Subscribe(OnRemoveConversations)}
        set PlayerTriggersConversation03TriggerSubscriptionOption = option {ShowConversation03UITrigger.TriggeredEvent.Subscribe(OnShowConversation03Triggered)}
        set PlayerEntersConversation04VolumeSubscriptionOption = option {ShowConversation04UIVolume.AgentEntersEvent.Subscribe(OnShowConversation04Triggered)}
        set PlayerExitsConversation04VolumeSubscriptionOption = option {ShowConversation04UIVolume.AgentExitsEvent.Subscribe(OnHideConversation04Triggered)}
        set PlayerEntersConversation01VolumeSubscriptionOption = option {ShowConversation01UIVolume.AgentEntersEvent.Subscribe(OnShowConversation01Triggered)}
        set PlayerExitsConversation01VolumeSubscriptionOption = option {ShowConversation01UIVolume.AgentExitsEvent.Subscribe(OnHideConversation01Triggered)}
        set PlayerEntersConversation02VolumeSubscriptionOption = option {ShowConversation02UIVolume.AgentEntersEvent.Subscribe(OnShowConversation02Triggered)}
        set PlayerExitsConversation02VolumeSubscriptionOption = option {ShowConversation02UIVolume.AgentExitsEvent.Subscribe(OnHideConversation02Triggered)}

    OnEnd<override>():void =
        CancelSubscription(PersonaSpawnsSubscriptionOption)
        CancelSubscription(PlayerEntersConversationsVolumeSubscriptionOption)
        CancelSubscription(PlayerExitsConversationsVolumeSubscriptionOption)
        CancelSubscription(PlayerTriggersConversation03TriggerSubscriptionOption)
        CancelSubscription(PlayerEntersConversation04VolumeSubscriptionOption)
        CancelSubscription(PlayerExitsConversation04VolumeSubscriptionOption)
        CancelSubscription(PlayerEntersConversation01VolumeSubscriptionOption)
        CancelSubscription(PlayerExitsConversation01VolumeSubscriptionOption)
        CancelSubscription(PlayerEntersConversation02VolumeSubscriptionOption)
        CancelSubscription(PlayerExitsConversation02VolumeSubscriptionOption)
        CancelSubscription(LeaveConversationsSubscriptionOption)

    OnNPCSpawned(Agent : agent) : void =
        if (NPCEntity := Agent.GetFortCharacter[].GetEntity[]):
            if (ExistingComp := NPCEntity.GetComponent[persona_component]):
                set PersonaCompOpt = option{ ExistingComp }
            else:
                NewComp := persona_component{ Entity := NPCEntity }
                NPCEntity.AddComponents(array{ NewComp })
                set PersonaCompOpt = option{ NewComp }
            if (PersonaComp := PersonaCompOpt?):
                set MaybeNPCAgent      = option{ Agent }
                set MaybePersonaEntity = option{ NPCEntity }
                if (DescComp := NPCEntity.GetComponent[description_component]):
                    set PersonaName = DescComp.Name  
                    Conversation03.UpdatePersonaName(PersonaName)   
                    Conversation04.UpdatePersonaName(PersonaName)    
                    Conversation01.UpdatePersonaName(PersonaName)       
                if (IconComp := NPCEntity.GetComponent[icon_component]):
                    set PersonaIcon = option{ IconComp.Icon }                   
                if (Icon := PersonaIcon?):   
                    Conversation03.UpdatePersonaAvatar(Icon)
                    Conversation04.UpdatePersonaAvatar(Icon)
                    Conversation01.UpdatePersonaAvatar(Icon)
                    Conversation02.UpdatePersonaAvatar(Icon)
                PersonaComp.StartHearEvent.Subscribe(OnStartHear)
                PersonaComp.StopHearEvent.Subscribe(OnStopHear)
                PersonaComp.StartSayEvent.Subscribe(OnStartSay)
                PersonaComp.StopSayEvent.Subscribe(OnStopSay)
                PersonaComp.PromptFailureEvent.Subscribe(OnPromptFailure)
                PersonaComp.CommitSayEvent.Subscribe(OnCommitSay)
        else:

    JoinVoiceChannel(InstigatingAgent : agent) : void =
        if (Player := player[InstigatingAgent]):
            if (MaybeVoiceChannel = false):
                CreateVoiceChannelAndAddPlayer(Player)
            if:
                VoiceChannel := MaybeVoiceChannel?
            then:
                set CanSpeakTo = EvalCanSpeakTo(Player)
            else:

    ExitVoiceChannel(InstigatingAgent : agent) : void = 
         if (Player := player[InstigatingAgent]):
            if (MaybeVoiceChannel = true):
                RemoveVoiceChannelAndRemovePlayer(Player)
                VoiceChannel := false
                set CanSpeakTo = EvalCanSpeakTo(Player)


    CreateVoiceChannelAndAddPlayer(Player:player):void=
        if:
            NPCAgent       := MaybeNPCAgent?
            NPCEntity      := MaybePersonaEntity?
            SimEntity      := NPCEntity.GetSimulationEntity[]
        then:
            TeamGroup:agent_group(team_member_info) = agent_group(team_member_info){}
            TeamVoiceChannel:voice_channel(team_member_info) = voice_channel(team_member_info):
                Name := StringToMessage("Team Chat")
                Group := TeamGroup
            TeamGroup.AddMember(Player, team_member_info{})
            TeamGroup.AddMember(NPCAgent, team_member_info{})
            SimEntity.AddChatChannel(TeamVoiceChannel)
            set MaybeAgentGroup = option{TeamGroup}
            set MaybeVoiceChannel = option{TeamVoiceChannel}

    RemoveVoiceChannelAndRemovePlayer(Player: player):void = 
        if:
            NPCAgent       := MaybeNPCAgent?
            NPCEntity      := MaybePersonaEntity?
            SimEntity      := NPCEntity.GetSimulationEntity[]
        then:
            if (TeamGroup := MaybeAgentGroup?):            
                if (TeamVoiceChannel:= MaybeVoiceChannel?):
                    TeamGroup.RemoveMember(Player)
                    TeamGroup.RemoveMember(NPCAgent)
                    SimEntity.RemoveChatChannel(TeamVoiceChannel)
                    set MaybeAgentGroup = false
                    set MaybeVoiceChannel = false

    EvalCanSpeakTo(Player : player) : logic =
        if:
            NPCAgent      := MaybeNPCAgent?
            PersonaEntity := MaybePersonaEntity?
            PersonaComp   := PersonaEntity.GetComponent[persona_component]
            Pair          := Player.GetConversationTarget[]
            Pair(0) = PersonaComp
            Members       := Pair(1).Group.GetMemberMap()
            Members[NPCAgent]
            Members[Player]
        then:
            set HasFailedToProcess = false
            set IsRecording = false
            Conversation03.SetCanSpeakState(TalkDialogMessage)     
            Conversation04.SetCanSpeakState(TalkMessage)  
            Conversation01.SetCanSpeakState(TalkMessage)
            Conversation02.SetCanSpeakState(TalkMessage)               
            true
        else:
            set HasFailedToProcess = true
            set IsRecording = false
            Conversation03.SetUnavailableState(UnavailableMessage) 
            Conversation04.SetUnavailableState(UnavailableMessage)     
            Conversation01.SetUnavailableState(UnavailableMessage)  
            Conversation02.SetUnavailableState(UnavailableMessage)  
            false

    SetPersonaConversationTarget(InstigatingAgent : agent) : void =
        if:
            Player        := player[InstigatingAgent]
            NPCAgent      := MaybeNPCAgent?
            PersonaEntity := MaybePersonaEntity?
            PersonaComp   := PersonaEntity.GetComponent[persona_component]
            SimEntity     := PersonaEntity.GetSimulationEntity[]
        then:
            for (VoiceChannel : SimEntity.GetVoiceChannels()):
                if (VoiceChannel.Group.GetMemberMap()[NPCAgent]):
                    Player.SetConversationTarget(PersonaComp, VoiceChannel)
                    set CanSpeakTo = EvalCanSpeakTo(Player)                                                 
                    return
        else:

    ClearPersonaConversationTarget(InstigatingAgent : agent) : void = 
        if:
            Player        := player[InstigatingAgent]
            NPCAgent      := MaybeNPCAgent?
            PersonaEntity := MaybePersonaEntity?
            PersonaComp   := PersonaEntity.GetComponent[persona_component]
            SimEntity     := PersonaEntity.GetSimulationEntity[]
        then:
            for (VoiceChannel : SimEntity.GetVoiceChannels()):
                if (VoiceChannel.Group.GetMemberMap()[NPCAgent]):
                    Player.ClearConversationTarget()
                    set CanSpeakTo = false                                                 
                    return
        else:

    RunNormalizedCountdown()<suspends> : void =
        Conversation03.UpdateRecordingProgress(1.0)
        Conversation04.UpdateRecordingProgress(1.0)
        Conversation01.UpdateRecordingProgress(1.0)
        Conversation02.UpdateRecordingProgress(1.0)
        DurationS : float = RecordingDuration / 1000.0
        StartTime : float = GetSecondsSinceEpoch()
        loop:
            Sleep(0.0)
            if (not IsRecording?):
                Conversation03.UpdateRecordingProgress(0.0)
                Conversation04.UpdateRecordingProgress(0.0)
                Conversation01.UpdateRecordingProgress(0.0)
                Conversation02.UpdateRecordingProgress(0.0)
                return
            Elapsed := GetSecondsSinceEpoch() - StartTime
            if (Elapsed >= DurationS):
                return
            Conversation03.UpdateRecordingProgress(1.0 - (Elapsed / DurationS))
            Conversation04.UpdateRecordingProgress(1.0 - (Elapsed / DurationS))
            Conversation01.UpdateRecordingProgress(1.0 - (Elapsed / DurationS))
            Conversation02.UpdateRecordingProgress(1.0 - (Elapsed / DurationS))

    OnInitialiseConversations(InAgent : agent) : void = 
        if (Player := player[InAgent]):
            JoinVoiceChannel(Player)
            SetPersonaConversationTarget(Player)
            HUDController.Enable()        

    OnRemoveConversations(InAgent : agent) : void = 
        if (Player := player[InAgent]):
            ExitVoiceChannel(Player)
            ClearPersonaConversationTarget(Player)
            HUDController.Disable()

    OnStartHear(SpokenByAgent : agent) : void = 
        set IsRecording = true
        set IsProcessing = false
        set IsPlayingAudio = false
        set HasFailedToProcess = false
        Conversation03.SetRecordingState(RecordingMessage)
        Conversation04.SetRecordingState(RecordingMessage)
        Conversation01.SetRecordingState(RecordingMessage)
        Conversation02.SetRecordingState(RecordingMessage)
        spawn { RunNormalizedCountdown() }

    OnStopHear(SpokenByAgent : agent) : void = 
        set IsRecording = false
        set IsProcessing = true
        set IsPlayingAudio = false
        set HasFailedToProcess = false
        Conversation03.SetProcessingState(ProcessingMessage)
        Conversation04.SetProcessingState(ProcessingMessage)
        Conversation01.SetProcessingState(ProcessingMessage)
        Conversation02.SetProcessingState(ProcessingMessage)

    OnStartSay(Payload : tuple(?agent, cancelable)) : void = 
        set IsRecording = false
        set IsProcessing = false
        set IsPlayingAudio = true
        set HasFailedToProcess = false
        Conversation03.SetSpeakingState()
        Conversation04.SetSpeakingState()
        Conversation01.SetSpeakingState(SpeakingMessage)
        Conversation02.SetSpeakingState(SpeakingMessage)

    OnStopSay(WasInterrupted : logic) : void = 
        set IsRecording = false
        set IsProcessing = false
        set IsPlayingAudio = false
        if (WasInterrupted = true):
            spawn {ShowErrorState()}
        else:
            set HasFailedToProcess = false
            Conversation03.SetCanSpeakState(TalkDialogMessage)
            Conversation04.SetCanSpeakState(TalkMessage)
            Conversation01.SetCanSpeakState(TalkMessage)
            Conversation02.SetCanSpeakState(TalkMessage)

    ShowErrorState()<suspends> : void =
        var Exit : logic = false
        loop:
            if (Exit = false):
                set HasFailedToProcess = true
                Conversation03.SetUnavailableState(UnavailableMessage)
                Conversation04.SetUnavailableState(UnavailableMessage)
                Conversation01.SetUnavailableState(UnavailableMessage)
                Conversation02.SetUnavailableState(UnavailableMessage)
                Sleep(1.0)
                set HasFailedToProcess = false
                Conversation03.SetCanSpeakState(TalkDialogMessage)
                Conversation04.SetCanSpeakState(TalkMessage)
                Conversation01.SetCanSpeakState(TalkMessage)
                Conversation02.SetCanSpeakState(TalkMessage)
                set Exit = true
            else:
                break;

    OnCommitSay(Payload : tuple(?agent, message)) : void = 
        Conversation03.ShowPersonaAnswer(Payload(1))

    OnPromptFailure(Error : ai_error) : void = 
        set IsRecording = false
        set IsProcessing = false
        set IsPlayingAudio = false
        spawn {ShowErrorState()}
        if (ai_throttled_error[Error]):
        else:

    RestoreState() : void = 
        if (IsProcessing = true):
            Conversation03.SetProcessingState(ProcessingMessage)
            Conversation04.SetProcessingState(ProcessingMessage)
            Conversation01.SetProcessingState(ProcessingMessage)
            Conversation02.SetProcessingState(ProcessingMessage)
        else if (IsPlayingAudio = true):
            Conversation03.SetSpeakingState()
            Conversation04.SetSpeakingState()
            Conversation01.SetSpeakingState(SpeakingMessage)
            Conversation02.SetSpeakingState(SpeakingMessage)
        else if (HasFailedToProcess = true):
            Conversation03.SetUnavailableState(UnavailableMessage)
            Conversation04.SetUnavailableState(UnavailableMessage)
            Conversation01.SetUnavailableState(UnavailableMessage)
            Conversation02.SetUnavailableState(UnavailableMessage)
        else:
            Conversation03.SetCanSpeakState(TalkDialogMessage)
            Conversation04.SetCanSpeakState(TalkMessage)
            Conversation01.SetCanSpeakState(TalkMessage)
            Conversation02.SetCanSpeakState(TalkMessage)

    OnShowConversation03Triggered(MaybeAgent : ?agent) : void =
        RestoreState()
        if (Agent := MaybeAgent?):
            Conversation03.BuildUI(Agent)

        Conversation03DoorSequence.Play()

        if (Agent := MaybeAgent?):
            Conversation03Controls.AddTo(Agent)
            Conversation03Camera.AddTo(Agent)

        if (Player := player[MaybeAgent?]):
            if (PlayerInput := GetPlayerInput[Player]):
                CancelSubscription(LeaveConversationsSubscriptionOption)
                Subscription := PlayerInput.GetInputEvents(Character.Reload).BeginDetectEvent.Subscribe(OnHideConversation03Triggered)
                set LeaveConversationsSubscriptionOption = option{Subscription}

    OnShowConversation04Triggered(InAgent :agent) : void = 
        RestoreState()
        Conversation04.BuildUI(InAgent)

    OnShowConversation01Triggered(InAgent :agent) : void =
        RestoreState() 
        Conversation01.BuildUI(InAgent)

    OnShowConversation02Triggered(InAgent :agent) : void = 
        RestoreState()
        Conversation02.BuildUI(InAgent)
    
    OnHideConversation03Triggered(Args : tuple(player, logic)) : void =
        Conversation03.RemoveUI(Args(0))

        Conversation03DoorSequence.PlayReverse()        

        Conversation03Controls.RemoveFrom(Args(0))
        Conversation03Camera.RemoveFrom(Args(0))

        CancelSubscription(LeaveConversationsSubscriptionOption)
        set LeaveConversationsSubscriptionOption = false

    OnHideConversation04Triggered(InAgent : agent) : void = 
        if (Player := player[InAgent]):   
            Conversation04.RemoveUI(Player) 

    OnHideConversation01Triggered(InAgent : agent) : void = 
        if (Player := player[InAgent]):   
            Conversation01.RemoveUI(Player) 

    OnHideConversation02Triggered(InAgent : agent) : void = 
        if (Player := player[InAgent]):   
            Conversation02.RemoveUI(Player)
