# Knowledge-neutral Verse snippet; project names and asset IDs generalized.
# id: snippet-183

# Copyright Epic Games, Inc. All Rights Reserved.

using { /Verse.org/Simulation }
using { /Fortnite.com/Devices }
using { /Fortnite.com/UI }
using { /UnrealEngine.com/Temporary/UI }
using { /Fortnite.com/Characters }
using { /Fortnite.com/Vehicles }

# ============================================================
#  countdown_page
#
#  The final screen before the race begins. Displays a 3-second
#  countdown, then removes menu UI.
# ============================================================
ui_countdown_page := class(ui_menu_page):

    CountdownWidget : UserInterfaces.UIFeatureExamples.Menu.Widgets.Countdown.UW_Menu_Countdown = UserInterfaces.UIFeatureExamples.Menu.Widgets.Countdown.UW_Menu_Countdown{}

    GetContentWidget<override>() : widget = CountdownWidget
    NeedBackButton<override> : logic = false
    ShowTitle<override> : logic = false
    ShowCurrency<override> : logic = false
    NeedSelectButton<override> : logic = false

    OnShow<override>() : void =
        spawn{ RunCountdown() }

    RunCountdown()<suspends> : void =
        set CountdownWidget.StartCountdown = true
        Sleep(4.5)
        Session.CloseAll()
