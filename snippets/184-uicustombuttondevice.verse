# Knowledge-neutral Verse snippet; project names and asset IDs generalized.
# id: snippet-184

using { /Fortnite.com/Devices }
using { /Verse.org/Simulation }
using { /UnrealEngine.com/Temporary/Diagnostics }
using { /Fortnite.com/UI }
using { /UnrealEngine.com/Temporary/UI }
using { /UnrealEngine.com/Temporary/SpatialMath }
using { /Verse.org/Assets }
using { /Verse.org/Colors }

DefaultButtonLabel<localizes> : message = ""

# A Verse button with configurable visuals #/
CustomizableButton := class():
    var IdleMaterial:material = UserInterfaces.VerseUIExamples.CustomButton.MaterialInstances.MI_UI_Button_Idle{}
    var HighlightMaterial:material = UserInterfaces.VerseUIExamples.CustomButton.MaterialInstances.MI_UI_Button_Highlighted{}
    var IconMaterial:material = UserInterfaces.VerseUIExamples.CustomButton.MaterialInstances.MI_UI_Icon{}
    var LabelText:message = DefaultButtonLabel
    var IdleTextColor:color = color{R:=0.0,G:=0.0,B:=0.0}
    var HighlightedTextColor:color = color{R:=1.0,G:=1.0,B:=1.0}
    var ButtonVisual:UserInterfaces.VerseUIExamples.CustomButton.Widgets.UW_VerseButton01 = UserInterfaces.VerseUIExamples.CustomButton.Widgets.UW_VerseButton01{}
    var VerseButton:button = button{ Slot := button_slot{ Widget := overlay{} } }
    var RootOverlay:overlay = overlay{}

    Initialize()<suspends>:void =
        set ButtonVisual.Label = LabelText
        set ButtonVisual.Icon = IconMaterial
        set ButtonVisual.Background = IdleMaterial
        set ButtonVisual.TextColor = IdleTextColor
        set ButtonVisual.WasHighlighted = false
        set ButtonVisual.WasClicked = false
        VerseButton.HighlightEvent().Subscribe(OnButtonHighlighted)
        VerseButton.UnhighlightEvent().Subscribe(OnButtonUnhighlighted)
        VerseButton.OnClick().Subscribe(OnClicked)
        VerseButton.SetWidget(button_slot{
            Widget := ButtonVisual,
            HorizontalAlignment := horizontal_alignment.Fill,
            VerticalAlignment := vertical_alignment.Fill
        })
        RootOverlay.AddWidget(overlay_slot{
            Widget := VerseButton,
            HorizontalAlignment := horizontal_alignment.Fill,
            VerticalAlignment := vertical_alignment.Fill
        })

    OnButtonHighlighted(M:widget_message):void =
        set ButtonVisual.Background = HighlightMaterial
        set ButtonVisual.TextColor = HighlightedTextColor
        set ButtonVisual.WasHighlighted = true
        Print("Highlighted")

    OnButtonUnhighlighted(M:widget_message):void =
        set ButtonVisual.Background = IdleMaterial
        set ButtonVisual.TextColor = IdleTextColor
        set ButtonVisual.WasHighlighted = false
        Print("Unhighlighted")

    SetText(NewText:message):void =
        set LabelText = NewText
        set ButtonVisual.Label = NewText

    SetIcon(NewIcon:material):void =
        set IconMaterial = NewIcon
        set ButtonVisual.Icon = NewIcon

    SetIdleMaterial(NewMaterial:material):void =
        set IdleMaterial = NewMaterial
        set ButtonVisual.Background = NewMaterial

    SetHighlightedMaterial(NewMaterial:material):void =
        set HighlightMaterial = NewMaterial

    SetIdleTextColor(NewColor:color):void =
        set IdleTextColor = NewColor
        set ButtonVisual.TextColor = NewColor

    SetHighlightedTextColor(NewColor:color):void =
        set HighlightedTextColor = NewColor

    SetSize(NewHeight:float):void =
        set ButtonVisual.Height = NewHeight

    OnClicked(M:widget_message):void =
        set ButtonVisual.WasClicked = true

    GetWidget():overlay = RootOverlay

    GetClickListenable():listenable(widget_message) = VerseButton.OnClick()

ui_custom_button_device := class(creative_device):
    @editable
    VolumeShow:volume_device = volume_device{}

    Btn1Label<localizes>:message = "Continue"
    Btn2Label<localizes>:message = "Restart"
    Btn3Label<localizes>:message = "Gallery"
    Btn4Label<localizes>:message = "Settings"

    var MenuColumn:stack_box = stack_box{ Orientation := orientation.Vertical }
    var ScreenOverlay:overlay = overlay{}

    var Background:color_block =
        color_block{ DefaultOpacity := 0.3, DefaultColor := color{R:=0.5,G:=0.5,B:=0.5} }

    var Btn1:CustomizableButton = CustomizableButton{}
    var Btn2:CustomizableButton = CustomizableButton{}
    var Btn3:CustomizableButton = CustomizableButton{}
    var Btn4:CustomizableButton = CustomizableButton{}

    OnBegin<override>()<suspends>:void =
        VolumeShow.AgentEntersEvent.Subscribe(ShowToAllPlayers)
        BuildMenu()

    BuildMenu()<suspends>:void =
        IdleColor:color = color{R:=1.0,G:=1.0,B:=1.0}
        HighlightedColor:color = color{R:=0.0,G:=0.0,B:=0.0}

        # Configure all 4 buttons to look the same
        Btn1.SetText(Btn1Label)
        Btn1.SetIdleMaterial(UserInterfaces.VerseUIExamples.CustomButton.MaterialInstances.MI_UI_Button_Idle{})
        Btn1.SetHighlightedMaterial(UserInterfaces.VerseUIExamples.CustomButton.MaterialInstances.MI_UI_Button_Highlighted{})
        Btn1.SetIdleTextColor(IdleColor)
        Btn1.SetHighlightedTextColor(HighlightedColor)

        Btn2.SetText(Btn2Label)
        Btn2.SetIdleMaterial(UserInterfaces.VerseUIExamples.CustomButton.MaterialInstances.MI_UI_Button_Idle{})
        Btn2.SetHighlightedMaterial(UserInterfaces.VerseUIExamples.CustomButton.MaterialInstances.MI_UI_Button_Highlighted{})
        Btn2.SetIdleTextColor(IdleColor)
        Btn2.SetHighlightedTextColor(HighlightedColor)

        Btn3.SetText(Btn3Label)
        Btn3.SetIdleMaterial(UserInterfaces.VerseUIExamples.CustomButton.MaterialInstances.MI_UI_Button_Idle{})
        Btn3.SetHighlightedMaterial(UserInterfaces.VerseUIExamples.CustomButton.MaterialInstances.MI_UI_Button_Highlighted{})
        Btn3.SetIdleTextColor(IdleColor)
        Btn3.SetHighlightedTextColor(HighlightedColor)

        Btn4.SetText(Btn4Label)
        Btn4.SetIdleMaterial(UserInterfaces.VerseUIExamples.CustomButton.MaterialInstances.MI_UI_Button_Idle{})
        Btn4.SetHighlightedMaterial(UserInterfaces.VerseUIExamples.CustomButton.MaterialInstances.MI_UI_Button_Highlighted{})
        Btn4.SetIdleTextColor(IdleColor)
        Btn4.SetHighlightedTextColor(HighlightedColor)

        Btn1.Initialize()
        Btn2.Initialize()
        Btn3.Initialize()
        Btn4.Initialize()

        # Subscribe to click event for each button
        Btn1.GetClickListenable().Subscribe(OnClicked_Btn1)
        Btn2.GetClickListenable().Subscribe(OnClicked_Btn2)
        Btn3.GetClickListenable().Subscribe(OnClicked_Btn3)
        Btn4.GetClickListenable().Subscribe(OnClicked_Btn4)

        # Add each button to the layout
        MenuColumn.AddWidget(stack_box_slot{ Widget := Btn1.GetWidget(), Padding := margin{ Top := 3.0, Bottom := 3.0 } })
        MenuColumn.AddWidget(stack_box_slot{ Widget := Btn2.GetWidget(), Padding := margin{ Top := 3.0, Bottom := 3.0 } })
        MenuColumn.AddWidget(stack_box_slot{ Widget := Btn3.GetWidget(), Padding := margin{ Top := 3.0, Bottom := 3.0 } })
        MenuColumn.AddWidget(stack_box_slot{ Widget := Btn4.GetWidget(), Padding := margin{ Top := 3.0, Bottom := 3.0 } })

        # Add background to the layout
        ScreenOverlay.AddWidget(overlay_slot{
            Widget := Background,
            HorizontalAlignment := horizontal_alignment.Fill,
            VerticalAlignment := vertical_alignment.Fill
        })

        # Add the menu to the layout
        ScreenOverlay.AddWidget(overlay_slot{
            Widget := MenuColumn,
            HorizontalAlignment := horizontal_alignment.Left,
            VerticalAlignment := vertical_alignment.Bottom,
            Padding := margin{ Top := 0.0, Bottom := 200.0, Left := 64.0, Right := 0.0 }
        })

    # Click events
    OnClicked_Btn1(M:widget_message):void =
        Print("Clicked Button 1")
        HideForAllPlayers(false)

    OnClicked_Btn2(M:widget_message):void =
        Print("Clicked Button 2")

    OnClicked_Btn3(M:widget_message):void =
        Print("Clicked Button 3")

    OnClicked_Btn4(M:widget_message):void =
        Print("Clicked Button 4")

    # Show UI
    ShowToAllPlayers(InAgent:agent):void =
        for:
            P : GetPlayspace().GetPlayers()
            PUI := GetPlayerUI[P]
        do:
            PUI.AddWidget(ScreenOverlay, player_ui_slot{ InputMode := ui_input_mode.All })

    # Hide UI
    HideForAllPlayers(InAgent:?agent):void =
        for:
            P : GetPlayspace().GetPlayers()
            PUI := GetPlayerUI[P]
        do:
            PUI.RemoveWidget(ScreenOverlay)
