# Knowledge-neutral Verse snippet; project names and asset IDs generalized.
# id: snippet-185

using { /Fortnite.com/Devices }
using { /Verse.org/Simulation }
using { /UnrealEngine.com/Temporary/Diagnostics }
using { /UnrealEngine.com/Temporary/UI }
using { /Verse.org/Input }

MouseKeyboardMessage<localizes> : message = "Use mouse wheel or Shift + Number key to select"
GamepadCycleMessage<localizes> : message = "Use LB and RB to cycle"
GamepadDirectMessage<localizes> : message = "Hold RB + button to select"
TouchMessage<localizes> : message = "Tap on slot to select"

ui_custom_button_hotbar_device := class(creative_device):

    @editable
    ItemGranter : item_granter_device = item_granter_device {}

    @editable
    Volume : volume_device = volume_device {}

    @editable
    HUDController : hud_controller_device = hud_controller_device {}

    @editable
    PreviousInputTrigger : input_trigger_device = input_trigger_device {}

    @editable
    NextInputTrigger: input_trigger_device = input_trigger_device {}

    @editable
    ToggleModeTrigger: trigger_device = trigger_device {}    

    HotbarWidget : UserInterfaces.UIFeatureExamples.Hotbar.Widgets.UW_CustomButton_Hotbar02 = UserInterfaces.UIFeatureExamples.Hotbar.Widgets.UW_CustomButton_Hotbar02 {}

    var SelectedSlotIndex : int = 0
    var PreviousSelectedSlotIndex : int = -1
    MaxSlots : int = 8

    var CycleGamePadMode : logic = true

    var IsPlayerExitsVolume : logic = false

    var AgentPressedNextSubscriptionOption : ?cancelable = false
    var AgentPressedPreviousSubscriptionOption : ?cancelable = false
    var AgentSelectedSlot01SubscriptionOption : ?cancelable_event = false
    var AgentSelectedSlot02SubscriptionOption : ?cancelable_event = false
    var AgentSelectedSlot03SubscriptionOption : ?cancelable_event = false
    var AgentSelectedSlot04SubscriptionOption : ?cancelable_event = false
    var AgentSelectedSlot05SubscriptionOption : ?cancelable_event = false
    var AgentSelectedSlot06SubscriptionOption : ?cancelable_event = false
    var AgentSelectedSlot07SubscriptionOption : ?cancelable_event = false
    var AgentSelectedSlot08SubscriptionOption : ?cancelable_event = false
    var AgentEntersVolumeSubscriptionOption : ?cancelable = false
    var AgentExitsVolumeSubscriptionOption : ?cancelable = false

    OnBegin<override>()<suspends>:void= 
        set AgentEntersVolumeSubscriptionOption = option {Volume.AgentEntersEvent.Subscribe(ShowUI)}
        set AgentExitsVolumeSubscriptionOption = option {Volume.AgentExitsEvent.Subscribe(HideUI)}

        ToggleModeTrigger.TriggeredEvent.Subscribe(OnGamepadModeTriggered)
        
        Slot01Clicked := SubscribeToEvent(HotbarWidget.Slot01Clicked, OnSlot01Clicked)
        Slot02Clicked := SubscribeToEvent(HotbarWidget.Slot02Clicked, OnSlot02Clicked)
        Slot03Clicked := SubscribeToEvent(HotbarWidget.Slot03Clicked, OnSlot03Clicked)
        Slot04Clicked := SubscribeToEvent(HotbarWidget.Slot04Clicked, OnSlot04Clicked)
        Slot05Clicked := SubscribeToEvent(HotbarWidget.Slot05Clicked, OnSlot05Clicked)
        Slot06Clicked := SubscribeToEvent(HotbarWidget.Slot06Clicked, OnSlot06Clicked)
        Slot07Clicked := SubscribeToEvent(HotbarWidget.Slot07Clicked, OnSlot07Clicked)
        Slot08Clicked := SubscribeToEvent(HotbarWidget.Slot08Clicked, OnSlot08Clicked)
        set AgentSelectedSlot01SubscriptionOption = option {Slot01Clicked}
        set AgentSelectedSlot02SubscriptionOption = option {Slot02Clicked}
        set AgentSelectedSlot03SubscriptionOption = option {Slot03Clicked}
        set AgentSelectedSlot04SubscriptionOption = option {Slot04Clicked}
        set AgentSelectedSlot05SubscriptionOption = option {Slot05Clicked}
        set AgentSelectedSlot06SubscriptionOption = option {Slot06Clicked}
        set AgentSelectedSlot07SubscriptionOption = option {Slot07Clicked}
        set AgentSelectedSlot08SubscriptionOption = option {Slot08Clicked}

    ShowUI(InAgent : agent) : void =
        set AgentPressedNextSubscriptionOption = option {NextInputTrigger.PressedEvent.Subscribe(OnNextInputPressed)}
        set AgentPressedPreviousSubscriptionOption = option {PreviousInputTrigger.PressedEvent.Subscribe(OnPreviousInputPressed)}

        set IsPlayerExitsVolume = false
        set CycleGamePadMode = true
        set HotbarWidget.ShowHideAnim = false
        HUDController.Enable()
        NextInputTrigger.Enable()
        PreviousInputTrigger.Enable()
        ItemGranter.Enable()        
        for (Player : GetPlayspace().GetPlayers()):
            if (PlayerUI := GetPlayerUI[Player]):
                PlayerUI.AddWidget(HotbarWidget)                
                if (PlayerInput := GetPlayerInput[Player]):
                    PlayerInput.AddInputMapping(Gameplay.HotbarMapping)
                    spawn {AdaptUIToInput(PlayerInput)}
        
        set HotbarWidget.ShowIntroAnim = true
        spawn {PlaySelectAnim()}
        spawn {PlayDeselectAnim()}
        ChangeSelectedSlot()

    HideUI(InAgent : agent) : void =
        CancelSubscription(AgentPressedNextSubscriptionOption)
        CancelSubscription(AgentPressedPreviousSubscriptionOption)
        
        set IsPlayerExitsVolume = true
        set HotbarWidget.ShowIntroAnim = false
        HUDController.Disable()
        PreviousInputTrigger.Disable()
        NextInputTrigger.Disable()
        ItemGranter.Disable()
        for (Player : GetPlayspace().GetPlayers()):
            if (PlayerUI := GetPlayerUI[Player]):
                spawn{PlayHideAnim(PlayerUI)}
            if (PlayerInput := GetPlayerInput[Player]):
                PlayerInput.RemoveInputMapping(Gameplay.HotbarMapping)
        
    AdaptUIToInput(PlayerInput:player_input)<suspends> : void = 
        loop:
            if (IsPlayerExitsVolume = false):
                case (PlayerInput.PreferredInputMethod):
                    input_method.Gamepad =>                        
                        if (CycleGamePadMode = true):
                            set HotbarWidget.IsGamepadInput = false
                            set HotbarWidget.MessageText = GamepadCycleMessage
                        else:
                            set HotbarWidget.IsGamepadInput = true
                            set HotbarWidget.MessageText = GamepadDirectMessage
                        set HotbarWidget.IsKeyboardInput = false
                    input_method.Touch =>
                        set HotbarWidget.IsGamepadInput = false
                        set HotbarWidget.IsKeyboardInput = false
                        set HotbarWidget.MessageText = TouchMessage
                    input_method.KeyboardAndMouse =>
                        set CycleGamePadMode = true
                        NextInputTrigger.Enable()
                        PreviousInputTrigger.Enable()
                        set HotbarWidget.IsGamepadInput = false
                        set HotbarWidget.IsKeyboardInput = true
                        set HotbarWidget.MessageText = MouseKeyboardMessage
                    _=>  
            else: 
                break

            Sleep(0.0)

    PlayHideAnim(PlayerUI : player_ui)<suspends> : void = 
        var Exit : logic = false
        loop:
            if (Exit = false):
                set HotbarWidget.ShowHideAnim = true
                Sleep(0.3)
                set HotbarWidget.ShowHideAnim = false
                PlayerUI.RemoveWidget(HotbarWidget)
                set Exit = true
            else:
                break
            Sleep(0.0)

    OnEnd<override>():void =
        CancelSubscription(AgentEntersVolumeSubscriptionOption)
        CancelSubscription(AgentExitsVolumeSubscriptionOption)

        if (Slot01Clicked := AgentSelectedSlot01SubscriptionOption?):
            Slot01Clicked.Cancel()

        if (Slot02Clicked := AgentSelectedSlot02SubscriptionOption?):
            Slot02Clicked.Cancel()

        if (Slot03Clicked := AgentSelectedSlot03SubscriptionOption?):
            Slot03Clicked.Cancel()

        if (Slot04Clicked := AgentSelectedSlot04SubscriptionOption?):
            Slot04Clicked.Cancel()
        
        if (Slot05Clicked := AgentSelectedSlot05SubscriptionOption?):
            Slot05Clicked.Cancel()

        if (Slot06Clicked := AgentSelectedSlot06SubscriptionOption?):
            Slot06Clicked.Cancel()

        if (Slot07Clicked := AgentSelectedSlot07SubscriptionOption?):
            Slot07Clicked.Cancel()

        if (Slot08Clicked := AgentSelectedSlot08SubscriptionOption?):        
            Slot08Clicked.Cancel()

    OnNextInputPressed(Agent:agent):void = 
        if (SelectedSlotIndex + 1 > MaxSlots -1):
            set PreviousSelectedSlotIndex = SelectedSlotIndex
            set SelectedSlotIndex = 0
        else:
            set PreviousSelectedSlotIndex = SelectedSlotIndex
            set SelectedSlotIndex += 1
        ChangeSelectedSlot()

    OnPreviousInputPressed(Agent:agent):void = 
        if (SelectedSlotIndex - 1 < 0):
            set PreviousSelectedSlotIndex = SelectedSlotIndex
            set SelectedSlotIndex = MaxSlots - 1
        else:
            set PreviousSelectedSlotIndex = SelectedSlotIndex
            set SelectedSlotIndex -=1
       ChangeSelectedSlot()

    PlaySelectAnim()<suspends> : void = 
        loop:
            if (IsPlayerExitsVolume = false):
                if (SelectedSlotIndex <> PreviousSelectedSlotIndex):
                    if (SelectedSlotIndex = 0):
                        set HotbarWidget.Slot01SelectIntroAnim = true                        
                        set HotbarWidget.Slot01Selected = true
                    if (SelectedSlotIndex = 1):
                        set HotbarWidget.Slot02SelectIntroAnim = true                        
                        set HotbarWidget.Slot02Selected = true
                    if (SelectedSlotIndex = 2):
                        set HotbarWidget.Slot03SelectIntroAnim = true                        
                        set HotbarWidget.Slot03Selected = true
                    if (SelectedSlotIndex = 3):
                        set HotbarWidget.Slot04SelectIntroAnim = true                        
                        set HotbarWidget.Slot04Selected = true
                    if (SelectedSlotIndex = 4):
                        set HotbarWidget.Slot05SelectIntroAnim = true                        
                        set HotbarWidget.Slot05Selected = true
                    if (SelectedSlotIndex = 5):
                        set HotbarWidget.Slot06SelectIntroAnim = true                        
                        set HotbarWidget.Slot06Selected = true
                    if (SelectedSlotIndex = 6):
                        set HotbarWidget.Slot07SelectIntroAnim = true                        
                        set HotbarWidget.Slot07Selected = true
                    if (SelectedSlotIndex = 7):
                        set HotbarWidget.Slot08SelectIntroAnim = true                        
                        set HotbarWidget.Slot08Selected = true
            else:
                break
            Sleep(0.0)

    PlayDeselectAnim()<suspends> : void = 
        loop:
            if (IsPlayerExitsVolume = false):
                if (SelectedSlotIndex <> PreviousSelectedSlotIndex):
                    if (PreviousSelectedSlotIndex = 0):
                        set HotbarWidget.Slot01Selected = false
                        set HotbarWidget.Slot01SelectIntroAnim = false
                    if (PreviousSelectedSlotIndex = 1):
                        set HotbarWidget.Slot02Selected = false
                        set HotbarWidget.Slot02SelectIntroAnim = false
                    if (PreviousSelectedSlotIndex = 2):
                        set HotbarWidget.Slot03Selected = false
                        set HotbarWidget.Slot03SelectIntroAnim = false
                    if (PreviousSelectedSlotIndex = 3):
                        set HotbarWidget.Slot04Selected = false
                        set HotbarWidget.Slot04SelectIntroAnim = false
                    if (PreviousSelectedSlotIndex = 4):
                        set HotbarWidget.Slot05Selected = false
                        set HotbarWidget.Slot05SelectIntroAnim = false
                    if (PreviousSelectedSlotIndex = 5):
                        set HotbarWidget.Slot06Selected = false
                        set HotbarWidget.Slot06SelectIntroAnim = false
                    if (PreviousSelectedSlotIndex = 6):
                        set HotbarWidget.Slot07Selected = false
                        set HotbarWidget.Slot07SelectIntroAnim = false
                    if (PreviousSelectedSlotIndex = 7):
                        set HotbarWidget.Slot08Selected = false
                        set HotbarWidget.Slot08SelectIntroAnim = false
            else:
                break
            Sleep(0.0)

    ChangeSelectedSlot():void =
        if (SelectedSlotIndex <> PreviousSelectedSlotIndex):
            if (SelectedSlotIndex = 0):
                ItemGranter.GrantItemIndex(0)
            if (SelectedSlotIndex = 1):
                ItemGranter.GrantItemIndex(1)
            if (SelectedSlotIndex = 2):
                ItemGranter.GrantItemIndex(2)
            if (SelectedSlotIndex = 3):
                ItemGranter.GrantItemIndex(3)
            if (SelectedSlotIndex = 4):
                ItemGranter.GrantItemIndex(4)
            if (SelectedSlotIndex = 5):
                ItemGranter.GrantItemIndex(5)
            if (SelectedSlotIndex = 6):
                ItemGranter.GrantItemIndex(6)
            if (SelectedSlotIndex = 7):
                ItemGranter.GrantItemIndex(7)

    OnGamepadModeTriggered(InAgent : ?agent) : void = 
        if (CycleGamePadMode = true):
            NextInputTrigger.Disable()
            PreviousInputTrigger.Disable()
            set CycleGamePadMode = false            
        else:
            NextInputTrigger.Enable()
            PreviousInputTrigger.Enable()
            set CycleGamePadMode = true            


    OnSlot01Clicked(Args : tuple()) : void = 
        set PreviousSelectedSlotIndex = SelectedSlotIndex
        set SelectedSlotIndex = 0        
        ChangeSelectedSlot()

    OnSlot02Clicked(Args : tuple()) : void = 
        set PreviousSelectedSlotIndex = SelectedSlotIndex
        set SelectedSlotIndex = 1        
        ChangeSelectedSlot()

    OnSlot03Clicked(Args : tuple()) : void = 
        set PreviousSelectedSlotIndex = SelectedSlotIndex
        set SelectedSlotIndex = 2        
        ChangeSelectedSlot()

    OnSlot04Clicked(Args : tuple()) : void = 
        set PreviousSelectedSlotIndex = SelectedSlotIndex
        set SelectedSlotIndex = 3        
        ChangeSelectedSlot()

    OnSlot05Clicked(Args : tuple()) : void = 
        set PreviousSelectedSlotIndex = SelectedSlotIndex
        set SelectedSlotIndex = 4        
        ChangeSelectedSlot()

    OnSlot06Clicked(Args : tuple()) : void = 
        set PreviousSelectedSlotIndex = SelectedSlotIndex
        set SelectedSlotIndex = 5        
        ChangeSelectedSlot()

    OnSlot07Clicked(Args : tuple()) : void = 
        set PreviousSelectedSlotIndex = SelectedSlotIndex
        set SelectedSlotIndex = 6        
        ChangeSelectedSlot()

    OnSlot08Clicked(Args : tuple()) : void = 
        set PreviousSelectedSlotIndex = SelectedSlotIndex
        set SelectedSlotIndex = 7        
        ChangeSelectedSlot()
