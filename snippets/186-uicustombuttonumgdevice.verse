# Knowledge-neutral Verse snippet; project names and asset IDs generalized.
# id: snippet-186

# Copyright Epic Games, Inc. All Rights Reserved.

using { /Fortnite.com/Devices }
using { /Fortnite.com/UI }
using { /UnrealEngine.com/Temporary/UI }
using { /UnrealEngine.com/Temporary/SpatialMath }
using { /Verse.org/Simulation }

ui_custom_button_umg_device := class(creative_device):

    @editable
    ShowVolume : volume_device = volume_device{}

    Widget : UserInterfaces.VerseUIExamples.CustomButton.Widgets.UW_CustomButton = UserInterfaces.VerseUIExamples.CustomButton.Widgets.UW_CustomButton{}

    var Subscription : ?cancelable_event = false
    var CurrentPlayer : ?player = false

    OnBegin<override>() : void =
        ShowVolume.AgentEntersEvent.Subscribe(OnAgentEntered)
        GetPlayspace().PlayerRemovedEvent().Subscribe(OnPlayerLeft)

    OnAgentEntered(Agent : agent) : void =
        if (Player := player[Agent]):
            if (CurrentPlayer?):
                return
            
            OpenMenu(Player)

    OnPlayerLeft(Player : player) : void =
        if (ExistingPlayer := CurrentPlayer?, ExistingPlayer = Player):
            CloseMenu()

    OpenMenu(Player : player) : void =
        if (PlayerUI := GetPlayerUI[Player]):

            PlayerUI.SetFocus(Widget)
            
            PlayerUI.AddWidget(
                Widget, 
                player_ui_slot{ InputMode := ui_input_mode.All }
            )

            set Widget.TriggerIntro = true
            
            set CurrentPlayer = option{Player}

            SubClose := SubscribeToEvent(Widget.CloseEvent, OnCloseEvent)
            set Subscription = option{SubClose}

    OnCloseEvent(Args : tuple()) : void =
        CloseMenu()

    CloseMenu() : void =
        if (Player := CurrentPlayer?, PlayerUI := GetPlayerUI[Player]):
            set Widget.TriggerIntro = false
            PlayerUI.RemoveWidget(Widget)
        
        if (SubClose := Subscription?):
            SubClose.Cancel()
        
        set Subscription = false
        set CurrentPlayer = false
