# Knowledge-neutral Verse snippet; project names and asset IDs generalized.
# id: snippet-187

# Copyright Epic Games, Inc. All Rights Reserved.

using { /Fortnite.com/Devices }
using { /Fortnite.com/UI }
using { /UnrealEngine.com/Temporary/UI }
using { /UnrealEngine.com/Temporary/SpatialMath }
using { /Verse.org/Simulation }
using { /Verse.org/Input }     
using { /Verse.org/Input/UI }  

ui_custom_button_umg_input_device := class(creative_device):

    @editable
    ShowVolume : volume_device = volume_device{}

    Widget : UserInterfaces.VerseUIExamples.CustomButton.Widgets.UW_CustomButton_Input = UserInterfaces.VerseUIExamples.CustomButton.Widgets.UW_CustomButton_Input{}

    var CurrentPlayer : ?player = false

    OnBegin<override>() : void =
        ShowVolume.AgentEntersEvent.Subscribe(OnAgentEntered)
        GetPlayspace().PlayerRemovedEvent().Subscribe(OnPlayerLeft)

    OnAgentEntered(Agent : agent) : void =
        if (Player := player[Agent]):
            if (CurrentPlayer?):
                return
            
            OpenMenu(Player)

    OnPlayerLeft(Player : player) : void =
        if (ExistingPlayer := CurrentPlayer?, ExistingPlayer = Player):
            CloseMenu()

    OpenMenu(Player : player) : void =
        if (PlayerUI := GetPlayerUI[Player], PlayerInput := GetPlayerInput[Player]):

            PlayerUI.SetFocus(Widget)
            
            PlayerUI.AddWidget(
                Widget, 
                player_ui_slot{ InputMode := ui_input_mode.All }
            )

            spawn {TriggerAnimation()}
            
            set CurrentPlayer = option{Player}

            PlayerInput.AddInputMapping(UI.InventoryMenuMapping)
            PlayerInput.AddInputMapping(UI.MenuNavigationMapping)

            PlayerInput.GetInputEvents(UI.PreviousTab).BeginDetectEvent.Subscribe(OnPrevTab)
            PlayerInput.GetInputEvents(UI.NextTab).BeginDetectEvent.Subscribe(OnNextTab)
            PlayerInput.GetInputEvents(UI.Use).BeginDetectEvent.Subscribe(OnClose)

    TriggerAnimation()<suspends> : void =
        set Widget.TriggerIntro = true
        Sleep(1.0)
        set Widget.TriggerLoop = true

    OnPrevTab(Args: tuple(player, logic)) : void =
        if (Widget.TabIndex = 1):
            set Widget.TabIndex = 0

    OnNextTab(Args: tuple(player, logic)) : void =
        if (Widget.TabIndex = 0):
            set Widget.TabIndex = 1

    OnClose(Args: tuple(player, logic)) : void =
        CloseMenu()

    CloseMenu() : void =
        if (Player := CurrentPlayer?, PlayerUI := GetPlayerUI[Player], PlayerInput := GetPlayerInput[Player]):
            set Widget.TriggerIntro = false
            set Widget.TriggerLoop = false
            PlayerUI.RemoveWidget(Widget)
            Print("Removed widget")
            PlayerInput.RemoveInputMapping(UI.InventoryMenuMapping)
            PlayerInput.RemoveInputMapping(UI.MenuNavigationMapping)
        
        set CurrentPlayer = false
        set Widget.TabIndex = 0
