# Knowledge-neutral Verse snippet; project names and asset IDs generalized.
# id: snippet-188

using { /Fortnite.com/Devices }
using { /Verse.org/Simulation }
using { /UnrealEngine.com/Temporary/Diagnostics }
using { /UnrealEngine.com/Temporary/UI }
using { /Verse.org/Input }

ui_custom_button_umg_input_hud_device := class(creative_device):

    @editable
    HUDController : hud_controller_device = hud_controller_device {}

    @editable
    Volume : volume_device = volume_device {}

    @editable
    PreviousInputTrigger : input_trigger_device = input_trigger_device {}

    @editable
    NextInputTrigger: input_trigger_device = input_trigger_device {}

    HotbarWidget :UserInterfaces.UIFeatureExamples.Hotbar.Widgets.UW_CustomButton_Hotbar01 = UserInterfaces.UIFeatureExamples.Hotbar.Widgets.UW_CustomButton_Hotbar01 {}

    var SelectedSlotIndex : int = 0
    MaxSlots : int = 4

    var IsPlayerExitsVolume : logic = false

    var AgentPressedNextSubscriptionOption : ?cancelable = false
    var AgentPressedPreviousSubscriptionOption : ?cancelable = false
    var AgentSelectedSlot01SubscriptionOption : ?cancelable_event = false
    var AgentSelectedSlot02SubscriptionOption : ?cancelable_event = false
    var AgentSelectedSlot03SubscriptionOption : ?cancelable_event = false
    var AgentSelectedSlot04SubscriptionOption : ?cancelable_event = false
    var AgentEntersVolumeSubscriptionOption : ?cancelable = false
    var AgentExitsVolumeSubscriptionOption : ?cancelable = false

    OnBegin<override>()<suspends>:void= 
        set AgentEntersVolumeSubscriptionOption = option {Volume.AgentEntersEvent.Subscribe(ShowUI)}
        set AgentExitsVolumeSubscriptionOption = option {Volume.AgentExitsEvent.Subscribe(HideUI)}


    AdaptUIToInput(PlayerInput:player_input)<suspends> : void = 
        loop:
            if (IsPlayerExitsVolume = false):
                case (PlayerInput.PreferredInputMethod):
                    input_method.Gamepad =>
                        set HotbarWidget.ShowActionWidgets = true                    
                    input_method.Touch =>
                        set HotbarWidget.ShowActionWidgets = false
                    input_method.KeyboardAndMouse =>
                        set HotbarWidget.ShowActionWidgets = false
                    _=>  
            else: 
                break

            Sleep(0.0)

    ShowUI(InAgent : agent) : void =
        set AgentPressedNextSubscriptionOption = option {NextInputTrigger.PressedEvent.Subscribe(OnNextInputPressed)}
        set AgentPressedPreviousSubscriptionOption = option {PreviousInputTrigger.PressedEvent.Subscribe(OnPreviousInputPressed)}


        Slot01Clicked := SubscribeToEvent(HotbarWidget.Slot01Clicked, OnSlot01Clicked)
        Slot02Clicked := SubscribeToEvent(HotbarWidget.Slot02Clicked, OnSlot02Clicked)
        Slot03Clicked := SubscribeToEvent(HotbarWidget.Slot03Clicked, OnSlot03Clicked)
        Slot04Clicked := SubscribeToEvent(HotbarWidget.Slot04Clicked, OnSlot04Clicked)
        set AgentSelectedSlot01SubscriptionOption = option {Slot01Clicked}
        set AgentSelectedSlot02SubscriptionOption = option {Slot02Clicked}
        set AgentSelectedSlot03SubscriptionOption = option {Slot03Clicked}
        set AgentSelectedSlot04SubscriptionOption = option {Slot04Clicked}

        set IsPlayerExitsVolume = false
        HUDController.Enable()
        PreviousInputTrigger.Enable()
        NextInputTrigger.Enable()
        for (Player : GetPlayspace().GetPlayers()):
            BuildUI(Player)
            if (PlayerInput := GetPlayerInput[Player]):
                PlayerInput.AddInputMapping(Gameplay.HotbarMapping)
                spawn {AdaptUIToInput(PlayerInput)}
            
        ChangeSelectedSlot()

    HideUI(InAgent : agent) : void =
        set IsPlayerExitsVolume = true
        HUDController.Disable()
        PreviousInputTrigger.Disable()
        NextInputTrigger.Disable()
        for (Player : GetPlayspace().GetPlayers()):
            RemoveUI(Player)
            if (PlayerInput := GetPlayerInput[Player]):
                PlayerInput.RemoveInputMapping(Gameplay.HotbarMapping)

        CancelSubscription(AgentPressedNextSubscriptionOption)
        CancelSubscription(AgentPressedPreviousSubscriptionOption)


        if (Slot01Clicked := AgentSelectedSlot01SubscriptionOption?):
            Slot01Clicked.Cancel()

        if (Slot02Clicked := AgentSelectedSlot02SubscriptionOption?):
            Slot02Clicked.Cancel()

        if (Slot03Clicked := AgentSelectedSlot03SubscriptionOption?):
            Slot03Clicked.Cancel()

        if (Slot04Clicked := AgentSelectedSlot04SubscriptionOption?):
            Slot04Clicked.Cancel()

    OnEnd<override>():void = 
        CancelSubscription(AgentEntersVolumeSubscriptionOption)
        CancelSubscription(AgentExitsVolumeSubscriptionOption)

    BuildUI(Player:player):void =
        if (PlayerUI := GetPlayerUI[Player]):
            PlayerUI.AddWidget(HotbarWidget)

    RemoveUI(Player:player):void =
        if (PlayerUI := GetPlayerUI[Player]):
            PlayerUI.RemoveWidget(HotbarWidget)

    OnNextInputPressed(Agent:agent):void = 
        if (SelectedSlotIndex + 1 > MaxSlots -1):
            set SelectedSlotIndex = 0
        else:
            set SelectedSlotIndex += 1
        ChangeSelectedSlot()

    OnPreviousInputPressed(Agent:agent):void = 
        if (SelectedSlotIndex - 1 < 0):
           set SelectedSlotIndex = MaxSlots - 1 
        else:
            set SelectedSlotIndex -=1
        ChangeSelectedSlot()

    ChangeSelectedSlot():void =
        set HotbarWidget.Slot01Selected = false
        set HotbarWidget.Slot02Selected = false
        set HotbarWidget.Slot03Selected = false
        set HotbarWidget.Slot04Selected = false
        if (SelectedSlotIndex = 0):
            set HotbarWidget.Slot01Selected = true
        if (SelectedSlotIndex = 1):
            set HotbarWidget.Slot02Selected = true
        if (SelectedSlotIndex = 2):
            set HotbarWidget.Slot03Selected = true
        if (SelectedSlotIndex = 3):
            set HotbarWidget.Slot04Selected = true

    OnSlot01Clicked(Args : tuple()) : void = 
        set SelectedSlotIndex = 0
        ChangeSelectedSlot()

    OnSlot02Clicked(Args : tuple()) : void = 
        set SelectedSlotIndex = 1
        ChangeSelectedSlot()

    OnSlot03Clicked(Args : tuple()) : void = 
        set SelectedSlotIndex = 2
        ChangeSelectedSlot()

    OnSlot04Clicked(Args : tuple()) : void = 
        set SelectedSlotIndex = 3
        ChangeSelectedSlot()
