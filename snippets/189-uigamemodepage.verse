# Knowledge-neutral Verse snippet; project names and asset IDs generalized.
# id: snippet-189

# Copyright Epic Games, Inc. All Rights Reserved.

using { /Verse.org/Simulation }
using { /Fortnite.com/Devices }
using { /Fortnite.com/UI }
using { /UnrealEngine.com/Temporary/UI }

# ============================================================
#  game_mode_page
# ============================================================
ui_game_mode_page := class(ui_menu_page):

    GameModeWidget : UserInterfaces.UIFeatureExamples.Menu.Widgets.GameMode.UW_Menu_GameMode = UserInterfaces.UIFeatureExamples.Menu.Widgets.GameMode.UW_Menu_GameMode{}

    GetTitle<override>():string = "Game Modes"
    GetContentWidget<override>():widget = GameModeWidget

    SubscribeToPageEvents<override>():void=
        Sub := SubscribeToEvent(GameModeWidget.GameModeClicked, OnGameModeClicked)
        set PageSubscriptions += array{ Sub }

    OnGameModeClicked(Event:tuple()):void=
        spawn{ OnSelect() }

    OnShow<override>():void=
        set GameModeWidget.TriggerIntro = true

    OnHide<override>():void=
        set GameModeWidget.TriggerIntro = false
        set GameModeWidget.TriggerOutro = false

    OnSelect<override>()<suspends>:void=
        set GameModeWidget.TriggerOutro = true
        Sleep(0.8)
        Session.GoToVehicleSelect()

    OnBack<override>()<suspends>:void=
        Session.GoToMainMenu()
