# Knowledge-neutral Verse snippet; project names and asset IDs generalized.
# id: snippet-190

# Copyright Epic Games, Inc. All Rights Reserved.

# Modules #
UserInterfaces<public> := module:
    VerseUIExamples<public> := module:
        MaterialBlock<public> := module:
            MaterialInstances<public>:=module{}
            Textures<public>:=module{}
        InputActions<public> := module:
            Widgets<public>:=module{}
            Textures<public>:=module{}
            MaterialInstances<public>:=module{}
        Vehicle<public> := module:
            MaterialInstances<public>:=module{}
            Textures<public>:=module{}
            Widgets<public>:=module{}
        CustomButton<public> := module:
            MaterialInstances<public>:=module{}
            Widgets<public>:=module{}
        VerseFields<public> := module:
            Widgets<public>:=module{}
        VerseEvents<public> := module:
            Widgets<public>:=module{}
        PlayerPortrait<public> := module:
            Widgets<public>:=module{}
        VirtualPointer<public> := module:
            Widgets<public> := module {}
        TouchscreenControlLayout<public> := module:
            Widgets<public> := module {}
    SceneGraphUIExamples<public> := module:
        InWorldUMG<public> := module:
            Widgets<public>:=module{}
    Shared<public> := module:
        Widgets<public> := module{}
    UIArtExamples<public> := module:
        MaterialCollection<public> := module:
            Widgets<public>:=module{}
    UIFeatureExamples<public> := module:
        Conversations<public> := module:
            Widgets<public> := module{}
        InIslandTransactions<public> := module:
            Demo01<public> := module:
                MaterialInstances<public>:=module{}
                Textures<public>:=module{}
                Widgets<public>:=module{}
            Demo02<public>:=module:
                Widgets<public>:=module{}
                Textures<public>:=module{}
                MaterialInstances<public>:=module{}
        Hotbar<public> := module:
            Widgets<public> := module{}
        Menu<public> := module:
            Widgets<public>:= module:
                MainMenu<public>      := module{}
                GameMode<public>      := module{}
                UpgradeSelect<public> := module:
                VehicleSelect<public> := module{}
                Settings<public>      := module{}
                Countdown<public>     := module{}
                Basic<public>         := module{}
            Textures<public>:= module:
                UpgradeIcons<public>:= module{}
