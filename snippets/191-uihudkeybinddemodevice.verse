# Knowledge-neutral Verse snippet; project names and asset IDs generalized.
# id: snippet-191

using { /Fortnite.com/Devices }
using { /Fortnite.com/UI }
using { /Verse.org/Simulation }
using { /Verse.org/Colors }
using { /UnrealEngine.com/Temporary/SpatialMath }
using { /UnrealEngine.com/Temporary/Diagnostics }
using { /UnrealEngine.com/Temporary/UI }
using { /Verse.org/Input }
using { /Fortnite.com/Characters }
using { /Fortnite.com/Input }
using {UserInterfaces.VerseUIExamples.InputActions.Textures}

# A Verse-authored creative device that can be placed in a level
ui_hud_keybind_demo_device := class(creative_device):

    @editable
    ShowTrigger:volume_device = volume_device{}    
    @editable
    RemoveTriggers:[]volume_device = array{}

    # Verse UI for the demo
    var PlayerStanceHuds : [player][]overlay = map {}

    # Materials used
    var PlayerStanceMaterials:[player]UserInterfaces.VerseUIExamples.InputActions.MaterialInstances.MI_UI_Stance = map{}
    var CrouchedSubscriptions:[player][]cancelable = map{}

    OnBegin<override>()<suspends>:void=
        for (RemoveTrigger : RemoveTriggers):
            RemoveTrigger.AgentExitsEvent.Subscribe(OnHide)

        # Setup our input on each player in the playspace, and get notified for when another player joins.
        Playspace := GetPlayspace()
        for (Player : Playspace.GetPlayers()):
            OnPlayerAdded(Player)
        # Handle any new players joining or leaving
        Playspace.PlayerAddedEvent().Subscribe(OnPlayerAdded)
        Playspace.PlayerRemovedEvent().Subscribe(OnPlayerRemoved)
        # Entering/exiting the volume device shows/hides the UI
        ShowTrigger.AgentEntersEvent.Subscribe(OnPlayerInitialize)

    OnPlayerInitialize(InAgent : agent):void=
        if (Player := player[InAgent], PlayerUI := GetPlayerUI[Player]):
            # Clean up existing widgets and subscriptions
            if (ExistingHuds := PlayerStanceHuds[Player]):
                for (Hud : ExistingHuds):
                    PlayerUI.RemoveWidget(Hud)
            if (ExistingSubs := CrouchedSubscriptions[Player]):
                for (Sub : ExistingSubs):
                    Sub.Cancel()

            if (FortCharacter := InAgent.GetFortCharacter[]):
                Sub := FortCharacter.CrouchedEvent().Subscribe(OnCrouchedEvent)
                if (set CrouchedSubscriptions[Player] = array{Sub}) {}

            PlayerMaterial := UserInterfaces.VerseUIExamples.InputActions.MaterialInstances.MI_UI_Stance{Texture := T_UI_Stand}    
            
            # Material block used for stance UI
            StanceMaterialBlock := material_block:
                DefaultImage := PlayerMaterial
                DefaultDesiredSize := vector2{X:=100.0, Y:=100.0}
                
            # UI overlay for stance
            StanceOverlay := overlay:
                Slots := array:
                    overlay_slot:
                        Widget := StanceMaterialBlock
                        Padding := margin{Left:=0.0, Top:=0.0, Right:=50.0, Bottom:=80.0}
                        HorizontalAlignment := horizontal_alignment.Right
                        VerticalAlignment := vertical_alignment.Bottom

            if (set PlayerStanceHuds[Player] = array{StanceOverlay}) {}
            if (set PlayerStanceMaterials[Player] = PlayerMaterial) {}
            # Show the stance UI
            PlayerUI.AddWidget(StanceOverlay)              
            
    # When a player is added to the playspace, add mapping contexts.
    OnPlayerAdded(Player:player):void=
        if (PlayerInput := GetPlayerInput[Player]):
            Print("Player added")
            # Adds the input mapping itself, which is what will actually
            PlayerInput.AddInputMapping(Character.TraversalMapping)
            PlayerInput.AddInputMapping(Character.RangedWeaponMapping)

    # When a player is removed, remove mapping contexts.
    OnPlayerRemoved(Player:player):void=
        if (PlayerInput := GetPlayerInput[Player]):
            Print("Player removed")
            # Adds the input mapping itself, which is what will actually
            PlayerInput.RemoveInputMapping(Character.TraversalMapping)
            PlayerInput.RemoveInputMapping(Character.RangedWeaponMapping)

    # Event to change the crouch UI indicator on HUD
    OnCrouchedEvent(FortCharacter:fort_character, isCrouch:logic):void=
        if (Agent := FortCharacter.GetAgent[], Player := player[Agent], Material := PlayerStanceMaterials[Player]):
            if (isCrouch = true):
                Print("Player crouched")
                set Material.Texture = T_UI_Crouch
            else:
                Print("Player standing")
                set Material.Texture = T_UI_Stand

    # Hide the UI
    OnHide(InAgent:agent):void=
        if (Player := player[InAgent]):
            if (ExistingSubs := CrouchedSubscriptions[Player]):
                for (Sub : ExistingSubs):
                    Sub.Cancel()
            if (set CrouchedSubscriptions[Player] = array{}) {}
            if (PlayerUI := GetPlayerUI[Player], Huds := PlayerStanceHuds[Player]):
                for (Hud : Huds):
                    PlayerUI.RemoveWidget(Hud)
            if (set PlayerStanceHuds[Player] = array{}) {}
