# Knowledge-neutral Verse snippet; project names and asset IDs generalized.
# id: snippet-192

# Copyright Epic Games, Inc. All Rights Reserved.

using { /Fortnite.com/Devices }
using { /Verse.org/Simulation }
using { /Verse.org/SceneGraph }
using { /Verse.org/Assets }
using { /Verse.org/Input }
using { /Fortnite.com/Input }
using { /Fortnite.com/Input/Character }
using { /Fortnite.com/Characters }
using { /UnrealEngine.com/Temporary/UI }
using { /UnrealEngine.com/Temporary/Diagnostics }
using { UserInterfaces.SceneGraphUIExamples.InWorldUMG.Widgets }

# Number of upgrade steps required to reach a full bar
UpgradeStepsToMax : int = 10

# Helper function to zero-pad an int to two digits
PadTwo(N : int) : string =
    if (N < 10):
        "0{N}"
    else:
        "{N}"

# Helper function to format seconds to MM:SS
FormatTime(TotalSeconds : int) : string =
    if (Minutes := Quotient[TotalSeconds, 60], Seconds := Mod[TotalSeconds, 60]):
        "{PadTwo(Minutes)}:{PadTwo(Seconds)}"
    else:
        "00:00"

# Helpers for the two text bindings on the widget.
UMGTimeMsg<localizes>(Text : string) : message = "{Text}"
UMGStepsMsg<localizes>(Steps : int, MaxSteps : int) : message = "{Steps}/{MaxSteps}"

ui_in_world_umg_device := class(creative_device):

    @editable
    ShowVolume : volume_device = volume_device{}

    @editable
    ShowWidget : UserInterfaces.SceneGraphUIExamples.InWorldUMG.Widgets.UW_InWorldUMG01_Layer02_entity = UserInterfaces.SceneGraphUIExamples.InWorldUMG.Widgets.UW_InWorldUMG01_Layer02_entity{}

    TimerDurationSeconds : float = 1.0

    var CurrentStep : int = 0

    var AgentEntersSubscriptionOption : ?cancelable = false
    var AgentExitsSubscriptionOption : ?cancelable = false
    var InputSubscriptionOption : ?cancelable = false

    OnBegin<override>()<suspends> : void =
        # Track when the player is inside the volume
        set AgentEntersSubscriptionOption = option { ShowVolume.AgentEntersEvent.Subscribe(OnAgentEntersVolume) }
        set AgentExitsSubscriptionOption = option { ShowVolume.AgentExitsEvent.Subscribe(OnAgentExitsVolume) }

        # Show the player's portrait
        SetPortrait()

        # Run the widget loop
        RunLoop()

    OnEnd<override>() : void =
        CancelSubscription(AgentEntersSubscriptionOption)
        CancelSubscription(AgentExitsSubscriptionOption)
        CancelSubscription(InputSubscriptionOption)

    # Player entered the volume: flag the widget and start listening for input
    OnAgentEntersVolume(InAgent : agent) : void =
        set ShowWidget.PlayerInVolume = true
        if (Player := player[InAgent], PlayerInput := GetPlayerInput[Player]):
            Subscription := PlayerInput.GetInputEvents(Character.Reload).BeginDetectEvent.Subscribe(OnInputPressed)
            set InputSubscriptionOption = option { Subscription }

    # Player left the volume: clear the flag and stop listening for input
    OnAgentExitsVolume(InAgent : agent) : void =
        set ShowWidget.PlayerInVolume = false
        CancelSubscription(InputSubscriptionOption)
        set InputSubscriptionOption = false

    # Get the Player's Icon Componenet (Player->PlayerEntity->IconComponent)
    SetPortrait() : void =
        if:
            Player := GetPlayspace().GetPlayers()[0]
            PlayerEntity := Player.GetFortCharacter[].GetEntity[]
            IconComponent := PlayerEntity.GetComponent[icon_component]
        then:
            set ShowWidget.PortraitTexture = IconComponent.Icon

    #Input handler. Input is only subscribed while in the volume
    OnInputPressed(Args : tuple(player, logic)) : void =
        if (ShowWidget.CanUpgrade?):
            ApplyUpgradeProgress(CurrentStep + 1)
            set ShowWidget.CanUpgrade = false

    ApplyUpgradeProgress(Step : int) : void =
        set CurrentStep = Step
        set ShowWidget.UpgradeProgress = (Step * 1.0) / (UpgradeStepsToMax * 1.0)
        set ShowWidget.UpgradeProgressText = UMGStepsMsg(Step, UpgradeStepsToMax)

    RunLoop()<suspends> : void =
        loop:
            ApplyUpgradeProgress(0)

            loop:
                # Countdown phase
                set ShowWidget.CanUpgrade = false
                RunCountdown()

                # Upgrade phase: timer is at 0, allow anupgrade
                set ShowWidget.TimeNormalized = 0.0
                set ShowWidget.TimeText = UMGTimeMsg(FormatTime(0))
                set ShowWidget.CanUpgrade = true

                # Wait for the IA_Input handler to consume the upgrade, then restart the timer
                loop:
                    if (not ShowWidget.CanUpgrade?):
                        break
                    Sleep(0.0)

                # The press that reaches the final step, breaks here, so the bar resets immediately on that upgrade
                if (CurrentStep >= UpgradeStepsToMax):
                    break

            # Bar is full: increment the; the outer loop resets progress back to 0 for the next level
            Sleep(0.5)

            set ShowWidget.Level = ShowWidget.Level + 1

    # Counts down TimerDurationSeconds, updating TimeText and TimeNormalized
    RunCountdown()<suspends> : void =
        StartTime : float = GetSecondsSinceEpoch()

        loop:
            Elapsed := GetSecondsSinceEpoch() - StartTime
            Remaining := TimerDurationSeconds - Elapsed

            if (Remaining <= 0.0):
                set ShowWidget.TimeNormalized = 0.0
                set ShowWidget.TimeText = UMGTimeMsg(FormatTime(0))
                return
            set ShowWidget.TimeNormalized = Remaining / TimerDurationSeconds

            if (DisplaySeconds := Ceil[Remaining]):
                set ShowWidget.TimeText = UMGTimeMsg(FormatTime(DisplaySeconds))
            Sleep(0.0)
