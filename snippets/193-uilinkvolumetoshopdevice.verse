# Knowledge-neutral Verse snippet; project names and asset IDs generalized.
# id: snippet-193

using { /Fortnite.com/Devices }
using { /Verse.org/Simulation }
using { /UnrealEngine.com/Temporary/Diagnostics }


# See https://dev.epicgames.com/documentation/en-us/uefn/create-your-own-device-in-verse for how to create a verse device.

# A Verse-authored creative device that can be placed in a level
ui_link_volume_to_shop_device := class(creative_device):

    @editable
    var ShopTemplate:ui_shop_setup_device_template = ui_shop_setup_device_template{}
    
    @editable
    var OfferVolume:volume_device = volume_device{}

    @editable
    var OfferVolume02:volume_device = volume_device{}


    OnBegin<override>()<suspends>:void =          
        OfferVolume.AgentEntersEvent.Subscribe(OnVolumeEnter) 
        OfferVolume02.AgentEntersEvent.Subscribe(OnVolume02Enter)
   
    OnVolumeEnter(Agent:agent):void=
        if(Player:= player[Agent]):
            spawn:
                ShopTemplate.ShowArrayOfOffers(Player) 
                
    OnVolume02Enter(Agent:agent):void=
        if(Player:= player[Agent]):
            var shop : ui_shop02 = ui_shop02{Player := Player}
            shop.ShowUI()
