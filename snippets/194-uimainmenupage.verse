# Knowledge-neutral Verse snippet; project names and asset IDs generalized.
# id: snippet-194

# Copyright Epic Games, Inc. All Rights Reserved.

using { /Verse.org/Simulation }
using { /Fortnite.com/Devices }
using { /Fortnite.com/UI }
using { /UnrealEngine.com/Temporary/UI }
using { /Verse.org/Colors }
using { /Fortnite.com/Characters }
using { /Verse.org/SceneGraph }
using { /Verse.org/Input }

# ============================================================
#  main_menu_page
#
#  From here the player can:
#    - Press Play → game mode selection
#    - Press Upgrades → upgrade shop
#    - Press Settings → toggle virtual pointer
# ============================================================
ui_main_menu_page := class(ui_menu_page):

    MainMenuWidget : UserInterfaces.UIFeatureExamples.Menu.Widgets.MainMenu.UW_Menu_MainMenu = UserInterfaces.UIFeatureExamples.Menu.Widgets.MainMenu.UW_Menu_MainMenu{}

    # When true, show players who are in the lobby but haven't started a game yet
    ShowLobbyPlayers : logic = true

    # Overlay that holds both the main menu widget and the session players panel
    ContentOverlay : overlay = overlay{}

    GetFocusWidget<override>() : widget = MainMenuWidget

    SessionPlayersWidget : UserInterfaces.UIFeatureExamples.Menu.Widgets.MainMenu.UW_Menu_Lobby = UserInterfaces.UIFeatureExamples.Menu.Widgets.MainMenu.UW_Menu_Lobby{}

    SessionPlayersOverlay : overlay = overlay{}

    var LastPlayerStack : ?stack_box = false

    var LeaderboardUpdateCancelable : ?cancelable_event = false

    var IsMobile : logic = false

    # Each value maps to a button so the widget can fire the respective button's press animation
    NoAnimatedButton : int = -1
    PlayButtonIndex : int = 0
    UpgradesButtonIndex : int = 1
    GarageButtonIndex : int = 2
    DriverClubButtonIndex : int = 3
    SettingsButtonIndex : int = 4

    # How long to let the mobile button press animation play before navigating.
    MobileButtonAnimDelay : float = 0.25

    NeedNavigationMapping<override> : logic = false

    GetTitle<override>() : string = ""
    GetContentWidget<override>() : widget = ContentOverlay
    NeedBackButton<override> : logic = false
    ShowTitle<override> : logic = false
    ShowCurrency<override> : logic = true
    NeedSelectButton<override> : logic = false

    PlayButtonText<localizes>: message = "Play"
    UpgradesButtonText<localizes>: message = "Upgrades"
    GarageButtonText<localizes> : message = "Garage"
    SettingsButtonText<localizes> : message = "Settings"
    ExitButtonText<localizes> : message = "Exit"

    InitData<override>() : void =
        set MainMenuWidget.PlayButtonText = PlayButtonText
        set MainMenuWidget.UpgradesButtonText = UpgradesButtonText
        set MainMenuWidget.GarageButtonText = GarageButtonText
        set MainMenuWidget.SettingsButtonText = SettingsButtonText
        set MainMenuWidget.ExitButtonText = ExitButtonText

        MainMenuSlot := overlay_slot:
            Widget := MainMenuWidget
            HorizontalAlignment := horizontal_alignment.Fill
            VerticalAlignment := vertical_alignment.Fill
        ContentOverlay.AddWidget(MainMenuSlot)

        SessionPlayersContainerOverlay : overlay = overlay{}

        SessionPlayersWidgetSlot := overlay_slot:
            Widget := SessionPlayersWidget
            HorizontalAlignment := horizontal_alignment.Fill
            VerticalAlignment := vertical_alignment.Fill
        SessionPlayersContainerOverlay.AddWidget(SessionPlayersWidgetSlot)

        PlayerStackSlot := overlay_slot:
            Widget := SessionPlayersOverlay
            HorizontalAlignment := horizontal_alignment.Left
            VerticalAlignment := vertical_alignment.Top
            Padding := margin{ Top := 75.0, Left := 75.0 }
        SessionPlayersContainerOverlay.AddWidget(PlayerStackSlot)

        BottomLeftSlot := overlay_slot:
            Widget := SessionPlayersContainerOverlay
            HorizontalAlignment := horizontal_alignment.Left
            VerticalAlignment := vertical_alignment.Bottom
        ContentOverlay.AddWidget(BottomLeftSlot)

    SubscribeToPageEvents<override>() : void =
        PlaySub := SubscribeToEvent(MainMenuWidget.PlayClicked, OnPlayClicked)
        UpgradesSub := SubscribeToEvent(MainMenuWidget.UpgradesClicked, OnUpgradesClicked)
        SettingsSub := SubscribeToEvent(MainMenuWidget.SettingsClicked, OnSettingsClicked)
        ExitSub := SubscribeToEvent(MainMenuWidget.ExitClicked, OnExitClicked)
        set PageSubscriptions += array{ PlaySub, UpgradesSub, SettingsSub, ExitSub }

    OnPlayClicked(Event : tuple()) : void =
        spawn{ OnPlayClickedAsync() }

    OnPlayClickedAsync()<suspends> : void =
        OnSelect()

    OnUpgradesClicked(Event : tuple()) : void =
        spawn{ OnUpgradesClickedAsync() }

    OnUpgradesClickedAsync()<suspends> : void =
        PlayMobileButtonAnim(UpgradesButtonIndex)
        Session.GoToUpgrades()

    OnSettingsClicked(Event : tuple()) : void =
        spawn{ OnSettingsClickedAsync() }

    OnSettingsClickedAsync()<suspends> : void =
        PlayMobileButtonAnim(SettingsButtonIndex)
        Session.GoToSettings()

    OnExitClicked(Event : tuple()) : void =
        spawn{ OnExitClickedAsync() }

    OnExitClickedAsync()<suspends> : void =
        Session.CloseAll()

    # Fire the button's press animation and wait before navigating (mobile only)
    PlayMobileButtonAnim(ButtonIndex : int)<suspends> : void =
        if (IsMobile?):
            set MainMenuWidget.CurrentAnimatedButton = ButtonIndex
            Sleep(MobileButtonAnimDelay)
            set MainMenuWidget.CurrentAnimatedButton = NoAnimatedButton

    # 1. The menu car cinematic sequence plays
    # 2. Wait, then go to game mode selection
    OnSelect<override>()<suspends> : void =
        set MainMenuWidget.TriggerOutro = true
        Session.MenuDevice.MenuCarSequence.Play(Session.Player)
        Sleep(0.8)
        Session.GoToGameMode()

    OnShow<override>():void=
        set MainMenuWidget.TriggerIntro = true
        set MainMenuWidget.CurrentAnimatedButton = NoAnimatedButton

        # Detect touch/mobile so page elements visibility can be toggled accordingly
        set IsMobile = if (Session.IsMobile?) then true else false
        BackButtonWidget.SetVisibility(widget_visibility.Collapsed)
        Session.MenuDevice.MainMenuCamera.AddTo(Session.Player)

        set MainMenuWidget.PlayerNickname = GetUsername(Session.Player)
        if:
            FortCharacter := Session.Player.GetFortCharacter[]
            PlayerEntity := FortCharacter.GetEntity[]
            IconComponent := PlayerEntity.GetComponent[icon_component]
        then:
            set MainMenuWidget.PlayerPortrait = IconComponent.Icon

        # Start the leaderboard update loop
        Cancelable := cancelable_event{}
        set LeaderboardUpdateCancelable = option{ Cancelable }
        spawn{ RunLeaderboardLoop(Cancelable) }

    # Remove this player from the camera and stop the leaderboard loop
    OnHide<override>():void=
        Session.MenuDevice.MainMenuCamera.RemoveFrom(Session.Player)

        # Stop leaderboard update loop
        if (Cancelable := LeaderboardUpdateCancelable?):
            Cancelable.Cancel()
        set LeaderboardUpdateCancelable = false
        set LastPlayerStack = false

        set MainMenuWidget.TriggerIntro = false
        set MainMenuWidget.TriggerOutro = false

    # Rebuilds the player stack while the page is visible
    RunLeaderboardLoop(Cancelable : cancelable_event)<suspends> : void =
        race:
            Cancelable.Await()
            loop:
                RebuildLeaderboard()
                Sleep(1.0)

    # Clears and rebuilds the player details stack from current player data
    RebuildLeaderboard() : void =
        if (OldStack := LastPlayerStack?):
            SessionPlayersOverlay.RemoveWidget(OldStack)

        InGamePlayers := Session.MenuDevice.GetSortedLeaderboards()

        # If ShowLobbyPlayers is true, find any players not already in the leaderboard and append them
        var AllPlayers : []tuple(player, int) = InGamePlayers
        if (ShowLobbyPlayers = true):
            for (Player : Session.MenuDevice.GetPlayspace().GetPlayers()):
                var AlreadyInList : logic = false
                for (InGamePlayer : InGamePlayers):
                    if (InGamePlayer(0) = Player):
                        set AlreadyInList = true
                if (AlreadyInList = false):
                    set AllPlayers += array{ (Player, 0) }

        if (AllPlayers.Length = 0):
            # Flag if no players have started a game session yet.
            set SessionPlayersWidget.WaitingForPlayers = true
        else:
            set SessionPlayersWidget.WaitingForPlayers = false

            NewStack : stack_box = stack_box{ Orientation := orientation.Horizontal }

            for (LeaderboardItem : AllPlayers):
                PlayerDetailsWidget : UserInterfaces.UIFeatureExamples.Menu.Widgets.MainMenu.UW_Menu_PlayerDetails_Lobby =
                    UserInterfaces.UIFeatureExamples.Menu.Widgets.MainMenu.UW_Menu_PlayerDetails_Lobby{}
                set PlayerDetailsWidget.Nickname = GetUsername(LeaderboardItem(0))

                # Fetch the player's portrait from their icon component if available.
                if:
                    FortCharacter := LeaderboardItem(0).GetFortCharacter[]
                    PlayerEntity := FortCharacter.GetEntity[]
                    IconComponent := PlayerEntity.GetComponent[icon_component]
                then:
                    set PlayerDetailsWidget.Portrait = IconComponent.Icon
                NewStackSlot := stack_box_slot{ Widget := PlayerDetailsWidget }
                NewStack.AddWidget(NewStackSlot)

            set LastPlayerStack = option{ NewStack }

            # Replace the contents of the session players overlay with the new stack.
            NewStackOverlaySlot := overlay_slot:
                Widget := NewStack
                HorizontalAlignment := horizontal_alignment.Fill
                VerticalAlignment := vertical_alignment.Fill
            SessionPlayersOverlay.AddWidget(NewStackOverlaySlot)

    OnBack<override>()<suspends> : void = {}
