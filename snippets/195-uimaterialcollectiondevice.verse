# Knowledge-neutral Verse snippet; project names and asset IDs generalized.
# id: snippet-195

using { /Fortnite.com/Devices }
using { /Fortnite.com/UI }
using { /UnrealEngine.com/Temporary/UI }
using { /UnrealEngine.com/Temporary/SpatialMath }
using { /Verse.org/Simulation }
using { /Verse.org/Random }

CloseLabel<localizes> : message = "CLOSE"

ui_material_collection_device := class(creative_device):
    @editable ShowVolume : volume_device = volume_device{}

    var MaterialCollectionScreen : UserInterfaces.UIArtExamples.MaterialCollection.Widgets.UW_MaterialCollection_Menu =
        UserInterfaces.UIArtExamples.MaterialCollection.Widgets.UW_MaterialCollection_Menu{}

    var RootOverlay : overlay = overlay{}
    var CloseBtn : button_loud = button_loud{ DefaultText := CloseLabel }
    var IsBuilt : logic = false
    var IsOpen : logic = false
    var LoopsStarted: logic = false

    # Animation tuning
    TickSeconds : float = 0.02
    ProgressUnitsPerSec : float = 0.80

    # Linear meter delta animation tuning
    DamageEveryMin : float = 0.80
    DamageEveryMax : float = 2.00
    DamageMin : float = 0.10
    DamageMax : float = 0.30
    DeltaAnimSeconds : float = 0.50
    ResetThreshold : float = 0.05
    ResetDelaySeconds : float = 0.75

    var CurrentHealthProgress : float = 1.0

    OnBegin<override>()<suspends>:void =
        ShowVolume.AgentEntersEvent.Subscribe(OnShow)

    Build():void =
        if (IsBuilt = true):
            return

        set RootOverlay = overlay{}
        # UMG
        RootOverlay.AddWidget(overlay_slot{
            Widget := MaterialCollectionScreen,
            HorizontalAlignment := horizontal_alignment.Fill,
            VerticalAlignment := vertical_alignment.Fill
        })
        # Close button
        RootOverlay.AddWidget(overlay_slot{
            Widget := CloseBtn,
            HorizontalAlignment := horizontal_alignment.Center,
            VerticalAlignment := vertical_alignment.Bottom,
            Padding := margin{ Bottom := 145.0 }
        })
        CloseBtn.OnClick().Subscribe(OnCloseClicked)

        set IsBuilt = true

    OnShow(InAgent:agent):void =
        Build()
        set IsOpen = true
        if (Player := player[InAgent]):
            if (PlayerUI := GetPlayerUI[Player]):
                PlayerUI.AddWidget(RootOverlay, player_ui_slot{ InputMode := ui_input_mode.All })

        InitFloats()

        if (LoopsStarted = false):
            set LoopsStarted = true
            spawn { RepeatLoop() }
            spawn { PingPongLoop() }
            spawn { LinearDeltaLoop() }

    OnCloseClicked(Msg:widget_message):void =
        if (IsOpen = true):
            set IsOpen = false
            if (PlayerUI := GetPlayerUI[Msg.Player]):
                PlayerUI.RemoveWidget(RootOverlay)

    InitFloats():void =
        set MaterialCollectionScreen.ProgressRepeat = 0.0
        set MaterialCollectionScreen.ProgressPingPong = 0.0
        set MaterialCollectionScreen.LinearMeterDeltaProgress = 1.0
        set MaterialCollectionScreen.LinearMeterDeltaAmount = 0.0
        set MaterialCollectionScreen.LinearMeterDeltaAnim = 0.0
        set CurrentHealthProgress = 1.0

    # 0->1 repeat animation loop
    RepeatLoop()<suspends>:void =
        var current : float = 0.0
        loop:
            if (IsOpen = true):
                set MaterialCollectionScreen.ProgressRepeat = current

            step := ProgressUnitsPerSec * TickSeconds
            set current += step
            if (current >= 1.0):
                set current = current - 1.0

            Sleep(TickSeconds)

    # 0->1->0 ping-pong animation loop
    PingPongLoop()<suspends>:void =
        var current : float = 0.0
        var direction : float = 1.0
        loop:
            if (IsOpen = true):
                set MaterialCollectionScreen.ProgressPingPong = current

            step := ProgressUnitsPerSec * TickSeconds * direction
            set current += step

            if (current >= 1.0):
                set current = 1.0
                set direction = -1.0
            else if (current <= 0.0):
                set current = 0.0
                set direction = 1.0

            Sleep(TickSeconds)

    # Linear meter delta loop
    LinearDeltaLoop()<suspends>:void =
        set CurrentHealthProgress = 1.0
        set MaterialCollectionScreen.LinearMeterDeltaProgress = CurrentHealthProgress
        set MaterialCollectionScreen.LinearMeterDeltaAmount = 0.0
        set MaterialCollectionScreen.LinearMeterDeltaAnim = 0.0

        loop:
            if (IsOpen = false):
                Sleep(TickSeconds)
            else:
                if (CurrentHealthProgress <= ResetThreshold):
                    set CurrentHealthProgress = 1.0
                    set MaterialCollectionScreen.LinearMeterDeltaProgress = CurrentHealthProgress
                    set MaterialCollectionScreen.LinearMeterDeltaAmount = 0.0
                    set MaterialCollectionScreen.LinearMeterDeltaAnim = 0.0
                    Sleep(ResetDelaySeconds)
                else:
                    Sleep(GetRandomFloat(DamageEveryMin, DamageEveryMax))

                    rawDamage := GetRandomFloat(DamageMin, DamageMax)
                    damage := if (rawDamage > CurrentHealthProgress) { CurrentHealthProgress } else { rawDamage }

                    ApplyMeterLoss(damage)

    TriggerMeterLoss(Amount:float):void =
        spawn { ApplyMeterLoss(Amount) }

    # Linear meter delta falloff
    ApplyMeterLoss(Amount:float)<suspends>:void =
        if (Amount <= 0.0):
            return

        loss := if (Amount > CurrentHealthProgress) { CurrentHealthProgress } else { Amount }
        var targetProgress : float = CurrentHealthProgress - loss
        if (targetProgress < 0.0):
            set targetProgress = 0.0

        set MaterialCollectionScreen.LinearMeterDeltaAmount = loss
        set MaterialCollectionScreen.LinearMeterDeltaAnim = 1.0

        duration := if (DeltaAnimSeconds > 0.0) { DeltaAnimSeconds } else { 0.0 }
        if (duration = 0.0):
            set CurrentHealthProgress = targetProgress
            set MaterialCollectionScreen.LinearMeterDeltaProgress = CurrentHealthProgress
            set MaterialCollectionScreen.LinearMeterDeltaAnim = 0.0
        else:
            fadeRate : float = TickSeconds / duration
            progressStep: float = loss * fadeRate

            loop:
                if (IsOpen = false):
                    Sleep(TickSeconds)
                else:
                    if (CurrentHealthProgress > targetProgress):
                        nextProgress : float = CurrentHealthProgress - progressStep
                        set CurrentHealthProgress =
                            if (nextProgress <= targetProgress) { targetProgress } else { nextProgress }
                        set MaterialCollectionScreen.LinearMeterDeltaProgress = CurrentHealthProgress

                    nextAnim : float = MaterialCollectionScreen.LinearMeterDeltaAnim - fadeRate
                    set MaterialCollectionScreen.LinearMeterDeltaAnim =
                        if (nextAnim <= 0.0) { 0.0 } else { nextAnim }

                    reachedTarget : logic = logic{(CurrentHealthProgress = targetProgress)}
                    if (reachedTarget = true):
                        finishedAnim : logic = logic{(MaterialCollectionScreen.LinearMeterDeltaAnim = 0.0)}
                        if (finishedAnim = true):
                            break

                    Sleep(TickSeconds)

        set MaterialCollectionScreen.LinearMeterDeltaAmount = 0.0
