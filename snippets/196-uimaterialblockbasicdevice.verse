# Knowledge-neutral Verse snippet; project names and asset IDs generalized.
# id: snippet-196

using { /Fortnite.com/Devices }
using { /Fortnite.com/UI }
using { /Verse.org/Simulation }
using { /Verse.org/Colors }
using { /UnrealEngine.com/Temporary/SpatialMath }
using { /UnrealEngine.com/Temporary/Diagnostics }
using { /UnrealEngine.com/Temporary/UI }
using {UserInterfaces.VerseUIExamples.MaterialBlock.MaterialInstances}
using {UserInterfaces.VerseUIExamples.MaterialBlock.Textures}
            

# A Verse-authored creative device that can be placed in a level
ui_materialblock_basic_device := class(creative_device):

    @editable
    ShowTrigger:volume_device = volume_device{}    

    @editable
    HideOnExitVolume:volume_device = volume_device{}

    @editable
    HideOnEnterVolume:volume_device = volume_device{}

    # Verse UI for the demo
    var MyUI : overlay = overlay {}

    # Materials used
    RadialMeter:MI_UI_MatBlock_RadialMeter = MI_UI_MatBlock_RadialMeter{Progress := 1.0}
    DissolveArt:M_UI_Dissolve = M_UI_Dissolve{Dissolve := 0.0}

    OnBegin<override>()<suspends>:void=
        # Subscribe to volume enter and exit event 
        ShowTrigger.AgentEntersEvent.Subscribe(OnShow)

        HideOnExitVolume.AgentExitsEvent.Subscribe(OnHide)
        HideOnEnterVolume.AgentEntersEvent.Subscribe(OnHide)

        # Material block used for radial
        RadialMaterialBlock := material_block:
            DefaultImage := RadialMeter
            DefaultDesiredSize := vector2{X:=436.0, Y:=436.0}

        # Material block used for dissolve art
        DissolveMaterialBlock := material_block:
            DefaultImage := DissolveArt
            DefaultDesiredSize := vector2{X:=336.0, Y:=336.0}

        # Create UI
        MyHudUI := overlay:
            Slots := array:
                overlay_slot:
                    VerticalAlignment := vertical_alignment.Center
                    HorizontalAlignment := horizontal_alignment.Center   
                    Padding := margin{Left:=0.0, Top:=0.0, Right:=64.0, Bottom:=0.0}                   
                    Widget := stack_box:
                        Orientation := orientation.Horizontal
                        Slots := array:
                            stack_box_slot:
                                Widget := RadialMaterialBlock
                                Padding := margin{Left:=24.0, Top:=24.0, Right:=24.0, Bottom:=24.0}
                            stack_box_slot:
                                Widget := DissolveMaterialBlock
                                Padding := margin{Left:=24.0, Top:=24.0, Right:=24.0, Bottom:=24.0}
        set MyUI = MyHudUI

    # Show the UI (agent is optional and ignorted for this demo)
    OnShow(InAgent:agent):void=

        for:
            Player : GetPlayspace().GetPlayers()
            PlayerUI := GetPlayerUI[Player]
        do:
            PlayerUI.AddWidget(MyUI)
        
        spawn{AnimationLoop()}

    # Hide the UI
    OnHide(InAgent:agent):void=
        for:
            Player:GetPlayspace().GetPlayers()
            PlayerUI := GetPlayerUI[Player]
        do:
            PlayerUI.RemoveWidget(MyUI)

    # Loops an animation created by setting material block properties
    AnimationLoop()<suspends>:void=
        set RadialMeter.Progress = 0.00
        set DissolveArt.Dissolve  = 0.255

        loop:
            set RadialMeter.Progress += 0.01
            set DissolveArt.Dissolve += 0.004
            Sleep(0.0)

            if ((RadialMeter.Progress >= 1.0) and (DissolveArt.Dissolve >= 0.7)):
                set RadialMeter.Progress = 0.00
                set DissolveArt.Dissolve  = 0.255                
                break
