# Knowledge-neutral Verse snippet; project names and asset IDs generalized.
# id: snippet-197

using { /Fortnite.com/Devices }
using { /Fortnite.com/UI }
using { /Verse.org/Simulation }
using { /Verse.org/Colors }
using { /UnrealEngine.com/Temporary/SpatialMath }
using { /UnrealEngine.com/Temporary/Diagnostics }
using { /UnrealEngine.com/Temporary/UI }
using { /Fortnite.com/AI }
using { /Fortnite.com/Characters }
using { /Fortnite.com/Game }
using {UserInterfaces.VerseUIExamples.MaterialBlock.MaterialInstances}
using {UserInterfaces.VerseUIExamples.MaterialBlock.Textures}

# See https://dev.epicgames.com/documentation/en-us/uefn/create-your-own-device-in-verse for how to create a verse device.

# A Verse-authored creative device that can be placed in a level
ui_materialblock_gameplay_device := class(creative_device):

    # Provide access to guard spawner and a testing trigger
    @editable
    Spawner:guard_spawner_device = guard_spawner_device{}
    @editable
    ShowTrigger:volume_device = volume_device{}
    @editable
    RemoveVolume:[]volume_device = array{}
    @editable
    HUDController:hud_controller_device = hud_controller_device{}

    # Verse UI for the demo
    var MyUI : overlay = overlay {}

    # The materials used in the demo:
    #   RadialMeter: enemy's health meter
    #   PortaitArt: enemy's portrait (pulses when the enemy takes damage, animates when a new enemy spawns)
    #   CountMeter: enemy spawner (shows how many spawns are left)
    #   StatusIcon: enemy's awareness status (alerted, lost target, suspicious, unaware)
    RadialMeter:MI_UI_MatBlock_Health = MI_UI_MatBlock_Health{Progress := 1.0}
    PortaitArt:MI_UI_MatBlock_Portrait = MI_UI_MatBlock_Portrait{TexturePositionY := 0.0}
    CountMeter:MI_UI_MatBlock_Counter = MI_UI_MatBlock_Counter{Progress := 0.0}
    StatusIcon:MI_UI_MatBlock_Status = MI_UI_MatBlock_Status{IsBouncing := 0.0, IsScaling := 0.0}

    # Variables used in the demo
    var EnemiesEliminated:int = 0
    var EnemiesLeftInt:int = 0
    var SpawnerActive:logic = false
    var EnemiesLeftFloat:float = 0.0
    var IntroEnemyAnimValue: float = 0.0
    var AgentIsInsideVolume : logic = false

    # Text block for the UI
    var GuardsLeft : text_block = text_block:
        DefaultTextColor := NamedColors.White
        DefaultShadowColor:= color{R:=0.0, G:=0.0, B:=0.0}
        DefaultShadowOffset := option{vector2{X:=5.0, Y:=2.0}}

    # Text helper function
    SetDynamicText<localizes>(IntValue:int):message = "Left: {IntValue}"

    OnBegin<override>()<suspends>:void=
        # Subscribe to volume enter event 
        ShowTrigger.AgentEntersEvent.Subscribe(OnShow)
        ShowTrigger.AgentExitsEvent.Subscribe(OnAgentExitsShowVolume)

        for (Vol : RemoveVolume):
            Vol.AgentExitsEvent.Subscribe(ResetGuardStates)

        # Subscribe to guard spawner events
        Spawner.SpawnedEvent.Subscribe(GuardSpawned)
        Spawner.AlertedEvent.Subscribe(GuardAlerted)
        Spawner.TargetLostEvent.Subscribe(GuardLostTarget)
        Spawner.SuspiciousEvent.Subscribe(GuardSuspicious)
        Spawner.UnawareEvent.Subscribe(GuardUnaware)
        Spawner.DamagedEvent.Subscribe(GuardDamaged)
        Spawner.EliminatedEvent.Subscribe(GuardEliminated)
        Spawner.EliminatingEvent.Subscribe(GuardEliminating)

        # Material block used for the enemy health radial
        RadialMaterialBlock := material_block:
            DefaultImage := RadialMeter
            DefaultDesiredSize := vector2{X:=436.0, Y:=436.0}

        # Material block used for the enemy portrait
        PortraitMaterialBlock := material_block:
            DefaultImage := PortaitArt
            DefaultDesiredSize := vector2{X:=336.0, Y:=336.0}

        # Material block used for the spawn counter   
        SpawnCountMaterialBlock := material_block:
            DefaultImage := CountMeter
            DefaultDesiredSize := vector2{X:=278.0, Y:=278.0}
        
        # Material block used for the enemy status indicator
        StatusMaterialBlock := material_block:
            DefaultImage := StatusIcon
            DefaultDesiredSize := vector2{X:=72.0, Y:=72.0}
        
        # Create UI
        MyHudUI := overlay:
            Slots := array:
                overlay_slot:
                    VerticalAlignment := vertical_alignment.Center
                    HorizontalAlignment := horizontal_alignment.Right   
                    Padding := margin{Left:=0.0, Top:=0.0, Right:=64.0, Bottom:=0.0}                   
                    Widget := stack_box:
                        Orientation := orientation.Vertical
                        Slots := array:
                            stack_box_slot:
                                Widget := overlay:
                                    Slots := array:
                                        overlay_slot:
                                            Widget := RadialMaterialBlock
                                            Padding := margin{Left:=0.0, Top:=100.0, Right:=0.0, Bottom:=32.0}
                                        overlay_slot:
                                            Widget := PortraitMaterialBlock
                                            Padding := margin{Left:=0.0, Top:=100.0, Right:=0.0, Bottom:=32.0}
                            stack_box_slot:
                                Widget := SpawnCountMaterialBlock
                                VerticalAlignment := vertical_alignment.Top
                                Padding := margin{Left:=0.0, Top:=-164.0, Right:=0.0, Bottom:=0.0}
                            stack_box_slot:
                                Widget := GuardsLeft
                                Padding := margin{Left:=0.0, Top:=-42.0, Right:=0.0, Bottom:=0.0}
                overlay_slot:
                    VerticalAlignment := vertical_alignment.Bottom
                    HorizontalAlignment := horizontal_alignment.Right   
                    Padding := margin{Left := 0.0, Top := 0.0, Right := 64.0, Bottom := 354.0}                   
                    Widget := StatusMaterialBlock
        set MyUI = MyHudUI

    # Show the UI (agent is optional and ignored for this demo)
    OnShow(InAgent:agent):void=
        if (player[InAgent]):
            Spawner.Reset()
            set EnemiesEliminated = 0
            set EnemiesLeftInt = Spawner.GetSpawnLimit()
            set EnemiesLeftFloat = EnemiesLeftInt * 1.0
            GuardsLeft.SetText(SetDynamicText(EnemiesLeftInt))
            set AgentIsInsideVolume = true
            HUDController.Enable()

            if (SpawnerActive=false and EnemiesLeftInt=3):
                Spawner.Enable()
                Spawner.Spawn()
                for:
                    Player : GetPlayspace().GetPlayers()
                    PlayerUI := GetPlayerUI[Player]
                do:
                    set CountMeter.Progress = EnemiesLeftFloat
                    set CountMeter.ProgressRangeMax = 1.0 * Spawner.GetSpawnLimit()
                    set CountMeter.NumSteps = 1.0 * Spawner.GetSpawnLimit()
                    PlayerUI.AddWidget(MyUI)
        
    # Hide the UI
    OnHide(InAgent:agent):void=
        for:
            Player:GetPlayspace().GetPlayers()
            PlayerUI := GetPlayerUI[Player]
        do:
            PlayerUI.RemoveWidget(MyUI)

    # Guard Events
    GuardSpawned(InAgent: agent):void=
        Print ("Guard Event: Spawned")
        # Set enemy's health radial meter value
        if (FortChar := InAgent.GetFortCharacter[]):
            EnemyHealthValue := FortChar.GetHealth()
            set RadialMeter.Progress = EnemyHealthValue
            Print ("Spawned a guard with health set to {EnemyHealthValue}")

        # Set health radial visual properties
        set RadialMeter.PositionEnd = 0.5
        set PortaitArt.OutlineAlpha = 0.0
        set PortaitArt.Glow = 0.0248
        set PortaitArt.OutlineColor1 = MakeColorFromSRGB(1.0, 0.0, 0.0)
        set PortaitArt.OutlineColor2 = MakeColorFromSRGB(0.1, 0.0, 0.0)
        set PortaitArt.OutlineColor3 = MakeColorFromSRGB(1.0, 0.0, 0.0)

    GuardAlerted(Result: device_ai_interaction_result):void=
        Print ("Guard Event: Alerted")
        # Set status icon to the alerted state (with material animation)
        set StatusIcon.Texture = T_UI_EnemyStatus_Alert
        set StatusIcon.IsScaling = 1.0
        set StatusIcon.IsBouncing = 0.0

    GuardLostTarget(Result: device_ai_interaction_result):void=
        Print ("Guard Event: Lost Target")
        # Set status icon to the lost target state (no material animation)
        set StatusIcon.Texture = T_UI_EnemyStatus_ClosedEye
        set StatusIcon.IsScaling = 0.0
        set StatusIcon.IsBouncing = 0.0        

    GuardSuspicious(InAgent: agent):void=
        Print ("Guard Event: Suspicious")
        # Set status icon to the suspicious state (with material animation)
        set StatusIcon.Texture =  T_UI_EnemyStatus_Eye
        set StatusIcon.IsScaling = 0.0
        set StatusIcon.IsBouncing = 1.0           

    GuardUnaware(InAgent: agent):void=
        Print ("Guard Event: Unaware")
        # Set status icon to the suspicious state (no material animation)
        set StatusIcon.Texture =  T_UI_EnemyStatus_Unaware
        set StatusIcon.IsScaling = 0.0
        set StatusIcon.IsBouncing = 0.0           

    GuardDamaged(Result: device_ai_interaction_result):void=
        Print ("Guard Event: Damaged")
        if (MaybeAgent := Result.Target?):
            if (FortChar := MaybeAgent.GetFortCharacter[]):
                EnemyHealthValue := FortChar.GetHealth()
                
                # Update health radial to show health value
                set RadialMeter.Progress = EnemyHealthValue
                
                # Pulse the portrait outline and change the color
                set PortaitArt.OutlineAlpha = 1.0
                set PortaitArt.OutlineColor1 = MakeColorFromSRGB(1.0, 1.0, 1.0)
                set PortaitArt.OutlineColor2 = MakeColorFromSRGB(1.0, 1.0, 1.0)
                set PortaitArt.OutlineColor3 = MakeColorFromSRGB(1.0, 1.0, 1.0)
                spawn{PulsePortraitOutline()}

    GuardEliminated(Result: device_ai_interaction_result):void=
        Print ("Guard Event: Guard Eliminated")
        set EnemiesEliminated += 1
        set EnemiesLeftInt = Spawner.GetSpawnLimit() - EnemiesEliminated
        set EnemiesLeftFloat = EnemiesLeftInt * 1.0
        GuardsLeft.SetText(SetDynamicText(EnemiesLeftInt))
        
        # Set Spawn Meter to new value
        set CountMeter.Progress = EnemiesLeftFloat

        # Start the intro sequence for a new enemy and spawn the enemy
        if (EnemiesLeftInt > 0):
            set RadialMeter.PositionEnd = 0.0
            set StatusIcon.Texture =  T_UI_EnemyStatus_Unaware
            set StatusIcon.IsScaling = 0.0
            set StatusIcon.IsBouncing = 0.0        
            set IntroEnemyAnimValue = -1.0
            spawn{IntroNewEnemy()}
        else:
            # Show "no more enemies state"
            set SpawnerActive=false
            Spawner.Reset()
            if:
                MaybeAgent := Result.Target?
                InAgent := MaybeAgent.GetFortCharacter[].GetAgent[]
            then:
                OnHide(MaybeAgent)
 
    GuardEliminating(Result: device_ai_interaction_result):void=
        Print ("Guard Event: Eliminating")

    IntroNewEnemy()<suspends>:void=
        loop:
            # Animate the portrait material
            set IntroEnemyAnimValue += 0.02 
            set PortaitArt.TexturePositionY = IntroEnemyAnimValue
 
            # Animate the health bar material
            if (RadialMeter.PositionEnd < 0.5):
                set RadialMeter.PositionEnd += 0.01

            Sleep (0.0) 

            # check if number is greater or equal to zero
            if (IntroEnemyAnimValue >= 0.0):
                # exit loop
                set IntroEnemyAnimValue = 0.0

                # Spawn a new enemy
                if (EnemiesLeftInt > 0):
                    Spawner.Spawn()

                break

    PulsePortraitOutline()<suspends>:void=
        loop:
            # Animate the portrait outline
            set PortaitArt.OutlineAlpha -= 0.05 
            set PortaitArt.Glow += 0.005

            # Check if outline is 0 or less
            if (PortaitArt.OutlineAlpha <= 0.0):
                set PortaitArt.OutlineAlpha = 0.00
                set PortaitArt.Glow = 0.0248

                break
 
            Sleep (0.0)
    
    # Full reset on the spawner anytime the player steps into other volumes or teleports
    ResetGuardStates(InAgent:agent):void=
        if (player[InAgent]):
            if (AgentIsInsideVolume = false):
                OnHide(InAgent)
                Spawner.Despawn()
                Spawner.Reset()
                Spawner.Disable()
                set SpawnerActive = false
                set EnemiesLeftInt = Spawner.GetSpawnLimit()
                HUDController.Disable()

    OnAgentExitsShowVolume(InAgent:agent):void=
        if (player[InAgent]):
            set AgentIsInsideVolume = false
