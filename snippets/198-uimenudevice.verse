# Knowledge-neutral Verse snippet; project names and asset IDs generalized.
# id: snippet-198

# Copyright Epic Games, Inc. All Rights Reserved.

using { /Verse.org/Simulation }
using { /Fortnite.com/Devices }
using { /Fortnite.com/Characters }
using { /Fortnite.com/UI }
using { /UnrealEngine.com/Temporary/UI }
using { /Verse.org/Input }
using { /Verse.org/Input/UI }
using { /UnrealEngine.com/Temporary }

# ============================================================
#  menu_device
#
#  The root of the menu system. Placed once in the level.
#  When the game starts, it opens a menu session
#  for each player currently in the match.
# ============================================================
ui_menu_device := class(creative_device):

    @editable
    Trigger : trigger_device = trigger_device{}

    @editable
    HUDController : hud_controller_device = hud_controller_device{}

    # The camera shown on the main menu screen.
    @editable
    MainMenuCamera : gameplay_camera_fixed_point_device = gameplay_camera_fixed_point_device{}

    # Default camera per vehicle, ordered to match the vehicle list.
    @editable
    VehicleCameras : []gameplay_camera_fixed_point_device = array{}

    # Camera inspect sequences shared by all vehicles
    @editable
    VehicleCamSeqLeft : cinematic_sequence_device = cinematic_sequence_device{}
    @editable
    VehicleCamSeqRight : cinematic_sequence_device = cinematic_sequence_device{}

    # The  sequence that plays when the player hits Play on the main menu
    @editable
    MenuCarSequence : cinematic_sequence_device = cinematic_sequence_device{}

    # Starting currency given to each player
    @editable
    var PlayerCurrency<public> : int = 1000

    var PlayerSessions : [player]menu_session = map{} 

    var Leaderboards : [player]int = map{}

    var PlayerTriggersShowMenuTriggerSubscriptionOption : ?cancelable = false

    # Helper function to sort descending by score
    ComparePlayerScore<private>(A : tuple(player, int), B : tuple(player, int))<computes><decides>:void =
        B(1) < A(1)

    OnBegin<override>()<suspends>:void=
        set PlayerTriggersShowMenuTriggerSubscriptionOption = option {Trigger.TriggeredEvent.Subscribe(OnShowMenuTriggered)}

    OnEnd<override>():void =
        CancelSubscription(PlayerTriggersShowMenuTriggerSubscriptionOption)

    OnShowMenuTriggered(MaybeAgent : ?agent) : void =
        MainMenuCamera.Enable()
        HUDController.Enable()
        for (Cam : VehicleCameras):
            Cam.Enable()
        GetPlayspace().PlayerAddedEvent().Subscribe(OnPlayerAdded)
        GetPlayspace().PlayerRemovedEvent().Subscribe(OnPlayerRemoved)
        for (Player : GetPlayspace().GetPlayers()):
            spawn{ WaitForCharacterAndOpenMenu(Player) }

    OnPlayerAdded(Player : player):void=
        if (AddPlayerToLeaderboards[Player]):
        spawn{ WaitForCharacterAndOpenMenu(Player) }

    GetSortedLeaderboards<public>():[]tuple(player, int) =
        return SortBy(Leaderboards.ToArray(), ComparePlayerScore)

    # When a player leaves the party, drop their menu session
    OnPlayerRemoved(LeavingPlayer : player):void=
        HUDController.Disable()
        var NewSessions : [player]menu_session = map{}
        for (Existing -> Session : PlayerSessions):
            if (Existing <> LeavingPlayer):
                if (set NewSessions[Existing] = Session) {}
        set PlayerSessions = NewSessions

        var NewLeaderboards : [player]int = map{}
        for (Existing -> Score : Leaderboards):
            if (Existing <> LeavingPlayer):
                if (set NewLeaderboards[Existing] = Score) {}
        set Leaderboards = NewLeaderboards

    AddPlayerToLeaderboards(Player : player)<decides><transacts>:void =
        # Skip if already on board
        not Leaderboards[Player]
        set Leaderboards[Player] = 0

    # Ensures a player is spawned before opening the menu
    WaitForCharacterAndOpenMenu(Player : player)<suspends>:void=
        loop:
            if (Player.GetFortCharacter[]):
                break
            Sleep(0.1)
        OpenMenuForPlayer(Player)
        Sleep(0.5)

    # Creates a new session for a player, or re-opens their existing one.
    OpenMenuForPlayer(Player:player):void=
        if (Existing := PlayerSessions[Player]):
            Existing.OpenUI()
            Existing.GoToMainMenu()
        else:
            NewSession := menu_session{ MenuDevice := Self, Player := Player }
            NewSession.Initialize()
            if (set PlayerSessions[Player] = NewSession) {}
            NewSession.GoToMainMenu()


# ============================================================
#  menu_session
#
#  One instance per player. It owns:
#    - The root overlay added to the player's screen
#    - All  page instances for that player
# ============================================================
menu_session := class:
    MenuDevice<public> : ui_menu_device
    Player<public> : player

    var PlayerCurrency<public> : int = 1000

    # Page instances 
    var MainMenuPage<public> : ?ui_main_menu_page = false
    var GameModePage<public> : ?ui_game_mode_page = false
    var UpgradePage<public> : ?ui_upgrade_page = false
    var VehicleSelectPage<public> : ?ui_vehicle_select_page = false 
    var CountdownPage<public> : ?ui_countdown_page= false
    var SettingsPage<public> : ?ui_settings_page = false

    var CurrentPage : ?ui_menu_page = false

    # All pages are children of this overlay.
    RootOverlay : overlay = overlay{}

    var IsUIOpen : logic = false
    var IsPlayerReady : logic = false
    var IsMobile : logic = false
    var IsGamepad : logic = false


    # Creates all page instances, builds their widget trees, and adds them (collapsed) to the root overlay.
    Initialize<public>():void=
        set PlayerCurrency = MenuDevice.PlayerCurrency
        NewMainMenu := ui_main_menu_page { Session := Self }
        NewGameMode := ui_game_mode_page{ Session := Self }
        NewUpgrade := ui_upgrade_page { Session := Self }
        NewVehicleSelect := ui_vehicle_select_page{ Session := Self }
        NewCountdown:= ui_countdown_page{ Session := Self }
        NewSettings:= ui_settings_page{ Session := Self }

        set MainMenuPage = option{ NewMainMenu }
        set GameModePage= option{ NewGameMode }
        set UpgradePage = option{ NewUpgrade }
        set VehicleSelectPage = option{ NewVehicleSelect }
        set CountdownPage= option{ NewCountdown }
        set SettingsPage = option{ NewSettings }

        AddPageToRoot(NewMainMenu)
        AddPageToRoot(NewGameMode)
        AddPageToRoot(NewUpgrade)
        AddPageToRoot(NewVehicleSelect)
        AddPageToRoot(NewCountdown)
        AddPageToRoot(NewSettings)

        OpenUI()

        spawn{DetectInput()}

    DetectInput()<suspends> : void = 
        loop:
            if (IsUIOpen = true):
                if (PlayerInput := GetPlayerInput[Player]):
                    case (PlayerInput.PreferredInputMethod):
                        input_method.Gamepad =>
                            set IsGamepad = true
                            set IsMobile = false
                        input_method.Touch =>
                            set IsMobile = true
                            set IsGamepad = false
                        input_method.KeyboardAndMouse =>
                            set IsMobile = false
                            set IsGamepad = false
                        _=>  
            else: 
                break

            Sleep(0.0)

    OpenUI<public>():void=
        if (not IsUIOpen?):
            if (PlayerUI := GetPlayerUI[Player]):
                PlayerUI.AddWidget(RootOverlay, player_ui_slot{ InputMode := ui_input_mode.All })
                PlayerUI.SetFocus(RootOverlay)
            set IsUIOpen = true

    # Builds a page's widget tree and inserts into the root overlay.
    AddPageToRoot(Page : ui_menu_page) : void =
        Page.InitWidget()
        Slot := overlay_slot:
            Widget := Page.GetPageWidget()
            HorizontalAlignment := horizontal_alignment.Fill
            VerticalAlignment := vertical_alignment.Fill
        RootOverlay.AddWidget(Slot)

    ShowPage<public>(Page : ui_menu_page) : void =
        if (Current := CurrentPage?):
            Current.Hide()
        set CurrentPage = option{ Page }
        Page.Show()

    CloseAll<public>() : void =
        MenuDevice.HUDController.Disable()
        if (Current := CurrentPage?):
            Current.Hide()
        set CurrentPage = false
        if (PlayerUI := GetPlayerUI[Player]):
            PlayerUI.RemoveWidget(RootOverlay)
        set IsUIOpen = false

        # Remove VP mapping
        if (SettingsScreen := SettingsPage?):
            SettingsScreen.ResetVirtualPointer()

    # Navigation methods called by each page's OnSelect / OnBack
    GoToMainMenu<public>():void=if (P := MainMenuPage?) { ShowPage(P) }
    GoToGameMode<public>():void= if (P := GameModePage?) { ShowPage(P) }
    GoToUpgrades<public>():void= if (P := UpgradePage?) { ShowPage(P) }
    GoToVehicleSelect<public>():void= if (P := VehicleSelectPage?){ ShowPage(P) }
    GoToCountdown<public>():void= if (P := CountdownPage?){ ShowPage(P) }
    GoToSettings<public>():void= if (P := SettingsPage?){ ShowPage(P) }
