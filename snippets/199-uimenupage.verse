# Knowledge-neutral Verse snippet; project names and asset IDs generalized.
# id: snippet-199

# Copyright Epic Games, Inc. All Rights Reserved.

using { /Verse.org/Simulation }
using { /Verse.org/Input }
using { /Verse.org/Input/UI }
using { /Fortnite.com/Devices }
using { /Fortnite.com/Characters }
using { /Fortnite.com/UI }
using { /UnrealEngine.com/Temporary/UI }

# ============================================================
#  menu_page (base class)
#
#  Every screen in the menu extends this.
#  It handles things every page needs:
#    - Building and showing/hiding the widget tree
#    - Back and Select button input
#    - Subscribing to and cleaning up widget events
#    - Syncing the title and currency display
#
#  To make a new page, extend menu_page and override
#  whichever methods you need. 
# ============================================================
ui_menu_page := class<abstract>:

    Session<public> : menu_session

    TitleWidget : UserInterfaces.UIFeatureExamples.Menu.Widgets.Basic.UW_Menu_Title = UserInterfaces.UIFeatureExamples.Menu.Widgets.Basic.UW_Menu_Title{}

    CurrencyWidget : UserInterfaces.UIFeatureExamples.Menu.Widgets.Basic.UW_Menu_Currency = UserInterfaces.UIFeatureExamples.Menu.Widgets.Basic.UW_Menu_Currency{}

    BackButtonWidget : UserInterfaces.UIFeatureExamples.Menu.Widgets.Basic.UW_Menu_Button_Input_Back = UserInterfaces.UIFeatureExamples.Menu.Widgets.Basic.UW_Menu_Button_Input_Back{}

    # Verse buttons that capture keyboard and gamepad inputs.
    var BackButton : button = button{
        TriggeringInputAction := option{ UI.Back },
        Slot := button_slot{ Widget := overlay{} }
    }

    var SelectButton : button = button{
        TriggeringInputAction := option{ UI.Inspect },
        Slot := button_slot{ Widget := overlay{} }
    }

    NextTabButton : button = button{
        TriggeringInputAction := option{ UI.NextTab },
        Slot := button_slot{ Widget := overlay{} }
    }

    PrevTabButton : button = button{
        TriggeringInputAction := option{ UI.PreviousTab },
        Slot := button_slot{ Widget := overlay{} }
    }

    NeedNavigationMapping<public> : logic = true

    NeedBackButton<public>: logic = true

    NeedSelectButton<public> : logic = true

    NeedTabButtons<public>: logic = false

    ShowTitle<public> : logic = true

    ShowCurrency<public> : logic = true

    var BackButtonClickedSubscriptionOption : ?cancelable = false
    var SelectButtonClickedSubscriptionOption : ?cancelable = false
    var NextTabSubscriptionOption: ?cancelable = false
    var PrevTabSubscriptionOption: ?cancelable = false

    var PageSubscriptions : []cancelable_event = array{}

    var PageRootWidget : ?overlay = false

    # Return the main content widget for this page.
    GetContentWidget<public>() : widget = overlay{}

    GetFocusWidget<public>() : widget = GetContentWidget()

    GetTitle<public>() : string = ""

    SubscribeToPageEvents<public>() : void = {}

    InitData<public>() : void = {}

    OnShow<public>() : void = {}

    OnHide<public>() : void = {}

    OnSelect<public>()<suspends> : void = {}

    OnBack<public>()<suspends> : void = {}

    OnNextTab<public>() : void = {}
    OnPrevTab<public>() : void = {}

    BackButtonText<localizes> : message = "Back"

    InitWidget<public>() : void =
        # Set up the UMG back button visual and attach it to the verse button
        set BackButton = button{
            TriggeringInputAction := option{ UI.Back },
            Slot := button_slot{ Widget := BackButtonWidget }
        }
        # Select button has no UMG visual, it is an invisible input capture
        set SelectButton = button{
            TriggeringInputAction := option{ UI.Inspect },
            Slot := button_slot{ Widget := overlay{} }
        }

        # Hide the back button visual on pages that don't need it
        if (NeedBackButton = false):
            BackButtonWidget.SetVisibility(widget_visibility.Collapsed)

        # Hide any shared widgets this page doesn't need
        if (ShowTitle = false):
            TitleWidget.SetVisibility(widget_visibility.Collapsed)
        if (ShowCurrency = false):
            CurrencyWidget.SetVisibility(widget_visibility.Collapsed)

        Built := BuildPageWidget()
        set PageRootWidget = option{ Built }
        Built.SetVisibility(widget_visibility.Collapsed)
        InitData()

    GetPageWidget<public>() : overlay =
        if (Widget := PageRootWidget?):
            Widget
        else:
            overlay{}

    Show<public>() : void =
        # Sync shared widgets
        set TitleWidget.ScreenTitle = StringToMessage(GetTitle())
        set CurrencyWidget.CurrencyAmount = Session.PlayerCurrency
        if (W := PageRootWidget?):
            W.SetVisibility(widget_visibility.Visible)

        # Route input focus
        if (PlayerUI := GetPlayerUI[Session.Player]):
            PlayerUI.SetFocus(GetFocusWidget())

        SubscribeToPageEvents()
        
        # Register input mappings for the navigation keybinds
        if (PlayerInput := GetPlayerInput[Session.Player]):
            if (NeedNavigationMapping = true):
                PlayerInput.AddInputMapping(UI.MenuNavigationMapping)
            PlayerInput.AddInputMapping(UI.InventoryMenuMapping)

        BackSub := BackButton.OnClick().Subscribe(OnBackClicked)
        set BackButtonClickedSubscriptionOption = option{ BackSub }
        BackWidgetSub := SubscribeToEvent(BackButtonWidget.OnClick, OnBackClickedFromWidget)
        set PageSubscriptions += array{ BackWidgetSub }

        # Select keybind is subscribed on pages that require it
        if (NeedSelectButton = true):
            SelectSub := SelectButton.OnClick().Subscribe(OnSelectClicked)
            set SelectButtonClickedSubscriptionOption = option{ SelectSub }

        if (NeedTabButtons = true):
            NextSub := NextTabButton.OnClick().Subscribe(OnNextTabClicked)
            PrevSub := PrevTabButton.OnClick().Subscribe(OnPrevTabClicked)
            set NextTabSubscriptionOption = option{ NextSub }
            set PrevTabSubscriptionOption = option{ PrevSub }
        OnShow()

    Hide<public>() : void =
        OnHide()
        # clear every button subscription.
        CancelSubscription(BackButtonClickedSubscriptionOption)
        CancelSubscription(SelectButtonClickedSubscriptionOption)
        CancelSubscription(NextTabSubscriptionOption)
        CancelSubscription(PrevTabSubscriptionOption)
        set BackButtonClickedSubscriptionOption = false
        set SelectButtonClickedSubscriptionOption = false
        set NextTabSubscriptionOption = false
        set PrevTabSubscriptionOption = false

        # Cancel all event listeners.
        for (Sub : PageSubscriptions):
            Sub.Cancel()
        set PageSubscriptions = array{}

        # Remove input mappings
        if (PlayerInput := GetPlayerInput[Session.Player]):
            if (NeedNavigationMapping = true):
                PlayerInput.RemoveInputMapping(UI.MenuNavigationMapping)
            PlayerInput.RemoveInputMapping(UI.InventoryMenuMapping)
            
        if (W := PageRootWidget?):
            W.SetVisibility(widget_visibility.Collapsed)


    OnBackClickedFromWidget(Event : tuple()) : void =
        spawn{ OnBackAsync() }

    OnBackClicked(Message : widget_message) : void =
        spawn{ OnBackAsync() }

    OnBackAsync()<suspends> : void =
        OnBack()

    OnSelectClickedFromWidget(Event : tuple()) : void =
        spawn{ OnSelectAsync() }

    OnSelectClicked(Message : widget_message) : void =
        spawn{ OnSelectAsync() }

    OnSelectAsync()<suspends> : void =
        OnSelect()

    OnNextTabClicked(Message : widget_message) : void =
        OnNextTab()

    OnPrevTabClicked(Message : widget_message) : void =
        OnPrevTab()

    # ----------------------------------------------------------
    # BuildPageWidget
    #
    # Builds the full layout for this page:
    #   Layer 1 — Page content
    #   Layer 2 — Title (top left)
    #   Layer 3 — Currency (top right)
    #   Layer 4 — Button area (bottom right)
    #             Back button is visible here. Select and tab buttons
    #             are invisible as they are keybind capture only.
    # ----------------------------------------------------------
    BuildPageWidget() : overlay =
        Content := GetContentWidget()

        NewOverlay : overlay = overlay{}

        # Layer 1
        ContentSlot := overlay_slot:
            Widget := Content
            HorizontalAlignment := horizontal_alignment.Fill
            VerticalAlignment := vertical_alignment.Fill

        # Layer 2
        TitleSlot := overlay_slot:
            Widget := TitleWidget
            HorizontalAlignment := horizontal_alignment.Left
            VerticalAlignment := vertical_alignment.Top

        # Layer 3
        CurrencySlot := overlay_slot:
            Widget := CurrencyWidget
            HorizontalAlignment := horizontal_alignment.Right
            VerticalAlignment := vertical_alignment.Top

        # Layer 4
        ButtonAreaOverlay : overlay = overlay{}

        # Select button
        HiddenSelectSlot := overlay_slot:
            Widget := SelectButton
            Padding := margin{ Left := -9999.0 }
        ButtonAreaOverlay.AddWidget(HiddenSelectSlot)

        # Tab buttons
        if (NeedTabButtons = true):
            NextTabSlot := overlay_slot:
                Widget := NextTabButton
            PrevTabSlot := overlay_slot:
                Widget := PrevTabButton
            ButtonAreaOverlay.AddWidget(NextTabSlot)
            ButtonAreaOverlay.AddWidget(PrevTabSlot)

        # Visible button bar
        ButtonBar : stack_box = stack_box{ Orientation := orientation.Horizontal }

        BackButtonBarSlot := stack_box_slot:
            Widget := BackButton
        ButtonBar.AddWidget(BackButtonBarSlot)

        VisibleBarSlot := overlay_slot:
            Widget := ButtonBar
            HorizontalAlignment := horizontal_alignment.Right
            VerticalAlignment := vertical_alignment.Bottom
        ButtonAreaOverlay.AddWidget(VisibleBarSlot)

        ButtonAreaSlot := overlay_slot:
            Widget := ButtonAreaOverlay
            HorizontalAlignment := horizontal_alignment.Right
            VerticalAlignment := vertical_alignment.Bottom
            Padding := margin{ Right := 64.0, Bottom := 24.0 }

        NewOverlay.AddWidget(ContentSlot)
        NewOverlay.AddWidget(TitleSlot)
        NewOverlay.AddWidget(CurrencySlot)
        if (NeedBackButton = true):
            NewOverlay.AddWidget(ButtonAreaSlot)

        return NewOverlay
