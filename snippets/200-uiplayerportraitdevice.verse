# Knowledge-neutral Verse snippet; project names and asset IDs generalized.
# id: snippet-200

using { /UnrealEngine.com/Temporary/SpatialMath }
using { /Fortnite.com/Devices }
using { /Verse.org/Simulation }
using { /UnrealEngine.com/Temporary/Diagnostics }
using { /Fortnite.com/Vehicles }
using { /Fortnite.com/Characters }
using { /UnrealEngine.com/Temporary/UI }
using { /Verse.org/SceneGraph }

ui_player_portrait_device := class(creative_device):

    @editable
    VolumeDevice : volume_device = volume_device{}

    @editable
    HUDControllerDevice : hud_controller_device = hud_controller_device {}

    # Main UI Widget
    var MyUI : overlay = overlay {}

    var AnimationInProgress : logic = false

    # A map to store each player's UI widget reference so it can be removed later
    var MaybeMyUIPerPlayer : [player]?overlay = map{}   
    
    var PortraitsWidget : UserInterfaces.VerseUIExamples.PlayerPortrait.Widgets.UW_PlayerPortrait_Full = UserInterfaces.VerseUIExamples.PlayerPortrait.Widgets.UW_PlayerPortrait_Full {}

    # Runs when the device is started in a running game
    OnBegin<override>()<suspends>:void=
        set MyUI = overlay {}
        # Subscribe to trigger events
        VolumeDevice.AgentEntersEvent.Subscribe(Initialize)


    # Purely material-based example HUD
    Initialize(InAgent:agent):void =

        HUDControllerDevice.Enable()

        if (AnimationInProgress = false):     
            set AnimationInProgress = true   
            # Get the Player's Icon Componenet (Player->PlayerEntity->IconComponent)        
            if:
                Player:=player[InAgent]
                PlayerEntity:=Player.GetFortCharacter[].GetEntity[]
                PlayerUI := GetPlayerUI[Player]
                IconComponent:=PlayerEntity.GetComponent[icon_component]
            then:
                set PortraitsWidget.PortraitTextureBlue3 = IconComponent.Icon
                set PortraitsWidget.NicknameBlue3 = GetUsername(InAgent)
            
            # Create UI
            MyHudUI := overlay:
                Slots := array:
                    overlay_slot:
                        VerticalAlignment := vertical_alignment.Fill
                        HorizontalAlignment := horizontal_alignment.Fill                                       
                        Widget := PortraitsWidget

            set PortraitsWidget.NicknameBlue1 = StringToMessage("Teammate1")
            set PortraitsWidget.NicknameBlue2 = StringToMessage("Teammate2")
            set PortraitsWidget.NicknameRed1 = StringToMessage("Opponent1")
            set PortraitsWidget.NicknameRed2 = StringToMessage("Opponent2")
            set PortraitsWidget.NicknameRed3 = StringToMessage("Opponent3")

            set MyUI = MyHudUI
            RemoveUI(InAgent)
            AssignUI(InAgent)            

    # Function to add the UI to a specific player
    AssignUI(InAgent: agent): void =
        # Use an if expression to safely unpack the optional agent
        if (Player := player[InAgent]):
            # Get the player's UI system
            if (PlayerUI := GetPlayerUI[Player]):
                # Add the widget and store a reference to it in the map
                PlayerUI.AddWidget(MyUI)
                spawn{PlayAnimations(InAgent)}
                # Store it in the map for later removal
                if (set MaybeMyUIPerPlayer[Player] = option{MyUI}) {}
        # If MaybeAgent was None, the code within the if block is skipped

    PlayAnimations(InAgent : agent)<suspends>:void =
        loop:
            set PortraitsWidget.ShowMiniPortrait1 = false
            set PortraitsWidget.ShowMiniPortrait2 = false
            set PortraitsWidget.ShowMiniPortrait3 = false
            set PortraitsWidget.PlayIntroAnim = true
            if (PortraitsWidget.PlayIntroAnim = true):
                Sleep(0.7)
                set PortraitsWidget.PlayIntroAnim = false
                set PortraitsWidget.PlayPortrait1Anim = true
                set PortraitsWidget.ShowPortrait1 = true
                Sleep(0.55)
                set PortraitsWidget.PlayPortrait1Anim = false
                set PortraitsWidget.PlayPortrait2Anim = true
                set PortraitsWidget.ShowPortrait2 = true
                Sleep(0.55)
                set PortraitsWidget.PlayPortrait2Anim = false
                set PortraitsWidget.PlayPortrait3Anim = true
                set PortraitsWidget.ShowPortrait3 = true
                Sleep(3.0)
                set PortraitsWidget.PlayOutroAnim = true
                set PortraitsWidget.PlayOutroPortraitsAnim = true
                Sleep(0.35)
                set PortraitsWidget.PlayPortrait3Anim = false
                set PortraitsWidget.PlayOutroAnim = false
                set PortraitsWidget.ShowPortrait1 = false
                set PortraitsWidget.ShowPortrait2 = false
                set PortraitsWidget.ShowPortrait3 = false
                set PortraitsWidget.PlayOutroPortraitsAnim = false
                set PortraitsWidget.PlayMiniPortrait1Anim = true
                set PortraitsWidget.ShowMiniPortrait1 = true
                Sleep(0.5)
                set PortraitsWidget.PlayMiniPortrait1Anim = false
                set PortraitsWidget.PlayMiniPortrait2Anim = true
                set PortraitsWidget.ShowMiniPortrait2 = true
                Sleep(0.5)
                set PortraitsWidget.PlayMiniPortrait2Anim = false
                set PortraitsWidget.PlayMiniPortrait3Anim = true
                set PortraitsWidget.ShowMiniPortrait3 = true
                Sleep(0.5)
                set PortraitsWidget.PlayMiniPortrait3Anim = false
                set PortraitsWidget.PlayCurrentPlayerAnim = true
                Sleep(0.3)
                set PortraitsWidget.PlayCurrentPlayerAnim = false
                set AnimationInProgress = false
                HUDControllerDevice.Disable()
                Sleep(6.0)
                RemoveUI(InAgent)
                break
        
    RemoveUI(Agent:agent): void =
        if (IfPlayer := player[Agent]):
            if (PlayerUI := GetPlayerUI[IfPlayer]):
                if (SomeUI := MaybeMyUIPerPlayer[IfPlayer]?):
                    PlayerUI.RemoveWidget(SomeUI)
