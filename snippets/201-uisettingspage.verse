# Knowledge-neutral Verse snippet; project names and asset IDs generalized.
# id: snippet-201

# Copyright Epic Games, Inc. All Rights Reserved.

using { /Verse.org/Simulation }
using { /Fortnite.com/Devices }
using { /Fortnite.com/UI }
using { /UnrealEngine.com/Temporary/UI }
using { /Verse.org/Input }
using { /Verse.org/Input/UI }

# ============================================================
#  settings_page
#
#  Contains the virtual pointer toggle
# ============================================================
ui_settings_page := class(ui_menu_page):

    SettingsWidget : UserInterfaces.UIFeatureExamples.Menu.Widgets.Settings.UW_Menu_Settings = UserInterfaces.UIFeatureExamples.Menu.Widgets.Settings.UW_Menu_Settings{}

    var ViewPointerEnabled : logic = false

    GetTitle<override>():string = "Settings"

    GetContentWidget<override>():widget = SettingsWidget

    ShowCurrency<override> : logic = false

    InitData<override>():void=
        set SettingsWidget.ViewPointerEnabled = ViewPointerEnabled

    SubscribeToPageEvents<override>():void=
        Sub := SubscribeToEvent(SettingsWidget.VirtualPointerEnabledChanged, OnVirtualPointerEnabledChanged)
        set PageSubscriptions += array{ Sub }

    OnShow<override>():void=
        if (ViewPointerEnabled?):
            EnableVP()
        else:
            DisableVP()
        set SettingsWidget.ViewPointerEnabled = ViewPointerEnabled

    ResetVirtualPointer<public>() : void =
        set ViewPointerEnabled = false
        DisableVP()
        set SettingsWidget.ViewPointerEnabled = false

    OnVirtualPointerEnabledChanged(Event : tuple()):void=
        if (ViewPointerEnabled?):
            set ViewPointerEnabled = false
            DisableVP()
        else:
            set ViewPointerEnabled = true
            EnableVP()
        set SettingsWidget.ViewPointerEnabled = ViewPointerEnabled

    OnBack<override>()<suspends>:void=
        Session.GoToMainMenu()

    # Enable/Disable Virtual Pointer by adding/removing "TouchMapping"
    EnableVP() : void =
        if (PlayerInput := GetPlayerInput[Session.Player]):
            PlayerInput.AddInputMapping(TouchMapping)
    DisableVP() : void =
        if (PlayerInput := GetPlayerInput[Session.Player]):
            PlayerInput.RemoveInputMapping(TouchMapping)
