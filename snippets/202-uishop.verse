# Knowledge-neutral Verse snippet; project names and asset IDs generalized.
# id: snippet-202

using { /Fortnite.com/UI }
using { /UnrealEngine.com/Temporary/UI }
using { /UnrealEngine.com/Temporary/SpatialMath }
using { /Fortnite.com/Marketplace }
using { /Verse.org/Simulation }

ui_shop := class():
    
    CloseButtonText<localizes><public> : message = "Close"

    CloseButton : button_quiet = button_quiet {}

    var Player : player

    var OverlayRoot : overlay = overlay {}

    var OfferDetailsOverlayRoot : overlay = overlay {}

    var OffersStackBox : stack_box = stack_box {Orientation := orientation.Vertical}    

    BackgroundWidget : UserInterfaces.UIFeatureExamples.InIslandTransactions.Demo01.Widgets.UW_Shop01_Background = 
        UserInterfaces.UIFeatureExamples.InIslandTransactions.Demo01.Widgets.UW_Shop01_Background {}

    HeaderWidget : UserInterfaces.UIFeatureExamples.InIslandTransactions.Demo01.Widgets.UW_Shop01_Header = 
        UserInterfaces.UIFeatureExamples.InIslandTransactions.Demo01.Widgets.UW_Shop01_Header {}

    ShopContentBackgroundShadowMaterial : UserInterfaces.UIFeatureExamples.InIslandTransactions.Demo01.MaterialInstances.MI_UI_Shop_ContentBackground_Shadow =
        UserInterfaces.UIFeatureExamples.InIslandTransactions.Demo01.MaterialInstances.MI_UI_Shop_ContentBackground_Shadow {}

    ShopContentBackgroundMaterial : UserInterfaces.UIFeatureExamples.InIslandTransactions.Demo01.MaterialInstances.MI_UI_Shop_ContentBackground =
        UserInterfaces.UIFeatureExamples.InIslandTransactions.Demo01.MaterialInstances.MI_UI_Shop_ContentBackground {}

    ShopContentBackgroundPatternMaterial : UserInterfaces.UIFeatureExamples.InIslandTransactions.Demo01.MaterialInstances.MI_UI_WoodenPattern =
        UserInterfaces.UIFeatureExamples.InIslandTransactions.Demo01.MaterialInstances.MI_UI_WoodenPattern {}

    ShopContentBackgroundPatternShadowMaterial : UserInterfaces.UIFeatureExamples.InIslandTransactions.Demo01.MaterialInstances.MI_UI_WoodenPatternLight =
        UserInterfaces.UIFeatureExamples.InIslandTransactions.Demo01.MaterialInstances.MI_UI_WoodenPatternLight {}

    LineMaterial : UserInterfaces.UIFeatureExamples.InIslandTransactions.Demo01.MaterialInstances.MI_UI_Line =
        UserInterfaces.UIFeatureExamples.InIslandTransactions.Demo01.MaterialInstances.MI_UI_Line {}

    OfferDetailsWidget : UserInterfaces.UIFeatureExamples.InIslandTransactions.Demo01.Widgets.UW_Shop01_Item_Offer_Details =
        UserInterfaces.UIFeatureExamples.InIslandTransactions.Demo01.Widgets.UW_Shop01_Item_Offer_Details {}    

    BuyButton : button_loud = button_loud {}

    BuyButtonText<localizes><public> : message = "Buy"

    ShowUI(InOffers:[]offer) : void =
        PopulateOffers(InOffers)
        CreateUI()        

        if (PlayerUI := GetPlayerUI[Player]):
            PlayerUI.AddWidget(OverlayRoot, player_ui_slot {InputMode := ui_input_mode.All})

    HandleCloseButtonClicked(Message : widget_message) : void =         
        set Player = Message.Player
        if (PlayerUI := GetPlayerUI[Player]):
            HideUI(PlayerUI)

    HideUI(PlayerUI : player_ui) : void = 
        PlayerUI.RemoveWidget(OverlayRoot)

    CreateUI() : void =

        CloseButton.SetText(CloseButtonText)
        BuyButton.SetText(BuyButtonText)

        CloseButton.OnClick().Subscribe(HandleCloseButtonClicked)

        set OfferDetailsOverlayRoot = overlay:
            Slots := array:
                overlay_slot:
                    Widget := OfferDetailsWidget
                    VerticalAlignment := vertical_alignment.Top                        
                overlay_slot:
                    Widget := overlay:                        
                        Slots := array:
                            overlay_slot:
                                Widget := material_block:
                                    DefaultImage := LineMaterial
                                    DefaultDesiredSize := vector2 {X := 320.0, Y := 0.0}
                            overlay_slot:
                                Widget := BuyButton                                
                                HorizontalAlignment := horizontal_alignment.Fill
                                Padding := margin {Left := 10.0}                            
                    Padding := margin {Bottom := 50.0}
                    VerticalAlignment := vertical_alignment.Bottom

        set OverlayRoot = overlay:
            Slots := array:
                overlay_slot:
                    Widget := BackgroundWidget                       
                    VerticalAlignment := vertical_alignment.Fill
                    HorizontalAlignment := horizontal_alignment.Fill                    
                overlay_slot:
                    Widget := CloseButton
                    Padding := margin {Left := 824.0, Bottom := 30.0}
                    VerticalAlignment := vertical_alignment.Bottom
                    HorizontalAlignment := horizontal_alignment.Center
                overlay_slot:
                    Widget := overlay:
                        Slots := array:
                            overlay_slot:
                                Widget := material_block:
                                    DefaultImage := ShopContentBackgroundShadowMaterial
                                    DefaultDesiredSize := vector2 {X := 1438.0 , Y := 784.0}
                                VerticalAlignment := vertical_alignment.Bottom
                                HorizontalAlignment := horizontal_alignment.Fill
                                Padding := margin {Bottom := -20.0}
                            overlay_slot:
                                Widget := material_block:
                                    DefaultImage := ShopContentBackgroundMaterial
                                    DefaultDesiredSize := vector2 {X := 1438.0 , Y := 784.0}
                                VerticalAlignment := vertical_alignment.Fill
                                HorizontalAlignment := horizontal_alignment.Fill
                            overlay_slot:
                                Widget := material_block:
                                    DefaultImage := ShopContentBackgroundPatternMaterial
                                    DefaultDesiredSize := vector2 {X := 32.0 , Y := 32.0}
                                VerticalAlignment := vertical_alignment.Fill
                                HorizontalAlignment := horizontal_alignment.Fill
                                Padding := margin {Bottom := 2.0, Top := 2.0, Left := 2.0, Right := 2.0}
                            overlay_slot:
                                Widget := material_block:
                                    DefaultImage := ShopContentBackgroundPatternShadowMaterial
                                    DefaultDesiredSize := vector2 {X := 32.0 , Y := 32.0}
                                VerticalAlignment := vertical_alignment.Fill
                                HorizontalAlignment := horizontal_alignment.Fill
                                Padding := margin {Bottom := 2.0, Top := 2.0, Left := 2.0, Right := 2.0}                                                           
                            overlay_slot:
                                Widget := stack_box:
                                    Orientation := orientation.Horizontal
                                    Slots := array:
                                        stack_box_slot:
                                            Widget := stack_box:
                                                Orientation := orientation.Vertical
                                                Slots := array:
                                                    stack_box_slot:
                                                        Widget := HeaderWidget
                                                        Padding := margin {Bottom := 14.0}
                                                        VerticalAlignment := vertical_alignment.Fill
                                                        HorizontalAlignment := horizontal_alignment.Fill
                                                    stack_box_slot:
                                                        Widget := OffersStackBox
                                                        VerticalAlignment := vertical_alignment.Fill
                                                        HorizontalAlignment := horizontal_alignment.Fill
                                            VerticalAlignment := vertical_alignment.Fill
                                            HorizontalAlignment := horizontal_alignment.Fill
                                            Padding := margin {Left := 43.0, Top := 42.0}
                                        stack_box_slot:
                                            Widget := OfferDetailsOverlayRoot
                                            Padding := margin {Top := -12.0, Right := 25.0}
                                            VerticalAlignment := vertical_alignment.Top
                                            HorizontalAlignment := horizontal_alignment.Right
                                            Distribution := option{1.0}
                                VerticalAlignment := vertical_alignment.Fill
                                HorizontalAlignment := horizontal_alignment.Fill
                    Padding := margin {Left := 200.0, Bottom := 144.0}
                    VerticalAlignment := vertical_alignment.Bottom
                    HorizontalAlignment := horizontal_alignment.Center
                                
    PopulateOffers(InOffers:[]offer) : void =
        
        for (Offer : InOffers):

            var OfferItemWidget : ui_shop_offer_item = ui_shop_offer_item {}

            OfferItemWidget.UpdateData(Offer)

            StackBoxSlot := stack_box_slot:
                Widget := OfferItemWidget.CreateLayout()
            
            OffersStackBox.AddWidget(StackBoxSlot)            
            
            spawn{HandleViewButtonClicked(OfferItemWidget, Offer)}
            
        if (FirstOffer := InOffers[0]):
            UpdateSelectedOfferDetails(FirstOffer)

    HandleViewButtonClicked(OfferWidget : ui_shop_offer_item, SelectedOffer : offer)<suspends> : void = 
        loop:
            OfferWidget.ViewButton.OnClick().Await()
            
            UpdateSelectedOfferDetails(SelectedOffer)
            
    UpdateSelectedOfferDetails(SelectedOffer : offer) : void =

        var PriceFloat : float = 0.0

            if (PriceVB := price_vbucks[SelectedOffer.Price]):
                set PriceFloat = GetPriceVBucks(PriceVB)

        set OfferDetailsWidget.OfferName = SelectedOffer.Name
        set OfferDetailsWidget.OfferIcon = SelectedOffer.Icon
        set OfferDetailsWidget.OfferShortDescription = SelectedOffer.ShortDescription
        set OfferDetailsWidget.OfferPrice = PriceFloat

        spawn {HandleBuyButtonClicked(SelectedOffer)}

    HandleBuyButtonClicked(SelectedOffer : offer) <suspends> : void =
        
        BuyButton.OnClick().Await()
        
        if (PlayerUI := GetPlayerUI[Player]):
            HideUI(PlayerUI)
        
        BuyOffer(Player, SelectedOffer)
