# Knowledge-neutral Verse snippet; project names and asset IDs generalized.
# id: snippet-203

using { /Fortnite.com/Devices }
using { /Verse.org/Simulation }
using { /UnrealEngine.com/Temporary/UI }
using { /UnrealEngine.com/Temporary/SpatialMath }
using { /Verse.org/Assets }
using { /Fortnite.com/UI }

ui_shop02 := class():

    var Player : player

    BackButtonText<localizes> : message = "Back"    

    var OverlayRoot : overlay = overlay {}    

    var ShopWidget : UserInterfaces.UIFeatureExamples.InIslandTransactions.Demo02.Widgets.UW_Shop02_Background =
        UserInterfaces.UIFeatureExamples.InIslandTransactions.Demo02.Widgets.UW_Shop02_Background {}

    var Offers : offers_data = offers_data {}    

    var FirstOfferItem : ui_shop02_offer_item = ui_shop02_offer_item {}
    var SecondOfferItem : ui_shop02_offer_item = ui_shop02_offer_item {}
    var ThirdOfferItem : ui_shop02_offer_item = ui_shop02_offer_item {}
    var FourthOfferItem : ui_shop02_offer_item = ui_shop02_offer_item {}
    var FifthOfferItem : ui_shop02_offer_item = ui_shop02_offer_item {}
    var SixthOfferItem : ui_shop02_offer_item = ui_shop02_offer_item {}

    var ItemOffers : []ui_shop02_offer_item = array{}

    var SelectedOffer : ui_shop02_offer_item = ui_shop02_offer_item {}    

    var BuyButton : ui_shop02_buy_button = ui_shop02_buy_button {}

    var BackButton : button_quiet = button_quiet {}

    ShowUI() : void =        
        CreateUI()        

        if (PlayerUI := GetPlayerUI[Player]):
            PlayerUI.AddWidget(OverlayRoot, player_ui_slot {InputMode := ui_input_mode.All})

    CreateUI() : void =       
            
        BackButton.SetText(BackButtonText)

        set ItemOffers = array{FirstOfferItem} + array{SecondOfferItem} + array{ThirdOfferItem} + array{FourthOfferItem} + array{FifthOfferItem} + array{SixthOfferItem}
            
        SetOffersFromFakeData()            

            OffersStackBox : stack_box = stack_box:
                Orientation := orientation.Vertical
                Slots := array:
                    stack_box_slot:
                        Widget := stack_box:
                            Orientation := orientation.Horizontal
                            Slots := array:
                                stack_box_slot:
                                    Widget := FirstOfferItem.Build()
                                stack_box_slot:
                                    Widget := SecondOfferItem.Build()
                                stack_box_slot:
                                    Widget := ThirdOfferItem.Build()
                    stack_box_slot:
                        Widget := stack_box:
                            Orientation := orientation.Horizontal
                            Slots := array:
                                stack_box_slot:
                                    Widget := FourthOfferItem.Build()
                                stack_box_slot:
                                    Widget := FifthOfferItem.Build()
                                stack_box_slot:
                                    Widget := SixthOfferItem.Build()
            
        FirstOfferItem.SelectItem()

        set SelectedOffer = FirstOfferItem

        UpdateSelectedOfferInfo()

            for (Index := 0..ItemOffers.Length-1):
                if (Item := ItemOffers[Index]):                    
                    spawn{HandleViewButtonClicked(Item)}

            var OverlayContent : overlay = overlay :
                Slots := array:
                    overlay_slot:
                        Widget := ShopWidget
                        VerticalAlignment := vertical_alignment.Fill
                        HorizontalAlignment := horizontal_alignment.Fill
                    overlay_slot:
                        Widget := OffersStackBox
                        VerticalAlignment := vertical_alignment.Center
                        HorizontalAlignment := horizontal_alignment.Center
                        Padding := margin {Left := 15.0, Top := 52.0}
                    overlay_slot:
                        Widget := BuyButton.Build()
                        VerticalAlignment := vertical_alignment.Bottom
                        HorizontalAlignment := horizontal_alignment.Center   
                        Padding := margin {Bottom := 156.0, Left := 725.0}
                    overlay_slot:
                        Widget := BackButton    
                        VerticalAlignment := vertical_alignment.Bottom
                        HorizontalAlignment := horizontal_alignment.Right
                        Padding := margin {Right := 74.0, Bottom := 32.0}
            
        BackButton.OnClick().Subscribe(HandleBackButtonClicked)

        BuyButton.OnClick().Subscribe(HandleBuyButtonClicked)           

        set OverlayRoot = OverlayContent        

    UpdateSelectedOfferInfo() : void =
        set ShopWidget.SelectedOfferName = SelectedOffer.OfferItemWidget.OfferName
        set ShopWidget.SelectedOfferDescription = SelectedOffer.OfferItemWidget.OfferDescription
        set ShopWidget.SelectedOfferShortDescription = SelectedOffer.OfferItemWidget.OfferShortDescription
        set ShopWidget.SelectedOfferIcon = SelectedOffer.OfferItemWidget.OfferIcon
        set ShopWidget.SelectedOfferPrice = SelectedOffer.OfferItemWidget.OfferPrice
        set ShopWidget.SelectedOfferCount = SelectedOffer.OfferItemWidget.OfferCount

    HandleViewButtonClicked(OfferWidget : ui_shop02_offer_item)<suspends> : void = 
        loop:
            OfferWidget.HitTestButton.OnClick().Await()
            
            SelectedOffer.UnselectItem()

            OfferWidget.SelectItem()

            set SelectedOffer = OfferWidget
            
            UpdateSelectedOfferInfo()
    
    HandleBackButtonClicked(Message : widget_message) : void =
        InPlayer := Message.Player
        if (PlayerUI := GetPlayerUI[InPlayer]):
            PlayerUI.RemoveWidget(OverlayRoot)

    HandleBuyButtonClicked(Message : widget_message) : void =
        Print ("Item was bought")    

    SetOffersFromFakeData() : void =        

        FirstOfferItem.SetData(Offers.FirstOfferName, Offers.FirstOfferDescription, Offers.FirstOfferShortDescription,
            Offers.FirstOfferIcon, Offers.FirstOfferPrice, Offers.FirstOfferCount, Offers.FirstOfferRarity, Offers.FirstOfferRarityTextColor)

        SecondOfferItem.SetData(Offers.SecondOfferName, Offers.SecondOfferDescription, Offers.SecondOfferShortDescription,
            Offers.SecondOfferIcon, Offers.SecondOfferPrice, Offers.SecondOfferCount, Offers.SecondOfferRarity, Offers.SecondOfferRarityTextColor)

        ThirdOfferItem.SetData(Offers.ThirdOfferName, Offers.ThirdOfferDescription, Offers.ThirdOfferShortDescription,
            Offers.ThirdOfferIcon, Offers.ThirdOfferPrice, Offers.ThirdOfferCount, Offers.ThirdOfferRarity, Offers.ThirdOfferRarityTextColor)

        FourthOfferItem.SetData(Offers.FourthOfferName, Offers.FourthOfferDescription, Offers.FourthOfferShortDescription,
            Offers.FourthOfferIcon, Offers.FourthOfferPrice, Offers.FourthOfferCount, Offers.FourthOfferRarity, Offers.FourthOfferRarityTextColor)

        FifthOfferItem.SetData(Offers.FifthOfferName, Offers.FifthOfferDescription, Offers.FifthOfferShortDescription,
            Offers.FifthOfferIcon, Offers.FifthOfferPrice, Offers.FifthOfferCount, Offers.FifthOfferRarity, Offers.FifthfferRarityTextColor)

        SixthOfferItem.SetData(Offers.SixthOfferName, Offers.SixthOfferDescription, Offers.SixthOfferShortDescription,
            Offers.SixthOfferIcon, Offers.SixthOfferPrice, Offers.SixthOfferCount, Offers.SixthOfferRarity, Offers.SixthOfferRarityTextColor)
