# Knowledge-neutral Verse snippet; project names and asset IDs generalized.
# id: snippet-204

using { /UnrealEngine.com/Temporary/UI }
using { /UnrealEngine.com/Temporary/SpatialMath }

ui_shop02_buy_button := class:

    var RootOverlay : overlay = overlay {}

    var ButtonWidget : UserInterfaces.UIFeatureExamples.InIslandTransactions.Demo02.Widgets.UW_Shop02_Button_Buy =
        UserInterfaces.UIFeatureExamples.InIslandTransactions.Demo02.Widgets.UW_Shop02_Button_Buy {}

    var HitTestButton : button = button { 
        Slot := button_slot { 
            Widget := texture_block {DefaultImage := UserInterfaces.UIFeatureExamples.InIslandTransactions.Demo02.Textures.T_UI_Blank}
            VerticalAlignment := vertical_alignment.Fill
            HorizontalAlignment := horizontal_alignment.Fill
            }
        }

    Build() : overlay =   
        
        HitTestButton.HighlightEvent().Subscribe(OnHighlight)
        HitTestButton.UnhighlightEvent().Subscribe(OnUnhighlight)

        var ButtonOverlayRoot : overlay = overlay:            
            Slots := array:
                overlay_slot:
                    Widget := texture_block:
                        DefaultImage := UserInterfaces.UIFeatureExamples.InIslandTransactions.Demo02.Textures.T_UI_Blank
                        DefaultDesiredSize := vector2 {X := 245.0, Y := 76.0}
                    VerticalAlignment := vertical_alignment.Fill
                    HorizontalAlignment := horizontal_alignment.Fill    
                overlay_slot:
                    Widget := ButtonWidget
                    VerticalAlignment := vertical_alignment.Fill
                    HorizontalAlignment := horizontal_alignment.Fill                               
                overlay_slot:
                    Widget := HitTestButton
                    VerticalAlignment := vertical_alignment.Fill
                    HorizontalAlignment := horizontal_alignment.Fill

        return ButtonOverlayRoot

    OnClick<public>():listenable(widget_message) =
        HitTestButton.OnClick()

    OnHighlight(_m:widget_message) : void =
        set ButtonWidget.Hovered = true

    OnUnhighlight(_m:widget_message) : void =
        set ButtonWidget.Hovered = false
