# Knowledge-neutral Verse snippet; project names and asset IDs generalized.
# id: snippet-205

using { /Verse.org/Assets }
using { /UnrealEngine.com/Temporary/UI }
using { /UnrealEngine.com/Temporary/SpatialMath }
using { /Fortnite.com/Marketplace }
using { /Verse.org/Colors }

ui_shop02_offer_item := class:

    var RootOverlay : overlay = overlay {}    

    var OfferItemWidget : UserInterfaces.UIFeatureExamples.InIslandTransactions.Demo02.Widgets.UW_Shop02_Item_Offer =
        UserInterfaces.UIFeatureExamples.InIslandTransactions.Demo02.Widgets.UW_Shop02_Item_Offer {}    

    var HitTestButton : button = button { 
        Slot := button_slot { 
            Widget := texture_block {DefaultImage := UserInterfaces.UIFeatureExamples.InIslandTransactions.Demo02.Textures.T_UI_Blank}
            VerticalAlignment := vertical_alignment.Fill
            HorizontalAlignment := horizontal_alignment.Fill
            }
        }

    Build() : overlay =

        set OfferItemWidget.Selected = false        
        
        HitTestButton.HighlightEvent().Subscribe(OnHighlight)
        HitTestButton.UnhighlightEvent().Subscribe(OnUnhighlight)

        var ItemOverlayRoot : overlay = overlay:            
            Slots := array:
                overlay_slot:
                    Widget := OfferItemWidget
                overlay_slot:
                    Widget := HitTestButton
                    VerticalAlignment := vertical_alignment.Fill
                    HorizontalAlignment := horizontal_alignment.Fill

        return ItemOverlayRoot

    SetData(InOfferName : message, InOfferDescription: message, InOfferShortDescription: message, InOfferIcon : texture,
     InOfferPrice : float, MaxCount : int, Rarity : material, RarityTextColor : color) : void =

        set OfferItemWidget.OfferName = InOfferName
        set OfferItemWidget.OfferDescription = InOfferDescription
        set OfferItemWidget.OfferShortDescription = InOfferShortDescription
        set OfferItemWidget.OfferIcon = InOfferIcon       
        set OfferItemWidget.OfferCount = MaxCount
        set OfferItemWidget.OfferPrice = InOfferPrice
        set OfferItemWidget.Rarity = Rarity  
        set OfferItemWidget.RarityTextColor = RarityTextColor    

    UnselectItem() : void =        
        set OfferItemWidget.Selected = false     
        
    SelectItem() : void =
        set OfferItemWidget.Selected = true

    OnHighlight(_m:widget_message):void =        
        set OfferItemWidget.Hovered = true

    OnUnhighlight(_m:widget_message):void =        
        set OfferItemWidget.Hovered = false
