# Knowledge-neutral Verse snippet; project names and asset IDs generalized.
# id: snippet-206

using { /Verse.org/Assets }
using { /Verse.org/Colors }

offers_data := class:  

    FirstOfferName<localizes> : message = "Soda Blurst"
    FirstOfferDescription<localizes> : message = "A mostly food-like take on a classic soda flavor. It's blursting with flavor blursts."
    FirstOfferShortDescription<localizes> : message = "First Offer Short Description"
    FirstOfferIcon : texture = UserInterfaces.UIFeatureExamples.InIslandTransactions.Demo02.Textures.T_UI_Item01
    FirstOfferPrice : float = 420.0
    FirstOfferCount : int = 17
    FirstOfferRarity : material = UserInterfaces.UIFeatureExamples.InIslandTransactions.Demo02.MaterialInstances.MI_UI_ItemRare {}
    FirstOfferRarityTextColor : color = color {R := 0.4666, G := 0.2235, B := 0.5529}

    SecondOfferName<localizes> : message = "Noodle Nod"
    SecondOfferDescription<localizes> : message = "A steamy cup of noodle-shaped polymers to eat, just like real food!"
    SecondOfferShortDescription<localizes> : message = "Second Offer Short Description"
    SecondOfferIcon : texture = UserInterfaces.UIFeatureExamples.InIslandTransactions.Demo02.Textures.T_UI_Item02
    SecondOfferPrice : float = 630.0
    SecondOfferCount : int = 9
    SecondOfferRarity : material = UserInterfaces.UIFeatureExamples.InIslandTransactions.Demo02.MaterialInstances.MI_UI_ItemCommon {}
    SecondOfferRarityTextColor : color = color {R := 0.0, G := 0.512, B := 0.8078}

    ThirdOfferName<localizes> : message = "Gimme Chocolate!"
    ThirdOfferDescription<localizes> : message = "Say hello to the Choco Delight Bar, literally. Be nice because it is easily offended and usually quite vengeful."
    ThirdOfferShortDescription<localizes> : message = "Third Offer Short Description"
    ThirdOfferIcon : texture = UserInterfaces.UIFeatureExamples.InIslandTransactions.Demo02.Textures.T_UI_Item03
    ThirdOfferPrice : float = 310.0
    ThirdOfferCount : int = 22
    ThirdOfferRarity : material = UserInterfaces.UIFeatureExamples.InIslandTransactions.Demo02.MaterialInstances.MI_UI_ItemRare {}
    ThirdOfferRarityTextColor : color = color {R := 0.4313, G := 0.1254, B := 0.4901}

    FourthOfferName<localizes> : message = "Chip Snack"
    FourthOfferDescription<localizes> : message = "A pack of crunchy chips, bursting with salty goodness with every bite. Light and highly addictive."
    FourthOfferShortDescription<localizes> : message = "Fourth Offer Short Description"
    FourthOfferIcon : texture = UserInterfaces.UIFeatureExamples.InIslandTransactions.Demo02.Textures.T_UI_Item04
    FourthOfferPrice : float = 100.0
    FourthOfferCount : int = 14
    FourthOfferRarity : material = UserInterfaces.UIFeatureExamples.InIslandTransactions.Demo02.MaterialInstances.MI_UI_ItemUncommon {}
    FourthOfferRarityTextColor : color = color {R := 0.1176, G := 0.4745, B := 0.1607}

    FifthOfferName<localizes> : message = "Sparkle Tea"
    FifthOfferDescription<localizes> : message = "Crafted for players who need a quick surge of focus. One sip delivers a wave of sharpness that keeps you moving forward. Contains no tea."
    FifthOfferShortDescription<localizes> : message = "Fifth Offer Short Description"
    FifthOfferIcon : texture = UserInterfaces.UIFeatureExamples.InIslandTransactions.Demo02.Textures.T_UI_Item05
    FifthOfferPrice : float = 750.0
    FifthOfferCount : int = 7
    FifthOfferRarity : material = UserInterfaces.UIFeatureExamples.InIslandTransactions.Demo02.MaterialInstances.MI_UI_ItemCommon {}
    FifthfferRarityTextColor : color = color {R := 0.0, G := 0.512, B := 0.8078}

    SixthOfferName<localizes> : message = "Starcorn"
    SixthOfferDescription<localizes> : message = "Fluffy, buttery, and full of character! Brings a cozy cinema vibe. Now with more starfish!"
    SixthOfferShortDescription<localizes> : message = "Sixth Offer Short Description"
    SixthOfferIcon : texture = UserInterfaces.UIFeatureExamples.InIslandTransactions.Demo02.Textures.T_UI_Item06
    SixthOfferPrice : float = 100.0
    SixthOfferCount : int = 3
    SixthOfferRarity : material = UserInterfaces.UIFeatureExamples.InIslandTransactions.Demo02.MaterialInstances.MI_UI_ItemIdol {}
    SixthOfferRarityTextColor : color = color {R := 0.4666, G := 0.2235, B := 0.5529}
