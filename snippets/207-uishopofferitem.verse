# Knowledge-neutral Verse snippet; project names and asset IDs generalized.
# id: snippet-207

using { /UnrealEngine.com/Temporary/UI }
using { /Verse.org/Assets }
using { /Fortnite.com/UI }
using { /Fortnite.com/Marketplace }
using { /Verse.org/Simulation }

ui_shop_offer_item := class:
    
    ItemWidget : UserInterfaces.UIFeatureExamples.InIslandTransactions.Demo01.Widgets.UW_Shop01_Item_Offer = 
        UserInterfaces.UIFeatureExamples.InIslandTransactions.Demo01.Widgets.UW_Shop01_Item_Offer {}

    var OverlayRoot : overlay = overlay {}   

    ViewButton : button_quiet = button_quiet {}

    ViewButtonText<localizes><public> : message = "View"

    CreateLayout() : overlay = 

        ViewButton.SetText(ViewButtonText)       

        set OverlayRoot = overlay:
            Slots:= array:
                overlay_slot:
                    Widget := ItemWidget
                overlay_slot:
                    Widget := ViewButton
                    VerticalAlignment := vertical_alignment.Center
                    HorizontalAlignment := horizontal_alignment.Right
                    Padding := margin {Right := 32.0}

        return OverlayRoot

    UpdateData(Offer : offer) : void =

        var PriceFloat : float = 0.0

            if (PriceVB := price_vbucks[Offer.Price]):
                set PriceFloat = GetPriceVBucks(PriceVB)

        set ItemWidget.OfferIcon = Offer.Icon
        set ItemWidget.OfferName = Offer.Name
        set ItemWidget.OfferPrice = PriceFloat
