# Knowledge-neutral Verse snippet; project names and asset IDs generalized.
# id: snippet-208

using { UserInterfaces.UIFeatureExamples.InIslandTransactions.Demo01.Textures }
using { /Fortnite.com/Devices }
using { /Fortnite.com/Marketplace }
using { /UnrealEngine.com/Temporary/Diagnostics }
using { /Verse.org/Assets }
using { /Verse.org/Simulation }
using { /UnrealEngine.com/Temporary/UI }

# Setup your entitlements and offers data, in this case that's your Names, Descriptions & Short Descriptions as well as any other data you want to be constant
EntitlementInfo<public> := module:
    CornSeedPacket<public> := module:
        Name<public><localizes>:message = "Corn Seed Packet"
        Description<public><localizes>:message = "Contains 10 corn seeds"
        ShortDescription<public><localizes>:message = "10 corn seeds"
        Count<public>:int = 5

    Shovel<public> := module:
        Name<public><localizes>:message = "Primo Shovel"
        Description<public><localizes>:message = "A better than average shovel"
        ShortDescription<public><localizes>:message = "A shovel"

    Gnome<public> := module:
        Name<public><localizes>:message = "Garden Gnome"
        Description<public><localizes>:message = "A gnome to help your garden grow"
        ShortDescription<public><localizes>:message = "A gnome"

    HeapPack<public> := module:
        Name<public><localizes>:message = "A bunch of stuff"
        Description<public><localizes>:message = "It has stuff you need"
        ShortDescription<public><localizes>:message = "Stuff"     

    AgeOfAdultHoodInTheUK<public>:int = 18

    # Country codes use the ISO-3166-1 A-2 standard
    # Subdivision codes use the ISO-3166-2 code (excluding Country Code portion)
    CountryCodes<public> := module:
        Antarctica<public>:string = "AQ"
        GreatBritain<public>:string = "GB"

Entitlements<public> := module:
    using {EntitlementInfo}

    # The base entitlement you should define for ALL your entitlements in your experience
    feature_example_entitlement<public> := class<abstract><castable>(entitlement){}

    cornseedpacket<public> := class<concrete>(feature_example_entitlement):
        var Name<override>:message = CornSeedPacket.Name
        var Description<override>:message = CornSeedPacket.Description
        var ShortDescription<override>:message = CornSeedPacket.ShortDescription
        var Icon<override>:texture = UserInterfaces.UIFeatureExamples.InIslandTransactions.Demo01.Textures.T_UI_CornSeedPacket 
        MaxCount<override>:int= 5
        Consumable<override>:logic = true        

    shovel<public> := class<concrete>(feature_example_entitlement):
        var Name<override>:message = Shovel.Name
        var Description<override>:message = Shovel.Description
        var ShortDescription<override>:message = Shovel.ShortDescription
        var Icon<override>:texture = UserInterfaces.UIFeatureExamples.InIslandTransactions.Demo01.Textures.T_UI_Shovel        

    gnome<public> := class<concrete>(feature_example_entitlement):
        var Name<override>:message = Gnome.Name
        var Description<override>:message = Gnome.Description
        var ShortDescription<override>:message = Gnome.ShortDescription
        var Icon<override>:texture = UserInterfaces.UIFeatureExamples.InIslandTransactions.Demo01.Textures.T_UI_Gnome        


ExampleOffers<public> := module:
    using {EntitlementInfo}

    shovel<public> := class(entitlement_offer):
        var Name<override>:message                = EntitlementInfo.Shovel.Name
        var Description<override>:message         = EntitlementInfo.Shovel.Description
        var ShortDescription<override>:message    = EntitlementInfo.Shovel.ShortDescription
        var Icon<override>:texture                = UserInterfaces.UIFeatureExamples.InIslandTransactions.Demo01.Textures.T_UI_Shovel

        EntitlementType<override>:concrete_subtype(entitlement) = Entitlements.shovel
        Price<override>:price_dimension = MakePriceVBucks(500.0)

        # Optional overrideable function you can use to inform Epic systems what the minimum age a player needs to be to purchase this offer
        GetMinPurchaseAge<#<override>#><public>(CountryCode:string, SubdivisionCode:string)<decides><computes>:int =
            # A Hypothetical example where you only want to sell swords to people older than 18 in Great Britain
            # to everyone everywhere else regardless of age, unless they live in Antarctica
            CountryCode <> CountryCodes.Antarctica

            if (CountryCode = CountryCodes.GreatBritain):
                AgeOfAdultHoodInTheUK
            else:
                0


    cornseedpacket<public> := class(entitlement_offer):
        var Name<override>:message                = EntitlementInfo.CornSeedPacket.Name
        var Description<override>:message         = EntitlementInfo.CornSeedPacket.Description
        var ShortDescription<override>:message    = EntitlementInfo.CornSeedPacket.ShortDescription
        var Icon<override>:texture                = UserInterfaces.UIFeatureExamples.InIslandTransactions.Demo01.Textures.T_UI_CornSeedPacket

        EntitlementType<override>:concrete_subtype(entitlement) = Entitlements.cornseedpacket
        Price<override>:price_dimension = MakePriceVBucks(100.0)

    gnome<public> := class(entitlement_offer):
        var Name<override>:message                = EntitlementInfo.Gnome.Name
        var Description<override>:message         = EntitlementInfo.Gnome.Description
        var ShortDescription<override>:message    = EntitlementInfo.Gnome.ShortDescription
        var Icon<override>:texture                = UserInterfaces.UIFeatureExamples.InIslandTransactions.Demo01.Textures.T_UI_Gnome

        EntitlementType<override>:concrete_subtype(entitlement) = Entitlements.gnome
        Price<override>:price_dimension = MakePriceVBucks(500.0)        

    heap_pack<public> := class(bundle_offer) :
        var Name<override>:message                = EntitlementInfo.HeapPack.Name
        var Description<override>:message         = EntitlementInfo.HeapPack.Description
        var ShortDescription<override>:message    = EntitlementInfo.HeapPack.ShortDescription
        var Icon<override>:texture                = UserInterfaces.UIFeatureExamples.InIslandTransactions.Demo01.Textures.T_UI_HeapBundle

        Offers<override>:[]tuple(offer, int) = array{(cornseedpacket{}, 2)}
        Price<override>:price_dimension = MakePriceVBucks(400.0)

# A Verse-authored creative device that can be placed in a level
ui_shop_setup_device_template := class(creative_device):
    
    # Track which players you're listening for entitlement changes for
    var EntitlementChangeSubscription:[player]?cancelable = map{}

    # Track Player Join subscription
    var PlayerJoinSubscription:?cancelable = false

    # Track Player Left subscription
    var PlayerLeftSubscription:?cancelable = false

    # Runs when the device is started in a running game, register to listen for player joins
    OnBegin<override>()<suspends>:void =        
        SubJoin:= GetPlayspace().PlayerAddedEvent().Subscribe(OnPlayerJoin)
        set PlayerJoinSubscription = option{ SubJoin }

        SubLeft:= GetPlayspace().PlayerRemovedEvent().Subscribe(OnPlayerLeft)
        set PlayerLeftSubscription = option{ SubLeft }        
        
        Playspace := GetPlayspace()
        for (Player : Playspace.GetPlayers()):
            OnPlayerJoin(Player)

    OnEnd<override>() : void =
        if(Subscription := PlayerJoinSubscription?):
            Subscription.Cancel()

        if(Subscription := PlayerLeftSubscription?):
            Subscription.Cancel()

    # When a player joins you subscribe to listen to any changes in what they own.
    OnPlayerJoin(InPlayer:player):void =        
        Subscription := GetEntitlementsChangedEvent(InPlayer, Entitlements.feature_example_entitlement).Subscribe(OnPurchasesChanged)

        if (set EntitlementChangeSubscription[InPlayer] = option{Subscription}):
            Print("Adding Entitlement Change Subscription player subscription")       

        # On players joining you are likely going to want to run some validation checks to make sure that any data you save
        # in your experience matches what the Marketplace API says they own.
        spawn{ ValidatePreviousPurchases(InPlayer) }

    # clean up any Subscriptions made for that player
    OnPlayerLeft(Player:player):void =
        if (Subscription := EntitlementChangeSubscription[Player]?):
            Subscription.Cancel()

        var TempMap:[player]?cancelable = map{}

        for (Key -> Value : EntitlementChangeSubscription, Key <> Player):
            set TempMap = ConcatenateMaps(TempMap, map{Key => Value})

        set EntitlementChangeSubscription = TempMap

    # How to respond to the Player's purchases having changed for whatever reason, purchasing something
    OnPurchasesChanged(Player:player, EntitlementType:[]entitlement_change(entitlement)):void= ()
        # Adjust Game Values
        # Process refunds       

    ValidatePreviousPurchases(Player:player)<suspends>:void=
        Purchases := GetPurchasedEntitlements(Player, Entitlements.feature_example_entitlement)       
        
        # Do any checks you need to make sure your saved data matches what Marketplace says the player owns

    # Base Implementation of how to present players with an offer to purchase in your experience
    TryBuyOffer(Player:player, Offer:entitlement_offer)<suspends>:void =
        Result := BuyOffer(Player, Offer)

        if (Result?):
            # do nothing it should respond in the purchase subscription, see OnPurchasesChanged for details
        else:
            Print("Failed to buy the offer")

    # Base Implementation of how to flag a consumable in your experience as being consumed
    TryConsumeEntitlement(Player:player, Entitlement:concrete_subtype(entitlement), NumberConsuming:int)<suspends>:void =
        Result := ConsumeEntitlement(Player, Entitlement, ?Count := NumberConsuming)

        if (Result?):
            Print("Successfully consumed entitlement!")
            # give the player the thing/react in your game, you can also respond in OnPurchasesChanged as with purchase if you choose
        else:
            Print("Failed to consume the entitlement")

    # Base Implementation of how to give players a entitlement in your experience without them purchasing it
    TryGrantEntitlement(Player:player, Entitlement:concrete_subtype(entitlement), NumberToGrant:int)<suspends>:void =
        Result := GrantEntitlement(Player, Entitlement, ?Count := NumberToGrant)

        if (Result?):
            Print("Successfully granted a player your entitlement!")
            # give the player the thing/react in your game, you can also respond in OnPurchasesChanged as with purchase if you choose
        else:
            Print("Failed to grant the entitlement")

    # Base Implementation of how to show an array of offers to the player that are available for purchase
    ShowArrayOfOffers(Player:player)<suspends>:void =        
        var Shop : ui_shop = ui_shop {Player := Player}

        Shop.ShowUI(array{ExampleOffers.shovel{}, ExampleOffers.cornseedpacket{}, ExampleOffers.heap_pack{}})

    # Base Implementation of how to show an array of offers to the player that are available for purchase
    ShowOffersUI(Player:player)<suspends>:void =
        ShowOffersDialog(Player, array{ExampleOffers.shovel{}, ExampleOffers.cornseedpacket{}, ExampleOffers.heap_pack{}})
