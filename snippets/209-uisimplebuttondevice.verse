# Knowledge-neutral Verse snippet; project names and asset IDs generalized.
# id: snippet-209

using { /Fortnite.com/Devices }
using { /Verse.org/Simulation }
using { /Fortnite.com/UI }
using { /UnrealEngine.com/Temporary/UI }
using { /UnrealEngine.com/Temporary/SpatialMath }
using { /Verse.org/Colors }

ui_simple_button_device := class(creative_device):

    @editable
    ShowVolume : volume_device = volume_device{}

    # Text for the Verse button states
    VerseButtonTextIdle<localizes> : message = "BUTTON"
    VerseButtonTextHighlighted<localizes> : message = "HIGHLIGHTED"

    CloseButtonText<localizes> : message = "CLOSE"

    # Widgets
    TextWidget : text_block = text_block{}
    TextWidget2 : text_block = text_block{}
    var RootOverlay : overlay = overlay{}

    OnBegin<override>()<suspends>:void =
        # Show UI when a player enters the volume
        ShowVolume.AgentEntersEvent.Subscribe(ShowToAllPlayers)

    ShowToAllPlayers(InAgent:agent):void =
        set RootOverlay = overlay{}

        # Configure the text used inside VerseButton
        TextWidget.SetText(VerseButtonTextIdle)
        TextWidget.SetTextSize(56.0)
        TextWidget.SetTextColor(color{R:=1.0, G:=1.0, B:=1.0})
        TextWidget.SetShadowOffset(option{vector2{X := 4.0, Y := 4.0}})
        TextWidget.SetShadowColor(color{R:=0.0, G:=0.0, B:=0.0})
        TextWidget.SetShadowOpacity(0.8)

        # Configure the text for the second button (same styling)
        TextWidget2.SetText(VerseButtonTextIdle)
        TextWidget2.SetTextSize(56.0)
        TextWidget2.SetTextColor(color{R:=1.0, G:=1.0, B:=1.0})
        TextWidget2.SetShadowOffset(option{vector2{X := 4.0, Y := 4.0}})
        TextWidget2.SetShadowColor(color{R:=0.0, G:=0.0, B:=0.0})
        TextWidget2.SetShadowOpacity(0.8)

        # Create VerseButtons
        VerseButton:button = button{ Slot := button_slot{ Widget := TextWidget } }
        VerseButton2:button = button{ Slot := button_slot{ Widget := TextWidget2 } }

        var CloseButton : button_quiet = button_quiet{ DefaultText := CloseButtonText }

        # Subscribe to VerseButton events
        VerseButton.OnClick().Subscribe(OnClicked)
        VerseButton.HighlightEvent().Subscribe(Button1_OnHighlighted)
        VerseButton.UnhighlightEvent().Subscribe(Button1_OnUnhighlighted)
        CloseButton.OnClick().Subscribe(OnClicked)

        # Subscribe to VerseButton2 events (same behavior)
        VerseButton2.OnClick().Subscribe(OnClicked)
        VerseButton2.HighlightEvent().Subscribe(Button2_OnHighlighted)
        VerseButton2.UnhighlightEvent().Subscribe(Button2_OnUnhighlighted)

        # Stack them vertically and show centered
        Stack := stack_box{ Orientation := orientation.Vertical }

        Stack.AddWidget(stack_box_slot{
            Widget := VerseButton
        })
        Stack.AddWidget(stack_box_slot{
            Widget := VerseButton2
        })
        Stack.AddWidget(stack_box_slot{
            Widget := CloseButton
        })

        RootOverlay.AddWidget(overlay_slot{
            Widget := Stack,
            HorizontalAlignment := horizontal_alignment.Center,
            VerticalAlignment := vertical_alignment.Center
        })

        for:
            Player : GetPlayspace().GetPlayers()
            PlayerUI := GetPlayerUI[Player]
        do:
            PlayerUI.SetFocus(VerseButton)
            PlayerUI.AddWidget(RootOverlay, player_ui_slot{ InputMode := ui_input_mode.All })

    OnClicked(InMessage:widget_message):void =
        for:
            Player:GetPlayspace().GetPlayers()
            PlayerUI := GetPlayerUI[Player]
        do:
            PlayerUI.RemoveWidget(RootOverlay)

    Button1_OnHighlighted(InMessage:widget_message):void =
        TextWidget.SetText(VerseButtonTextHighlighted)
        TextWidget.SetTextSize(64.0)

    Button1_OnUnhighlighted(InMessage:widget_message):void =
        TextWidget.SetText(VerseButtonTextIdle)
        TextWidget.SetTextSize(56.0)
        
    Button2_OnHighlighted(InMessage:widget_message):void =
        TextWidget2.SetText(VerseButtonTextHighlighted)
        TextWidget2.SetTextSize(64.0)

    Button2_OnUnhighlighted(InMessage:widget_message):void =
        TextWidget2.SetText(VerseButtonTextIdle)
        TextWidget2.SetTextSize(56.0)
