# Knowledge-neutral Verse snippet; project names and asset IDs generalized.
# id: snippet-210

using { /Fortnite.com/Devices }
using { /Fortnite.com/UI }
using { /Verse.org/Simulation }
using { /Verse.org/Random }
using { /Verse.org/Colors }
using { /UnrealEngine.com/Temporary/Diagnostics }
using { /UnrealEngine.com/Temporary/UI }

# See https://dev.epicgames.com/documentation/en-us/uefn/create-your-own-device-in-verse for how to create a verse device.

# A Verse-authored creative device that can be placed in a level
ui_textsize_device := class(creative_device):

    @editable
    ShowTrigger:volume_device = volume_device{}    

    # Messages
    SmallText<localizes> : message = "This is some small text!"
    MedText<localizes> : message = "This is some medium sized text!"
    LargeText<localizes> : message = "This is some really large text!"
    RandomText<localizes> : message = "Random text sizes galore!"
    RandomButtonText<localizes> : message = "Random Size"
    CloseButtonText<localizes> : message = "Close"

    # Verse UI for the demo
    var MyUI : overlay = overlay{}
    TextSmall:text_block = text_block{DefaultTextSize := 16.0, DefaultTextColor := NamedColors.White}
    TextMed:text_block = text_block{DefaultTextSize := 28.0, DefaultTextColor := NamedColors.White}
    TextLarge:text_block = text_block{DefaultTextSize := 52.0, DefaultTextColor := NamedColors.White}

    # Runs when the device is started in a running game
    OnBegin<override>()<suspends>:void=        
        ShowTrigger.AgentEntersEvent.Subscribe(OnShow)

        # Initialize text blocks
        InitializeTextBlocks()
        
        # Create my buttons
        RandomButton : button_loud = button_loud{DefaultText := RandomButtonText}
        CloseButton : button_quiet = button_quiet{DefaultText := CloseButtonText}

        # Subscribe button clicks to functions
        RandomButton.OnClick().Subscribe(HandleRandomizeButtonClick)
        CloseButton.OnClick().Subscribe(OnHide)

        # Create UI
        MyHudUI := overlay:
            Slots := array:
                overlay_slot:
                    Widget := color_block{DefaultColor := NamedColors.Black, DefaultOpacity := 0.3}
                    VerticalAlignment := vertical_alignment.Fill
                    HorizontalAlignment := horizontal_alignment.Fill
                    Padding := margin{Left:=256.0, Top:=256.0, Right:=256.0, Bottom:=256.0}
                overlay_slot:
                    VerticalAlignment := vertical_alignment.Center
                    HorizontalAlignment := horizontal_alignment.Center 
                    Widget := stack_box:
                        Orientation := orientation.Vertical
                        Slots := array:
                            stack_box_slot:
                                Widget := TextSmall
                            stack_box_slot:
                                Widget := TextMed
                            stack_box_slot:
                                Widget := TextLarge
                                Padding := margin{Bottom:=24.0}
                            stack_box_slot:
                                Widget := stack_box:
                                    Orientation := orientation.Horizontal
                                    Slots := array:
                                        stack_box_slot:
                                            Widget := RandomButton
                                            Padding := margin{Right:=16.0}
                                        stack_box_slot:
                                            Widget := CloseButton
        set MyUI = MyHudUI

    OnShow(InAgent:agent):void=
        for:
            Player : GetPlayspace().GetPlayers()
            PlayerUI := GetPlayerUI[Player]
        do:
            PlayerUI.AddWidget(MyUI,player_ui_slot{InputMode := ui_input_mode.All})
    
    OnHide(Message : widget_message):void=
        for:
            Player:GetPlayspace().GetPlayers()
            PlayerUI := GetPlayerUI[Player]
        do:
            PlayerUI.RemoveWidget(MyUI)
    
    HandleRandomizeButtonClick(Message : widget_message):void=
        TextSmall.SetText(RandomText)
        TextSmall.SetTextSize(GetRandomFloat(16.0,92.0))
        TextMed.SetText(RandomText)
        TextMed.SetTextSize(GetRandomFloat(16.0,92.0))
        TextLarge.SetText(RandomText)
        TextLarge.SetTextSize(GetRandomFloat(16.0,92.0))

    InitializeTextBlocks():void=
        TextSmall.SetText(SmallText)
        TextMed.SetText(MedText)
        TextLarge.SetText(LargeText)
