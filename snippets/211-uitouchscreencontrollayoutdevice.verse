# Knowledge-neutral Verse snippet; project names and asset IDs generalized.
# id: snippet-211

using { /Fortnite.com/Devices }
using { /Verse.org/Simulation }
using { /UnrealEngine.com/Temporary/Diagnostics }
using { /UnrealEngine.com/Temporary/UI }
using { /Verse.org/Input }
using { /Verse.org/Assets }
using { /Fortnite.com/Input/Character }
using { /Fortnite.com/UI }
using { /UnrealEngine.com/Temporary/SpatialMath }

# A Verse-authored creative device that can be placed in a level
ui_touchscreen_control_layout_device := class(creative_device):

    @editable
    HudLayoutDevice01 : hud_controller_device = hud_controller_device{}
    @editable
    HudLayoutDevice02 : hud_controller_device = hud_controller_device{}
    @editable
    HudLayoutDevice03 : hud_controller_device = hud_controller_device{}
    @editable
    HudLayoutDevice04 : hud_controller_device = hud_controller_device{}
    @editable
    OpenMenuTrigger : trigger_device = trigger_device{}
    @editable
    Volume : volume_device = volume_device {}

    TouchLayoutMenu : UserInterfaces.VerseUIExamples.TouchscreenControlLayout.Widgets.UW_TCL_Menu = UserInterfaces.VerseUIExamples.TouchscreenControlLayout.Widgets.UW_TCL_Menu {}
    
    # Declare a variable to store the cancelable subscription so we can clean it up later
    var OpenMenuTriggeredSubscription : ?cancelable = false
    var PlayerEntersVolumeSubscriptionOption : ?cancelable = false

    # Runs when the device is started in a running game
    OnBegin<override>()<suspends>:void=
        set OpenMenuTriggeredSubscription = option{OpenMenuTrigger.TriggeredEvent.Subscribe(OnMenuActionPressed)}
        SubscribeToEvent(TouchLayoutMenu.Layout1Selected, OnLayout1Selected)
        SubscribeToEvent(TouchLayoutMenu.Layout2Selected, OnLayout2Selected)
        SubscribeToEvent(TouchLayoutMenu.Layout3Selected, OnLayout3Selected)
        SubscribeToEvent(TouchLayoutMenu.Layout4Selected, OnLayout4Selected)
        set PlayerEntersVolumeSubscriptionOption = option{Volume.AgentEntersEvent.Subscribe(DisableDevices)}

    OnEnd<override>():void = 
        CancelSubscription(OpenMenuTriggeredSubscription)
        CancelSubscription(PlayerEntersVolumeSubscriptionOption)

    DisableDevices(Agent : agent) : void =
        HudLayoutDevice01.Disable()
        HudLayoutDevice02.Disable()
        HudLayoutDevice03.Disable()
        HudLayoutDevice04.Disable() 

    # Define the handler function that reacts to the jump
    OnJump(Payload : tuple(player, logic)) : void =
        for (Player : GetPlayspace().GetPlayers()):
            if (PlayerUI := GetPlayerUI[Player]):
                PlayerUI.AddWidget(TouchLayoutMenu, player_ui_slot{InputMode := ui_input_mode.All})
                PlayerUI.SetFocus(TouchLayoutMenu)

    OnLayoutSelected() : void =
        for (Player : GetPlayspace().GetPlayers()):
            if (PlayerUI := GetPlayerUI[Player]):
                PlayerUI.RemoveWidget(TouchLayoutMenu)  
                if (PlayerInput := GetPlayerInput[Player]):
                    PlayerInput.AddInputMapping(TraversalMapping)
                    PlayerInput.GetInputEvents(Jump).TriggerActivationEvent.Subscribe(OnJump)
                
    OnLayout1Selected() : void =
        OnLayoutSelected()
        ConfigureHCD(0)

    OnLayout2Selected() : void =
        OnLayoutSelected()
        ConfigureHCD(1)

    OnLayout3Selected() : void =
        OnLayoutSelected()
        ConfigureHCD(2)

    OnLayout4Selected() : void =
        OnLayoutSelected()
        ConfigureHCD(3)

    OnMenuActionPressed(InAgent:?agent) : void =
        for (Player : GetPlayspace().GetPlayers()):
            if (PlayerUI := GetPlayerUI[Player]):
                PlayerUI.AddWidget(TouchLayoutMenu, player_ui_slot{InputMode := ui_input_mode.All})
                PlayerUI.SetFocus(TouchLayoutMenu)
                if (PlayerInput := GetPlayerInput[Player]):
                    PlayerInput.RemoveInputMapping(TraversalMapping)

    ConfigureHCD(Active : int) : void =
        Print("Active = {Active}")
        HudLayoutDevice01.Disable()
        HudLayoutDevice02.Disable()
        HudLayoutDevice03.Disable()
        HudLayoutDevice04.Disable()
        if (Active = 0):
            HudLayoutDevice01.Enable()
        else if (Active = 1):
            HudLayoutDevice02.Enable()
        else if (Active = 2):
            HudLayoutDevice03.Enable()
        else if (Active = 3):
            HudLayoutDevice04.Enable()
