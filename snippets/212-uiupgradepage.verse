# Knowledge-neutral Verse snippet; project names and asset IDs generalized.
# id: snippet-212

# Copyright Epic Games, Inc. All Rights Reserved.

using { /Verse.org/Simulation }
using { /Verse.org/Input }
using { /Verse.org/Input/UI }
using { /Fortnite.com/Devices }
using { /Fortnite.com/Characters }
using { /Fortnite.com/UI }
using { /UnrealEngine.com/Temporary/UI }

# ============================================================
#  upgrade_page
#
#  Contains the upgrade data as well as the logic for purchasing upgrades
#  To add a new upgrade to the game, just add a new entry to the Upgrades list below.
# ============================================================

# PurchaseState tells the UI which visual state to show on the  buy button:
#   0 = Available to buy
#   1 = Already owned
#   2 = Cannot afford
#   3 = Prerequisites not yet met

upgrade_data := struct:
    Name : string
    Description : string
    Cost : int
    Prerequisites : []int
    WeaponBuffName: string
    ArmorBuffName : string
    GearBuffName: string
    WeaponBuffValue : int
    ArmorBuffValue: int
    GearBuffValue : int
    HasWeaponBuff : logic
    HasArmorBuff: logic
    HasGearBuff : logic

ui_upgrade_page := class(ui_menu_page):

    UpgradeWidget : UserInterfaces.UIFeatureExamples.Menu.Widgets.UpgradeSelect.UW_Menu_Upgrades = UserInterfaces.UIFeatureExamples.Menu.Widgets.UpgradeSelect.UW_Menu_Upgrades{}

    var Upgrades : []upgrade_data = array{}

    # Tracks which upgrades the player has already purchased.
    var EquippedSlots : []logic = array{
        false, false, false, false, false, false, false,
        false, false, false, false, false, false
    }

    var CurrentHighlightedIndex : int = 0

    var CurrentSelectedIndex : int = 0

    var InitialCurrency : int = 0

    var IsGamepad : logic = false

    GetTitle<override>():string = "Upgrades"
    GetContentWidget<override>():widget = UpgradeWidget

    # This is where all upgrade data is defined.
    InitData<override>():void=
        set Upgrades = array{

            upgrade_data{ # Upgrade 0:  Weapon. No prerequisites.
                Name := "Spike Strip",
                Description := "Deploys a strip of spikes behind your vehicle. Punishes anyone following too close.",
                Cost:= 150,
                Prerequisites := array{},
                WeaponBuffName := "Spike Damage",WeaponBuffValue := 25, HasWeaponBuff := true,
                ArmorBuffName := "", ArmorBuffValue := 0, HasArmorBuff := false,
                GearBuffName := "", GearBuffValue := 0,HasGearBuff:= false
            },

            upgrade_data{ # Upgrade 1:  Armor. No prerequisites.
                Name:= "Armour Plating",
                Description := "Heavy steel panels bolted to the chassis. You will feel every hit less.",
                Cost := 250,
                Prerequisites := array{},
                WeaponBuffName := "", WeaponBuffValue := 0, HasWeaponBuff := false,
                ArmorBuffName := "Damage Resist", ArmorBuffValue := 25, HasArmorBuff:= true,
                GearBuffName := "", GearBuffValue := 0,HasGearBuff := false
            },

            upgrade_data{ # Upgrade 2:  Gear. No prerequisites.
                Name := "Turbo Engine",
                Description := "A high-output engine swap. Massively increases top speed and raw power.",
                Cost := 1500,
                Prerequisites := array{},
                WeaponBuffName := "", WeaponBuffValue := 0,HasWeaponBuff := false,
                ArmorBuffName:= "", ArmorBuffValue:= 0, HasArmorBuff:= false,
                GearBuffName := "Top Speed", GearBuffValue := 30, HasGearBuff := true
            },

            upgrade_data{ # Upgrade 3:  Weapon + Armor combo. Requires upgrade 0 and 1.
                Name := "Ram Bumper",
                Description := "A reinforced front bumper built to dish out and absorb punishment in equal measure.",
                Cost := 200,
                Prerequisites := array{0, 1},
                WeaponBuffName := "Ram Damage", WeaponBuffValue := 20, HasWeaponBuff := true,
                ArmorBuffName := "Collision Resist", ArmorBuffValue := 18, HasArmorBuff := true,
                GearBuffName := "", GearBuffValue := 0,HasGearBuff := false
            },

            upgrade_data{ # Upgrade 4:  Weapon. Requires upgrade 0.
                Name := "EMP Cannon",
                Description := "Fires a targeted electromagnetic pulse. Disables nearby rivals temporarily.",
                Cost:= 2000,
                Prerequisites := array{0},
                WeaponBuffName := "Disable Duration", WeaponBuffValue := 22, HasWeaponBuff := true,
                ArmorBuffName := "", ArmorBuffValue := 0, HasArmorBuff := false,
                GearBuffName := "", GearBuffValue := 0, HasGearBuff := false
            },

            upgrade_data{ # Upgrade 5:  Armor. Requires upgrade 1.
                Name := "Roll Cage",
                Description := "A full safety cage welded to the frame. Protects the driver through any impact.",
                Cost:= 1800,
                Prerequisites := array{1},
                WeaponBuffName := "", WeaponBuffValue := 0, HasWeaponBuff := false,
                ArmorBuffName := "Collision Resist", ArmorBuffValue := 35, HasArmorBuff := true,
                GearBuffName := "", GearBuffValue := 0, HasGearBuff := false
            },

            upgrade_data{ # Upgrade 6:  Gear. Requires upgrade 2.
                Name:= "Aero Kit",
                Description := "A full aerodynamic body kit. Cuts drag and improves high-speed stability.",
                Cost := 2500,
                Prerequisites := array{2},
                WeaponBuffName := "",WeaponBuffValue := 0, HasWeaponBuff := false,
                ArmorBuffName := "", ArmorBuffValue := 0, HasArmorBuff:= false,
                GearBuffName := "Top Speed", GearBuffValue := 15, HasGearBuff := true
            },

            upgrade_data{ # Upgrade 7:  Weapon. Requires upgrade 4.
                Name := "Battering Ram",
                Description := "A heavy-duty front ram. Designed to push anything out of the way.",
                Cost := 1200,
                Prerequisites := array{4},
                WeaponBuffName := "Ram Damage", WeaponBuffValue := 30, HasWeaponBuff := true,
                ArmorBuffName := "", ArmorBuffValue := 0, HasArmorBuff := false,
                GearBuffName := "", GearBuffValue := 0,HasGearBuff := false
            },

            upgrade_data{ # Upgrade 8:  Armor. Requires upgrade 5.
                Name := "Heat Shield",
                Description := "An advanced thermal barrier around the engine bay. Sustains performance under heat.",
                Cost := 3000,
                Prerequisites := array{5},
                WeaponBuffName := "", WeaponBuffValue := 0, HasWeaponBuff := false,
                ArmorBuffName := "Engine Durability", ArmorBuffValue:= 28, HasArmorBuff:= true,
                GearBuffName := "", GearBuffValue := 0,HasGearBuff := false
            },

            upgrade_data{ # Upgrade 9:  Armor + Gear combo. Requires upgrade 8 and 6.
                Name := "Nitrous Tank",
                Description := "A secondary nitrous system with a reinforced tank. Speed and resilience in one.",
                Cost := 2200,
                Prerequisites := array{8, 6},
                WeaponBuffName := "", WeaponBuffValue := 0,HasWeaponBuff := false,
                ArmorBuffName := "Tank Durability", ArmorBuffValue := 15, HasArmorBuff:= true,
                GearBuffName := "Boost Duration", GearBuffValue := 20, HasGearBuff := true
            },

            upgrade_data{ # Upgrade 10:  Gear. Requires upgrade 6.
                Name := "Carbon Frame",
                Description := "Strips the body down to carbon fibre. Lighter, faster, more responsive.",
                Cost := 4000,
                Prerequisites := array{6},
                WeaponBuffName := "", WeaponBuffValue := 0,HasWeaponBuff := false,
                ArmorBuffName := "", ArmorBuffValue := 0, HasArmorBuff:= false,
                GearBuffName := "Weight", GearBuffValue := 20, HasGearBuff := true
            },

            upgrade_data{ # Upgrade 11:  Weapon + Armor combo. Requires upgrade 7 and 8.
                Name := "Radar Jammer",
                Description := "Broadcasts interference across rival tracking systems. Hard to see, hard to hit.",
                Cost := 1600,
                Prerequisites:= array{7, 8},
                WeaponBuffName := "Stealth Rating", WeaponBuffValue := 30, HasWeaponBuff := true,
                ArmorBuffName := "Detection Range", ArmorBuffValue := 25, HasArmorBuff := true,
                GearBuffName := "", GearBuffValue := 0,HasGearBuff := false
            },

            upgrade_data{ # Upgrade 12:  Armor + Gear combo. Requires upgrade 9.
                Name := "Overclocked ECU",
                Description := "Pushes the ECU beyond factory limits. Everything gets faster and runs hotter.",
                Cost := 5000,
                Prerequisites := array{9},
                WeaponBuffName := "",WeaponBuffValue := 0, HasWeaponBuff := false,
                ArmorBuffName:= "Thermal Resist", ArmorBuffValue := 20, HasArmorBuff := true,
                GearBuffName := "Power Output", GearBuffValue := 25, HasGearBuff := true
            }
        }
        set InitialCurrency = Session.PlayerCurrency
        UpdateUpgradeDisplay()

    SubscribeToPageEvents<override>():void=
        SelectSub := SubscribeToEvent(UpgradeWidget.SelectUpgrade, OnSelectUpgrade)
        HighlightSub := SubscribeToEvent(UpgradeWidget.HighlightUpgrade, OnHighlightUpgrade)
        ConfirmSub := SubscribeToEvent(UpgradeWidget.ConfirmUpgrade, OnConfirmUpgrade)
        set PageSubscriptions += array{ SelectSub, HighlightSub, ConfirmSub }

    # When the player opens this screen, reset the state
    OnShow<override>():void=
        set IsGamepad = if (Session.IsGamepad?) then true else false
        set Session.PlayerCurrency = InitialCurrency
        set EquippedSlots = array{
            false, false, false, false, false, false, false,
            false, false, false, false, false, false
        }
        set CurrentHighlightedIndex = 0
        set CurrentSelectedIndex = 0
        SyncCurrency()
        SyncEquippedAndConnectorStates()
        UpdateUpgradeDisplay()

    # Checks whether the player has already purchased every prerequisite for an  upgrade
    ArePrerequisitesMet(UpgradeIndex : int) : logic=
        if (UpgradeEntry := Upgrades[UpgradeIndex]):
            var AllPrerequisitesMet : logic = true
            for (RequiredUpgradeIndex : UpgradeEntry.Prerequisites):
                AlreadyOwned := if (SlotValue := EquippedSlots[RequiredUpgradeIndex]) then SlotValue else false
                if (AlreadyOwned = false):
                    set AllPrerequisitesMet = false
            AllPrerequisitesMet
        else:
            false

    # Runs whenever the player moves their cursor to a different upgrade.
    OnHighlightUpgrade(Index : tuple(int)):void=
        set CurrentHighlightedIndex = Index(0)
        if (IsGamepad?):
            LockInUpgrade(Index(0))

    # Locks in an upgrade as the current selection and refreshes the display.
    LockInUpgrade(UpgradeIndex : int):void=
        set CurrentSelectedIndex = UpgradeIndex
        UpdateUpgradeDisplay()

    # Pushes all data for the locked-in (selected) upgrade to the UI widget
    UpdateUpgradeDisplay():void=
        if (UpgradeEntry := Upgrades[CurrentSelectedIndex]):
            set UpgradeWidget.CurrentHighlightedUpgrade= CurrentSelectedIndex
            set UpgradeWidget.CurrentSelectedUpgrade = CurrentSelectedIndex
            set UpgradeWidget.CurrentHighlightedUpgradeName = StringToMessage(UpgradeEntry.Name)
            set UpgradeWidget.CurrentHighlightedUpgradeDescription= StringToMessage(UpgradeEntry.Description)
            set UpgradeWidget.CurrentHighlightedUpgradeCost = UpgradeEntry.Cost
            set UpgradeWidget.CurrentHighlightedUpgradeWeaponBuffName = StringToMessage(UpgradeEntry.WeaponBuffName)
            set UpgradeWidget.CurrentHighlightedUpgradeArmorBuffName = StringToMessage(UpgradeEntry.ArmorBuffName)
            set UpgradeWidget.CurrentHighlightedUpgradeGearBuffName = StringToMessage(UpgradeEntry.GearBuffName)
            set UpgradeWidget.CurrentHighlightedUpgradeWeaponBuff = UpgradeEntry.WeaponBuffValue
            set UpgradeWidget.CurrentHighlightedUpgradeArmorBuff = UpgradeEntry.ArmorBuffValue
            set UpgradeWidget.CurrentHighlightedUpgradeGearBuff = UpgradeEntry.GearBuffValue
            set UpgradeWidget.HasWeaponBuff = UpgradeEntry.HasWeaponBuff
            set UpgradeWidget.HasArmorBuff = UpgradeEntry.HasArmorBuff
            set UpgradeWidget.HasGearBuff = UpgradeEntry.HasGearBuff
            SetCurrentIcon(CurrentSelectedIndex)

            # Determine the state that the buy button should be in
            PrereqsMet := ArePrerequisitesMet(CurrentSelectedIndex)

            AlreadyOwned := if (SlotValue := EquippedSlots[CurrentSelectedIndex]) then SlotValue else false
            var BuyButtonState : int = 3
            if (AlreadyOwned = true):
                set BuyButtonState = 1
            else if (PrereqsMet = false):
                set BuyButtonState = 3
            else if (Session.PlayerCurrency < UpgradeEntry.Cost):
                set BuyButtonState = 2
            else:
                set BuyButtonState = 0
            set UpgradeWidget.PurchaseState = BuyButtonState

    # Fires when the player clicks/taps an upgrade 
    OnSelectUpgrade(Index : tuple(int)):void=
        LockInUpgrade(Index(0))

    # When the player taps the on-screen Upgrade button (mobile), buy the upgrade.
    OnConfirmUpgrade(Event : tuple()):void=
        AttemptUpgradePurchase(CurrentSelectedIndex)

    # When the player presses the confirm input action (gamepad/keyboard), buy the upgrade
    OnSelect<override>()<suspends> : void =
        AttemptUpgradePurchase(CurrentSelectedIndex)

    # Attempts to purchase the upgrade at the given index
    AttemptUpgradePurchase(UpgradeIndex : int):void=
        if (UpgradeEntry := Upgrades[UpgradeIndex]):
            AlreadyOwned := if (SlotValue := EquippedSlots[UpgradeIndex]) then SlotValue else false
            if (AlreadyOwned = true):
                return
            PrereqsMet := ArePrerequisitesMet(UpgradeIndex)
            if (PrereqsMet = false):
                return
            if (Session.PlayerCurrency < UpgradeEntry.Cost):
                return
            if (set EquippedSlots[UpgradeIndex] = true):
                set Session.PlayerCurrency = Session.PlayerCurrency - UpgradeEntry.Cost
                SyncCurrency()
                SyncEquippedAndConnectorStates()
                UpdateUpgradeDisplay()

    SyncCurrency():void=
        set CurrencyWidget.CurrencyAmount = Session.PlayerCurrency

    # Updates every upgrade slot icon and every connector line on the tree
    SyncEquippedAndConnectorStates():void=
        # Push owned/not-owned state for every upgrade slot icon
        set UpgradeWidget.Upgrade0Equipped = if (SlotValue := EquippedSlots[0]) then SlotValue else false
        set UpgradeWidget.Upgrade1Equipped = if (SlotValue := EquippedSlots[1]) then SlotValue else false
        set UpgradeWidget.Upgrade2Equipped = if (SlotValue := EquippedSlots[2]) then SlotValue else false
        set UpgradeWidget.Upgrade3Equipped = if (SlotValue := EquippedSlots[3]) then SlotValue else false
        set UpgradeWidget.Upgrade4Equipped = if (SlotValue := EquippedSlots[4]) then SlotValue else false
        set UpgradeWidget.Upgrade5Equipped = if (SlotValue := EquippedSlots[5]) then SlotValue else false
        set UpgradeWidget.Upgrade6Equipped = if (SlotValue := EquippedSlots[6]) then SlotValue else false
        set UpgradeWidget.Upgrade7Equipped = if (SlotValue := EquippedSlots[7]) then SlotValue else false
        set UpgradeWidget.Upgrade8Equipped = if (SlotValue := EquippedSlots[8]) then SlotValue else false
        set UpgradeWidget.Upgrade9Equipped = if (SlotValue := EquippedSlots[9]) then SlotValue else false
        set UpgradeWidget.Upgrade10Equipped = if (SlotValue := EquippedSlots[10]) then SlotValue else false
        set UpgradeWidget.Upgrade11Equipped = if (SlotValue := EquippedSlots[11]) then SlotValue else false
        set UpgradeWidget.Upgrade12Equipped = if (SlotValue := EquippedSlots[12]) then SlotValue else false

        Owns0 := if (SlotValue := EquippedSlots[0]) then SlotValue else false
        Owns1 := if (SlotValue := EquippedSlots[1]) then SlotValue else false
        Owns2 := if (SlotValue := EquippedSlots[2]) then SlotValue else false
        Owns3 := if (SlotValue := EquippedSlots[3]) then SlotValue else false
        Owns4 := if (SlotValue := EquippedSlots[4]) then SlotValue else false
        Owns5 := if (SlotValue := EquippedSlots[5]) then SlotValue else false
        Owns6 := if (SlotValue := EquippedSlots[6]) then SlotValue else false
        Owns7 := if (SlotValue := EquippedSlots[7]) then SlotValue else false
        Owns8 := if (SlotValue := EquippedSlots[8]) then SlotValue else false
        Owns9 := if (SlotValue := EquippedSlots[9]) then SlotValue else false
        Owns10 := if (SlotValue := EquippedSlots[10]) then SlotValue else false
        Owns11 := if (SlotValue := EquippedSlots[11]) then SlotValue else false
        Owns12 := if (SlotValue := EquippedSlots[12]) then SlotValue else false

        # These determine what connectors should be filled depending on owned upgrades.
        set UpgradeWidget.Connector0Filled = if (Owns0 = true, Owns1 = true) then true else false
        set UpgradeWidget.Connector1Filled = Owns3
        set UpgradeWidget.Connector2Filled = if (Owns3 = true, Owns5 = true) then true else false
        set UpgradeWidget.Connector3Filled = if (Owns4 = true, Owns5 = true) then true else false
        set UpgradeWidget.Connector4Filled = Owns7
        set UpgradeWidget.Connector5Filled = Owns8
        set UpgradeWidget.Connector6Filled = Owns9
        set UpgradeWidget.Connector7Filled = Owns9
        set UpgradeWidget.Connector8Filled = Owns10
        set UpgradeWidget.Connector9Filled = if (Owns7 = true, Owns8 = true) then true else false
        set UpgradeWidget.Connector10Filled = Owns11
        set UpgradeWidget.Connector11Filled = Owns6
        set UpgradeWidget.Connector12Filled = Owns12

    # Maps each upgrade index to its icon texture.
    SetCurrentIcon(UpgradeIndex : int):void=
        if (UpgradeIndex = 0):
            set UpgradeWidget.CurrentHighlightedUpgradeTexture = UserInterfaces.UIFeatureExamples.Menu.Textures.UpgradeIcons.T_UI_ICN_Upgrade_13
        else if (UpgradeIndex = 1):
            set UpgradeWidget.CurrentHighlightedUpgradeTexture = UserInterfaces.UIFeatureExamples.Menu.Textures.UpgradeIcons.T_UI_ICN_Upgrade_08
        else if (UpgradeIndex = 2):
            set UpgradeWidget.CurrentHighlightedUpgradeTexture = UserInterfaces.UIFeatureExamples.Menu.Textures.UpgradeIcons.T_UI_ICN_Upgrade_03
        else if (UpgradeIndex = 3):
            set UpgradeWidget.CurrentHighlightedUpgradeTexture = UserInterfaces.UIFeatureExamples.Menu.Textures.UpgradeIcons.T_UI_ICN_Upgrade_10
        else if (UpgradeIndex = 4):
            set UpgradeWidget.CurrentHighlightedUpgradeTexture = UserInterfaces.UIFeatureExamples.Menu.Textures.UpgradeIcons.T_UI_ICN_Upgrade_12
        else if (UpgradeIndex = 5):
            set UpgradeWidget.CurrentHighlightedUpgradeTexture = UserInterfaces.UIFeatureExamples.Menu.Textures.UpgradeIcons.T_UI_ICN_Upgrade_08
        else if (UpgradeIndex = 6):
            set UpgradeWidget.CurrentHighlightedUpgradeTexture = UserInterfaces.UIFeatureExamples.Menu.Textures.UpgradeIcons.T_UI_ICN_Upgrade_02
        else if (UpgradeIndex = 7):
            set UpgradeWidget.CurrentHighlightedUpgradeTexture = UserInterfaces.UIFeatureExamples.Menu.Textures.UpgradeIcons.T_UI_ICN_Upgrade_11
        else if (UpgradeIndex = 8):
            set UpgradeWidget.CurrentHighlightedUpgradeTexture = UserInterfaces.UIFeatureExamples.Menu.Textures.UpgradeIcons.T_UI_ICN_Upgrade_06
        else if (UpgradeIndex = 9):
            set UpgradeWidget.CurrentHighlightedUpgradeTexture = UserInterfaces.UIFeatureExamples.Menu.Textures.UpgradeIcons.T_UI_ICN_Upgrade_04
        else if (UpgradeIndex = 10):
            set UpgradeWidget.CurrentHighlightedUpgradeTexture = UserInterfaces.UIFeatureExamples.Menu.Textures.UpgradeIcons.T_UI_ICN_Upgrade_01
        else if (UpgradeIndex = 11):
            set UpgradeWidget.CurrentHighlightedUpgradeTexture = UserInterfaces.UIFeatureExamples.Menu.Textures.UpgradeIcons.T_UI_ICN_Upgrade_09
        else if (UpgradeIndex = 12):
            set UpgradeWidget.CurrentHighlightedUpgradeTexture = UserInterfaces.UIFeatureExamples.Menu.Textures.UpgradeIcons.T_UI_ICN_Upgrade_05

    OnBack<override>()<suspends>:void=
        Session.GoToMainMenu()
