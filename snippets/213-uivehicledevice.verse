# Knowledge-neutral Verse snippet; project names and asset IDs generalized.
# id: snippet-213

using { /UnrealEngine.com/Temporary/SpatialMath }
using { /Fortnite.com/Devices }
using { /Verse.org/Simulation }
using { /UnrealEngine.com/Temporary/Diagnostics }
using { /Fortnite.com/Vehicles }
using { /Fortnite.com/Characters }
using { /UnrealEngine.com/Temporary/UI }


ui_vehicle_device := class(creative_device):
    @editable
    Volume1 : volume_device = volume_device{}
    @editable
    Volume2 : volume_device = volume_device{}
    @editable
    Volume3 : volume_device = volume_device{} 
    @editable    
    VolumeStop : volume_device = volume_device{}
    @editable
    VolumeStop2: volume_device = volume_device{}
    @editable 
    BikeSpawner : vehicle_spawner_sportbike_device = vehicle_spawner_sportbike_device{}

    # Main UI Widget
    var MyUI : overlay = overlay {}

    # A map to store each player's UI widget reference so it can be removed later
    var MaybeMyUIPerPlayer : [player]?overlay = map{}       

    # Gameplay variables
    var MaxSpeed : float = 90.00
    var OldBoost : float = 0.0

    # UI Widget Asset Examples
    var SpeedometerWidgetType1 : UserInterfaces.VerseUIExamples.Vehicle.Widgets.UW_Vehicle_Speedometer01 = UserInterfaces.VerseUIExamples.Vehicle.Widgets.UW_Vehicle_Speedometer01 {}
    var SpeedometerWidgetType2 : UserInterfaces.VerseUIExamples.Vehicle.Widgets.UW_Vehicle_Speedometer02 = UserInterfaces.VerseUIExamples.Vehicle.Widgets.UW_Vehicle_Speedometer02 {}
    var BoostmeterWidgetType1 : UserInterfaces.VerseUIExamples.Vehicle.Widgets.UW_Vehicle_Speedometer01_Boost = UserInterfaces.VerseUIExamples.Vehicle.Widgets.UW_Vehicle_Speedometer01_Boost {}   
    var SpeedometerWidgetType3 : UserInterfaces.VerseUIExamples.Vehicle.Widgets.UW_Vehicle_Speedometer03 = UserInterfaces.VerseUIExamples.Vehicle.Widgets.UW_Vehicle_Speedometer03 {} 

    # Runs when the device is started in a running game
    OnBegin<override>()<suspends>:void=

        set MyUI = overlay {}

        # Subscribe to Volume events
        Volume1.AgentEntersEvent.Subscribe(SetUIMode1)
        Volume2.AgentEntersEvent.Subscribe(SetUIMode2)
        Volume3.AgentEntersEvent.Subscribe(SetUIMode3)
        VolumeStop.AgentEntersEvent.Subscribe(RemoveUI)
        VolumeStop2.AgentEntersEvent.Subscribe(RemoveUI)
        
        # Set gameplay variables
        set SpeedometerWidgetType1.MaxSpeed = MaxSpeed
        set SpeedometerWidgetType2.MaxSpeed = MaxSpeed

    # Example HUD Version 1
    SetUIMode1(InAgent:agent):void =

        # Create UI
        MyHudUI := overlay:
            Slots := array:
                overlay_slot:
                    VerticalAlignment := vertical_alignment.Bottom
                    HorizontalAlignment := horizontal_alignment.Right                                       
                    Widget := stack_box:
                        Orientation := orientation.Vertical
                        Slots := array:
                            stack_box_slot:
                                Widget := BoostmeterWidgetType1
                                Padding := margin{Left:=0.0, Top:=0.0, Right:=0.0, Bottom:=-100.0}
                                HorizontalAlignment := horizontal_alignment.Center                         
                            stack_box_slot:
                                Widget := SpeedometerWidgetType1
                                Padding := margin{Left:=0.0, Top:=0.0, Right:=0.0, Bottom:=0.0}
                                HorizontalAlignment := horizontal_alignment.Center
        set MyUI = MyHudUI
        RemoveUI(InAgent)
        AssignUI(option{InAgent})


    # Example HUD Version 2
    SetUIMode2(InAgent:agent):void =
        set SpeedometerWidgetType2.MaxSpeed = MaxSpeed

        # Create UI
        MyHudUI := overlay:
            Slots := array:
                overlay_slot:
                    VerticalAlignment := vertical_alignment.Bottom
                    HorizontalAlignment := horizontal_alignment.Center                                       
                    Widget := stack_box:
                        Orientation := orientation.Vertical
                        Slots := array:
                            stack_box_slot:
                                Widget := SpeedometerWidgetType2
                                Padding := margin{Left:=0.0, Top:=0.0, Right:=0.0, Bottom:=0.0}
                                HorizontalAlignment := horizontal_alignment.Center
                                
        set MyUI = MyHudUI   
        RemoveUI(InAgent)
        AssignUI(option{InAgent})
           
# Example HUD Version 3
    SetUIMode3(InAgent:agent):void =
        set SpeedometerWidgetType3.MaxSpeed = MaxSpeed

        # Create UI
        MyHudUI := overlay:
            Slots := array:
                overlay_slot:
                    HorizontalAlignment := horizontal_alignment.Right
                    VerticalAlignment := vertical_alignment.Center                               
                    Widget := stack_box:
                        Orientation := orientation.Vertical
                        Slots := array:
                            stack_box_slot:
                                Widget := SpeedometerWidgetType3
                                Padding := margin{Left:=0.0, Top:=0.0, Right:=24.0, Bottom:=0.0}
                                HorizontalAlignment := horizontal_alignment.Center
                                VerticalAlignment := vertical_alignment.Center
        set MyUI = MyHudUI   
        RemoveUI(InAgent)
        AssignUI(option{InAgent})

    # HUD Update loop
    UpdateUI(InPlayer:agent)<suspends>:void=
        if:
            FortCharacter := InPlayer.GetFortCharacter[]
        then:        
            # This race expression runs two blocks concurrently. The first block to complete cancels the other.
            race:               
                # Block 1: done on tick
                block:                                        
                    loop:
                        if (CurrentVehicle := FortCharacter.GetVehicle[]):

                            # Get vehicle speed, Boost, Boost Capacity, and Health
                            Speed : float := Abs(CurrentVehicle.Speed)
                            Boost : float = CurrentVehicle.BoostRemaining? or 0.0  
                            BoostCapacity : float = CurrentVehicle.BoostCapacity? or 0.0  
                            MaxHealth := CurrentVehicle.GetMaxHealth()
                            Health := CurrentVehicle.GetHealth()                                                         
                                                        
                            # Update example 1 UI
                            set BoostmeterWidgetType1.Boost = Boost
                            set SpeedometerWidgetType1.Speed = Speed

                            # Update example 2 UI
                            set SpeedometerWidgetType2.Boost = Boost * 100 
                            set SpeedometerWidgetType2.Speed = Speed
                            if (OldBoost > Boost):
                                set SpeedometerWidgetType2.IsBoosting = true
                            else:
                                set SpeedometerWidgetType2.IsBoosting = false
                            set OldBoost = Boost  

                            # Update example 3 UI
                            set SpeedometerWidgetType3.Speed = Speed     
                            set SpeedometerWidgetType3.Boost = Boost * 100
                            set SpeedometerWidgetType3.BoostCapacity = BoostCapacity
                            set SpeedometerWidgetType3.Health = Health   
                            set SpeedometerWidgetType3.MaxHealth = MaxHealth                      

                        Sleep(0.0) # Wait for the next frame 
                # Block 2: Acts as an exit condition - when player actiavtes the Volume the Await() completes, canceling Block 1 and breaking the outer loop iteration.
                VolumeStop.AgentEntersEvent.Await() 
        else:
            Print("unable to initialize")

    # Function to add the UI to a specific player
    AssignUI(MaybeAgent: ?agent): void =
        if (Agent := MaybeAgent?):
            RemoveUI
            if (Player := player[Agent]):
                # Get the player's UI system
                if (PlayerUI := GetPlayerUI[Player]):
                    # Add the widget and store a reference to it in the map
                    PlayerUI.AddWidget(MyUI)
                    # Store it in the map for later removal
                    if (set MaybeMyUIPerPlayer[Player] = option{MyUI}) {}
                    spawn { UpdateUI(Player) }  
        
    RemoveUI(Agent:agent): void =
        if (IfPlayer := player[Agent]):
            if (PlayerUI := GetPlayerUI[IfPlayer]):
                if (SomeUI := MaybeMyUIPerPlayer[IfPlayer]?):
                    PlayerUI.RemoveWidget(SomeUI)
