# Knowledge-neutral Verse snippet; project names and asset IDs generalized.
# id: snippet-214

# Copyright Epic Games, Inc. All Rights Reserved.

using { /Verse.org/Simulation }
using { /Verse.org/Input }
using { /Verse.org/Input/UI }
using { /Fortnite.com/Devices }
using { /Fortnite.com/Characters }
using { /Fortnite.com/UI }
using { /UnrealEngine.com/Temporary/UI }
using { /Verse.org/Assets }

# ============================================================
#  vehicle_select_page
#
#  Contains the vehicle data as well as the logic for switching active cameras
#  To add a new vehicle to the game, just add a new entry to the Vehicles list below.
# ============================================================

vehicle_data := struct:
    Name        : string
    Nickname    : string
    Description : string

    Speed        : float
    Acceleration : float
    Handling     : float
    Durability   : float

    IsLocked : logic

    Texture : texture = UserInterfaces.UIFeatureExamples.Menu.Textures.T_UI_Vehicle_Icon_White

ui_vehicle_select_page := class(ui_menu_page):

    VehicleWidget : UserInterfaces.UIFeatureExamples.Menu.Widgets.VehicleSelect.UW_Menu_VehicleSelect = UserInterfaces.UIFeatureExamples.Menu.Widgets.VehicleSelect.UW_Menu_VehicleSelect{}

    var Vehicles : []vehicle_data = array{}

    var CurrentVehicleIndex : int = 0

    var ActiveCamera : ?gameplay_camera_fixed_point_device = false

    # Current inspection angle: 0 Default, 1 Left, 2 Right.
    var CurrentAngleIndex : int = 0

    var IsInspectLocked : logic = false

    InspectLockSeconds : float = 0.6

    # True when the player is on touch/mobile.
    # On gamepad/keyboard, hovering a vehicle (HighlightVehicle) highlights it
    # On mobile the player must click a vehicle to highlight it instead
    var IsMobile : logic = false

    # Invisible buttons that capture Right/Left Trigger
    NextPageButton : button = button{
        TriggeringInputAction := option{ UI.NextPage },
        Slot := button_slot{ Widget := overlay{} }
    }
    PrevPageButton : button = button{
        TriggeringInputAction := option{ UI.PreviousPage },
        Slot := button_slot{ Widget := overlay{} }
    }

    var NextPageSubscription : ?cancelable = false
    var PrevPageSubscription : ?cancelable = false

    ContentOverlay : overlay = overlay{}

    GetTitle<override>():string = "Vehicle Select"
    GetContentWidget<override>():widget = ContentOverlay
    GetFocusWidget<override>():widget = VehicleWidget
    NeedTabButtons<override> : logic = true

    InitData<override>():void=
        ContentOverlay.AddWidget(overlay_slot{ Widget := VehicleWidget, HorizontalAlignment := horizontal_alignment.Fill, VerticalAlignment := vertical_alignment.Fill })
        ContentOverlay.AddWidget(overlay_slot{ Widget := NextPageButton })
        ContentOverlay.AddWidget(overlay_slot{ Widget := PrevPageButton })
        set Vehicles = array{

            vehicle_data{
                Name := "Ironclad",
                Nickname := "The Wall",
                Description := "A heavily armoured beast built for punishment. Slow but unstoppable.",
                Speed := 0.4,
                Acceleration := 0.3,
                Handling := 0.5,
                Durability := 0.95,
                IsLocked:= false,
                Texture := UserInterfaces.UIFeatureExamples.Menu.Textures.T_UI_Vehicle_Icon_Red
            },

            vehicle_data{
                Name := "Phantom",
                Nickname := "Ghost Runner",
                Description := "A featherweight racer stripped to the essentials. Blink and it's gone.",
                Speed := 0.95,
                Acceleration := 0.9,
                Handling := 0.8,
                Durability := 0.35,
                IsLocked := false,
                Texture := UserInterfaces.UIFeatureExamples.Menu.Textures.T_UI_Vehicle_Icon_White
            },

            vehicle_data{
                Name := "Bruiser",
                Nickname := "Mean Machine",
                Description := "Balanced muscle with a mean shoulder charge. Hits hard, takes hits harder.",
                Speed := 0.7,
                Acceleration := 0.65,
                Handling := 0.6,
                Durability:= 0.8,
                IsLocked := false,
                Texture := UserInterfaces.UIFeatureExamples.Menu.Textures.T_UI_Vehicle_Icon_Blue
            }
        }
        UpdateVehicleDisplay()

    SubscribeToPageEvents<override>():void=
        ConfirmSub := SubscribeToEvent(VehicleWidget.ConfirmSelection, OnConfirmSelection)
        HighlightSub := SubscribeToEvent(VehicleWidget.HighlightVehicle, OnHighlightVehicle)
        SelectSub := SubscribeToEvent(VehicleWidget.SelectVehicle, OnVehicleClicked)
        LeftInspectSub := SubscribeToEvent(VehicleWidget.LeftInspectClicked, OnLeftInspectClicked)
        RightInspectSub := SubscribeToEvent(VehicleWidget.RightInspectClicked, OnRightInspectClicked)
        set PageSubscriptions += array{ ConfirmSub, HighlightSub, SelectSub, LeftInspectSub, RightInspectSub }

    OnShow<override>():void=
        set IsMobile = if (Session.IsMobile?) then true else false
        set CurrentAngleIndex = 0
        set IsInspectLocked = false
        ShowCameraForVehicle(CurrentVehicleIndex)
        NextSub := NextPageButton.OnClick().Subscribe(OnNextPageClicked)
        PrevSub := PrevPageButton.OnClick().Subscribe(OnPrevPageClicked)
        set NextPageSubscription = option{ NextSub }
        set PrevPageSubscription = option{ PrevSub }

    OnHide<override>():void=
        ResetToDefault()
        RemoveActiveCamera()
        CancelSubscription(NextPageSubscription)
        CancelSubscription(PrevPageSubscription)
        set NextPageSubscription = false
        set PrevPageSubscription = false

    OnNextTab<override>():void=
        VehicleCount := Vehicles.Length
        if (VehicleCount > 0):
            MaybeNext := FindNextUnlocked(CurrentVehicleIndex, 1, VehicleCount)
            if (Next := MaybeNext?):
                set CurrentVehicleIndex = Next
                ReleaseSequenceView()
                ShowCameraForVehicle(CurrentVehicleIndex)
                UpdateVehicleDisplay()

    OnPrevTab<override>():void=
        VehicleCount := Vehicles.Length
        if (VehicleCount > 0):
            MaybePrev := FindNextUnlocked(CurrentVehicleIndex, -1, VehicleCount)
            if (Prev := MaybePrev?):
                set CurrentVehicleIndex = Prev
                ReleaseSequenceView()
                ShowCameraForVehicle(CurrentVehicleIndex)
                UpdateVehicleDisplay()

    # Walks the array in the given direction (1 or -1), warpping until it finds an unlocked vehicle
    FindNextUnlocked(FromIndex : int, Direction : int, Count : int) : ?int=
        var Steps : int = 1
        var Result : ?int = false
        loop:
            if (Steps >= Count):
                break
            CandidateRaw := FromIndex + (Direction * Steps)
            Candidate := if (CandidateRaw < 0) then CandidateRaw + Count else if (R := Mod[CandidateRaw, Count]) then R else 0
            if (Data := Vehicles[Candidate], Data.IsLocked = false):
                set Result = option{Candidate}
                break
            set Steps = Steps + 1
        Result

    # Called when the player hovers a vehicle directly in the widget.
    # Hover only exists on gamepad/keyboard, ignored on mobile.
    OnHighlightVehicle(Index : tuple(int)):void=
        if (IsMobile = false):
            HighlightVehicleAtIndex(Index(0))

    # Called when the player clicks/taps a vehicle in the widget.
    # Only highlights on mobile.
    OnVehicleClicked(Index : tuple(int)):void=
        if (IsMobile?):
            HighlightVehicleAtIndex(Index(0))

    # Highlights the vehicle at the given index, switching the camera and pushing its data to the widget.
    HighlightVehicleAtIndex(Index : int):void=
        if (Data := Vehicles[Index]):
            if (Data.IsLocked = false):
                set CurrentVehicleIndex = Index
                ReleaseSequenceView()
                ShowCameraForVehicle(CurrentVehicleIndex)
                UpdateVehicleDisplay()

    UpdateVehicleDisplay():void=
        if (Data := Vehicles[CurrentVehicleIndex]):
            set VehicleWidget.CurrentHighlightedVehicle= CurrentVehicleIndex
            set VehicleWidget.CurrentHighlightedVehicleName= StringToMessage(Data.Name)
            set VehicleWidget.CurrentHighlightedVehicleNickname = StringToMessage(Data.Nickname)
            set VehicleWidget.CurrentHighlightedVehicleDescription = StringToMessage(Data.Description)
            set VehicleWidget.Stat1Value = Data.Speed
            set VehicleWidget.Stat2Value = Data.Acceleration
            set VehicleWidget.Stat3Value = Data.Handling
            set VehicleWidget.Stat4Value = Data.Durability

        for (Index -> VehicleEntry : Vehicles):
            if (Index = 0):
                set VehicleWidget.Vehicle1Texture = VehicleEntry.Texture
            else if (Index = 1):
                set VehicleWidget.Vehicle2Texture = VehicleEntry.Texture
            else if (Index = 2):
                set VehicleWidget.Vehicle3Texture = VehicleEntry.Texture
            else if (Index = 3):
                set VehicleWidget.Vehicle4Texture = VehicleEntry.Texture
            else if (Index = 4):
                set VehicleWidget.Vehicle5Texture = VehicleEntry.Texture

    # Switches the player's view to the given vehicle's camera
    ShowCameraForVehicle(Index : int):void=
        if (Cam := Session.MenuDevice.VehicleCameras[Index]):
            if (Prev := ActiveCamera?):
                Prev.RemoveFrom(Session.Player)
            Cam.AddTo(Session.Player)
            set ActiveCamera = option{ Cam }

    RemoveActiveCamera():void=
        if (Cam := ActiveCamera?):
            Cam.RemoveFrom(Session.Player)
        set ActiveCamera = false

    ReleaseSequenceView():void=
        Session.MenuDevice.VehicleCamSeqLeft.Stop(Session.Player)
        Session.MenuDevice.VehicleCamSeqRight.Stop(Session.Player)
        set IsInspectLocked = false

    ResetToDefault():void=
        ReleaseSequenceView()
        set CurrentAngleIndex = 0

    InspectRight():void=
        spawn{ DoInspect(true) }

    InspectLeft():void=
        spawn{ DoInspect(false) }

    DoInspect(GoRight : logic)<suspends>:void=
        if (IsInspectLocked = false):
            var DidPlay : logic = false
            if (GoRight?):
                if (CurrentAngleIndex = 1):
                    set IsInspectLocked = true
                    set DidPlay = true
                    set CurrentAngleIndex = 0
                    Session.MenuDevice.VehicleCamSeqLeft.PlayReverse(Session.Player)
                else if (CurrentAngleIndex = 0):
                    set IsInspectLocked = true
                    set DidPlay = true
                    set CurrentAngleIndex = 2
                    Session.MenuDevice.VehicleCamSeqRight.Play(Session.Player)
            else:
                if (CurrentAngleIndex = 2):
                    set IsInspectLocked = true
                    set DidPlay = true
                    set CurrentAngleIndex = 0
                    Session.MenuDevice.VehicleCamSeqRight.PlayReverse(Session.Player)
                else if (CurrentAngleIndex = 0):
                    set IsInspectLocked = true
                    set DidPlay = true
                    set CurrentAngleIndex = 1
                    Session.MenuDevice.VehicleCamSeqLeft.Play(Session.Player)
            if (DidPlay?):
                Sleep(InspectLockSeconds)
                set IsInspectLocked = false

    OnNextPageClicked(Message : widget_message):void=
        InspectRight()

    OnPrevPageClicked(Message : widget_message):void=
        InspectLeft()

    OnRightInspectClicked(Event : tuple()):void=
        InspectRight()

    OnLeftInspectClicked(Event : tuple()):void=
        InspectLeft()

    OnConfirmSelection(Event : tuple()):void=
        spawn{ OnSelect() }

    OnSelect<override>()<suspends>:void=
        # Disable Virtual Pointer as this is the last page in the menu
        DisableVP()
        Session.GoToCountdown()

    OnBack<override>()<suspends>:void=
        Session.GoToGameMode()

    DisableVP() : void =
        if (PlayerInput := GetPlayerInput[Session.Player]):
            PlayerInput.RemoveInputMapping(TouchMapping)
