# Knowledge-neutral Verse snippet; project names and asset IDs generalized.
# id: snippet-215

# Copyright Epic Games, Inc. All Rights Reserved.

using { /Fortnite.com/Devices }
using { /Fortnite.com/UI }
using { /UnrealEngine.com/Temporary/UI }
using { /UnrealEngine.com/Temporary/SpatialMath }
using { /Verse.org/Simulation }

ui_verse_events_device := class(creative_device):

    @editable
    ShowVolume : volume_device = volume_device{}

    @editable
    DoorSequence : cinematic_sequence_device = cinematic_sequence_device{}

    Widget : UserInterfaces.VerseUIExamples.VerseEvents.Widgets.UW_VerseFields_Events = UserInterfaces.VerseUIExamples.VerseEvents.Widgets.UW_VerseFields_Events{}

    var SubscriptionOpenDoor : ?cancelable_event = false
    var SubscriptionCloseDoor : ?cancelable_event = false
    var SubscriptionExit : ?cancelable_event = false
    var DoorOpen : logic = false
    
    var CurrentPlayer : ?player = false

    TextDoorOpen<localizes> : message = "Door is now open."
    TextDoorClosed<localizes> : message = "Door is now closed."

    OnBegin<override>() : void =
        ShowVolume.AgentEntersEvent.Subscribe(OnAgentEntered)
        GetPlayspace().PlayerRemovedEvent().Subscribe(OnPlayerLeft)

    OnAgentEntered(Agent : agent) : void =
        if (Player := player[Agent]):
            if (CurrentPlayer?):
                return
            
            OpenMenu(Player)

    OnPlayerLeft(Player : player) : void =
        if (ExistingPlayer := CurrentPlayer?, ExistingPlayer = Player):
            CloseMenu()

    OpenMenu(Player : player) : void =
        if (PlayerUI := GetPlayerUI[Player]):
            
            set Widget.StatusText = TextDoorClosed

            PlayerUI.AddWidget(
                Widget, 
                player_ui_slot{ InputMode := ui_input_mode.All }
            )
            
            PlayerUI.SetFocus(Widget)
            
            set CurrentPlayer = option{Player}

            SubOpen := SubscribeToEvent(Widget.OpenDoorEvent, OnOpenDoorClicked)
            set SubscriptionOpenDoor = option{SubOpen}

            SubClose := SubscribeToEvent(Widget.CloseDoorEvent, OnCloseDoorClicked)
            set SubscriptionCloseDoor = option{SubClose}

            SubExit := SubscribeToEvent(Widget.ExitEvent, OnExitClicked)
            set SubscriptionExit = option{SubExit}

    OnOpenDoorClicked(Args : tuple()) : void =
        if (DoorOpen = false):
            set Widget.StatusText = TextDoorOpen
            DoorSequence.Play()
            set DoorOpen = true
        else:
            TriggerInvalidAnimation()

    OnCloseDoorClicked(Args : tuple()) : void =
        if (DoorOpen = true):
            set Widget.StatusText = TextDoorClosed
            DoorSequence.PlayReverse()
            set DoorOpen = false
        else:
            TriggerInvalidAnimation()

    TriggerInvalidAnimation() : void =
        if (Widget.TriggerInvalid?):
            set Widget.TriggerInvalid = false
        else:
            set Widget.TriggerInvalid = true


    OnExitClicked(Args : tuple()) : void =
        CloseMenu()

    CloseMenu() : void =
        if (Player := CurrentPlayer?, PlayerUI := GetPlayerUI[Player]):
            PlayerUI.RemoveWidget(Widget)
        
        if (SubOpen := SubscriptionOpenDoor?):
            SubOpen.Cancel()
        
        if (SubClose := SubscriptionCloseDoor?):
            SubClose.Cancel()

        if (SubExit := SubscriptionExit?):
            SubExit.Cancel()

        if (DoorOpen = true):
            set Widget.StatusText = TextDoorClosed
            DoorSequence.PlayReverse()
            set DoorOpen = false
        
        set SubscriptionOpenDoor = false
        set SubscriptionCloseDoor = false
        set SubscriptionExit = false
        set CurrentPlayer = false
