# Knowledge-neutral Verse snippet; project names and asset IDs generalized.
# id: snippet-216

using { /Fortnite.com/Devices }
using { /Verse.org/Simulation }
using { /UnrealEngine.com/Temporary/Diagnostics }
using { /Fortnite.com/UI }
using { /UnrealEngine.com/Temporary/UI }
using { /UnrealEngine.com/Temporary/SpatialMath }
using { /Verse.org/Assets }
using { /Verse.org/Colors }
using { /Verse.org/Random }

# See https://dev.epicgames.com/documentation/en-us/uefn/create-your-own-device-in-verse for how to create a verse device.

# A Verse-authored creative device that can be placed in a level
ui_verse_fields_demo_device := class(creative_device):

    @editable 
    VolumeShow : volume_device = volume_device{}    

    MyUI:overlay = overlay{ Slots := array{} }    
    DemoWidget : UserInterfaces.VerseUIExamples.VerseFields.Widgets.UW_VerseFields = UserInterfaces.VerseUIExamples.VerseFields.Widgets.UW_VerseFields{}
    HeartMaterial : UserInterfaces.VerseUIExamples.CustomButton.MaterialInstances.MI_UI_Icon = UserInterfaces.VerseUIExamples.CustomButton.MaterialInstances.MI_UI_Icon{}

    # Variable to track how many times the loop has run
    var LoopCount : int = 0    
    var IsActive : logic = false    

    # Runs when the device is started in a running game
    OnBegin<override>()<suspends>:void=
        set LoopCount = 0  
        # Subscribe to triger device event to show the UI demo
        VolumeShow.AgentEntersEvent.Subscribe(ShowToAllPlayers)
        VolumeShow.AgentExitsEvent.Subscribe(HideFromAllPlayers)

    ShowToAllPlayers(InAgent:agent):void =
        # Add widget
        MyUI.AddWidget(overlay_slot{ Widget := DemoWidget, HorizontalAlignment := horizontal_alignment.Center, VerticalAlignment := vertical_alignment.Center })
        set DemoWidget.MaterialField = HeartMaterial
        set LoopCount = 0
        set DemoWidget.TextField = StringToMessage("Counter: 0")
        set DemoWidget.IntField = 0
        set DemoWidget.FloatField = 0.0
        set DemoWidget.LogicField = false
        set IsActive = true
        spawn { ChangeData() }
        
        # Show the main layout container to all players
        for:
            Player : GetPlayspace().GetPlayers()
            PlayerUI := GetPlayerUI[Player]
        do:
            PlayerUI.AddWidget(MyUI)   

    ChangeData()<suspends>:void =
        loop:
            # Wait for the specified delay
            Sleep(1.0) #

            # Code to execute every loop iteration goes here
            if (IsActive = false):
                break
            set LoopCount += 1

            # Change UI Widget's text field
            time : string = "Counter: {LoopCount}"
            set DemoWidget.TextField = StringToMessage(time)

            # Change UI Widget's int field
            set DemoWidget.IntField += 1
            
            # Change UI Widget's float field
            set DemoWidget.FloatField += 0.1

            # Change UI Widget's logic field
            if (GetRandomInt(0, 1) = 0):
                set DemoWidget.LogicField = true   
            else:
                set DemoWidget.LogicField = false   
            
            # Condition to break the loop
            if (LoopCount >= 10):
                # Hide the UI when the timer hits 20
                for:
                    Player : GetPlayspace().GetPlayers()
                    PlayerUI := GetPlayerUI[Player]
                do:
                    PlayerUI.RemoveWidget(MyUI)
                set IsActive = false
                break # Exits the loop

    HideFromAllPlayers(InAgent:agent):void =
        set IsActive = false
        for:
            Player : GetPlayspace().GetPlayers()
            PlayerUI := GetPlayerUI[Player]
        do:
            PlayerUI.RemoveWidget(MyUI)
