# Knowledge-neutral Verse snippet; project names and asset IDs generalized.
# id: snippet-217

using { /Fortnite.com/Devices }
using { /Verse.org/Simulation }
using { /Verse.org/Input }
using { /Verse.org/Input/UI }
using { /UnrealEngine.com/Temporary/UI }
using { /Verse.org/SpatialMath }

ui_vp_device := class(creative_device):

    @editable
    HUDController : hud_controller_device = hud_controller_device {}

    @editable
    Trigger : trigger_device = trigger_device {}

    # Max seconds between two taps to count as a swipe
    @editable
    FlickWindow : float = 0.4

    # Minimum screen-space distance between two taps to classify as a swipe
    @editable
    SwipeMinDistance : float = 0.08

    KeyboardInputText<localizes><public> : message = "Mouse & Keyboard"
    GamepadInputText<localizes><public> : message = "Gamepad"
    TouchInputText<localizes><public> : message = "Touchscreen"
    MessageText<localizes><public> : message = "Action clicked!"

    MenuWidget : UserInterfaces.VerseUIExamples.VirtualPointer.Widgets.UW_VP_Menu = UserInterfaces.VerseUIExamples.VirtualPointer.Widgets.UW_VP_Menu {}
    CoordinatesWidget : UserInterfaces.VerseUIExamples.VirtualPointer.Widgets.UW_VP_ScreenCoordinates = UserInterfaces.VerseUIExamples.VirtualPointer.Widgets.UW_VP_ScreenCoordinates {}

    var MyUI : overlay = overlay {}
    var MyCoordinatesUI : overlay = overlay {}

    var IsAnimationPlaying : logic = false
    var EnabledVP : logic = false

    var CurrentPlayer : ?player = false

    var LastTapPos : ?vector3 = false
    var LastTapTime : float = 0.0
    var SwipeTapSub : ?cancelable = false

    # Fired on swipe: (player, directionIndex 0-7, worldSwipeDir, screenEndPos)
    # 0=Up 1=TopRight 2=Right 3=BottomRight 4=Down 5=BottomLeft 6=Left 7=TopLeft
    SwipeEvent<public> : event(tuple(player, int, vector3, vector3)) = event(tuple(player, int, vector3, vector3)){}

    # Fired on tap (non-swipe): (player, screenPos)
    TapEvent<public> : event(tuple(player, vector3)) = event(tuple(player, vector3)){}

    var PlayerTriggersShowMenuSubscriptionOption : ?cancelable = false

    OnBegin<override>()<suspends>:void= 

        set PlayerTriggersShowMenuSubscriptionOption = option{Trigger.TriggeredEvent.Subscribe(OpenMenuFromTriggerDevice)}

        # Create UI
        NewMenuOverlay := overlay:
            Slots := array:
                overlay_slot:
                    VerticalAlignment := vertical_alignment.Fill
                    HorizontalAlignment := horizontal_alignment.Fill                                       
                    Widget := MenuWidget
        
        set MyUI = NewMenuOverlay

        NewHUDOverlay := overlay:
            Slots := array:
                overlay_slot:
                    VerticalAlignment := vertical_alignment.Fill
                    HorizontalAlignment := horizontal_alignment.Fill                                       
                    Widget := CoordinatesWidget
        
        set MyCoordinatesUI = NewHUDOverlay

        SubscribeToEvent(MenuWidget.OnAddVPClicked, OnAddVPClicked)
        SubscribeToEvent(MenuWidget.OnRemoveVPClicked, OnRemoveVPClicked)
        SubscribeToEvent(MenuWidget.OnCloseClicked, OnCloseClicked)
        SubscribeToEvent(CoordinatesWidget.OnActionClicked, OnActionClicked)
        SubscribeToEvent(CoordinatesWidget.OnExitClicked, OnExitClicked)

        set EnabledVP = false

    OnEnd<override>():void = 
        CancelSubscription(PlayerTriggersShowMenuSubscriptionOption)
        CancelSubscription(SwipeTapSub)

        set CurrentPlayer = false
    
    DisableDevices(Agent : agent) : void =
        if (Player := player[Agent]):
            HUDController.Disable()

    # Enable/Disable Virtual Pointer by adding/removing "TouchMapping"
    EnableVP(Player: player) : void =
        if (PlayerInput := GetPlayerInput[Player]):
            PlayerInput.AddInputMapping(TouchMapping)
            TapSub := PlayerInput.GetInputEvents(PointerSelect).TriggerActivationEvent.Subscribe(OnPointerTap)
            set SwipeTapSub = option{TapSub}
        else:
        set MenuWidget.IsEnabledVP = true
        set CoordinatesWidget.ShowUI = true
        set EnabledVP = true

    DisableVP(Player : player) : void =
        if (PlayerInput := GetPlayerInput[Player]):
            PlayerInput.RemoveInputMapping(TouchMapping)
        set MenuWidget.IsEnabledVP = false
        set CoordinatesWidget.ShowUI = false
        set EnabledVP = false
        CancelSubscription(SwipeTapSub)

    OpenMenuFromTriggerDevice(Agent : ?agent) : void =
        if (Player := player[Agent?]):
            OpenMenu(Player)
            HUDController.Enable()
            if (PlayerUI := GetPlayerUI[Player]):
                PlayerUI.AddWidget(MyCoordinatesUI)

            set CoordinatesWidget.ShowUI = false

    OpenMenu(Player : player) : void =
        if (PlayerInput := GetPlayerInput[Player]):
            case (PlayerInput.PreferredInputMethod):
                input_method.Gamepad => 
                    set MenuWidget.CurrentInput = GamepadInputText
                input_method.Touch => 
                    set MenuWidget.CurrentInput = TouchInputText
                input_method.KeyboardAndMouse => 
                    set MenuWidget.CurrentInput = KeyboardInputText
                _=>            

        if (PlayerUI := GetPlayerUI[Player]):
            PlayerUI.AddWidget(MyUI, player_ui_slot{ InputMode := ui_input_mode.All })

            set CurrentPlayer = option{Player}
        
        set CoordinatesWidget.MessageText = StringToMessage("")

    OnAddVPClicked() : void = 
        if(Player:=CurrentPlayer?):
            EnableVP(Player)

    OnRemoveVPClicked() : void = 
        if(Player:=CurrentPlayer?):
            DisableVP(Player)

    OnCloseClicked() : void =
        if(Player:=CurrentPlayer?):
            if (PlayerUI := GetPlayerUI[Player]):
                PlayerUI.RemoveWidget(MyUI)
                if (EnabledVP = false):
                    PlayerUI.RemoveWidget(MyCoordinatesUI)
                    set CoordinatesWidget.ShowButtons = false
                else:
                    set CoordinatesWidget.ShowButtons = true

    OnActionClicked() : void =
        if (IsAnimationPlaying = false):
            spawn {ShowMessageAnim()}

    ShowMessageAnim()<suspends> : void =
        loop:
            if(IsAnimationPlaying = false):   
                set IsAnimationPlaying = true
                set CoordinatesWidget.MessageText = MessageText   
                Sleep(2.2)
                set CoordinatesWidget.MessageText = StringToMessage("")      
            else:
                set IsAnimationPlaying = false
                break
            Sleep(0.0)            

    OnExitClicked() : void =
        if(Player:=CurrentPlayer?):
            OpenMenu(Player)
            set CoordinatesWidget.ShowButtons = false

    OnPointerTap(Arg : tuple(player, vector3)) : void =
        Player := Arg(0)
        TapPos := Arg(1)
        CurrentTime := GetSimulationElapsedTime()
        set CoordinatesWidget.X_val = TapPos.Left
        set CoordinatesWidget.Y_val = TapPos.Up
        var IsSwipe : logic = false
        if (LastPos := LastTapPos?):
            DL := TapPos.Left - LastPos.Left
            DU := TapPos.Up - LastPos.Up
            DLSq := DL * DL
            DUSq := DU * DU
            TimeDelta := CurrentTime - LastTapTime
            if (TimeDelta < FlickWindow, DLSq + DUSq > SwipeMinDistance * SwipeMinDistance):
                set IsSwipe = true
                DirIndex := ClassifySwipe(DL, DU)
                StartRay := Player.DeprojectViewportToWorld(vector3{Left := LastPos.Left, Up := LastPos.Up, Forward := 0.0})
                EndRay := Player.DeprojectViewportToWorld(vector3{Left := TapPos.Left, Up := TapPos.Up, Forward := 0.0})
                SwipeWorldDir := EndRay.Direction - StartRay.Direction
                SwipeLen := SwipeWorldDir.Length()
                if (SwipeLen > 0.0001):
                    NormDir := SwipeWorldDir * (1.0 / SwipeLen)
                    SwipeEvent.Signal((Player, DirIndex, NormDir, TapPos))
                set LastTapPos = false
        if (IsSwipe = false):
            TapEvent.Signal((Player, TapPos))
            set LastTapPos = option{TapPos}
            set LastTapTime = CurrentTime

    # Classifies a screen-space delta into 8 directions (0=Up, clockwise to 7=TopLeft)
    ClassifySwipe(DL : float, DU : float) : int =
        DLSq := DL * DL
        DUSq := DU * DU
        var Dir : int = 0
        # Cardinal if one axis dominates by tan(22.5°)^2 ≈ 0.172
        if (DLSq < 0.172 * DUSq):
            if (DU > 0.0):
                set Dir = 0  # Up
            else:
                set Dir = 4  # Down
        else if (DUSq < 0.172 * DLSq):
            if (DL > 0.0):
                set Dir = 2  # Right
            else:
                set Dir = 6  # Left
        else:
            if (DU > 0.0):
                if (DL > 0.0):
                    set Dir = 1  # TopRight
                else:
                    set Dir = 7  # TopLeft
            else:
                if (DL > 0.0):
                    set Dir = 3  # BottomRight
                else:
                    set Dir = 5  # BottomLeft
        Dir
