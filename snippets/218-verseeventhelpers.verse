# Knowledge-neutral Verse snippet; project names and asset IDs generalized.
# id: snippet-218

# Copyright Epic Games, Inc. All Rights Reserved.

using { /Verse.org/Simulation }

cancelable_event := class:
    Event:event() = event(){}

    Cancel():void=
        Event.Signal()
    Await()<suspends>:void=
        Event.Await()

AwaitForEvent(Event:event(t), Callback(:t):void, Cancelable:cancelable_event where t : type)<suspends>:void=
    race:
        loop:
            Payload := Event.Await()
            Callback(Payload)
        Cancelable.Await()

SubscribeToEvent(Event:event(t), Callback(:t):void where t : type):cancelable_event=
    Cancelable := cancelable_event{}
    spawn{ AwaitForEvent(Event, Callback, Cancelable) }
    Cancelable

StringToMessage<localizes>(value:string) : message = "{value}"

UnknownMsg<localizes> : message = "Unknown"

GetUsername<localizes>(Agent : agent) : message = "{Agent}"

CancelSubscription(SubscriptionOption : ?cancelable):void=
    if (Subscription := SubscriptionOption?):
        Subscription.Cancel()

(Map:[t]u where t:subtype(comparable), u:type ).ToArray()<transacts>:[]tuple(t, u)=
    for (A -> B : Map):
        (A, B)
