# Knowledge-neutral Verse snippet; project names and asset IDs generalized.
# id: snippet-219

# This file contains the data representation of all the commands that the NPC can receive
# and utilities for the data to help with debugging and troubleshooting issues.

# Each type of command that the NPC can perform will be an instance of this class.
# The class has the computes specifier to be able to instantiate it at module-scope.
# The class has the abstract specifier so it cannot be instantiated directly, and
# requires subclasses to implement any non-initialized functions, like DebugString().
command := class<computes><abstract>:
    DebugString():string

# The Commands module contains definitions of commands that the NPC can perform.
# This example uses instances of command subclasses instead of the enum type
# so you can add more commands after the initial published version of the project.
Commands := module:
    # The following are subclasses of the command class
    # that implement the abstract command class and define what commands are valid.
    # Each has their own implementation of DebugString(), for example,
    # so when you print a command value it has the correct string associated with it.
    # Since the command class is abstract, it means these subclasses are the only valid commands,
    # otherwise there will be a compiler error.
    forward_command<public> := class<computes>(command):
        DebugString<override>():string = "Forward"

    turn_command<public> := class<computes><abstract>(command):

    turn_right_command<public> := class<computes>(turn_command):
        DebugString<override>():string = "TurnRight"

    turn_left_command<public> := class<computes>(turn_command):
        DebugString<override>():string = "TurnLeft"
        
    # Instances of each command type that can be used in the minigame.
    Forward<public>:forward_command = forward_command{}
    TurnRight<public>:turn_right_command = turn_right_command{}
    TurnLeft<public>:turn_left_command = turn_left_command{} 

# Convert the command data to a string to be able to print it when debugging issues.
# For example, Print("Player selected {Command} command.")
ToString(Command:command):string=
    Command.DebugString()
