# Knowledge-neutral Verse snippet; project names and asset IDs generalized.
# id: snippet-220

using { /Fortnite.com/Characters }
using { /Fortnite.com/Devices }
using { /Fortnite.com/FortPlayerUtilities }
using { /Fortnite.com/UI }
using { /UnrealEngine.com/Temporary/Diagnostics }
using { /UnrealEngine.com/Temporary/SpatialMath}
using { /UnrealEngine.com/Temporary/UI }
using { /Verse.org/Simulation }
using { /Verse.org/Simulation/Tags } 

# Gameplay tag for objects that should be displayed or hidden during cinematic
cinematic_objects := class(tag):

# Script that sets the player into a cinematic mode so its not in the cinematic, and then resets them when the cinematic is done
billboard_cinematic := class<concrete>:

    # A reference to the device that contains the level sequence to be played
    @editable
    CinematicSequence:cinematic_sequence_device := cinematic_sequence_device{}

    # A reference to the device that will hide the HUD
    @editable
    HUDController:hud_controller_device := hud_controller_device{}

    # Should the Skip Intro button be visible during the cinematic
    @editable
    ShowUI:logic := true

    # Should Devices and Billboards be visible during the cinematic
    @editable
    ShowCreativeObjects:logic := true

    # Should play the cinematic at the beginning of game
    @editable
    PlayAtStart:logic = true

    # Reference to the Radio that plays music
    @editable
    Music:radio_device := radio_device{}

    # Signaled when the cinematic has completed.
    CinematicCompleted<public>:event() = event(){}

    # Runs at the beginning of the script.
    EnterCinematicModeForAll<public>(InPlayers:[]player)<suspends>:void=
        if (not PlayAtStart?):
            HideObjectsInCinematic()
            return
        
        HUDController.Enable()

        if (ShowCreativeObjects?):
            ShowObjectsInCinematic()
        else:
            HideObjectsInCinematic()

        for:
            Player : InPlayers
            not Player.IsSpectator[]
        do:
            EnterCinematicMode(Player)
        
        Music.Play()
        
        CinematicSequence.StoppedEvent.Await()
        CinematicCompleted.Signal()
        EndCinematicModeForAll(InPlayers)

    # Sets players to visible, re-enables their movement, and removes the skip button.
    EndCinematicModeForAll<public>(InPlayers:[]player):void=
        Music.Stop()
        for:
            Player : InPlayers
            not Player.IsSpectator[]
        do:
            EndCinematicMode(Player)

        HUDController.Disable()
        HideObjectsInCinematic()

    # UI to show during the cinematic
    CinematicUI<private>:cinematic_ui = cinematic_ui{}

    # Handler for player choosing to skip cinematic
    HandleSkipIntroInteraction<private>(WidgetMessage:widget_message):void=
        EndCinematicMode(WidgetMessage.Player)

    # Sets players to invisible, removes their movement, optionally gives them a skip button, then waits for the cinematic to end before signaling that it has completed.
    EnterCinematicMode<private>(Player:player)<suspends>:void=
        HidePlayer(Player)

        if:
            ShowUI?
            PlayerUI := GetPlayerUI[Player]
        then:
            Listenable := CinematicUI.Show(Player, PlayerUI)
            Listenable.Subscribe(HandleSkipIntroInteraction)

        CinematicSequence.Play(Player)

    # Stop cinematic for player
    EndCinematicMode<private>(Player:player):void=
        CinematicSequence.Stop(Player)
        ShowPlayer(Player)
        CinematicUI.Hide(Player)

    # Make player invisible and unmoveable
    HidePlayer<private>(Player:player):void=
        if (FortCharacter := Player.GetFortCharacter[]):
            CinematicStasis := stasis_args{AllowTurning := false, AllowFalling := false, AllowEmotes := false}

            FortCharacter.PutInStasis(CinematicStasis)

    # Make player visible and moveable
    ShowPlayer<private>(Player:player):void=
        if (FortCharacter := Player.GetFortCharacter[]):
            FortCharacter.ReleaseFromStasis()

    # Find all objects tagged and make them visible
    ShowObjectsInCinematic<private>():void=
        ObjectsToShow := HUDController.FindCreativeObjectsWithTag(cinematic_objects{})

        for (Object : ObjectsToShow):
            if (Billboard := billboard_device[Object]):
                Billboard.ShowText()
            else if (Device := creative_device[Object]):
                Device.Show()
            else if (Radio := radio_device[Object]):
                Radio.Show()
            else if (AudioPlayer := audio_player_device[Object]):
                AudioPlayer.Show()

    # Find all objects tagged and make them invisible
    HideObjectsInCinematic<private>():void=
        ObjectsToHide := HUDController.FindCreativeObjectsWithTag(cinematic_objects{})

        for (Object : ObjectsToHide):
            if (Billboard := billboard_device[Object]):
                Billboard.HideText()
            else if (Device := creative_device[Object]):
                Device.Hide()
            else if (Radio := radio_device[Object]):
                Radio.Hide()
            else if (AudioPlayer := audio_player_device[Object]):
                AudioPlayer.Hide()

# Class in charge of UI for cinematic
cinematic_ui := class:

    # Create a canvas widget that displays a skip button for the cinematic
    CreateUI():tuple(canvas, listenable(widget_message))=
        SkipButton:button_regular = button_regular{DefaultText := SkipCinematicText}
        SkipButton.SetEnabled(true)

        UICanvas:canvas = canvas:
            Slots := array:
                canvas_slot:
                    Anchors := anchors{Minimum := vector2{X := 0.9, Y := 0.9}, Maximum := vector2{X := 0.9, Y := 0.9}}
                    Offsets := margin{Top := 0.0, Left := 0.0, Right := 0.0, Bottom := 0.0}
                    Alignment := vector2{X := 1.0, Y := 1.0}
                    SizeToContent := true
                    Widget := SkipButton

        return (UICanvas, SkipButton.OnClick())

    # Shows the UI for the player and returns the listenable for the button
    Show(Player:player, PlayerUI:player_ui):listenable(widget_message)=
        UIResult := CreateUI()
        UICanvas := UIResult(0)
        Listenable := UIResult(1)

        PlayerUISlot := player_ui_slot{ZOrder := 0, InputMode := ui_input_mode.All}
        PlayerUI.AddWidget(UICanvas, PlayerUISlot)
        if (set UIPerPlayer[Player] = option{UICanvas}):
        
        return Listenable

    # Hides the UI for the player
    Hide(Player:player):void=
        if:
            PlayerUI := GetPlayerUI[Player]
            CustomUI := UIPerPlayer[Player]?
        then:
            PlayerUI.RemoveWidget(CustomUI)
            if (set UIPerPlayer[Player] = false):

    # A localizable message to display as text in the UI
    SkipCinematicText<localizes><private>:message = "Skip Intro"

    var UIPerPlayer<private>:[player]?canvas = map{}
