# Knowledge-neutral Verse snippet; project names and asset IDs generalized.
# id: snippet-221

# This file contains the gameboard structure and how it works.
# The gameboard knows where the NPC is supposed to start on the board
# how many obstacles / barriers there are to getting to the end goal,
# and what the end goal is and when it has been reached.
using { /Fortnite.com/Devices }
using { /UnrealEngine.com/Temporary/SpatialMath }
using { /Verse.org/Simulation }
using { /Verse.org/Simulation/Tags }

# An obstacle on the gameboard, with an associated set of
# Barrier devices, Trigger devices that deactivate the barriers,
# and Cinematic Sequence devices that play sequences for barriers dissolving and appearing for visual feedback on what is happening.
# This class has the concrete specifier so it can be an editable property on a Verse device.
obstacle := class<concrete>:

    # Data for storing whether barriers are currently enabled/disabled.
    var IsObstaclePassed:logic = false

    # The array of barriers for this obstacle.
    @editable
    Barriers:[]barrier_device = array{}

    # The cinematic sequences for barriers disappearing.
    @editable
    BarrierDissolves:[]cinematic_sequence_device = array{}

    #The cinematic sequences for barriers appearing.
    @editable
    BarrierAppears:[]cinematic_sequence_device = array{}

    # The array of triggers for this obstacle.
    @editable
    Trigger:trigger_device = trigger_device{}

    # Runs when the NPC steps on the obstacle Trigger device.
    # Disables the trigger, each barrier assigned to this obstacle,
    # and plays the cinematic for the barrier dissolving for visual feedback.
    OnObstaclePassed(Agent:?agent):void=
        set IsObstaclePassed = true
        Trigger.Disable()
        for(BarrierDissolve:BarrierDissolves):
            BarrierDissolve.Play()
        for(Barrier:Barriers):
            Barrier.Disable()

    # Set up the obstacle by subscribing to the Trigger device event.
    Setup():void=
        Trigger.TriggeredEvent.Subscribe(OnObstaclePassed)
    
    # Reset the obstacle by enabling the trigger and barriers assigned to this obstacle, 
    # and playing the cinematic for the barriers appearing.
    Reset():void=
        if: 
            IsObstaclePassed?
        then:
            set IsObstaclePassed = false
            Trigger.Enable()
            for(BarrierAppear:BarrierAppears):
                BarrierAppear.Play()
            for (Barrier:Barriers):
                Barrier.Enable()

# This class represents the gameboard and how it behaves.
# This class has the concrete specifier so it can be an editable property on a Verse device.
gameboard<public> := class<concrete>:

    # The size of each tile on the gameboard. By default set to 
    # the default Fortnite tile size of 512x512x384.
    @editable
    TileSize<public>:vector3 = vector3{X:=512.0, Y:=512.0, Z:=384.0}

    # The fixed point camera that provides a top-down view of the gameboard
    @editable
    Camera:gameplay_camera_fixed_point_device = gameplay_camera_fixed_point_device{}

    # The trigger device the NPC needs to activate to complete the board.
    @editable
    EndGoal:trigger_device = trigger_device{}

    # The array of obstacles for this gameboard.
    @editable
    Obstacles:[]obstacle = array{}

    # The transform of the NPC’s starting position on this board.
    @editable
    var StartingCharacterPosition<public>:creative_prop = creative_prop{}

    @editable
    OpeningCinematic:cinematic_sequence_device = cinematic_sequence_device{}

    # The Z offset to add to the NPC’s starting position when spawning them.
    ZOffset:float = 100.0

    # Event that signals when the end goal has been reached.
    EndGoalReached<public>:event() = event(){}

    # Gets the starting transform of the NPC on this gameboard by adding the ZOffest to the starting transform.
    GetStartingCharacterPosition():transform=
        var Transform:transform = StartingCharacterPosition.GetTransform()
        # Add a small Z offset to the transform of the starting position to spawn the NPC
        # slightly in the air. This prevent the NPC from colliding with the ground or other 
        # objects when spawning.
        set Transform.Translation.Z += ZOffset
        Transform

    # Call Setup on each obstacle, and subscribe to the EndGoal trigger event.
    Setup<public>():void=
        EndGoal.TriggeredEvent.Subscribe(OnEndGoalEmitted)
        for (Obstacle : Obstacles):
            Obstacle.Setup()
    
    # Swap the camera view to top-down and play an opening cinematic.
    Start<public>():void=
        Camera.AddToAll()
        OpeningCinematic.Play()

    # Remove the current top-down camera view.
    Stop<public>():void=
        Camera.RemoveFromAll()

    # Signal that the end goal has been reached.
    OnEndGoalEmitted<private>(MaybeAgent:?agent):void=
        EndGoalReached.Signal()

    # Reset all obstacles.
    Reset<public>():void=
        for (Obstacle : Obstacles):
            Obstacle.Reset()
