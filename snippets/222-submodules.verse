# Knowledge-neutral Verse snippet; project names and asset IDs generalized.
# id: snippet-222

# This file sets the access levels for submodules in the project.
# Specifically, this file is making art assets accessible to the Verse files in the project
# so textures, meshes, and materials can be used in the UI and in the game.

# The submodule Textures.MiniGameUI below is the same as the folder
# Textures/MinigameUI that you can see in the project's Content Browser.
Textures := module:
    MiniGameUI<public> := module{}

# The submodule VerseCommander below is the same as the folder
# VerseCommander that you can see in the project's Content Browser.
VerseCommander := module:
    Meshes<public> := module{}
    Materials<public> := module{}
    MiniGameTextures<public> := module{}
