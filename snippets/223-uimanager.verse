# Knowledge-neutral Verse snippet; project names and asset IDs generalized.
# id: snippet-223

# This file contains all the code to create and modify the UI
# in the Verse Commander minigame.
# The UI for the game contains:
#  - Buttons that map to commands for the NPC: forward, turn left, turn right.
#  - An execute button that tells the NPC to perform the queue of commands.
#  - A remove button that removes the last command added.
#  - A reset button that resets the current board and clears out the command queue.
#  - A dynamic list of commands that grows wider when a player adds commands and shrinks when a player removes commands.

using { /Fortnite.com/Devices }
using { /Fortnite.com/UI }
using { /UnrealEngine.com/Temporary/Diagnostics }
using { /UnrealEngine.com/Temporary/UI }
using { /UnrealEngine.com/Temporary/SpatialMath }
using { /Verse.org/Colors }
using { /Verse.org/Assets }
using { /Verse.org/Simulation }

# This text is localizable and must be a message type to be used in Verse UI.
ForwardText<localizes>:message = "Forward"
TurnRightText<localizes>:message = "Turn Right"
TurnLeftText<localizes>:message = "Turn Left"
RemoveText<localizes>:message = "Remove"
ExecuteText<localizes>:message = "Execute"
ResetText<localizes>:message = "Reset"
PlaceholderText<localizes>:message = ""

# This is the class that holds all the UI element references
# and is in charge of the designs for each UI element and how a UI element responds. 
verse_commander_minigame_ui := class:

    # These are references to the canvases used in the UI.
    # The interactive UI canvas holds all the buttons that the player can interact with.
    # The QueueUI holds the dynamic list of commands that the player can only interact with using the interactive UI.
    var InteractiveCanvas:?canvas = false
    var QueueUI:?canvas = false

    # A counter for the number of command in the queue.
    var CommandCounter:int = 0

    # References to each button. The class minigame_button handles the look and feel of each button and how it responds to player interaction.
    ButtonExecute:minigame_button = minigame_button:
        ActiveBackgroundTexture := Textures.MiniGameUI.T_UI_Button_Yellow_Rounded
        ActiveIconTexture := Textures.MiniGameUI.T_UI_Icon_Execute
        ActiveTextColor := MakeColorFromHex("0D137E")
        InactiveBackgroundTexture := Textures.MiniGameUI.T_UI_Button_Yellow_Rounded_Dim
        InactiveIconTexture := Textures.MiniGameUI.T_UI_Icon_Execute_Dim
        InactiveTextColor := MakeColorFromHex("030628")
        TextMessage := ExecuteText

    ButtonRemove:minigame_button = minigame_button:
        ActiveBackgroundTexture := Textures.MiniGameUI.T_UI_Button_Yellow_Rounded
        ActiveIconTexture := Textures.MiniGameUI.T_UI_Icon_Remove
        ActiveTextColor := MakeColorFromHex("0D137E")
        InactiveBackgroundTexture := Textures.MiniGameUI.T_UI_Button_Yellow_Rounded_Dim
        InactiveIconTexture := Textures.MiniGameUI.T_UI_Icon_Remove_Dim
        InactiveTextColor := MakeColorFromHex("030628")
        TextMessage := RemoveText

    ButtonReset:minigame_button = minigame_button:
        ActiveBackgroundTexture := Textures.MiniGameUI.T_UI_Button_Yellow_Rounded
        ActiveIconTexture := Textures.MiniGameUI.T_UI_Icon_Reset
        ActiveTextColor := MakeColorFromHex("0D137E")
        InactiveBackgroundTexture := Textures.MiniGameUI.T_UI_Button_Yellow_Rounded_Dim
        InactiveIconTexture := Textures.MiniGameUI.T_UI_Icon_Reset_Dim
        InactiveTextColor := MakeColorFromHex("030628")
        TextMessage := ResetText

    ButtonForward:minigame_button = minigame_button:
        ActiveBackgroundTexture := Textures.MiniGameUI.T_UI_Button_Blue_Rounded
        ActiveIconTexture := Textures.MiniGameUI.T_UI_Arrow_ForwardRight
        ActiveTextColor := MakeColorFromHex("ABD0D0")
        InactiveBackgroundTexture := Textures.MiniGameUI.T_UI_Button_Blue_Rounded_Dim
        InactiveIconTexture := Textures.MiniGameUI.T_UI_Arrow_Forward_Dim
        InactiveTextColor := MakeColorFromHex("364242")
        TextMessage := ForwardText

    ButtonTurnRight:minigame_button = minigame_button:
        ActiveBackgroundTexture := Textures.MiniGameUI.T_UI_Button_Blue_Rounded
        ActiveIconTexture := Textures.MiniGameUI.T_UI_Arrow_Right
        ActiveTextColor := MakeColorFromHex("ABD0D0")
        InactiveBackgroundTexture := Textures.MiniGameUI.T_UI_Button_Blue_Rounded_Dim
        InactiveIconTexture := Textures.MiniGameUI.T_UI_Arrow_Right_Dim
        InactiveTextColor := MakeColorFromHex("364242")
        TextMessage := TurnRightText

    ButtonTurnLeft:minigame_button = minigame_button:
        ActiveBackgroundTexture := Textures.MiniGameUI.T_UI_Button_Blue_Rounded
        ActiveIconTexture := Textures.MiniGameUI.T_UI_Arrow_Left
        ActiveTextColor := MakeColorFromHex("ABD0D0")
        InactiveBackgroundTexture := Textures.MiniGameUI.T_UI_Button_Blue_Rounded_Dim
        InactiveIconTexture := Textures.MiniGameUI.T_UI_Arrow_Left_Dim
        InactiveTextColor := MakeColorFromHex("364242")
        TextMessage := TurnLeftText

    # An array that stores the commands displayed on the UI.
    # These commands are passed to the NPC behavior when the player chooses to execute the commands on the NPC.
    var CommandQueueSlots<private>:[]command_queue = array{}

    # Widget reference for the commands to be able to add and remove widgets (stored in CommandQueueSlots) later.
    var CommandQueueWidget:stack_box = stack_box{Orientation := orientation.Horizontal}

    # Generate all the UI from scratch in Verse and store it to draw and reference later.
    Create<public>():void=
        CommandQueueCanvas := canvas:
            Slots := array:
                CreateCommandQueue()
        set QueueUI = option{CommandQueueCanvas}

        NewInteractiveCanvas := canvas:
            Slots := array:
                CreateButtons()
        set InteractiveCanvas = option{NewInteractiveCanvas}

    # Empty out all the commands from the current dynamic list.
    ResetCommandQueue<private>():void=
        for (Index -> CommandQueueSlot : CommandQueueSlots, Index > 1):
            CommandQueueWidget.RemoveWidget(CommandQueueSlot.Widget)

        # When the command queue is empty, there are still two slots to display
        # for the empty left and right ends of the queue to show that there are no commands in the UI. 
        if:
            FirstCommandQueueSlot := CommandQueueSlots[0]
            LastCommandQueueSlot := CommandQueueSlots[1]
        then:
            # Set as the left side.
            FirstCommandQueueSlot.SetAsEnd(false)

            # Set as the right side.
            LastCommandQueueSlot.SetAsEnd(true)

            set CommandQueueSlots = array{FirstCommandQueueSlot, LastCommandQueueSlot}

    # Create an empty command queue at the bottom of the screen and save the widget for reference later.
    CreateCommandQueue<private>():canvas_slot=
        # Create an end slot and set it as the left side.
        LeftEndSlot := command_queue{}
        LeftEndSlot.CreateEnd(false)

        # Create an end slot and set it as the right side.
        RightEndSlot := command_queue{}
        RightEndSlot.CreateEnd(true)
        set CommandQueueSlots = array{LeftEndSlot, RightEndSlot}

        # Save the widget to add and remove widgets from later.
        set CommandQueueWidget = stack_box:
            Orientation := orientation.Horizontal
            Slots := array:
                stack_box_slot:
                    Widget := LeftEndSlot.Widget
                stack_box_slot:
                    Widget := RightEndSlot.Widget

        # Creates a canvas slot that contains the horizontal stackbox of commands and anchors the widget in the lower middle of the screen.
        CanvasSlot := canvas_slot: # Lower Middle of Screen
            Anchors := anchors{ Minimum := vector2{X := 0.5, Y := 0.9}, Maximum := vector2{X := 0.5, Y := 0.9}}
            Offsets := margin{Top := 0.0, Left := 0.0, Right := 0.0, Bottom := 0.0}
            Alignment := vector2{X := 0.5, Y := 1.0}
            SizeToContent := true
            Widget := CommandQueueWidget

        return CanvasSlot
        
    # Create all the buttons and position them at the top of the screen.
    CreateButtons<private>():canvas_slot=
        ButtonForward.Create()
        ButtonTurnRight.Create()
        ButtonTurnLeft.Create()
        ButtonExecute.Create()
        ButtonRemove.Create()
        ButtonReset.Create()

        # Creates a canvas slot that contains the horizontal stackbox of buttons and positions them at top of screen.
        # Padding is added between buttons to space them apart from each other but keep them all still centered.
        canvas_slot:
            Anchors := anchors{Minimum := vector2{X := 0.5, Y := 0.0}, Maximum := vector2{X := 0.5, Y := 0.0} }
            Offsets := margin{Top := 0.0, Left := 0.0, Right := 0.0, Bottom := 0.0}
            Alignment := vector2{X := 0.5, Y := 0.0}
            SizeToContent := true
            Widget := stack_box:
                Orientation := orientation.Horizontal
                Slots := array:
                    stack_box_slot:
                        Padding := margin{Top := 0.0, Left := 0.0, Right := 2.5, Bottom := 0.0}
                        Widget := ButtonTurnLeft.Widget
                    stack_box_slot:
                        Padding := margin{Top := 0.0, Left := 2.5, Right := 2.5, Bottom := 0.0}
                        Widget := ButtonForward.Widget
                    stack_box_slot:
                        Padding := margin{Top := 0.0, Left := 2.5, Right := 2.5, Bottom := 0.0}
                        Widget := ButtonTurnRight.Widget
                    stack_box_slot:
                        Padding := margin{Top := 0.0, Left := 2.5, Right := 2.5, Bottom := 0.0}
                        Widget := ButtonExecute.Widget
                    stack_box_slot:
                        Padding := margin{Top := 0.0, Left := 2.5, Right := 2.5, Bottom := 0.0}
                        Widget := ButtonRemove.Widget
                    stack_box_slot:
                        Padding := margin{Top := 0.0, Left := 2.5, Right := 0.0, Bottom := 0.0}
                        Widget := ButtonReset.Widget          

    # Adds the custom UI created in this class to the player’s UI.
    Draw<public>(PlayerUI:player_ui):void=
        if:
            CommandQueueCanvas := QueueUI?
            ButtonCanvas := InteractiveCanvas?
        then:
            PlayerUI.AddWidget(CommandQueueCanvas)

            # Input mode is set to All so the player can click the buttons.
            PlayerUI.AddWidget(ButtonCanvas, player_ui_slot{InputMode := ui_input_mode.All})

    # Clear out the queue and reset the command counter.
    ClearQueue<public>():void=
        ResetCommandQueue()
        set CommandCounter = 0

    # Removes the custom UI created in this class from the player’s UI.
    Remove<public>(PlayerUI:player_ui):void=
        if:
            CommandQueueCanvas := QueueUI?
            ButtonCanvas := InteractiveCanvas?
        then:
            PlayerUI.RemoveWidget(CommandQueueCanvas)
            PlayerUI.RemoveWidget(ButtonCanvas)

    # Appends the command icon to the command queue at the bottom of the screen.
    # If the commands in the queue are at the max number of commands allowed, no more commands can be added.
    AddCommandToDisplay(Command:command, UICommandLimit:int):void=
        # Checks if the queue has already reached the max number of commands allowed.
        # The check ignores the left and right end slots since they don’t contain commands, which is why 2 is subtracted from the command length.
        if:
            (CommandQueueSlots.Length - 2) = UICommandLimit
        then:
            return

        # Increase the command counter and add the command to the end of the queue.
        set CommandCounter += 1
        if:
            LastCommandQueueSlot := CommandQueueSlots[CommandCounter]
        then:
            LastCommandQueueSlot.SetCommand(Command)

        # Create a new slot to add to the end of the queue and set it as the right end.
        CommandQueueSlotToAdd := command_queue{}
        CommandQueueSlotToAdd.CreateEnd(true)

        # Update the references and add the new end widget to the stackbox.
        if:
            NewArray := CommandQueueSlots.Insert[CommandCounter + 1, array{CommandQueueSlotToAdd}]
        then:
            set CommandQueueSlots = NewArray
            CommandQueueWidget.AddWidget(stack_box_slot{Widget := CommandQueueSlotToAdd.Widget})

    # Removes the last command icon from the command queue at the bottom of the screen.
    # If there are no commands already in the queue, nothing happens.
    RemoveLastCommandFromDisplay():void=
        if:
            CommandCounter = 0
        then:
            return

        # Set the last command as the right end slot instead.
        if:
            CommandQueueSlot := CommandQueueSlots[CommandCounter]
        then:
            CommandQueueSlot.SetAsEnd(true)

        # Remove the previous right end slot from the display.
        if:
            LastCommandQueueSlot := CommandQueueSlots[CommandCounter + 1]
        then:
            CommandQueueWidget.RemoveWidget(LastCommandQueueSlot.Widget)

        # Update the references.
        if:
            NewArray := CommandQueueSlots.RemoveElement[CommandQueueSlots.Length - 1]
        then:
            set CommandQueueSlots = NewArray
            
        # Decrease the command counter.
        set CommandCounter -= 1

    # Gets the current queue of commands from the command queue UI.
    GetCommandQueue():[]command=
        CommandQueue := for(Slot : CommandQueueSlots, Command := Slot.CommandData?):
            Command
    
        return CommandQueue

    # Set whether the player can interact with the buttons.
    # When IsEnabled is true, then the buttons become interactable and bright.
    # When IsEnabled is false, then the buttons become non-interactable and dim.
    SetQueueButtonInteractivity(IsEnabled:logic):void=
        ButtonForward.SetEnabled(IsEnabled)
        ButtonTurnRight.SetEnabled(IsEnabled)
        ButtonTurnLeft.SetEnabled(IsEnabled)
        ButtonExecute.SetEnabled(IsEnabled)
        ButtonRemove.SetEnabled(IsEnabled)

# The class for storing all the widget references for a button and setting its look.
minigame_button := class:

    # Store a reference of the overlay for easy removal later.
    var Widget:overlay = overlay:

    # Store a reference of the background texture widget to change easily later.
    BackgroundImage:texture_block = texture_block:
        DefaultImage := Textures.MiniGameUI.T_UI_Button_Blue_Rounded
        DefaultDesiredSize := vector2{X := 256.0, Y := 128.0}

    ActiveBackgroundTexture:texture = Textures.MiniGameUI.T_UI_Button_Blue_Rounded
    ActiveIconTexture:texture = Textures.MiniGameUI.T_UI_Arrow_ForwardRight
    ActiveTextColor:color = MakeColorFromHex("ABD0D0")
    InactiveBackgroundTexture:texture = Textures.MiniGameUI.T_UI_Button_Blue_Rounded_Dim
    InactiveIconTexture:texture = Textures.MiniGameUI.T_UI_Arrow_Forward_Dim
    InactiveTextColor:color = MakeColorFromHex("364242")
    TextMessage:message = ForwardText

    # Store a reference of the icon texture widget to change easily later.
    Icon:texture_block = texture_block:
        DefaultImage := Textures.MiniGameUI.T_UI_Arrow_ForwardRight
        DefaultDesiredSize := vector2{X := 64.0, Y := 64.0}

    # Store a reference of the text widget to change easily later.
    Text:text_block = text_block:
        DefaultText := ForwardText
        DefaultTextColor := MakeColorFromHex("ABD0D0")

    # Store a reference of the button widget to change easily later.
    Button:button_quiet = button_quiet:
        DefaultText := PlaceholderText

    # Set the interactivity of the button.
    # When InIsEnabled is true, enable the button and add the bright versions of textures and text color.
    # When InIsEnabled is false, disable the button and add the dim versions of textures and text color.
    SetEnabled(InIsEnabled:logic):void=
        Button.SetEnabled(InIsEnabled)

        if (InIsEnabled?):
            BackgroundImage.SetImage(ActiveBackgroundTexture)
            Icon.SetImage(ActiveIconTexture)
            Text.SetTextColor(ActiveTextColor)
        else:
            BackgroundImage.SetImage(InactiveBackgroundTexture)
            Icon.SetImage(InactiveIconTexture)
            Text.SetTextColor(InactiveTextColor)

    # Create the button based on its type and set its default textures and text.
    # The button widget overlays the text and icon, which overlay the background.
    # This is to have the custom look of the button but maintain the functionality and responsiveness of the button design.
    Create():void=
        set Widget = overlay:
            Slots := array:
                overlay_slot: # background of button
                    Widget := BackgroundImage
                overlay_slot:
                    Widget := stack_box:
                        Orientation := orientation.Horizontal
                        Slots := array:
                            stack_box_slot:
                                Widget := Icon
                            stack_box_slot:
                                Widget := Text
                overlay_slot: # button
                    HorizontalAlignment := horizontal_alignment.Fill
                    Widget := Button

        BackgroundImage.SetImage(ActiveBackgroundTexture)
        Icon.SetImage(ActiveIconTexture)
        Text.SetTextColor(ActiveTextColor)
        Text.SetText(TextMessage)

# A class that stores the image (texture) that is used to represent a command in the queue on screen
# and the enum associated with that command.
command_queue := class:
    # Store a reference of the overlay for easy removal later.
    var Widget:overlay = overlay{}

    # Store a reference of the background texture block to change the image easily later.
    BackgroundImage:texture_block = texture_block:
        DefaultImage := Textures.MiniGameUI.TransparentImage512
        DefaultDesiredSize := vector2{X := 88.0, Y := 88.0}

    # Store a reference of the command texture block to change the image easily later.
    CommandImage:texture_block = texture_block:
        DefaultImage := Textures.MiniGameUI.TransparentImage256
        DefaultDesiredSize := vector2{X := 88.0, Y := 88.0}

    # Store the command data that's also displayed in the UI for easy retrieval later.
    var CommandData:?command = false

    # Create the overlay widget with the background and command images.
    CreateEnd(IsRight:logic):void=
        set Widget = overlay:
            Slots := array:
                overlay_slot:
                    Widget := BackgroundImage
                overlay_slot:
                    Widget := CommandImage
        SetAsEnd(IsRight)

    # Update the widget to show the new command.
    SetCommand(Command:command):void=
        BackgroundImage.SetImage(Textures.MiniGameUI.T_UI_Prompt_Middle)
        if (Commands.forward_command[Command]):
            CommandImage.SetImage(Textures.MiniGameUI.T_UI_Arrow_ForwardRight)
        else if (Commands.turn_right_command[Command]):
            CommandImage.SetImage(Textures.MiniGameUI.T_UI_Arrow_Right)
        else if (Commands.turn_left_command[Command]):
            CommandImage.SetImage(Textures.MiniGameUI.T_UI_Arrow_Left)
        else:
            CommandImage.SetImage(Textures.MiniGameUI.TransparentImage256)
        set CommandData = option{Command}

    # Update the widget to be the end of the UI.
    # When IsRight is true, the right end texture is used.
    # When IsRight is false, the left end texture is used.
    SetAsEnd(IsRight:logic):void=
        if:
            IsRight?
        then:
            BackgroundImage.SetImage(Textures.MiniGameUI.T_UI_Prompt_EndRight)
        else:
            BackgroundImage.SetImage(Textures.MiniGameUI.T_UI_Prompt_End)
        CommandImage.SetImage(Textures.MiniGameUI.TransparentImage256)
        set CommandData = false
