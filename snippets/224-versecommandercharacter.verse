# Knowledge-neutral Verse snippet; project names and asset IDs generalized.
# id: snippet-224

# This file handles the logic for the gameboard NPC. The NPC waits for commands from
# verse_commander_minigame, then executes movement based on those commands using navigation targets.
# It also tracks the NPC's position using a visual arrow that updates after each command.

using { /Fortnite.com/AI }
using { /Fortnite.com/Characters }
using { /Fortnite.com/Devices }
using { /UnrealEngine.com/Temporary/Diagnostics }
using { /UnrealEngine.com/Temporary/SpatialMath }
using { /Verse.org/Assets }
using { /Verse.org/Simulation }

# Create a dedicated debug channel to draw to for this behavior
verse_commander_character_debug_draw := class(debug_draw_channel) {}

# This module contains functions which return messages used for helpful debugging information that are
# localized and paired with the NPC ID.
verse_commander_character_message_module:= module:

    OnBeginMessage<public><localizes>(Agent:agent):message =
        "NPC Agent = {Agent}: OnBegin triggered let's get started."
    OnEndMessage<public><localizes>(Agent:agent):message =
        "NPC Agent = {Agent}: OnEnd triggered let's cleanup."
    OnNavigateBeginMessage<public><localizes>(Agent:agent, X:float, Y:float, Z:float):message =
        "NPC Agent = {Agent}: Is moving to [{X},{Y},{Z}]"
    OnNavigateErrorMessage<public><localizes>(Agent:agent, X:float, Y:float, Z:float):message =
        "NPC Agent = {Agent}: Hit error moving to [{X},{Y},{Z}], please refer to Island Setting's Navigation debug"

# A Verse-authored NPC Behavior that can be used within an NPC Definition or a Character Spawner device's Behavior Script Override.
verse_commander_character<public> := class(npc_behavior):

    # The VFX that play when the NPC teleports out.
    @editable
    CharacterTeleportOutVFX:vfx_spawner_device = vfx_spawner_device{}

    # The VFX that play when the NPC teleports in.
    @editable
    CharacterTeleportInVFX:vfx_spawner_device = vfx_spawner_device{}

    # The amount of time to wait for the VFX to finish playing.
    @editable
    VFXWaitTime:float = 1.75

    # The amount of time to wait between executing commands.
    @editable
    CommandWaitTime:float = 0.25

    # The amount of time to Telekinetic focus on a target.
    @editable
    FocusTime:float = 0.5

    # The radius for the character to reach their navigation target.
    @editable
    ReachRadius:float = 0.1

    # The offset of the arrow asset from the NPC's position.
    @editable
    ForwardArrowAssetOffset:vector3 = vector3{X := 150.0, Y := 150.0, Z := 0.0}

    # The creative_prop_asset used to spawn the ForwardArrow prop.
    ForwardArrowAsset:creative_prop_asset = DefaultCreativePropAsset

    # The arrow prop that displays the orientation of the NPC.
    var ForwardArrow:creative_prop = creative_prop{}
    
    # Reference to the verse commander minigame device in the level.
    var VerseCommanderMinigame<private>:verse_commander_minigame = verse_commander_minigame{}

    # This function runs when the NPC is spawned in the world and ready to follow a behavior.
    OnBegin<override>()<suspends>:void=

        # Get the Verse Commander Minigame Device.
        # Assumption is that there is only one device in the level.
        CreativeObjects := CharacterTeleportOutVFX.FindCreativeObjectsWithTag(verse_commander_minigame_tag{})
        for(CreativeObject:CreativeObjects):
            if:
                MinigameManager := verse_commander_minigame[CreativeObject]
            then:
                if:
                    # Get the Agent (the NPC).
                    Agent := GetAgent[]

                    # Gets the Fortnite Character interface, which gets you access to its gameplay data 
                    # including its AI module for navigation and focus.
                    Character := Agent.GetFortCharacter[]

                    # Get the Navigatable Interface, this allows you to tell it to move.
                    Navigatable := Character.GetNavigatable[]

                    # Get the Focus Interface, this allows you to tell it to look at something or somewhere.
                    Focus := Character.GetFocusInterface[]
                then:
                    # Keep a reference of the Verse Commander Minigame so the character can receive commands from it.
                    set VerseCommanderMinigame = MinigameManager
                    
                    # Create the 
                    CreateArrow(Agent)

                    # Wait two seconds.
                    Sleep(CommandWaitTime)

                    # Race between the minigame ending and the character performing commands on loop.
                    race:
                        loop:
                            CharacterCommandLoop()
                        VerseCommanderMinigame.GameEndedEvent.Await()
                    return
                else:
                    ProjectLog("Not able to get character")
                    
        # If code falls here something failed when setting up the character for the game.
        ProjectLog("Error in NPC Behavior Script on NPC Setup")

    # Creates an arrow prop at the NPC's position that visually shows the orientation of the NPC.
    CreateArrow(Agent:agent):void=
        if :
            Character := Agent.GetFortCharacter[]
        then:
            var Transform:transform = Character.GetTransform()
            # Spawn the arrow prop, then set the mesh and material for the prop.
            SpawnPropResult := SpawnProp(ForwardArrowAsset, Transform)
            if:
                SpawnedProp := SpawnPropResult(0)?
            then:
                set ForwardArrow = SpawnedProp
                ForwardArrow.SetMesh(VerseCommander.Meshes.CP_VerseCommander_Floor_01_asset)
                ForwardArrow.SetMaterial(VerseCommander.Materials.MI_CP_VerseCommander_CharacterArrow_01{})
                # Update the arrow prop to the position and orientation of the NPC.
                MoveArrow(Agent)
            else:
                SpawnPropIssue:string = case(SpawnPropResult(1)):
                    spawn_prop_result.Ok => "Ok"
                    spawn_prop_result.InvalidSpawnPoint => "Invalid Spawn Point"
                    spawn_prop_result.SpawnPointOutOfBounds => "Spawn Point Outside Island's Boundaries"
                    spawn_prop_result.InvalidAsset => "Asset Is Not a Valid creative_prop"
                    spawn_prop_result.TooManyProps => "Too Many Props Spawned Than Permitted by Island."
                    _ => "Unknown Error"
                ProjectLog("Error spawning prop: {SpawnPropIssue}")

    # Update the arrow prop to the position and orientation of the NPC, 
    # then make the arrow prop visible.
    MoveArrow(Agent:agent):void=
        if:
            Character := Agent.GetFortCharacter[]
        then:
            CharacterTransform := Character.GetTransform()
            var ArrowTransform:transform = CharacterTransform
            LocalForward := ArrowTransform.Rotation.GetLocalForward()
            set ArrowTransform.Rotation = ArrowTransform.Rotation.ApplyWorldRotationZ(DegreesToRadians(90.0))
            set ArrowTransform.Translation = ArrowTransform.Translation + LocalForward * ForwardArrowAssetOffset
            set ArrowTransform.Scale = vector3{X := 0.5, Y:= 0.5, Z := 1.0}
            if (ForwardArrow.TeleportTo[ArrowTransform]):
            ForwardArrow.Show()

    # Waits for the current board to be reset, then moves the
    # NPC back to the starting position of the board along with VFX.
    AwaitReset()<suspends>:void=
        # Wait for the current board to be reset.
        # The event payload is the starting position for the board.
        StartPosition := VerseCommanderMinigame.BoardResetEvent.Await()
        spawn{PlayVFXAndMoveCharacter(StartPosition)}

    # Hides the NPC and the arrow prop, then teleports both to a new position, 
    # playing VFX for teleporting in and teleporting out.
    PlayVFXAndMoveCharacter(StartPosition:transform)<suspends>:void=
        if:
            Agent := GetAgent[]
            FortCharacter := Agent.GetFortCharacter[]
        then:
            # Hide the NPC and the arrow.
            FortCharacter.Hide()
            ForwardArrow.Hide()

            # Get the character's current position to play the VFX at.
            CurrentPosition:transform = FortCharacter.GetTransform()

            # Move the VFX to the character's current position.
            if (CharacterTeleportOutVFX.TeleportTo[CurrentPosition]):

            # Play the VFX of the character teleporting out.
            CharacterTeleportOutVFX.Enable()

            # Wait for the VFX to finish playing.
            Sleep(VFXWaitTime)

            # Stop the VFX of the character teleporting out.
            CharacterTeleportOutVFX.Disable()
            
            # Move the VFX to the starting position of the board.
            if (CharacterTeleportInVFX.TeleportTo[StartPosition]):

            # Play the VFX of the character teleporting in.
            CharacterTeleportInVFX.Enable()

            # Move the character to the starting position of the board.
            if (MoveToTile[StartPosition]) {}
            else:
                # Since this expression, moving the character to the start again,
                # breaks the game, its failure is logged to help with debugging.
                # The other failable expressions only play and move the VFX device
                # which are not game breaking and so no need to add extra logging unless actively debugging.
                ProjectLog("Could not teleport character")

            # Show the character and wait for the VFX to finish.
            FortCharacter.Show()

            # Wait for the VFX to finish playing.
            Sleep(VFXWaitTime)

            # Show the NPC and the arrow.
            MoveArrow(Agent)
            ForwardArrow.Show()

            # Stop the VFX of the character teleporting in.
            CharacterTeleportInVFX.Disable()

    # Waits for commands to be sent from the verse_commander_minigame, then
    # executes each command. 
    AwaitCommands()<suspends>:void=
        if:
            Agent := GetAgent[]
        then:
            # Wait for commands to be sent from the verse commander minigame.
            ExecuteResult := VerseCommanderMinigame.ExecuteCommandsEvent.Await()
            
            # For each execute result tuple, execute the command and pass the tile size from the tuple.
            # Then wait for a period of time.
            for (Command : ExecuteResult(0)):
                ForwardArrow.Hide()
                ExecuteCommand(Command, ExecuteResult(1))
                MoveArrow(Agent)
                ForwardArrow.Show()
                Sleep(CommandWaitTime)

            # When finished executing commands, signal that the character has finished moving.
            VerseCommanderMinigame.CharacterFinishedMovingEvent.Signal()
        else:
            ProjectLog("No agent in awaiting commands")

    # Race between resetting the character to start of the board and awaiting commands for that character.
    CharacterCommandLoop()<suspends>:void=
        race:
            AwaitReset()
            loop:
                AwaitCommands()

    # Teleports the NPC to the given transform.
    MoveToTile(Transform:transform)<transacts><decides>:void=
        # Get the Agent (the NPC).
        Agent := GetAgent[]
        
        # Gets the Fortnite Character interface, which gets you access to its gameplay data 
        # including its AI module for navigation and focus.
        Character := Agent.GetFortCharacter[]

        var NewTransform:transform = Transform
        set NewTransform.Translation.Z += 20.0
        
        # Teleport the character to the given transform.
        Character.TeleportTo[Transform.Translation, Transform.Rotation]

    # Gets a new navigation target for the NPC based on the current transform and the given command.
    GetNavTarget(CurrentTransform:transform, Command:command, TileDistance:vector3):transform=
        # Based on the command, get the character's local foward, right, or left (negative right).
        Direction :=
            if (Commands.forward_command[Command]):
                CurrentTransform.Rotation.GetLocalForward()
            else if (Commands.turn_right_command[Command]):
                CurrentTransform.Rotation.GetLocalRight()
            else if (Commands.turn_left_command[Command]):
                -CurrentTransform.Rotation.GetLocalRight()
            else:
                CurrentTransform.Rotation.GetLocalForward()

        # Create a new transform for the character to navigate to from the character's current transform.
        var NewTransform:transform = CurrentTransform

        # Multiply the direction of the command and the tile distance then add them to the new transform's X and Y.
        # This creates a new target for the character to navigate to that is either ahead, to the right, or to the left of
        # the direction the character is currently facing.
        set NewTransform.Translation = vector3:
            X := NewTransform.Translation.X + Direction.X * TileDistance.X
            Y := NewTransform.Translation.Y + Direction.Y * TileDistance.Y
            Z := NewTransform.Translation.Z
        
        # Return the new target for the character to navigate to.
        return NewTransform         

    # Executes the given command, either moving the NPC forward one tile or turning them left
    # or right. 
    ExecuteCommand(Command:command, TileSize:vector3)<suspends>:void=
        if:
            # Get the Agent (the NPC).
            Agent := GetAgent[]
            
            # Gets the Fortnite Character interface, which gets you access to its gameplay data 
            # including its AI module for navigation and focus.
            Character := Agent.GetFortCharacter[]

            # Get the Navigatable Interface, this allows you to tell it to move.
            Navigatable := Character.GetNavigatable[]

            # Get the focus interface, which lets you Telekinetic a character to focus on a target.
            Focusable := Character.GetFocusInterface[]
        then:
            CharacterTransform := Character.GetTransform()
            TargetTile := GetNavTarget(CharacterTransform, Command, TileSize)

            # If the command is forward, create a new navigation target from the target tile.
            if:
                Commands.forward_command[Command]
            then:
                NavTarget := MakeNavigationTarget(TargetTile.Translation)
                # Navigate the character to the navigation target. The ReachRadius is set to a small float
                # here rather than zero to allow the character a small leniency in how close it needs to get to the 
                # exact position of the target.
                NavResult := Navigatable.NavigateTo(NavTarget, ?ReachRadius := ReachRadius)
            # If the command is right or left, Telekinetic the character to maintain focus on the taget for
            # a short period of time to turn the character to face that target.
            else if:
                Commands.turn_command[Command]
            then:
                race:
                    Focusable.MaintainFocus(TargetTile.Translation)
                    Sleep(FocusTime)
        else:
            ProjectLog("No character to issue commands to")
