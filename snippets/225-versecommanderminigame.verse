# Knowledge-neutral Verse snippet; project names and asset IDs generalized.
# id: snippet-225

# This device is in charge of managing the minigame.
# In its Details panel in the editor, you can:
# - Define how many gameboards there are
# - Fhe order you play the gameboards
# - All the details of the gameboard such as their obstacles.

using { /Fortnite.com/Devices }
using { /Fortnite.com/UI }
using { /UnrealEngine.com/Temporary }
using { /UnrealEngine.com/Temporary/Diagnostics }
using { /UnrealEngine.com/Temporary/UI }
using { /UnrealEngine.com/Temporary/SpatialMath }
using { /Verse.org/Colors }
using { /Verse.org/Simulation }
using { /Verse.org/Simulation/Tags }

# A Verse-authored creative device that can be placed in a level
verse_commander_minigame := class(creative_device):

    # The HUD Controller device for the minigame.
    @editable
    HUDController<private>:hud_controller_device = hud_controller_device{}

    # The button that starts the minigame.
    @editable
    PlayVerseCommanderButton:button_device = button_device{}

    # The NPC spawner that spawns the NPC used in the minigame.
    @editable
    NPCSpawner:npc_spawner_device = npc_spawner_device{}

    # An array of gameboards that the user can customize in the editor.
    # Each element in this array is a level that the character has to complete.
    @editable
    var Gameboards<private>:[]gameboard = array{}

    # The cinematic sequence used to show off design billboards.
    @editable
    Cinematic:billboard_cinematic = billboard_cinematic{}

    # A variable to track which gameboard the character is currently on.
    var CurrentBoard<private>:int = 0

    # Whether this is the first time a gameboard has been played. Used to know when to play
    # intro VFX for the board.
    var FirstTimeOnCurrentBoard<private>:logic = true

    # Minigame events that the character and Verse device can use to communicate with each other.
    ExecuteCommandsEvent:event(tuple([]command, vector3)) = event(tuple([]command, vector3)){}
    BoardResetEvent:event(transform) = event(transform){}
    GameStartedEvent:event() = event(){}
    GameEndedEvent:event() = event(){}
    CharacterFinishedMovingEvent:event() = event(){}

    # Internal events for handling player UI selection.
    ExecuteButtonSelected<private>:event() = event(){}
    ResetButtonSelected<private>:event() = event(){}

    # Define max limit of commands that can be queued on screen.
    @editable
    UICommandLimit<private>:int = 10

    # Keep a reference of the UI for each player so we can update it for each player.
    var UIPerPlayer<private>:[player]verse_commander_minigame_ui = map{}
    
    # Runs when the device is started in a running game.
    OnBegin<override>()<suspends>:void=
        spawn{Cinematic.EnterCinematicModeForAll(GetPlayspace().GetPlayers())}
        loop:
            PlayVerseCommanderButton.InteractedWithEvent.Await()

            Setup()
            # Wait for all Gameboards to be set up.
            Sleep(2.0)

            Start()

    # Creates a verse_commander_minigame_ui for each player, and adds buttons for each NPC
    # command to the player's UI. Then setups all gameboards, and spawns the NPC that the player commands.
    Setup()<suspends>:void=
        # For each player, create a new verse_commander_minigame_ui. 
        for:
            Number -> Player : GetPlayspace().GetPlayers()
            PlayerUI := GetPlayerUI[Player]
        do:
            # Add buttons for foward, turn right, turn left, remove, execute, and reset commands to each player's UI. 
            # Subscribe those buttons to their associated functions.
            NewUI := verse_commander_minigame_ui{}
            NewUI.Create()
            NewUI.ButtonForward.Button.OnClick().Subscribe(OnForwardSelected)
            NewUI.ButtonTurnRight.Button.OnClick().Subscribe(OnTurnRightSelected)
            NewUI.ButtonTurnLeft.Button.OnClick().Subscribe(OnTurnLeftSelected)
            NewUI.ButtonRemove.Button.OnClick().Subscribe(OnRemoveSelected)
            NewUI.ButtonExecute.Button.OnClick().Subscribe(OnExecuteSelected)
            NewUI.ButtonReset.Button.OnClick().Subscribe(OnResetSelected)
            # Store a reference of the player's UI to be able to update it later.
            if (set UIPerPlayer[Player] = NewUI):
        
        # Set up all the gameboards.
        for (Gameboard : Gameboards):
            Gameboard.Setup()

        # Spawn the NPC that the player commands.
        NPCSpawner.Spawn()

    # Adds the foward command to each player's command queue.
    OnForwardSelected(Widget:widget_message):void=
        AddCommandForAllPlayers(Commands.Forward)

    # Adds the turn right command to each player's command queue.
    OnTurnRightSelected(Widget:widget_message):void=
        AddCommandForAllPlayers(Commands.TurnRight)
      
    # Adds the turn left command to each player's command queue.
    OnTurnLeftSelected(Widget:widget_message):void=
        AddCommandForAllPlayers(Commands.TurnLeft)

    # Signals that the execute command button was selected.
    OnExecuteSelected(Widget:widget_message):void=
        ExecuteButtonSelected.Signal()

    # Signals that the execute command button was selected.
    OnResetSelected(Widget:widget_message):void=
        ResetButtonSelected.Signal()

    # Adds the given command to each player's command queue UI.
    AddCommandForAllPlayers(Command:command):void=
        for:
            Player : GetPlayspace().GetPlayers()
            CustomUI := UIPerPlayer[Player]
        do:
            CustomUI.AddCommandToDisplay(Command, UICommandLimit)

    # Removes the last command from each player's command queue UI.
    OnRemoveSelected(Widget:widget_message):void=
        for:
            Player : GetPlayspace().GetPlayers()
            CustomUI := UIPerPlayer[Player]
        do:
            CustomUI.RemoveLastCommandFromDisplay()

    # Draws the custom UI for each player in the game.
    DrawUIForAllPlayers():void=
        for:
            Player : GetPlayspace().GetPlayers()
            PlayerUI := GetPlayerUI[Player]
            CustomUI := UIPerPlayer[Player]
        do:
            CustomUI.Draw(PlayerUI)

    # Enable the HUDController, draw the UI for all players, and start the Gameloop. 
    # Exit the minigame when the final board is completed. 
    Start<private>()<suspends>:void=
        
        HUDController.Enable()

        DrawUIForAllPlayers()

        GameLoop()

        Exit()

    # Loops over the current gameboard and resets them.
    GameLoop<private>()<suspends>:void=
        # For the current board, swap to that gameboard's camera and reset the character to the gameboard's starting position.
        loop:
            if:
                Gameboard := Gameboards[CurrentBoard]
            then:
                # If first time on this board, set up the board
                # and move character to starting position.
                if:
                    FirstTimeOnCurrentBoard?
                then:
                    set FirstTimeOnCurrentBoard = false
                    BoardResetEvent.Signal(Gameboard.GetStartingCharacterPosition())
                    Gameboard.Start()
                
                # Race between the gameplay loop and the board reset.
                race:
                    LevelLoop(Gameboard)
                    AwaitReset(Gameboard)
            else:
                break
       
    # Handles command logic for the current gameboard. 
    LevelLoop<private>(Gameboard:gameboard)<suspends>:void=
        # On the current board, race between completing the board and looping player commands.
        # The race expression will cancel whichever action doesn't finish first.
        race:
            loop:
                defer:
                    # If the loop is canceled because the character reached the end goal of the level,
                    # Or the character finished performing their commands,
                    # reset the UI for all the players so they can interact with it and have no commands in the queue.
                    ResetUIForAllPlayers()

                # Wait for the player to choose to execute the commands.
                CommandQueue:[]command := WaitForExecuteCommand()

                if (CommandQueue.Length > 0):
                    # Turn off button interactivity for all players until a player chooses to reset or the character finishe
                    TurnOffButtonInteractivityForAllPlayers()

                    # Send all the commands to the character to perform.
                    ExecuteCommandsEvent.Signal(CommandQueue, Gameboard.TileSize)

                    # Wait till the character finishes moving. Then reset the UI for all players. When the race completes,=
                    CharacterFinishedMovingEvent.Await()
            Gameboard.EndGoalReached.Await()
        
        # Advance the CurrentBoard to the next board in the Gameboards array.
        # And mark that it's first time on the board.
        set CurrentBoard += 1
        set FirstTimeOnCurrentBoard = true

        # Wait one second.
        Sleep(1.0)

        # Clean up geamobard and stop the current level.
        Gameboard.Stop()

    # Wait for the Execute button to be selected, then returns the command queue of 
    # the first player in the playspace.
    WaitForExecuteCommand()<suspends>:[]command=
        ExecuteButtonSelected.Await()
        # Because all players share the same UI, get the command queue of the first player in the playspace.
        # Then return that command queue, or a blank array if this fails.
        if:
            APlayer := GetPlayspace().GetPlayers()[0]
            CustomUI := UIPerPlayer[APlayer]
        then:
            CustomUI.GetCommandQueue()
        else:
            array{}

    # Reset the UI for all players by enabling all their buttons again  and removing commands from the queue.
    ResetUIForAllPlayers():void=
        for (Player : GetPlayspace().GetPlayers(), CustomUI := UIPerPlayer[Player]):
            CustomUI.SetQueueButtonInteractivity(true)
            CustomUI.ClearQueue()

    # Disables UI button interactivity while waiting for commands to finish executing.
    TurnOffButtonInteractivityForAllPlayers():void=
        # For each player, get the UI for that player, and disable the interactivity of their UI buttons.
        for (Player : GetPlayspace().GetPlayers(), CustomUI := UIPerPlayer[Player]):
            CustomUI.SetQueueButtonInteractivity(false)

    # Waits for the Reset button to be selected, then resets the current gameboard
    # and NPC.
    AwaitReset<private>(Gameboard:gameboard)<suspends>:void=
        ResetButtonSelected.Await()
        # Reset the current gameboard, returning the game character to the starting position and 
        # resetting any barriers or triggers on the board.
        BoardResetEvent.Signal(Gameboard.GetStartingCharacterPosition())

        # Reset Gameboard
        Gameboard.Reset()

    # Clean up and end the minigame.
    Exit<private>():void=
        # Signal the minigame is over.
        GameEndedEvent.Signal()
        
        # Remove the HUD used for the minigame.
        HUDController.Disable()

        # Reset board data for next playthrough.
        set CurrentBoard = 0
        set FirstTimeOnCurrentBoard = true

        # Reset NPC Spawner for next playthrough.
        NPCSpawner.DespawnAll(false)
        NPCSpawner.Reset()

        # Remove the custom UI for all players.
        for:
            Player : GetPlayspace().GetPlayers()
            PlayerUI := GetPlayerUI[Player]
            CustomUI := UIPerPlayer[Player]
        do:
            CustomUI.Remove(PlayerUI)

# A log channel for printing messages to.
verse_commander_log_channel := class(log_channel):

# A project-wide Logger to print messages from functions that are not in a class with a log.
ProjectLog(Message:[]char, ?Level:log_level = log_level.Normal)<transacts>:void=
    Logger := log{Channel := verse_commander_log_channel}
    Logger.Print(Message, ?Level := Level)
