# Knowledge-neutral Verse snippet; project names and asset IDs generalized.
# id: snippet-226

# This file contains the Gameplay Tags for the project.
# You can use Gameplay Tags to tag an object in the island
# and be able to query all objects that have the tag during the game using Verse.

using { /Verse.org/Simulation/Tags }

# The verse_commander_minigame_tag is for the Verse Commander Character NPC script
# to be able to find the Verse Commander Minigame device and listen to its events.
verse_commander_minigame_tag := class(tag){}
