# Knowledge-neutral Verse snippet; project names and asset IDs generalized.
# id: snippet-227

using { /Verse.org/Simulation }
using { /UnrealEngine.com/Temporary/SpatialMath }

# Defines a box in 3D space
box<public> := struct<computes>:

    # Centre of box
    Center<public>:vector3 = vector3{X:=0.0,Y:=0.0,Z:=0.0}
    # Radii of box
    Extent<public>:vector3 = vector3{X:=0.0,Y:=0.0,Z:=0.0}

MakeBox<public>(InCenter:vector3, InExtent:vector3)<computes>:box=
    box:
        Center := InCenter
        Extent := InExtent

ToString(Box:box)<transacts>:string=
    "{Box.Center} : {Box.Extent}"

(Box:box).Max()<computes>:vector3=
    Box.Center + Box.Extent

(Box:box).Min()<computes>:vector3=
    Box.Center - Box.Extent
    

MakeZeroBox()<computes>:box=
    MakeBox(MakeZeroVector3(), MakeZeroVector3())


(Box:box).Overlaps(Other:box)<transacts><decides>:void=
    ThisMin := Box.Min()
    ThisMax := Box.Max()
    OtherMin := Other.Min()
    OtherMax := Other.Max()

    ThisMax.X >= OtherMin.X and
    ThisMin.X <= OtherMax.X and
    ThisMax.Y >= OtherMin.Y and
    ThisMin.Y <= OtherMax.Y and
    ThisMax.Z >= OtherMin.Z and
    ThisMin.Z <= OtherMax.Z
