# Knowledge-neutral Verse snippet; project names and asset IDs generalized.
# id: snippet-228

using { /Verse.org/Colors }
using { /Verse.org/Simulation }
using { /Verse.org/Colors/NamedColors }
using { /Verse.org/Assets }
using { /UnrealEngine.com/Temporary/SpatialMath }
using { /UnrealEngine.com/Temporary/UI }
using { /Fortnite.com/UI }
using { /Fortnite.com/Devices }
using { Textures }
using { Lib }

BuildingText<localizes> : message =  "Building"
ParkText<localizes> : message =  "Park"
NullText<localizes> : message =  "None"

CategoryToDisplayText(Category:build_category):message=
    case(Category):
        build_category.Building1 => BuildingText
        build_category.Park => ParkText
        build_category.Null => NullText


# Simple HUD that shows currently selected building type (building, park, none)
build_hud<public> := class(base_ui):
   
    # Score
    CategoryTextWidget<private> : text_block = text_block:
        DefaultTextColor := White
        DefaultJustification := text_justification.Left
        DefaultShadowColor := Black
        DefaultShadowOffset := option{vector2{X := 2.0, Y := 2.0}}


    CreateCanvas<override>():void=

        set Canvas = canvas:
            Slots := array:
                # Category text
                canvas_slot:
                    Widget := CategoryTextWidget
                    Anchors := anchors{Minimum := vector2{X := 0.5, Y := 0.9}, Maximum := vector2{X := 0.5, Y := 0.9}}
                    Alignment := vector2{X := 0.5, Y := 0.5}


    # Change the currently displayed category
    SetSelectedCategory<public>(InCategory:build_category):void=
        CategoryTextWidget.SetText(CategoryToDisplayText(InCategory))
