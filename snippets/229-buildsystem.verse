# Knowledge-neutral Verse snippet; project names and asset IDs generalized.
# id: snippet-229

using { /Verse.org/Simulation }
using { /Verse.org/Colors }
using { /Verse.org/Assets }
using { /Verse.org/Random }
using { /UnrealEngine.com/Temporary/SpatialMath }
using { /UnrealEngine.com/Temporary/Diagnostics }
using { /Fortnite.com/Devices }
using { Lib }
using { Lib.WFC }
using { SpaceGrammar }
using { Materials }

# Determines if a category uses the WFC method
CategoryUsesWFC(Category:build_category)<transacts><decides>:void=
    Category = build_category.Park


# Information about one box we found and will mesh
box_info<public> := class<unique>:
    # Actual box
    Box<public> : voxel_box
    # Category of box
    Category<public> : build_category
    # If we are about to remove this box
    var PendingRemoval : logic = false

# Util to see if Category is present in OnlyCategories
# If OnlyCategories is empty, it always matches
CategoryFilter(Category:build_category, OnlyCategories:[]build_category)<transacts><decides>:void=
    # If we supplied a filter, and this category was not in it, then we don't match
    OnlyCategories.Length = 0 or OnlyCategories.Find[Category]

# Debug util for printing array of categories
PrintCategoryFilter(OnlyCategories:[]build_category):void=
    Print("PrintCategoryFilter:")
    if(OnlyCategories.Length = 0):
        Print("  EMPTY")
    else:
        for(Cat : OnlyCategories):
            Print("  {Cat}")


# See if a modified box could require another box to be remeshed
BoxAffectedCheck(ModifiedBox:voxel_box, AffectedBox:voxel_box)<computes><decides>:void=
    ModifiedBox.Origin.X - AffectedBox.EndX() <= 1 and
    AffectedBox.Origin.X - ModifiedBox.EndX() <= 1 and
    ModifiedBox.Origin.Y - AffectedBox.EndY() <= 1 and
    AffectedBox.Origin.Y - ModifiedBox.EndY() <= 1 and
    ModifiedBox.Origin.Z - AffectedBox.EndZ() <= 1 and
    AffectedBox.Origin.Z - ModifiedBox.EndZ() <= 1

# Main class for the building system
# Contains the 3D grid of voxels, and main functions for decomposing that into convex boxes, and running the shape grammar rules.
build_system<public> := class:
    Logger:log = log{Channel:=LogClass}

    # 3D grid containing spawned Props (meshes)
    MeshGrid<public> : mesh_grid = mesh_grid{}
    # 3D grid of voxels edited by the player
    VoxelGrid<public> : voxel_grid = voxel_grid{}
    # Back reg to build size that owns this build system
    var BuildSite<public> : ?build_zone = false

    # Wave Function Collapse system used for this build system/site
    WFCSystem : wfc_system = wfc_system{}
    # Specific WFC Model for 'Park'
    var Park_WFCModel : wfc_model = wfc_model{}

    # Intermediate array of voxels, sorted by how far away they are from an empty voxel. Updating during processing.
    var VoxelsSortedByDepth : []tuple(vector3i, int) = array{}   
    # Length of VoxelsSortedByDepth array
    var VoxelsSortedByDepthLength : int = 0
    # Tracks how deep we recurse during sorting of VoxelsSortedByDepth
    var MaxRecursionReached : int = 0
    # Used to track progress while iterating through VoxelsSortedByDepth during convex decomposition
    var CurrentSortedVoxelIndex : int = 0

    # Current set of convex box regions covering all placed voxels
    var BoxInfos : []box_info = array{}
    # 3D grid (same size as VoxelGrid) which maps from voxel location to box_info which currently contains it
    var VoxelToBoxInfoLookup : []?box_info = array{}

    # Set during rule process - current box_info being processed by shape grammar rules
    var CurrentBoxInfo : ?box_info = false
    # Set during rule process - current rule set being processed by shape grammar rules
    var CurrentRuleSet : ?rs_base = false
    # Total number of props spawned during processing
    var PropCount:int = 0

    # Have shape grammar rule sets been initialized
    var RuleSetsInitialized : logic = false
    # Rule set for 'Building' category
    var RS_Building1:rs_base = rs_base{}
    # Rule set for 'Park' category
    var RS_WFC:rs_base = rs_base{}

    # Initilizes the grid systems, for the supplied 
    Init<public>(InMainDevice:build_zone, XSize:int, YSize:int, ZSize:int):void=
        set BuildSite = option{InMainDevice}
        MeshGrid.Init(Self, XSize, YSize, ZSize)
        VoxelGrid.Init(Self, XSize, YSize, ZSize)

        # Init the rule sets
        InitRuleSets()

        # Create Park WFC model and init WFC system
        PT := InMainDevice.RootDevice.SpawnReferencesDevice.GetPT()
        WFCFactory := wfc_model_factory{}
        set Park_WFCModel = WFCFactory.GetParkModel(PT)

        WFCSystem.Initialize(Self)        

        # Init voxel->box info map to all false
        set VoxelToBoxInfoLookup = for (X := 0..VoxelGrid.GetVoxelCount()-1):
            false

        # Init sorted voxel array to max size (don't want to resize at runtime)
        set VoxelsSortedByDepth = for (X := 0..VoxelGrid.GetVoxelCount()-1):
            (vector3i{}, 0)   


    #completely reset the build grids (this is only meant to happen when the mission is completed and all players leave the build site)
    Reset<public>():void=

        set BoxInfos = array{}
            
        set VoxelToBoxInfoLookup = for (X := 0..VoxelGrid.GetVoxelCount()-1):
            false

        MeshGrid.Clear()
        VoxelGrid.Clear()
        WFCSystem.ResetWFC()

    # Debug util for printing state of VoxelsSortedByDepth
    PrintVoxelsSortedByDepth(Low:int, High:int):void=
        for:
            V := Low .. High
            Depth := VoxelsSortedByDepth[V](1)
        do:
            Print("{V} : {Depth}")

    # Partition the VoxelsSortedByDepth given the supplied low and high indices
    PartitionVoxelsSortedByDepth(Low:int, High:int):?int=
        # See if all elements are the same
        var AllTheSame:logic = true
        if(Check := VoxelsSortedByDepth[Low]):       
            CheckVal := Check(1)
            for:
                C := Low+1 .. High            
                Val := VoxelsSortedByDepth[C](1)
                AllTheSame?
                Val <> CheckVal
            do:
                set AllTheSame = false
            
        if:
            not AllTheSame? # Don't partition if all the same
            PivotIndex:int := Floor((Low + High)/2)
            PivotVal := VoxelsSortedByDepth[PivotIndex](1)
        then:
            var I:int := Low - 1

            # Swap PivotIndex and High elements (put Pivot at High)
            if:
                NewHigh := VoxelsSortedByDepth[PivotIndex]
                NewPivotIndex := VoxelsSortedByDepth[High]                    
                set VoxelsSortedByDepth[PivotIndex] = NewPivotIndex
                set VoxelsSortedByDepth[High] = NewHigh

            for(J := Low .. High-1):
                if(VoxelsSortedByDepth[J](1) > PivotVal):
                    set I += 1
                    # Swap I and J elements
                    if:
                        NewJ := VoxelsSortedByDepth[I]
                        NewI := VoxelsSortedByDepth[J]                    
                        set VoxelsSortedByDepth[I] = NewI
                        set VoxelsSortedByDepth[J] = NewJ

            # Swap I+1 and High elements (put pivot at I+1)
            if:
                NewHigh := VoxelsSortedByDepth[I+1]
                NewIPlus1 := VoxelsSortedByDepth[High]                    
                set VoxelsSortedByDepth[I+1] = NewIPlus1
                set VoxelsSortedByDepth[High] = NewHigh

            #if(VoxelsSortedByDepthLength < 20):
            #   PrintVoxelsSortedByDepth(Low, High)

            return option{I+1}
        else:
            #if(VoxelsSortedByDepthLength < 20):
            #    Print("{Low}-{High} : All same, not change")

            return false

    # Quick sort a section of the VoxelsSortedByDepth array, defined by supplied indices
    QuickSortVoxelsSortedByDepth(Low:int, High:int, RecursionLevel:int):void=
        if(Low < High):
            set MaxRecursionReached = Max(MaxRecursionReached, RecursionLevel)

            if(RecursionLevel < 50):
                POpt := PartitionVoxelsSortedByDepth(Low, High)
                if(P := POpt?):
                    #Print("P:{P}")
                    QuickSortVoxelsSortedByDepth(Low, P-1, RecursionLevel+1)
                    QuickSortVoxelsSortedByDepth(P+1, High, RecursionLevel+1)

            else:
                Print("Recursion limit reached!")

    # Find the next 'deepest' voxel (ie furthest from an empty voxel), that is not already part of a convex box 
    FindNextDeepestVoxel()<decides><transacts>:vector3i=
        var VoxelOpt:?vector3i = false

        # Keep walking through VoxelsSortedByDepth to find the next valid entry
        loop:
            if(CurrentSortedVoxelIndex > VoxelsSortedByDepthLength-1 or VoxelOpt?): # quit once we hit the end of the array, or have a valid cell
                break

            # If this voxel is still unused, return it
            if:
                VInf := VoxelsSortedByDepth[CurrentSortedVoxelIndex]
                VIndex := VoxelGrid.GetVoxelIndex(VInf(0))
                BoxInfoOpt := VoxelToBoxInfoLookup[VIndex]
                ((not BoxInfoOpt?) or (BoxInfoOpt?.PendingRemoval?)) # either no box, or box is pending removal
            then:
                set VoxelOpt = option{VInf(0)}
            
            set CurrentSortedVoxelIndex += 1

        VoxelOpt?

    # Expand a box one voxel out from a face of the box
    ExpandBoxFaceInSDF(Box:voxel_box, Face:voxel_box_face, Category:build_category)<transacts><decides>:voxel_box=
        # Get set of cells we want to move into. Will fail if against edge of grid.
        FaceCells:voxel_box = VoxelGrid.GetCellsTouchingBoxFace[Box, Face]
        
        # See if they are all valid 
        for:
            X := FaceCells.Origin.X .. FaceCells.EndX()
            Y := FaceCells.Origin.Y .. FaceCells.EndY()
            Z := FaceCells.Origin.Z .. FaceCells.EndZ()
        do:
            # If not filled in, or already used for a box, or the wrong category, can't expand
            VIdx := VoxelGrid.GetVoxelIndex(X, Y, Z)
            VoxelGrid.SDF[VIdx]?
            VoxelGrid.Cells[VIdx]?.Category = Category
            BoxInfoOpt := VoxelToBoxInfoLookup[VIdx]                            
            ((not BoxInfoOpt?) or (BoxInfoOpt?.PendingRemoval?)) # either no box, or box is pending removal

        Box.Expand(Face)


    # Expand the box as far as possible along the face directions provided
    ExpandBoxFacesInSDF(InBox:voxel_box, InCategory:build_category, InFaceArray:[]?voxel_box_face):voxel_box=

        var ResultBox:voxel_box = InBox
        var FaceArray:[]?voxel_box_face = InFaceArray
        NumFaces := FaceArray.Length
        var FaceCount:int = 0      
        loop:
            for(FIdx := 0 .. NumFaces-1): # For each face..                
                if(Face := FaceArray[FIdx]?): # Should we try to expand this face                 
                    if(ExpandResult := ExpandBoxFaceInSDF[ResultBox, Face, InCategory]): # Try and expand the box...
                        set ResultBox = ExpandResult # We expanded, update result
                    else:
                        # We couldn't, clear entry so we don't try this face again (and keep count of how many faces have been cleared)
                        if( set FaceArray[FIdx] = false ) {}
                        set FaceCount += 1                                

            if(FaceCount = NumFaces): # If we've gone as far as we can along all faces, stop
                break
        
        return ResultBox

    # Expand this box as far as possible along all faces
    ExpandBoxInSDF(InVoxel:vector3i, InCategory:build_category):voxel_box=
        # Init array
        var XYFaceArray:[]?voxel_box_face = array:
            option{voxel_box_face.PosX}
            option{voxel_box_face.NegX}
            option{voxel_box_face.PosY}
            option{voxel_box_face.NegY}

        var ZFaceArray:[]?voxel_box_face = array:
            option{voxel_box_face.PosZ}
            option{voxel_box_face.NegZ}
        
        var ResultBox:voxel_box = MakeVoxelBox(InVoxel, MakeVector3i(1, 1, 1))            
        set ResultBox = ExpandBoxFacesInSDF(ResultBox, InCategory, XYFaceArray)
        set ResultBox = ExpandBoxFacesInSDF(ResultBox, InCategory, ZFaceArray)

        return ResultBox


    # Util to set all VoxelToBoxInfoLookup entries in this box to an index
    SetVoxelToBoxInfoLookup(Box:voxel_box, BoxInfo:?box_info):void=
        for(X := Box.Origin.X .. Box.EndX()):
            for(Y := Box.Origin.Y .. Box.EndY()):
                for(Z := Box.Origin.Z .. Box.EndZ()):
                    if:
                        set VoxelToBoxInfoLookup[VoxelGrid.GetVoxelIndex(X, Y, Z)] = BoxInfo

    # Add all voxels in supplied box to the VoxelsSortedByDepth array (at the end)
    AddBoxToVoxelsForSorting(Box:voxel_box):void=
                
        for:
            X := Box.Origin.X .. Box.EndX()
            Y := Box.Origin.Y .. Box.EndY()
            Z := Box.Origin.Z .. Box.EndZ()
            VIdx := VoxelGrid.GetVoxelIndex(X,Y,Z)
            Depth := VoxelGrid.SDF[VIdx]?.SumMinDist
        do:
            if( set VoxelsSortedByDepth[VoxelsSortedByDepthLength] = (vector3i{X:=X,Y:=Y,Z:=Z}, Depth) ) {}
            set VoxelsSortedByDepthLength += 1


    # Update BoxInfos array - set of convex boxes covering all voxels in the grid
    # This is done incrementally, so only affected boxes are modified/reprocessed.
    # OnlyCategories - only updates boxes for voxels in these categories
    # ModifiedBox - only consider voxels in this box as having changed
    UpdateBoxesFromSDF(OnlyCategories:[]build_category, ModifiedBox:voxel_box):void=

        #VoxelGrid.VoxelDebugDraw.Clear() # Clear debug drawing so we can draw it now

        var PendingRemoveBoxInfos:[]box_info = array{}
        var KeptBoxInfos:[]box_info = array{}

        # Boxes to mesh
        var BoxInfosToMesh:[]box_info = array{} 

        set VoxelsSortedByDepthLength = 0
        AddBoxToVoxelsForSorting(ModifiedBox)

        #profile("    RemoveBoxes"):
        block:

            for(I -> BoxInfo : BoxInfos):

                # Affected and a category we are updating, mark box info as pending delete
                BoxAffected := logic{ BoxAffectedCheck[ModifiedBox, BoxInfo.Box] }
                if:
                    CategoryFilter[BoxInfo.Category, OnlyCategories]
                    BoxAffected?
                then:
                    set PendingRemoveBoxInfos += array{BoxInfo}
                    set BoxInfo.PendingRemoval = true

                    # Add voxels in this box to set to rebox
                    AddBoxToVoxelsForSorting(BoxInfo.Box)

                # Not a category we are updating, keep as it is
                else:
                    set KeptBoxInfos += array{BoxInfo}            

                    #if(DrawDebugBoxes?) then VoxelGrid.DebugDrawVoxelBox(BoxInfo.Box, NamedColors.Blue) # DEBUG

                    # See if this box touches affected area. If so Telekinetic a remesh
                    if(BoxAffected?):
                        set BoxInfosToMesh += array{BoxInfo}

                    else if(DrawDebugBoxes?):
                        VoxelGrid.DebugDrawVoxelBox(BoxInfo.Box, NamedColors.Gray, 5.0) # DEBUG

        set BoxInfos = KeptBoxInfos

        if(DrawDebugBoxes?):
            for(BI : BoxInfos):
                VoxelGrid.DebugDrawVoxelBox(BI.Box, NamedColors.Grey, 5.0) # DEBUG

        #profile("    SortVoxels"):
        block:
            if(VoxelsSortedByDepthLength > 1):
                set MaxRecursionReached = 0
                QuickSortVoxelsSortedByDepth(0, VoxelsSortedByDepthLength-1, 0)                
            
        set CurrentSortedVoxelIndex = 0        

        #profile("    BuildBoxes"):
        block:

            # Make boxes, starting from the 'deepest' voxels
            loop:
                # Find the next deepest (unused) voxel 
                if:
                    StartVoxel:vector3i = FindNextDeepestVoxel[]
                then:
                    # Debug draw starting voxel for box
                    #if(DrawDebugBoxes?) then VoxelGrid.DebugDrawVoxel(StartVoxel.X, StartVoxel.Y, StartVoxel.Z, NamedColors.PeachPuff, 20.0)                    
                    
                    if:
                        Category := VoxelGrid.GetCellCategory[StartVoxel.X, StartVoxel.Y, StartVoxel.Z] # Get the category of this voxel
                    then:
                
                        #Print("-Start {StartVoxel} {Category}")
                        var Box:voxel_box = ExpandBoxInSDF(StartVoxel, Category) # Start from voxel and expand as much as you can
                        #Print("-Made {Box}")

                        # See if we already have a BoxInfo for this box and category
                        var FoundBox:logic = false

                        StartVIdx := VoxelGrid.GetVoxelIndex(StartVoxel)
                        if:
                            PreviousBoxInfo := VoxelToBoxInfoLookup[StartVIdx]?
                            PreviousBoxInfo.Box.Equals[Box]   
                            PreviousBoxInfo.Category = Category
                        then:
                            # Keep this box
                            set PreviousBoxInfo.PendingRemoval = false # clear removal flag
                            set PendingRemoveBoxInfos = for(BI : PendingRemoveBoxInfos, BI <> PreviousBoxInfo): # remove from pending removal array
                                BI
                            set FoundBox = true
                            set BoxInfos += array{PreviousBoxInfo} # add to boxes we are keeping

                            #Print("--Found existing")

                            # See if this box we are keeping is touching the modified area and needs to be remeshed
                            if(BoxAffectedCheck[ModifiedBox, PreviousBoxInfo.Box]):
                                set BoxInfosToMesh += array{PreviousBoxInfo}                                    

                            if(DrawDebugBoxes?) then VoxelGrid.DebugDrawVoxelBox(Box, NamedColors.Yellow, 5.0) # DEBUG
                        else:
                            #Print("--No existing")

                            # Fully remove boxes that overlap this new box
                            var KeepPending:[]box_info = array{}

                            for(BI : PendingRemoveBoxInfos):                             
                                if(BI.Box.Intersects[Box]):
                                    SetVoxelToBoxInfoLookup(BI.Box, false) # Remove from VoxelToBoxInfoLookup   TODO PERF Could just rely on entries being overwritten by new box?
                                    #Print("--Destroy {BI.Box}")
                                else:
                                    set KeepPending += array{BI}
                            set PendingRemoveBoxInfos = KeepPending                    
                                                                
                                                                
                        # Couldn't find an existing box, make a new one
                        if(not FoundBox?):
                            #Print("--Make new {Box}")

                            NewBoxInfo := box_info{ Box:=Box, Category:=Category }
                            set BoxInfos += array{NewBoxInfo} # Add to BoxInfos array                            
                            set BoxInfosToMesh += array{NewBoxInfo} # Always mesh new boxes
                            SetVoxelToBoxInfoLookup(Box, option{NewBoxInfo}) # Store in lookup                                                        

                            if(DrawDebugBoxes?) then VoxelGrid.DebugDrawVoxelBox(Box, NamedColors.Red, 5.0) # DEBUG


                # Failed to get a cell, so break
                else:
                    break   

            # Remove any remaining boxes that were not saved
            for(BI : PendingRemoveBoxInfos):                             
                SetVoxelToBoxInfoLookup(BI.Box, false) # Remove from VoxelToBoxInfoLookup   TODO PERF Could just rely on entries being overwritten by new box?
    

        #profile("    ApplyMeshing"):
        block:

            # Apply meshing
            set PropCount = 0
            for(I -> BoxInfo : BoxInfosToMesh):
                if(Rules := SelectRuleSet[BoxInfo]):

                    set CurrentBoxInfo = option{BoxInfo} # Current box info can be used by rules
                    set CurrentRuleSet = option{Rules}
                    Rules.ApplyRulesToBox(BoxInfo.Box, Self) # Use style to apply props
                    set CurrentBoxInfo = false
                    set CurrentRuleSet = false
            Logger.Print("     {BoxInfosToMesh.Length} boxes meshed, {PropCount} Props spawned")


    #Print("Mesh Update Done ({PropCount} Props updated)") 

  
    # Allocate/initialize all rule sets
    InitRuleSets():void=

        if(BZ := BuildSite?):

            PT := BZ.RootDevice.SpawnReferencesDevice.GetPT()

            set RS_Building1 = rs_building1{}
            RS_Building1.SetupRules(PT)        

            set RS_WFC = rs_wfc{}
            RS_WFC.SetupRules(PT)


    # Select the rule set to apply to the supplied box
    SelectRuleSet(BoxInfo:box_info)<transacts><decides>:rs_base=
        case(BoxInfo.Category):
            build_category.Building1 => RS_Building1
            build_category.Park => RS_WFC


    # Clear a rectangular region 
    ClearRegion<public>(ClearBox:voxel_box):void=

        var ModifiedCategories:[]build_category = array{}

        for:
            X := ClearBox.Origin.X .. ClearBox.EndX()
            Y := ClearBox.Origin.Y .. ClearBox.EndY()            
            Z := ClearBox.Origin.Z .. ClearBox.EndZ()            
        do:
            Voxel := MakeVector3i(X,Y,Z)
            if(PreviousCell := VoxelGrid.GetCell[Voxel]):
                PreviousCategory := PreviousCell.Category

                # Add to list of modified categories, if not already present
                if(not ModifiedCategories.Find[PreviousCategory]):
                    set ModifiedCategories += array{PreviousCategory}

                # Remove WFC tile
                if(CategoryUsesWFC[PreviousCell.Category]):
                    RemoveWFCTile(Voxel, 0)

            # Clear voxel
            VoxelGrid.ClearVoxel(Voxel)

        VoxelGrid.RemoveBoxFromSDF(ClearBox, ModifiedCategories)
        MeshGrid.ClearPropsForBox(ClearBox)
        UpdateBoxesFromSDF(ModifiedCategories, ClearBox)


    # Add an entire rectangular region of the specified category 
    AddRegion<public>(AddBox:voxel_box, Category:build_category):void=

        # Set each voxel in the supplied box to the desired category
        for:
            X := AddBox.Origin.X .. AddBox.EndX()
            Y := AddBox.Origin.Y .. AddBox.EndY()            
            Z := AddBox.Origin.Z .. AddBox.EndZ()            
        do:
            Voxel := MakeVector3i(X,Y,Z)
            VoxelGrid.SetVoxel(Voxel, Category)

        if(Category = build_category.Park):
            AddParkWFCRegion(AddBox, 1)

        # Udpate Signed Distance Field, given the modified region and category
        ModifiedCategories := array{Category}
        VoxelGrid.AddBoxToSDF(AddBox, ModifiedCategories)

        # Update boxes from SDF, for supplied category and region. Will re-run rules
        UpdateBoxesFromSDF(ModifiedCategories, AddBox) 



    # Adds a park WFC tile at this voxel
    AddParkWFCTile(Voxel:vector3i, RecalcRadius:int):void=
        WFCSystem.AddTile(Voxel, Park_WFCModel, RecalcRadius)

    # Adds a park WFC region 
    AddParkWFCRegion(Region:voxel_box, RecalcRadius:int):void=
        WFCSystem.AddRegion(Region, Park_WFCModel, RecalcRadius)

    # Remove WFC tile from this voxel
    RemoveWFCTile(Voxel:vector3i, RecalcRadius:int):void=
        WFCSystem.RemoveTile(Voxel, RecalcRadius)
        
    # Notify when a tile is changed
    NotifyWFCTileChanged(Voxel:vector3i, Category:build_category, RecalcRadius:int):void=
        WFCSystem.NotifyChangedVoxel(Voxel, Category, RecalcRadius)
                           
    # Returns if a voxel is consdiered 'solid' in the vertical direction
    VoxelIsVerticallySolid(InX:int, InY:int, InZ:int)<transacts><decides>:void=
        var IsSolid:logic = false       
        if(Category := VoxelGrid.GetCellCategory[InX, InY, InZ]):
            set IsSolid = GetSolidityOfCategory(Category)(1)
        IsSolid?

    # Returns if a voxel is consdiered 'solid' in the horizontal direction
    VoxelIsHorizontallySolid(InX:int, InY:int, InZ:int)<transacts><decides>:void=
        var IsSolid:logic = false       
        if(Category := VoxelGrid.GetCellCategory[InX, InY, InZ]):
            set IsSolid = GetSolidityOfCategory(Category)(0)
        IsSolid?
