# Knowledge-neutral Verse snippet; project names and asset IDs generalized.
# id: snippet-230

using { /Verse.org/Simulation }
using { /Verse.org/Random }
using { /Verse.org/Colors/NamedColors }
using { /UnrealEngine.com/Temporary/Diagnostics }
using { /UnrealEngine.com/Temporary/SpatialMath }
using { /Fortnite.com/Characters }
using { /Fortnite.com/AI }
using { /Fortnite.com/Devices }
using { /Fortnite.com/UI }
using { Lib }

EnteredSiteMessage<localizes><public> : message = "Entered site"
UnableToEnterSiteMessage<localizes><public> : message = "Could not enter site"
ExitedSiteMessage<localizes><public> : message = "Left site"
SiteAlreadyFull<localizes><public> : message = "Site already has maximum players"

BuildGridSize:vector3 := vector3{X:=500.0, Y:=500.0, Z:=400.0}

# Device  that represents one buildable region in the level
build_zone := class<unique>(creative_device):
    Logger:log = log{Channel:=LogClass}

    ZoneDebugDraw<public> : debug_draw = debug_draw{}

    # Root device reference
    @editable
    RootDevice<public> : root_device = root_device{}

    # Device used to spawn props
    @editable
    SpawnerDevice : prop_spawner_device = prop_spawner_device{}

    # Grid size in voxels
    @editable
    var GridSize : vector3i = vector3i{}
    
    # volume for entering site
    @editable
    var EnterSiteVolume : volume_device = volume_device{}

    # volume for leaving site
    @editable
    var ExitSiteVolume : volume_device = volume_device{}

    # Button for adding random park
    @editable
    var GenerateParkButton : button_device = button_device{}

    # Button for adding random building
    @editable
    var GenerateBuildingButton : button_device = button_device{}

    # Button to reset zone
    @editable
    var ResetButton : button_device = button_device{}

    # Players playing on the build zone
    var Agents : []agent = array{}

    # Stores player data
    var AgentDataMap<public> : [agent]player_data = map{}

    DebugDraw<public> : debug_draw = debug_draw{}

    # World space location of build site origin
    var ZoneOrigin<public> : vector3 = vector3{}
    
    # Stores voxel grid status, and processes spawning of meshes (props) via shape grammar/WFC rules
    BuildSystem<public> : build_system = build_system{}

    # Initialize a build zone
    InitBuildZone<public>():void=
        Print("InitBuildZone")

        # show debug drawing channel by default
        DebugDraw.ShowChannel()
        
        # Register events when a player enters or leaves the zone
        EnterSiteVolume.AgentEntersEvent.Subscribe(EnteredSiteVolume)
        ExitSiteVolume.AgentExitsEvent.Subscribe(ExitedSiteVolume)

        # Use device location as origin of the build site grid
        set ZoneOrigin = Self.GetTransform().Translation

        # Init the build system
        BuildSystem.Init(Self, GridSize.X, GridSize.Y, GridSize.Z)

        GenerateParkButton.InteractedWithEvent.Subscribe(PressedGeneratePark)
        GenerateBuildingButton.InteractedWithEvent.Subscribe(PressedGenerateBuilding)
        ResetButton.InteractedWithEvent.Subscribe(PressedReset)
        
    # Player entered the site volume
    EnteredSiteVolume(Agent:agent):void=  

        # Check we are not already in zone    
        if(not Agents.Find[Agent]):           
            AddPlayer(Agent)        

    # Player left the site volume
    ExitedSiteVolume(Agent:agent):void=      

        # Leave if in zone
        if(Agents.Find[Agent]):
            RootDevice.MessageHUD.SetText(ExitedSiteMessage)
            RootDevice.MessageHUD.Show(Agent)
            RemovePlayer(Agent)

    # Player pressed the 'generate park' button. 
    # Randomly generate an area of park
    PressedGeneratePark(Agent:agent):void=     
        MaxRandomParkSize := 5 
        XMinAndSize := GetRandMinAndSize(GridSize.X, MaxRandomParkSize)
        YMinAndSize := GetRandMinAndSize(GridSize.Y, MaxRandomParkSize)

        ParkOrigin := MakeVector3i(XMinAndSize(0), YMinAndSize(0), 0)
        ParkSize := MakeVector3i(XMinAndSize(1), YMinAndSize(1), 1)
        AddBox := MakeVoxelBox(ParkOrigin, ParkSize)

        BuildSystem.ClearRegion(AddBox)
        BuildSystem.AddRegion(AddBox, build_category.Park)
        
    # Player pressed the 'generate building' button
    # Randomly generate a volume of building
    PressedGenerateBuilding(Agent:agent):void=     
        MaxRandomBuildingXYSize := 4
        XMinAndSize := GetRandMinAndSize(GridSize.X, MaxRandomBuildingXYSize)
        YMinAndSize := GetRandMinAndSize(GridSize.Y, MaxRandomBuildingXYSize)
        ZSize := GetRandomInt(1,5)

        BuildingOrigin := MakeVector3i(XMinAndSize(0), YMinAndSize(0), 0)
        BuildingSize := MakeVector3i(XMinAndSize(1), YMinAndSize(1), ZSize)
        AddBox := MakeVoxelBox(BuildingOrigin, BuildingSize)

        BuildSystem.ClearRegion(AddBox)
        BuildSystem.AddRegion(AddBox, build_category.Building1)

    # Player pressed the 'reset' button
    PressedReset(Agent:agent):void=     
        ResetBuildzone()

    # Add a player to this site
    AddPlayer<public>(Agent:agent):void=
        if(GPD := RootDevice.AgentToGlobalDataMap[Agent]):

            # If too many in that site
            if(Agents.Length >= 4): 
                # Show message
                RootDevice.MessageHUD.SetText(SiteAlreadyFull)
                RootDevice.MessageHUD.Show(Agent)
            else:
                # If already in another site, leave that first
                if(CurrentSite := GPD.CurrentZone?):
                    CurrentSite.RemovePlayer(Agent)

                set Agents += array{Agent}

                if:
                    PlayerData := player_data:
                        Agent := Agent
                        GlobalData := GPD
                        Character := Agent.GetFortCharacter[]
                        BuildZoneDevice := Self
                        RootDevice := RootDevice
                        UIManager := GPD.UIManager
                    set AgentDataMap[Agent] = PlayerData

                    set GPD.CurrentZone = option{Self}
                then:
                    PlayerData.InitPlayerData()

                # Show message
                RootDevice.MessageHUD.SetText(EnteredSiteMessage)
                RootDevice.MessageHUD.Show(Agent)                

    # Remove a player from the site
    RemovePlayer<public>(Agent:agent):void=
        if:
            TempAgents := Agents.RemoveFirstElement[Agent] # fails if player is not in the site, no further work will happen
            set Agents = TempAgents
            GPD := RootDevice.AgentToGlobalDataMap[Agent]
            ExistingPlayerData := AgentDataMap[Agent]
        then:
            set GPD.CurrentZone = false # clear zone ref in global player data

            ExistingPlayerData.Removed() # tell player data it is done

            # Remove entry for Agent from AgentDataMap
            var NewAgentDataMap:[agent]player_data = map{}
            for (Key -> Value : AgentDataMap, Key <> ExistingPlayerData):
                set NewAgentDataMap = ConcatenateMaps(NewAgentDataMap, map{Key => Value})
            set AgentDataMap = NewAgentDataMap


    #Resets the buildzone to initial state
    ResetBuildzone<public>():void=
        Print("Reset Build Zone")
        BuildSystem.Reset()




    
    # Util to get a random min and size within a range
    GetRandMinAndSize(Range:int, MaxSize:int):tuple(int,int)=
        ResultSize := GetRandomInt(1, MaxSize)
        MaxMin := Range - ResultSize
        ResultMin := GetRandomInt(0, MaxMin)
        return (ResultMin, ResultSize)
