# Knowledge-neutral Verse snippet; project names and asset IDs generalized.
# id: snippet-231

# Handy global configuration options

# Should draw boxes for debugging shape grammar process
DrawDebugBoxes:logic = false
