# Knowledge-neutral Verse snippet; project names and asset IDs generalized.
# id: snippet-232

using { /Fortnite.com/UI }
using { /UnrealEngine.com/Temporary/SpatialMath }
using { /UnrealEngine.com/Temporary/UI }
using { /Verse.org/Colors }
using { /Verse.org/Simulation }
using { /Verse.org/Colors/NamedColors }
using { /Fortnite.com/Devices }
using { /Verse.org/Assets }
using { Textures }
using { Lib }

# Simple UI that places a dot in the center of the scren
cursor_ui<public> := class(base_ui):

    ReticleWidget:color_block := color_block:
        DefaultColor := White
        DefaultDesiredSize := vector2{X := 10.0, Y := 10.0}

    CreateCanvas<override>():void=
        set Canvas = canvas:
            Slots := array:
                canvas_slot:
                    Widget := ReticleWidget
                    Anchors := anchors{Minimum := vector2{X := 0.5, Y := 0.5}, Maximum := vector2{X := 0.5, Y := 0.5}}
                    Alignment := vector2{X := 0.5, Y := 0.5}

    # Update whether reticle is visible
    SetReticleVisible<public>(Visible:logic):void=
        if(Visible?):
            ReticleWidget.SetVisibility(widget_visibility.Visible)
        else:
            ReticleWidget.SetVisibility(widget_visibility.Hidden)
