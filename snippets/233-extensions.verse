# Knowledge-neutral Verse snippet; project names and asset IDs generalized.
# id: snippet-233

using { /Verse.org/Simulation }
using { /Verse.org/Colors }
using { /UnrealEngine.com/Temporary/SpatialMath }
using { /Fortnite.com/Devices }
using { /Fortnite.com/Characters }
using { /Fortnite.com/AI }


# Get the box for a player
(Character:fort_character).GetBox()<transacts>:box=
    var ResultBox:box = MakeZeroBox()
    set ResultBox.Center = Character.GetTransform().Translation
    set ResultBox.Extent = MakeVector3(50.0, 50.0, 192.0)
    return ResultBox

Max<public>(X:int, Y:int, Z:int)<computes>:int=
    Max(X,Max(Y,Z))

SafeDivide(Dividend:float, Divisor:float)<transacts>:float=
    Divisor = 0.0 and Inf or Dividend / Divisor

ToString(Color:color):string=
    "R={Color.R} G={Color.G} B={Color.B}"

InvertLogic(Logic:logic):logic=
    return logic{not Logic?}

LogicToString<public>(Logic:logic)<computes>:string=
    if(Logic?) then "true" else "false"

MakeVector3<public>(InX:float, InY:float, InZ:float)<computes>:vector3=
    vector3{X:=InX, Y:=InY, Z:=InZ}

MakeZeroVector3<public>()<computes>:vector3=
    vector3{X:=0.0, Y:=0.0, Z:=0.0}

MakeUnitVector3<public>()<computes>:vector3=
    vector3{X:=1.0, Y:=1.0, Z:=1.0}

MakeVector2i<public>(InX:int, InY:int)<computes>:vector2i=
    vector2i{X:=InX, Y:=InY}

MakeColor<public>(R:float, G:float, B:float)<computes>:color=
    color{R:=R, G:=G, B:=B}
    
EmptyMessage<public><localizes>:message = ""
