# Knowledge-neutral Verse snippet; project names and asset IDs generalized.
# id: snippet-234

using { /Fortnite.com/Devices }
using { /Verse.org/Simulation }
using { /UnrealEngine.com/Temporary/UI }
using { /UnrealEngine.com/Temporary/SpatialMath }
using { /Fortnite.com/Characters }



# Class containing game data about one player
global_player_data<public> := class:
    # Player owning this data
    Player<public>:player
    # Root device instance
    RootDevice<public>:root_device
    # UI manager for this player
    UIManager<public> : ui_manager = ui_manager {}
    # Current build zone this player is in
    var CurrentZone<public>:?build_zone = false

    InitGlobalPlayerData<public>():void=
        UIManager.InitUI(Player, RootDevice, Self)
