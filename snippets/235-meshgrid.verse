# Knowledge-neutral Verse snippet; project names and asset IDs generalized.
# id: snippet-235

using { /Fortnite.com/Characters }
using { /Fortnite.com/Devices }
using { /UnrealEngine.com/Temporary/Diagnostics }
using { /UnrealEngine.com/Temporary/SpatialMath }
using { /Verse.org/Simulation }
using { /Verse.org/Simulation/Tags }
using { /Verse.org/Assets }

# Util for creating a transform
held_transform<public> := class:

    var<private> Rotation<public> : rotation = IdentityRotation()
    var<private> Scale<public> : vector3 = vector3{X := 1.0, Y := 1.0, Z := 1.0}

    RotatePiece<public>(Yaw:prop_yaw)<transacts>:void=
        set Rotation = Rotation.ApplyYaw(PropYawToRadians(Yaw))

    ToTransform<public>(Translation:vector3)<transacts>:transform=
        transform{Rotation := Rotation, Scale := Scale, Translation := Translation}

# Storage class to handle the build state (what pieces have been placed and where)
mesh_grid<public> := class:
    Logger:log = log{Channel:=LogClass}

    var BuildSystem<public> : ?build_system = false

    # 3D array of cells, each potentially containing a prop spawned with that cell as its origin
    var PropGrid : []?prop_entry = array{}

    # Current size of grid
    var Size : tuple(int, int, int) = (-1,-1,-1)
    
    # Get 1D array index from 3D voxel location
    GetVoxelIndex<public>(X:int, Y:int, Z:int)<transacts>:int=
        (X * (Size(1)*Size(2))) + (Y * Size(2)) + Z

    GetVoxelIndex<public>(Vec:vector3i)<transacts>:int=
        (Vec.X * (Size(1)*Size(2))) + (Vec.Y * Size(2)) + Vec.Z

    # Get total number of voxels in grid
    GetVoxelCount<public>()<transacts>:int=
        Size(0)*Size(1)*Size(2)

    # Initialize mesh grid
    Init<public>(_BuildSystem:build_system, XSize:int, YSize:int, ZSize:int):void=
        set Size = (XSize, YSize, ZSize)
        set BuildSystem = option{_BuildSystem}

        set PropGrid = for (X := 0..GetVoxelCount()-1):
            false

    # Clear mesh grid, destroying all spawned props
    Clear<public>():void=
        if(BZ := BuildSystem?.BuildSite?):

            for(Idx := 0 .. GetVoxelCount()-1):
                # Clear prop for this voxel
                if(PE := PropGrid[Idx]?):
                    BZ.SpawnerDevice.CheckIn(PE)

                if( set PropGrid[Idx] = false ) {} # Empty array

    # Clear any props at the supplied voxel
    ClearPropForVoxel(Voxel:vector3i):void=
        if(BZ := BuildSystem?.BuildSite?):
            Idx := GetVoxelIndex(Voxel)
            # Clear prop for this voxel
            if(PE := PropGrid[Idx]?):
                BZ.SpawnerDevice.CheckIn(PE)

            if( set PropGrid[Idx] = false ) {} # Empty array

    # Clear any props in the supplied box
    ClearPropsForBox<public>(Box:voxel_box):void=

        if(BZ := BuildSystem?.BuildSite?):
            # iterate over box
            for: 
                X := Box.Origin.X .. Box.EndX()
                Y := Box.Origin.Y .. Box.EndY()
                Z := Box.Origin.Z .. Box.EndZ()
            do:
                Idx := GetVoxelIndex(X,Y,Z)

                # Clear prop for this voxel
                if(PE := PropGrid[Idx]?):
                    BZ.SpawnerDevice.CheckIn(PE)

                if( set PropGrid[Idx] = false ) {} # Empty array


    # Spawn a prop of type 'Piece' with 'Voxel' as its origin, and 'Rotation' as its rotation
    SpawnProp<public>(Piece:piece_type, Voxel:vector3i, Rotation:prop_yaw):void=
        var HeldTransform : held_transform = held_transform{}

        if:            
            Zone := BuildSystem?.BuildSite?
        then:
            
            HeldTransform.RotatePiece(Rotation)

            PieceSize := Piece.Size
            var HalfSize : vector3 = (PieceSize * 0.5)

            # We always store the mesh at its 'minimum' corner
            if:
                PieceSize.X <> PieceSize.Y # See if the build piece is not radially symmetrical 
                Rotation = prop_yaw.deg_90 or Rotation = prop_yaw.deg_270 # See if if piece is 90 degrees in either direction
            then:  # if it is, then we need to rotate the voxel size
                set HalfSize = MakeVector3(HalfSize.Y, HalfSize.X, HalfSize.Z) 
                    
            CursorCord:vector3 = Zone.ZoneOrigin + vector3:
                X := (Voxel.X * BuildGridSize.X) + (HalfSize.X * BuildGridSize.X)
                Y := (Voxel.Y * BuildGridSize.Y) + (HalfSize.Y * BuildGridSize.Y)
                Z := (Voxel.Z * BuildGridSize.Z) + (HalfSize.Z * BuildGridSize.Z)

            # Tell spawner device to create prop
            Zone.SpawnerDevice.CheckOut(
                Piece, 
                Voxel, 
                HeldTransform.ToTransform(CursorCord), 
                Rotation,
                Zone.BuildSystem)

    # Adds a prop to the grid at supplied cell
    AddProp<public>(Cell:vector3i, PropEntry:prop_entry):void=

        #Print("AddProp {Cell}")

        # Add to grid
        Idx := GetVoxelIndex(Cell)
        if(PE := PropGrid[Idx]?):
            Print("Already PropEntry in grid!")

        if( set PropGrid[Idx] = option{ PropEntry } ) {}
