# Knowledge-neutral Verse snippet; project names and asset IDs generalized.
# id: snippet-236

using { /Fortnite.com/Characters }
using { /Fortnite.com/Devices }
using { /UnrealEngine.com/Temporary/Diagnostics }
using { /UnrealEngine.com/Temporary/SpatialMath }
using { /UnrealEngine.com/Temporary/UI }
using { /Verse.org/Simulation }
using { /Verse.org/Assets }
using { Materials }

# How far to trace from the camera to find a voxel to hit
PlayerTraceDistance:float := 7000.0

CursorMaterialMap : [build_category]material = map:
    # category specific valid cursor material

CursorInvalidMaterialMap : [build_category]material = map:
    # category specific invalid cursor material
    

# Class that holds specific player data for a specific build site
# Note that this is recreated each time a player enters a site
player_data<public> := class<final><unique>:
    Logger:log = log{Channel:=LogClass}

    # Player that this data is for
    Agent<public> : agent
    # Fortnite character for the player
    Character<public> : fort_character
    # Event called when player leaves, used to stop any running routines
    RemovedEvent<public> : event() = event(){}
    # Reference to the root device in the world
    RootDevice<public> : root_device
    # Reference to global data for this player (persistent as they enter/leave build sites)
    GlobalData<public> : global_player_data
    # UI manager for this player
    UIManager<public> : ui_manager
    # Build zone that owns this data
    BuildZoneDevice<public> : build_zone

    # If player is currently placing pieces
    var PlacePiece<private> : logic = false
    # If player is currently deleting pieces
    var DeletePiece<private> : logic = false
    # Currently selected category to place
    var CurrentCategory<public> : build_category = build_category.Building1 

    # Prop used to draw the cursor on voxel face we are looking at
    var FaceCursorProp : ?creative_prop = false

    # If holding down build button, voxel we started on
    var TurboBuildStart<private> : ?vector3i = false
    # If holding build button, face of voxel we started on
    var TurboBuildFace<private> : ?voxel_box_face = false

    PlayerDebugDraw<public> : debug_draw = debug_draw{}

    # Initialized this player data
    InitPlayerData<public>():void=
        # Start processes that will continue to run while player is in site
        spawn {MainLoop()}
        spawn {AwaitInput()}

        # Show the cursor when entering site
        UIManager.CursorUI.Show()

        # show build HUD and update current category
        UIManager.BuildHUD.Show()
        UIManager.BuildHUD.SetSelectedCategory(CurrentCategory)

        # Turn on additional input
        RootDevice.EnableBuildInputForPlayer(Agent)

        # Create cursor prop
        InitTM := transform:
            Scale:=vector3{X:=1.0, Y:=1.0, Z:=1.0}
            Rotation:=IdentityRotation()
            Translation:=vector3{X:=0.0, Y:=0.0, Z:=-10000.0}

        SpawnCursorResult := SpawnProp(RootDevice.BP_Cursor, InitTM)        
        if(NewProp := SpawnCursorResult(0)?):
            set FaceCursorProp = option{ NewProp }          

    # Called when player leaves site
    Removed<public>():void=
        HideCursor()
        UIManager.CursorUI.Hide()
        UIManager.BuildHUD.Hide()
        RemovedEvent.Signal() # kills MainLoop and AwaitInput
        RootDevice.DisableBuildInputForPlayer(Agent)

    # Function that loops to receive input from Input Trigger devices and set relevant state
    AwaitInput<private>()<suspends>:void=
        race:
            RemovedEvent.Await()
            loop:
                race:
                    # Building controls
                    block:
                        _Agent := RootDevice.InputDevice_Place.PressedEvent.Await()
                        if(Agent = _Agent):
                            set PlacePiece = true                            
                    block:
                        Tuple := RootDevice.InputDevice_Place.ReleasedEvent.Await()
                        if(Agent = Tuple(0)):
                            set PlacePiece = false
                    block:
                        _Agent := RootDevice.InputDevice_Delete.PressedEvent.Await()
                        if(Agent = _Agent):
                            set DeletePiece = true
                    block:
                        _Agent := RootDevice.InputDevice_PrevCategory.PressedEvent.Await()
                        if(Agent = _Agent):
                            CycleCurrentCategory(false)
                    block:
                        _Agent := RootDevice.InputDevice_NextCategory.PressedEvent.Await()
                        if(Agent = _Agent):
                            CycleCurrentCategory(true)

    # Get the 'next' category, given the current one
    GetNextCategory(Category:build_category, Forward:logic):build_category=
        if(Forward?):
            return case(Category):
                build_category.Building1 => build_category.Park
                build_category.Park => build_category.Null
                build_category.Null => build_category.Building1
        else:
            return case(Category):
                build_category.Building1 => build_category.Null
                build_category.Park => build_category.Building1
                build_category.Null => build_category.Park
                

    # Cycles to next available category
    CycleCurrentCategory(Forward:logic):void=    
        NewCategory := GetNextCategory(CurrentCategory, Forward)
        ChangeCategory(NewCategory)

    # Hide the cursor
    HideCursor<public>():void=
        if(FaceCursorPropAct := FaceCursorProp?):
            FaceCursorPropAct.Hide()   

    # Change the currently chosen category (will update HUD)
    ChangeCategory<public>(InCategory:build_category):void=
        set CurrentCategory = InCategory
        UIManager.BuildHUD.SetSelectedCategory(CurrentCategory)

        UIManager.CursorUI.SetReticleVisible(logic{CurrentCategory <> build_category.Null})


    # Main loop that updates cursor and adds voxels if control is pressed
    MainLoop<private>()<suspends>:void=
        race:
            RemovedEvent.Await()
            loop:
                Sleep(0.0)
                SelectModeTick_Trace()
           
               
    # Move the face cursor prop to the desired location (usually face hit by raycast)
    UpdateFaceCursor(Voxel:vector3i, Face:voxel_box_face, InCategory:build_category, Valid:logic):void=       
        if(FaceCursorPropAct := FaceCursorProp?):
            # Make visible
            FaceCursorPropAct.Show()

            # Calc position
            VoxelCentre := BuildZoneDevice.ZoneOrigin + (Voxel * BuildGridSize) + (0.5 * BuildGridSize)
            FaceCentreOffset := 0.5 * BuildGridSize * FaceToVector3(Face)
            CursorCentre := VoxelCentre + FaceCentreOffset

            CursorRot := case(Face):
                voxel_box_face.PosX => MakeRotationFromYawPitchRollDegrees(0.0, -90.0, 0.0)
                voxel_box_face.NegX => MakeRotationFromYawPitchRollDegrees(0.0, 90.0, 0.0)
                voxel_box_face.PosY => MakeRotationFromYawPitchRollDegrees(0.0, 0.0, 90.0)
                voxel_box_face.NegY => MakeRotationFromYawPitchRollDegrees(0.0, 0.0, -90.0)
                voxel_box_face.PosZ => MakeRotationFromYawPitchRollDegrees(0.0, 0.0, 0.0)
                voxel_box_face.NegZ => MakeRotationFromYawPitchRollDegrees(0.0, 180.0, 0.0)            

            CursorScale := case(Face):
                voxel_box_face.PosX => MakeVector3(0.8, 1.0, 1.0)
                voxel_box_face.NegX => MakeVector3(0.8, 1.0, 1.0)
                voxel_box_face.PosY => MakeVector3(1.0, 0.8, 1.0)
                voxel_box_face.NegY => MakeVector3(1.0, 0.8, 1.0)
                voxel_box_face.PosZ => MakeVector3(1.0, 1.0, 1.0)
                voxel_box_face.NegZ => MakeVector3(1.0, 1.0, 1.0)

            # Move them
            CursorTM := transform:
                Scale := CursorScale
                Rotation := CursorRot
                Translation := CursorCentre
            if( FaceCursorPropAct.TeleportTo[CursorTM] ) {}
               
            # Set cursor material
            ShowInvalid := logic{ not Valid? and not TurboBuildStart? }
            if:
                ShowInvalid?
                InvalidMaterial := CursorInvalidMaterialMap[InCategory]
            then:
                FaceCursorPropAct.SetMaterial(InvalidMaterial)
            else if(CursorMaterial := CursorMaterialMap[InCategory]):
                FaceCursorPropAct.SetMaterial(CursorMaterial)

    # Hide the face cursor prop
    HideFaceCursor():void=
        if(FaceCursorPropAct := FaceCursorProp?):
            FaceCursorPropAct.Hide()


    # See if any player overlaps this voxel
    CheckPlayerOverlapsVoxel(Voxel:vector3i)<transacts>:logic=
        var HaveOverlap:logic = false

        for (Player : BuildZoneDevice.GetPlayspace().GetPlayers(), not HaveOverlap?):
            if(Char := Player.GetFortCharacter[]):
                VoxelBox := BuildZoneDevice.BuildSystem.VoxelGrid.GetBoxFromVoxel(Voxel)
                PlayerBox := Char.GetBox()

                if(VoxelBox.Overlaps[PlayerBox]):
                    set HaveOverlap = true

        return HaveOverlap          

    # See if we can add a voxel at the specified location
    CheckVoxelCanBeBuilt(Voxel:vector3i, SurfaceOnly:logic)<transacts>:logic=
        VoxelGrid := BuildZoneDevice.BuildSystem.VoxelGrid
        GridSize := VoxelGrid.Size

        # Can't build if already solid
        if(VoxelGrid.Trace_VoxelIsSolid(Voxel)?):
            return false

        # If 'surface only', check that we are on the ground, or voxel below is solid
        if(SurfaceOnly?):
            if( Voxel.Z = 0 or VoxelGrid.Trace_VoxelIsSolid(Voxel + MakeVector3i(0,0,-1))? ):
                return true
            else:
                return false
        # If not 'surface only', see if any neighbour is solid (or we're on the ground)
        else:
            if:
                (Voxel.Z = 0) or
                (Voxel.X > 0 and VoxelGrid.Trace_VoxelIsSolid(Voxel + MakeVector3i(-1,0,0))?) or
                (Voxel.X < GridSize.X-1 and VoxelGrid.Trace_VoxelIsSolid(Voxel + MakeVector3i(1,0,0))?) or
                (Voxel.Y > 0 and VoxelGrid.Trace_VoxelIsSolid(Voxel + MakeVector3i(0,-1,0))?) or
                (Voxel.Y < GridSize.Y-1 and VoxelGrid.Trace_VoxelIsSolid(Voxel + MakeVector3i(0,1,0))?) or
                (Voxel.Z > 0 and VoxelGrid.Trace_VoxelIsSolid(Voxel + MakeVector3i(0,0,-1))?) or
                (Voxel.Z < GridSize.Z-1 and VoxelGrid.Trace_VoxelIsSolid(Voxel + MakeVector3i(0,0,1))?)
            then:
                return true
            else:
                return false

    # Main function that performs a trace, and updates voxel grid as needed
    SelectModeTick_Trace<private>():void=
        #profile("TICK"):

            #PlayerDebugDraw.Clear()

            # Clear turbo build if not holding place
            if(not PlacePiece?):
                set TurboBuildStart = false
                set TurboBuildFace = false

            PlayerPos := Character.GetTransform().Translation
            PlayerView := Character.GetViewRotation().GetLocalForward()
            PlayerViewLocation := Character.GetViewLocation()
            BuildSystem := BuildZoneDevice.BuildSystem
            VoxelGrid := BuildSystem.VoxelGrid
            MeshGrid := BuildSystem.MeshGrid
            SurfaceOnly := logic{ CategoryIsSurfaceOnly[CurrentCategory] }

            var Result:ray_cast_result = ray_cast_result{}
            #profile(" Trace"):

            # While turbo building, trace a virtual plane in space, so you can build overhangs
            if:
                TBStart := TurboBuildStart?
                TBFace := TurboBuildFace?
            then:
                TBPlaneResult := VoxelGrid.TracePlane(PlayerViewLocation, PlayerView, PlayerTraceDistance, TBStart, TBFace)
                
                # We got a hit
                if(TBPlaneResult.Hit?):
                    # Check this voxel can be built (isn't already, and is touching something solid)
                    AddVoxel := TBPlaneResult.HitVoxel + FaceToVector3i(TBPlaneResult.HitFace)
                    CanBuild := CheckVoxelCanBeBuilt(AddVoxel, SurfaceOnly)
                    if(CanBuild?):
                        set Result = TBPlaneResult
            else:
                set Result = VoxelGrid.Trace(PlayerViewLocation, PlayerView, PlayerTraceDistance)


            LimitsBox := MakeVoxelBox(ZeroVector3i, BuildZoneDevice.GridSize)

            # Check we got a hit within bounds of build region
            var ValidAdd:logic = false
            var ValidDelete:logic = false
            var RemoveCursorVoxel:vector3i = ZeroVector3i
            var AddCursorVoxel:vector3i = ZeroVector3i
            if(Result.Hit? and CurrentCategory <> build_category.Null):

                # Voxel to remove is the one we hit
                set RemoveCursorVoxel = Result.HitVoxel
                if(LimitsBox.Contains[RemoveCursorVoxel]):
                    set ValidDelete = true

                # Voxel to add is adjacent to hit face
                set AddCursorVoxel = RemoveCursorVoxel + FaceToVector3i(Result.HitFace)
                if(LimitsBox.Contains[AddCursorVoxel]):
                    set ValidAdd = true

            if(ValidAdd? or ValidDelete?):

                # Check that a player isn't in this voxel
                if(ValidAdd? and CheckPlayerOverlapsVoxel(AddCursorVoxel)?):
                    set ValidAdd = false

                # Surface-only can only be added on tops of voxels
                if(ValidAdd? and SurfaceOnly? and Result.HitFace <> voxel_box_face.PosZ):
                    set ValidAdd = false
                    

                # Don't allow turbo build unless we are in correct plane and hit face 
                if:
                    ValidAdd?
                    TBStart := TurboBuildStart?
                    TBFace := TurboBuildFace?
                then:
                    if(Result.HitFace <> TBFace):
                        set ValidAdd = false
                    else if((TBFace = voxel_box_face.PosX or TBFace = voxel_box_face.NegX) and AddCursorVoxel.X <> TBStart.X):
                        set ValidAdd = false
                    else if((TBFace = voxel_box_face.PosY or TBFace = voxel_box_face.NegY) and AddCursorVoxel.Y <> TBStart.Y):
                        set ValidAdd = false
                    else if((TBFace = voxel_box_face.PosZ or TBFace = voxel_box_face.NegZ) and AddCursorVoxel.Z <> TBStart.Z):
                        set ValidAdd = false

                # Update cursor
                UpdateFaceCursor(RemoveCursorVoxel, Result.HitFace, CurrentCategory, ValidAdd)

                # Check if voxel above 'remove voxel' is surface-only
                var SurfaceOnlyVoxelAbove:logic = false
                var SOV_Voxel:vector3i = MakeVector3i(0,0,0)
                var SOV_Category:build_category = build_category.Null
                var SOV_Resources:int = 0
                var SOV_PlayerID:int = -1
                if:
                    VoxelAbove := RemoveCursorVoxel + MakeVector3i(0,0,1)
                    LimitsBox.Contains[VoxelAbove]
                    CellAbove := VoxelGrid.GetCell[VoxelAbove]
                    CategoryIsSurfaceOnly[CellAbove.Category]
                then:
                    set SurfaceOnlyVoxelAbove = true
                    set SOV_Category = CellAbove.Category
                    set SOV_Voxel = VoxelAbove

                # Special case for deleting the top of a voxel which has a veg voxel above it - we delete the veg instead
                if(SurfaceOnlyVoxelAbove? and Result.HitFace = voxel_box_face.PosZ):
                    set RemoveCursorVoxel = SOV_Voxel
                    set ValidDelete = logic{ LimitsBox.Contains[RemoveCursorVoxel] } # Needed for deleting veg from z=0
        
                if(PlacePiece? or DeletePiece?):
                    profile("MODIFY"):

                        var ModifiedCategories:[]build_category = array{}

                        # Perform Add
                        if(PlacePiece? and ValidAdd?):
                            #profile(" Add"):

                                var CanStartTurbo:logic = true
                                var AddBox:voxel_box = voxel_box{}

                                # If replacing an existing voxel...
                                var Skip:logic = false
                                if(PreviousCell := VoxelGrid.GetCell[AddCursorVoxel]):
                                    if(PreviousCell.Category = CurrentCategory): # Skip if this voxel is already the category we want!
                                        set Skip = true
                                    else:
                                        set ModifiedCategories = array{PreviousCell.Category}
                
                                        # If replacing a park tile, update WFC
                                        if(CategoryUsesWFC[PreviousCell.Category]):
                                            BuildSystem.RemoveWFCTile(RemoveCursorVoxel, 1)

                                if(not Skip?):
                                    set AddBox = MakeVoxelBox(AddCursorVoxel, MakeVector3i(1,1,1))

                                    VoxelGrid.SetVoxel(AddCursorVoxel, CurrentCategory)
                                    set ModifiedCategories += array{CurrentCategory}

                                    #profile("  UpdateWFC"):
                                        if(CurrentCategory = build_category.Park):
                                            BuildSystem.AddParkWFCTile(AddCursorVoxel, 1)
                                        else:
                                            BuildSystem.NotifyWFCTileChanged(AddCursorVoxel, CurrentCategory, 1)


                                if(ModifiedCategories.Length > 0):

                                    VoxelGrid.VoxelDebugDraw.Clear() # Clear debug drawing so we can draw it now

                                    # Update SDF and recalc boxes for new and previous voxel types                                
                                    #profile("  AddBoxToSDF"):
                                        VoxelGrid.AddBoxToSDF(AddBox, ModifiedCategories)

                                    #profile("  UpdateBoxesFromSDF"):
                                        BuildSystem.UpdateBoxesFromSDF(ModifiedCategories, AddBox) 

                                    # Remember start of turbo build
                                    if(not TurboBuildStart? and CanStartTurbo?):
                                        set TurboBuildStart = option{AddCursorVoxel}
                                        set TurboBuildFace = option{Result.HitFace}

                        # Perform delete
                        else if(DeletePiece? and ValidDelete?):
                            #profile(" Delete"):

                                VoxelGrid.VoxelDebugDraw.Clear() # Clear debug drawing so we can draw it now

                                if(PreviousCell := VoxelGrid.GetCell[RemoveCursorVoxel]):
                                    set ModifiedCategories += array{PreviousCell.Category}

                                    var RemoveBox:voxel_box := MakeVoxelBox(RemoveCursorVoxel, MakeVector3i(1,1,1))
                                    VoxelGrid.ClearVoxel(RemoveCursorVoxel)

                                    if(CategoryUsesWFC[PreviousCell.Category]):
                                        BuildSystem.RemoveWFCTile(RemoveCursorVoxel, 1)
                                    else:
                                        BuildSystem.NotifyWFCTileChanged(RemoveCursorVoxel, PreviousCell.Category, 1)

                                    # If we are removing a non-surface-only voxel with surface-only above it, remove the surface-only as well
                                    if(SurfaceOnlyVoxelAbove? and not CategoryIsSurfaceOnly[ModifiedCategories[0]]): 
                                        VoxelGrid.ClearVoxel(SOV_Voxel)
                                        
                                        if(CategoryUsesWFC[SOV_Category]):
                                            BuildSystem.RemoveWFCTile(SOV_Voxel, 1)   

                                        set RemoveBox.Size.Z += 1 # Expand box to cover both voxels
                                        set ModifiedCategories += array{SOV_Category}


                                    #profile("  RemoveBoxFromSDF"):
                                        VoxelGrid.RemoveBoxFromSDF(RemoveBox, ModifiedCategories)
                                    MeshGrid.ClearPropsForBox(RemoveBox)

                                    #profile("  UpdateBoxesFromSDF"):
                                        BuildZoneDevice.BuildSystem.UpdateBoxesFromSDF(ModifiedCategories, RemoveBox) # Recalc boxes for types being removed (or do nothing if nothing was removed)


            else:
                # If we didn't hit anything, hide cursor and tooltip
                HideCursor()

            set DeletePiece = false
