# Knowledge-neutral Verse snippet; project names and asset IDs generalized.
# id: snippet-237

using { /Fortnite.com/Characters }
using { /UnrealEngine.com/Temporary/SpatialMath }
using { /Verse.org/Simulation }



# This wraps the ray cast logic into an interator through voxels
# It is similar to a vector form equation of a line
ray_caster<public> := class:
    # Size of cell in grid we are casting through
    GridSize<public> : vector3
    # Start location of ray
    InitialPosition<public> : vector3
    # Ray vector
    Ray<public> : vector3
    # Sign of ray in each dimension (1 or -1)
    RaySgn<public> : vector3i

    # Total distance travelled by trace
    var DistanceTraced<public> : float = 0.0

    # Current Position, but snapped to our build coordinate system
    var<private> CurrentVoxel<public> : vector3i = vector3i{}

    # Current position (CurrentPosition = Ray * t + InitialPosition)
    # The world position of the current ray collision (this is the output of the vector form equation of a line)
    var<private> CurrentPosition<public> : vector3 = vector3{}

    var<private> CurrentHitFace<public> : voxel_box_face = voxel_box_face.PosZ

    # How far along our ray we should go (this is the parameter for a vector form equation of a line)
    var t<private> : float = 0.0

    # Moves the Ray Cast to the next closest block, returning (vector of the world position, vector of the grid snapped position in block coords)
    Next<public>()<transacts>:tuple(vector3, vector3i)=
        # Calculate distances to the next block (dt means delta-t)
        var dt:vector3 = vector3:
            X := ((CurrentVoxel.X * 1.0 + (if(Ray.X > 0.0) then 1.0 else 0.0)) * GridSize.X) - CurrentPosition.X
            Y := ((CurrentVoxel.Y * 1.0 + (if(Ray.Y > 0.0) then 1.0 else 0.0)) * GridSize.Y) - CurrentPosition.Y
            Z := ((CurrentVoxel.Z * 1.0 + (if(Ray.Z > 0.0) then 1.0 else 0.0)) * GridSize.Z) - CurrentPosition.Z

        # The distance to the next nth block in `m` direction is infinite if the view is 0 for `m`
        set dt = vector3:
            X := SafeDivide(dt.X, Ray.X)
            Y := SafeDivide(dt.Y, Ray.Y)
            Z := SafeDivide(dt.Z, Ray.Z)

        
        # Figure out in which direction is the next closest block
        # and take the next closest block's distance
        if (dt.X < dt.Y and dt.X < dt.Z):
            set CurrentVoxel.X += RaySgn.X
            set t = dt.X
            set CurrentHitFace = if(RaySgn.X = 1) then voxel_box_face.NegX else voxel_box_face.PosX
        else if (dt.Y < dt.X and dt.Y < dt.Z):
            set CurrentVoxel.Y += RaySgn.Y
            set t = dt.Y
            set CurrentHitFace = if(RaySgn.Y = 1) then voxel_box_face.NegY else voxel_box_face.PosY
        else:
            set CurrentVoxel.Z += RaySgn.Z
            set t = dt.Z
            set CurrentHitFace = if(RaySgn.Z = 1) then voxel_box_face.NegZ else voxel_box_face.PosZ
        
        StepDist := t + 0.001 # small offset to ensure we are properly in the voxel 
        set DistanceTraced += StepDist
        set CurrentPosition = CurrentPosition + (Ray * StepDist)

        return(CurrentPosition, CurrentVoxel)

# Create a ray_caster object, using the supplied initial parameters
InitRayCaster<constructor><public>(Ray:vector3, InitialPosition:vector3, GridSize:vector3)<decides><transacts> := 

    ray_caster:
        CurrentVoxel := vector3i:
            X := Floor[InitialPosition.X / GridSize.X]
            Y := Floor[InitialPosition.Y / GridSize.Y]
            Z := Floor[InitialPosition.Z / GridSize.Z]

        CurrentPosition := InitialPosition
        GridSize := GridSize
        InitialPosition := InitialPosition
        Ray := Ray
        RaySgn := vector3i:
            X := Round[Sgn(Ray.X)]
            Y := Round[Sgn(Ray.Y)]
            Z := Round[Sgn(Ray.Z)]
