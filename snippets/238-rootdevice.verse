# Knowledge-neutral Verse snippet; project names and asset IDs generalized.
# id: snippet-238

using { /Verse.org/Simulation/Tags }
using { /Verse.org/Simulation }
using { /Verse.org/Random }
using { /UnrealEngine.com/Temporary/UI }
using { /UnrealEngine.com/Temporary/SpatialMath }
using { /UnrealEngine.com/Temporary/Diagnostics }
using { /Fortnite.com/Devices }
using { /Fortnite.com/Playspaces }
using { /Fortnite.com/UI }
using { Lib }


LogClass := class(log_channel){}

# Make asset subfolders public
Textures<public> := module:
Materials<public> := module:

# Device that is the singleton entry point for building system
# Only one instance should exist in the level
root_device<public> := class(creative_device):

    # Input Trigger Devices used to gather input from all players and route to their building site
    @editable
    InputDevice_Place<public> : input_trigger_device = input_trigger_device{}
    @editable
    InputDevice_Delete<public> : input_trigger_device = input_trigger_device{}
    @editable
    InputDevice_PrevCategory<public> : input_trigger_device = input_trigger_device{}
    @editable
    InputDevice_NextCategory<public> : input_trigger_device = input_trigger_device{}

    # Prop used for cursor
    @editable
    var BP_Cursor : creative_prop_asset = DefaultCreativePropAsset

    #Reference to each build site
    @editable
    var BuildZones : []build_zone = array{}

    # Used to display message on screen for players
    @editable 
    MessageHUD<public> : hud_message_device = hud_message_device {}

    # Device used to map to actual props to spawn
    @editable 
    SpawnReferencesDevice : prop_spawner_references_device = prop_spawner_references_device{}

    # Map from player to global data about that player, that exist the entire time they are in the game (common across build sites)
    var AgentToGlobalDataMap : weak_map(agent, global_player_data) = map{}

    OnBegin<override>()<suspends>:void=

        # Iterate over build zones and initialize them
        for(Zone : BuildZones):
            Sleep(0.5)
            Zone.InitBuildZone()
            
        SetupPlayers()

    # Setup existing players, and register events for player that join later
    SetupPlayers<private>():void=
        Playspace := Self.GetPlayspace()
        Playspace.PlayerAddedEvent().Subscribe(AddPlayer)
        Playspace.PlayerRemovedEvent().Subscribe(RemovePlayer)
        Players := Playspace.GetPlayers()
        for (Player : Players):
            AddPlayer(Player)

    # Remove a player from the game
    RemovePlayer<private>(Agent:agent):void=
        for(Zone:BuildZones):
            Zone.RemovePlayer(Agent)


    # Adds joining players to player map
    AddPlayer<private>(Agent:agent):void=
        if:
            Player := player[Agent]
            NewGPD := global_player_data:
                Player := Player
                RootDevice := Self
            set AgentToGlobalDataMap[Agent] = NewGPD
        then:            
            NewGPD.InitGlobalPlayerData()
            
           
    # Enable building input for a player
    EnableBuildInputForPlayer(Agent:agent):void=
        InputDevice_Place.Register(Agent)
        InputDevice_Delete.Register(Agent)
        InputDevice_PrevCategory.Register(Agent)
        InputDevice_NextCategory.Register(Agent)

    # Disable building input for a player
    DisableBuildInputForPlayer(Agent:agent):void=
        InputDevice_Place.Unregister(Agent)
        InputDevice_Delete.Unregister(Agent)
        InputDevice_PrevCategory.Unregister(Agent)
        InputDevice_NextCategory.Unregister(Agent)
