# Knowledge-neutral Verse snippet; project names and asset IDs generalized.
# id: snippet-239

using { /Verse.org/Simulation }

# Base class for space grammar rule set
# A rule sets takes a voxel box as a starting point, and calls a series of rules, resulting in a number of BuildProps assets being spawned
rs_base<public> := class:

    RuleSetName<public> : string = "Base"

    # Base function that applies this style to a box
    ApplyRulesToBox<public>(Box:voxel_box, BuildSystem:build_system):void=
        VO_Root.Operate(Box, prop_yaw.deg_0, BuildSystem)

    # Function in which to actually create rules
    SetupRules<public>(PT:piece_type_dir):void={}

    # Root operator to start from, initial box is passed to this rule, and it is expected to pass 'sub boxes' to subsequent rules
    var VO_Root:vo_base = vo_base{}
