# Knowledge-neutral Verse snippet; project names and asset IDs generalized.
# id: snippet-240

using { /Verse.org/Simulation }

# Rules for 'Building1' style
rs_building1<public> := class(rs_base):
    
    RuleSetName<override>:string = "Building1"

    SetupRules<override>(PT:piece_type_dir):void=

        RoofTileCheck := vo_positioncheck: # if we are the top floor, fill centre of top floor to make a roof
            VO_Pass := vo_tilefill: 
                VO_Tile := vo_prop:
                    Prop := PT.Building1_box
                    AlignToTop := true

        # Check four cardinal directions and choose appropriate piece
        FourWay := vo_fourwaycheck:
            VO_4Clear := vo_prop:
                Prop := PT.Building1_O
            VO_3Clear := vo_prop:
                Prop := PT.Building1_U 
            VO_2ClearCorner := vo_prop:
                Prop := PT.Building1_corner
            VO_2ClearParallel := vo_prop:
                Prop := PT.Building1_II
            VO_1Clear := vo_prop:
                Prop := PT.Building1_face
            VO_0Clear := vo_prop:
                Prop := PT.Building1_box

        # Rules for one floor of the building
        FloorRules := vo_cornerwallsplit:
            VO_CornerLength1 := vo_corner:
                VO_Corner := vo_prop:
                    Prop := PT.Building1_corner
                VO_Face := vo_prop:
                    Prop := PT.Building1_face      
                VO_Covered := vo_prop:
                    Prop := PT.Building1_box                                          
            VO_Wall := vo_wall:
                VO_Width1 := vo_facecoveragecheck_1voxel:
                    VO_Clear := vo_prop:
                        Prop := PT.Building1_face
                    VO_Covered := vo_prop:
                        Prop := PT.Building1_box
            VO_Centre := RoofTileCheck
            VO_TooSmall := vo_tilefill: 
                VO_Tile := FourWay

        set VO_Root = vo_floorsplit:
                VO_FloorHeight1 := FloorRules
