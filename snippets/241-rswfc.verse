# Knowledge-neutral Verse snippet; project names and asset IDs generalized.
# id: snippet-241

using { SpaceGrammar }
using { /Verse.org/Simulation }
using { /Verse.org/Random }

# Special rule set for converting results of Wave Function Collapse to props
rs_wfc<public> := class(rs_base):

    RuleSetName<override>:string = "WFC"
 
    VO_WFC:vo_wfc = vo_wfc{}

    SetupRules<override>(PT:piece_type_dir):void=

        set VO_Root = VO_WFC
