# Knowledge-neutral Verse snippet; project names and asset IDs generalized.
# id: snippet-242

using { /Fortnite.com/Devices }
using { Lib }
using { /Verse.org/Simulation }

# Object used to indicate a particular prop that we want to spawn
piece_type<public> := class<unique>:
    # Size of prop in voxels
    Size<public> : vector3i = vector3i{X:=1, Y:=1, Z:=1}

# Container of all piece_type objects, used to indicate specific props
piece_type_dir<public> := class:

    Box:piece_type := piece_type{}

    #BUILDING
    Building1_box:piece_type := piece_type{}
    Building1_corner:piece_type := piece_type{}
    Building1_face:piece_type := piece_type{}
    Building1_U:piece_type := piece_type{}
    Building1_O:piece_type := piece_type{}
    Building1_II:piece_type := piece_type{}

    #PARK
    Park_Grass:piece_type := piece_type{}
    Park_Water:piece_type := piece_type{}
    Park_Grass_Water_Straight:piece_type := piece_type{}
    Park_Grass_Water_Incorn:piece_type := piece_type{}
    Park_Grass_Water_Outcorn:piece_type := piece_type{}

# Device that maps from piece_type to actual creative_prop_asset that can be spawned
# This is only needed because we do not currently reflect creative_prop_asset into Verse (unlike mesh, texture etc)
prop_spawner_references_device<public> := class(creative_device):

    #FALLBACK
    @editable
    BP_box : creative_prop_asset = DefaultCreativePropAsset

    #BUILDING
    @editable
    BP_Building1_box : creative_prop_asset = DefaultCreativePropAsset
    @editable
    BP_Building1_corner : creative_prop_asset = DefaultCreativePropAsset
    @editable
    BP_Building1_face : creative_prop_asset = DefaultCreativePropAsset
    @editable
    BP_Building1_U : creative_prop_asset = DefaultCreativePropAsset
    @editable
    BP_Building1_O : creative_prop_asset = DefaultCreativePropAsset
    @editable
    BP_Building1_II : creative_prop_asset = DefaultCreativePropAsset

    #PARK
    @editable
    BP_Park_Grass : creative_prop_asset = DefaultCreativePropAsset
    @editable
    BP_Park_Water : creative_prop_asset = DefaultCreativePropAsset
    @editable
    BP_Park_Grass_Water_Straight : creative_prop_asset = DefaultCreativePropAsset
    @editable
    BP_Park_Grass_Water_Incorn : creative_prop_asset = DefaultCreativePropAsset
    @editable
    BP_Park_Grass_Water_Outcorn : creative_prop_asset = DefaultCreativePropAsset

    var PT_Opt:?piece_type_dir = false

    var PieceTypeToAssetMap : [piece_type]creative_prop_asset = map{}

    var MapHasBeenBuilt:logic = false

    # Get the piece_type_dir singleton
    GetPT():piece_type_dir=
        if(PT := PT_Opt?):
            return PT
        else:
            NewPT := piece_type_dir{}
            set PT_Opt = option{NewPT}
            return NewPT

    # Get the creative_prop_asset for a particular piece_type
    GetPropAssetForPieceType(PieceType:piece_type):creative_prop_asset = 

        # Build map if we haven't done so
        if(not MapHasBeenBuilt?):
            PT := GetPT()
            BuildPieceTypeToAssetMap(PT)
            set MapHasBeenBuilt = true

        # Look up asset from map using piece type object
        var PropAsset:creative_prop_asset = BP_Building1_box
        if(PA := PieceTypeToAssetMap[PieceType]):
            set PropAsset = PA
        else:
            Print("FAILED TO FIND ASSET")
        return PropAsset

    # Create map from piece_type to creative_prop_asset
    BuildPieceTypeToAssetMap(PT:piece_type_dir):void=

        set PieceTypeToAssetMap = map:
            PT.Box => BP_box        
            
            PT.Building1_box => BP_Building1_box
            PT.Building1_corner => BP_Building1_corner
            PT.Building1_face => BP_Building1_face
            PT.Building1_U => BP_Building1_U
            PT.Building1_O => BP_Building1_O
            PT.Building1_II => BP_Building1_II

            PT.Park_Grass => BP_Park_Grass
            PT.Park_Water => BP_Park_Water
            PT.Park_Grass_Water_Straight => BP_Park_Grass_Water_Straight
            PT.Park_Grass_Water_Incorn => BP_Park_Grass_Water_Incorn
            PT.Park_Grass_Water_Outcorn => BP_Park_Grass_Water_Outcorn
