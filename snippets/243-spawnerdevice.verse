# Knowledge-neutral Verse snippet; project names and asset IDs generalized.
# id: snippet-243

using { /Verse.org/Simulation }
using { /Verse.org/Assets }
using { /Verse.org/Colors }
using { /UnrealEngine.com/Temporary/Diagnostics }
using { /UnrealEngine.com/Temporary/SpatialMath }
using { /Fortnite.com/Devices }
using { Lib }
using { Materials }


# Enum for indicating yaw rotation
prop_yaw<public> := enum:
    deg_0
    deg_90
    deg_180
    deg_270

# Util for converting from rotation enum to 
PropYawToRadians(Rot:prop_yaw)<computes>:float=
    case(Rot):
        prop_yaw.deg_0 => 0.0
        prop_yaw.deg_90 => 0.5 * PiFloat
        prop_yaw.deg_180 => PiFloat
        prop_yaw.deg_270 => 1.5 * PiFloat

# Util to increase yaw by 90 degrees
PropYawAdd90(Rot:prop_yaw)<computes>:prop_yaw=
    case(Rot):
        prop_yaw.deg_0 => prop_yaw.deg_90
        prop_yaw.deg_90 => prop_yaw.deg_180
        prop_yaw.deg_180 => prop_yaw.deg_270
        prop_yaw.deg_270 => prop_yaw.deg_0

# Util to decrease yaw by 90 degrees
PropYawSubtract90(Rot:prop_yaw)<computes>:prop_yaw=
    case(Rot):
        prop_yaw.deg_0 => prop_yaw.deg_270
        prop_yaw.deg_90 => prop_yaw.deg_0
        prop_yaw.deg_180 => prop_yaw.deg_90
        prop_yaw.deg_270 => prop_yaw.deg_180

ToString(Yaw:prop_yaw):string=
    case(Yaw):
        prop_yaw.deg_0 => "0"
        prop_yaw.deg_90 => "90"
        prop_yaw.deg_180 => "180"
        prop_yaw.deg_270 => "270"

########
        
# Reference to a Prop instance, indicating type and rotation
prop_entry<public> := struct<computes>:
    Prop<public>:creative_prop
    PieceType<public>:piece_type
    Rotation<public>:prop_yaw

MakePropEntry<public>(Prop:creative_prop, PieceType:piece_type, Rotation:prop_yaw)<computes>:prop_entry=
    prop_entry:
        Prop := Prop
        PieceType := PieceType
        Rotation := Rotation


ToString(SpawnResult:spawn_prop_result):string=
    case(SpawnResult):
        spawn_prop_result.Ok => "Ok"
        spawn_prop_result.UnknownError => "UnknownError"
        spawn_prop_result.InvalidSpawnPoint => "InvalidSpawnPoint"
        spawn_prop_result.SpawnPointOutOfBounds => "SpawnPointOutOfBounds"
        spawn_prop_result.InvalidAsset => "InvalidAsset"
        spawn_prop_result.TooManyProps => "TooManyProps"

# One entry in the prop spawning queue
prop_spawn_info := class:
    var PieceType:piece_type = piece_type{}
    var Voxel:vector3i = vector3i{}
    var Transform:transform = transform{}
    var Rotation:prop_yaw = prop_yaw.deg_0
    var BuildSystem:build_system = build_system{}


# Device for actually spawning props. 
prop_spawner_device := class(creative_device):

    # Event used to trigger spawning process
    SpawnPropsEvent<public>:event() = event(){}

    # Queue of props waiting to spawn
    var CurrentSpawnInfos:[]prop_spawn_info = array{}

    Logger:log = log{Channel:=LogClass}

    # Device that maps from piece_type object to a prop asset reference
    @editable
    var References : prop_spawner_references_device = prop_spawner_references_device{}

    # Destroy a prop
    CheckIn<public>(PropEntry:prop_entry):void= 
        PropEntry.Prop.Dispose()

    # Create a new prop at the specified location, store in the specified voxel
    CheckOut(PieceType: piece_type, Voxel:vector3i, Transform: transform, Rotation:prop_yaw, BuildSystem:build_system):void=

        # Add info to queue to process
        NewInfo := prop_spawn_info:
            PieceType := PieceType
            Voxel := Voxel
            Transform := Transform
            Rotation := Rotation
            BuildSystem := BuildSystem

        set CurrentSpawnInfos += array{NewInfo}

        SpawnPropsEvent.Signal() # Kick off spawning process


    # Process any pending props
    SpawnPendingProps()<suspends>:void=
        loop:
            SpawnPropsEvent.Await() # Blocks until signalled from CheckOut above

            # Iterate over any pending props
            for:
                CurrentSpawnInfo : CurrentSpawnInfos
            do:
                # get prop asset (BP) for piece
                CreativeAsset := References.GetPropAssetForPieceType(CurrentSpawnInfo.PieceType)
            
                var SpawnPropResult:tuple(?creative_prop, spawn_prop_result) = (false, spawn_prop_result.Ok)
                #profile("     DoSpawn"):
                    # Spawn the prop, continue if its good
                    set SpawnPropResult = SpawnProp(CreativeAsset, CurrentSpawnInfo.Transform)
                    
                SpawnResult := SpawnPropResult(1)

                if:
                    SpawnResult = spawn_prop_result.Ok                    
                    NewProp := SpawnPropResult(0)?
                then:
                    #Logger.Print("SD ADD:  [{SpawnerIndex}] Count:{SpawnedPropCount}")
                    
                    PropEntry := MakePropEntry(NewProp, CurrentSpawnInfo.PieceType, CurrentSpawnInfo.Rotation)
                    CurrentSpawnInfo.BuildSystem.MeshGrid.AddProp(CurrentSpawnInfo.Voxel, PropEntry)

                else:
                    Print("SD: SPAWN FAILED! {SpawnResult}")

            set CurrentSpawnInfos = array{}

    # Start the spawning process
    OnBegin<override>()<suspends>:void=
        SpawnPendingProps()
