# Knowledge-neutral Verse snippet; project names and asset IDs generalized.
# id: snippet-244

using { /Verse.org/Simulation }
using { /UnrealEngine.com/Temporary/UI }
using { /Fortnite.com/UI }

# Base class for UI elements (e.g. menus and HUD)
base_ui<public> := class<unique>:
    var _Agent<public>:?agent = false
    var Canvas<public>:canvas = canvas {}
    var _RootDevice<public>:?root_device = false
    var _GlobalData<public> : ?global_player_data = false

    Init<public><final>(Agent:agent, RootDevice:root_device, GlobalData:global_player_data):void=
        set _Agent = option{Agent}
        set _RootDevice = option{RootDevice}
        set _GlobalData = option{GlobalData}
        CreateCanvas()

    CreateCanvas<public>():void=
        Print("Creating canvas")
        
    ShowInteract<public>():void=
        if(PlayerUI := GetPlayerUI[player[_Agent?]]):
            PlayerUI.AddWidget(Canvas, player_ui_slot{InputMode := ui_input_mode.All})
            MenuOpened()

    Show<public><final>():void=
        if(PlayerUI := GetPlayerUI[player[_Agent?]]):
            PlayerUI.AddWidget(Canvas, player_ui_slot{InputMode := ui_input_mode.None})
            MenuOpened()

    Hide<public><final>():void=
        if(PlayerUI := GetPlayerUI[player[_Agent?]]):
            PlayerUI.RemoveWidget(Canvas)        
            MenuClosed()

    MenuOpened<public>():void={}
        #Print("menu opened")

    MenuClosed<public>():void={}
        #Print("menu closed")

# Manager that stores all UI elements for one player
ui_manager<public> := class:
    # UI element for cursor
    var CursorUI : cursor_ui = cursor_ui {}
    # UI element for showing category
    var BuildHUD : build_hud = build_hud {}

    # Global data for player owning this UI
    var _GlobalData<public> : ?global_player_data = false

    # Array of all UIs
    var AllUI : []base_ui = array {}

    InitUI<public>(Agent:agent, RootDevice:root_device, GlobalData:global_player_data):void=

        set _GlobalData = option{GlobalData}

        set AllUI = array:
            CursorUI
            BuildHUD
        
        for(_UI : AllUI):
            _UI.Init(Agent, RootDevice, GlobalData)


    HideAll<public>():void=
        for(BUI : AllUI):
            BUI.Hide()
