# Knowledge-neutral Verse snippet; project names and asset IDs generalized.
# id: snippet-245

using { /Verse.org/Simulation }
using { /UnrealEngine.com/Temporary/SpatialMath }

# 3-dimensional vector with `int` components
vector3i<public> := struct<computes><concrete>:
    @editable
    X<public>:int = 0

    @editable
    Y<public>:int = 0

    @editable
    Z<public>:int = 0


ZeroVector3i:vector3i := vector3i{X:=0, Y:=0, Z:=0}

MakeVector3i<public>(InX:int, InY:int, InZ:int)<computes>:vector3i=
    vector3i{X:=InX, Y:=InY, Z:=InZ}

operator'+'<public>(A:vector3i, B:vector3i)<computes>:vector3i=
    vector3i{X:=A.X + B.X, Y:=A.Y + B.Y, Z:=A.Z + B.Z}

operator'-'<public>(A:vector3i, B:vector3i)<computes>:vector3i=
    vector3i{X:=A.X - B.X, Y:=A.Y - B.Y, Z:=A.Z - B.Z}

operator'*'<public>(A:vector3i, B:int)<computes>:vector3i=
    vector3i{X:=A.X * B, Y:=A.Y * B, Z:=A.Z * B}

operator'*'<public>(A:vector3i, B:float)<computes>:vector3=
    vector3{X:=A.X * B, Y:=A.Y * B, Z:=A.Z * B}

operator'*'<public>(A:vector3i, B:vector3)<computes>:vector3=
    vector3{X:=A.X * B.X, Y:=A.Y * B.Y, Z:=A.Z * B.Z}

operator'*'<public>(A:vector3, B:vector3i)<computes>:vector3=
    vector3{X:=A.X * B.X, Y:=A.Y * B.Y, Z:=A.Z * B.Z}

Min<public>(A:vector3i, B:vector3i)<computes>:vector3i=
    vector3i{X:= Min(A.X, B.X), Y:= Min(A.Y, B.Y), Z:= Min(A.Z, B.Z)}

Max<public>(A:vector3i, B:vector3i)<computes>:vector3i=
    vector3i{X:= Max(A.X, B.X), Y:= Max(A.Y, B.Y), Z:= Max(A.Z, B.Z)}

ToString(Vec:vector3i)<computes>:string=
    "{Vec.X},{Vec.Y},{Vec.Z}"

(Vec:vector3i).Equals(A:vector3i)<computes><decides>:void=
    Vec.X = A.X and Vec.Y = A.Y and Vec.Z = A.Z


RandomIntFromVector(Vec:vector3i, RandMax:int, Primes:tuple(int,int,int,int)):int=
    var Result:int = 0
    set Result = (Vec.X*Primes(0)) + (Vec.Y*Primes(1)) + (Vec.Z*Primes(2))
    if(TempResult := Mod[Result, Primes(3)]):
        set Result = TempResult
    if(TempResult := Mod[Result, RandMax+1]):
        set Result = TempResult
    return Result
