# Knowledge-neutral Verse snippet; project names and asset IDs generalized.
# id: snippet-246

using { /Verse.org/Simulation }

# Base class of volume operator (VO). 
# Each volume operate takes a voxel box, and calls the next VOs with smaller sub-boxes
vo_base<public> := class:

    Operate<public>(Box:voxel_box, Rot:prop_yaw, BuildSystem:build_system):void={}
