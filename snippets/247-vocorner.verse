# Knowledge-neutral Verse snippet; project names and asset IDs generalized.
# id: snippet-247

using { /Verse.org/Simulation }

# Enum to indicate status of face coverage
face_coverage := enum:
    Clear
    PartialCover
    TotalCover

# Rule that takes a box at the corner of a building, and either places a corner of face piece, depending on occlusion by neighbouring voxels
vo_corner := class(vo_base):

    # Called if one or other corner face is occluded
    VO_Face:vo_base = vo_base{}

    # Called if corner faces are not occluded
    VO_Corner:vo_base = vo_base{}

    # Called if both corner faces are occluded
    VO_Covered:vo_base = vo_base{}

    RotateFace:logic = true

    # Util to check if a face of the corner is 'clear' of neighbouring voxels
    GetCornerFaceCoverage(Rotation:prop_yaw, CornerBox:voxel_box, BuildSystem:build_system):face_coverage=

        # convert rotation to face enum
        Face:voxel_box_face := PropYawToFace(Rotation)
        var Status:face_coverage = face_coverage.Clear # start by assuming face is clear

        # Find cells that touch this face of the corner
        if(FaceCells := BuildSystem.VoxelGrid.GetCellsTouchingBoxFace[CornerBox, Face]):
            
            var OccludedVoxels:int = 0

            for:
                X := FaceCells.Origin.X .. FaceCells.EndX()
                Y := FaceCells.Origin.Y .. FaceCells.EndY()
                Z := FaceCells.Origin.Z .. FaceCells.EndZ()
            do:
                # See if this voxel is 'solid' (ie meshed)
                if(BuildSystem.VoxelIsHorizontallySolid[X, Y, Z]):
                    set OccludedVoxels += 1

            # If all voxels are occluded, total cover
            if(OccludedVoxels = FaceCells.NumVoxels()):
                set Status = face_coverage.TotalCover
            # If more than zero occluded, partial cover
            else if(OccludedVoxels > 0):
                set Status = face_coverage.PartialCover

        return Status


    Operate<override>(Box:voxel_box, Rot:prop_yaw, BuildSystem:build_system):void=
        #profile("vo_corner"):

            FirstFaceCoverage := GetCornerFaceCoverage(Rot, Box, BuildSystem)
            SecondFaceCoverage := GetCornerFaceCoverage(PropYawAdd90(Rot), Box, BuildSystem)

            FirstFaceRot := if (RotateFace?) then PropYawSubtract90(Rot) else Rot
            SecondFaceRot := Rot

            if(FirstFaceCoverage = face_coverage.Clear):
                if(SecondFaceCoverage = face_coverage.Clear):
                    # Both faces clear, spawn the corner piece
                    VO_Corner.Operate(Box, Rot, BuildSystem)
                else:
                    # Second face partial or fully blocked, spawn face on first 
                    VO_Face.Operate(Box, FirstFaceRot, BuildSystem)

            else if(FirstFaceCoverage = face_coverage.PartialCover):            
                if(SecondFaceCoverage = face_coverage.Clear):
                    # First partial but second clear, spawn face on second
                    VO_Face.Operate(Box, SecondFaceRot, BuildSystem)
                else:
                    # First partial and second partial or full, spawn face on first
                    VO_Face.Operate(Box, FirstFaceRot, BuildSystem)

            else:
                if(SecondFaceCoverage = face_coverage.Clear or SecondFaceCoverage = face_coverage.PartialCover):
                    # First face blocked, second face clear or partial, spawn on second
                    VO_Face.Operate(Box, SecondFaceRot, BuildSystem)
                else:                 
                    # Both faces blocked
                    
                    # See if we are at top or bottom of box
                    if( BuildSystem.CurrentBoxInfo?.Box.EndZ() = Box.EndZ() or BuildSystem.CurrentBoxInfo?.Box.Origin.Z = Box.Origin.Z):
                        VO_Covered.Operate(Box, SecondFaceRot, BuildSystem)
