# Knowledge-neutral Verse snippet; project names and asset IDs generalized.
# id: snippet-248

using { /Verse.org/Simulation }
using { /Verse.org/Colors }


# Split a box into corners and walls.
# Tries to use corners of size CornerLengthN but falls back to size 1 if that is not possible
vo_cornerwallsplit := class(vo_base):

    # Rule to use for corners that are 1 voxel
    VO_CornerLength1:vo_base = vo_base{}

    # Rule to use for corners that are N voxels
    VO_CornerLengthN:vo_base = vo_base{}

    # Rule to use for regions between corners. Rotation is set based on face direction
    VO_Wall:vo_base = vo_base{}

    # Rule for region inside corners and walls
    VO_Centre:vo_base = vo_base{}

    # Rule when region is too small to cut out corners and walls
    VO_TooSmall:vo_base = vo_base{}

    # Size of 'N' corner
    # If set to 1, will always use VO_CornerLength1 (never VO_CornerLengthN)
    CornerLengthN:int = 1

    Operate<override>(Box:voxel_box, Rot:prop_yaw, BuildSystem:build_system):void=
        #Print("  vo_cornerwallsplit.Operate {Box}")
        #profile("vo_cornerwallsplit"):
        
            var SubBox:voxel_box = Box # Init to input box

            # Choose whether to use size 1 or N corner
            var CornerLength:int = CornerLengthN
            var VO_Corner:vo_base = VO_CornerLengthN

            # See if we need to fall back to all 1-size corners
            TargetCombinedCornerLength := CornerLengthN * 2
            if:
                CornerLengthN = 1 or 
                Box.Size.X < TargetCombinedCornerLength or 
                Box.Size.Y < TargetCombinedCornerLength
            then:
                set CornerLength = 1
                set VO_Corner = VO_CornerLength1
        
            ## Corners
            CombinedCornerLength := 2 * CornerLength

            # If we have space for all corners, send boxes to corner op
            if(Box.Size.X >= CombinedCornerLength and Box.Size.Y >= CombinedCornerLength):

                #profile("vo_cornerwallsplit corners"):

                    # MinX, MinY
                    set SubBox.Origin = Box.Origin
                    set SubBox.Size = MakeVector3i(CornerLength, CornerLength, Box.Size.Z)
                    VO_Corner.Operate(SubBox, prop_yaw.deg_180, BuildSystem)
                    #BuildSystem.VoxelGrid.DebugDrawVoxelBox(SubBox, NamedColors.Salmon, 5.0) # DEBUG

                    # MinX, MaxY
                    set SubBox.Origin = Box.Origin + MakeVector3i(0, Box.Size.Y - CornerLength, 0)
                    set SubBox.Size = MakeVector3i(CornerLength, CornerLength, Box.Size.Z)
                    VO_Corner.Operate(SubBox, prop_yaw.deg_90, BuildSystem)
                    #BuildSystem.VoxelGrid.DebugDrawVoxelBox(SubBox, NamedColors.Salmon, 5.0) # DEBUG

                    # MaxX, MinY
                    set SubBox.Origin = Box.Origin + MakeVector3i(Box.Size.X - CornerLength, 0, 0)
                    set SubBox.Size = MakeVector3i(CornerLength, CornerLength, Box.Size.Z)
                    VO_Corner.Operate(SubBox, prop_yaw.deg_270, BuildSystem)
                    #BuildSystem.VoxelGrid.DebugDrawVoxelBox(SubBox, NamedColors.Salmon, 5.0) # DEBUG

                    # MaxX, MaxY
                    set SubBox.Origin = Box.Origin + MakeVector3i(Box.Size.X - CornerLength, Box.Size.Y - CornerLength, 0)
                    set SubBox.Size = MakeVector3i(CornerLength, CornerLength, Box.Size.Z)
                    VO_Corner.Operate(SubBox, prop_yaw.deg_0, BuildSystem)
                    #BuildSystem.VoxelGrid.DebugDrawVoxelBox(SubBox, NamedColors.Salmon, 5.0) # DEBUG

                XFaceLength := Box.Size.X - CombinedCornerLength
                YFaceLength := Box.Size.Y - CombinedCornerLength

                ## Walls

                var XFaceDepth:int = CornerLength
                var YFaceDepth:int = CornerLength

                # Faces along Y
                if(YFaceLength > 0):
                    set SubBox.Size = MakeVector3i(YFaceDepth, YFaceLength, Box.Size.Z)

                    set SubBox.Origin = Box.Origin + MakeVector3i(0, CornerLength, 0)
                    VO_Wall.Operate(SubBox, prop_yaw.deg_90, BuildSystem)
                    #BuildSystem.VoxelGrid.DebugDrawVoxelBox(SubBox, NamedColors.Yellow, 5.0) # DEBUG
                    
                    set SubBox.Origin = Box.Origin + MakeVector3i(Box.Size.X - YFaceDepth, CornerLength, 0)
                    VO_Wall.Operate(SubBox, prop_yaw.deg_270, BuildSystem)
                    #BuildSystem.VoxelGrid.DebugDrawVoxelBox(SubBox, NamedColors.Yellow, 5.0) # DEBUG

                # Faces along X
                if(XFaceLength > 0):
                    set SubBox.Size = MakeVector3i(XFaceLength, XFaceDepth, Box.Size.Z)

                    set SubBox.Origin = Box.Origin + MakeVector3i(CornerLength, 0, 0)
                    VO_Wall.Operate(SubBox, prop_yaw.deg_180, BuildSystem)
                    #BuildSystem.VoxelGrid.DebugDrawVoxelBox(SubBox, NamedColors.YellowGreen, 5.0) # DEBUG

                    set SubBox.Origin = Box.Origin + MakeVector3i(CornerLength, Box.Size.Y - XFaceDepth, 0)
                    VO_Wall.Operate(SubBox, prop_yaw.deg_0, BuildSystem)
                    #BuildSystem.VoxelGrid.DebugDrawVoxelBox(SubBox, NamedColors.YellowGreen, 5.0) # DEBUG


                # If there is an area in the centre, pass that through
                if(XFaceLength > 0 and YFaceLength > 0):
                    set SubBox.Size = MakeVector3i(XFaceLength, YFaceLength, Box.Size.Z)
                    set SubBox.Origin = Box.Origin + MakeVector3i(CornerLength, CornerLength, 0)
                    VO_Centre.Operate(SubBox, Rot, BuildSystem)
                    #BuildSystem.VoxelGrid.DebugDrawVoxelBox(SubBox, NamedColors.Blue, 5.0) # DEBUG


            # Not enough space for corners, fall back to 'too small' path
            else:
                VO_TooSmall.Operate(Box, Rot, BuildSystem)
                #BuildSystem.VoxelGrid.DebugDrawVoxelBox(SubBox, NamedColors.RoyalBlue, 5.0) # DEBUG
