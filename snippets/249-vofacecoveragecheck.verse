# Knowledge-neutral Verse snippet; project names and asset IDs generalized.
# id: snippet-249

using { /Verse.org/Simulation }



# General test for any face of a box is obstructed.
vo_facecoveragecheck := class(vo_base):
    
    # Called if any voxels on specified face are obstructed
    VO_Covered:vo_base = vo_base{}

    # Called if no voxels on specified face are obstructed
    VO_Clear:vo_base = vo_base{}

    # Whether to apply supplied rotation to specified face before testing
    UseRot:logic = true

    # Which face of specified box to test
    CheckFace:voxel_box_face = voxel_box_face.NegZ

    Operate<override>(Box:voxel_box, Rot:prop_yaw, BuildSystem:build_system):void=

        var IsCovered:logic = false

        # Use rotation or fixed face depending on option
        Face := if(UseRot?) then PropYawToFace(PropYawAdd90(Rot)) else CheckFace

        # Special case of building on bottom of box, and looking at bottom face
        if:
            Face = voxel_box_face.NegZ
            Box.Origin.Z = 0
        then:
            set IsCovered = true
            
        else if(FaceBox := BuildSystem.VoxelGrid.GetCellsTouchingBoxFace[Box, Face]):
            for:
                X := FaceBox.Origin.X .. FaceBox.EndX()
                Y := FaceBox.Origin.Y .. FaceBox.EndY()
                Z := FaceBox.Origin.Z .. FaceBox.EndZ()
                not IsCovered?
            do:
                # If voxel is solid, we are 'covered'
                if(Face = voxel_box_face.NegZ or Face = voxel_box_face.PosZ):
                    if(BuildSystem.VoxelIsVerticallySolid[X, Y, Z]):
                        set IsCovered = true   
                else:
                    if(BuildSystem.VoxelIsHorizontallySolid[X, Y, Z]):
                        set IsCovered = true           
    
        if(IsCovered?):
            VO_Covered.Operate(Box, Rot, BuildSystem)
        else:
            VO_Clear.Operate(Box, Rot, BuildSystem)
