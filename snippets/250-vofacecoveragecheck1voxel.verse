# Knowledge-neutral Verse snippet; project names and asset IDs generalized.
# id: snippet-250

using { /Verse.org/Simulation }



# Special (faster) case for checking if the 'outward' face (based on supplied rotation) of the supplied box is obstructed
# Useful to swap a wall piece to a box piece if there is something in front of it
vo_facecoveragecheck_1voxel := class(vo_base):
    
    # Called if face is covered
    VO_Covered:vo_base = vo_base{}

    # Called if face is free
    VO_Clear:vo_base = vo_base{}

    Operate<override>(Box:voxel_box, Rot:prop_yaw, BuildSystem:build_system):void=

        var IsCovered:logic = false

        if(Box.Size.X <> 1 or Box.Size.Y <> 1 or Box.Size.Z <> 1):
            Print("vo_facecoveragecheck_1voxel only works for 1x1x1 voxels!")

        if:
            Rot = prop_yaw.deg_0
            Box.Origin.Y < BuildSystem.VoxelGrid.Size.Y - 1
            BuildSystem.VoxelIsHorizontallySolid[Box.Origin.X, Box.Origin.Y+1, Box.Origin.Z]
        then:
            set IsCovered = true
        else if:
            Rot = prop_yaw.deg_90
            Box.Origin.X > 0
            BuildSystem.VoxelIsHorizontallySolid[Box.Origin.X-1, Box.Origin.Y, Box.Origin.Z]
        then:
            set IsCovered = true
        else if:
            Rot = prop_yaw.deg_180
            Box.Origin.Y > 0
            BuildSystem.VoxelIsHorizontallySolid[Box.Origin.X, Box.Origin.Y-1, Box.Origin.Z]
        then:
            set IsCovered = true
        else if:
            Rot = prop_yaw.deg_270
            Box.Origin.X < BuildSystem.VoxelGrid.Size.X - 1
            BuildSystem.VoxelIsHorizontallySolid[Box.Origin.X+1, Box.Origin.Y, Box.Origin.Z]
        then:
            set IsCovered = true                
    
        if(IsCovered?):
            VO_Covered.Operate(Box, Rot, BuildSystem)
        else:
            VO_Clear.Operate(Box, Rot, BuildSystem)
