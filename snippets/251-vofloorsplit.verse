# Knowledge-neutral Verse snippet; project names and asset IDs generalized.
# id: snippet-251

using { /Verse.org/Simulation }

# Split volume into floors of height 1 or N
# Will fit as many N floors as possible, starting from the bottom, then 1 height floors for remaining space
vo_floorsplit := class(vo_base):

    # Called for 1 voxel high floors
    VO_FloorHeight1:vo_base = vo_base{}

    # Called for N voxel high floors
    VO_FloorHeightN:vo_base = vo_base{}

    # Height of N-high floors
    # If set to 1, will always use VO_FloorHeight1 (never VO_FloorHeightN)
    NHeight:int = 1



    Operate<override>(Box:voxel_box, Rot:prop_yaw, BuildSystem:build_system):void=

        var SubBox:voxel_box = Box # Init to input box

        # Start from bottom and split into floors until we reach the top
        var CurrentZ:int = 0

        loop:
            # If we are at the top, then stop
            if(CurrentZ >= Box.Size.Z):
                break
            # If we can fit an N voxel floor, do it
            else if(NHeight > 1 and Box.Size.Z - CurrentZ >= NHeight): 
                set SubBox.Origin.Z = Box.Origin.Z + CurrentZ
                set SubBox.Size.Z = NHeight
                VO_FloorHeightN.Operate(SubBox, Rot, BuildSystem)
                set CurrentZ += NHeight
            # Otherwise add a 1
            else:
                set SubBox.Origin.Z = Box.Origin.Z + CurrentZ
                set SubBox.Size.Z = 1
                VO_FloorHeight1.Operate(SubBox, Rot, BuildSystem)
                set CurrentZ += 1
