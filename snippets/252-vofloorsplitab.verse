# Knowledge-neutral Verse snippet; project names and asset IDs generalized.
# id: snippet-252

using { /Verse.org/Simulation }
using { /Verse.org/Colors }

# Split volume into alternating 1-high floors
vo_floorsplit_ab := class(vo_base):

    VO_FloorA:vo_base = vo_base{}

    VO_FloorB:vo_base = vo_base{}

    # Should alternating be 'local' to box passed in, or 'global' to build site volume
    SiteSpaceAlternating:logic = true

    Operate<override>(Box:voxel_box, Rot:prop_yaw, BuildSystem:build_system):void=
        #BuildSystem.VoxelGrid.DebugDrawVoxelBox(Box, NamedColors.White, 5.0) # DEBUG

        var SubBox:voxel_box = Box # Init to input box
        set SubBox.Size.Z = 1

        # Start from bottom and split into floors until we reach the top
        var CurrentZ:int = 0

        loop:
            # If we are at the top, then stop
            if(CurrentZ >= Box.Size.Z):
                break

            set SubBox.Origin.Z = Box.Origin.Z + CurrentZ

            AlternateBasis := if(SiteSpaceAlternating?) then SubBox.Origin.Z else CurrentZ

            if(Div := Mod[AlternateBasis, 2]):
                # If we can fit an N voxel floor, do it
                if(Div = 0): 
                    VO_FloorA.Operate(SubBox, Rot, BuildSystem)
                # Otherwise add a 1
                else:
                    VO_FloorB.Operate(SubBox, Rot, BuildSystem)

            set CurrentZ += 1
