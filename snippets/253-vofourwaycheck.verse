# Knowledge-neutral Verse snippet; project names and asset IDs generalized.
# id: snippet-253

using { /Verse.org/Simulation }
using { /Verse.org/Colors }

# Takes a single voxel, and checks all 4 cardinal directions, picking the output which matches the set of unoccluded faces
# Applies rotation as well
vo_fourwaycheck := class(vo_base):

    # All 4 faces clear
    VO_4Clear:vo_base = vo_base{}

    # 3 faces clear
    VO_3Clear:vo_base = vo_base{}

    # 2 faces clear, making corner
    VO_2ClearCorner:vo_base = vo_base{}

    # 2 faces clear, on opposite sides
    VO_2ClearParallel:vo_base = vo_base{}

    # 1 face clear
    VO_1Clear:vo_base = vo_base{}

    # 0 faces clear
    VO_0Clear:vo_base = vo_base{}


    Operate<override>(Box:voxel_box, Rot:prop_yaw, BuildSystem:build_system):void=
        #BuildSystem.VoxelGrid.DebugDrawVoxelBox(Box, NamedColors.White, 5.0) # DEBUG

        if(Box.Size.X > 1 or Box.Size.Y > 1 or Box.Size.Z > 1):
            if(RS := BuildSystem.CurrentRuleSet?):
                Print("vo_fourwaycheck {RS.RuleSetName} : Box larger than 1 not supported!")
        else if(CurrentBoxInfo := BuildSystem.CurrentBoxInfo?):
            GridSize := BuildSystem.VoxelGrid.Size
            PosXSolid := logic{ Box.Origin.X < GridSize.X-1 and BuildSystem.VoxelIsHorizontallySolid[Box.Origin.X + 1, Box.Origin.Y, Box.Origin.Z] }
            PosYSolid := logic{ Box.Origin.Y < GridSize.Y-1 and BuildSystem.VoxelIsHorizontallySolid[Box.Origin.X, Box.Origin.Y + 1, Box.Origin.Z] }
            NegXSolid := logic{ Box.Origin.X > 0 and BuildSystem.VoxelIsHorizontallySolid[Box.Origin.X - 1, Box.Origin.Y, Box.Origin.Z] }
            NegYSolid := logic{ Box.Origin.Y > 0 and BuildSystem.VoxelIsHorizontallySolid[Box.Origin.X, Box.Origin.Y - 1, Box.Origin.Z] }
    
            #Print("4WC: {LogicToString(PosXSolid)} {LogicToString(PosYSolid)} {LogicToString(NegXSolid)} {LogicToString(NegYSolid)}")

            # PosX PosY NegX NegY

            if(PosXSolid?):             # 1 ? ? ?
                if(PosYSolid?):         # 1 1 ? ?
                    if(NegXSolid?):     # 1 1 1 ?                        
                        if(NegYSolid?): # 1 1 1 1 
                            # Only trigger Clear if we are top or bottom of box (not in the middle)
                            if(Box.Origin.Z = CurrentBoxInfo.Box.Origin.Z or Box.Origin.Z = CurrentBoxInfo.Box.EndZ()):
                                VO_0Clear.Operate(Box, prop_yaw.deg_0, BuildSystem)
                        else:           # 1 1 1 0 
                            VO_1Clear.Operate(Box, prop_yaw.deg_180, BuildSystem)
                    else:               # 1 1 0 ?
                        if(NegYSolid?): # 1 1 0 1
                            VO_1Clear.Operate(Box, prop_yaw.deg_90, BuildSystem)
                        else:           # 1 1 0 0
                            VO_2ClearCorner.Operate(Box, prop_yaw.deg_180, BuildSystem)
                else:                   # 1 0 ? ?
                    if(NegXSolid?):     # 1 0 1 ?
                        if(NegYSolid?): # 1 0 1 1
                            VO_1Clear.Operate(Box, prop_yaw.deg_0, BuildSystem)
                        else:           # 1 0 1 0
                            VO_2ClearParallel.Operate(Box, prop_yaw.deg_0, BuildSystem)
                    else:               # 1 0 0 ?
                        if(NegYSolid?): # 1 0 0 1
                            VO_2ClearCorner.Operate(Box, prop_yaw.deg_90, BuildSystem)
                        else:           # 1 0 0 0
                            VO_3Clear.Operate(Box, prop_yaw.deg_90, BuildSystem)
            else:                       # 0 ? ? ?
                if(PosYSolid?):         # 0 1 ? ?
                    if(NegXSolid?):     # 0 1 1 ?
                        if(NegYSolid?): # 0 1 1 1
                            VO_1Clear.Operate(Box, prop_yaw.deg_270, BuildSystem)
                        else:           # 0 1 1 0
                            VO_2ClearCorner.Operate(Box, prop_yaw.deg_270, BuildSystem)
                    else:               # 0 1 0 ?
                        if(NegYSolid?): # 0 1 0 1
                            VO_2ClearParallel.Operate(Box, prop_yaw.deg_270, BuildSystem)
                        else:           # 0 1 0 0
                            VO_3Clear.Operate(Box, prop_yaw.deg_180, BuildSystem)
                else:                   # 0 0 ? ?
                    if(NegXSolid?):     # 0 0 1 ?
                        if(NegYSolid?): # 0 0 1 1
                            VO_2ClearCorner.Operate(Box, prop_yaw.deg_0, BuildSystem)
                        else:           # 0 0 1 0 
                            VO_3Clear.Operate(Box, prop_yaw.deg_270, BuildSystem)
                    else:               # 0 0 0 ?
                        if(NegYSolid?): # 0 0 0 1 
                            VO_3Clear.Operate(Box, prop_yaw.deg_0, BuildSystem)
                        else:           # 0 0 0 0 
                            VO_4Clear.Operate(Box, prop_yaw.deg_0, BuildSystem)
