# Knowledge-neutral Verse snippet; project names and asset IDs generalized.
# id: snippet-254

using { /Verse.org/Simulation }

position_face<public> := enum:
    PosZ
    NegZ

# See if this sub-box is touching a particular face of the original box
vo_positioncheck := class(vo_base):
    
    PositionFace:position_face = position_face.PosZ
    VO_Pass:vo_base = vo_base{}
    VO_Fail:vo_base = vo_base{}


    Operate<override>(Box:voxel_box, Rot:prop_yaw, BuildSystem:build_system):void=

        var IsTouching:logic = false

        if( PositionFace = position_face.PosZ and BuildSystem.CurrentBoxInfo?.Box.EndZ() = Box.EndZ()):
            set IsTouching = true

        if( PositionFace = position_face.NegZ and BuildSystem.CurrentBoxInfo?.Box.Origin.Z = Box.Origin.Z):
            set IsTouching = true
    
        if(IsTouching?):
            VO_Pass.Operate(Box, Rot, BuildSystem)
        else:
            VO_Fail.Operate(Box, Rot, BuildSystem)
