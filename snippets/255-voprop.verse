# Knowledge-neutral Verse snippet; project names and asset IDs generalized.
# id: snippet-255

using { /Verse.org/Simulation }
using { /Verse.org/Colors }

# Spawn a prop at the location and rotation of the supplied box
vo_prop := class(vo_base):

    # Which piece to spawn (see prop_spawner_references_device for mapping from this object to actual prop asset)
    Prop:piece_type

    # Should attempt to fill 'empty' space (if prop is smaller than box)
    AttemptBackfill:logic = false

    # Only attempt to backfill if we are the top of the box being processed
    OnlyBackfillAtTop:logic = true

    # Which prop to use to backfill
    BackfillProp:piece_type = piece_type{}

    # Should this prop align to the top of the supplied box, instead of the bottom
    AlignToTop : logic = false


    MatchPropEntry(Entry:prop_entry, NewProp:tuple(piece_type, prop_yaw))<transacts><decides>:void=
        Entry.PieceType = NewProp(0) and Entry.Rotation = NewProp(1)

    # Update mesh at this grid location, if necessary
    SyncMeshGrid(Voxel:vector3i, NewProp:tuple(piece_type, prop_yaw), BuildSystem:build_system):void=

        Index := BuildSystem.MeshGrid.GetVoxelIndex(Voxel)

        if:
            CurrentEntry := BuildSystem.MeshGrid.PropGrid[Index]?
            MatchPropEntry[CurrentEntry, NewProp]
        then:
            # We match, no need to do anything
        else:
            # It's different, clear old prop and spawn new one
            BuildSystem.MeshGrid.ClearPropForVoxel(Voxel)
            BuildSystem.MeshGrid.SpawnProp(NewProp(0), Voxel, NewProp(1))
            set BuildSystem.PropCount += 1
         


    Operate<override>(Box:voxel_box, Rot:prop_yaw, BuildSystem:build_system):void=
        #Print("vo_prop.Operate {Box}")
        #BuildSystem.VoxelGrid.DebugDrawVoxelBox(Box, NamedColors.White)  

        var MainSpawnPos:vector3i = MakeVector3i(-1,-1,-1)

        PieceSize := Prop.Size

        # We want to align the +Y face of the prop with the desired face of the box passed in
        # This means offsetting if it's on the +X or +Y face
        set MainSpawnPos = Box.Origin
        if(Rot = prop_yaw.deg_0):
            set MainSpawnPos.Y = Box.Origin.Y + (Box.Size.Y - PieceSize.Y)
        else if(Rot = prop_yaw.deg_270):
            set MainSpawnPos.X = Box.Origin.X + (Box.Size.X - PieceSize.Y)

        # Align to top of box if desired
        if(AlignToTop?):
            set MainSpawnPos.Z = Box.Origin.Z + (Box.Size.Z - PieceSize.Z)
        
        var MainSpawnProp:tuple(piece_type, prop_yaw) = (Prop, Rot)

        SyncMeshGrid(MainSpawnPos, MainSpawnProp, BuildSystem)

        var BackfillSpawnPos:vector3i = MakeVector3i(-1,-1,-1)

        # See if we want to spawn backfill piece
        if(AttemptBackfill? and (not OnlyBackfillAtTop? or BuildSystem.CurrentBoxInfo?.Box.EndZ() = Box.EndZ())):
            var BackfillSpawnPropOpt:?tuple(piece_type, prop_yaw) = false

            if(Rot = prop_yaw.deg_0):
                Gap := Box.Size.Y - PieceSize.Y
                if(Gap > 0):
                    set BackfillSpawnPos = Box.Origin
                    set BackfillSpawnPropOpt = option{(BackfillProp, Rot)}
            else if(Rot = prop_yaw.deg_90):
                Gap := Box.Size.X - PieceSize.Y
                if(Gap > 0):
                    set BackfillSpawnPos = Box.Origin + MakeVector3i(PieceSize.Y,0,0)
                    set BackfillSpawnPropOpt = option{(BackfillProp, Rot)}
            else if(Rot = prop_yaw.deg_180):
                Gap := Box.Size.Y - PieceSize.Y
                if(Gap > 0):
                    set BackfillSpawnPos = Box.Origin + MakeVector3i(0,PieceSize.Y,0)
                    set BackfillSpawnPropOpt = option{(BackfillProp, Rot)}
            else if(Rot = prop_yaw.deg_270):
                Gap := Box.Size.X - PieceSize.Y
                if(Gap > 0):
                    set BackfillSpawnPos = Box.Origin
                    set BackfillSpawnPropOpt = option{(BackfillProp, Rot)}

            if(BackfillSpawnProp := BackfillSpawnPropOpt?):
                SyncMeshGrid(BackfillSpawnPos, BackfillSpawnProp, BuildSystem)

        # Need to ensure any cells in supplied box that we didn't update are empty
        for: 
            X := Box.Origin.X .. Box.EndX()
            Y := Box.Origin.Y .. Box.EndY()
            Z := Box.Origin.Z .. Box.EndZ()
        do:
            Cell := MakeVector3i(X,Y,Z)
            if:
                not Cell.Equals[MainSpawnPos]
                not Cell.Equals[BackfillSpawnPos]
            then:
                BuildSystem.MeshGrid.ClearPropForVoxel(Cell)
