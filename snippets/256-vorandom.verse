# Knowledge-neutral Verse snippet; project names and asset IDs generalized.
# id: snippet-256

using { /Verse.org/Simulation }
using { /Verse.org/Random }


vo_random_entry := struct:
    # Child VO to call
    VO_Random<public>:vo_base = vo_base{}
    # Min height to use this child
    MinHeight<public>:int = 0

PrimesA:tuple(int,int,int,int) := (31, 47, 59, 167)
PrimesB:tuple(int,int,int,int) := (37, 53, 61, 191)

# Pick a random child
vo_random := class(vo_base):

    Entries:[]vo_random_entry = array{}

    RandomRotation:logic = false

    Operate<override>(Box:voxel_box, Rot:prop_yaw, BuildSystem:build_system):void=

        # Build array of entries that are at or below current height
        FilteredEntries := for(Entry : Entries, Entry.MinHeight <= Box.Size.Z) do Entry

        NumOptions := FilteredEntries.Length
        if(NumOptions > 0):

            Option := RandomIntFromVector(Box.Origin, NumOptions-1, PrimesA) 

            # Random rotation if desired
            var UseRot:prop_yaw = Rot
            if(RandomRotation?):
                RandInt := RandomIntFromVector(Box.Origin, 3, PrimesB)

                set UseRot = case(RandInt):
                    0 => prop_yaw.deg_0
                    1 => prop_yaw.deg_90
                    2 => prop_yaw.deg_180
                    3 => prop_yaw.deg_270
                    _ => prop_yaw.deg_0
        
            if(RandomEntry := FilteredEntries[Option]):
                RandomEntry.VO_Random.Operate(Box, UseRot, BuildSystem)
