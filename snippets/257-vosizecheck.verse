# Knowledge-neutral Verse snippet; project names and asset IDs generalized.
# id: snippet-257

using { /Verse.org/Simulation }

# Check all dimensions of a box are >= a certain size
vo_sizecheck := class(vo_base):
    Size:vector3i = vector3i{X:=0, Y:=0, Z:=0}
    VO_Pass:vo_base = vo_base{}
    VO_Fail:vo_base = vo_base{}


    Operate<override>(Box:voxel_box, Rot:prop_yaw, BuildSystem:build_system):void=
        if(Box.Size.X >= Size.X and Box.Size.Y >= Size.Y and Box.Size.Z >= Size.Z):
            VO_Pass.Operate(Box, Rot, BuildSystem)
        else:
            VO_Fail.Operate(Box, Rot, BuildSystem)
