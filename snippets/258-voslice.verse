# Knowledge-neutral Verse snippet; project names and asset IDs generalized.
# id: snippet-258

using { /Verse.org/Simulation }
using { /Verse.org/Colors }

# How we want to divide the incoming box
slice_type<public> := enum:
    # Take a fixed number of voxels
    Absolute
    # Take a % of the box
    Proportion

# Cut voxels off one side of the supplied box
# If box is not big enough to slice desired amount, entire box is passed to VO_Remainder
vo_slice := class(vo_base):

    # Which face to slice from
    SliceFace:voxel_box_face = voxel_box_face.PosZ
    
    SliceType:slice_type = slice_type.Absolute

    # Voxels to slice off (if SliceType = Absolute)
    SliceSize:int = 1

    # Proportion of voxels to slice off (if SliceType = Proportion)
    SliceProportion:float = 0.5

    # Apply rotation of box to face to slice (no effect for PosZ/NegZ)
    ApplyRotToSliceFace:logic = false
    
    # Box that was sliced off
    VO_Sliced:vo_base = vo_base{}

    # Remainder that was left
    VO_Remainder:vo_base = vo_base{}


    Operate<override>(Box:voxel_box, Rot:prop_yaw, BuildSystem:build_system):void=
        #Print("vo_slice.Operate {SliceSize}")

        var Sliced:logic = true
        var SlicedBox:voxel_box = Box
        var RemainderBox:voxel_box = Box

        # Get face to slice (apply rotation if desired)
        var FaceToSlice:voxel_box_face = SliceFace
        if(ApplyRotToSliceFace?):
            set FaceToSlice = ApplyRotToFace(FaceToSlice, Rot)

        var UseSliceSize:int = SliceSize

        XDir:logic := if(FaceToSlice = voxel_box_face.PosX or FaceToSlice = voxel_box_face.NegX) then true else false
        YDir:logic := if(FaceToSlice = voxel_box_face.PosY or FaceToSlice = voxel_box_face.NegY) then true else false
        ZDir:logic := if(FaceToSlice = voxel_box_face.PosZ or FaceToSlice = voxel_box_face.NegZ) then true else false

        if(SliceType = slice_type.Proportion):
            if:
                XDir?
                PropSliceSize := Floor[Box.Size.X * Clamp(SliceProportion, 0.0, 1.0)]
                set UseSliceSize = PropSliceSize

            if:
                YDir?
                PropSliceSize := Floor[Box.Size.Y * Clamp(SliceProportion, 0.0, 1.0)]
                set UseSliceSize = PropSliceSize

            if:
                ZDir?
                PropSliceSize := Floor[Box.Size.Z * Clamp(SliceProportion, 0.0, 1.0)]
                set UseSliceSize = PropSliceSize
    

        #Print("vo_slice.Operate {Rot} {FaceToSlice} {UseSliceSize}")
        

        # POS/NEG X
        if(XDir? and Box.Size.X >= UseSliceSize):
            # Offset boxes depending on pos or neg face being sliced
            if(FaceToSlice = voxel_box_face.PosX):
                set SlicedBox.Origin.X = Box.Origin.X + (Box.Size.X - UseSliceSize)
            else:
                set RemainderBox.Origin.X = Box.Origin.X + UseSliceSize
            # Sizes are the same either way
            set SlicedBox.Size.X = UseSliceSize
            set RemainderBox.Size.X = Box.Size.X - UseSliceSize

        # POS/NEG Y
        else if(YDir? and Box.Size.Y >= UseSliceSize):
            if(FaceToSlice = voxel_box_face.PosY):
                set SlicedBox.Origin.Y = Box.Origin.Y + (Box.Size.Y - UseSliceSize)
            else:
                set RemainderBox.Origin.Y = Box.Origin.Y + UseSliceSize
            set SlicedBox.Size.Y = UseSliceSize
            set RemainderBox.Size.Y = Box.Size.Y - UseSliceSize

        # POS/NEG Z
        else if(ZDir? and Box.Size.Z >= UseSliceSize):
            if(FaceToSlice = voxel_box_face.PosZ):
                set SlicedBox.Origin.Z = Box.Origin.Z + (Box.Size.Z - UseSliceSize)
            else:
                set RemainderBox.Origin.Z = Box.Origin.Z + UseSliceSize
            set SlicedBox.Size.Z = UseSliceSize
            set RemainderBox.Size.Z = Box.Size.Z - UseSliceSize

        # FAIL
        else:
            set Sliced = false

        # If we had space to slice..
        if(Sliced?):
            # Pass sliced part
            VO_Sliced.Operate(SlicedBox, Rot, BuildSystem)
            #BuildSystem.VoxelGrid.DebugDrawVoxelBox(SlicedBox, NamedColors.White)

            # If we have something left, pass that as well
            if(RemainderBox.Size.X > 0 and RemainderBox.Size.Y > 0 and RemainderBox.Size.Z > 0):
                VO_Remainder.Operate(RemainderBox, Rot, BuildSystem)
                #BuildSystem.VoxelGrid.DebugDrawVoxelBox(RemainderBox, NamedColors.Blue)


        # If there wasn't space, we pass everything to remainder
        else:
            VO_Remainder.Operate(Box, Rot, BuildSystem)
