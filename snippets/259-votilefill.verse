# Knowledge-neutral Verse snippet; project names and asset IDs generalized.
# id: snippet-259

using { /Verse.org/Simulation }

# Simple rule that divides incoming box into 1x1 voxel 'tiles' across the XY plane
vo_tilefill := class(vo_base):

    # Called for each resulting tile
    VO_Tile:vo_base := vo_base{}

    # If we should check that this voxel has something under it before passing it on
    CheckSupported:logic = false

    Operate<override>(Box:voxel_box, Rot:prop_yaw, BuildSystem:build_system):void=

        # Spawn prop for all floor voxels
        for(X := 0..Box.Size.X-1; Y := 0..Box.Size.Y-1):

            SpawnPos := Box.Origin + MakeVector3i(X, Y, 0)
            SubBox := MakeVoxelBox(SpawnPos, MakeVector3i(1,1,Box.Size.Z))

            # If above bottom of grid, check if voxel below is solid
            var IsSupported:logic = true
            if(CheckSupported? and Box.Origin.Z > 1):
                set IsSupported = logic{ BuildSystem.VoxelIsVerticallySolid[SpawnPos.X, SpawnPos.Y, Box.Origin.Z - 1] }

            #Print("{SpawnPos.X} {SpawnPos.Y} {Box.Origin.Z - 1} {LogicToString(IsSupported)}")

            # If something to place on, call child
            if(IsSupported?):
                VO_Tile.Operate(SubBox, Rot, BuildSystem)
