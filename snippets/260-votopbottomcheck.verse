# Knowledge-neutral Verse snippet; project names and asset IDs generalized.
# id: snippet-260

using { /Verse.org/Simulation }

# Rule that checks if this box is at the top, bottom or in the middle of the overall box being processed
vo_topbottomcheck := class(vo_base):
    
    # Box is passed to this if touching the top of overall box
    VO_Top:vo_base = vo_base{}
    # Box is passed to this if in the middle of the overall box
    VO_Middle:vo_base = vo_base{}
    # Box is passed to this if touching the bottom of overall box
    VO_Bottom:vo_base = vo_base{}


    Operate<override>(Box:voxel_box, Rot:prop_yaw, BuildSystem:build_system):void=

        if(BuildSystem.CurrentBoxInfo?.Box.Origin.Z = Box.Origin.Z):
            VO_Bottom.Operate(Box, Rot, BuildSystem)
        else if( BuildSystem.CurrentBoxInfo?.Box.EndZ() = Box.EndZ()):
            VO_Top.Operate(Box, Rot, BuildSystem)
        else:
            VO_Middle.Operate(Box, Rot, BuildSystem)
