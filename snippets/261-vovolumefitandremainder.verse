# Knowledge-neutral Verse snippet; project names and asset IDs generalized.
# id: snippet-261

using { /Verse.org/Simulation }

# Rule that attempts to fit a given volume into the supplied box as many times as possible, and passes on the remaining area as well
vo_volumefitandremainder := class(vo_base):

    # Called for each place that the volume can fit into the supplied box
    VO_Volume:vo_base := vo_base{}
    # Called areas that are left over after fitting volume
    VO_Remainder:vo_base := vo_base{} 

    # Size of volume we want to fit
    VolumeSize:vector3i := vector3i{}

    Operate<override>(Box:voxel_box, Rot:prop_yaw, BuildSystem:build_system):void=

        var UseRot:prop_yaw = prop_yaw.deg_0
        var UseVolumeSize:vector3i = VolumeSize
        if:
            var NumX:int := Quotient[Box.Size.X, VolumeSize.X]
            var NumY:int := Quotient[Box.Size.Y, VolumeSize.Y]
            var NumZ:int := Quotient[Box.Size.Z, VolumeSize.Z]
        then:          

            var CanFit:logic = true

            # If box is too small to fit volume in, try rotating it
            if(NumX = 0 or NumY = 0 or NumZ = 0):

                if(TempNumX := Quotient[Box.Size.X, VolumeSize.Y]):
                    set NumX = TempNumX
                if(TempNumY := Quotient[Box.Size.Y, VolumeSize.X]):
                    set NumY = TempNumY


                if(NumX = 0 or NumY = 0 or NumZ = 0):
                    # Still doesn't fit
                    VO_Remainder.Operate(Box, Rot, BuildSystem) 
                    set CanFit = false
                else:
                    # It does fit, so use rotation
                    set UseRot = prop_yaw.deg_90
                    set UseVolumeSize.X = VolumeSize.Y
                    set UseVolumeSize.Y = VolumeSize.X
    
        

            if(CanFit?):
                # Call Volume for each volume we can fit in
                for:
                    X := 0 .. NumX-1
                    Y := 0 .. NumY-1
                    Z := 0 .. NumZ-1                 
                do:
                    Origin := Box.Origin + MakeVector3i(X * UseVolumeSize.X, Y * UseVolumeSize.Y, Z * UseVolumeSize.Z)
                    VolumeBox := MakeVoxelBox(Origin, UseVolumeSize)
                    VO_Volume.Operate(VolumeBox, UseRot, BuildSystem)

                # Now call Remainder on bits left over
                if:
                    RemX := Mod[Box.Size.X, UseVolumeSize.X]
                    RemY := Mod[Box.Size.Y, UseVolumeSize.Y]
                    RemZ := Mod[Box.Size.Z, UseVolumeSize.Z]
                then:
                    # Z remainder (box on pos Z)
                    if(RemZ > 0):
                        RemOrigin := Box.Origin + MakeVector3i(0, 0, NumZ * UseVolumeSize.Z)
                        RemSize := MakeVector3i(Box.Size.X, Box.Size.Y, RemZ)
                        RemBox := MakeVoxelBox(RemOrigin, RemSize)
                        VO_Remainder.Operate(RemBox, Rot, BuildSystem)

                    # X remainder (box on pos X)
                    if(RemX > 0):
                        RemOrigin := Box.Origin + MakeVector3i(NumX * UseVolumeSize.X, 0, 0)
                        RemSize := MakeVector3i(RemX, Box.Size.Y, NumZ * UseVolumeSize.Z)
                        RemBox := MakeVoxelBox(RemOrigin, RemSize)
                        VO_Remainder.Operate(RemBox, Rot, BuildSystem) 

                    # Y remainder (box on pos Y)
                    if(RemY > 0):
                        RemOrigin := Box.Origin + MakeVector3i(0, NumY * UseVolumeSize.Y, 0)
                        RemSize := MakeVector3i(NumX * UseVolumeSize.X, RemY, NumZ * UseVolumeSize.Z)
                        RemBox := MakeVoxelBox(RemOrigin, RemSize)
                        VO_Remainder.Operate(RemBox, Rot, BuildSystem)
