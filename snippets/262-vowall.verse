# Knowledge-neutral Verse snippet; project names and asset IDs generalized.
# id: snippet-262

using { /Verse.org/Simulation }
using { /Verse.org/Colors }

# Take a box and divide into 1- or N- wide subboxes. The direction of splitting is dictated by the rotation passed in.
vo_wall := class(vo_base):

    # Passed for each set of N-wide voxels that fit along this wall
    VO_WidthN:vo_base = vo_base{}
    # Passed in for any voxels which could not be grouped into N-wide block
    VO_Width1:vo_base = vo_base{}

    # Size of N
    # If set to 1, will always use VO_Width1 (never VO_WidthN)
    NWidth:int = 1

    Operate<override>(Box:voxel_box, Rot:prop_yaw, BuildSystem:build_system):void=

        var SubBox:voxel_box = Box # Init to input box

        # Init as Y face (ie run along X)
        var Dir:vector3i = MakeVector3i(1,0,0)
        var TotalLength:int = Box.Size.X
        var DirX:logic = true
        # X faces (ie run along Y)
        if(Rot = prop_yaw.deg_90 or Rot = prop_yaw.deg_270):
            set Dir = MakeVector3i(0,1,0)        
            set TotalLength = Box.Size.Y
            set DirX = false


        var CurrentPos:int = 0

        loop:
            # If we are at the end, then stop
            if(CurrentPos >= TotalLength):
                break
            # If N>1 and we can fit an N wide piece, do it
            else if(NWidth > 1 and TotalLength - CurrentPos >= NWidth):                 
                set SubBox.Origin = Box.Origin + (Dir * CurrentPos)
                set SubBox.Size = if(DirX?) then MakeVector3i(NWidth, Box.Size.Y, Box.Size.Z) else MakeVector3i(Box.Size.X, NWidth, Box.Size.Z)
                VO_WidthN.Operate(SubBox, Rot, BuildSystem)
                set CurrentPos += NWidth
            # Otherwise, do a 1
            else:
                set SubBox.Origin = Box.Origin + (Dir * CurrentPos)
                set SubBox.Size = if(DirX?) then MakeVector3i(1, Box.Size.Y, Box.Size.Z) else MakeVector3i(Box.Size.X, 1, Box.Size.Z)
                VO_Width1.Operate(SubBox, Rot, BuildSystem)
                set CurrentPos += 1
