# Knowledge-neutral Verse snippet; project names and asset IDs generalized.
# id: snippet-263

using { /Verse.org/Simulation }
using { /Verse.org/Colors }

# Divide incoming box into alternating 1-voxel wide A and B sub-boxes. The direction of splitting is dictated by the rotation passed in.
vo_wallalternate := class(vo_base):

    # First possibility for 1-voxel wide outputs
    VO_A:vo_base = vo_base{}
    # Second possibility for 1-voxel wide outputs
    VO_B:vo_base = vo_base{}

    Operate<override>(Box:voxel_box, Rot:prop_yaw, BuildSystem:build_system):void=

        var SubBox:voxel_box = Box # Init to input box

        # Init as Y face (ie run along X)
        var Dir:vector3i = MakeVector3i(1,0,0)
        var TotalLength:int = Box.Size.X
        var DirX:logic = true
        # X faces (ie run along Y)
        if(Rot = prop_yaw.deg_90 or Rot = prop_yaw.deg_270):
            set Dir = MakeVector3i(0,1,0)        
            set TotalLength = Box.Size.Y
            set DirX = false

        var CurrentPos:int = 0

        loop:
            # If we are at the end, then stop
            if(CurrentPos >= TotalLength):
                break

            set SubBox.Origin = Box.Origin + (Dir * CurrentPos)
            set SubBox.Size = if(DirX?) then MakeVector3i(1, Box.Size.Y, Box.Size.Z) else MakeVector3i(Box.Size.X, 1, Box.Size.Z)

            AlternateBasis := if(DirX?) then SubBox.Origin.X else SubBox.Origin.Y

            if(Mod[AlternateBasis,2] = 0):
                VO_A.Operate(SubBox, Rot, BuildSystem)
            else:
                VO_B.Operate(SubBox, Rot, BuildSystem)

            set CurrentPos += 1
