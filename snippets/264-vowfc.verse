# Knowledge-neutral Verse snippet; project names and asset IDs generalized.
# id: snippet-264

using { Lib.WFC }
using { /Verse.org/Simulation }
using { /Verse.org/Random }


# Converts a rotation as an int into a yaw enum
RotIntToPropYaw(RotInt:int):prop_yaw=
    case(RotInt):
        0 => prop_yaw.deg_0
        1 => prop_yaw.deg_90
        2 => prop_yaw.deg_180
        3 => prop_yaw.deg_270
        _ => 
            Print("RotIntToPropYaw: INVALID") 
            prop_yaw.deg_0

# Special operator that creates props based on status of WFC system held by BuildSystem
vo_wfc := class(vo_base):

    Operate<override>(Box:voxel_box, Rot:prop_yaw, BuildSystem:build_system):void=
        SpawnPropsFromTiles(Box, BuildSystem)

    FindRoofHeight(StartVoxel:vector3i, BuildSystem:build_system):int=
        # Check voxels above this box to see how much free space there is
        GridSizeZ := BuildSystem.VoxelGrid.Size.Z
        var FoundRoof:logic = false
        var RoofDistance:int = 3 # default to max
        for(ZOffset := 1..3, not FoundRoof?):
            Z := StartVoxel.Z + ZOffset
            if(Z < GridSizeZ-1):
                if(BuildSystem.VoxelIsVerticallySolid[StartVoxel.X, StartVoxel.Y, Z]):
                    set RoofDistance = ZOffset
                    set FoundRoof = true
        return RoofDistance


    SpawnPropIfNeeded(PropToSpawn:piece_type, Pos:vector3i, Rot:prop_yaw, BuildSystem:build_system):void=
        # See if we already have the right prop at this location
        if:
            Index := BuildSystem.MeshGrid.GetVoxelIndex(Pos)
            Entry := BuildSystem.MeshGrid.PropGrid[Index]?
            Entry.PieceType = PropToSpawn
            Entry.Rotation = Rot
        then:
            # it's the prop we want, do nothing
            #Print("Existing mesh in place")
        else:
            # Clear existing prop and spawn a new one
            BuildSystem.MeshGrid.ClearPropForVoxel(Pos)                            
            BuildSystem.MeshGrid.SpawnProp(PropToSpawn, Pos, Rot)
            set BuildSystem.PropCount += 1

    SpawnPropsFromTiles(Box:voxel_box, BuildSystem:build_system):void=

        PropPrimes:tuple(int,int,int,int) := (37, 53, 61, 191)
        RotPrimes:tuple(int,int,int,int) := (37, 53, 61, 191)


        # Iterate over all tiles
        for:
            X := Box.Origin.X .. Box.EndX()
            Y := Box.Origin.Y .. Box.EndY()
        do:
            Pos := MakeVector3i(X,Y,Box.Origin.Z)
            if(Tile := BuildSystem.WFCSystem.GetTile[Pos]): 
               
                var PropToSpawnOpt:?piece_type = false
                var SpawnRot:prop_yaw = prop_yaw.deg_0

                # If we only have one option (ie fully collapsed), spawn desired prop
                if:
                    Tile.NumRemainingOptions = 1
                    Option := Tile.GetFirstOption[]
                then:
                    var SelectedPropIndex:int = 0
                    NumProps := Option.Mesh.Props.Length
                    if(NumProps > 1):
                        set SelectedPropIndex = RandomIntFromVector(Pos, NumProps-1, PropPrimes) 

                    #PropIdx := GetRandomInt(0, NumProps-1) 
                    if(Prop := Option.Mesh.Props[SelectedPropIndex]):

                        set PropToSpawnOpt = option{ Prop }

                        if(Option.Mesh.RandomPropRotation?):
                            set SpawnRot = RotIntToPropYaw(RandomIntFromVector(Pos, 3, RotPrimes))
                            #set SpawnRot = RotIntToPropYaw(GetRandomInt(0, 3))
                        else:
                            set SpawnRot = RotIntToPropYaw(Option.Rot)                    
                            
                    # Tile was not collapsed, spawn fallback 
                    else:
                        set PropToSpawnOpt = Tile.Model.FallbackPropOpt


                    # Update mesh grid to match desired outcome
                    if(PropToSpawn := PropToSpawnOpt?):
                        SpawnPropIfNeeded(PropToSpawn, Pos, SpawnRot, BuildSystem)

                    else:
                        # We want nothing, so just clear. Shouldn't really get here.
                        BuildSystem.MeshGrid.ClearPropForVoxel(Pos)
