# Knowledge-neutral Verse snippet; project names and asset IDs generalized.
# id: snippet-265

using { /Verse.org/Simulation }
using { /UnrealEngine.com/Temporary/SpatialMath }

# Defines a box within the voxel grid
voxel_box<public> := struct<computes>:
    # Min corner of box
    Origin<public>:vector3i = vector3i{X:=0,Y:=0,Z:=0}
    # Size of box
    Size<public>:vector3i = vector3i{X:=0,Y:=0,Z:=0}

voxel_box_face<public> := enum:
    PosX
    NegX
    PosY
    NegY
    PosZ
    NegZ

# Array of all faces of a box
AllVoxelBoxFaces:[]voxel_box_face := array{voxel_box_face.PosX, voxel_box_face.NegX, voxel_box_face.PosY, voxel_box_face.NegY, voxel_box_face.PosZ, voxel_box_face.NegZ}

ToString(Face:voxel_box_face):string=
    case(Face):
        voxel_box_face.PosX => "PosX"
        voxel_box_face.NegX => "NegX"            
        voxel_box_face.PosY => "PosY"
        voxel_box_face.NegY => "NegY"
        voxel_box_face.PosZ => "PosZ"
        voxel_box_face.NegZ => "NegZ"

FaceToVector3(Face:voxel_box_face):vector3=  
    case(Face):
        voxel_box_face.PosX => MakeVector3( 1.0, 0.0, 0.0)
        voxel_box_face.NegX => MakeVector3(-1.0, 0.0, 0.0)
        voxel_box_face.PosY => MakeVector3( 0.0, 1.0, 0.0)
        voxel_box_face.NegY => MakeVector3( 0.0,-1.0, 0.0)
        voxel_box_face.PosZ => MakeVector3( 0.0, 0.0, 1.0)
        voxel_box_face.NegZ => MakeVector3( 0.0, 0.0,-1.0)

FaceToVector3i(Face:voxel_box_face)<computes>:vector3i=  
    case(Face):
        voxel_box_face.PosX => MakeVector3i(1,0,0)
        voxel_box_face.NegX => MakeVector3i(-1,0,0)
        voxel_box_face.PosY => MakeVector3i(0,1,0)
        voxel_box_face.NegY => MakeVector3i(0,-1,0)
        voxel_box_face.PosZ => MakeVector3i(0,0,1)
        voxel_box_face.NegZ => MakeVector3i(0,0,-1)

PropYawToFace(Yaw:prop_yaw):voxel_box_face=
    case(Yaw):
        prop_yaw.deg_0 => voxel_box_face.PosX
        prop_yaw.deg_90 => voxel_box_face.PosY
        prop_yaw.deg_180 => voxel_box_face.NegX
        prop_yaw.deg_270 => voxel_box_face.NegY


FaceToPropYaw(Face:voxel_box_face):prop_yaw=
    case(Face):
        voxel_box_face.PosX => prop_yaw.deg_0
        voxel_box_face.PosY => prop_yaw.deg_90
        voxel_box_face.NegX => prop_yaw.deg_180
        voxel_box_face.NegY => prop_yaw.deg_270
        _ => prop_yaw.deg_0 # Doesn't work for Z faces...


ApplyRotToFace(Face:voxel_box_face, Yaw:prop_yaw):voxel_box_face=

    var Result:voxel_box_face = Face

    # Only works for X and Y
    if(Face <> voxel_box_face.PosZ and Face <> voxel_box_face.NegZ):
        # convert face to rot
        var FaceRot:prop_yaw = FaceToPropYaw(Face)

        # update rotation
        if(Yaw = prop_yaw.deg_90):
            set FaceRot = PropYawAdd90(FaceRot)
        else if(Yaw = prop_yaw.deg_180):
            set FaceRot = PropYawAdd90(PropYawAdd90(FaceRot))
        else if(Yaw = prop_yaw.deg_270):
            set FaceRot = PropYawSubtract90(FaceRot)

        # convert rot to face
        set Result = PropYawToFace(FaceRot)

    return Result

#----

MakeVoxelBox<public>(InOriginX:int, InOriginY:int, InOriginZ:int, InSizeX:int, InSizeY:int, InSizeZ:int)<computes>:voxel_box=
    voxel_box:
        Origin := MakeVector3i(InOriginX, InOriginY, InOriginZ)
        Size := MakeVector3i(InSizeX, InSizeY, InSizeZ)

MakeVoxelBox<public>(InOrigin:vector3i, InSize:vector3i)<computes>:voxel_box=
    voxel_box:
        Origin := InOrigin
        Size := InSize


(Box:voxel_box).EndX()<computes>:int=
    Box.Origin.X + Box.Size.X - 1

(Box:voxel_box).EndY()<computes>:int=
    Box.Origin.Y + Box.Size.Y - 1

(Box:voxel_box).EndZ()<computes>:int=
    Box.Origin.Z + Box.Size.Z - 1  



# Create new box containing box supplied boxes
operator'+'<public>(A:voxel_box, B:voxel_box)<computes>:voxel_box=
    MinX := Min(A.Origin.X, B.Origin.X)
    MaxX := Max(A.EndX(), B.EndX())
    MinY := Min(A.Origin.Y, B.Origin.Y)
    MaxY := Max(A.EndY(), B.EndY())
    MinZ := Min(A.Origin.Z, B.Origin.Z)
    MaxZ := Max(A.EndZ(), B.EndZ())

    Origin := MakeVector3i(MinX, MinY, MinZ)
    Size := MakeVector3i(MaxX - MinX + 1, MaxY - MinY + 1, MaxZ - MinZ + 1)
    return MakeVoxelBox(Origin, Size)
   
# Create new box containing box and supplied voxel
operator'+'<public>(A:voxel_box, B:vector3i)<computes>:voxel_box=
    MinX := Min(A.Origin.X, B.X)
    MaxX := Max(A.EndX(), B.X)
    MinY := Min(A.Origin.Y, B.Y)
    MaxY := Max(A.EndY(), B.Y)
    MinZ := Min(A.Origin.Z, B.Z)
    MaxZ := Max(A.EndZ(), B.Z)

    Origin := MakeVector3i(MinX, MinY, MinZ)
    Size := MakeVector3i(MaxX - MinX + 1, MaxY - MinY + 1, MaxZ - MinZ + 1)
    return MakeVoxelBox(Origin, Size)

(Box:voxel_box).SizeGreaterEqualThan(A:vector3i)<decides><transacts>:void=
    Box.Size.X >= A.X
    Box.Size.Y >= A.Y
    Box.Size.Z >= A.Z

(Box:voxel_box).Contains(A:vector3i)<decides><transacts>:void=
    A.X >= Box.Origin.X
    A.X <= Box.EndX()
    A.Y >= Box.Origin.Y
    A.Y <= Box.EndY()
    A.Z >= Box.Origin.Z
    A.Z <= Box.EndZ()

(Box:voxel_box).Intersects(A:voxel_box)<decides><transacts>:void=
    A.EndX() >= Box.Origin.X
    A.Origin.X <= Box.EndX()
    A.EndY() >= Box.Origin.Y
    A.Origin.Y <= Box.EndY()
    A.EndZ() >= Box.Origin.Z
    A.Origin.Z <= Box.EndZ()

# Return box that is Box made to fit within Bounds
# Fails if box is totally clipped away (does not overlap bounds)
ClipBoxToBounds<public>(Box:voxel_box, Bounds:voxel_box)<decides><transacts>:voxel_box=
    var Result:voxel_box = Box
    if(Result.Origin.X < Bounds.Origin.X):
        set Result.Size.X = Result.Size.X - (Bounds.Origin.X - Result.Origin.X)
        set Result.Origin.X = Bounds.Origin.X

    if(Result.EndX() > Bounds.EndX()):
        set Result.Size.X = Result.Size.X - (Result.EndX() - Bounds.EndX())

    if(Result.Origin.Y < Bounds.Origin.Y):
        set Result.Size.Y = Result.Size.Y - (Bounds.Origin.Y - Result.Origin.Y)
        set Result.Origin.Y = Bounds.Origin.Y
    
    if(Result.EndY() > Bounds.EndY()):
        set Result.Size.Y = Result.Size.Y - (Result.EndY() - Bounds.EndY())

    if(Result.Origin.Z < Bounds.Origin.Z):
        set Result.Size.Z = Result.Size.Z - (Bounds.Origin.Z - Result.Origin.Z)
        set Result.Origin.Z = Bounds.Origin.Z
    
    if(Result.EndZ() > Bounds.EndZ()):
        set Result.Size.Z = Result.Size.Z - (Result.EndZ() - Bounds.EndZ())

    var ReturnBox:?voxel_box = false
    if(Result.Size.X > 0 and Result.Size.Y > 0 and Result.Size.Z > 0):
        set ReturnBox = option{Result}

    ReturnBox?


        

ToString<public>(Box:voxel_box)<computes>:string=
    "{Box.Origin} : {Box.Size}"


(Box:voxel_box).NumVoxels()<computes>:int=
    Box.Size.X * Box.Size.Y * Box.Size.Z

(Box:voxel_box).Equals<public>(A:voxel_box)<computes><decides>:void=
    Box.Origin.Equals[A.Origin] and Box.Size.Equals[A.Size]

# Expand this box by Amount units in all directions
(Box:voxel_box).Expand(Amount:int)<transacts>:voxel_box=
    var NewBox:voxel_box = voxel_box{}
    AmountVec := MakeVector3i(Amount, Amount, Amount)
    set NewBox.Origin = Box.Origin - AmountVec
    set NewBox.Size = Box.Size + (AmountVec * 2)
    return NewBox

(Box:voxel_box).AddVoxelsToArray(Array:[]vector3i)<computes>:[]vector3i=   
    BoxVoxels := for:
        X := Box.Origin.X .. Box.EndX()
        Y := Box.Origin.Y .. Box.EndY()
        Z := Box.Origin.Z .. Box.EndZ()
    do:
        MakeVector3i(X,Y,Z)

    return Array + BoxVoxels


(Box:voxel_box).Expand(Face:voxel_box_face)<transacts>:voxel_box=
    var NewBox:voxel_box = Box

    # X
    if(Face=voxel_box_face.PosX):
        set NewBox.Size.X += 1
    else if(Face=voxel_box_face.NegX):
        set NewBox.Origin.X -= 1
        set NewBox.Size.X += 1
    # Y
    else if(Face=voxel_box_face.PosY):
        set NewBox.Size.Y += 1
    else if(Face=voxel_box_face.NegY):
        set NewBox.Origin.Y -= 1
        set NewBox.Size.Y += 1
    # Z
    else if(Face=voxel_box_face.PosZ):
        set NewBox.Size.Z += 1
    else if(Face=voxel_box_face.NegZ):
        set NewBox.Origin.Z -= 1
        set NewBox.Size.Z += 1

    return NewBox
