# Knowledge-neutral Verse snippet; project names and asset IDs generalized.
# id: snippet-266

using { /Verse.org/Simulation }
using { /Verse.org/Simulation/Tags }
using { /Verse.org/Assets }
using { /Verse.org/Colors }
using { /Verse.org/Random }
using { /UnrealEngine.com/Temporary/Diagnostics }
using { /UnrealEngine.com/Temporary/SpatialMath }
using { /Fortnite.com/Characters }
using { /Fortnite.com/Devices }

# Category of voxel in our build grid
build_category<public> := enum:
    Building1
    Park
    Null

# Enum to string util for debug logging
ToString(Category:build_category)<computes>:string=
    case(Category):
        build_category.Building1 => "Building"
        build_category.Park => "Park"
        build_category.Null => "Null"

# Returns if category is 'surface only', can only sit on top of another category, and does not block raycasts
CategoryIsSurfaceOnly(Category:build_category)<transacts><decides>:void=
    Category = build_category.Park

# Get 'solidity' of a category, used for meshing. First value is horizontal, second is vertical
GetSolidityOfCategory(Category:build_category)<computes>:tuple(logic,logic)=
    case(Category):
        build_category.Building1 => (true, true)
        build_category.Park => (false, false)
        build_category.Null => (false, false)

# When building SDF, max distance to look in each direction
MaxDistanceSearch:int := 6
  
# Result from a raycast against voxel grid
ray_cast_result<public> := struct<computes>:
    Hit<public>:logic = false
    HitLocation<public>:vector3 = vector3{X:=0.0, Y:=0.0, Z:=0.0}
    HitVoxel<public>:vector3i = vector3i{X:=0, Y:=0, Z:=0}
    HitFace<public>:voxel_box_face = voxel_box_face.PosZ
    HitDistance<public>:float = 0.0

# Structure stored for each occupied voxel
voxel_cell<public> := struct<computes>:
    Category<public>:build_category

# One entry in Signed Distance Field structure
# Contains distance in each direction to an empty voxel, as well as sum of minimums along each axis
sdf_entry<public> := struct<computes>:
    PosXDist<public>:int = 0
    NegXDist<public>:int = 0
    PosYDist<public>:int = 0
    NegYDist<public>:int = 0
    PosZDist<public>:int = 0
    NegZDist<public>:int = 0
    SumMinDist<public>:int = 0

# Class that holds 3D voxel grid, and supplies functions for modifying and raycasting against it
voxel_grid<public> := class:
    Logger:log = log{Channel:=LogClass}

    # Main array representing the 3D grid of voxels
    var Cells<public> : []?voxel_cell = array{} 
    # Size of grid 
    var Size<public> : vector3i = vector3i{}
    # Build system that owns this grid
    var BuildSystem<public> : ?build_system = false

    # Signed distance field like structure, stores distance from each occupied voxel to an empty voxel, in each direction
    var SDF<public> : []?sdf_entry = array{}

    # Current bounds of occupied voxels
    var Bounds : voxel_box = voxel_box{}
    # Are Bounds currently valid (false when no voxels are occupied)
    var BoundsValid : logic = false
    # Number of voxels in grid that are occupied
    var NumOccupiedVoxels : int = 0

    VoxelDebugDraw<public> : debug_draw = debug_draw{}

    # Gets the 1D array index from a 3D location in grid
    GetVoxelIndex<public>(X:int, Y:int, Z:int)<transacts>:int=
        # TODO PERF Remove before shipping
        #if(X < 0 or X >= Size(0) or Y < 0 or Y >= Size(1) or Z < 0 or Z >= Size(2)):
        #    Logger.Print("GetVoxelIndexA: Out of range!")

        return (X * (Size.Y*Size.Z)) + (Y * Size.Z) + Z

    GetVoxelIndex<public>(Vec:vector3i)<transacts>:int=
        # TODO PERF Remove before shipping
        #if(Vec.X < 0 or Vec.X >= Size(0) or Vec.Y < 0 or Vec.Y >= Size(1) or Vec.Z < 0 or Vec.Z >= Size(2)):
        #    Logger.Print("GetVoxelIndexB: Out of range!")

        return (Vec.X * (Size.Y*Size.Z)) + (Vec.Y * Size.Z) + Vec.Z

    # Get total number of voxels in 3D grid
    GetVoxelCount<public>()<transacts>:int=
        Size.X*Size.Y*Size.Z

    Init<public>(_BuildSystem:build_system, XSize:int, YSize:int, ZSize:int):void=
        set Size = MakeVector3i(XSize, YSize, ZSize)
        set BuildSystem = option{_BuildSystem}

        set Cells = for (X := 0..GetVoxelCount()-1):
            false

        # Reset SDF
        set SDF = for (X := 0..GetVoxelCount()-1):
            false         
            
    # Clear all voxels in grid
    Clear<public>():void=
        for (I -> Cell : Cells):           
            # Set voxel to empty
            if( set Cells[I] = false ) {} 

        # Empty SDF
        set SDF = for (X := 0..GetVoxelCount()-1):
            false  

        set BoundsValid = false

    # Sets the status of the voxel and the supplied position to the specified category
    SetVoxel<public>(Voxel:vector3i, Category:build_category):void=
        
        if(Cell := GetCell[Voxel.X, Voxel.Y, Voxel.Z]):
            # Clear SDF entries here, where we know previous category
            RemoveSDFEntry(Voxel, Cell.Category)
          
        NewCell := voxel_cell:
            Category := Category

        if( set Cells[GetVoxelIndex(Voxel)] = option{ NewCell } ) {}


    # Set the voxel at the specified location to empty
    ClearVoxel<public>(Voxel:vector3i):void=
      
        if(Cell := GetCell[Voxel.X, Voxel.Y, Voxel.Z]):
            # Clear SDF entries here, where we know previous category
            RemoveSDFEntry(Voxel, Cell.Category)

        if( set Cells[GetVoxelIndex(Voxel)] = false ) {}


    # Set all voxels in the specified box to empty
    ClearVoxels<public>(Box:voxel_box):void=
        for: 
            X := Box.Origin.X .. Box.EndX()
            Y := Box.Origin.Y .. Box.EndY()
            Z := Box.Origin.Z .. Box.EndZ()
        do:
            ClearVoxel(MakeVector3i(X,Y,Z))
        
    GetCell<public>(X:int, Y:int, Z:int)<decides><transacts>:voxel_cell=
        Cells[GetVoxelIndex(X, Y, Z)]?

    GetCell<public>(Vec:vector3i)<decides><transacts>:voxel_cell=
        GetCell[Vec.X, Vec.Y, Vec.Z]

    # Note - this fails if cell is not occupied
    GetCellCategory<public>(X:int, Y:int, Z:int)<decides><transacts>:build_category=
        GetCell[X, Y, Z].Category

    GetCellCategory<public>(Vec:vector3i)<decides><transacts>:build_category=
        GetCell[Vec.X, Vec.Y, Vec.Z].Category        

    OptCellToString(Cell:?voxel_cell)<computes>:string=
        if(Cell?):
            "CELL"
        else:
            "FALSE"
      

    # Get world-space box for the supplied voxel
    GetBoxFromVoxel(Voxel:vector3i)<transacts>:box=
        var ResultBox:box = MakeZeroBox()

        LocalCenter := (Voxel * BuildGridSize) + (0.5 * BuildGridSize)
        if(Center:vector3 = BuildSystem?.BuildSite?.ZoneOrigin + LocalCenter):
            set ResultBox = MakeBox(Center, 0.5 * BuildGridSize)

        return ResultBox
        

    # Get world space location of the center of the voxel
    GetCenterOfVoxel(Voxel:vector3i):vector3=
        var Result:vector3 = (Voxel * BuildGridSize) + (0.5 * BuildGridSize)
        if(TempResult := Result + BuildSystem?.BuildSite?.ZoneOrigin):
            set Result = TempResult
        return Result

    # Get world space location of the center of the voxel box
    GetCenterOfVoxelBox(InBox:voxel_box):vector3=
        var Result:vector3 = MakeZeroVector3()
        
        BoxExtent := InBox.Size * 0.5 * BuildGridSize
        set Result = InBox.Origin * BuildGridSize + BoxExtent
        if(TempResult := Result + BuildSystem?.BuildSite?.ZoneOrigin):
            set Result = TempResult

        return Result
        
    # Get the world space location of the top of the supplied voxel 
    GetTopOfVoxel(Voxel:vector3i):vector3=
        return GetCenterOfVoxel(Voxel) + MakeVector3(0.0, 0.0, 200.0)

    # Get the world space location of the top of the supplied voxel box
    GetTopOfVoxelBox(InBox:voxel_box):vector3=
        return GetCenterOfVoxelBox(InBox) + MakeVector3(0.0, 0.0, InBox.Size.Z * 200.0)

    # Get the world space location of the bottom of the supplied voxel box
    GetBottomOfVoxelBox(InBox:voxel_box):vector3=
        return GetCenterOfVoxelBox(InBox) - MakeVector3(0.0, 0.0, InBox.Size.Z * 200.0)        


    DebugDrawVoxel(InX:int, InY:int, InZ:int, InColor:color, InInset:float):void=
        Voxel := MakeVector3i(InX, InY, InZ)
        VoxelCenter := (Voxel * BuildGridSize) + (0.5 * BuildGridSize)

        # Drawn a box inset from actual voxel, so we don't overlap 
        if(Center:vector3 = BuildSystem?.BuildSite?.ZoneOrigin + VoxelCenter):
            VoxelDebugDraw.DrawBox(
                Center, 
                IdentityRotation(), 
                ?Extent := (0.5 * BuildGridSize) - MakeVector3(InInset, InInset, InInset), 
                ?Color := InColor, 
                ?DrawDurationPolicy := debug_draw_duration_policy.Persistent)


    DebugDrawVoxelBox(InBox:voxel_box, InColor:color, InInset:float):void=
        BoxExtent := InBox.Size * BuildGridSize * 0.5
        BoxCenter := (InBox.Origin * BuildGridSize) + BoxExtent

        if(Center:vector3 = BuildSystem?.BuildSite?.ZoneOrigin + BoxCenter):
            VoxelDebugDraw.DrawBox(
                Center, 
                IdentityRotation(), 
                ?Extent := BoxExtent - vector3{X:=InInset, Y:=InInset, Z:=InInset}, 
                ?Color := InColor, 
                ?DrawDurationPolicy := debug_draw_duration_policy.Persistent)
       
    GetVoxelBoxCenter(InBox:voxel_box):vector3=
        BoxExtent := InBox.Size * BuildGridSize * 0.5
        BoxCenter := (InBox.Origin * BuildGridSize) + BoxExtent

        var Center:vector3 = BoxCenter
        if(BZ := BuildSystem?.BuildSite?):
            set Center = Center + BZ.ZoneOrigin
        return Center

    DebugDrawBoxCenter(InBox:voxel_box, InColor:color):void=
        Center := GetVoxelBoxCenter(InBox)
        VoxelDebugDraw.DrawSphere(
            Center, 
            ?Radius := 10.0,
            ?Color := InColor,
            ?DrawDurationPolicy := debug_draw_duration_policy.Persistent)

    DebugDrawLineBetweenBoxes(BoxA:voxel_box, BoxB:voxel_box, InColor:color):void=
        ACenter := GetVoxelBoxCenter(BoxA)
        BCenter := GetVoxelBoxCenter(BoxB)
        VoxelDebugDraw.DrawLine(
            ACenter, 
            BCenter,
            ?Color := InColor,
            ?DrawDurationPolicy := debug_draw_duration_policy.Persistent)

    # Walk in both directions from this voxel and find mind dist before we hit an edge (or the edge of the grid)
    FindDistInPosX(Voxel:vector3i, Category:build_category):int=
        var PosDist:int = 0
        var TestVox:vector3i = Voxel
        loop:
            set TestVox.X += 1
            if( PosDist < MaxDistanceSearch and TestVox.X < Size.X and GetCellCategory[TestVox.X, TestVox.Y, TestVox.Z] = Category ):
                set PosDist += 1
            else:
                break
        return PosDist

    FindDistInNegX(Voxel:vector3i, Category:build_category):int=
        var NegDist:int = 0
        var TestVox:vector3i = Voxel
        loop:
            set TestVox.X -= 1
            # We can stop if we go further than we did in the other direction, we want the min of the two
            if( NegDist < MaxDistanceSearch and TestVox.X >= 0 and GetCellCategory[TestVox.X, TestVox.Y, TestVox.Z] = Category ):
                set NegDist += 1
            else:
                break
        return NegDist

    # Y
    FindDistInPosY(Voxel:vector3i, Category:build_category):int=
        var PosDist:int = 0
        var TestVox:vector3i = Voxel
        loop:
            set TestVox.Y += 1
            if( PosDist < MaxDistanceSearch and TestVox.Y < Size.Y and GetCellCategory[TestVox.X, TestVox.Y, TestVox.Z] = Category ):
                set PosDist += 1
            else:
                break
        return PosDist

    FindDistInNegY(Voxel:vector3i, Category:build_category):int=
        var NegDist:int = 0
        var TestVox:vector3i = Voxel
        loop:
            set TestVox.Y -= 1
            # We can stop if we go further than we did in the other direction, we want the min of the two
            if( NegDist < MaxDistanceSearch and TestVox.Y >= 0 and GetCellCategory[TestVox.X, TestVox.Y, TestVox.Z] = Category ):
                set NegDist += 1
            else:
                break
        return NegDist

    # Z
    FindDistInPosZ(Voxel:vector3i, Category:build_category):int=
        var PosDist:int = 0
        var TestVox:vector3i = Voxel
        loop:
            set TestVox.Z += 1
            if( PosDist < MaxDistanceSearch and TestVox.Z < Size.Z and GetCellCategory[TestVox.X, TestVox.Y, TestVox.Z] = Category ):
                set PosDist += 1
            else:
                break
        return PosDist

    FindDistInNegZ(Voxel:vector3i, Category:build_category):int=
        var NegDist:int = 0
        var TestVox:vector3i = Voxel
        loop:
            set TestVox.Z -= 1
            # We can stop if we go further than we did in the other direction, we want the min of the two
            if( NegDist < MaxDistanceSearch and TestVox.Z >= 0 and GetCellCategory[TestVox.X, TestVox.Y, TestVox.Z] = Category ):
                set NegDist += 1
            else:
                break
        return NegDist

    # Count how many voxels in this box have entries in the SDF
    NumVoxelsInBoxHaveSDFEntries(Box:voxel_box):int=
        var NumOccupied:int = 0
        for(X := Box.Origin.X..Box.EndX(); Y := Box.Origin.Y..Box.EndY(); Z := Box.Origin.Z..Box.EndZ()):
            if(SDF[GetVoxelIndex(X, Y, Z)]?):
                set NumOccupied += 1

        return NumOccupied


    # Update Bounds, NumOccupiedVoxels and SDF
    # Box needs to be within grid bounds
    AddBoxToSDF<public>(Box:voxel_box, OnlyCategories:[]build_category):void=

        #Print("AddBoxToSDF {Box}")

        # Expand bounds to new box
        if(not BoundsValid?):
            set Bounds = Box
            set BoundsValid = true
        else:
            set Bounds = Bounds + Box

        # Update number of voxels that are now occupied
        AlreadyOccupied := NumVoxelsInBoxHaveSDFEntries(Box)
        set NumOccupiedVoxels += (Box.NumVoxels() - AlreadyOccupied)

        #Print("ADD ~ NumVoxels = {NumOccupiedVoxels},  Bounds = {Bounds}")

        UpdateSDF_2(Box, OnlyCategories)             


    # Update Bounds, NumOccupiedVoxels and SDF
    # Box needs to be within grid bounds
    RemoveBoxFromSDF(Box:voxel_box, OnlyCategories:[]build_category):void=
                
        # Removing this box can only affect the bounds if it touches the existing bounds
        if:
            Box.Origin.X = Bounds.Origin.X or
            Box.Origin.Y = Bounds.Origin.Y or
            Box.Origin.Z = Bounds.Origin.Z or
            Box.EndX() = Bounds.EndX() or
            Box.EndY() = Bounds.EndY() or
            Box.EndZ() = Bounds.EndZ()
        # Rebuild bounds in this case
        then:
            set BoundsValid = false
            set NumOccupiedVoxels = 0
            
            #profile("   RecomputeBounds"):
                # Removing a box of voxels can only make bounds smaller, so we only need to iterate over previous bounds
                OldBounds := Bounds
                var MinX:int = 99999
                var MaxX:int = -99999
                var MinY:int = 99999
                var MaxY:int = -99999
                var MinZ:int = 99999
                var MaxZ:int = -99999
                for:
                    X := OldBounds.Origin.X .. OldBounds.EndX()
                    Y := OldBounds.Origin.Y .. OldBounds.EndY()
                    Z := OldBounds.Origin.Z .. OldBounds.EndZ()
                do:
                    # Get category of this cell (will fail if empty)
                    if(Cell := GetCell[X, Y, Z]):
                        set MinX = Min(MinX, X)
                        set MaxX = Max(MaxX, X)
                        set MinY = Min(MinY, Y)
                        set MaxY = Max(MaxY, Y)
                        set MinZ = Min(MinZ, Z)
                        set MaxZ = Max(MaxZ, Z)                            
                        set NumOccupiedVoxels += 1

                if(NumOccupiedVoxels > 0):
                    BoundsSize := MakeVector3i((MaxX-MinX)+1, (MaxY-MinY)+1, (MaxZ-MinZ)+1)
                    BoundsOrigin := MakeVector3i(MinX, MinY, MinZ)
                    set Bounds = MakeVoxelBox(BoundsOrigin, BoundsSize)
                    set BoundsValid = true

        # Don't need to adjust bounds, but do need to update NumOccupiedVoxels
        else:
            AlreadyOccupied := NumVoxelsInBoxHaveSDFEntries(Box)
            set NumOccupiedVoxels -= AlreadyOccupied

        UpdateSDF_2(Box, OnlyCategories)

    # Remove entry from SDF structure
    RemoveSDFEntry(Voxel:vector3i, PreviousCategory:build_category):void=
        # If we have an existing entry we need to clear
        VIdx := GetVoxelIndex(Voxel)
        if(ExistingEntry := SDF[VIdx]?):
            if( set SDF[VIdx] = false ) {}

    # Update a single entry in the SDF structure. Will check in each direction to find empty voxel.
    UpdateSDF_2_Entry(X:int, Y:int, Z:int, OnlyCategories:[]build_category, ModifiedDirOpt:?voxel_box_face):void=
        # If cell has a category, and it's the right one, add/update SDF entry
        if:
            Category := GetCellCategory[X, Y, Z]
            CategoryFilter[Category, OnlyCategories]
        then:
            Voxel := MakeVector3i(X, Y, Z)
            VIdx := GetVoxelIndex(Voxel)

            var NewEntry:sdf_entry = sdf_entry{}
            var RecalcAll:logic = true

            if(ExistingEntry := SDF[VIdx]?):

                # Existing entry and direction given, only recalc distances in that direction (take others from existing entry)
                if(ModifiedDir := ModifiedDirOpt?):
                    set NewEntry = ExistingEntry
                    case(ModifiedDir):
                        voxel_box_face.PosX => set NewEntry.PosXDist = FindDistInPosX(Voxel, Category)
                        voxel_box_face.NegX => set NewEntry.NegXDist = FindDistInNegX(Voxel, Category)
                        voxel_box_face.PosY => set NewEntry.PosYDist = FindDistInPosY(Voxel, Category)
                        voxel_box_face.NegY => set NewEntry.NegYDist = FindDistInNegY(Voxel, Category)
                        voxel_box_face.PosZ => set NewEntry.PosZDist = FindDistInPosZ(Voxel, Category)
                        voxel_box_face.NegZ => set NewEntry.NegZDist = FindDistInNegZ(Voxel, Category)    

                    set RecalcAll = false

            # New entry, or no direction specified, recalc all distances
            if(RecalcAll?):
                set NewEntry.PosXDist = FindDistInPosX(Voxel, Category)
                set NewEntry.NegXDist = FindDistInNegX(Voxel, Category)
                set NewEntry.PosYDist = FindDistInPosY(Voxel, Category)
                set NewEntry.NegYDist = FindDistInNegY(Voxel, Category)
                set NewEntry.PosZDist = FindDistInPosZ(Voxel, Category)
                set NewEntry.NegZDist = FindDistInNegZ(Voxel, Category)

            #Print("{Voxel} : {NewEntry.PosXDist} {NewEntry.NegXDist} - {NewEntry.PosYDist} {NewEntry.NegYDist} - {NewEntry.PosZDist} {NewEntry.NegZDist}")
            set NewEntry.SumMinDist = Min(NewEntry.PosXDist, NewEntry.NegXDist) + Min(NewEntry.PosYDist, NewEntry.NegYDist) + Min(NewEntry.PosZDist, NewEntry.NegZDist)

            # Update SDF entry
            if( set SDF[VIdx] = option{NewEntry} ) {}



    # SDF-type calculation which sums min distance in each direction to an empty voxel (or edge of grid)
    # Pass in box of voxels that have changed, will only recompute values that could have been affected
    # This does not clear existing entries, only adds new ones if a voxel has a category. Removal is done in Set/ClearVoxel
    UpdateSDF_2<public>(Box:voxel_box, OnlyCategories:[]build_category):void=
        #Print("UpdateSDF_2 : {Box}")
        # We only need to update distances within MaxDistanceSearch of the voxels that have changed
        

        # Cache ends
        BoxEndX := Box.EndX()
        BoxEndY := Box.EndY()
        BoxEndZ := Box.EndZ()

        BoundsEndX := Bounds.EndX()
        BoundsEndY := Bounds.EndY()
        BoundsEndZ := Bounds.EndZ()

        # Modified region itself
        for(X := Box.Origin.X .. BoxEndX, Y := Box.Origin.Y .. BoxEndY, Z := Box.Origin.Z .. BoxEndZ):
            UpdateSDF_2_Entry(X, Y, Z, OnlyCategories, false)

        # X
        if(BoxEndX < BoundsEndX):
            for(X := BoxEndX+1 .. Min(BoxEndX+MaxDistanceSearch, BoundsEndX), Y := Box.Origin.Y .. BoxEndY, Z := Box.Origin.Z .. BoxEndZ):
                UpdateSDF_2_Entry(X, Y, Z, OnlyCategories, option{voxel_box_face.NegX})

        if(Box.Origin.X > 0):
            for(X := Max(Box.Origin.X-MaxDistanceSearch, 0) .. Box.Origin.X-1, Y := Box.Origin.Y .. BoxEndY, Z := Box.Origin.Z .. BoxEndZ):
                UpdateSDF_2_Entry(X, Y, Z, OnlyCategories, option{voxel_box_face.PosX})


        # Y
        if(BoxEndY < BoundsEndY):
            for(X := Box.Origin.X .. BoxEndX, Y := BoxEndY+1 .. Min(BoxEndY+MaxDistanceSearch, BoundsEndY), Z := Box.Origin.Z .. BoxEndZ):
                UpdateSDF_2_Entry(X, Y, Z, OnlyCategories, option{voxel_box_face.NegY})

        if(Box.Origin.Y > 0):
            for(X := Box.Origin.X .. BoxEndX, Y := Max(Box.Origin.Y-MaxDistanceSearch, 0) .. Box.Origin.Y-1, Z := Box.Origin.Z .. BoxEndZ):
                UpdateSDF_2_Entry(X, Y, Z, OnlyCategories, option{voxel_box_face.PosY})


        # Z
        if(BoxEndZ < BoundsEndZ):
            for(X := Box.Origin.X .. BoxEndX, Y := Box.Origin.Y .. BoxEndY, Z := BoxEndZ+1 .. Min(BoxEndZ+MaxDistanceSearch, BoundsEndZ)):
                UpdateSDF_2_Entry(X, Y, Z, OnlyCategories, option{voxel_box_face.NegZ})

        if(Box.Origin.Z > 0):
            for(X := Box.Origin.X .. BoxEndX, Y := Box.Origin.Y .. BoxEndY, Z := Max(Box.Origin.Z-MaxDistanceSearch, 0) .. Box.Origin.Z-1):
                UpdateSDF_2_Entry(X, Y, Z, OnlyCategories, option{voxel_box_face.PosZ})


        # debug draw  
        <#
        for(X := Bounds.Origin.X .. Bounds.EndX(), Y := Bounds.Origin.Y .. Bounds.EndY(), Z := Bounds.Origin.Z .. Bounds.EndZ()):
            Voxel := MakeVector3i(X, Y, Z)
            VIdx := GetVoxelIndex(Voxel)

            if(Entry := SDF[VIdx]?):
                Frac := Clamp( (((Entry.SumMinDist * 1.0) / 10.0) * 360.0), 0.0, 360.0 )
                Color := MakeColorFromHSV(Frac, 1.0, 1.0)
                DebugDrawVoxel(X,Y,Z, Color, 10.0)
        #>


    # Generate an array of all cells that are touching a particular face of the supplied box
    GetCellsTouchingBoxFace(Box:voxel_box, Face:voxel_box_face)<decides><transacts>:voxel_box=

        var FBox:voxel_box = Box # init output box to supplied box
        var MakeCells:logic = true

        # X FACES
        if(Face = voxel_box_face.NegX or Face = voxel_box_face.PosX):
            set FBox.Size.X = 1

            if(Face = voxel_box_face.NegX): # NegX
                if(Box.Origin.X <= 0):
                    set MakeCells = false
                else:
                    set FBox.Origin.X = Box.Origin.X - 1
            else: # PosX
                if(Box.EndX() >= Size.X-1):
                    set MakeCells = false
                else:
                    set FBox.Origin.X = Box.Origin.X + Box.Size.X
        
        # Y FACES
        if(Face = voxel_box_face.NegY or Face = voxel_box_face.PosY):
            set FBox.Size.Y = 1

            if(Face = voxel_box_face.NegY): # NegY
                if(Box.Origin.Y <= 0):
                    set MakeCells = false
                else:
                    set FBox.Origin.Y = Box.Origin.Y - 1
            else: # PosY
                if(Box.EndY() >= Size.Y-1):
                    set MakeCells = false
                else:
                    set FBox.Origin.Y = Box.Origin.Y + Box.Size.Y

        # Z FACES
        if(Face = voxel_box_face.NegZ or Face = voxel_box_face.PosZ):
            set FBox.Size.Z = 1

            if(Face = voxel_box_face.NegZ): # NegZ
                if(Box.Origin.Z <= 0):
                    set MakeCells = false
                else:
                    set FBox.Origin.Z = Box.Origin.Z - 1
            else: # PosZ
                if(Box.EndZ() >= Size.Z-1):
                    set MakeCells = false
                else:
                    set FBox.Origin.Z = Box.Origin.Z + Box.Size.Z

        var OutBox:?voxel_box = false
        if(MakeCells?):
            set OutBox = option{FBox}
        OutBox?

    # Get voxels that touch a face of this one, within the grid
    GetNeighbouringVoxels(Voxel:vector3i, IncludePosZ:logic, IncludeNegZ:logic, IncludeXY:logic):[]vector3i=
        var Neighbours:[]vector3i = array{}
        if(IncludeXY? and Voxel.X > 0):
            set Neighbours += array{Voxel + FaceToVector3i(voxel_box_face.NegX)}
        if(IncludeXY? and Voxel.X < Size.X-1):
            set Neighbours += array{Voxel + FaceToVector3i(voxel_box_face.PosX)}
        if(IncludeXY? and Voxel.Y > 0):
            set Neighbours += array{Voxel + FaceToVector3i(voxel_box_face.NegY)}
        if(IncludeXY? and Voxel.Y < Size.Y-1):
            set Neighbours += array{Voxel + FaceToVector3i(voxel_box_face.PosY)}
        if(IncludeNegZ? and Voxel.Z > 0):
            set Neighbours += array{Voxel + FaceToVector3i(voxel_box_face.NegZ)}
        if(IncludePosZ? and Voxel.Z < Size.Z-1):
            set Neighbours += array{Voxel + FaceToVector3i(voxel_box_face.PosZ)}

        return Neighbours


        
    # Util used to determine if a voxel is 'solid' and should stop raycast
    Trace_VoxelIsSolid(Voxel:vector3i)<transacts>:logic=
        var Solid:logic = false
        if:
            Category := GetCellCategory[Voxel]
            not CategoryIsSurfaceOnly[Category]
        then:
            set Solid = true

        return Solid

    # Trace ray in voxel grid and return hit voxel
    Trace(Start:vector3, Dir:vector3, TraceDistance:float):ray_cast_result=

        var Result:ray_cast_result = ray_cast_result{}

        if:
            BuildSys := BuildSystem?
            ZoneOrigin := BuildSys.BuildSite?.ZoneOrigin
            LocalStart := Start - ZoneOrigin
            Caster := InitRayCaster[Dir, LocalStart, BuildGridSize]
        then:
           
            loop:
                # If we've gone too far
                if(Caster.DistanceTraced > TraceDistance):
                    break

                # Special case for hitting floor
                HitFloor := logic{Caster.CurrentVoxel.Z = -1 and Dir.Z < 0.0}

                # If this cell is in bounds, and set to a category, or the floor, we have a hit!
                if((BoundsValid? and Bounds.Contains[Caster.CurrentVoxel]) or HitFloor?):
                    if(HitFloor? or Trace_VoxelIsSolid(Caster.CurrentVoxel)?):
                        set Result.Hit = true
                        set Result.HitVoxel = Caster.CurrentVoxel
                        set Result.HitLocation = Caster.CurrentPosition + ZoneOrigin # Transform back to world space
                        set Result.HitFace = Caster.CurrentHitFace
                        set Result.HitDistance = Caster.DistanceTraced
                        break

                Caster.Next()


        return Result

    # Trace a ray onto a plane defined by a voxel and face of that voxel
    TracePlane(Start:vector3, Dir:vector3, TraceDistance:float, VoxOnPlane:vector3i, PlaneNormal:voxel_box_face):ray_cast_result=
        var Result:ray_cast_result = ray_cast_result{}

        if:
            BuildSys := BuildSystem?
            ZoneOrigin := BuildSys.BuildSite?.ZoneOrigin
        then:
            LocalStart := Start - ZoneOrigin
            HalfGridSize := (0.5 * BuildGridSize)
            VoxCentre := (VoxOnPlane * BuildGridSize) + HalfGridSize
            var TimeToPlane:float = 0.0

            if(PlaneNormal = voxel_box_face.PosX):
                if(Dir.X < 0.0):
                    set TimeToPlane = ((VoxCentre.X - HalfGridSize.X) - LocalStart.X)/Dir.X
                    if:
                        TimeToPlane > 0.0 and TimeToPlane < TraceDistance
                        Hit := LocalStart + (TimeToPlane * Dir)
                        HitVoxY := Floor[Hit.Y/BuildGridSize.Y]
                        HitVoxZ := Floor[Hit.Z/BuildGridSize.Z]
                    then:
                        set Result.Hit = true
                        set Result.HitVoxel = MakeVector3i(VoxOnPlane.X-1, HitVoxY, HitVoxZ)
            else if(PlaneNormal = voxel_box_face.NegX):
                if(Dir.X > 0.0):
                    set TimeToPlane = ((VoxCentre.X + HalfGridSize.X) - LocalStart.X)/Dir.X
                    if:
                        TimeToPlane > 0.0 and TimeToPlane < TraceDistance
                        Hit := LocalStart + (TimeToPlane * Dir)
                        HitVoxY := Floor[Hit.Y/BuildGridSize.Y]
                        HitVoxZ := Floor[Hit.Z/BuildGridSize.Z]
                    then:
                        set Result.Hit = true
                        set Result.HitVoxel = MakeVector3i(VoxOnPlane.X+1, HitVoxY, HitVoxZ)
            else if(PlaneNormal = voxel_box_face.PosY):
                if(Dir.Y < 0.0):
                    set TimeToPlane = ((VoxCentre.Y - HalfGridSize.Y) - LocalStart.Y)/Dir.Y
                    if:
                        TimeToPlane > 0.0 and TimeToPlane < TraceDistance
                        Hit := LocalStart + (TimeToPlane * Dir)
                        HitVoxX := Floor[Hit.X/BuildGridSize.X]
                        HitVoxZ := Floor[Hit.Z/BuildGridSize.Z]
                    then:
                        set Result.Hit = true
                        set Result.HitVoxel = MakeVector3i(HitVoxX, VoxOnPlane.Y-1, HitVoxZ)
            else if(PlaneNormal = voxel_box_face.NegY):
                if(Dir.Y > 0.0):
                    set TimeToPlane = ((VoxCentre.Y + HalfGridSize.Y) - LocalStart.Y)/Dir.Y
                    if:
                        TimeToPlane > 0.0 and TimeToPlane < TraceDistance
                        Hit := LocalStart + (TimeToPlane * Dir)
                        HitVoxX := Floor[Hit.X/BuildGridSize.X]
                        HitVoxZ := Floor[Hit.Z/BuildGridSize.Z]
                    then:
                        set Result.Hit = true
                        set Result.HitVoxel = MakeVector3i(HitVoxX, VoxOnPlane.Y+1, HitVoxZ)
            else if(PlaneNormal = voxel_box_face.PosZ):
                if(Dir.Z < 0.0):
                    set TimeToPlane = ((VoxCentre.Z - HalfGridSize.Z) - LocalStart.Z)/Dir.Z
                    if:
                        TimeToPlane > 0.0 and TimeToPlane < TraceDistance
                        Hit := LocalStart + (TimeToPlane * Dir)
                        HitVoxX := Floor[Hit.X/BuildGridSize.X]
                        HitVoxY := Floor[Hit.Y/BuildGridSize.Y]
                    then:
                        set Result.Hit = true
                        set Result.HitVoxel = MakeVector3i(HitVoxX, HitVoxY, VoxOnPlane.Z-1)
            else if(PlaneNormal = voxel_box_face.NegZ):
                if(Dir.Z > 0.0):
                    set TimeToPlane = ((VoxCentre.Z + HalfGridSize.Z) - LocalStart.Z)/Dir.Z
                    if:
                        TimeToPlane > 0.0 and TimeToPlane < TraceDistance
                        Hit := LocalStart + (TimeToPlane * Dir)
                        HitVoxX := Floor[Hit.X/BuildGridSize.X]
                        HitVoxY := Floor[Hit.Y/BuildGridSize.Y]
                    then:
                        set Result.Hit = true
                        set Result.HitVoxel = MakeVector3i(HitVoxX, HitVoxY, VoxOnPlane.Z+1)


            set Result.HitDistance = TimeToPlane
            set Result.HitFace = PlaneNormal

        return Result
