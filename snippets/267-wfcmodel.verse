# Knowledge-neutral Verse snippet; project names and asset IDs generalized.
# id: snippet-267

using { /Verse.org/Simulation }
using { /UnrealEngine.com/Temporary/SpatialMath }
using { Lib }

# Add two rotations (expressed as integer multiples of 90 degrees)
AddRotations(A:int, B:int):int=
    if(Result := Mod[A + B, 4]):
        return Result
    else:
        return 0

# Subtract two rotations (expressed as integer multiples of 90 degrees)
SubtractRotations(A:int, B:int):int=
    if(Result := Mod[A - B, 4]):
        return Result
    else:
        return 0

# Flip a rotation (expressed as integer multiples of 90 degrees)
FlipRotation(Rot:int):int=
    case(Rot):
        0 => 2
        1 => 3
        2 => 0
        3 => 1
        _ =>
            Print("FlipRotation: INVALID") 
            0

# EDGES/DIRECTIONS:
#        0
#       +-+
#     3 | | 1
#   Y   +-+
#   ↑    2
# X←˩


# A unique 'label' for the edge of a WFC tile. Used when deciding which tiles can be placed next to each other.
wfc_edge_label<public> := class<unique>:
    # Name, used for debugging
    Name<public>:[]char 
    # if this edge is symmetric
    Symmetric<public>:logic

ToString(Label:wfc_edge_label)<computes>:string=
    "{Label.Name}"

# Util to create a tile just given a set of labels
CreateTileFromEdgeLabels(BorderLabels:[]wfc_edge_label, Model:wfc_model)<transacts>:wfc_tile=

    # Create option for each label
    var Options:[wfc_mesh_option]int = map{}
    for(BorderLabel : BorderLabels):
        BorderMesh := wfc_mesh:
            Props := array{}
            Name := "BorderTile: {ToString(BorderLabel)}"
            Edges := array:
                BorderLabel
                BorderLabel
                BorderLabel
                BorderLabel            

        O := wfc_mesh_option{Mesh := BorderMesh}
        if( set Options[O] = 0 ){}

        # If not symmetric, add flipped version as well
        if(not BorderLabel.Symmetric?):
            FlippedLabel := Model.FlipEdgeLabel(BorderLabel)
            BorderMeshFlipped := wfc_mesh:
                Props := array{}
                Name := "BorderTile: {ToString(FlippedLabel)}"
                Edges := array:
                    FlippedLabel
                    FlippedLabel
                    FlippedLabel
                    FlippedLabel         
            FO := wfc_mesh_option{Mesh := BorderMeshFlipped}
            if( set Options[FO] = 0 ){}

    NewBorderTile := wfc_tile:
        RemainingOptions := Options
        NumRemainingOptions := Options.Length
        AllOptions := false
        Model := Model
    NewBorderTile.UpdateShannonEntropy()
    return NewBorderTile

# One mesh that can be placed in the WFC grid
wfc_mesh<public> := class:
    # Actual Props to spawn, selected from randomly
    Props<public>:[]piece_type 

    # Name just used for debugging
    Name<public>:[]char 

    # Info about each of the 4 edges (no fixed sized arrays in Verse)
    Edges<public>:[]wfc_edge_label

    # Don't allow this tile to be rotated as options
    AllowRotation<public>:logic = true

    # Randomize prop rotation when we pick
    RandomPropRotation<public>:logic = false

# One particular option (rotation) for this mesh
wfc_mesh_option<public> := class<unique>: 
    Mesh<public>:wfc_mesh
    Rot<public>:int = 0 # 0, 1, 2 or 3, rotating by 90 degrees clockwise

    GetMeshEdgeLabel(Direction:int):wfc_edge_label=
        TotalRot := SubtractRotations(Direction, Rot)
        if(Label := Mesh.Edges[TotalRot]):
            return Label
        else:
            Print("GetMeshEdge: Invalid!")
            return wfc_edge_label:
                Name:="Fail"
                Symmetric:=true


ToString<public>(Option:wfc_mesh_option)<computes>:string=
    "{Option.Mesh.Name} Rot:{Option.Rot}"    

# A set of possible mesh options
wfc_mesh_options_set<public> := class:
    var Map<public>:[wfc_mesh_option]int = map{}

# Options for a particular edge type in a particular direction
wfc_edge_options<public> := class:
    var OptionSets<public>:[]wfc_mesh_options_set = array:
        wfc_mesh_options_set{}
        wfc_mesh_options_set{}
        wfc_mesh_options_set{}
        wfc_mesh_options_set{}

    GetOptionsForDirection(Direction:int):wfc_mesh_options_set=
        # If this edge is symmetric, or we want non-flipped options, return normal set
        if(OptionSet := OptionSets[Direction]):
            return OptionSet
        else:
            return wfc_mesh_options_set{}

    AddOptionForDirection(Direction:int, Option:wfc_mesh_option):void=  
        OptionsSet := GetOptionsForDirection(Direction)
        if( set OptionsSet.Map[Option] = 1 ) {} 

# Class that represents a particular WFC model. 
# This includes the mapping from edge labels to mesh options that have that label
wfc_model<public> := class<unique>:

    # Map from an edge label to options that match that label
    var EdgeOptions:[wfc_edge_label]wfc_edge_options = map{}
    # All mesh options
    var AllMeshOptions:[wfc_mesh_option]int = map{}
    # Size of AllMeshOptions array
    var NumAllMeshOptions:int = 0

    # Map from label to flipped version of that label
    var EdgeLabelFlipMap:[wfc_edge_label]wfc_edge_label = map{}

    # Special tiles used for borders
    var BorderTile:?wfc_tile = false
    
    # Prop to use if we get over-constrained (no options possible)
    var FallbackPropOpt<public>:?piece_type = false

    GetOptionsForEdgeLabel(EdgeLabel:wfc_edge_label):wfc_edge_options=
        if(ExistingOptions := EdgeOptions[EdgeLabel]):
            return ExistingOptions
        else:
            NewOptions := wfc_edge_options{}
            if( set EdgeOptions[EdgeLabel] = NewOptions ) {}

            return NewOptions

    FlipEdgeLabel(InLabel:wfc_edge_label)<transacts>:wfc_edge_label=
        if(InLabel.Symmetric?):
            return InLabel
        else:
            if(FlippedLabel := EdgeLabelFlipMap[InLabel]):
                return FlippedLabel
            else:
                Print("Unable to find flipped version of {InLabel}!")
                return InLabel

    AddLabelFlipPair(A:wfc_edge_label, B:wfc_edge_label):void=
        if:
            set EdgeLabelFlipMap[A] = B
            set EdgeLabelFlipMap[B] = A

    # Add a mesh to the model
    # Will create options and update label->mesh option map
    AddMesh<public>(Mesh:wfc_mesh):void=

        # Create mesh option for each rotation of this mesh
        NewMeshOption0 := wfc_mesh_option{Mesh := Mesh, Rot := 0}
        NewMeshOption1 := wfc_mesh_option{Mesh := Mesh, Rot := 1}
        NewMeshOption2 := wfc_mesh_option{Mesh := Mesh, Rot := 2}
        NewMeshOption3 := wfc_mesh_option{Mesh := Mesh, Rot := 3}

        if ( set AllMeshOptions[NewMeshOption0] = 0 ){}
        set NumAllMeshOptions += 1

        if(Mesh.AllowRotation?):
            if ( set AllMeshOptions[NewMeshOption1] = 0 ){}
            if ( set AllMeshOptions[NewMeshOption2] = 0 ){}
            if ( set AllMeshOptions[NewMeshOption3] = 0 ){}
            set NumAllMeshOptions += 3

        # For each edge
        for(EdgeIndex := 0..3):
            if(EdgeLabel := Mesh.Edges[EdgeIndex]):

                # Get options object (will create if missing)
                Options := GetOptionsForEdgeLabel(EdgeLabel)

                # Add option for each rotation of mesh 
                Options.AddOptionForDirection(AddRotations(EdgeIndex, 0), NewMeshOption0)

                if(Mesh.AllowRotation?):
                    Options.AddOptionForDirection(AddRotations(EdgeIndex, 1), NewMeshOption1)
                    Options.AddOptionForDirection(AddRotations(EdgeIndex, 2), NewMeshOption2)
                    Options.AddOptionForDirection(AddRotations(EdgeIndex, 3), NewMeshOption3)
