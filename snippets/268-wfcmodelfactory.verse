# Knowledge-neutral Verse snippet; project names and asset IDs generalized.
# id: snippet-268

# Factory class for creating a wfc_model
wfc_model_factory<public> := class:


    # Create a WFC model for our 'park'
    GetParkModel<public>(PT:piece_type_dir):wfc_model =

        var Model:wfc_model := wfc_model{}

        # Define edge labels

        GrassEL:wfc_edge_label := wfc_edge_label:
            Name:="Grass"
            Symmetric := true

        WaterEL:wfc_edge_label := wfc_edge_label:
            Name:="Water"
            Symmetric := true
            
        GrassToWaterEL:wfc_edge_label := wfc_edge_label:
            Name:="SandToWater"
            Symmetric := false

        WaterToGrassEL:wfc_edge_label := wfc_edge_label:
            Name:="WaterToSand"
            Symmetric := false

        # Indicate that these two edges are mirrors of each other
        Model.AddLabelFlipPair(GrassToWaterEL, WaterToGrassEL)

        # Define tiles, including props to use, and label on each edge (clockwise from +Y)

        Park_Grass:wfc_mesh := wfc_mesh:
            Props := array:
                PT.Park_Grass
            Name := "Park_Grass"
            AllowRotation := false
            RandomPropRotation := true
            Edges := array:
                GrassEL
                GrassEL
                GrassEL
                GrassEL
        
        Park_Water:wfc_mesh := wfc_mesh:
            Props := array:
                PT.Park_Water
            Name := "Park_Water"
            AllowRotation := false
            RandomPropRotation := true
            Edges := array:
                WaterEL
                WaterEL
                WaterEL
                WaterEL

        Park_Grass_Water_InCorn:wfc_mesh := wfc_mesh:
            Props := array:
                PT.Park_Grass_Water_Incorn
            Name := "Park_Grass_Water_Incorn"
            Edges := array:
                WaterEL
                WaterEL
                WaterToGrassEL
                GrassToWaterEL

        Park_Grass_Water_OutCorn:wfc_mesh := wfc_mesh:
            Props := array:
                PT.Park_Grass_Water_Outcorn
            Name := "Park_Grass_Water_OutCorn"
            Edges := array:
                GrassToWaterEL
                WaterToGrassEL
                GrassEL
                GrassEL

        Park_Grass_Water_Straight:wfc_mesh := wfc_mesh:
            Props := array:
                PT.Park_Grass_Water_Straight
            Name := "Park_Grass_Water_Straight"
            Edges := array:
                WaterEL
                WaterToGrassEL
                GrassEL
                GrassToWaterEL

        # Add tiles to model

        Model.AddMesh(Park_Grass)
        Model.AddMesh(Park_Water)
        Model.AddMesh(Park_Grass_Water_InCorn)
        Model.AddMesh(Park_Grass_Water_OutCorn)
        Model.AddMesh(Park_Grass_Water_Straight)

        set Model.FallbackPropOpt = option{PT.Park_Grass}

        # No restrictions on border, but this is how it would be done
        #BorderLabels := array{GrassEL}
        #BorderTile := CreateTileFromEdgeLabels(BorderLabels, Model)

        return Model
