# Knowledge-neutral Verse snippet; project names and asset IDs generalized.
# id: snippet-269

using { Lib }
using { /Verse.org/Simulation }
using { /Verse.org/Random }
using { /UnrealEngine.com/Temporary/SpatialMath }
using { /UnrealEngine.com/Temporary/Diagnostics }

# Base Tile Struct which holds an array of remaining Options and its Shannon Entropy value
wfc_tile<public> := class:
    # Indication of how many options remain for this tile
    var ShannonEntropy:float = 0.0
    # Set of options that this tile could be
    var RemainingOptions<public>:[wfc_mesh_option]int = map{}
    # Size of RemainingOptions set
    var NumRemainingOptions<public>:int = 0
    # Mode that we are using for this tile
    var Model<public>:wfc_model
    # If tile could be any option (optimization)
    var AllOptions:logic = false

    UpdateShannonEntropy()<transacts>:void=
        # TODO: ShannonEntropy calc is more complex if we add weights for options
        set ShannonEntropy = NumRemainingOptions * 1.0

    IsCollapsed()<transacts><decides>:void=
        NumRemainingOptions <= 1

    IsFailed()<transacts><decides>:void=
        NumRemainingOptions = 0

    GetFirstOption<public>()<transacts><decides>:wfc_mesh_option=
        var Result:?wfc_mesh_option = false
        for(O -> I : RemainingOptions, not Result?):
            set Result = option{O}
        return Result?

    Reset():void=
        set AllOptions = true
        set RemainingOptions = map{}
        set NumRemainingOptions = Model.NumAllMeshOptions

        UpdateShannonEntropy() 


# A helper struct used for queuing during Observation and Propagation phases
wfc_queue_entry := struct<computes>:
    AffectedPos<public>:vector3i
    InstigatorTile<public>:?wfc_tile
    Direction<public>:int # Pointing FROM AffectedTile TO InstigatorTile

#ToString(Entry:wfc_queue_entry)<computes>:string=
#    "({Entry.AffectedTilePos}), ({Entry.InstigatorTilePos}), {Entry.Direction}"

MakeQueueEntry(InAffectedPos:vector3i, InInstigatorTile:wfc_tile, InDirection:int)<computes>:wfc_queue_entry=
    wfc_queue_entry:
        AffectedPos := InAffectedPos
        InstigatorTile := option{InInstigatorTile}
        Direction := InDirection

# System that runs Wave Function Collapse process to generate set of tiles
wfc_system<public> := class:
    Logger:log = log{Channel:=LogClass}

    # 3D array of tiles
    var Tiles:[]?wfc_tile = array{}

    # Dimensions of grid
    var GridSize:vector3i = vector3i{X:= 0, Y:= 0, Z:=0}

    # Array of remaining Tiles
    var RemainingTiles:[]vector3i = array{}

    # Pending propagation operations
    var PropagationQueue:[]wfc_queue_entry = array{}
  
    # Build system that owns this WFC system
    var BuildSystemOpt:?build_system = false

    # Convert 3D coordinate into 1D array index
    GetGridIndex<public>(X:int, Y:int, Z:int)<transacts>:int=
        # TODO PERF Remove before shipping
        #if(X < 0 or X >= GridSize.X or Y < 0 or Y >= GridSize.Y or Z < 0 or Z >= GridSize.Z):
        #    Logger.Print("GetGridIndex: Out of range!")

        return (X * (GridSize.Y*GridSize.Z)) + (Y * GridSize.Z) + Z

    # Convert 3D coordinate into 1D array index
    GetGridIndex<public>(Voxel:vector3i)<transacts>:int=
        # TODO PERF Remove before shipping
        #if(Voxel.X < 0 or Voxel.X >= GridSize.X or Voxel.Y < 0 or Voxel.Y >= GridSize.Y or Voxel.Z < 0 or Voxel.Z >= GridSize.Z):
        #    Logger.Print("GetGridIndex: Out of range!")

        return (Voxel.X * (GridSize.Y*GridSize.Z)) + (Voxel.Y * GridSize.Z) + Voxel.Z


    GetTile<public>(X:int, Y:int, Z:int)<decides><transacts>:wfc_tile=
        Tiles[GetGridIndex(X,Y,Z)]?

    GetTile<public>(Pos:vector3i)<decides><transacts>:wfc_tile=
        Tiles[GetGridIndex(Pos.X,Pos.Y,Pos.Z)]?




    # Initialize the WFC system, with a specifc model
    Initialize<public>(BuildSystem:build_system):void=
        set BuildSystemOpt = option{BuildSystem}
        set GridSize = BuildSystem.VoxelGrid.Size

        # Init grid of tiles to false
        ResetWFC()

    # Reset state of entire WFC grid
    ResetWFC<public>():void=
        set Tiles = for(X := 0 .. GridSize.X-1; Y := 0 .. GridSize.Y-1; Z := 0 .. GridSize.Z-1):
            false

    # Remove a tile from the WFC system, and regenerate affected tiles (and nearby, if RecalcRadius > 0)
    RemoveTile<public>(Voxel:vector3i, RecalcRadius:int):void=
        if(set Tiles[GetGridIndex(Voxel)] = false) {}

        var Region:voxel_box = MakeXYBoxFromCentreAndRadius(Voxel, RecalcRadius)

        ResetRegion(Region)
        Regenerate()
        
    # Add a tile to the WFC system, and regenerate that tile (and nearby, if RecalcRadius > 0)
    AddTile<public>(Voxel:vector3i, Model:wfc_model, RecalcRadius:int):void=
        var NewTile:wfc_tile = wfc_tile:
            Model := Model

        if(set Tiles[GetGridIndex(Voxel)] = option{NewTile}) {}
        # We leave this tile empty as ResetRegion below will init to all options

        var Region:voxel_box = MakeXYBoxFromCentreAndRadius(Voxel, RecalcRadius)

        ResetRegion(Region)
        Regenerate()

    # Add a box region to the WFC system, and regenerate tiles (and nearby, if RecalcRadius > 0)
    AddRegion<public>(Region:voxel_box, Model:wfc_model, RecalcRadius:int):void=
        for:
            X := Region.Origin.X .. Region.EndX()
            Y := Region.Origin.Y .. Region.EndY()            
            Z := Region.Origin.Z .. Region.EndZ()  
        do:
            Voxel := MakeVector3i(X,Y,Z)
            NewTile := wfc_tile{ Model := Model }
            if(set Tiles[GetGridIndex(Voxel)] = option{NewTile}) {}

        RegionToReset := ExpandBoxXYSafely(Region, RecalcRadius)
        ResetRegion(RegionToReset)
        Regenerate()

    # Reset tile to all options
    ResetTile(Pos:vector3i):void=
        if:
            Tile := GetTile[Pos]
        then:
            Tile.Reset()

    # Indicate that voxel has changed, as it may require regenerating nearby tiles
    NotifyChangedVoxel<public>(Voxel:vector3i, Category:build_category, RecalcRadius:int):void=
        # If voxel changed is 'horizontally solid'
        var HorizontallySolid:logic = GetSolidityOfCategory(Category)(0)
        if(HorizontallySolid?):
            var Region:voxel_box = MakeXYBoxFromCentreAndRadius(Voxel, RecalcRadius)

            # If not the top most voxel, update region above as well
            if(Voxel.Z < GridSize.Z-1):
                set Region.Size.Z = 2

            ResetRegion(Region)
            Regenerate()
    
    # Iteratively collapse a tile and propagate results until RemainingTiles is empty
    Regenerate():void=
        #Print("---------------- Regenerate")
        # Keep observing and propagating until one of them fails
        loop:
            ObserveSuccess := Observe()
            if(not ObserveSuccess?):
                break

            PropagateSuccess := Propagate()
            if(not PropagateSuccess?):
                break

    # Get indices into Remaining Tiles array with the lowest entropy
    GetRemainingTilesWithMinEntropy():[]int=
        var MinEntropyTiles:[]int = array{}
        var MinEntropy:float = 999999999.0
        for(RemainingTileIdx -> TilePos : RemainingTiles):
            if:
                Tile := Tiles[GetGridIndex(TilePos)]?
                Entropy := Tile.ShannonEntropy
            then:
                # New low entropy, clear array and add this
                if(Entropy < MinEntropy - 0.0001):
                    set MinEntropyTiles = array{RemainingTileIdx}
                    set MinEntropy = Entropy
                # Almost the same as current low entropy, add to array
                else if(Entropy < MinEntropy + 0.0001):
                    set MinEntropyTiles += array{RemainingTileIdx}

        return MinEntropyTiles

    # Pick next tile to collapse
    Observe():logic=
        #Print("Observe options:")
        set PropagationQueue = array{} # Empty array

        # Get tile(s) with lowest entropy
        MinEntropyRemainingTileIndices := GetRemainingTilesWithMinEntropy()        
        #for(I : MinEntropyRemainingTileIndices):
            #Print("  {IndexAsPosition(I)}")

        if:
            SelectedRemainingTileIndex := MinEntropyRemainingTileIndices[GetRandomInt(0, MinEntropyRemainingTileIndices.Length-1)]
            SelectedTilePos := RemainingTiles[SelectedRemainingTileIndex]
            SelectedTile := Tiles[GetGridIndex(SelectedTilePos)]?
        then:
            if(SelectedTile.IsCollapsed[]):
                Print("Observe: Found collapsed tile!")          

            # Select the option for this tile ie collapse it

            # If collapsing a tile that hasn't been constrained (still has all options)
            if(SelectedTile.AllOptions?):
               set SelectedTile.RemainingOptions = SelectedTile.Model.AllMeshOptions
               set SelectedTile.AllOptions = false        

            var OptionsArray:[]wfc_mesh_option = for(O -> I : SelectedTile.RemainingOptions):
                O

            SelectedOptionIndex := GetRandomInt(0, OptionsArray.Length-1)
            if(SelectedOption := OptionsArray[SelectedOptionIndex]):
                #Print("Picking Random {SelectedTilePos}")               
                #Print("   Collapse to: {SelectedOption}")
                set SelectedTile.RemainingOptions = map{ SelectedOption => 0 }
                set SelectedTile.NumRemainingOptions = 1
                SelectedTile.UpdateShannonEntropy()                


            # Remove entry from RemainingTiles
            if(TmpRemainingTiles := RemainingTiles.RemoveElement[SelectedRemainingTileIndex]):
                set RemainingTiles = TmpRemainingTiles

            # Add Adjacent Tile Indices to Queue
            AddAdjacentIndicesToQueue(SelectedTilePos, SelectedTile)

            # Continue To Propagation
            return true

        else:
            #Print("  Nothing to observe")

            # Do Not Continue to Propagation
            return false

        
    # Util to create a box from a centre and radius
    MakeXYBoxFromCentreAndRadius(Centre:vector3i, Radius:int):voxel_box=            
        MinX := Max(Centre.X - Radius, 0)
        MaxX := Min(Centre.X + Radius, GridSize.X-1)
        MinY := Max(Centre.Y - Radius, 0)
        MaxY := Min(Centre.Y + Radius, GridSize.Y-1)
        return MakeVoxelBox(MinX,MinY,Centre.Z,  MaxX-MinX+1,MaxY-MinY+1,1)

    ExpandBoxXYSafely(Box:voxel_box, ExpandBy:int):voxel_box=
        MinX := Max(Box.Origin.X - ExpandBy, 0)
        MaxX := Min(Box.EndX() + ExpandBy, GridSize.X-1)
        MinY := Max(Box.Origin.Y - ExpandBy, 0)
        MaxY := Min(Box.EndY() + ExpandBy, GridSize.Y-1)
        return MakeVoxelBox(MinX,MinY,Box.Origin.Z,  MaxX-MinX+1,MaxY-MinY+1,1)
  


    # Reset tiles in this region to uncollapsed. Will propagate any constraints from outside this region.
    ResetRegion(Region:voxel_box):void=

        #Print("ResetRegion {Region}")

        MinX := Region.Origin.X
        MaxX := Region.EndX()
        MinY := Region.Origin.Y
        MaxY := Region.EndY()

        for:
            X := MinX .. MaxX
            Y := MinY .. MaxY
            Z := Region.Origin.Z .. Region.EndZ()
        do:
            Pos := MakeVector3i(X,Y,Z)
            # See if we already have a tile for this location
            if:
                Tile := GetTile[Pos]
            then:
                ResetTile(Pos) # Reset to all options

                set RemainingTiles += array{Pos}

                # Add entries to propagation queue for any tiles on region border, or any borders

                if(Y < GridSize.Y-1):
                    CheckPos := Pos + MakeVector3i(0,1,0)
                    if(InstigatorTile := GetTile[CheckPos]):                        
                        if(Y = MaxY): # If on the border of region, and we neighbour a tile outside the region, add a ref to queue
                            set PropagationQueue += array{ MakeQueueEntry(Pos, InstigatorTile, 0) }  # Direction points up, from affected to instigator
                    else if(BorderTile := Tile.Model.BorderTile?): # No tile, so use border tile                     
                        set PropagationQueue += array{ MakeQueueEntry(Pos, BorderTile, 0) }  
                else: # Outside region, use border tile
                    if(BorderTile := Tile.Model.BorderTile?):
                        set PropagationQueue += array{ MakeQueueEntry(Pos, BorderTile, 0) }

                if(X > 0):
                    CheckPos := Pos + MakeVector3i(-1,0,0)
                    if(InstigatorTile := GetTile[CheckPos]):
                        if(X = MinX):
                            set PropagationQueue += array{ MakeQueueEntry(Pos, InstigatorTile, 1) }
                    else if(BorderTile := Tile.Model.BorderTile?):
                        set PropagationQueue += array{ MakeQueueEntry(Pos, BorderTile, 1) }  
                else:
                    if(BorderTile := Tile.Model.BorderTile?): 
                        set PropagationQueue += array{ MakeQueueEntry(Pos, BorderTile, 1) }

                if(Y > 0):
                    CheckPos := Pos + MakeVector3i(0,-1,0)
                    if(InstigatorTile := GetTile[CheckPos]):
                        if(Y = MinY):
                            set PropagationQueue += array{ MakeQueueEntry(Pos, InstigatorTile, 2) }
                    else if(BorderTile := Tile.Model.BorderTile?):
                        set PropagationQueue += array{ MakeQueueEntry(Pos, BorderTile, 2) }  
                else:
                    if(BorderTile := Tile.Model.BorderTile?): 
                        set PropagationQueue += array{ MakeQueueEntry(Pos, BorderTile, 2) }
            
                if(X < GridSize.X-1):
                    CheckPos := Pos + MakeVector3i(1,0,0)
                    if(InstigatorTile := GetTile[CheckPos]):
                        if(X = MaxX):
                            set PropagationQueue += array{ MakeQueueEntry(Pos, InstigatorTile, 3) }
                    else if(BorderTile := Tile.Model.BorderTile?):
                        set PropagationQueue += array{ MakeQueueEntry(Pos, BorderTile, 3) }  
                else:
                    if(BorderTile := Tile.Model.BorderTile?): 
                        set PropagationQueue += array{ MakeQueueEntry(Pos, BorderTile, 3) }

        Propagate()
            
    # Add constraints from this tile to neighbours
    AddAdjacentIndicesToQueue(InstigatorPos:vector3i, InstigatorTile:wfc_tile):void=
        
        if:
            InstigatorPos.Y > 0 # check not at edge of grid
            AffectedPos := InstigatorPos + MakeVector3i(0,-1,0)
            AffectedTile := GetTile[AffectedPos] # only add if we have a tile
            not AffectedTile.IsCollapsed[] # skip if tile is already collapsed (ie already decided)
        then:
            set PropagationQueue += array{ MakeQueueEntry(AffectedPos, InstigatorTile, 0) }  # Direction points up, from affected to instigator


        if:
            InstigatorPos.X < GridSize.X-1
            AffectedPos := InstigatorPos + MakeVector3i(1,0,0)
            AffectedTile := GetTile[AffectedPos]
            not AffectedTile.IsCollapsed[]
        then:
            set PropagationQueue += array{ MakeQueueEntry(AffectedPos, InstigatorTile, 1) } 


        if:
            InstigatorPos.Y < GridSize.Y-1
            AffectedPos := InstigatorPos + MakeVector3i(0,1,0)
            AffectedTile := GetTile[AffectedPos]
            not AffectedTile.IsCollapsed[]
        then:
            set PropagationQueue += array{ MakeQueueEntry(AffectedPos, InstigatorTile, 2) } 


        if:
            InstigatorPos.X > 0
            AffectedPos := InstigatorPos + MakeVector3i(-1,0,0)
            AffectedTile := GetTile[AffectedPos]
            not AffectedTile.IsCollapsed[]
        then:
            set PropagationQueue += array{ MakeQueueEntry(AffectedPos, InstigatorTile, 3) } 


    # Reduce the options for a tile given a set of options from a particular direction
    ConstrainTileByEdgeLabels(Tile:wfc_tile, Direction:int, Labels:[]wfc_edge_label):logic=
               
        #Print("  MeshEdges:")
        #for(ME : MeshEdges):
        #    Print("   {ME}")

        # All options this tile could now be, based on MeshEdges
        var OptionsToCheckAgainst:[wfc_mesh_option]int = map{}

        for(Label : Labels):
            # We always have to match flipped to unflipped. This has no effect if label is marked as symmetric.
            DesiredLabel := Tile.Model.FlipEdgeLabel(Label)

            # Get all options that have that label on desired edge
            EdgeOptions := Tile.Model.GetOptionsForEdgeLabel(DesiredLabel)
            OptionsSet := EdgeOptions.GetOptionsForDirection(Direction)

            # Add all options to set (this handles duplicates)
            set OptionsToCheckAgainst = ConcatenateMaps(OptionsToCheckAgainst, OptionsSet.Map)

        var OptionsChanged:logic = false

        # If this is the first time the tile has been constrained, we can just copy the constrained set over
        if(Tile.AllOptions?):
            set Tile.RemainingOptions = OptionsToCheckAgainst
            set Tile.AllOptions = false
            set OptionsChanged = true
        # Otherwise we filter current set by new allowed set
        else:
            var NewRemainingOptions:[wfc_mesh_option]int = map{}

            # For each option the tile currently has..
            for(Option -> I : Tile.RemainingOptions):
                # See if it is still allowed..
                if(OptionsToCheckAgainst[Option]):
                    # If so, add to new set (shouldn't need to check for uniqueness, because this is just a subset of RemainingOptions)
                    if( set NewRemainingOptions[Option] = 0 ){}
                else:
                    # We removed an option, so we need to update neighbours
                    set OptionsChanged = true

            if(OptionsChanged?):
                NewNumOptions := NewRemainingOptions.Length
                set Tile.RemainingOptions = NewRemainingOptions
                set Tile.NumRemainingOptions = NewNumOptions
                Tile.UpdateShannonEntropy()

            #Print("  New Options:")
            #    for(Opt : Tile.RemainingOptions):
            #        Print("   {Opt}")
        #else:
        #    Print("  No Change")

        return OptionsChanged


    # Process any pending constraint propagation
    Propagate():logic=

        #Print("--- Propagate")
        # Keep iterating as long as there are still tiles that need to be updated
        var QueueIndex:int = 0
        loop:

            #Print("QUEUE:")
            #for(I := QueueIndex .. PropagationQueue.Length-1):
            #    if(E := PropagationQueue[I]):
            #        Print(" {I}: {E}")

            if(Entry := PropagationQueue[QueueIndex]):
                if:
                    AffectedTile := GetTile[Entry.AffectedPos] # Get the 'affected' tile
                    not AffectedTile.IsCollapsed[] # Check it's not already collapsed 
                    InstigatorTile := Entry.InstigatorTile? # get the 'instigator' tile
                    not InstigatorTile.IsFailed[] # Do nothing if instigator failed (overconstrained)
                then:
                    var UseInstigatorTile:wfc_tile = InstigatorTile

                    #Print("P: {Entry.AffectedPos} dir {Entry.Direction}")

                    # If models don't match, use flat boundary
                    if:
                        AffectedTile.Model <> InstigatorTile.Model
                        OtherBorder := AffectedTile.Model.BorderTile?
                    then:
                        set UseInstigatorTile = OtherBorder

                    # All options this tile could now be, based on things instigator could be, and our adjecency to it
                    var OptionsToCheckAgainst:[wfc_mesh_option]int = map{}

                    # Instigator should never be unconstrained
                    if(UseInstigatorTile.AllOptions?):
                        Print("Instigator with AllOption=true!")

                    # Find new (more restricted) options that affected tile could have, based on adjacency to instigator
                    var EdgeLabelMap:[wfc_edge_label]int = map{}

                    # For each option the 'instigator' has...
                    FlippedDir := FlipRotation(Entry.Direction)
                    for(InstigatorOption -> I : UseInstigatorTile.RemainingOptions):

                        # Get edge label that we want to match. This is the opposite of the side we want it to be on this (affected) tile
                        EdgeLabel := InstigatorOption.GetMeshEdgeLabel(FlippedDir)

                        if( set EdgeLabelMap[EdgeLabel] = 0 ) {}

                    # Convert map to array
                    EdgeLabels:[]wfc_edge_label := for(L -> I : EdgeLabelMap):
                        L

                    # Reduce options for tile given this set of edges
                    AddToPropagationQueue := ConstrainTileByEdgeLabels(AffectedTile, Entry.Direction, EdgeLabels)

                    # If we constrained options for this tile, need to update neighbours
                    if(AddToPropagationQueue?):
                        AddAdjacentIndicesToQueue(Entry.AffectedPos, AffectedTile)
                       
                        # If we have fully collapsed 
                        if(AffectedTile.NumRemainingOptions <= 1):
                            if(AffectedTile.NumRemainingOptions = 0):
                                # TODO Should return false and try again...?
                                Print("Propagate: Overconstrained!")
                            #else:
                            #    Print("Propagate: Collapsed!")


                            # Remove entry from RemainingTiles
                            set RemainingTiles = for(T : RemainingTiles, not T.Equals[Entry.AffectedPos]):
                                T
                           

                set QueueIndex += 1
            else:
                break

        return true
