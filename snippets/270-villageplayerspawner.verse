# Knowledge-neutral Verse snippet; project names and asset IDs generalized.
# id: snippet-270

# This device manages player spawning and assigns players to a specific home.
# In its Details panel in the editor, you can:
# - Add player spawn info classes to the PlayerSpawnList Array
# - Set the information for each player spawn area
# - This includes the spawn pad, camera device, and counter being used


using { /Fortnite.com/Devices }
using { /Verse.org/Simulation }
using { /UnrealEngine.com/Temporary/Diagnostics }
using { /Fortnite.com/Game }
using { /Fortnite.com/FortPlayerUtilities }
using { /Fortnite.com/Characters }


PlayerSpawnerTip<localizes>:message := "This is the player spawner being used in the Homestead."
IntroCameraTip<localizes>:message := "This is the intro camera (pointed at the player) when the player spawns for the first time."
HomeCounterTip<localizes>:message := "This is a Player Counter used to check if a player is in a Homestead at game start."
SpawnListTip<localizes>:message := "This is a collection of Player Spawn Information. Each one represents a separate Homestead"
DialogTip<localizes>:message := "The initial dialog that pops up at game start"
IntroSequenceTip<localizes>:message := "This is the fade from black during the intro."
StasisControlModeTip<localizes>:message := "This is used to hold the player in position during the intro."




# This class handles collecting the information used to assign a player to a homestead
initial_player_spawn_info := class<concrete>():
    # The initial player spawner used when a player starts a game
    @editable {ToolTip := PlayerSpawnerTip}
    PlayerSpawner:player_spawner_device = player_spawner_device{}
    
    # The initial camera tied to the player spawn pad, this is the camera that points at the player during the intro,
    @editable {ToolTip := IntroCameraTip}
    PlayerIntroCamera:gameplay_camera_fixed_point_device = gameplay_camera_fixed_point_device{}


    # This counter is used as a failsafe in case the we miss the PlayerSpawned Event
    @editable {ToolTip := HomeCounterTip}
    HomeCounter:player_counter_device = player_counter_device{}
   


# This is being used to handle players "Owning" a particular spawn pad.
# This assigns a Homestead to a player, and handles their intro camera and initial pop up dialog.
village_player_spawner := class(creative_device):


    # This tracks each spawn set up for each home.
    @editable {ToolTip := SpawnListTip}
    PlayerSpawnList:[]initial_player_spawn_info = array{}


 
    # This handles the intro fade from black.
    @editable {ToolTip := IntroSequenceTip}
    StartSequence:cinematic_sequence_device = cinematic_sequence_device{}
    
    # The intro dialog “Welcome to Village…”.
    @editable {ToolTip := DialogTip}
    FirstDialog:popup_dialog_device = popup_dialog_device{}
    
    # An extra failsafe method to put the player in stasis during the intro
    @editable {ToolTip := StasisControlModeTip}
    ControlModeStasis:gameplay_controls_third_person_device = gameplay_controls_third_person_device{}


    #Used to Fighter the player to a spawner
    var PlayerToSpawner:[agent]player_spawner_device = map{}


    # On begin turn on the events to listen for players spawning in and listen for players leaving the game.
    OnBegin<override>()<suspends>:void=


        InitializeListener()
        GetPlayspace().PlayerRemovedEvent().Subscribe(ReactivateSpawner)
        #Backup in case PlayerSpawned didn't get called.
        for (PlayerSpawnInfo : PlayerSpawnList):
            PlayerSpawnInfo.HomeCounter.TransmitForAllCounted()


    #When a player is spawned, disable the spawner and add the player to the map so we can enable that spawner later.
    PlayerSpawned(Agent:agent,Spawner:player_spawner_device,Camera:gameplay_camera_fixed_point_device)<suspends>:void=
       
        if(FortCharacter := Agent.GetFortCharacter[]):
            #Check if the player already is in a map of player spawners
            if (PlayerToSpawner[Agent]):
                return
            if (set PlayerToSpawner[Agent] = Spawner):
                FortCharacter.PutInStasis(stasis_args{AllowFalling := true})            
                StartSequence.Play(Agent)
                Spawner.Disable()
                Camera.RemoveFrom(Agent)
                Camera.AddTo(Agent)
                race:
                    block:
                        StartSequence.StoppedEvent.Await()
                    block:
                        #Timeout in case the Sequence fails for some reason.
                        Sleep(6.0)
                        StartSequence.GoToEndAndStop(Agent)
                FirstDialog.Show(Agent)
                FortCharacter.ReleaseFromStasis()
                ControlModeStasis.RemoveFrom(Agent)


    InitializeListener():void=
        for (PlayerSpawnInfo:PlayerSpawnList):
            PlayerSpawnListener:player_spawned_listener = player_spawned_listener{PlayerSpawnerInfo:=PlayerSpawnInfo,VillagePlayerSpawnerRef:=Self}
                spawn{PlayerSpawnedListenerEvent(PlayerSpawnListener,PlayerSpawnInfo.PlayerSpawner)}
                spawn{PlayerCountedListenerEvent(PlayerSpawnListener,PlayerSpawnInfo.HomeCounter)}

    #Listen for the PlayerSpawned event.  Await() is used here because the SpawnPad asset that called the event is needed later to shut it off.
    PlayerSpawnedListenerEvent(ClassListener:player_spawned_listener,PlayerSpawnPad:player_spawner_device)<suspends>:void=
        loop:
            Agent := PlayerSpawnPad.SpawnedEvent.Await()
            ClassListener.PlayerSpawned(Agent)
    PlayerCountedListenerEvent(ClassListener:player_spawned_listener,Counter:player_counter_device)<suspends>:void=
        loop:
            Agent := Counter.CountedEvent.Await()
            ClassListener.PlayerSpawned(Agent)




    #Turn That Spawn Pad back on once a player leaves the match, so it is available for the next player to join
    ReactivateSpawner(Player:player):void=      
        if:
            Agent := agent[Player]
            SpawnerToReactivate := PlayerToSpawner[Agent]
        then: 
            SpawnerToReactivate.Enable()
            RemoveAgentFromMap(Agent)
   
    RemoveAgentFromMap(Agent:agent):void=
        var NewMap:[agent]player_spawner_device = map{}
        for (Key -> Value : PlayerToSpawner, Key <> Agent):
            set NewMap = ConcatenateMaps(NewMap, map{Key => Value})
        set PlayerToSpawner = NewMap


player_spawned_listener := class:


    PlayerSpawnerInfo:initial_player_spawn_info
    VillagePlayerSpawnerRef:village_player_spawner
       
    PlayerSpawned(Agent:agent)<suspends>:void=
        VillagePlayerSpawnerRef.PlayerSpawned(Agent,PlayerSpawnerInfo.PlayerSpawner,PlayerSpawnerInfo.PlayerIntroCamera)
