# Knowledge-neutral Verse snippet; project names and asset IDs generalized.
# id: snippet-271

# This device manages the time in village, allowing designers to hook events up to specific hours.
# In its Details panel in the editor, you can:
# - Set the time when it becomes “daytime” in village
# - Set the time when it becomes “nighttime” in village


using { /Fortnite.com/Devices }
using { /Verse.org/Simulation }
using { /UnrealEngine.com/Temporary/Diagnostics }




MessageMakerTimeHour<localizes>(HourToSet:int):message = "{HourToSet}:00"


HudDayTip<localizes>:message := "This determines the time HUD displayed during the day."
HudNightTip<localizes>:message := "This determines the time HUD displayed during the night."
TimerTip<localizes>:message := "This determines the length of each hour of the day."
DayNightSwitchTip<localizes>:message := "This track if it is day or night, and sends an event when the switch has happened"
MorningHourTip<localizes>:message := "This determines the time of day should shift to day."
EveningHourTip<localizes>:message := "This determines the time of day should shift to night."
DayLengthTip<localizes>:message := "This determines how long a day is in hours."

village_time_manager := class(creative_device):


    # Displays the UI for day
    @editable {ToolTip := HudDayTip}
    TimerHudDeviceDay:hud_message_device = hud_message_device{}


    # Displays the UI for night
    @editable {ToolTip := HudNightTip}
    TimerHudDeviceNight:hud_message_device = hud_message_device{}


    # Tracks each hour of the day
    @editable {ToolTip := TimerTip}
    TimerDay:timer_device = timer_device{}
    
    # Used to track if it is day or night in game, and sends an event when the switch has happened
    @editable {ToolTip := DayNightSwitchTip}
    DayNightSwitch:switch_device = switch_device{}


    var CurrentHour:int = 6
    # This can be changed to set when it is considered "morning" in village.
    @editable {ToolTip := MorningHourTip}
    var MorningHour:int = 6
    # This can be changed to set when it is considered "night" in village.
    @editable {ToolTip := EveningHourTip}
    var EveningHour:int = 20
    # This can be changed to set the length of a day in village.
    @editable {ToolTip := DayLengthTip}
    var HoursInDay:int = 24
	


    #An event other devices can await if they want to listen for a specific time. The int payload represents the current hour.
    HourIncrementEvent:event(int) = event(int){}


    # Runs when the device is started in a running game
    OnBegin<override>()<suspends>:void=
        TimerDay.SuccessEvent.Subscribe(IncrementHour)
        DayNightSwitch.TurnedOffEvent.Subscribe(DayForcedOff)
        DayNightSwitch.TurnedOnEvent.Subscribe(DayForcedOn)


    IncrementHour(Agent:?agent):void=
        GenericAgents := GetPlayspace().GetPlayers()
        #The switch device needs an agent to turn the switch on and off.
        if(GenericAgent := GenericAgents[0]):
            set CurrentHour+=1

            if (CurrentHour = MorningHour):
                DayNightSwitch.TurnOn(GenericAgent)
            if (CurrentHour = EveningHour):
                DayNightSwitch.TurnOff(GenericAgent)
            if (CurrentHour = HoursInDay):
                set CurrentHour = 0

            HourIncrementEvent.Signal(CurrentHour)
            SetHourText(CurrentHour)


    #Update the messaging on the HUD
    SetHourText(Hour:int):void=
        if (IsDay[]):
            TimerHudDeviceDay.SetText(MessageMakerTimeHour(Hour))
        else:
            TimerHudDeviceNight.SetText(MessageMakerTimeHour(Hour))


    #Making sure the current hour tracks properly if the Core switch is turned on/off        
    DayForcedOff(Agent:agent):void=
        set CurrentHour = EveningHour   
        TimerHudDeviceNight.SetText(MessageMakerTimeHour(CurrentHour))
        TimerHudDeviceNight.Show()
        TimerHudDeviceDay.Hide()
       
    DayForcedOn(Agent:agent):void=
        set CurrentHour = MorningHour 
        TimerHudDeviceDay.SetText(MessageMakerTimeHour(CurrentHour))
        TimerHudDeviceDay.Show()
        TimerHudDeviceNight.Hide()


    IsDay()<decides><transacts>:void=
        CurrentHour >= MorningHour or CurrentHour < EveningHour
