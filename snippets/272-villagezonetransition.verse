# Knowledge-neutral Verse snippet; project names and asset IDs generalized.
# id: snippet-272

# This device manages transitions from the exterior map to interior spaces and back again, as well as teleports within the mines.
# The transitions are broken up between button and trigger classes for easier tracking and management in the editor.


using { /Fortnite.com/Characters }
using { /Fortnite.com/Devices }
using { /Fortnite.com/Playspaces }
using { /Verse.org/Simulation }
using { /UnrealEngine.com/Temporary/Diagnostics }
using { /UnrealEngine.com/Temporary/SpatialMath }


TeleportDestinationTip<localizes>:message := "This determines the location the player is teleported to."
TeleportButtonTip<localizes>:message := "This determines the button used to start teleporting the instigating player"
TeleportTriggerTip<localizes>:message := "This determines volume used to start teleporting the instigating player"
TeleportButtonListTip<localizes>:message := "This contains all the buttons and their destinations for teleporting"
TeleportVolumeListTip<localizes>:message := "This contains all the volumes and their destinations for teleporting"


# Transitions that are initiated with a button device
zone_transition_buttons := class<concrete>():
    @editable {ToolTip := TeleportDestinationTip}
    TeleportDestination:teleporter_device = teleporter_device{}


    @editable {ToolTip := TeleportButtonTip}
    TransitionButton:button_device = button_device{}


# Transitions that are initiated with a trigger volume
zone_transition_triggers := class<concrete>():
    @editable {ToolTip := TeleportDestinationTip}
    TeleportDestination:teleporter_device = teleporter_device{}


    @editable {ToolTip := TeleportTriggerTip}
    TransitionTrigger:volume_device = volume_device{}


# Relevant data to send when a player is teleported through a transition.
# The TeleportDestination is the teleporter the player is headed to.
# The TeleportInstigator is the character class of the player being teleported.  This is used to put the player in stasis.
# The TeleportInstigatorAsAgent is the agent that is passed through for use with the teleport function.
zone_transition_payload := class():
    TeleportDestination:teleporter_device
    TeleportInstigator:fort_character
    TeleportInstigatorAsAgent:agent
    ParentSubscription:zone_transition_subscription


zone_transition_subscription := class:
    DoTransitionEvent:event(zone_transition_payload) = event(zone_transition_payload){}
    ZoneTransitionDevice:village_zonetransition
    ZoneDestination:teleporter_device


    var TransitioningPlayerAgents:[]agent = array{}


    TriggerTransition(Agent:agent):void=
        # Validate the agent used here is a player and not an NPC or vehicle
        if (FortCharacter := Agent.GetFortCharacter[]):   
            Payload := zone_transition_payload:
                TeleportDestination := ZoneDestination
                TeleportInstigator := FortCharacter
                TeleportInstigatorAsAgent := Agent
                ParentSubscription := Self
            DoTransitionEvent.Signal(Payload)

# This class handles managing the transitions in the world and doing the actual teleporting of the player.
# To use, add the buttons or volumes along with their respective teleporters to the ZoneButtonList or ZoneTriggerList arrays.
village_zonetransition := class(creative_device):
    @editable {ToolTip := TeleportButtonListTip}
    ZoneButtonList:[]zone_transition_buttons = array{}


    @editable {ToolTip := TeleportVolumeListTip}
    ZoneTriggerList:[]zone_transition_triggers = array{}


    var ZoneTransitionSubscriptions:[]zone_transition_subscription = array{}
    var PlayersInTransition:[]agent = array{}


    OnBegin<override>()<suspends>:void=
        CreateZoneTransitionSubscriptions()
        StartListenToTransitionEvents()


    CreateZoneTransitionSubscriptions():void=
        # Setup the subscription data for button/interact based zone transitions
        for (EachBinding:ZoneButtonList):
            NewSubscription := zone_transition_subscription:
                ZoneTransitionDevice := Self
                ZoneDestination := EachBinding.TeleportDestination
            EachBinding.TransitionButton.InteractedWithEvent.Subscribe(NewSubscription.TriggerTransition)
            NewArray:[]zone_transition_subscription = array{NewSubscription}
            set ZoneTransitionSubscriptions = Concatenate(ZoneTransitionSubscriptions,NewArray)


        # Setup the subscription data for volume event based zone transitions
        for (EachBinding:ZoneTriggerList):
            NewSubscription := zone_transition_subscription:
                ZoneTransitionDevice := Self
                ZoneDestination := EachBinding.TeleportDestination
            EachBinding.TransitionTrigger.AgentEntersEvent.Subscribe(NewSubscription.TriggerTransition)
            NewArray:[]zone_transition_subscription = array{NewSubscription}
            set ZoneTransitionSubscriptions = Concatenate(ZoneTransitionSubscriptions,NewArray)


    StartListenToTransitionEvents():void=
        # For each zone subscription, spawn an event to listen for a player attempting to move between zones.  
        # This is done using spawn events so it can pass along a payload
        for:
            ZoneSubscription : ZoneTransitionSubscriptions
        do:
           spawn{ListenToSubscription(ZoneSubscription)}


    ListenToSubscription(ZoneSubscription:zone_transition_subscription)<suspends>:void=
        loop:
            AwaitedPayload := ZoneSubscription.DoTransitionEvent.Await()
            spawn{TransitionToNewZone(AwaitedPayload)}


    TransitionToNewZone(Payload:zone_transition_payload)<suspends>:void=
        #Make sure we're trying to send a player, and not a chicken that may have wandered into a zone.
        if(AgentIsAPlayer:=player[Payload.TeleportInstigatorAsAgent]):
            Payload.TeleportInstigator.PutInStasis(stasis_args{AllowFalling := true})
            Payload.TeleportDestination.Teleport(Payload.TeleportInstigatorAsAgent)
            Sleep(2.0) #Give some time for the fade to black to play
            Payload.TeleportInstigator.ReleaseFromStasis()
