# Knowledge-neutral Verse snippet; project names and asset IDs generalized.
# id: snippet-273

using { /Fortnite.com/AI }
using { /Verse.org/Simulation }
using { /Verse.org/SceneGraph }
using { /Fortnite.com/Playspaces }
using { /UnrealEngine.com/Conversations }
using { /Verse.org/AgentGroup }
using { /Verse.org/Chat }
using { /Fortnite.com/Devices }
using { /Verse.org/Random }
 
# Getting started: https://www.epicgames.com/fortnite/en-US/creative/docs/uefn/Verse/onboarding-guide-to-programming-with-verse-in-unreal-editor-for-fortnite

# Initializing Voice Channel and Team classes related to Persona
team_member_info := class(has_voice_member_info){}
team_agent_group := class(agent_group(team_member_info)) {}

# Enum we will use to roll for a random Character for their vehicular needs in this game
Character := enum:
    FamilyPerson,
    HeavydutyWorker,
    SinglePerson,
    OutdoorsPerson,
    Collector

FinalOfferResponse := struct {
    @ai_description("Return true whenever you and the player have reached to an agreed price that you are both happy with.")
    FinalOffer:logic = false

    @ai_description("Return an int that represents the amount of money that you and the player have agreed upon.")
    FinalPrice:int = 0
}
KickResponse := struct {
    @ai_description("Return true if you were told to leave.")
    Kicked:logic = false
}

# A Verse-authored NPC Behavior that can be used within an NPC Character Definition or an NPC Spawner device's NPC Behavior Script Override.
persona_behavior := class(npc_behavior):

    # Helper function to make strings into messages at runtime
    MakeMessage<localizes>(String:string):message = "{String}"
    # Optional variables that we use to store the group of players in the conversation and the channel they are part of
    var OptionAgentGroup:?agent_group(team_member_info) = false
    var OptionVoiceChannel:?voice_channel(team_member_info) = false
    # Variables we use to keep track of our Structured Output
    var FinalOfferCancelable:?cancelable = false
    var KickCancelable:?cancelable = false
    # Devices and Verse classes we use to connect gameplay and UI
    @editable
    var Spawner:npc_spawner_device = npc_spawner_device{}
    var UIManager:ui_manager = ui_manager{}
    # -- Gameplay variables --
    # Message we pass as our prompt for our first interaction with the NPC
    GreetingMessage<localizes>:message = "You just entered the office. Say hello to the Car Salesman, and state that you are here to buy a car from them."
    # Message added between messages when using Join() to create our prompts.
    JoinCharacter<localizes>:message = "\n"
    # Helper function to create a message that includes our budget for our prompts. We use the term 'starting offer' instead of 'budget' for the prompt to ensure that the number is mutable when interpreted by the LLM; 'Budget' can be interpreted as a hard ceiling, while 'starting offer' is open to change.
    MakeBudgetPrompt<localizes>(Budget:int):message = "Your starting offer is {Budget} Gold."
    # Gold we remove if the NPC leaves
    var PenaltyScore:int = -200
    # Vars that control what will happen when the NPC stops talking
    # Should the NPC despawn? Could be due to it being told to leave or a final offer has been reached
    var ShouldDespawn:logic = false
    # Should we give the Player Gold equal to the AcceptedOffer?
    var ShouldAwardScore:logic = false
    # Should we apply a penalty to the Player's score?
    var ShouldApplyPenalty:logic = false
    # Var we use to save the returned amount of Gold from the Structured Output
    var AcceptedOffer:int = 0
    # Map that indexes our messages to our Character enum. We will use this to select a random character and assign the proper message to the prompt.
    CharacterPromptMap:[Character]string = map {
        Character.FamilyPerson => "You are a Family Person who needs a vehicle to travel with their family. Generate a backstory and your needs for your vehicle based on this.",
        Character.HeavydutyWorker => "You are a Heavy-duty Worker who uses their vehicle for work. Generate a backstory and your needs for your vehicle based on this.",
        Character.SinglePerson => "You are a Single Person who uses their vehicle mostly for commuting and short-distance traveling. Generate a backstory and your needs for your vehicle based on this.",
        Character.OutdoorsPerson => "You are an Outdoors Person who goes on adventures often. Generate a backstory and your needs for your vehicle based on this.",
        Character.Collector => "You are a Car Collector who is looking for a one-of-a-kind vehicles. Generate a backstory and your needs for your vehicle based on this."
    }
    # Similarly to CharacterPromptMap, we use this to find the proper range based on the Character selected to randomize a budget for the client. These numbers are multiplied by 50 when using GetBudget().
    CharacterBudgetMap:[Character]tuple(int, int) = map {
        Character.FamilyPerson => (37,53),
        Character.HeavydutyWorker => (38,54),
        Character.SinglePerson => (34,50),
        Character.OutdoorsPerson => (36,52),
        Character.Collector => (60,80)
    }

    # This function runs when the NPC is spawned in the world and ready to follow a behavior.
    OnBegin<override>()<suspends>:void=
        if:
            NPCEntity := GetEntity[]
            PersonaComponent := NPCEntity.GetComponent[persona_component]
            Player := NPCEntity.GetPlayspaceForEntity[].GetPlayers()[0]
        then:
            # Find all objects with the Tag 'ManagerTag' and check if they are a 'ui_manager'. If they are, keep a reference to it to send UI events/calls
            UIManagers := Self.FindCreativeObjectsWithTag(ManagerTag)
                for(Manager:UIManagers):
                    if (PossibleManager := ui_manager[Manager]):
                        set UIManager = PossibleManager
            # Randomize our personality and assign it to the persona component's Personality field
            ChooseCharacter()
            # Call our function that Creates the Voice channel, adds all players and NPCs, and enables it as the channel we will use to communicate
            CreateAudioTeamAndAddPlayer(Player)
            # Call and subscribe our two Structured Output bindings
            set FinalOfferCancelable = RegisterFinalOfferAction()
            set KickCancelable = RegisterKickAction()
            # Subscribe 'SetCloseCaptions' to the event 'CommitSayEvent'. This allows the NPC's last spoken message to be displayed on screen
            PersonaComponent.CommitSayEvent.Subscribe(SetCloseCaptions)
            # Subscribe 'ClearCloseCaption' to the event 'StopSayEvent'. This allows us to clean all of our Close Caption variables for the next message
            PersonaComponent.StopSayEvent.Subscribe(ClearCloseCaptions)
            # Subscribe our 'StopSayChecks' function to 'StopSayChecks'. The function allows us to determine NPC reactions based on the responses from Structured Output
            PersonaComponent.StopSayEvent.Subscribe(StopSayChecks)
            # Prompt the NPC to give its initial message to the player when they spawn
            InitialTalk(PersonaComponent)

    # This function runs when the NPC is despawned or eliminated from the world.
    OnEnd<override>():void= 
        # It is good practice to clean up your cancelable events OnEnd(), so we are doing that here
        if(OfferEvent := FinalOfferCancelable?):
            OfferEvent.Cancel()
        if(KickEvent := KickCancelable?):
            KickEvent.Cancel()

    # Function that allows us to create the channel where the Player will converse with the NPC
    CreateAudioTeamAndAddPlayer(Player:player):void =
        if:
            NPCEntity := GetEntity[]
            NPCAgent := GetAgent[]
            SimEntity := NPCEntity.GetSimulationEntity[]
            PersonaComponent := NPCEntity.GetComponent[persona_component]
        then:
            # Create an 'agent_group' with our new agent_group class that uses team_member_info
            TeamGroup:agent_group(team_member_info) = agent_group(team_member_info){}
            # Do the same with a voice channel. We will use this channel for any future methods and function in Persona that involve a channel as the parameter
            TeamVoiceChannel:voice_channel(team_member_info) = voice_channel(team_member_info):
                Name := MakeMessage("Conversation")
                Group := TeamGroup
            # Add the Player to the 'agent_group'
            TeamGroup.AddMember(Player, team_member_info{})
            # Add the NPC (the one containing this behavior) to the 'agent_group'
            TeamGroup.AddMember(NPCAgent, team_member_info{})
            # Add this channel as a valid channel to the voice chat system through the Simulation entity
            SimEntity.AddChatChannel(TeamVoiceChannel)
            # Save our newly created agent group and voice channel for future calls
            set OptionAgentGroup = option{TeamGroup}
            set OptionVoiceChannel = option{TeamVoiceChannel}
            # Finally, add the player as a target for this behavior to communicate
            AddPlayerAsConversationTarget(Player)

    # Helper function that sets the passed Player as a valid conversation target for this behavior (if they are in the voice channel)
    AddPlayerAsConversationTarget(Player:player):void=
        if:
            NPCEntity := GetEntity[]
            NPCAgent := GetAgent[]
            TeamVoiceChannel := OptionVoiceChannel?
        then:
            if (PersonaComponent := NPCEntity.GetComponent[persona_component]):
                Player.SetConversationTarget(PersonaComponent, TeamVoiceChannel)

    # Helper function to select a random character that we will use to modify our NPC's Personality
    ChooseCharacter():void=
        # Create an array of the possible Character enums that can be chosen
        Characters:[]Character = array{Character.FamilyPerson, Character.HeavydutyWorker, Character.SinglePerson, Character.OutdoorsPerson, Character.Collector}
        if:
            SelectedCharacter := Characters[GetRandomInt(0, Characters.Length - 1)]
            NPCEntity := GetEntity[]
            PersonaComponent := NPCEntity.GetComponent[persona_component]
            Budget := CharacterBudgetMap[SelectedCharacter]
            CharacterMapString := CharacterPromptMap[SelectedCharacter]
        then:
            # Create a random Budget based on the ranges provided by our SelectedCharacter
            BudgetPrompt := MakeBudgetPrompt(GetBudget(Budget))
            # Create a message variable from our mapped string
            CharacterMapMessage := MakeMessage(CharacterMapString)
            # Create a CharacterPrompt local variable that will create a prompt using our mapped message with CharacterMapMessage and our newly created BudgetPrompt.
            CharacterPrompt := Join(array{CharacterMapMessage, BudgetPrompt}, JoinCharacter)
            # Finally, we attempt to set the PersonaComponent personality by creating a new message that joins the Persona's current personality and our CharacterPrompt.
            option{PersonaComponent.SetPersonality[Join(array{PersonaComponent.Personality, CharacterPrompt}, JoinCharacter)]}
    
    # Function that prompts the NPC to talk to the Player. It uses our GreetingMessage as the prompt it should respond to
    InitialTalk(Persona:persona_component):void = 
        if(Channel := OptionVoiceChannel?):
            option{Persona.PromptToTalk[GreetingMessage, Channel]}

    # Function that creates a prompt binding and uses it to bind a function to a Structured Output response. This will fire whenever the NPC believes they have agreed on a price with the player.
    # It returns a '?cancelable' type that we can subscribe later on to keep track on the Structure Output
    RegisterFinalOfferAction():?cancelable =
        if:
            NPCEntity := GetEntity[]
            PersonaComponent := NPCEntity.GetComponent[persona_component]
            Session := PersonaComponent.GetAISession()
        then:
            # Define the binding we will use to determine when LLM should return a response through Structured Output and run the binded function
            FinalOfferBinding:prompt_binding_definition = prompt_binding_definition:
                Name:= MakeMessage("FinalOfferBinding")
                Description:= MakeMessage("Signals whenever you have received an agreement with the player on a final price for a vehicle that fits your needs and is within your budget.")
            # Attempt to register this action into Structured Output. If it succeeds, return the resulting cancelable.
            # To learn more about Structured Output and how this function works, head to the Persona documentation page
            return option{Session.RegisterAction(FinalOfferBinding , false, FinalOfferResponse, FinalOfferFunction)}
        # Return false as a failsafe in the case we couldn't register the action
        return false

    # This is the function binded to our 'FinalOfferBinding'. It uses the 'FinalOfferResponse' to receive the parameters passed by the LLM
    FinalOfferFunction(Response:FinalOfferResponse):void=
        if(Response.FinalOffer?):
            # Award score to the player. Will be called in 'StopSayChecks'
            set ShouldAwardScore = true
            # Save our agreed upon price between the Player and the LLM. This will be used by our ui_manager
            set AcceptedOffer = Response.FinalPrice
            # Despawn the NPC. Will be called in 'StopSayChecks'
            set ShouldDespawn = true

    # Similar to the function 'RegisterFinalOfferAction', but this one handles whether or not the NPC was asked to leave
    RegisterKickAction():?cancelable =
        if:
            NPCEntity := GetEntity[]
            PersonaComponent := NPCEntity.GetComponent[persona_component]
            Session := PersonaComponent.GetAISession()
        then:
            KickBinding:prompt_binding_definition = prompt_binding_definition:
                Name:= MakeMessage("KickBinding")
                Description:= MakeMessage("Signals whenever you have been told to leave, you are told they will no longer deal with you, to look for business elsewhere, or any sentiment similar to those mentioned before.")
            return option{Session.RegisterAction(KickBinding , false, KickResponse, KickFunction)}
        return false

    # Similar to 'FinalOfferFunction', but this will be called in response to the 'KickBinding' description being met
    KickFunction(Response:KickResponse):void=
        if(Response.Kicked?):
            # The NPC left, so give the Player a penalty. Will be called in 'StopSayChecks'
            set ShouldApplyPenalty = true
            # Despawn the NPC. Will be called in 'StopSayChecks'
            set ShouldDespawn = true

    # This function fires every time the NPC has finished saying something to the Player. We call all of our gameplay functions here so that the NPC can finish speaking BEFORE doing anything gameplay related
    StopSayChecks(Interrupted:logic):void=
        # Handles penalties to the Player's score
        if(ShouldApplyPenalty?):
            # This function handles the UI changes for the score
            spawn{UIManager.ScoreChange(PenaltyScore)}
            # Reset the value for the next call
            set ShouldApplyPenalty = false
        if(ShouldAwardScore?):
            # This function handles the UI changes for the score
            spawn{UIManager.ScoreChange(AcceptedOffer)}
            # Reset the value for the next call
            set ShouldAwardScore = false
        if(ShouldDespawn?):
            # Play our Fade animation to hide the NPC's despawning. Doesn't look good for them to just disappear/appear out of thin air
            spawn{UIManager.Fade()}
            # Call for the NPC to be despawned
            spawn{DespawnNPC()}
            # Reset the value for the next call
            set ShouldDespawn = false

    # This function calls ShowCaptions() to make the Captions element visible and set the message inside it to the latest talk point from the NPC
    SetCloseCaptions(NPC:?agent, Message:message):void=
        UIManager.ShowCaptions(Message)
    
    # This function clears the contents of the Captions element and hides it when the NPC is done talking
    ClearCloseCaptions(Interrupted:logic):void=
        UIManager.HideCaptions()

    # Helper function with a delay to match the Fade animation
    DespawnNPC()<suspends>:void=
        Sleep(0.8)
        Spawner.DespawnAll(false)

    # Helper function to generate a random amount of gold and round it down to the nearest multiple of 50
    GetBudget(Range:tuple(int, int)):int = 
        return Value:= GetRandomInt(Range) * 50
