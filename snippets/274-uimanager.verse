# Knowledge-neutral Verse snippet; project names and asset IDs generalized.
# id: snippet-274

using { /Fortnite.com/Devices }
using { /Verse.org/Simulation }
using { /UnrealEngine.com/Temporary/Diagnostics }
using { /UnrealEngine.com/Temporary/UI }
using { /Verse.org/Input }
using { /Verse.org/Input/UI }
using { /Verse.org/Simulation/Tags }

# See https://dev.epicgames.com/documentation/en-us/uefn/create-your-own-device-in-verse for how to create a verse device.

# Tag we will use to find this Verse device and reference it in other Verse classes
ManagerTag := class(tag){}

# Helper functions that allow us to create our own Subscribeable functions
event_subscription := class:
    CancelEvent:event() = event(){}
 
    Cancel():void=
        CancelEvent.Signal()
 
AwaitForEvent(Event:event(t), Callback(:t):void, Subscription:event_subscription where t : type)<suspends>:void=
    race:
        loop:
            # Await till event is signaled
            Payload := Event.Await()
            Callback(Payload)
        # Await till event is cancelled
        Subscription.CancelEvent.Await() 
 
(Event:event(t) where t : type).Subscribe(Callback(:t):void):event_subscription=
    Subscription := event_subscription{}
    spawn{ AwaitForEvent(Event, Callback, Subscription) }
    return Subscription

# Verse device that handles all of the UI transaction between the Game and the Widget
ui_manager := class(creative_device):
    # The Device we will use to hide/show the Catalog UI
    @editable
    var Button:button_device = button_device{}
    # Create the Widgets we will use in this game
    var Catalog:UIAssets.WB_Catalog = UIAssets.WB_Catalog{}
    var HUD:UIAssets.WB_Score = UIAssets.WB_Score{}
    # Empty message used to clear any message fields in the UI
    EmptyMessage<localizes>:message = ""

    # Runs when the device is started in a running game
    OnBegin<override>()<suspends>:void=
        # Initialize all of our UI
        InitializeUI()
        # Subscribe 'DeviceButtonInteraction' to 'InteractedWithEvent'. This will make it so that the Catalog UI is shown when the button is pressed
        Button.InteractedWithEvent.Subscribe(DeviceButtonInteraction)

    # Function that will add the HUD to the Player's screen, and handle all required initialization for the UI
    InitializeUI():void =
        if:
            Player := GetPlayspace().GetPlayers()[0]
            TargetUI := GetPlayerUI[Player]
        then:
            # Subscribe 'UICloseButtonInteraction' to 'CloseMenuEvent'. 'CloseMenuEvent' is an event called whenever the 'CLOSE' button is pressed. For more information, look at the 'WB_Score' class
            Catalog.CloseMenuEvent.Subscribe(UICloseButtonInteraction)
            # Add the HUD to the Player's screen
            TargetUI.AddWidget(HUD)
            # As a failsafe, we hide the Catalog
            ToggleCatalogVisibility(false)

    # Helper function to toggle whether the Catalog UI is enabled or not. 'Visible' determines whether it should be shown or hidden
    ToggleCatalogVisibility(Visible:logic):void=
        if:
            Player := GetPlayspace().GetPlayers()[0]
            TargetUI := GetPlayerUI[Player]
        then:
            if(Visible?):
                # Add the Catalog Widget to the screen. The reason why this is done instead of setting visibility, is due to the fact that input is consumed when it is added
                TargetUI.AddWidget(Catalog, Slot:player_ui_slot = player_ui_slot{InputMode:= ui_input_mode.All})
                # Disable the button so that the player doesn't see it on-screen at all times
                Button.Disable()
            else:
                # Remove the Catalog UI from the screen.
                TargetUI.RemoveWidget(Catalog)
                # Re-enable the button
                Button.Enable()

    # The function we bind to that removes the Catalog UI from the screen
    UICloseButtonInteraction():void=
        ToggleCatalogVisibility(false)

    # The function we bind to that adds the Catalog UI from the screen
    DeviceButtonInteraction(Agent:agent):void=
        ToggleCatalogVisibility(true)

    # This function takes a passed int to modify our score
    ScoreChange<public>(Score:int)<suspends>:void=
        # Save the score we currently have
        CurrentScore := HUD.PlayerScore
        # Add the new 'Score' to the total PlayerScore in the UI
        set HUD.PlayerScore = CurrentScore + Score
        # Set 'ScoreToAdd' to 'Score'. This is used for the animation that shows the score being added
        set HUD.ScoreToAdd = Score
        # Set 'AnimationPlay' to true. This causes 'ScoreAdded' animation to play in 'WB_Score'. To learn how this works, check the 'View Bindings' panel in 'WB_Score'
        set HUD.AnimationPlay = true
        Sleep(2.0)
        # Set 'AnimationPlay' to false. This allows the animation to be played again when we switch it back to true
        set HUD.AnimationPlay = false

    # Function that allows other Verse classes to show and modify the Captions
    ShowCaptions<public>(IncomingCaption:message):void=
        # Set 'CloseCaption' to 'IncomingCaption'. This variable is the direct message we use in the Caption's Widget element
        set HUD.CloseCaption = IncomingCaption
        # Set 'CCVisible' to true. This sets the Overlay element that holds the caption to be shown
        set HUD.CCVisible = true

    # Function that allows other Verse classes to hide and clear the Captions
    HideCaptions<public>():void=
        # Clear 'CloseCaption' by setting it with our empty message
        set HUD.CloseCaption = EmptyMessage
        # Set 'CCVisible' to false. This sets the Overlay element that holds the caption to be collapsed
        set HUD.CCVisible = false
    
    # Function that allows other Verse classes to begin the 'FadeAnimation'
    Fade<public>()<suspends>:void=
        # Set 'FadePlay' to true. This works similarly to 'set HUD.AnimationPlay', but it controls the 'FadeAnimation' instead
        set HUD.FadePlay = true
        Sleep(2.0)
        # Set 'FadePlay' to false. Allows for replay of the animation on future calls
        set HUD.FadePlay = false
